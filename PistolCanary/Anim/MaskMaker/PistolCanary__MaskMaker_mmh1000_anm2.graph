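# Level Sequence for the PistolCanary MaskMaker animation (mmh1000, anm2), animating a spawned MaskMaker mannequin and a tentacle creature.
# MaskMaker runs a chain of overlapping animation clips while its scale grows in steps, and a visibility track hides it near the end.
# A Fortnite biped Control Rig track adds keyed body motion, including repeating leg cycles.
# The tentacle creature plays four overlapping clips late in the timeline, with a transform track raising and repositioning it.
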LevelSequence /PistolCanary/Anim/MaskMaker/PistolCanary__MaskMaker_mmh1000_anm2
bindings [
 Tentacle_PistolCanary_Creature {
  id=#1
  possessed_class=FortSkeletalMeshActor
  tracks [
   {
    type=MovieSceneSkeletalAnimationTrack
    sections [
     {type=MovieSceneSkeletalAnimationSection, end_frame=4272, start_frame=4150}
     {type=MovieSceneSkeletalAnimationSection, end_frame=4334, start_frame=4227}
     {type=MovieSceneSkeletalAnimationSection, end_frame=4466, start_frame=4290}
     {type=MovieSceneSkeletalAnimationSection, end_frame=4598, start_frame=4416}
    ]
   }
   {type=MovieSceneSpawnTrack, sections [{type=MovieSceneSpawnSection, channels [None {type=MovieSceneScriptingBoolChannel, default=true}]}]}
   {
    type=MovieScene3DTransformTrack
    sections [{
      type=MovieScene3DTransformSection
      channels [
       Location.X {type=MovieSceneScriptingDoubleChannel, default=0, keys [[4155,123.5116],[4297,123.5116],[4374,900],[4422,0]]}
       Location.Y {type=MovieSceneScriptingDoubleChannel, default=-6000, keys [[4155,-6000],[4297,-6000],[4374,-8000],[4422,-6000]]}
       Location.Z {
        type=MovieSceneScriptingDoubleChannel
        default=-800
        keys [[4155,-18000],[4157.5,0],[4202.5,0],[4297,-600],[4374,-1000],[4377.5,-800],[4422,-600]]
       }
       Rotation.X {type=MovieSceneScriptingDoubleChannel, default=0, const [0,3]}
       Rotation.Y {type=MovieSceneScriptingDoubleChannel, default=-0.1356, const [-0.1356,3]}
       Rotation.Z {type=MovieSceneScriptingDoubleChannel, default=0, keys [[4155,0],[4297,0],[4374,-10],[4422,0]]}
       Scale.X {type=MovieSceneScriptingDoubleChannel, default=0.75, const [0.75,1]}
       Scale.Y {type=MovieSceneScriptingDoubleChannel, default=0.75, const [0.75,1]}
       Scale.Z {type=MovieSceneScriptingDoubleChannel, default=0.75, const [0.75,1]}
      ]
     }]
   }
  ]
 }
 MaskMaker {
  id=#2
  possessed_class=FortPlayerMannequin
  tracks [
   {
    type=MovieSceneSkeletalAnimationTrack
    sections [
     {type=MovieSceneSkeletalAnimationSection, end_frame=2486, start_frame=1790}
     {type=MovieSceneSkeletalAnimationSection, end_frame=2572, start_frame=2465}
     {type=MovieSceneSkeletalAnimationSection, end_frame=2634, start_frame=2566}
     {type=MovieSceneSkeletalAnimationSection, end_frame=2653, start_frame=2624}
     {type=MovieSceneSkeletalAnimationSection, end_frame=2709, start_frame=2647}
     {type=MovieSceneSkeletalAnimationSection, end_frame=2733, start_frame=2704}
     {type=MovieSceneSkeletalAnimationSection, end_frame=2788, start_frame=2727}
     {type=MovieSceneSkeletalAnimationSection, end_frame=3066, start_frame=2769}
     {type=MovieSceneSkeletalAnimationSection, end_frame=3085, start_frame=3065}
     {type=MovieSceneSkeletalAnimationSection, end_frame=3190, start_frame=3081}
     {type=MovieSceneSkeletalAnimationSection, end_frame=3967, start_frame=3172}
    ]
   }
   {type=MovieSceneSpawnTrack, sections [{type=MovieSceneSpawnSection, channels [None {type=MovieSceneScriptingBoolChannel, default=true}]}]}
   {
    type=MovieScene3DTransformTrack
    sections [
     {
      type=MovieScene3DTransformSection
      channels [
       Location.X {type=MovieSceneScriptingDoubleChannel, default=-1950, const [-100,1]}
       Location.Y {type=MovieSceneScriptingDoubleChannel, default=500, const [-6320.5469,1]}
       Location.Z {type=MovieSceneScriptingDoubleChannel, default=1100, const [1100,1]}
       Rotation.X {type=MovieSceneScriptingDoubleChannel, default=20, const [0,1]}
       Rotation.Y {type=MovieSceneScriptingDoubleChannel, default=0, const [0,1]}
       Rotation.Z {type=MovieSceneScriptingDoubleChannel, default=100, const [0,1]}
       Scale.X {
        type=MovieSceneScriptingDoubleChannel
        default=2
        thinned_from=12
        keys [[1854.5,1.25],[2523,1.25],[2552.5,1.75],[2584,1.75],[2610,2],[2636,2],[2657,2.5],[2696,2.5],[2718,3]]
       }
       Scale.Y {
        type=MovieSceneScriptingDoubleChannel
        default=2
        thinned_from=12
        keys [[1854.5,1.25],[2523,1.25],[2552.5,1.75],[2584,1.75],[2610,2],[2636,2],[2657,2.5],[2696,2.5],[2718,3]]
       }
       Scale.Z {
        type=MovieSceneScriptingDoubleChannel
        default=2
        thinned_from=12
        keys [[1854.5,1.25],[2523,1.25],[2552.5,1.75],[2584,1.75],[2610,2],[2636,2],[2657,2.5],[2696,2.5],[2718,3]]
       }
      ]
     }
     {
      type=MovieScene3DTransformSection
      channels [
       Location.X {type=MovieSceneScriptingDoubleChannel, default=-1950, const [-100,1]}
       Location.Y {type=MovieSceneScriptingDoubleChannel, default=500, const [-6320.5469,1]}
       Location.Z {type=MovieSceneScriptingDoubleChannel, default=3800, const [1100,1]}
       Rotation.X {type=MovieSceneScriptingDoubleChannel, default=-52, const [0,1]}
       Rotation.Y {type=MovieSceneScriptingDoubleChannel, default=5, const [0,1]}
       Rotation.Z {type=MovieSceneScriptingDoubleChannel, default=256, const [0,1]}
       Scale.X {
        type=MovieSceneScriptingDoubleChannel
        default=2.45
        thinned_from=15
        keys [[1864.5,1.25],[2522,1.25],[2552.5,1.75],[2583,1.75],[2599,1.95],[2605.5,2],[2637.5,2],[2642,2.1],[2649,2.35],[2652,2.45],[2656,2.5],[2695.5,2.5],[2714.5,3]]
       }
       Scale.Y {
        type=MovieSceneScriptingDoubleChannel
        default=2.45
        thinned_from=15
        keys [[1864.5,1.25],[2522,1.25],[2552.5,1.75],[2583,1.75],[2599,1.95],[2605.5,2],[2637.5,2],[2642,2.1],[2649,2.35],[2652,2.45],[2656,2.5],[2695.5,2.5],[2714.5,3]]
       }
       Scale.Z {
        type=MovieSceneScriptingDoubleChannel
        default=2.45
        thinned_from=15
        keys [[1864.5,1.25],[2522,1.25],[2552.5,1.75],[2583,1.75],[2599,1.95],[2605.5,2],[2637.5,2],[2642,2.1],[2649,2.35],[2652,2.45],[2656,2.5],[2695.5,2.5],[2714.5,3]]
       }
      ]
     }
    ]
   }
   {
    type=MovieSceneVisibilityTrack
    sections [{
      type=MovieSceneVisibilitySection
      channels [None {type=MovieSceneScriptingBoolChannel, default=true, keys [[1790,true],[1818.5,true],[4418,true],[4418.5,false]]}]
     }]
   }
   {
    type=MovieSceneControlRigParameterTrack
    display_name=Fortnite_Biped_CtrlRig
    sections [
     {
      type=MovieSceneControlRigParameterSection
      channels [
       leg_l_stretch_switch {type=MovieSceneScriptingBoolChannel, default=false, const [false,12]}
       leg_r_stretch_switch {type=MovieSceneScriptingBoolChannel, default=false, const [false,12]}
       arm_l_fk_local {type=MovieSceneScriptingBoolChannel, default=true, const [true,23]}
       Auto Interpolate_3 {type=MovieSceneScriptingBoolChannel, default=false, const [false,15]}
       Auto Interpolate {type=MovieSceneScriptingBoolChannel, default=false, const [false,12]}
       arm_l_stretch_switch {type=MovieSceneScriptingBoolChannel, default=false, const [false,12]}
       Keep Aligned {type=MovieSceneScriptingBoolChannel, default=false, const [false,12]}
       arm_r_fk_local {type=MovieSceneScriptingBoolChannel, default=true, const [true,22]}
       Auto Interpolate_4 {type=MovieSceneScriptingBoolChannel, default=false, const [false,13]}
       Auto Interpolate_2 {type=MovieSceneScriptingBoolChannel, default=false, const [false,12]}
       arm_r_stretch_switch {type=MovieSceneScriptingBoolChannel, default=false, const [false,12]}
       Keep Aligned_2 {type=MovieSceneScriptingBoolChannel, default=false, const [false,12]}
       head_local {type=MovieSceneScriptingBoolChannel, default=true, const [true,19]}
       Stretch {type=MovieSceneScriptingBoolChannel, default=false, const [false,19]}
       head_stretch {type=MovieSceneScriptingBoolChannel, default=false, const [false,12]}
       spine_stretch_switch {type=MovieSceneScriptingBoolChannel, default=false, const [false,12]}
       arm_l_fk_ik_switch {type=MovieSceneScriptingBoolChannel, default=false, const [false,12]}
       leg_l_fk_ik_switch {
        type=MovieSceneScriptingBoolChannel
        default=false
        keys [[1790,false],[2468,false],[2487,false],[2927,false],[2938,false],[2986,false],[3058,false],[3070,false],[3083,false],[3099,false],[3185,false],[4360,false],[4364,false],[4436,true]]
       }
       arm_r_fk_ik_switch {type=MovieSceneScriptingBoolChannel, default=false, const [false,12]}
       leg_r_fk_ik_switch {
        type=MovieSceneScriptingBoolChannel
        default=false
        keys [[1790,false],[2468,false],[2487,false],[2927,false],[2938,false],[2986,false],[3058,false],[3070,false],[3083,false],[3099,false],[3185,false],[4360,false],[4364,false],[4436,true]]
       }
       ShowBodyControls {type=MovieSceneScriptingBoolChannel, default=true, const [true,12]}
       spine_fk_ik_switch {type=MovieSceneScriptingBoolChannel, default=false, const [false,12]}
       neck_fk_ik_switch {type=MovieSceneScriptingBoolChannel, default=false, const [false,12]}
       space {type=MovieSceneScriptingByteChannel, default=1, const [1,19]}
       global_ctrl.Location.X {type=MovieSceneScriptingFloatChannel, default=0, const [0,12]}
       global_ctrl.Location.Y {type=MovieSceneScriptingFloatChannel, default=0, const [0,12]}
       global_ctrl.Location.Z {type=MovieSceneScriptingFloatChannel, default=0, const [0,12]}
       global_ctrl.Rotation.X {type=MovieSceneScriptingFloatChannel, default=0, const [0,12]}
       global_ctrl.Rotation.Y {type=MovieSceneScriptingFloatChannel, default=0, const [0,12]}
       global_ctrl.Rotation.Z {type=MovieSceneScriptingFloatChannel, default=0, const [0,12]}
       root_ctrl.Location.X {
        type=MovieSceneScriptingFloatChannel
        default=0
        thinned_from=12
        keys [[1790,0],[2318,0],[2780,-0.1302],[2927,0.5929],[2938,0.6061],[2986,9.3404],[3058,9.3404],[3070,6.7236],[4364,6.7236]]
       }
       root_ctrl.Location.Y {
        type=MovieSceneScriptingFloatChannel
        default=0
        thinned_from=18
        keys [[1790,0],[2449,0],[2471,-1.3091],[2496,-8.0693],[2503,0.0005],[2784,0.4867],[2927,-0.4309],[2938,-0.4593],[2986,-24.0952],[3058,-24.0952],[3070,11.4813],[4364,11.4813]]
       }
       root_ctrl.Location.Z {
        type=MovieSceneScriptingFloatChannel
        default=0
        thinned_from=20
        keys [[1790,0],[2467,0.0076],[2503,6.1068],[2507,8.0984],[2511,1.1822],[2518,13.2528],[2524,6.4697],[2530,10.066],[2762,12.0648],[2772,6.3121],[2780,7.4454],[2927,15.3543],[2938,15.4177],[2986,62.208],[3058,62.208],[3070,43.1919],[4364,43.1919]]
       }
       root_ctrl.Rotation.X {type=MovieSceneScriptingFloatChannel, default=0, const [0,10]}
       root_ctrl.Rotation.Y {type=MovieSceneScriptingFloatChannel, default=0, const [0,10]}
       root_ctrl.Rotation.Z {type=MovieSceneScriptingFloatChannel, default=0, const [0,10]}
       root_ctrl.Scale.X {type=MovieSceneScriptingFloatChannel, default=1, const [1,10]}
       root_ctrl.Scale.Y {type=MovieSceneScriptingFloatChannel, default=1, const [1,10]}
       root_ctrl.Scale.Z {type=MovieSceneScriptingFloatChannel, default=1, const [1,10]}
       body_offset_ctrl.Location.X {
        type=MovieSceneScriptingFloatChannel
        default=0
        thinned_from=14
        keys [[1790,0],[2318,0],[2780,-0.1302],[2927,0.5929],[2938,0.6061],[2986,9.3404],[3058,9.3404],[3070,6.7236],[3185,6.7236],[4360,0.1873],[4364,0.1726],[4436,0]]
       }
       body_offset_ctrl.Location.Y {
        type=MovieSceneScriptingFloatChannel
        default=0
        thinned_from=20
        keys [[1790,0],[2449,0],[2471,-1.3091],[2496,-8.0693],[2503,0.0005],[2784,0.4867],[2927,-0.4309],[2938,-0.4593],[2986,-24.0952],[3058,-24.0952],[3070,11.4813],[3185,11.4813],[4360,0.3199],[4364,0.2947],[4436,0]]
       }
       body_offset_ctrl.Location.Z {
        type=MovieSceneScriptingFloatChannel
        default=0
        thinned_from=22
        keys [[1790,0],[2467,0.0076],[2503,6.1068],[2507,8.0984],[2511,1.1822],[2518,13.2528],[2524,6.4697],[2530,10.066],[2762,12.0648],[2772,6.3121],[2780,7.4454],[2927,15.3543],[2938,15.4177],[2986,62.208],[3058,62.208],[3070,43.1919],[3185,43.1919],[4360,1.2034],[4364,1.1085],[4436,0]]
       }
       body_offset_ctrl.Rotation.X {type=MovieSceneScriptingFloatChannel, default=0, const [0,12]}
       body_offset_ctrl.Rotation.Y {type=MovieSceneScriptingFloatChannel, default=0, const [0,12]}
       body_offset_ctrl.Rotation.Z {type=MovieSceneScriptingFloatChannel, default=0, const [0,12]}
       body_offset_ctrl.Scale.X {type=MovieSceneScriptingFloatChannel, default=1, const [1,12]}
       body_offset_ctrl.Scale.Y {type=MovieSceneScriptingFloatChannel, default=1, const [1,12]}
       body_offset_ctrl.Scale.Z {type=MovieSceneScriptingFloatChannel, default=1, const [1,12]}
       thigh_l_fk_ctrl.Location.X {type=MovieSceneScriptingFloatChannel, default=0, thinned_from=263, keys [[1790,0],[4365,0]]}
       thigh_l_fk_ctrl.Location.Y {type=MovieSceneScriptingFloatChannel, default=0, thinned_from=973, keys [[1790,0],[4365,0]]}
       thigh_l_fk_ctrl.Location.Z {type=MovieSceneScriptingFloatChannel, default=0, thinned_from=384, keys [[1790,0],[4365,0]]}
       thigh_l_fk_ctrl.Rotation.X {
        type=MovieSceneScriptingFloatChannel
        default=6.1949
        keys [[1790,6.1949],[1821,2.5511],[1865,6.3543],[1901,2.5557],[1945,6.3542],[1981,2.551],[2025,6.3543],[2061,2.5559],[2105,6.3542],[2141,2.5511],[2185,6.3543],[2221,2.5557],[2271,2.6498],[2301,2.551],[2345,6.3543],[2381,2.5559],[2425,6.3542],[2461,2.5511],[2471,3.1259],[2472,3.3599],[2484,8.6496],[2501,8.6496],[2507,6.6029],[2511,0.6433],[2515,8.4646],[2519,1.6425],[2522,4.9167],[2525,2.6373],[2536,7.2623],[2757,4.4823],[2783,2.4005],[2809,2.3245],[2857,2.536],[2901,2.6662],[2938,2.797],[2986,21.2365],[3058,21.2365],[3070,15.7855],[3083,15.7855],[3087,16.2701],[3099,15.7855],[3185,17.7638],[3220,6.3542],[3256,2.5511],[3300,6.3543],[3336,2.5557],[3380,6.3542],[3416,2.551],[3460,6.3543],[3496,2.5559],[3540,6.3542],[3576,2.5511],[3620,6.3543],[3656,2.5557],[3700,6.3542],[3736,2.551],[3780,6.3543],[3816,2.5559],[3860,6.3542],[3904,6.3543],[3940,2.5559],[3984,6.3542],[4028,6.3543],[4064,2.5559],[4108,6.3542],[4153,6.3543],[4189,2.5559],[4233,6.3542],[4279,6.3543],[4315,2.5559],[4364,2.9261],[4365,2.9265],[4369,2.5814]]
       }
       thigh_l_fk_ctrl.Rotation.Y {
        type=MovieSceneScriptingFloatChannel
        default=4.8675
        thinned_from=75
        keys [[1790,4.8675],[1818,-1.8481],[1860,5.2273],[1898,-1.853],[1939,5.2367],[1978,-1.8476],[2020,5.2273],[2058,-1.8524],[2099,5.2365],[2138,-1.8481],[2180,5.2273],[2218,-1.853],[2259,5.2367],[2271,-2.4929],[2298,-1.8476],[2340,5.2273],[2378,-1.8524],[2419,5.2365],[2454,1.3103],[2468,3.9537],[2472,6.2967],[2485,14.1879],[2502,14.1879],[2510,15.9962],[2511,15.9896],[2513,16.4628],[2515,16.4379],[2519,19.0427],[2521,17.7158],[2526,20.0255],[2757,4.482],[2767,5.6284],[2779,2.7887],[2809,-3.4006],[2857,2.2286],[2901,0.5513],[2938,3.629],[2986,12.0906],[3058,12.0906],[3070,3.5393],[3099,3.5393],[3185,-6.2939],[3253,1.3607],[3295,5.2273],[3333,1.3558],[3374,5.2367],[3413,1.3613],[3455,5.2273],[3493,1.3564],[3534,5.2365],[3573,1.3607],[3615,5.2273],[3653,1.3558],[3694,5.2367],[3723,1.6597],[3733,1.3613],[3775,5.2273],[3813,1.3564],[3854,5.2365],[3899,5.2273],[3937,1.3564],[3978,5.2365],[4023,5.2273],[4061,1.3564],[4102,5.2365],[4148,5.2273],[4186,1.3564],[4227,5.2365],[4274,5.2273],[4312,1.3564],[4353,5.2365],[4364,1.9438],[4365,1.4746],[4369,-12.491]]
       }
       thigh_l_fk_ctrl.Rotation.Z {
        type=MovieSceneScriptingFloatChannel
        default=-2.7259
        thinned_from=80
        keys [[1790,-2.7259],[1791,-2.7347],[1831,0.8135],[1871,-2.7351],[1911,0.8135],[1951,-2.7347],[1991,0.8133],[2031,-2.7351],[2071,0.8139],[2111,-2.7347],[2151,0.8135],[2191,-2.7351],[2231,0.8135],[2271,2.2751],[2311,0.8133],[2351,-2.7351],[2391,0.8139],[2431,-2.7347],[2466,1.4961],[2472,-3.0131],[2484,-79.9699],[2504,-79.9699],[2506,-82.8465],[2510,-119.087],[2511,-114.3223],[2512,-105.5224],[2515,-28.5578],[2518,-0.603],[2520,12.1943],[2525,2.2113],[2757,14.6089],[2780,0.0694],[2809,-0.7816],[2857,-0.0299],[2901,-0.2739],[2938,0.2204],[2979,4.1261],[3008,21.3509],[3038,22.5292],[3058,21.7986],[3061,13.0064],[3069,-62.3172],[3080,-70.9378],[3086,-77.167],[3100,-72.5095],[3129,-69.658],[3178,-56.1033],[3226,-2.7347],[3266,0.8135],[3306,-2.7351],[3346,0.8135],[3386,-2.7347],[3426,0.8133],[3466,-2.7351],[3506,0.8139],[3546,-2.7347],[3586,0.8135],[3626,-2.7351],[3666,0.8135],[3691,-3.1307],[3720,-0.4272],[3746,0.8133],[3786,-2.7351],[3826,0.8139],[3866,-2.7347],[3910,-2.7351],[3950,0.8139],[3990,-2.7347],[4034,-2.7351],[4074,0.8139],[4114,-2.7347],[4159,-2.7351],[4199,0.8139],[4239,-2.7347],[4285,-2.7351],[4325,0.8139],[4364,-0.5127],[4365,-0.5644],[4369,-2.9677]]
       }
       thigh_l_fk_ctrl.Scale.X {type=MovieSceneScriptingFloatChannel, default=1, const [1,12]}
       thigh_l_fk_ctrl.Scale.Y {type=MovieSceneScriptingFloatChannel, default=1, const [1,12]}
       thigh_l_fk_ctrl.Scale.Z {type=MovieSceneScriptingFloatChannel, default=1, const [1,12]}
       calf_l_fk_ctrl.Location.X {
        type=MovieSceneScriptingFloatChannel
        default=0
        thinned_from=100
        keys [[1790,0],[2511,0],[2513,0.0002],[2530,0],[2761,0],[2771,0.0004],[2776,0.017],[2782,0.0794],[2804,0.1213],[2938,0.1217],[2986,0.1478],[3058,0.1478],[3070,0.0007],[3185,0.0007],[3217,0],[4365,0]]
       }
       calf_l_fk_ctrl.Location.Y {
        type=MovieSceneScriptingFloatChannel
        default=0
        thinned_from=887
        keys [[1790,0],[2768,0],[2777,0.0045],[2787,0.0198],[2927,0.0245],[2938,0.0245],[2986,0.0447],[3058,0.0447],[3070,0],[4366,0]]
       }
       calf_l_fk_ctrl.Location.Z {type=MovieSceneScriptingFloatChannel, default=0, thinned_from=1158, keys [[1790,0],[3058,-0.0001],[4365,0]]}
       calf_l_fk_ctrl.Rotation.X {
        type=MovieSceneScriptingFloatChannel
        default=-0.3023
        thinned_from=103
        keys [[1790,-0.3023],[1824,-0.2454],[1869,-0.3023],[1871,-0.3023],[1905,-0.2453],[1949,-0.3023],[1951,-0.3023],[1984,-0.2453],[2027,-0.3023],[2030,-0.3023],[2065,-0.2453],[2110,-0.3023],[2144,-0.2454],[2189,-0.3023],[2191,-0.3023],[2224,-0.2453],[2269,-0.3023],[2271,-0.3023],[2304,-0.2453],[2347,-0.3023],[2350,-0.3023],[2384,-0.2453],[2430,-0.3023],[2464,-0.2454],[2468,-0.4682],[2490,-0.6032],[2507,-0.59],[2511,-0.5548],[2513,-0.369],[2518,0.4206],[2521,0.2686],[2523,0.6],[2526,0.7187],[2759,0.3527],[2766,-0.0556],[2773,-0.203],[2788,-0.0239],[2938,-0.0243],[2967,0.4393],[2986,0.1465],[3058,0.1465],[3070,0.2529],[3099,0.2529],[3185,0.6365],[3225,-0.3023],[3259,-0.2454],[3304,-0.3023],[3306,-0.3023],[3339,-0.2453],[3384,-0.3023],[3386,-0.3023],[3419,-0.2453],[3462,-0.3023],[3464,-0.3023],[3500,-0.2453],[3546,-0.3023],[3579,-0.2454],[3624,-0.3023],[3626,-0.3023],[3659,-0.2453],[3704,-0.3023],[3706,-0.3023],[3739,-0.2453],[3782,-0.3023],[3786,-0.3023],[3820,-0.2453],[3865,-0.3023],[3910,-0.3023],[3944,-0.2453],[3989,-0.3023],[4034,-0.3023],[4068,-0.2453],[4113,-0.3023],[4159,-0.3023],[4193,-0.2453],[4238,-0.3023],[4285,-0.3023],[4319,-0.2453],[4365,-0.2572]]
       }
       calf_l_fk_ctrl.Rotation.Y {
        type=MovieSceneScriptingFloatChannel
        default=0.0494
        thinned_from=114
        keys [[1790,0.0494],[1824,0.0278],[1869,0.0494],[1905,0.0278],[1949,0.0494],[1984,0.0278],[2027,0.0494],[2030,0.0494],[2065,0.0278],[2108,0.0494],[2111,0.0494],[2144,0.0278],[2189,0.0494],[2224,0.0278],[2269,0.0494],[2304,0.0278],[2347,0.0494],[2350,0.0494],[2385,0.0278],[2428,0.0494],[2431,0.0494],[2464,0.0278],[2468,0.0919],[2484,0.8537],[2501,0.8537],[2504,0.9189],[2511,0.9852],[2512,0.9433],[2513,0.791],[2515,0.4135],[2519,0.114],[2521,0.0871],[2526,0.1871],[2760,0.1247],[2782,-0.0109],[2804,-0.0217],[2938,-0.0217],[2967,-0.2816],[2986,-0.1418],[3058,-0.1418],[3070,-0.2104],[3099,-0.2104],[3185,-0.6504],[3223,0.0494],[3225,0.0494],[3259,0.0278],[3304,0.0494],[3339,0.0278],[3384,0.0494],[3419,0.0278],[3462,0.0494],[3466,0.0494],[3500,0.0278],[3543,0.0494],[3546,0.0494],[3579,0.0278],[3624,0.0494],[3659,0.0278],[3704,0.0494],[3739,0.0278],[3782,0.0494],[3786,0.0494],[3820,0.0278],[3863,0.0494],[3910,0.0494],[3944,0.0278],[3987,0.0494],[4034,0.0494],[4068,0.0278],[4111,0.0494],[4159,0.0494],[4193,0.0278],[4236,0.0494],[4285,0.0494],[4319,0.0278],[4365,0.0326]]
       }
       calf_l_fk_ctrl.Rotation.Z {
        type=MovieSceneScriptingFloatChannel
        default=28.5674
        thinned_from=92
        keys [[1790,28.5674],[1824,22.949],[1867,28.5656],[1870,28.5691],[1905,22.9472],[1947,28.5656],[1950,28.5691],[1984,22.9472],[2025,28.562],[2028,28.569],[2065,22.9472],[2108,28.5674],[2144,22.949],[2187,28.5656],[2190,28.5691],[2224,22.9472],[2267,28.5656],[2270,28.5691],[2304,22.9472],[2345,28.562],[2348,28.569],[2384,22.9472],[2428,28.5674],[2464,22.949],[2467,24.2346],[2490,117.292],[2501,117.292],[2504,123.9793],[2511,131.2444],[2512,127.0024],[2513,106.4916],[2515,60.7825],[2519,9.5786],[2521,6.9463],[2526,19.6221],[2765,20.8073],[2777,23.4396],[2800,30.9092],[2938,31.3399],[2967,59.3762],[2986,49.898],[3058,49.898],[3061,49.6448],[3065,53.9859],[3067,62.4088],[3071,22.9891],[3078,21.897],[3099,22.0566],[3194,23.2007],[3213,43.7092],[3218,45.4285],[3239,29.0772],[3259,22.949],[3302,28.5656],[3305,28.5691],[3339,22.9472],[3382,28.5656],[3385,28.5691],[3419,22.9472],[3460,28.562],[3463,28.569],[3500,22.9472],[3543,28.5674],[3579,22.949],[3622,28.5656],[3625,28.5691],[3657,23.1594],[3705,28.5691],[3739,22.9472],[3780,28.562],[3783,28.569],[3820,22.9472],[3863,28.5674],[3907,28.569],[3944,22.9472],[3987,28.5674],[4031,28.569],[4068,22.9472],[4111,28.5674],[4156,28.569],[4193,22.9472],[4236,28.5674],[4282,28.569],[4319,22.9472],[4365,24.0087]]
       }
       calf_l_fk_ctrl.Scale.X {type=MovieSceneScriptingFloatChannel, default=1, const [1,12]}
       calf_l_fk_ctrl.Scale.Y {type=MovieSceneScriptingFloatChannel, default=1, const [1,12]}
       calf_l_fk_ctrl.Scale.Z {type=MovieSceneScriptingFloatChannel, default=1, const [1,12]}
       foot_l_fk_ctrl.Location.X {type=MovieSceneScriptingFloatChannel, default=0, thinned_from=42, keys [[1790,0],[4435,0],[4436,0.0005]]}
       foot_l_fk_ctrl.Location.Y {type=MovieSceneScriptingFloatChannel, default=0, thinned_from=1358, keys [[1790,0],[4436,0]]}
       foot_l_fk_ctrl.Location.Z {type=MovieSceneScriptingFloatChannel, default=0, thinned_from=47, keys [[1790,0],[4436,0]]}
       foot_l_fk_ctrl.Rotation.X {
        type=MovieSceneScriptingFloatChannel
        default=-3.9349
        thinned_from=79
        keys [[1790,-7.2678],[1791,-3.932],[1832,-4.5765],[1871,-3.932],[1912,-4.5765],[1951,-3.932],[1992,-4.5765],[2031,-3.9321],[2072,-4.5765],[2111,-3.932],[2152,-4.5765],[2191,-3.932],[2232,-4.5765],[2264,-6.8722],[2312,-4.5765],[2351,-3.9321],[2392,-4.5765],[2431,-3.932],[2468,13.2999],[2487,1.5537],[2507,-2.1653],[2511,-3.0952],[2518,4.5115],[2519,5.0859],[2521,10.5153],[2523,11.5106],[2530,10.0801],[2532,10.1332],[2536,10.1476],[2757,7.2112],[2758,7.9409],[2764,0.4928],[2767,-2.4613],[2776,-4.9444],[2781,-7.1139],[2799,-7.9946],[2938,-7.9705],[2986,-2.9007],[3058,-2.9007],[3070,-3.1334],[3185,-3.1334],[3226,-3.932],[3267,-4.5765],[3306,-3.932],[3347,-4.5765],[3386,-3.932],[3427,-4.5765],[3466,-3.9321],[3507,-4.5765],[3546,-3.932],[3587,-4.5765],[3626,-3.932],[3667,-4.5765],[3706,-3.932],[3747,-4.5765],[3786,-3.9321],[3827,-4.5765],[3866,-3.932],[3910,-3.9321],[3951,-4.5765],[3990,-3.932],[4034,-3.9321],[4075,-4.5765],[4114,-3.932],[4159,-3.9321],[4200,-4.5765],[4239,-3.932],[4285,-3.9321],[4326,-4.5765],[4360,-4.3129],[4364,-4.3541],[4376,-4.5765],[4415,-3.932],[4435,-4.2745],[4436,-0.0011]]
       }
       foot_l_fk_ctrl.Rotation.Y {
        type=MovieSceneScriptingFloatChannel
        default=-0.3461
        thinned_from=76
        keys [[1790,-1.8402],[1791,-0.3458],[1832,-0.391],[1871,-0.3458],[1912,-0.391],[1951,-0.3458],[1992,-0.391],[2031,-0.3458],[2072,-0.391],[2111,-0.3458],[2152,-0.391],[2191,-0.3458],[2232,-0.391],[2264,-2.171],[2312,-0.391],[2351,-0.3458],[2392,-0.391],[2431,-0.3458],[2468,-18.1704],[2471,-0.3844],[2487,-33.2769],[2502,-5.3689],[2507,-5.0076],[2512,-5.4967],[2518,-2.298],[2519,-2.9261],[2522,-0.238],[2524,-0.2722],[2532,0.9364],[2533,0.9078],[2757,-1.8396],[2763,0.2482],[2775,-0.549],[2788,-1.9666],[2927,-1.9095],[2938,-1.9081],[2986,-0.0167],[3058,-0.0167],[3070,1.0286],[3185,1.0286],[3226,-0.3458],[3267,-0.391],[3306,-0.3458],[3347,-0.391],[3386,-0.3458],[3427,-0.391],[3466,-0.3458],[3507,-0.391],[3546,-0.3458],[3587,-0.391],[3626,-0.3458],[3667,-0.391],[3706,-0.3458],[3747,-0.391],[3786,-0.3458],[3827,-0.391],[3866,-0.3458],[3910,-0.3458],[3951,-0.391],[3990,-0.3458],[4034,-0.3458],[4075,-0.391],[4114,-0.3458],[4159,-0.3458],[4200,-0.391],[4239,-0.3458],[4285,-0.3458],[4326,-0.391],[4360,-0.3725],[4364,-0.3754],[4376,-0.391],[4415,-0.3458],[4435,-0.378],[4436,-0.0006]]
       }
       foot_l_fk_ctrl.Rotation.Z {
        type=MovieSceneScriptingFloatChannel
        default=17.0812
        thinned_from=74
        keys [[1790,31.3563],[1791,17.0546],[1832,22.8518],[1871,17.0546],[1912,22.8525],[1951,17.0546],[1992,22.8525],[2031,17.0556],[2072,22.8525],[2111,17.0546],[2152,22.8518],[2191,17.0546],[2232,22.8525],[2264,38.8918],[2312,22.8525],[2351,17.0556],[2392,22.8525],[2431,17.0546],[2468,22.9276],[2487,20.0231],[2503,4.1122],[2507,-2.7301],[2511,6.3248],[2512,6.2496],[2515,14.7521],[2519,38.2828],[2521,27.5826],[2524,18.4389],[2525,18.6127],[2526,18.6113],[2757,28.9415],[2760,21.8802],[2927,21.0895],[2938,21.0778],[2986,14.3326],[3058,14.3326],[3070,21.9629],[3185,21.9629],[3226,17.0546],[3267,22.8518],[3306,17.0546],[3347,22.8525],[3386,17.0546],[3427,22.8525],[3466,17.0556],[3507,22.8525],[3546,17.0546],[3587,22.8518],[3626,17.0546],[3667,22.8525],[3706,17.0546],[3747,22.8525],[3786,17.0556],[3827,22.8525],[3866,17.0546],[3910,17.0556],[3951,22.8525],[3990,17.0546],[4034,17.0556],[4075,22.8525],[4114,17.0546],[4159,17.0556],[4200,22.8525],[4239,17.0546],[4285,17.0556],[4326,22.8525],[4360,20.4808],[4364,20.8513],[4376,22.8518],[4415,17.0546],[4435,20.1407],[4436,0.0054]]
       }
       foot_l_fk_ctrl.Scale.X {type=MovieSceneScriptingFloatChannel, default=1, const [1,14]}
       foot_l_fk_ctrl.Scale.Y {type=MovieSceneScriptingFloatChannel, default=1, const [1,14]}
       foot_l_fk_ctrl.Scale.Z {type=MovieSceneScriptingFloatChannel, default=1, const [1,14]}
       ball_l_fk_ctrl.Location.X {type=MovieSceneScriptingFloatChannel, default=0, thinned_from=1203, keys [[1790,0],[4435,0],[4436,0.0004]]}
       ball_l_fk_ctrl.Location.Y {type=MovieSceneScriptingFloatChannel, default=0, thinned_from=1154, keys [[1790,0],[4435,0],[4436,-0.0006]]}
       ball_l_fk_ctrl.Location.Z {type=MovieSceneScriptingFloatChannel, default=0, thinned_from=1150, keys [[1790,0],[4436,0]]}
       ball_l_fk_ctrl.Rotation.X {
        type=MovieSceneScriptingFloatChannel
        default=-2.9199
        thinned_from=1006
        keys [[1790,-2.9199],[2471,-2.9199],[2487,-1.4622],[2501,0],[2519,0.0637],[2523,0],[2767,-0.839],[2775,-1.6172],[2784,-0.1297],[2794,0.0052],[2938,0.0052],[2986,0.0033],[3058,0.0033],[3070,0.0008],[3185,0.0008],[3216,-2.9199],[4435,-2.9199],[4436,-0.0007]]
       }
       ball_l_fk_ctrl.Rotation.Y {
        type=MovieSceneScriptingFloatChannel
        default=-1.347
        thinned_from=177
        keys [[1790,-1.347],[2468,-1.347],[2487,-0.6102],[2501,0],[2512,0],[2519,0.0842],[2523,0],[2757,-0.0112],[2776,-0.7953],[2784,-0.3304],[2793,-0.278],[2938,-0.278],[2986,-0.1727],[3058,-0.1727],[3070,-0.044],[3185,-0.044],[3222,-1.347],[4435,-1.347],[4436,-0.0006]]
       }
       ball_l_fk_ctrl.Rotation.Z {
        type=MovieSceneScriptingFloatChannel
        default=-0.3363
        thinned_from=1249
        keys [[1790,-0.3363],[2471,-0.3363],[2487,-0.1801],[2501,0],[2512,0],[2515,-2.2794],[2519,-15.7476],[2521,-4.2195],[2523,0],[2763,-0.0558],[2776,-0.7937],[2781,-2.5612],[2789,-3.224],[2938,-3.2236],[2986,-2.0029],[3058,-2.0029],[3070,-1.4546],[3185,-1.4546],[3216,-0.3363],[4434,-0.3363],[4436,-0.0001]]
       }
       ball_l_fk_ctrl.Scale.X {type=MovieSceneScriptingFloatChannel, default=1, const [1,14]}
       ball_l_fk_ctrl.Scale.Y {type=MovieSceneScriptingFloatChannel, default=1, const [1,14]}
       ball_l_fk_ctrl.Scale.Z {type=MovieSceneScriptingFloatChannel, default=1, const [1,14]}
       foot_l_ik_ctrl.Location.X {
        type=MovieSceneScriptingFloatChannel
        default=8.1109
        thinned_from=308
        keys [[1780,-86.0461],[1791,-86.0606],[1815,-90.7509],[1821,-91.6108],[1824,-91.7359],[1832,-91.2012],[1859,-86.6662],[1865,-86.3542],[1869,-86.5492],[1871,-86.7768],[1896,-92.2638],[1900,-92.7916],[1904,-92.9866],[1907,-92.8938],[1911,-92.5136],[1941,-86.4908],[1945,-86.1106],[1949,-86.0308],[1951,-86.1305],[1975,-90.8483],[1981,-91.6302],[1984,-91.7428],[1988,-91.599],[1991,-91.3252],[2019,-86.6669],[2025,-86.3545],[2029,-86.5494],[2031,-86.777],[2056,-92.2626],[2060,-92.7902],[2064,-92.9852],[2068,-92.8228],[2071,-92.5121],[2101,-86.4883],[2104,-86.1775],[2111,-86.1277],[2135,-90.8421],[2141,-91.6234],[2144,-91.7359],[2152,-91.2012],[2179,-86.6662],[2185,-86.3542],[2189,-86.5492],[2191,-86.7768],[2216,-92.2638],[2220,-92.7916],[2224,-92.9866],[2227,-92.8938],[2231,-92.5136],[2261,-83.5509],[2265,-82.5094],[2268,-82.4273],[2270,-82.7882],[2271,-83.1042],[2295,-110.7068],[2301,-119.4299],[2304,-123.5342],[2308,-128.5912],[2311,-131.9511],[2339,-144.0023],[2345,-144.3815],[2351,-144.0278],[2376,-138.9118],[2380,-138.4517],[2384,-138.3225],[2391,-138.899],[2421,-145.0953],[2431,-145.5292],[2455,-141.2932],[2461,-140.6809],[2464,-140.6521],[2470,-141.5485],[2471,-141.516],[2484,-157.9749],[2487,-157.0152],[2496,-146.8075],[2501,-136.905],[2502,-132.0089],[2503,-131.5128],[2505,-131.2857],[2507,-131.4998],[2508,-131.4055],[2510,-130.5081],[2511,-131.0079],[2512,-133.073],[2513,-135.909],[2515,-151.4815],[2518,-166.2672],[2519,-168.2817],[2521,-167.9003],[2522,-169.1602],[2523,-170.8846],[2524,-172.1974],[2526,-173.8823],[2531,-176.7884],[2536,-178.8849],[2538,-180.1035],[2760,-232.8271],[2764,-213.0225],[2766,-204.5897],[2769,-199.5102],[2773,-193.5558],[2778,-190.379],[2784,-190.6351],[2791,-190.3568],[2927,-200.8313],[2938,68.5634],[2986,-314.6584],[3058,-314.6584],[3070,-76.1282],[3083,-76.1282],[3099,-350.2401],[3185,-350.2401],[3216,-182.3823],[3219,-183.7207],[3226,-184.7935],[3250,-176.7489],[3256,-175.512],[3259,-175.5362],[3267,-177.3603],[3294,-187.3619],[3300,-187.6833],[3304,-186.9312],[3306,-186.267],[3331,-173.5078],[3335,-172.6598],[3339,-172.5978],[3342,-173.0863],[3346,-174.2747],[3376,-187.5789],[3380,-188.1478],[3386,-187.6456],[3410,-176.6891],[3416,-175.4551],[3419,-175.4876],[3423,-176.1761],[3426,-177.0297],[3454,-187.3811],[3460,-187.6891],[3464,-186.9336],[3466,-186.2695],[3491,-173.5123],[3495,-172.6645],[3499,-172.6024],[3503,-173.3363],[3506,-174.2793],[3536,-187.5839],[3539,-188.0734],[3546,-187.6517],[3570,-176.7488],[3576,-175.5119],[3579,-175.5361],[3587,-177.3602],[3614,-187.3618],[3620,-187.6833],[3624,-186.9311],[3626,-186.2669],[3651,-173.5078],[3655,-172.6598],[3659,-172.5978],[3662,-173.0864],[3666,-174.2747],[3696,-193.2119],[3700,-194.5813],[3706,-194.8886],[3730,-188.012],[3736,-186.9123],[3739,-186.866],[3746,-187.9273],[3774,-195.5844],[3780,-196.0295],[3786,-195.3287],[3811,-185.9352],[3815,-184.9306],[3819,-184.5438],[3823,-184.8238],[3826,-185.3647],[3856,-195.1699],[3866,-195.8958],[3898,-195.5844],[3904,-196.0295],[3910,-195.3287],[3935,-185.9352],[3939,-184.9306],[3943,-184.5438],[3947,-184.8238],[3950,-185.3647],[3980,-195.1699],[3990,-195.8958],[4022,-195.5844],[4028,-196.0295],[4034,-195.3287],[4059,-185.9352],[4063,-184.9306],[4067,-184.5438],[4071,-184.8238],[4074,-185.3647],[4104,-195.1699],[4114,-195.8958],[4147,-195.5844],[4153,-196.0295],[4159,-195.3287],[4184,-185.9352],[4188,-184.9306],[4192,-184.5438],[4196,-184.8238],[4199,-185.3647],[4229,-195.1699],[4239,-195.8958],[4273,-195.5844],[4279,-196.0295],[4285,-195.3287],[4310,-185.9352],[4314,-184.9306],[4318,-184.5438],[4322,-184.8238],[4325,-185.3647],[4355,-195.1699],[4360,-139.4056],[4364,-442.7568],[4365,-457.3369],[4372,-4973.6343],[4376,-8291.6807],[4403,-35737.5117],[4409,-42029.2305],[4413,-46063.2617],[4415,-46556.3672],[4436,-46556.1523]]
       }
       foot_l_ik_ctrl.Location.Y {
        type=MovieSceneScriptingFloatChannel
        default=-7.6251
        thinned_from=308
        keys [[1780,-6331.1826],[1871,-6331.2295],[1896,-6261.374],[1911,-6180.3208],[1941,-5904.7964],[1951,-5783.7993],[1975,-5446.5864],[1991,-5189.2949],[2031,-4466.2695],[2071,-3707.7649],[2111,-2960.8872],[2152,-2233.4844],[2181,-1754.6631],[2191,-1600.8602],[2216,-1248.6101],[2227,-1108.5222],[2231,-1059.8708],[2261,-730.9561],[2271,-630.771],[2295,-447.2084],[2301,-418.1592],[2304,-406.0663],[2308,-392.3621],[2311,-383.8224],[2339,-355.913],[2351,-356.6652],[2391,-354.2794],[2431,-358.5948],[2468,-354.7823],[2471,-356.3622],[2484,-271.2956],[2487,-278.1201],[2496,-323.2889],[2501,-339.4597],[2502,-344.6358],[2503,-347.7607],[2507,-349.7909],[2508,-352.7624],[2510,-364.6984],[2511,-367.4814],[2512,-364.4111],[2513,-358.5569],[2515,-357.4763],[2518,-361.1989],[2519,-363.0481],[2536,-357.7015],[2538,-356.4658],[2760,-349.4806],[2764,-339.6061],[2766,-336.8608],[2773,-332.7584],[2778,-331.3175],[2784,-331.5238],[2791,-330.0089],[2927,-290.4642],[2938,-466.8503],[2986,-306.5935],[3058,-306.5935],[3070,-145.6823],[3083,-145.6823],[3099,-238.2056],[3185,-238.2056],[3216,-317.2762],[3219,-303.9962],[3226,-270.6149],[3250,-136.0068],[3256,-101.4325],[3259,-84.4418],[3263,-62.3578],[3265,-51.6127],[3267,-41.1036],[3294,72.3799],[3296,78.3436],[3300,89.0165],[3304,98.0371],[3306,101.9397],[3331,118.9551],[3339,121.5263],[3342,122.2177],[3346,122.8145],[3376,117.1049],[3386,114.7594],[3410,117.7624],[3419,120.7383],[3423,121.6207],[3426,122.0452],[3454,117.7971],[3464,115.7432],[3466,115.5156],[3491,118.9588],[3499,121.529],[3503,122.4062],[3506,122.8164],[3536,117.1007],[3546,114.7569],[3570,117.8827],[3576,119.9669],[3579,120.8233],[3585,121.9221],[3587,122.0827],[3616,117.2987],[3624,115.7351],[3626,115.5076],[3651,118.955],[3659,121.5263],[3662,122.2176],[3666,122.8144],[3696,120.3722],[3700,115.673],[3706,106.0718],[3730,58.1873],[3736,40.9789],[3739,31.2917],[3743,17.2516],[3746,5.9312],[3774,-119.2979],[3776,-128.4068],[3780,-146.1238],[3784,-163.0027],[3811,-273.7253],[3826,-341.6908],[3856,-458.3436],[3859,-464.9792],[3865,-473.8107],[3866,-474.6777],[3898,-119.2979],[3900,-128.4068],[3904,-146.1238],[3908,-163.0027],[3935,-273.7253],[3950,-341.6908],[3980,-458.3436],[3983,-464.9792],[3989,-473.8107],[3990,-474.6777],[4022,-119.2979],[4024,-128.4068],[4028,-146.1238],[4032,-163.0027],[4059,-273.7253],[4074,-341.6908],[4104,-458.3436],[4107,-464.9792],[4113,-473.8107],[4114,-474.6777],[4147,-119.2979],[4149,-128.4068],[4153,-146.1238],[4157,-163.0027],[4184,-273.7253],[4199,-341.6908],[4229,-458.3436],[4232,-464.9792],[4238,-473.8107],[4239,-474.6777],[4273,-119.2979],[4275,-128.4068],[4279,-146.1238],[4283,-163.0027],[4310,-273.7253],[4325,-341.6908],[4355,-458.3436],[4360,20.353],[4364,-481.0431],[4365,-476.1614],[4372,-164.2368],[4376,279.9738],[4403,6008.7422],[4405,6475.5679],[4409,7363.2725],[4413,8163.7563],[4415,8256.8115],[4435,8257.2207],[4436,8221.9746]]
       }
       foot_l_ik_ctrl.Location.Z {
        type=MovieSceneScriptingFloatChannel
        default=4.7335
        thinned_from=292
        keys [[1780,1108.2777],[1791,1108.3147],[1815,1116.2009],[1859,1110.6464],[1871,1121.5361],[1896,1148.0253],[1951,1182.0576],[1975,1207.0867],[2025,1236.1387],[2060,1271.5021],[2101,1293.3585],[2152,1337.8275],[2178,1407.5875],[2191,1488.3726],[2216,1692.3982],[2231,1829.8622],[2271,2225.1726],[2295,2454.5991],[2311,2577.5017],[2339,2720.9849],[2351,2750.7136],[2391,2761.2217],[2432,2755.823],[2468,2760.4038],[2471,2752.1919],[2487,2748.7561],[2496,2757.3862],[2501,2768.8977],[2503,2783.8176],[2508,2786.8906],[2512,2782.7893],[2513,2776.011],[2515,2771.7563],[2519,2776.8064],[2526,2768.8469],[2764,2814.7476],[2769,2800.9888],[2773,2797.5369],[2938,2802.134],[2986,3093.2786],[3058,3093.2786],[3070,2919.6492],[3083,2919.6492],[3099,3148.0798],[3185,3148.0798],[3216,2901.0784],[3226,2971.033],[3267,3335.5332],[3306,3632.3733],[3331,3760.282],[3346,3809.0679],[3386,3902.175],[3410,3939.6973],[3466,3922.0466],[3506,3936.2668],[3546,3922.4341],[3587,3934.1228],[3626,3922.0449],[3666,3936.2629],[3706,3908.9934],[3730,3919.7649],[3746,3892.3857],[3776,3826.4158],[3786,3824.5303],[3826,3836.6177],[3866,3828.2192],[3910,3824.5303],[3950,3836.6177],[3990,3828.2192],[4034,3824.5303],[4074,3836.6177],[4114,3828.2192],[4159,3824.5303],[4199,3836.6177],[4239,3828.2192],[4285,3824.5303],[4325,3836.6177],[4355,3823.4155],[4360,4041.0007],[4364,3850.5149],[4365,3853.8342],[4372,4909.3555],[4376,7407.2598],[4403,15123.6934],[4405,14912.8711],[4413,13765.5186],[4415,13705.1855],[4436,13693.8604]]
       }
       foot_l_ik_ctrl.Rotation.X {
        type=MovieSceneScriptingFloatChannel
        default=34.5449
        thinned_from=308
        keys [[1780,34.5449],[1791,34.5325],[1815,37.5337],[1821,38.4716],[1824,38.8449],[1828,39.1928],[1832,39.3206],[1861,35.3571],[1865,34.8842],[1869,34.6184],[1871,34.5823],[1896,37.8856],[1900,38.5192],[1904,39.0224],[1907,39.2793],[1911,39.4146],[1941,35.2966],[1945,34.8305],[1949,34.5694],[1951,34.534],[1975,37.6402],[1981,38.5785],[1984,38.9385],[1988,39.2507],[1991,39.3275],[2021,35.325],[2025,34.872],[2029,34.6182],[2031,34.5838],[2056,37.8871],[2060,38.5208],[2064,39.0239],[2068,39.3386],[2071,39.4161],[2101,35.2955],[2104,34.9288],[2109,34.5679],[2111,34.5325],[2135,37.5337],[2141,38.4716],[2144,38.8449],[2148,39.1928],[2152,39.3206],[2181,35.3571],[2185,34.8842],[2189,34.6184],[2191,34.5823],[2216,37.8856],[2220,38.5192],[2224,39.0224],[2227,39.2793],[2231,39.4146],[2261,35.2966],[2265,34.8305],[2270,34.543],[2271,34.534],[2295,37.6402],[2301,38.5785],[2304,38.9385],[2308,39.2507],[2311,39.3275],[2341,35.325],[2345,34.872],[2349,34.6182],[2351,34.5838],[2376,37.8871],[2380,38.5208],[2384,39.0239],[2388,39.3386],[2391,39.4161],[2421,35.2955],[2424,34.9288],[2429,34.5679],[2431,34.5325],[2455,37.7426],[2461,38.676],[2466,39.1828],[2470,39.3237],[2471,39.2812],[2484,64.0611],[2496,72.4186],[2501,75.7616],[2502,75.1091],[2503,73.338],[2504,70.728],[2505,67.5587],[2507,60.6608],[2508,57.4915],[2510,53.1104],[2511,52.4578],[2513,58.2096],[2515,52.4214],[2518,46.6333],[2519,49.3913],[2521,40.3807],[2522,34.1859],[2523,31.3701],[2524,31.4033],[2526,31.6567],[2530,32.796],[2531,33.1887],[2533,34.0747],[2536,35.5937],[2538,36.6851],[2760,39.9468],[2764,41.8532],[2767,43.0163],[2784,48.044],[2791,49.9682],[2927,65.3703],[2938,47.1679],[2986,73.2666],[3058,73.2666],[3070,-7.8213],[3083,-7.8213],[3099,15.1881],[3185,15.1881],[3216,70.029],[3219,67.4902],[3225,62.8306],[3226,62.1431],[3250,56.5113],[3256,57.4559],[3259,57.8311],[3265,58.2746],[3267,58.3075],[3296,54.3461],[3300,53.8649],[3304,53.5892],[3306,53.5484],[3331,56.8662],[3335,57.5053],[3339,58.0109],[3342,58.2676],[3346,58.4011],[3376,54.2851],[3380,53.8177],[3386,53.5138],[3410,56.618],[3416,57.563],[3419,57.9248],[3423,58.2379],[3426,58.3146],[3456,54.3142],[3460,53.8528],[3464,53.5891],[3466,53.5501],[3491,56.8667],[3495,57.5059],[3499,58.0119],[3503,58.3266],[3506,58.4028],[3539,53.9169],[3546,53.5108],[3570,56.5113],[3576,57.4559],[3579,57.8311],[3585,58.2746],[3587,58.3075],[3616,54.3461],[3620,53.8649],[3624,53.5892],[3626,53.5485],[3651,56.8662],[3655,57.5053],[3659,58.0109],[3662,58.2676],[3666,58.4011],[3696,33.4892],[3700,29.9764],[3703,29.1118],[3706,28.4253],[3730,26.8138],[3736,26.781],[3739,26.7175],[3743,26.5269],[3746,26.2709],[3774,21.0186],[3776,20.7194],[3780,20.2651],[3784,20.047],[3786,20.0428],[3811,24.2704],[3815,25.1298],[3819,25.8741],[3823,26.4418],[3826,26.7143],[3859,24.2571],[3863,24.1243],[3866,24.1693],[3898,21.0186],[3900,20.7194],[3904,20.2651],[3908,20.047],[3910,20.0428],[3935,24.2704],[3939,25.1298],[3943,25.8741],[3947,26.4418],[3950,26.7143],[3983,24.2571],[3987,24.1243],[3990,24.1693],[4022,21.0186],[4024,20.7194],[4028,20.2651],[4032,20.047],[4034,20.0428],[4059,24.2704],[4063,25.1298],[4067,25.8741],[4071,26.4418],[4074,26.7143],[4107,24.2571],[4111,24.1243],[4114,24.1693],[4147,21.0186],[4149,20.7194],[4153,20.2651],[4157,20.047],[4159,20.0428],[4184,24.2704],[4188,25.1298],[4192,25.8741],[4196,26.4418],[4199,26.7143],[4232,24.2571],[4236,24.1243],[4239,24.1693],[4273,21.0186],[4275,20.7194],[4279,20.2651],[4283,20.047],[4285,20.0428],[4310,24.2704],[4314,25.1298],[4318,25.8741],[4322,26.4418],[4325,26.7143],[4355,24.4714],[4360,22.4474],[4364,41.1235],[4365,40.031],[4372,119.1687],[4376,152.1479],[4403,155.8245],[4415,154.7321],[4422,155.336],[4429,156.5355],[4435,157.0821],[4436,120.0036]]
       }
       foot_l_ik_ctrl.Rotation.Y {
        type=MovieSceneScriptingFloatChannel
        default=2.5255
        thinned_from=308
        keys [[1780,2.5255],[1791,2.5148],[1815,-0.6385],[1821,-0.9818],[1824,-0.9195],[1828,-0.6681],[1830,-0.4839],[1832,-0.269],[1859,2.5314],[1861,2.5022],[1865,2.2857],[1869,1.898],[1871,1.6547],[1896,-1.594],[1900,-1.7071],[1904,-1.6107],[1907,-1.4258],[1911,-1.0583],[1941,2.5207],[1945,2.6171],[1948,2.5463],[1949,2.4937],[1951,2.3505],[1975,-0.7147],[1981,-0.9812],[1984,-0.919],[1988,-0.6675],[1991,-0.3794],[2019,2.5315],[2021,2.5022],[2025,2.2858],[2029,1.8981],[2031,1.6549],[2056,-1.5934],[2060,-1.7066],[2064,-1.6058],[2068,-1.3295],[2071,-1.0305],[2101,2.5626],[2104,2.6202],[2108,2.503],[2109,2.4407],[2110,2.3668],[2111,2.2822],[2135,-0.7283],[2141,-0.9818],[2144,-0.9195],[2148,-0.668],[2150,-0.4839],[2152,-0.269],[2179,2.5314],[2181,2.5022],[2185,2.2857],[2189,1.898],[2191,1.6547],[2216,-1.594],[2220,-1.7071],[2224,-1.6107],[2227,-1.4258],[2231,-1.0583],[2261,2.5206],[2265,2.6171],[2268,2.5463],[2269,2.4937],[2270,2.4281],[2271,2.3505],[2295,-0.7147],[2301,-0.9812],[2304,-0.919],[2308,-0.6675],[2311,-0.3794],[2339,2.5315],[2341,2.5022],[2345,2.2858],[2349,1.8981],[2351,1.6549],[2376,-1.5934],[2380,-1.7066],[2384,-1.6058],[2388,-1.3295],[2391,-1.0305],[2421,2.5626],[2424,2.6202],[2428,2.503],[2429,2.4407],[2430,2.3668],[2431,2.2822],[2455,-0.7283],[2461,-0.9818],[2464,-0.7041],[2468,0.0262],[2470,0.2679],[2471,0.3039],[2484,22.2028],[2487,21.1627],[2496,10.6667],[2501,1.1791],[2502,-3.3257],[2503,-3.6752],[2504,-4.006],[2505,-4.3192],[2507,-4.894],[2508,-5.1299],[2510,-5.4344],[2511,-5.4773],[2512,-3.3584],[2513,1.3033],[2515,14.2862],[2518,22.9594],[2519,23.7339],[2521,25.9259],[2522,27.1144],[2523,28.0181],[2524,28.3771],[2526,28.2778],[2530,27.6581],[2533,27.0535],[2536,26.5469],[2538,26.3605],[2760,19.7164],[2764,11.6143],[2766,8.5055],[2767,7.0271],[2768,5.6398],[2769,4.3741],[2773,1.1347],[2778,-0.5403],[2780,-1.1509],[2782,-1.702],[2784,-2.1768],[2791,-2.9915],[2927,-1.9249],[2938,-21.8563],[2986,11.4627],[3058,11.4627],[3070,65.0087],[3083,65.0087],[3099,-11.2539],[3185,-11.2539],[3216,4.4559],[3219,4.544],[3223,4.3981],[3224,4.3183],[3225,4.2224],[3226,4.1114],[3250,0.2019],[3256,-0.1817],[3259,-0.1692],[3263,0.0441],[3265,0.2233],[3267,0.4408],[3294,3.7142],[3296,3.7287],[3300,3.591],[3304,3.2615],[3306,3.0364],[3331,-0.654],[3335,-0.8599],[3339,-0.8317],[3342,-0.6725],[3346,-0.3057],[3376,3.7622],[3380,3.9133],[3383,3.871],[3384,3.8249],[3385,3.7642],[3386,3.6892],[3410,0.2111],[3416,-0.1858],[3419,-0.1736],[3423,0.0395],[3426,0.3231],[3454,3.7133],[3456,3.7281],[3460,3.5908],[3464,3.2616],[3466,3.0366],[3491,-0.633],[3495,-0.8411],[3499,-0.8153],[3503,-0.5799],[3506,-0.2857],[3536,3.798],[3539,3.902],[3543,3.8298],[3544,3.7747],[3545,3.7061],[3546,3.6245],[3570,0.2019],[3576,-0.1817],[3579,-0.1692],[3583,0.0441],[3585,0.2233],[3587,0.4408],[3614,3.7142],[3616,3.7288],[3620,3.591],[3624,3.2615],[3626,3.0364],[3651,-0.654],[3655,-0.8599],[3659,-0.8317],[3662,-0.6725],[3666,-0.3057],[3696,2.3856],[3700,2.2331],[3703,2.1029],[3704,2.0326],[3705,1.9502],[3706,1.8568],[3730,-1.2141],[3736,-1.4267],[3739,-1.3505],[3743,-1.1051],[3746,-0.8431],[3774,1.4269],[3776,1.3642],[3780,1.102],[3784,0.7005],[3786,0.4637],[3811,-2.1808],[3815,-2.2099],[3819,-2.055],[3823,-1.7603],[3826,-1.4706],[3856,1.7225],[3859,1.7631],[3863,1.6415],[3864,1.5816],[3865,1.5117],[3866,1.4327],[3898,1.4269],[3900,1.3642],[3904,1.102],[3908,0.7005],[3910,0.4637],[3935,-2.1808],[3939,-2.2099],[3943,-2.055],[3947,-1.7603],[3950,-1.4706],[3980,1.7225],[3983,1.7631],[3987,1.6415],[3988,1.5816],[3989,1.5117],[3990,1.4327],[4022,1.4269],[4024,1.3642],[4028,1.102],[4032,0.7005],[4034,0.4637],[4059,-2.1808],[4063,-2.2099],[4067,-2.055],[4071,-1.7603],[4074,-1.4706],[4104,1.7225],[4107,1.7631],[4111,1.6415],[4112,1.5816],[4113,1.5117],[4114,1.4327],[4147,1.4269],[4149,1.3642],[4153,1.102],[4157,0.7005],[4159,0.4637],[4184,-2.1808],[4188,-2.2099],[4192,-2.055],[4196,-1.7603],[4199,-1.4706],[4229,1.7225],[4232,1.7631],[4236,1.6415],[4237,1.5816],[4238,1.5117],[4239,1.4327],[4273,1.4269],[4275,1.3642],[4279,1.102],[4283,0.7005],[4285,0.4637],[4310,-2.1808],[4314,-2.2099],[4318,-2.055],[4322,-1.7603],[4325,-1.4706],[4355,1.7225],[4360,-21.976],[4364,-3.1954],[4365,-0.6083],[4372,1.9286],[4376,2.073],[4403,2.2434],[4405,2.377],[4409,2.7087],[4413,3.0725],[4415,3.2461],[4422,2.8422],[4429,1.4635],[4434,0.7163],[4435,0.6477],[4436,-0.0004]]
       }
       foot_l_ik_ctrl.Rotation.Z {
        type=MovieSceneScriptingFloatChannel
        default=-4.5125
        thinned_from=308
        keys [[1780,-4.5125],[1791,-4.5075],[1821,-2.2939],[1824,-2.1516],[1828,-2.0763],[1830,-2.0917],[1832,-2.1359],[1859,-4.0583],[1861,-4.1882],[1865,-4.3941],[1869,-4.5083],[1871,-4.5237],[1900,-2.3176],[1904,-2.1242],[1907,-2.076],[1911,-2.1327],[1941,-4.1923],[1945,-4.3891],[1948,-4.4803],[1949,-4.4981],[1951,-4.5127],[1975,-2.756],[1981,-2.2907],[1984,-2.1404],[1988,-2.0606],[1991,-2.0949],[2019,-4.0553],[2021,-4.186],[2025,-4.3932],[2029,-4.5082],[2031,-4.5237],[2056,-2.6848],[2060,-2.3738],[2064,-2.16],[2068,-2.0805],[2071,-2.116],[2101,-4.2253],[2104,-4.3777],[2108,-4.4964],[2110,-4.5125],[2111,-4.5075],[2141,-2.2939],[2144,-2.1516],[2148,-2.0763],[2150,-2.0917],[2152,-2.1359],[2179,-4.0583],[2181,-4.1882],[2185,-4.3941],[2189,-4.5083],[2191,-4.5237],[2220,-2.3176],[2224,-2.1242],[2227,-2.076],[2231,-2.1353],[2261,10.2788],[2265,18.6722],[2268,26.015],[2269,28.6418],[2270,31.3519],[2271,34.1425],[2295,113.3895],[2301,132.5009],[2304,141.2517],[2308,151.799],[2311,158.7181],[2339,176.4917],[2351,176.5697],[2384,180.4301],[2391,180.7901],[2431,180.1882],[2466,184.1413],[2468,182.2738],[2470,177.3858],[2471,174.2255],[2484,160.753],[2487,161.2224],[2496,166.2152],[2501,171.0572],[2502,174.6586],[2503,176.3994],[2508,179.0515],[2511,179.7298],[2512,178.0546],[2513,174.9077],[2515,173.2324],[2518,175.9416],[2519,178.6507],[2521,177.5389],[2523,175.7753],[2524,175.4637],[2526,175.6647],[2531,177.0449],[2538,180.3267],[2760,189.2735],[2769,192.9796],[2773,194.312],[2780,195.7264],[2791,195.6758],[2927,204.881],[2938,199.6817],[2986,231.0827],[3058,231.0827],[3070,80.1465],[3083,80.1465],[3099,182.1491],[3185,182.1491],[3216,198.02],[3226,197.0628],[3250,197.1754],[3267,197.8928],[3306,196.2597],[3346,197.6366],[3386,196.5015],[3426,197.8957],[3466,196.2597],[3506,197.6639],[3546,196.4795],[3587,197.8928],[3626,196.2597],[3666,197.6366],[3706,195.2679],[3739,198.1066],[3746,198.045],[3780,195.1726],[3786,195.2059],[3819,198.2619],[3826,198.1624],[3859,195.213],[3910,195.2059],[3943,198.2619],[3950,198.1624],[3983,195.213],[4034,195.2059],[4067,198.2619],[4074,198.1624],[4107,195.213],[4159,195.2059],[4192,198.2619],[4199,198.1624],[4232,195.213],[4285,195.2059],[4318,198.2619],[4325,198.1624],[4355,195.3688],[4360,184.6976],[4364,200.9847],[4365,194.3046],[4372,63.5786],[4376,45.8159],[4403,45.2241],[4405,45.2527],[4409,45.0913],[4413,44.6922],[4415,44.4237],[4422,42.7791],[4429,40.8092],[4434,39.9879],[4435,39.9358],[4436,40.9989]]
       }
       foot_l_ik_ctrl.Scale.X {
        type=MovieSceneScriptingFloatChannel
        default=1
        thinned_from=308
        keys [[1780,1.25],[2519,1.25],[2521,1.2669],[2538,1.555],[2760,2.8819],[2764,2.9592],[2767,2.993],[2769,3],[2927,3.0152],[2938,3.0096],[3185,3.0136],[3226,3],[4355,3],[4360,2.9944],[4365,3],[4436,3]]
       }
       foot_l_ik_ctrl.Scale.Y {
        type=MovieSceneScriptingFloatChannel
        default=1
        thinned_from=308
        keys [[1780,1.25],[2519,1.25],[2521,1.2669],[2538,1.555],[2760,2.8819],[2764,2.9592],[2767,2.993],[2769,3],[2927,3.0152],[2938,3.0096],[3185,3.0136],[3226,3],[4355,3],[4360,2.9944],[4365,3],[4436,3]]
       }
       foot_l_ik_ctrl.Scale.Z {
        type=MovieSceneScriptingFloatChannel
        default=1
        thinned_from=308
        keys [[1780,1.25],[2519,1.25],[2521,1.2669],[2538,1.555],[2760,2.8819],[2764,2.9592],[2767,2.993],[2769,3],[2927,3.0152],[2938,3.0096],[3185,3.0136],[3226,3],[4355,3],[4360,2.9944],[4365,3],[4436,3]]
       }
       ball_l_ik_ctrl.Location.X {type=MovieSceneScriptingFloatChannel, default=0, thinned_from=1243, keys [[1790,0],[4436,-0.0001]]}
       ball_l_ik_ctrl.Location.Y {type=MovieSceneScriptingFloatChannel, default=0, thinned_from=1248, keys [[1790,0],[4435,0],[4436,-0.0006]]}
       ball_l_ik_ctrl.Location.Z {type=MovieSceneScriptingFloatChannel, default=0, thinned_from=1241, keys [[1790,0],[4435,0],[4436,0.0004]]}
       ball_l_ik_ctrl.Rotation.X {
        type=MovieSceneScriptingFloatChannel
        default=-0.5199
        thinned_from=1279
        keys [[1790,-0.5199],[2470,-0.5199],[2487,-0.2481],[2501,0],[2512,0],[2515,-2.2701],[2519,-15.687],[2521,-4.2023],[2523,0],[2763,-0.0733],[2776,-0.8803],[2789,-3.236],[2938,-3.2244],[2986,-2.0016],[3058,-2.0016],[3070,-1.4456],[3185,-1.4456],[3215,-0.5199],[4435,-0.5199],[4436,-0.0002]]
       }
       ball_l_ik_ctrl.Rotation.Y {
        type=MovieSceneScriptingFloatChannel
        default=-1.3124
        thinned_from=1106
        keys [[1790,-1.3124],[2469,-1.3124],[2487,-0.5984],[2501,0],[2512,0],[2515,0.2065],[2519,1.4006],[2521,0.3817],[2523,0],[2775,-0.7276],[2783,-0.1035],[2792,0],[2804,0],[2927,-0.1344],[2938,-0.1322],[2986,-0.1141],[3058,-0.1141],[3070,-0.016],[3185,-0.016],[3216,-1.3124],[4435,-1.3124],[4436,-0.0006]]
       }
       ball_l_ik_ctrl.Rotation.Z {
        type=MovieSceneScriptingFloatChannel
        default=2.9187
        thinned_from=607
        keys [[1790,2.9187],[2471,2.9187],[2487,1.4593],[2501,0],[2508,0],[2515,-0.0149],[2519,-0.2683],[2523,0],[2767,0.8367],[2775,1.6189],[2784,0.1373],[2793,0.0002],[2938,0.0001],[2986,-0.0018],[3185,-0.0018],[3215,2.9187],[4435,2.9187],[4436,0.0007]]
       }
       ball_l_ik_ctrl.Scale.X {type=MovieSceneScriptingFloatChannel, default=1, const [1,14]}
       ball_l_ik_ctrl.Scale.Y {type=MovieSceneScriptingFloatChannel, default=1, const [1,14]}
       ball_l_ik_ctrl.Scale.Z {type=MovieSceneScriptingFloatChannel, default=1, const [1,14]}
       foot_bk1_l_ctrl.Location.X {type=MovieSceneScriptingFloatChannel, default=0, const [0,12]}
       foot_bk1_l_ctrl.Location.Y {type=MovieSceneScriptingFloatChannel, default=0, const [0,12]}
       foot_bk1_l_ctrl.Location.Z {type=MovieSceneScriptingFloatChannel, default=0, const [0,12]}
       foot_bk1_l_ctrl.Rotation.X {type=MovieSceneScriptingFloatChannel, default=0, const [0,12]}
       foot_bk1_l_ctrl.Rotation.Y {type=MovieSceneScriptingFloatChannel, default=0, const [0,12]}
       foot_bk1_l_ctrl.Rotation.Z {type=MovieSceneScriptingFloatChannel, default=0, const [0,12]}
       foot_bk1_l_ctrl.Scale.X {type=MovieSceneScriptingFloatChannel, default=1, const [1,12]}
       foot_bk1_l_ctrl.Scale.Y {type=MovieSceneScriptingFloatChannel, default=1, const [1,12]}
       foot_bk1_l_ctrl.Scale.Z {type=MovieSceneScriptingFloatChannel, default=1, const [1,12]}
       foot_roll_l_ctrl.Location.X {type=MovieSceneScriptingFloatChannel, default=0, const [0,12]}
       foot_roll_l_ctrl.Location.Y {type=MovieSceneScriptingFloatChannel, default=0, const [0,12]}
       foot_roll_l_ctrl.Location.Z {type=MovieSceneScriptingFloatChannel, default=0, const [0,12]}
       foot_roll_l_ctrl.Rotation.X {type=MovieSceneScriptingFloatChannel, default=0, const [0,12]}
       foot_roll_l_ctrl.Rotation.Y {type=MovieSceneScriptingFloatChannel, default=0, const [0,12]}
       foot_roll_l_ctrl.Rotation.Z {type=MovieSceneScriptingFloatChannel, default=0, const [0,12]}
       foot_roll_l_ctrl.Scale.X {type=MovieSceneScriptingFloatChannel, default=1, const [1,12]}
       foot_roll_l_ctrl.Scale.Y {type=MovieSceneScriptingFloatChannel, default=1, const [1,12]}
       foot_roll_l_ctrl.Scale.Z {type=MovieSceneScriptingFloatChannel, default=1, const [1,12]}
       roll_blend_l_ctrl {type=MovieSceneScriptingFloatChannel, default=1, const [1,13]}
       tip_l_ctrl.Location.X {type=MovieSceneScriptingFloatChannel, default=0, const [0,12]}
       tip_l_ctrl.Location.Y {type=MovieSceneScriptingFloatChannel, default=0, const [0,12]}
       tip_l_ctrl.Location.Z {type=MovieSceneScriptingFloatChannel, default=0, const [0,12]}
       tip_l_ctrl.Rotation.X {type=MovieSceneScriptingFloatChannel, default=0, const [0,12]}
       tip_l_ctrl.Rotation.Y {type=MovieSceneScriptingFloatChannel, default=0, const [0,12]}
       tip_l_ctrl.Rotation.Z {type=MovieSceneScriptingFloatChannel, default=0, const [0,12]}
       tip_l_ctrl.Scale.X {type=MovieSceneScriptingFloatChannel, default=1, const [1,12]}
       tip_l_ctrl.Scale.Y {type=MovieSceneScriptingFloatChannel, default=1, const [1,12]}
       tip_l_ctrl.Scale.Z {type=MovieSceneScriptingFloatChannel, default=1, const [1,12]}
       heel_l_ctrl.Location.X {type=MovieSceneScriptingFloatChannel, default=0, const [0,12]}
       heel_l_ctrl.Location.Y {type=MovieSceneScriptingFloatChannel, default=0, const [0,12]}
       heel_l_ctrl.Location.Z {type=MovieSceneScriptingFloatChannel, default=0, const [0,12]}
       heel_l_ctrl.Rotation.X {type=MovieSceneScriptingFloatChannel, default=0, const [0,12]}
       heel_l_ctrl.Rotation.Y {type=MovieSceneScriptingFloatChannel, default=0, const [0,12]}
       heel_l_ctrl.Rotation.Z {type=MovieSceneScriptingFloatChannel, default=0, const [0,12]}
       heel_l_ctrl.Scale.X {type=MovieSceneScriptingFloatChannel, default=1, const [1,12]}
       heel_l_ctrl.Scale.Y {type=MovieSceneScriptingFloatChannel, default=1, const [1,12]}
       heel_l_ctrl.Scale.Z {type=MovieSceneScriptingFloatChannel, default=1, const [1,12]}
       leg_l_softness {type=MovieSceneScriptingFloatChannel, default=0, const [0,12]}
       leg_l_pv_ik_ctrl.Location.X {
        type=MovieSceneScriptingFloatChannel
        default=-0.5179
        thinned_from=242
        keys [[1780,1140.7158],[1791,1155.9548],[1824,1159.2324],[1854,1153.2798],[1871,1169.6221],[1909,1199.5577],[1951,1229.7499],[1989,1258.8983],[2014,1273.5497],[2065,1317.7213],[2094,1333.6133],[2111,1349.9036],[2144,1376.6475],[2151,1380.0742],[2174,1436.333],[2188,1514.5146],[2191,1534.8986],[2225,1814.972],[2254,2092.7666],[2271,2264.6826],[2305,2566.3796],[2309,2595.3535],[2334,2738.1318],[2348,2782.2832],[2351,2788.2446],[2389,2796.0527],[2414,2790.5972],[2466,2796.4707],[2468,2838.498],[2469,2794.4436],[2484,2750.0527],[2487,2793.4514],[2496,2775.335],[2501,2802.0107],[2502,2812.6628],[2504,2823.665],[2508,2832.9832],[2511,2850.988],[2512,2845.2729],[2513,2825.7874],[2515,2806.9712],[2523,2787.1733],[2529,2799.0955],[2759,2876.2283],[2765,2891.6711],[2772,2885.0449],[2783,2896.123],[2801,2889.8032],[2927,2874.7661],[2938,2947.6375],[2986,3126.1575],[3058,3126.1575],[3070,2957.5708],[3083,2957.5708],[3099,3261.6326],[3185,3261.6326],[3223,2962.1704],[3259,3311.1846],[3289,3556.7927],[3306,3682.4446],[3340,3833.9956],[3386,3950.6692],[3420,3979.7124],[3449,3963.489],[3504,3979.0427],[3529,3964.4814],[3579,3980.1909],[3609,3963.6665],[3664,3979.0305],[3689,3995.8987],[3706,4034.8628],[3740,4038.6648],[3769,3978.4443],[3786,3971.3862],[3824,3972.1663],[3849,3958.5728],[3893,3978.4443],[3910,3971.3862],[3948,3972.1663],[3973,3958.5728],[4017,3978.4443],[4034,3971.3862],[4072,3972.1663],[4097,3958.5728],[4142,3978.4443],[4159,3971.3862],[4197,3972.1663],[4222,3958.5728],[4268,3978.4443],[4285,3971.3862],[4323,3972.1663],[4348,3958.5728],[4360,4212.6265],[4364,4539.4473],[4375,5830.6001],[4398,12040.5771],[4411,9910.1826],[4412,9686.96],[4414,9122.1982],[4415,9151.6719],[4435,9147.0264],[4436,9177.9648]]
       }
       leg_l_pv_ik_ctrl.Location.Y {
        type=MovieSceneScriptingFloatChannel
        default=9.5088
        thinned_from=242
        keys [[1780,-6308.9673],[1791,-6290.9604],[1871,-6290.9092],[1905,-6176.9258],[1909,-6153.3135],[1934,-5939.9873],[1951,-5743.8188],[1985,-5249.6777],[2014,-4743.4458],[2111,-2923.2327],[2144,-2335.0105],[2151,-2213.9626],[2174,-1829.4548],[2191,-1563.7102],[2222,-1133.9896],[2229,-1046.4081],[2254,-765.7035],[2267,-639.2153],[2271,-606.9105],[2302,-494.3],[2309,-482.054],[2334,-454.5389],[2351,-454.8312],[2389,-451.8592],[2431,-456.4579],[2466,-452.1471],[2468,-465.2686],[2469,-452.3866],[2471,-453.9625],[2484,-397.3776],[2487,-420.9402],[2496,-446.4816],[2501,-454.8102],[2502,-457.9091],[2504,-459.1038],[2508,-457.8244],[2511,-443.2822],[2512,-451.5485],[2513,-465.6261],[2515,-468.791],[2518,-460.3049],[2519,-455.2374],[2521,-453.7875],[2523,-458.7758],[2526,-462.1633],[2529,-462.6314],[2533,-461.5856],[2759,-524.0078],[2766,-527.5268],[2778,-526.0168],[2801,-528.3572],[2927,-482.5403],[2938,-731.6172],[2986,-467.0431],[3058,-467.0431],[3070,-316.3887],[3083,-316.3887],[3099,-281.1798],[3185,-281.1798],[3223,-511.49],[3226,-495.632],[3257,-308.028],[3259,-296.5366],[3266,-257.9842],[3289,-157.3143],[3302,-121.2224],[3305,-114.8708],[3306,-113.2725],[3337,-90.0526],[3340,-89.076],[3344,-88.0522],[3369,-93.0957],[3383,-98.1531],[3386,-98.8896],[3420,-88.4895],[3424,-87.5662],[3449,-92.6694],[3461,-97.1889],[3466,-98.4535],[3500,-88.4811],[3504,-87.5746],[3529,-92.5973],[3543,-98.0363],[3546,-98.7149],[3579,-88.7241],[3586,-87.397],[3609,-92.5557],[3623,-97.7831],[3626,-98.5324],[3660,-88.4554],[3664,-87.5741],[3689,-74.8633],[3702,-55.7476],[3706,-60.2653],[3737,-113.1506],[3740,-122.0848],[3744,-135.3683],[3769,-240.8939],[3781,-293.9897],[3786,-314.9231],[3817,-448.617],[3824,-481.9814],[3849,-591.3461],[3863,-626.0956],[3866,-629.5503],[3893,-240.8939],[3905,-293.9897],[3910,-314.9231],[3941,-448.617],[3948,-481.9814],[3973,-591.3461],[3987,-626.0956],[3990,-629.5503],[4017,-240.8939],[4029,-293.9897],[4034,-314.9231],[4065,-448.617],[4072,-481.9814],[4097,-591.3461],[4111,-626.0956],[4114,-629.5503],[4142,-240.8939],[4154,-293.9897],[4159,-314.9231],[4190,-448.617],[4197,-481.9814],[4222,-591.3461],[4236,-626.0956],[4239,-629.5503],[4268,-240.8939],[4280,-293.9897],[4285,-314.9231],[4316,-448.617],[4323,-481.9814],[4348,-591.3461],[4360,-236.8545],[4364,-10.1134],[4375,491.1156],[4398,6350.8447],[4412,9999.5137],[4414,10363.4531],[4415,10336.2275],[4435,10336.2402],[4436,10306.3154]]
       }
       leg_l_pv_ik_ctrl.Location.Z {
        type=MovieSceneScriptingFloatChannel
        default=11.845
        thinned_from=242
        keys [[1780,497.418],[1791,501.2488],[1824,507.1737],[1831,507.1298],[1854,503.0282],[1871,502.9332],[1905,506.2078],[1909,505.4051],[1934,493.4285],[1951,483.8181],[1985,469.8509],[1989,467.2067],[2031,431.175],[2065,410.847],[2069,407.7577],[2094,384.1425],[2111,369.5239],[2144,351.2003],[2174,330.7289],[2188,326.6629],[2191,326.502],[2229,338.6645],[2254,354.5282],[2267,378.5916],[2271,392.0577],[2302,449.8445],[2305,447.4868],[2309,443.0636],[2334,440.19],[2351,444.8981],[2389,438.1382],[2431,442.2017],[2466,435.597],[2468,428.6408],[2469,439.516],[2471,445.7273],[2484,465.3438],[2487,445.8047],[2496,442.6983],[2501,434.8195],[2502,431.1951],[2504,429.9043],[2508,429.0106],[2511,430.3578],[2512,429.2764],[2513,432.2683],[2515,440.2603],[2518,446.2188],[2519,446.8447],[2521,445.5858],[2524,446.323],[2529,450.6833],[2533,452.0408],[2759,471.7284],[2772,450.8545],[2778,446.5684],[2801,442.6521],[2927,451.5508],[2938,139.2234],[2986,498.6219],[3058,498.6219],[3070,471.4152],[3083,471.4152],[3099,677.2726],[3185,677.2726],[3223,428.4539],[3226,431.5334],[3266,451.9041],[3289,471.5954],[3306,483.3342],[3340,485.393],[3344,486.6457],[3369,500.7711],[3386,508.2983],[3420,500.4092],[3424,500.2186],[3463,509.3444],[3466,509.4082],[3500,499.1049],[3504,498.9418],[3543,509.937],[3546,510.1244],[3579,500.6517],[3586,500.3418],[3626,509.3885],[3660,499.0291],[3664,498.959],[3689,516.923],[3702,536.4721],[3706,538.4641],[3744,539.9266],[3769,546.5872],[3786,553.2275],[3824,548.7661],[3849,556.4079],[3866,562.2666],[3893,546.5872],[3910,553.2275],[3948,548.7661],[3973,556.4079],[3990,562.2666],[4017,546.5872],[4034,553.2275],[4072,548.7661],[4097,556.4079],[4114,562.2666],[4142,546.5872],[4159,553.2275],[4197,548.7661],[4222,556.4079],[4239,562.2666],[4268,546.5872],[4285,553.2275],[4323,548.7661],[4348,556.4079],[4360,476.7428],[4364,1578.1918],[4375,8061.1416],[4412,45937.7109],[4414,47234.0508],[4436,47247.5508]]
       }
       leg_l_pv_ik_ctrl.Rotation.X {
        type=MovieSceneScriptingFloatChannel
        default=-14.917
        thinned_from=242
        keys [[1780,37.595],[1791,20.8685],[1822,24.8369],[1824,24.2502],[1831,21.7191],[1854,17.8838],[1867,20.6846],[1870,21.573],[1871,21.9322],[1902,27.5569],[1904,27.0429],[1905,26.733],[1909,25.1924],[1934,17.567],[1947,20.1499],[1948,20.3806],[1950,20.7712],[1951,21.0275],[1982,24.6996],[1985,23.8209],[1989,22.3613],[2014,18.1426],[2025,20.1494],[2028,20.9815],[2031,21.9335],[2062,27.869],[2064,27.2544],[2065,26.8763],[2069,25.1918],[2094,19.1141],[2108,20.3665],[2111,20.9284],[2142,24.8633],[2144,24.2529],[2151,21.6871],[2174,17.8211],[2187,20.7714],[2190,21.5114],[2191,21.9322],[2222,27.5811],[2224,27.0429],[2225,26.7264],[2229,25.1172],[2254,-34.9217],[2267,-76.9939],[2268,-79.2352],[2270,-83.1681],[2271,-84.7721],[2302,-90.9234],[2305,-90.0332],[2309,-88.5456],[2334,-91.3299],[2345,-93.3111],[2348,-91.175],[2351,-88.346],[2382,-26.1454],[2384,-24.1131],[2385,-23.3239],[2389,-20.9254],[2414,-60.4115],[2428,-80.583],[2431,-74.0053],[2462,20.5115],[2464,22.4662],[2466,25.3255],[2468,98.0727],[2469,-33.6777],[2470,-70.1837],[2471,-86.5881],[2484,-148.1242],[2487,-204.715],[2496,-194.9058],[2501,41.2334],[2502,-8.859],[2504,-6.907],[2507,-5.5074],[2508,-5.5224],[2511,-6.3794],[2512,9.4059],[2513,21.8525],[2515,-209.9864],[2518,-196.4179],[2519,161.4726],[2523,158.1434],[2524,156.5126],[2526,155.472],[2528,157.804],[2529,158.369],[2759,131.4655],[2765,103.7989],[2766,98.6617],[2769,89.5713],[2772,82.8231],[2777,76.8344],[2778,76.4182],[2783,75.7872],[2801,74.113],[2927,85.6531],[2938,119.0142],[2986,97.1817],[3058,97.1817],[3070,-119.1329],[3083,-119.1329],[3099,-77.5528],[3185,-77.5528],[3223,109.6926],[3226,107.1201],[3257,89.7445],[3259,89.8902],[3266,91.3757],[3289,97.7713],[3302,97.6253],[3337,87.4285],[3340,87.6902],[3344,88.7403],[3369,96.8613],[3382,98.3067],[3386,97.7092],[3417,89.7065],[3420,89.9814],[3424,90.8726],[3449,97.8365],[3460,97.9877],[3463,97.3765],[3497,87.4611],[3500,87.7145],[3504,88.678],[3529,96.9879],[3543,98.301],[3546,97.7543],[3577,89.7554],[3579,89.8901],[3586,91.3669],[3609,97.7779],[3622,97.6197],[3657,87.4308],[3660,87.7003],[3664,88.6898],[3689,85.5221],[3702,69.0632],[3706,67.0882],[3737,55.5652],[3740,55.6475],[3744,56.107],[3769,58.3258],[3780,56.457],[3781,56.1848],[3786,54.2615],[3817,52.0086],[3820,52.7844],[3849,61.6864],[3863,62.0526],[3866,61.459],[3893,58.3258],[3904,56.457],[3905,56.1848],[3910,54.2615],[3941,52.0086],[3944,52.7844],[3973,61.6864],[3987,62.0526],[3990,61.459],[4017,58.3258],[4028,56.457],[4029,56.1848],[4034,54.2615],[4065,52.0086],[4068,52.7844],[4097,61.6864],[4111,62.0526],[4114,61.459],[4142,58.3258],[4153,56.457],[4154,56.1848],[4159,54.2615],[4190,52.0086],[4193,52.7844],[4222,61.6864],[4236,62.0526],[4239,61.459],[4268,58.3258],[4279,56.457],[4280,56.1848],[4285,54.2615],[4316,52.0086],[4319,52.7844],[4348,61.6864],[4360,166.1582],[4364,243.4998],[4375,31.1617],[4398,-66.4837],[4411,134.6775],[4412,140.5834],[4414,12.1718],[4415,27.6909],[4425,28.1788],[4430,28.4706],[4433,28.7129],[4434,28.8913],[4435,30.1714],[4436,38.4529]]
       }
       leg_l_pv_ik_ctrl.Rotation.Y {
        type=MovieSceneScriptingFloatChannel
        default=0.1989
        thinned_from=242
        keys [[1780,-70.9522],[1791,-78.4301],[1822,-82.183],[1824,-82.2042],[1831,-81.8355],[1854,-78.7637],[1868,-78.5485],[1871,-78.7468],[1902,-82.8039],[1905,-82.8037],[1909,-82.5936],[1934,-79.1546],[1948,-78.3202],[1951,-78.4243],[1982,-82.183],[1989,-82.008],[2014,-78.9634],[2026,-78.4755],[2031,-78.7467],[2062,-82.7731],[2065,-82.7825],[2069,-82.5987],[2094,-79.1244],[2108,-78.2983],[2111,-78.4188],[2142,-82.1867],[2144,-82.2043],[2151,-81.833],[2174,-78.7348],[2188,-78.5516],[2191,-78.7468],[2222,-82.8034],[2225,-82.8043],[2229,-82.5811],[2254,-79.9965],[2267,-65.9331],[2268,-63.6872],[2270,-58.7382],[2271,-56.1115],[2302,43.6016],[2304,49.2735],[2305,51.8725],[2309,62.5968],[2334,84.0503],[2346,84.8174],[2385,85.2467],[2414,88.2652],[2431,88.3621],[2462,85.6521],[2466,85.5964],[2468,73.4739],[2469,87.0473],[2470,85.4244],[2471,82.6955],[2484,68.426],[2487,75.6301],[2496,80.1073],[2501,88.3373],[2502,82.1615],[2504,80.5275],[2507,79.6932],[2508,79.6087],[2511,81.7641],[2512,83.0136],[2513,85.9961],[2515,83.0287],[2518,76.1013],[2519,75.0864],[2521,75.5374],[2523,74.2056],[2524,73.7945],[2526,74.0544],[2529,74.7356],[2533,75.3438],[2759,71.9413],[2765,77.9621],[2766,78.2185],[2769,77.7844],[2778,76.1122],[2783,75.6979],[2801,76.0097],[2927,85.9487],[2938,78.2524],[2986,55.7976],[3058,55.7976],[3070,6.674],[3083,6.674],[3099,89.5701],[3185,89.5701],[3223,73.3861],[3226,73.7121],[3266,72.9794],[3289,73.7513],[3306,74.6609],[3340,73.1655],[3344,73.0902],[3369,73.8436],[3386,74.5661],[3424,72.985],[3449,73.7552],[3466,74.6608],[3504,73.0865],[3529,73.8205],[3546,74.5518],[3579,73.1598],[3586,72.9788],[3609,73.7531],[3626,74.6609],[3660,73.165],[3664,73.0929],[3689,74.7611],[3702,76.534],[3706,76.6227],[3737,73.8081],[3744,73.7806],[3769,76.1814],[3781,76.9668],[3786,76.9305],[3817,73.4762],[3824,73.4658],[3849,75.9533],[3863,76.8938],[3893,76.1814],[3905,76.9668],[3910,76.9305],[3941,73.4762],[3948,73.4658],[3973,75.9533],[3987,76.8938],[4017,76.1814],[4029,76.9668],[4034,76.9305],[4065,73.4762],[4072,73.4658],[4097,75.9533],[4111,76.8938],[4142,76.1814],[4154,76.9668],[4159,76.9305],[4190,73.4762],[4197,73.4658],[4222,75.9533],[4236,76.8938],[4268,76.1814],[4280,76.9668],[4285,76.9305],[4316,73.4762],[4323,73.4658],[4348,75.9533],[4360,64.2142],[4364,32.5813],[4375,-42.0567],[4398,-47.8698],[4411,-3.4995],[4412,1.1255],[4414,-35.4048],[4415,-45.534],[4425,-47.84],[4430,-49.685],[4433,-50.513],[4434,-50.7117],[4435,-51.7313],[4436,-51.2527]]
       }
       leg_l_pv_ik_ctrl.Rotation.Z {
        type=MovieSceneScriptingFloatChannel
        default=0.7145
        thinned_from=242
        keys [[1780,-28.7726],[1791,-31.8422],[1822,-31.6029],[1831,-28.4281],[1854,-27.1616],[1868,-31.9495],[1871,-32.9443],[1902,-34.3973],[1904,-33.6786],[1905,-33.2684],[1909,-31.8077],[1934,-29.0362],[1947,-31.1593],[1951,-32.0553],[1982,-31.5736],[1985,-30.4282],[1989,-28.9746],[2014,-27.9676],[2025,-30.9518],[2031,-32.9434],[2062,-34.6451],[2065,-33.4675],[2069,-31.7719],[2094,-28.5613],[2111,-31.9527],[2142,-31.6223],[2151,-28.4254],[2174,-27.0469],[2187,-31.6318],[2191,-32.9442],[2222,-34.3704],[2224,-33.6785],[2225,-33.2805],[2229,-31.9255],[2254,26.9106],[2267,66.2071],[2268,68.4106],[2270,72.3631],[2271,74.0098],[2302,93.326],[2305,94.8039],[2309,96.4506],[2334,99.3353],[2345,97.4832],[2348,99.7998],[2351,102.6583],[2382,160.842],[2385,163.5175],[2389,165.9071],[2414,129.356],[2428,110.7064],[2431,117.3231],[2462,-152.3255],[2464,-150.3844],[2466,-147.258],[2468,-91.5306],[2469,-204.7987],[2470,-240.2753],[2471,105.1161],[2484,35.7119],[2487,-35.2101],[2496,16.0439],[2501,-90.1095],[2502,-137.4791],[2504,-130.6048],[2507,-125.2229],[2508,-121.5161],[2511,-100.1011],[2512,-94.2222],[2513,-105.3139],[2515,-6.9034],[2518,-12.1662],[2519,-22.6571],[2521,-25.3116],[2523,-24.1394],[2524,-23.7975],[2526,-22.6713],[2528,-21.0305],[2529,-21.1193],[2533,-24.9151],[2759,-51.6486],[2765,-74.2782],[2766,-79.1949],[2769,-87.7245],[2772,-94.2541],[2777,-99.6121],[2778,-99.8651],[2801,-98.9431],[2927,-96.6885],[2938,-10.8661],[2986,-88.049],[3058,-88.049],[3070,7.055],[3083,7.055],[3099,138.998],[3185,138.998],[3223,-87.0268],[3226,-87.5651],[3257,-101.1468],[3259,-101.0917],[3266,-99.4994],[3289,-89.8005],[3302,-88.5495],[3306,-89.761],[3337,-103.5985],[3340,-103.4634],[3344,-102.4454],[3369,-90.7493],[3382,-87.8715],[3385,-88.1929],[3386,-88.4266],[3417,-101.1328],[3420,-100.9923],[3424,-100.1082],[3449,-89.8134],[3460,-88.3478],[3463,-88.8341],[3466,-89.7602],[3497,-103.5639],[3500,-103.4404],[3504,-102.4753],[3529,-90.6926],[3543,-87.8415],[3546,-88.3809],[3577,-101.1429],[3579,-101.0917],[3586,-99.4773],[3609,-89.8243],[3622,-88.6123],[3625,-89.3756],[3657,-103.5975],[3660,-103.4591],[3664,-102.467],[3689,-91.0425],[3703,-93.296],[3737,-104.4583],[3740,-104.137],[3744,-103.1673],[3769,-96.1872],[3780,-96.6449],[3783,-97.5171],[3786,-98.7886],[3817,-106.8721],[3820,-106.4339],[3824,-105.3959],[3849,-96.0688],[3863,-94.943],[3866,-95.65],[3893,-96.1872],[3904,-96.6449],[3907,-97.5171],[3910,-98.7886],[3941,-106.8721],[3944,-106.4339],[3948,-105.3959],[3973,-96.0688],[3987,-94.943],[3990,-95.65],[4017,-96.1872],[4028,-96.6449],[4031,-97.5171],[4034,-98.7886],[4065,-106.8721],[4068,-106.4339],[4072,-105.3959],[4097,-96.0688],[4111,-94.943],[4114,-95.65],[4142,-96.1872],[4153,-96.6449],[4156,-97.5171],[4159,-98.7886],[4190,-106.8721],[4193,-106.4339],[4197,-105.3959],[4222,-96.0688],[4236,-94.943],[4239,-95.65],[4268,-96.1872],[4279,-96.6449],[4282,-97.5171],[4285,-98.7886],[4316,-106.8721],[4319,-106.4339],[4323,-105.3959],[4348,-96.0688],[4360,-292.3577],[4364,-88.99],[4375,85.6748],[4398,85.0431],[4411,47.0499],[4412,53.4698],[4414,90.7537],[4415,83.2915],[4433,84.5992],[4434,84.5704],[4435,83.1063],[4436,83.623]]
       }
       leg_l_pv_ik_ctrl.Scale.X {
        type=MovieSceneScriptingFloatChannel
        default=1
        thinned_from=242
        keys [[1780,1.25],[2519,1.25],[2521,1.2669],[2533,1.4703],[2759,2.8585],[2765,2.9732],[2766,2.9845],[2769,3],[2801,3],[2927,2.8584],[2938,2.917],[2986,2.8798],[3058,2.8798],[3070,2.8976],[3083,2.8976],[3099,2.8727],[3185,2.8727],[3223,3],[4348,3],[4360,3.0617],[4375,3],[4436,3]]
       }
       leg_l_pv_ik_ctrl.Scale.Y {
        type=MovieSceneScriptingFloatChannel
        default=1
        thinned_from=242
        keys [[1780,1.25],[2519,1.25],[2521,1.2669],[2533,1.4703],[2759,2.8585],[2765,2.9732],[2766,2.9845],[2769,3],[2801,3],[2927,2.8584],[2938,2.917],[2986,2.8798],[3058,2.8798],[3070,2.8976],[3083,2.8976],[3099,2.8727],[3185,2.8727],[3223,3],[4348,3],[4360,3.0617],[4375,3],[4436,3]]
       }
       leg_l_pv_ik_ctrl.Scale.Z {
        type=MovieSceneScriptingFloatChannel
        default=1
        thinned_from=242
        keys [[1780,1.25],[2519,1.25],[2521,1.2669],[2533,1.4703],[2759,2.8585],[2765,2.9732],[2766,2.9845],[2769,3],[2801,3],[2927,2.8584],[2938,2.917],[2986,2.8798],[3058,2.8798],[3070,2.8976],[3083,2.8976],[3099,2.8727],[3185,2.8727],[3223,3],[4348,3],[4360,3.0617],[4375,3],[4436,3]]
       }
       thigh_r_fk_ctrl.Location.X {type=MovieSceneScriptingFloatChannel, default=0, thinned_from=271, keys [[1790,0],[4365,0]]}
       thigh_r_fk_ctrl.Location.Y {type=MovieSceneScriptingFloatChannel, default=0, thinned_from=1257, keys [[1790,0],[4366,0]]}
       thigh_r_fk_ctrl.Location.Z {type=MovieSceneScriptingFloatChannel, default=0, thinned_from=394, keys [[1790,0],[4365,0]]}
       thigh_r_fk_ctrl.Rotation.X {
        type=MovieSceneScriptingFloatChannel
        default=4.3562
        thinned_from=74
        keys [[1790,4.3562],[1821,2.6071],[1865,4.4327],[1901,2.6094],[1945,4.4325],[1981,2.6071],[2025,4.4327],[2061,2.6094],[2105,4.4327],[2141,2.6071],[2185,4.4327],[2221,2.6094],[2265,4.4325],[2279,2.0011],[2301,2.6071],[2345,4.4327],[2381,2.6094],[2425,4.4327],[2461,2.6071],[2468,2.5243],[2497,2.9627],[2508,2.9627],[2520,2.4963],[2525,7.3824],[2527,7.0841],[2528,7.4874],[2530,7.02],[2532,7.541],[2757,8.8303],[2768,3.8982],[2782,2.8788],[2801,2.7981],[2849,3.1218],[2893,3.3763],[2938,3.5709],[2986,4.8087],[3058,4.8087],[3070,44.9325],[3099,44.9325],[3167,39.4166],[3188,34.177],[3240,5.2324],[3256,2.6071],[3300,4.4327],[3336,2.6094],[3380,4.4325],[3416,2.6071],[3460,4.4327],[3496,2.6094],[3540,4.4327],[3576,2.6071],[3620,4.4326],[3656,2.6094],[3700,4.4325],[3736,2.6071],[3780,4.4327],[3816,2.6094],[3860,4.4327],[3904,4.4327],[3940,2.6094],[3984,4.4327],[4028,4.4327],[4064,2.6094],[4108,4.4327],[4153,4.4327],[4189,2.6094],[4233,4.4327],[4279,4.4327],[4315,2.6094],[4364,2.7871],[4365,2.7873],[4369,-10.7764]]
       }
       thigh_r_fk_ctrl.Rotation.Y {
        type=MovieSceneScriptingFloatChannel
        default=5.8831
        thinned_from=74
        keys [[1790,5.8831],[1818,-3.7958],[1860,6.2437],[1898,-3.8006],[1939,6.2532],[1978,-3.7957],[2020,6.2437],[2058,-3.8006],[2099,6.2532],[2138,-3.7958],[2180,6.2437],[2218,-3.8006],[2259,6.2532],[2279,4.2829],[2298,-3.7957],[2340,6.2437],[2378,-3.8006],[2419,6.2532],[2458,0.019],[2468,-0.3056],[2498,4.3624],[2509,4.3624],[2519,22.6372],[2521,21.2209],[2524,21.9191],[2538,17.6348],[2757,15.7876],[2769,5.7125],[2780,3.6346],[2801,-2.5891],[2849,2.8505],[2893,0.9155],[2937,3.9084],[2938,3.9083],[2986,-5.5477],[3058,-5.5477],[3070,24.3584],[3099,24.3584],[3185,3.1204],[3223,-3.8919],[3295,6.2437],[3333,2.371],[3374,6.2532],[3413,2.376],[3455,6.2437],[3493,2.3711],[3534,6.2532],[3573,2.3758],[3615,6.2437],[3653,2.371],[3685,6.0813],[3694,6.2532],[3719,1.9179],[3722,1.6374],[3762,4.6214],[3777,5.3066],[3813,2.3711],[3854,6.2532],[3899,6.2437],[3937,2.3711],[3978,6.2532],[4023,6.2437],[4061,2.3711],[4102,6.2532],[4148,6.2437],[4186,2.3711],[4227,6.2532],[4274,6.2437],[4312,2.3711],[4353,6.2532],[4364,2.9592],[4365,2.4898],[4369,-0.9495]]
       }
       thigh_r_fk_ctrl.Rotation.Z {
        type=MovieSceneScriptingFloatChannel
        default=-3.1148
        thinned_from=78
        keys [[1790,-3.1148],[1791,-3.1319],[1832,0.8123],[1871,-3.1319],[1912,0.8127],[1951,-3.1319],[1992,0.8114],[2031,-3.1319],[2072,0.8118],[2111,-3.1319],[2152,0.8123],[2191,-3.1319],[2232,0.8127],[2279,-12.0144],[2312,0.8114],[2351,-3.1319],[2392,0.8118],[2431,-3.1319],[2468,1.1199],[2472,-1.7868],[2476,-17.0653],[2487,-56.7311],[2504,-58.2702],[2505,-59.3672],[2509,-100.5305],[2510,-108.3165],[2512,-94.3525],[2520,1.5293],[2522,4.4983],[2532,7.1632],[2533,6.9084],[2757,14.9765],[2770,0.6488],[2782,0.9769],[2801,-0.0961],[2849,0.3244],[2893,-0.5237],[2942,-2.0223],[2976,-11.1449],[3058,-1.9608],[3070,-71.5981],[3083,-71.5981],[3088,-73.2218],[3100,-68.2759],[3119,-70.5727],[3185,-54.8232],[3228,-17.5547],[3267,0.8123],[3306,-3.1319],[3347,0.8127],[3386,-3.1319],[3427,0.8114],[3466,-3.1319],[3507,0.8118],[3546,-3.1319],[3587,0.8123],[3626,-3.1319],[3667,0.8127],[3696,-3.1248],[3747,0.8114],[3786,-3.1319],[3827,0.8118],[3866,-3.1319],[3910,-3.1319],[3951,0.8118],[3990,-3.1319],[4034,-3.1319],[4075,0.8118],[4114,-3.1319],[4159,-3.1319],[4200,0.8118],[4239,-3.1319],[4285,-3.1319],[4326,0.8118],[4364,-0.7388],[4365,-0.8009],[4369,-6.7648]]
       }
       thigh_r_fk_ctrl.Scale.X {type=MovieSceneScriptingFloatChannel, default=1, const [1,12]}
       thigh_r_fk_ctrl.Scale.Y {type=MovieSceneScriptingFloatChannel, default=1, const [1,12]}
       thigh_r_fk_ctrl.Scale.Z {type=MovieSceneScriptingFloatChannel, default=1, const [1,12]}
       calf_r_fk_ctrl.Location.X {
        type=MovieSceneScriptingFloatChannel
        default=0
        thinned_from=1081
        keys [[1790,0],[2768,0],[2775,0.01],[2783,0.0806],[2789,0.0933],[2927,0.1129],[2938,0.1131],[2986,0.2196],[3058,0.2196],[3070,0],[4368,0]]
       }
       calf_r_fk_ctrl.Location.Y {
        type=MovieSceneScriptingFloatChannel
        default=0
        thinned_from=1140
        keys [[1790,0],[2768,0],[2792,0.0224],[2802,0.0219],[2938,0.0222],[2986,0.1504],[3058,0.1504],[3070,0],[4368,0]]
       }
       calf_r_fk_ctrl.Location.Z {
        type=MovieSceneScriptingFloatChannel
        default=0
        thinned_from=1182
        keys [[1790,0],[2768,0],[2790,-0.0001],[2938,-0.0002],[2986,-0.0004],[3058,-0.0004],[3070,0],[4368,0]]
       }
       calf_r_fk_ctrl.Rotation.X {
        type=MovieSceneScriptingFloatChannel
        default=-0.2565
        thinned_from=939
        keys [[1790,-0.2565],[1830,-0.2565],[1880,-0.217],[1911,-0.2565],[2147,-0.2565],[2200,-0.217],[2231,-0.2565],[2464,-0.2565],[2468,-0.3876],[2521,0.591],[2524,0.6598],[2532,0.1318],[2760,0.1991],[2773,-0.2213],[2788,-0.0325],[2938,-0.034],[2967,0.2551],[2986,-0.2574],[3058,-0.2574],[3070,0.5826],[3185,0.5826],[3258,-0.2565],[3263,-0.2565],[3315,-0.217],[3346,-0.2565],[3584,-0.2565],[3635,-0.217],[3666,-0.2565],[4368,-0.2565]]
       }
       calf_r_fk_ctrl.Rotation.Y {
        type=MovieSceneScriptingFloatChannel
        default=0.0297
        thinned_from=968
        keys [[1790,0.0297],[1830,0.0297],[1880,0.0182],[1914,0.0297],[2151,0.0297],[2200,0.0182],[2232,0.0297],[2463,0.0297],[2468,0.0517],[2469,0.091],[2499,0.5159],[2508,0.5159],[2522,0.1728],[2524,0.1837],[2532,0.0525],[2533,0.0525],[2760,0.0983],[2927,0.0877],[2938,0.0876],[2967,-0.1012],[2986,0.0404],[3058,0.0404],[3070,-0.625],[3185,-0.625],[3257,0.0297],[3263,0.0297],[3315,0.0182],[3346,0.0297],[3586,0.0297],[3635,0.0182],[3668,0.0297],[4365,0.0297]]
       }
       calf_r_fk_ctrl.Rotation.Z {
        type=MovieSceneScriptingFloatChannel
        default=23.0452
        thinned_from=46
        keys [[1790,23.0452],[1880,19.3973],[1911,23.0453],[1913,23.0452],[2200,19.3973],[2231,23.0453],[2460,23.0452],[2469,23.2016],[2493,125.1619],[2501,140.0712],[2508,142.4116],[2519,24.3513],[2521,15.8316],[2525,7.2603],[2532,3.4384],[2533,3.4439],[2761,15.3453],[2769,22.0557],[2785,24.1528],[2930,27.685],[2938,31.2876],[2966,73.5986],[2991,42.6179],[3006,38.7331],[3058,59.6715],[3060,63.8729],[3064,100.4039],[3066,119.1704],[3067,121.1147],[3079,88.3357],[3087,86.5607],[3185,86.5607],[3315,19.3973],[3346,23.0453],[3348,23.0452],[3641,19.1535],[3655,21.5567],[3714,26.4188],[3721,22.9923],[4365,19.8871]]
       }
       calf_r_fk_ctrl.Scale.X {type=MovieSceneScriptingFloatChannel, default=1, const [1,12]}
       calf_r_fk_ctrl.Scale.Y {type=MovieSceneScriptingFloatChannel, default=1, const [1,12]}
       calf_r_fk_ctrl.Scale.Z {type=MovieSceneScriptingFloatChannel, default=1, const [1,12]}
       foot_r_fk_ctrl.Location.X {type=MovieSceneScriptingFloatChannel, default=0, thinned_from=41, keys [[1790,0],[4435,0],[4436,0.0003]]}
       foot_r_fk_ctrl.Location.Y {type=MovieSceneScriptingFloatChannel, default=0, thinned_from=1267, keys [[1790,0],[4436,0]]}
       foot_r_fk_ctrl.Location.Z {type=MovieSceneScriptingFloatChannel, default=0, thinned_from=117, keys [[1790,0],[4435,0],[4436,0.0001]]}
       foot_r_fk_ctrl.Rotation.X {
        type=MovieSceneScriptingFloatChannel
        default=-4.2712
        thinned_from=73
        keys [[1790,-6.5103],[1806,-3.9153],[1850,-4.73],[1886,-3.9153],[1930,-4.7297],[1966,-3.9153],[2010,-4.73],[2046,-3.9153],[2090,-4.73],[2126,-3.9153],[2170,-4.73],[2206,-3.9153],[2250,-4.7297],[2286,-3.9153],[2330,-4.73],[2366,-3.9153],[2410,-4.73],[2446,-3.9153],[2467,-3.5877],[2487,-17.4099],[2496,-30.6259],[2497,-33.0533],[2508,-33.0533],[2509,-31.9275],[2520,2.7332],[2522,2.314],[2524,2.7635],[2526,2.558],[2538,3.59],[2776,-4.5074],[2799,-4.7227],[2938,-4.6767],[2986,4.9934],[3058,4.9934],[3070,-1.1033],[3185,-1.1033],[3241,-3.9153],[3285,-4.73],[3321,-3.9153],[3365,-4.7297],[3401,-3.9153],[3445,-4.73],[3481,-3.9153],[3525,-4.73],[3561,-3.9153],[3605,-4.73],[3641,-3.9153],[3685,-4.7297],[3721,-3.9153],[3765,-4.73],[3801,-3.9153],[3845,-4.73],[3881,-3.9153],[3925,-3.9153],[3969,-4.73],[4005,-3.9153],[4049,-3.9153],[4093,-4.73],[4129,-3.9153],[4174,-3.9153],[4218,-4.73],[4254,-3.9153],[4300,-3.9153],[4344,-4.73],[4360,-3.9863],[4364,-4.0148],[4394,-4.73],[4430,-3.9153],[4435,-3.9693],[4436,-0.001]]
       }
       foot_r_fk_ctrl.Rotation.Y {
        type=MovieSceneScriptingFloatChannel
        default=-0.2362
        thinned_from=71
        keys [[1790,-0.1935],[1806,-0.1048],[1850,-0.3976],[1886,-0.1048],[1930,-0.3975],[1966,-0.1048],[2010,-0.3976],[2046,-0.1048],[2090,-0.3976],[2126,-0.1048],[2170,-0.3976],[2206,-0.1048],[2250,-0.3975],[2286,-0.1048],[2330,-0.3976],[2366,-0.1048],[2410,-0.3976],[2446,-0.1048],[2458,-0.572],[2471,-0.3574],[2496,3.5186],[2497,3.9417],[2508,3.9417],[2522,-0.7288],[2525,-0.0793],[2526,-0.0846],[2538,1.1668],[2781,-0.365],[2927,-0.6883],[2938,-0.6922],[2986,-3.9637],[3058,-3.9637],[3070,-5.7345],[3185,-5.7345],[3241,-0.1048],[3285,-0.3976],[3321,-0.1048],[3365,-0.3975],[3401,-0.1048],[3445,-0.3976],[3481,-0.1048],[3525,-0.3976],[3561,-0.1048],[3605,-0.3976],[3641,-0.1048],[3685,-0.3975],[3721,-0.1048],[3765,-0.3976],[3801,-0.1048],[3845,-0.3976],[3881,-0.1048],[3925,-0.1048],[3969,-0.3976],[4005,-0.1048],[4049,-0.1048],[4093,-0.3976],[4129,-0.1048],[4174,-0.1048],[4218,-0.3976],[4254,-0.1048],[4300,-0.1048],[4344,-0.3976],[4360,-0.1303],[4364,-0.1405],[4394,-0.3976],[4430,-0.1048],[4435,-0.1252],[4436,-0.0006]]
       }
       foot_r_fk_ctrl.Rotation.Z {
        type=MovieSceneScriptingFloatChannel
        default=21.1681
        thinned_from=70
        keys [[1790,31.719],[1806,19.5312],[1849,23.2805],[1886,19.5312],[1928,23.2786],[1966,19.5312],[2009,23.2805],[2046,19.5312],[2089,23.2805],[2126,19.5312],[2169,23.2805],[2206,19.5312],[2248,23.2786],[2286,19.5312],[2329,23.2805],[2366,19.5312],[2409,23.2805],[2446,19.5312],[2463,19.8597],[2496,29.7844],[2507,29.7844],[2522,30.7062],[2524,25.9306],[2532,33.4166],[2760,28.0638],[2766,24.3847],[2778,23.0196],[2927,20.8394],[2938,20.8254],[2986,16.1405],[3058,16.1405],[3070,20.6831],[3185,20.6831],[3241,19.5312],[3284,23.2805],[3321,19.5312],[3363,23.2786],[3401,19.5312],[3444,23.2805],[3481,19.5312],[3524,23.2805],[3561,19.5312],[3604,23.2805],[3641,19.5312],[3683,23.2786],[3721,19.5312],[3764,23.2805],[3801,19.5312],[3844,23.2805],[3881,19.5312],[3925,19.5312],[3968,23.2805],[4005,19.5312],[4049,19.5312],[4092,23.2805],[4129,19.5312],[4174,19.5312],[4217,23.2805],[4254,19.5312],[4300,19.5312],[4343,23.2805],[4360,19.8723],[4364,20.0103],[4393,23.2805],[4430,19.5312],[4435,19.7795],[4436,0.0053]]
       }
       foot_r_fk_ctrl.Scale.X {type=MovieSceneScriptingFloatChannel, default=1, const [1,13]}
       foot_r_fk_ctrl.Scale.Y {type=MovieSceneScriptingFloatChannel, default=1, const [1,13]}
       foot_r_fk_ctrl.Scale.Z {type=MovieSceneScriptingFloatChannel, default=1, const [1,13]}
       ball_r_fk_ctrl.Location.X {type=MovieSceneScriptingFloatChannel, default=0, thinned_from=39, keys [[1790,0],[2484,0],[2487,0.0001],[2524,0],[4435,0],[4436,0.0003]]}
       ball_r_fk_ctrl.Location.Y {type=MovieSceneScriptingFloatChannel, default=0, thinned_from=40, keys [[1790,0],[4435,0],[4436,-0.0005]]}
       ball_r_fk_ctrl.Location.Z {type=MovieSceneScriptingFloatChannel, default=0, thinned_from=40, keys [[1790,0],[4436,0]]}
       ball_r_fk_ctrl.Rotation.X {
        type=MovieSceneScriptingFloatChannel
        default=0
        thinned_from=25
        keys [[1790,0],[2484,0],[2487,-0.0002],[2513,0],[2515,0.0219],[2519,0.5224],[2520,0.3574],[2521,0.1015],[2522,0.0223],[2524,0],[4436,0]]
       }
       ball_r_fk_ctrl.Rotation.Y {
        type=MovieSceneScriptingFloatChannel
        default=0
        thinned_from=21
        keys [[1790,0],[2487,0.0001],[2515,-0.0106],[2519,0.63],[2522,-0.0156],[2524,0],[4436,0]]
       }
       ball_r_fk_ctrl.Rotation.Z {
        type=MovieSceneScriptingFloatChannel
        default=0
        thinned_from=23
        keys [[1790,0],[2484,0],[2487,-0.0006],[2513,0],[2519,-13.4747],[2521,-6.3423],[2522,-4.1543],[2524,0],[4436,0]]
       }
       ball_r_fk_ctrl.Scale.X {type=MovieSceneScriptingFloatChannel, default=1, const [1,14]}
       ball_r_fk_ctrl.Scale.Y {type=MovieSceneScriptingFloatChannel, default=1, const [1,14]}
       ball_r_fk_ctrl.Scale.Z {type=MovieSceneScriptingFloatChannel, default=1, const [1,14]}
       foot_r_ik_ctrl.Location.X {
        type=MovieSceneScriptingFloatChannel
        default=1.4253
        thinned_from=314
        keys [[1780,-136.1185],[1793,-135.9138],[1812,-132.0813],[1818,-132.4507],[1832,-135.1507],[1853,-138.4532],[1861,-137.9438],[1877,-135.2795],[1893,-133.5556],[1898,-133.8305],[1901,-134.215],[1913,-136.5653],[1928,-138.5188],[1931,-138.433],[1942,-136.9773],[1956,-134.013],[1972,-132.0771],[1978,-132.4362],[1993,-135.2541],[2012,-137.9882],[2021,-137.3872],[2036,-134.9714],[2052,-133.3327],[2059,-133.8433],[2073,-136.5592],[2088,-138.5047],[2091,-138.4192],[2102,-136.9677],[2116,-134.0117],[2132,-132.0813],[2138,-132.4507],[2152,-135.1507],[2173,-138.4532],[2181,-137.9438],[2197,-135.2795],[2213,-133.5556],[2218,-133.8305],[2221,-134.215],[2233,-136.5653],[2248,-138.4939],[2251,-138.0566],[2261,-133.0005],[2262,-132.2166],[2273,-120.4933],[2292,-95.4036],[2297,-91.6307],[2298,-91.1103],[2312,-90.6304],[2313,-90.8887],[2341,-93.2885],[2356,-95.8455],[2372,-97.7725],[2379,-97.4302],[2393,-95.0177],[2408,-93.1633],[2411,-93.2401],[2422,-94.6342],[2436,-97.6443],[2452,-100.0737],[2458,-99.6294],[2471,-95.5219],[2484,-105.9697],[2487,-108.6431],[2496,-114.4436],[2501,-123.527],[2502,-126.702],[2505,-126.5786],[2506,-126.1999],[2507,-125.2336],[2509,-121.7482],[2511,-119.2317],[2512,-119.4279],[2513,-110.6492],[2515,-90.0042],[2518,-72.3782],[2519,-66.0658],[2520,-66.9957],[2521,-67.3534],[2525,-65.0511],[2527,-64.2867],[2528,-64.002],[2536,-63.0463],[2538,-62.5578],[2760,-44.5625],[2764,-65.007],[2768,-77.8475],[2771,-81.3275],[2774,-84.2974],[2775,-85.17],[2781,-88.0964],[2803,-86.6765],[2927,-88.5409],[2938,209.3792],[2986,-155.0094],[3058,-155.0094],[3070,73.4228],[3083,73.4228],[3099,-281.4387],[3185,-281.4387],[3216,-68.0986],[3228,-74.2818],[3231,-75.9496],[3247,-82.9971],[3252,-83.2738],[3253,-83.2246],[3267,-79.5242],[3288,-69.9064],[3296,-69.132],[3299,-69.3124],[3308,-71.1637],[3328,-78.7048],[3331,-79.1974],[3333,-79.3115],[3336,-79.2033],[3347,-76.4652],[3363,-71.293],[3366,-71.1499],[3376,-72.548],[3388,-76.45],[3391,-77.6565],[3407,-83.0093],[3412,-83.3007],[3413,-83.2588],[3427,-79.7436],[3447,-72.0779],[3456,-72.0017],[3459,-72.4087],[3467,-74.244],[3471,-75.5203],[3487,-80.1361],[3492,-80.309],[3494,-80.1298],[3507,-76.5448],[3523,-71.369],[3526,-71.2125],[3536,-72.5737],[3548,-76.4597],[3551,-77.662],[3567,-82.9996],[3572,-83.2737],[3573,-83.2245],[3587,-79.5241],[3608,-69.9063],[3616,-69.132],[3619,-69.3124],[3628,-71.1637],[3648,-78.7049],[3651,-79.1974],[3653,-79.3116],[3656,-79.2034],[3667,-76.4653],[3668,-76.0692],[3683,-72.0818],[3686,-72.9818],[3708,-83.2052],[3711,-84.9799],[3727,-93.6813],[3732,-94.5956],[3733,-94.6214],[3747,-90.5302],[3767,-79.9878],[3776,-79.3408],[3779,-79.8356],[3787,-82.6537],[3791,-84.6858],[3807,-92.2547],[3812,-92.806],[3814,-92.6447],[3827,-87.7423],[3843,-79.6006],[3846,-78.9642],[3856,-79.3675],[3857,-79.6011],[3868,-84.0505],[3871,-85.7136],[3887,-142.235],[3891,-79.9878],[3900,-79.3408],[3903,-79.8356],[3911,-82.6537],[3915,-84.6858],[3931,-92.2547],[3936,-92.806],[3938,-92.6447],[3951,-87.7423],[3967,-79.6006],[3970,-78.9642],[3980,-79.3675],[3981,-79.6011],[3992,-84.0505],[3995,-85.7136],[4011,-142.235],[4015,-79.9878],[4024,-79.3408],[4027,-79.8356],[4035,-82.6537],[4039,-84.6858],[4055,-92.2547],[4060,-92.806],[4062,-92.6447],[4075,-87.7423],[4091,-79.6006],[4094,-78.9642],[4104,-79.3675],[4105,-79.6011],[4116,-84.0505],[4119,-85.7136],[4135,-142.235],[4140,-79.9878],[4149,-79.3408],[4152,-79.8356],[4160,-82.6537],[4164,-84.6858],[4180,-92.2547],[4185,-92.806],[4187,-92.6447],[4200,-87.7423],[4216,-79.6006],[4219,-78.9642],[4229,-79.3675],[4230,-79.6011],[4241,-84.0505],[4244,-85.7136],[4260,-142.235],[4266,-79.9878],[4275,-79.3408],[4278,-79.8356],[4286,-82.6537],[4290,-84.6858],[4306,-92.2547],[4311,-92.806],[4313,-92.6447],[4326,-87.7423],[4342,-79.6006],[4345,-78.9642],[4355,-79.3675],[4356,-79.6011],[4360,-56.255],[4364,-338.2641],[4367,-560.0049],[4376,-8366.5117],[4397,-29399.6504],[4405,-37944.7773],[4408,-41082.8516],[4417,-46639.2031],[4436,-46625.2617]]
       }
       foot_r_ik_ctrl.Location.Y {
        type=MovieSceneScriptingFloatChannel
        default=-5.536
        thinned_from=314
        keys [[1780,-6328.5713],[1877,-6321.0371],[1901,-6237.0044],[1913,-6168.2598],[1931,-6012.4248],[1942,-5891.5518],[1956,-5715.9712],[1978,-5400.2607],[1993,-5157.0381],[2012,-4819.3228],[2116,-2866.7305],[2138,-2476.3274],[2152,-2235.0325],[2173,-1880.5049],[2184,-1703.0773],[2193,-1565.0477],[2197,-1506.2482],[2213,-1286.1259],[2221,-1183.7362],[2233,-1037.5399],[2248,-867.8524],[2251,-836.655],[2262,-733.395],[2273,-643.6451],[2276,-620.7259],[2292,-504.6837],[2298,-465.2205],[2312,-392.6841],[2313,-388.8864],[2332,-355.4731],[2344,-357.8103],[2352,-358.4432],[2356,-358.1918],[2379,-352.632],[2393,-350.506],[2411,-353.3456],[2422,-356.0279],[2433,-357.3431],[2436,-357.2311],[2452,-353.8012],[2471,-347.9151],[2484,-275.4839],[2487,-276.448],[2496,-273.7492],[2501,-263.6606],[2502,-265.6441],[2503,-265.1152],[2505,-262.2643],[2506,-261.7876],[2507,-265.1162],[2508,-273.1508],[2509,-282.758],[2511,-294.1303],[2512,-289.7285],[2513,-284.0139],[2515,-298.7659],[2518,-329.1943],[2519,-346.0239],[2520,-345.7948],[2524,-341.4523],[2536,-337.0528],[2538,-335.514],[2760,-288.5093],[2764,-290.0905],[2768,-293.7811],[2775,-295.5783],[2781,-295.7829],[2803,-286.5292],[2927,-252.2055],[2938,-463.1673],[2986,-247.0169],[3058,-247.0169],[3070,-190.7237],[3083,-190.7237],[3099,-150.578],[3185,-150.578],[3216,-276.3817],[3217,-272.337],[3228,-224.3911],[3231,-209.6602],[3247,-120.4357],[3252,-90.7727],[3253,-84.7284],[3267,-1.5069],[3288,91.5383],[3296,113.7718],[3299,120.2925],[3308,134.7507],[3312,139.2421],[3328,148.2354],[3336,154.2319],[3347,161.729],[3348,162.1135],[3363,162.6987],[3366,161.7038],[3377,156.0398],[3388,149.6722],[3391,148.5648],[3407,149.6872],[3413,152.8925],[3427,161.4544],[3428,161.8463],[3447,161.1325],[3456,156.7058],[3467,150.7495],[3471,149.4212],[3487,150.6036],[3494,154.1694],[3507,161.8951],[3508,162.2913],[3523,162.809],[3526,161.7791],[3537,156.0385],[3548,149.6774],[3551,148.571],[3567,149.6901],[3573,152.9004],[3587,161.6796],[3608,158.4286],[3619,150.4293],[3628,144.6996],[3632,143.5616],[3648,148.2355],[3656,154.232],[3667,161.7292],[3668,162.1136],[3686,164.2471],[3696,158.2996],[3697,156.9067],[3708,135.6344],[3727,97.6072],[3732,86.2193],[3733,83.7676],[3747,41.2888],[3748,37.5669],[3767,-47.0468],[3776,-91.3519],[3779,-105.7731],[3787,-142.023],[3791,-158.9578],[3814,-252.3009],[3828,-310.9104],[3843,-375.1276],[3846,-387.0395],[3856,-420.9808],[3857,-423.7269],[3868,-443.2844],[3871,-444.7535],[3887,-441.5018],[3891,-47.0468],[3900,-91.3519],[3903,-105.7731],[3911,-142.023],[3915,-158.9578],[3938,-252.3009],[3952,-310.9104],[3967,-375.1276],[3970,-387.0395],[3980,-420.9808],[3981,-423.7269],[3992,-443.2844],[3995,-444.7535],[4011,-441.5018],[4015,-47.0468],[4024,-91.3519],[4027,-105.7731],[4035,-142.023],[4039,-158.9578],[4062,-252.3009],[4076,-310.9104],[4091,-375.1276],[4094,-387.0395],[4104,-420.9808],[4105,-423.7269],[4116,-443.2844],[4119,-444.7535],[4135,-441.5018],[4140,-47.0468],[4149,-91.3519],[4152,-105.7731],[4160,-142.023],[4164,-158.9578],[4187,-252.3009],[4201,-310.9104],[4216,-375.1276],[4219,-387.0395],[4229,-420.9808],[4230,-423.7269],[4241,-443.2844],[4244,-444.7535],[4260,-441.5018],[4266,-47.0468],[4275,-91.3519],[4278,-105.7731],[4286,-142.023],[4290,-158.9578],[4313,-252.3009],[4327,-310.9104],[4342,-375.1276],[4345,-387.0395],[4355,-420.9808],[4356,-423.7269],[4360,64.5692],[4364,-347.1439],[4367,-464.8099],[4376,205.6319],[4397,4476.7148],[4405,6387.1294],[4408,7059.7725],[4417,8172.4438],[4435,8190.5044],[4436,8161.9087]]
       }
       foot_r_ik_ctrl.Location.Z {
        type=MovieSceneScriptingFloatChannel
        default=3.1284
        thinned_from=300
        keys [[1780,1106.2712],[1796,1107.4082],[1818,1115.8999],[1832,1113.1229],[1853,1105.7616],[1864,1112.1056],[1901,1151.7964],[1913,1158.5608],[1931,1165.3286],[1942,1171.6644],[1978,1209.0488],[2024,1233.5382],[2059,1271.241],[2073,1278.9751],[2102,1292.3284],[2138,1329.7207],[2152,1337.5284],[2173,1385.687],[2181,1423.6361],[2184,1440.6527],[2193,1500.1088],[2197,1530.0173],[2221,1737.6401],[2251,2022.6219],[2276,2273.7534],[2292,2427.6216],[2298,2479.5298],[2331,2689.2104],[2346,2740.7107],[2357,2756.397],[2372,2763.1665],[2420,2753.2104],[2468,2760.6445],[2471,2757.6541],[2484,2721.7207],[2496,2739.8833],[2501,2765.1653],[2502,2781.4421],[2507,2790.5093],[2512,2784.3318],[2513,2778.4668],[2519,2781.2817],[2522,2769.2651],[2764,2813.8564],[2772,2797.7209],[2781,2801.4556],[2803,2788.4204],[2927,2805.7546],[2986,3094.3977],[3058,3094.3977],[3070,2832.1343],[3083,2832.1343],[3099,3161.1643],[3185,3161.1643],[3216,2895.4937],[3228,2979.8706],[3253,3212.5713],[3267,3336.0637],[3288,3497.167],[3312,3660.7446],[3336,3778.8652],[3348,3817.3052],[3407,3936.1416],[3428,3934.333],[3459,3912.8057],[3494,3942.1526],[3508,3937.1426],[3537,3913.1038],[3551,3919.9895],[3573,3939.2971],[3587,3934.6533],[3619,3909.0369],[3656,3941.3359],[3668,3937.7129],[3697,3902.7781],[3727,3919.3484],[3733,3916.4341],[3748,3886.2661],[3767,3837.7314],[3779,3821.3523],[3814,3842.3354],[3857,3820.051],[3887,3846.0356],[3903,3821.3523],[3938,3842.3354],[3981,3820.051],[4011,3846.0356],[4027,3821.3523],[4062,3842.3354],[4105,3820.051],[4135,3846.0356],[4152,3821.3523],[4187,3842.3354],[4230,3820.051],[4260,3846.0356],[4278,3821.3523],[4313,3842.3354],[4356,3820.051],[4360,4116.2969],[4364,3913.9062],[4367,3856.1736],[4376,7410.9639],[4397,15150.3691],[4405,14907.4316],[4408,14485.4531],[4417,13694.2842],[4436,13693.8721]]
       }
       foot_r_ik_ctrl.Rotation.X {
        type=MovieSceneScriptingFloatChannel
        default=33.1934
        thinned_from=314
        keys [[1780,33.1934],[1793,32.9408],[1796,32.6881],[1812,35.4546],[1818,37.0926],[1832,39.3222],[1853,34.7148],[1861,32.0474],[1864,31.1546],[1873,29.3135],[1877,29.0852],[1893,33.3593],[1898,35.4793],[1901,36.6774],[1912,39.3114],[1913,39.3347],[1928,37.4724],[1942,34.3425],[1953,32.78],[1956,32.6875],[1972,35.3454],[1978,36.9493],[1992,39.3302],[1993,39.3445],[2012,36.5842],[2021,34.5356],[2024,33.9344],[2032,32.8267],[2036,32.664],[2052,35.328],[2059,37.1934],[2072,39.3219],[2073,39.3363],[2088,37.4737],[2102,34.3434],[2113,32.7807],[2116,32.6881],[2132,35.4546],[2138,37.0926],[2152,39.3222],[2173,34.7148],[2181,32.0474],[2184,31.1546],[2193,29.3135],[2197,29.0852],[2213,33.3593],[2218,35.4792],[2221,36.6774],[2232,39.3114],[2233,39.3346],[2248,37.4723],[2262,34.3425],[2273,32.78],[2276,32.6875],[2292,35.3454],[2298,36.9492],[2312,39.3301],[2313,39.3444],[2332,36.5842],[2341,34.5356],[2344,33.9344],[2352,32.8267],[2356,32.664],[2372,35.328],[2379,37.1934],[2392,39.3219],[2393,39.3363],[2408,37.4737],[2422,34.3434],[2433,32.7807],[2436,32.6881],[2452,35.9142],[2458,37.8609],[2468,40.3632],[2471,40.6998],[2484,53.1708],[2487,58.0682],[2496,76.8847],[2501,91.4007],[2502,97.8356],[2503,101.7736],[2507,109.2026],[2508,113.432],[2509,120.3223],[2511,129.5847],[2512,118.4687],[2513,94.8289],[2515,73.2271],[2518,58.5115],[2519,49.8743],[2520,46.35],[2521,43.1115],[2522,40.4445],[2524,37.968],[2527,38.1106],[2530,38.5006],[2533,39.0807],[2538,40.3153],[2760,42.1629],[2764,43.0324],[2768,43.5975],[2771,43.7402],[2772,43.7031],[2774,43.9055],[2781,44.0568],[2803,46.4911],[2927,51.1386],[2938,12.5784],[2986,48.0755],[3058,48.0755],[3070,114.103],[3083,114.103],[3099,80.2067],[3185,80.2067],[3216,69.7955],[3217,68.7752],[3228,58.9557],[3231,57.0109],[3247,54.3958],[3253,56.0325],[3267,58.2379],[3288,53.5738],[3296,50.9026],[3299,50.0147],[3308,48.2048],[3312,47.9927],[3328,52.3093],[3336,55.6249],[3347,58.2394],[3348,58.2603],[3363,56.3507],[3377,53.1715],[3388,51.6345],[3391,51.5566],[3407,54.2803],[3413,55.8892],[3427,58.2427],[3428,58.2538],[3447,55.4302],[3456,53.3736],[3459,52.7774],[3467,51.6987],[3471,51.5538],[3487,54.271],[3494,56.1394],[3507,58.2489],[3508,58.261],[3523,56.3519],[3537,53.1722],[3548,51.6352],[3551,51.5573],[3567,54.3896],[3573,56.0325],[3587,58.2379],[3608,53.5738],[3616,50.9026],[3619,50.0147],[3628,48.2048],[3632,47.9927],[3648,52.3093],[3656,55.6249],[3667,58.2393],[3668,58.2602],[3683,53.9092],[3686,49.5926],[3696,32.7607],[3697,31.5169],[3708,26.3142],[3711,25.6167],[3727,25.0916],[3733,25.644],[3747,26.2304],[3748,26.1493],[3767,22.2812],[3776,20.0828],[3779,19.4693],[3787,18.4064],[3791,18.3109],[3807,21.5425],[3814,23.7867],[3827,26.7473],[3828,26.8296],[3843,25.9882],[3846,25.5251],[3857,23.6829],[3868,22.5623],[3871,22.5541],[3887,30.4724],[3891,22.2812],[3900,20.0828],[3903,19.4693],[3911,18.4064],[3915,18.3109],[3931,21.5425],[3938,23.7867],[3951,26.7473],[3952,26.8296],[3967,25.9882],[3970,25.5251],[3981,23.6829],[3992,22.5623],[3995,22.5541],[4011,30.4724],[4015,22.2812],[4024,20.0828],[4027,19.4693],[4035,18.4064],[4039,18.3109],[4055,21.5425],[4062,23.7867],[4075,26.7473],[4076,26.8296],[4091,25.9882],[4094,25.5251],[4105,23.6829],[4116,22.5623],[4119,22.5541],[4135,30.4724],[4140,22.2812],[4149,20.0828],[4152,19.4693],[4160,18.4064],[4164,18.3109],[4180,21.5425],[4187,23.7867],[4200,26.7473],[4201,26.8296],[4216,25.9882],[4219,25.5251],[4230,23.6829],[4241,22.5623],[4244,22.5541],[4260,30.4724],[4266,22.2812],[4275,20.0828],[4278,19.4693],[4286,18.4064],[4290,18.3109],[4306,21.5425],[4313,23.7867],[4326,26.7473],[4327,26.8296],[4342,25.9882],[4345,25.5251],[4356,23.6829],[4360,5.5458],[4364,27.4918],[4367,44.3022],[4376,152.2777],[4397,155.207],[4408,151.698],[4417,149.5795],[4421,149.2223],[4435,152.1842],[4436,120.0092]]
       }
       foot_r_ik_ctrl.Rotation.Y {
        type=MovieSceneScriptingFloatChannel
        default=2.1467
        thinned_from=314
        keys [[1780,-2.1467],[1793,-2.0689],[1796,-1.8594],[1812,-0.0009],[1817,0.2773],[1818,0.2864],[1832,-0.3645],[1853,-2.2284],[1861,-2.6162],[1864,-2.6521],[1873,-2.2705],[1893,-0.5766],[1896,-0.4696],[1898,-0.4476],[1901,-0.4993],[1912,-1.3078],[1913,-1.4072],[1928,-2.6173],[1931,-2.669],[1941,-2.3246],[1942,-2.2591],[1953,-1.3327],[1956,-1.0501],[1972,0.154],[1977,0.2823],[1978,0.2862],[1992,-0.3555],[1993,-0.4371],[2012,-2.1223],[2021,-2.5756],[2024,-2.6109],[2032,-2.3073],[2036,-1.9912],[2052,-0.6212],[2057,-0.448],[2058,-0.4425],[2059,-0.4485],[2072,-1.3012],[2073,-1.4005],[2088,-2.6084],[2091,-2.6601],[2101,-2.3167],[2102,-2.2515],[2113,-1.3277],[2116,-1.046],[2132,0.1546],[2137,0.2824],[2138,0.2864],[2152,-0.3645],[2173,-2.2284],[2181,-2.6162],[2184,-2.6521],[2193,-2.2705],[2213,-0.5766],[2216,-0.4696],[2218,-0.4476],[2221,-0.4993],[2232,-1.3078],[2233,-1.4072],[2248,-2.6173],[2251,-2.669],[2261,-2.3246],[2262,-2.2591],[2273,-1.3327],[2276,-1.0501],[2292,0.154],[2297,0.2823],[2298,0.2862],[2312,-0.3555],[2313,-0.4371],[2332,-2.1223],[2341,-2.5756],[2344,-2.6109],[2352,-2.3073],[2356,-1.9912],[2372,-0.6212],[2377,-0.448],[2378,-0.4425],[2379,-0.4485],[2392,-1.3012],[2393,-1.4005],[2408,-2.6084],[2411,-2.6601],[2421,-2.3167],[2422,-2.2515],[2433,-1.3277],[2436,-1.046],[2452,0.1546],[2457,0.2824],[2458,0.2864],[2468,-0.9032],[2469,-1.0905],[2470,-1.2866],[2471,-1.6799],[2484,9.6166],[2496,17.1021],[2501,20.0963],[2502,19.861],[2503,19.2224],[2505,17.1384],[2507,14.651],[2508,13.5081],[2509,12.567],[2511,11.693],[2512,12.9756],[2513,7.0591],[2515,-5.9571],[2518,-18.9733],[2519,-24.8898],[2521,-24.3196],[2522,-24.5431],[2524,-25.1816],[2525,-25.1221],[2527,-24.7437],[2530,-24.014],[2532,-23.7978],[2533,-23.803],[2535,-23.3074],[2536,-23.134],[2538,-23.0505],[2760,-17.4446],[2764,-9.507],[2768,-4.588],[2771,-3.4491],[2772,-3.0908],[2774,-2.4337],[2775,-2.144],[2781,-1.2384],[2803,-2.0007],[2927,-2.4042],[2938,0.8404],[2986,-12.5432],[3058,-12.5432],[3070,26.599],[3083,26.599],[3099,25.2389],[3185,25.2389],[3216,-7.0728],[3217,-6.9363],[3228,-4.707],[3231,-4.0228],[3247,-1.4382],[3252,-1.2405],[3253,-1.2388],[3267,-2.2107],[3288,-4.8278],[3296,-5.3102],[3299,-5.3155],[3308,-4.6076],[3312,-4.0607],[3328,-2.0266],[3331,-1.8965],[3333,-1.8844],[3336,-1.9764],[3347,-3.1254],[3348,-3.2661],[3363,-5.1286],[3366,-5.2915],[3376,-5.1547],[3377,-5.0895],[3388,-3.8416],[3391,-3.4052],[3407,-1.4382],[3412,-1.2407],[3413,-1.2391],[3427,-2.2347],[3428,-2.357],[3447,-4.8066],[3456,-5.392],[3459,-5.397],[3467,-4.8158],[3471,-4.2953],[3487,-2.1708],[3492,-1.9306],[3493,-1.9276],[3494,-1.9398],[3507,-3.1289],[3508,-3.2679],[3523,-5.1194],[3526,-5.2823],[3536,-5.1472],[3537,-5.0821],[3548,-3.8369],[3551,-3.4013],[3567,-1.4377],[3572,-1.2405],[3573,-1.2388],[3587,-2.2107],[3608,-4.8278],[3616,-5.3102],[3619,-5.3155],[3628,-4.6076],[3632,-4.0607],[3648,-2.0266],[3651,-1.8965],[3653,-1.8844],[3656,-1.9764],[3667,-3.1254],[3668,-3.2661],[3683,-4.8304],[3686,-4.4964],[3696,-2.0369],[3697,-1.8043],[3708,-0.4233],[3711,-0.1229],[3727,1.0205],[3732,1.1745],[3733,1.1925],[3747,0.9819],[3748,0.9443],[3767,0.0904],[3776,-0.201],[3779,-0.2558],[3787,-0.201],[3791,-0.0779],[3807,0.5595],[3812,0.65],[3813,0.6524],[3814,0.6466],[3827,0.0057],[3828,-0.0679],[3843,-0.9189],[3846,-0.928],[3856,-0.569],[3857,-0.5151],[3868,0.1126],[3871,0.2843],[3887,0.5779],[3891,0.0904],[3900,-0.201],[3903,-0.2558],[3911,-0.201],[3915,-0.0779],[3931,0.5595],[3936,0.65],[3937,0.6524],[3938,0.6466],[3951,0.0057],[3952,-0.0679],[3967,-0.9189],[3970,-0.928],[3980,-0.569],[3981,-0.5151],[3992,0.1126],[3995,0.2843],[4011,0.5779],[4015,0.0904],[4024,-0.201],[4027,-0.2558],[4035,-0.201],[4039,-0.0779],[4055,0.5595],[4060,0.65],[4061,0.6524],[4062,0.6466],[4075,0.0057],[4076,-0.0679],[4091,-0.9189],[4094,-0.928],[4104,-0.569],[4105,-0.5151],[4116,0.1126],[4119,0.2843],[4135,0.5779],[4140,0.0904],[4149,-0.201],[4152,-0.2558],[4160,-0.201],[4164,-0.0779],[4180,0.5595],[4185,0.65],[4186,0.6524],[4187,0.6466],[4200,0.0057],[4201,-0.0679],[4216,-0.9189],[4219,-0.928],[4229,-0.569],[4230,-0.5151],[4241,0.1126],[4244,0.2843],[4260,0.5779],[4266,0.0904],[4275,-0.201],[4278,-0.2558],[4286,-0.201],[4290,-0.0779],[4306,0.5595],[4311,0.65],[4312,0.6524],[4313,0.6466],[4326,0.0057],[4327,-0.0679],[4342,-0.9189],[4345,-0.928],[4355,-0.569],[4356,-0.5151],[4360,-5.3511],[4364,-5.0561],[4367,-0.3244],[4376,-5.147],[4397,-6.1214],[4405,-6.2319],[4408,-6.0162],[4417,-4.0304],[4421,-2.8627],[4435,-0.019],[4436,-0.0009]]
       }
       foot_r_ik_ctrl.Rotation.Z {
        type=MovieSceneScriptingFloatChannel
        default=-8.4386
        thinned_from=314
        keys [[1780,8.4386],[1793,8.3082],[1796,7.9586],[1812,4.9783],[1817,4.6353],[1818,4.6415],[1832,5.7397],[1853,8.3774],[1861,8.7346],[1864,8.6459],[1873,7.5731],[1877,6.8737],[1893,4.554],[1896,4.4652],[1898,4.4919],[1901,4.6244],[1912,5.8087],[1928,8.1695],[1931,8.5221],[1941,9.1081],[1942,9.0979],[1953,7.9484],[1956,7.4256],[1972,4.8701],[1977,4.6352],[1978,4.6421],[1992,5.8401],[2012,8.6225],[2021,9.1073],[2024,9.0193],[2032,8.1094],[2036,7.4243],[2052,4.8679],[2057,4.633],[2058,4.6398],[2059,4.6599],[2072,5.8386],[2073,5.9778],[2088,8.1669],[2091,8.5196],[2101,9.1079],[2102,9.0978],[2113,7.9484],[2116,7.4255],[2132,4.8702],[2137,4.6353],[2138,4.6415],[2152,5.7397],[2173,8.3774],[2181,8.7346],[2184,8.6459],[2193,7.5731],[2197,6.8737],[2213,4.5539],[2216,4.4652],[2218,4.4919],[2221,4.6244],[2232,5.8087],[2233,5.9524],[2248,8.2526],[2251,9.8185],[2261,23.624],[2262,25.605],[2273,52.3677],[2276,60.9308],[2292,111.1522],[2298,130.2938],[2312,168.7772],[2313,170.9366],[2332,188.8505],[2344,189.7941],[2372,186.9151],[2379,187.024],[2393,188.974],[2411,192.3239],[2422,193.3931],[2436,192.3419],[2452,190.4907],[2458,190.5585],[2471,194.5153],[2484,179.9776],[2487,180.8064],[2496,189.6215],[2501,198.0882],[2502,204.2007],[2503,206.782],[2507,208.4748],[2508,210.5967],[2509,214.7189],[2511,220.5815],[2512,214.4828],[2513,203.1208],[2515,197.022],[2518,200.7213],[2519,204.4205],[2521,204.5664],[2524,205.8983],[2533,205.0439],[2538,203.7625],[2760,208.5176],[2775,203.9426],[2781,203.4707],[2803,205.09],[2927,205.9348],[2938,228.6424],[2986,260.4882],[3058,260.4882],[3070,155.4111],[3083,155.4111],[3099,256.8774],[3185,256.8774],[3216,206.2005],[3228,206.5474],[3247,204.656],[3253,204.4827],[3267,205.3107],[3288,207.206],[3299,207.3233],[3328,204.1195],[3336,204.2113],[3348,205.1736],[3366,207.2003],[3377,207.8799],[3391,206.6853],[3407,204.656],[3413,204.4831],[3428,205.5139],[3447,207.473],[3459,207.6912],[3487,204.4022],[3494,204.2614],[3508,205.1999],[3526,207.2008],[3537,207.8823],[3551,206.6866],[3567,204.6562],[3573,204.4827],[3587,205.3107],[3608,207.206],[3619,207.3233],[3648,204.1195],[3656,204.2113],[3668,205.1736],[3697,209.1981],[3711,207.498],[3727,204.7647],[3733,204.4957],[3748,205.9205],[3767,208.8775],[3779,209.3831],[3791,207.6848],[3807,204.8754],[3814,204.6366],[3828,206.1387],[3846,208.8742],[3857,209.3559],[3871,207.493],[3887,204.8383],[3891,208.8775],[3903,209.3831],[3915,207.6848],[3931,204.8754],[3938,204.6366],[3952,206.1387],[3970,208.8742],[3981,209.3559],[3995,207.493],[4011,204.8383],[4015,208.8775],[4027,209.3831],[4039,207.6848],[4055,204.8754],[4062,204.6366],[4076,206.1387],[4094,208.8742],[4105,209.3559],[4119,207.493],[4135,204.8383],[4140,208.8775],[4152,209.3831],[4164,207.6848],[4180,204.8754],[4187,204.6366],[4201,206.1387],[4219,208.8742],[4230,209.3559],[4244,207.493],[4260,204.8383],[4266,208.8775],[4278,209.3831],[4290,207.6848],[4306,204.8754],[4313,204.6366],[4327,206.1387],[4345,208.8742],[4356,209.3559],[4360,227.5925],[4364,207.4954],[4367,181.781],[4376,42.6656],[4397,34.8533],[4405,34.3349],[4408,34.4211],[4417,36.3424],[4421,37.6475],[4435,40.9785],[4436,41.0004]]
       }
       foot_r_ik_ctrl.Scale.X {
        type=MovieSceneScriptingFloatChannel
        default=1
        thinned_from=314
        keys [[1780,-1.25],[2520,-1.25],[2538,-1.555],[2760,-2.8819],[2764,-2.9592],[2768,-2.9982],[2927,-3.0152],[2938,-3.0096],[3185,-3.0136],[3217,-3],[4356,-3],[4360,-2.9944],[4367,-3],[4436,-3]]
       }
       foot_r_ik_ctrl.Scale.Y {
        type=MovieSceneScriptingFloatChannel
        default=1
        thinned_from=314
        keys [[1780,1.25],[2520,1.25],[2538,1.555],[2760,2.8819],[2764,2.9592],[2768,2.9982],[2927,3.0152],[2938,3.0096],[3185,3.0136],[3217,3],[4356,3],[4360,2.9944],[4367,3],[4436,3]]
       }
       foot_r_ik_ctrl.Scale.Z {
        type=MovieSceneScriptingFloatChannel
        default=1
        thinned_from=314
        keys [[1780,1.25],[2520,1.25],[2538,1.555],[2760,2.8819],[2764,2.9592],[2768,2.9982],[2927,3.0152],[2938,3.0096],[3185,3.0136],[3217,3],[4356,3],[4360,2.9944],[4367,3],[4436,3]]
       }
       ball_r_ik_ctrl.Location.X {type=MovieSceneScriptingFloatChannel, default=0, thinned_from=42, keys [[1790,0],[4436,0]]}
       ball_r_ik_ctrl.Location.Y {type=MovieSceneScriptingFloatChannel, default=0, thinned_from=40, keys [[1790,0],[4435,0],[4436,-0.0005]]}
       ball_r_ik_ctrl.Location.Z {type=MovieSceneScriptingFloatChannel, default=0, thinned_from=39, keys [[1790,0],[4435,0],[4436,0.0003]]}
       ball_r_ik_ctrl.Rotation.X {
        type=MovieSceneScriptingFloatChannel
        default=0
        thinned_from=23
        keys [[1790,0],[2513,0],[2519,-13.375],[2521,-6.3059],[2522,-4.1371],[2524,0],[4436,0]]
       }
       ball_r_ik_ctrl.Rotation.Y {
        type=MovieSceneScriptingFloatChannel
        default=0
        thinned_from=23
        keys [[1790,0],[2513,0],[2515,0.3313],[2519,1.7686],[2521,0.6867],[2524,0],[4436,0]]
       }
       ball_r_ik_ctrl.Rotation.Z {
        type=MovieSceneScriptingFloatChannel
        default=0
        thinned_from=23
        keys [[1790,0],[2513,0],[2515,-0.0319],[2519,-0.799],[2521,-0.1436],[2524,0],[4436,0]]
       }
       ball_r_ik_ctrl.Scale.X {type=MovieSceneScriptingFloatChannel, default=1, const [1,14]}
       ball_r_ik_ctrl.Scale.Y {type=MovieSceneScriptingFloatChannel, default=1, const [1,14]}
       ball_r_ik_ctrl.Scale.Z {type=MovieSceneScriptingFloatChannel, default=1, const [1,14]}
       foot_bk1_r_ctrl.Location.X {type=MovieSceneScriptingFloatChannel, default=0, const [0,12]}
       foot_bk1_r_ctrl.Location.Y {type=MovieSceneScriptingFloatChannel, default=0, const [0,12]}
       foot_bk1_r_ctrl.Location.Z {type=MovieSceneScriptingFloatChannel, default=0, const [0,12]}
       foot_bk1_r_ctrl.Rotation.X {type=MovieSceneScriptingFloatChannel, default=0, const [0,12]}
       foot_bk1_r_ctrl.Rotation.Y {type=MovieSceneScriptingFloatChannel, default=0, const [0,12]}
       foot_bk1_r_ctrl.Rotation.Z {type=MovieSceneScriptingFloatChannel, default=0, const [0,12]}
       foot_bk1_r_ctrl.Scale.X {type=MovieSceneScriptingFloatChannel, default=1, const [1,12]}
       foot_bk1_r_ctrl.Scale.Y {type=MovieSceneScriptingFloatChannel, default=1, const [1,12]}
       foot_bk1_r_ctrl.Scale.Z {type=MovieSceneScriptingFloatChannel, default=1, const [1,12]}
       foot_roll_r_ctrl.Location.X {type=MovieSceneScriptingFloatChannel, default=0, const [0,12]}
       foot_roll_r_ctrl.Location.Y {type=MovieSceneScriptingFloatChannel, default=0, const [0,12]}
       foot_roll_r_ctrl.Location.Z {type=MovieSceneScriptingFloatChannel, default=0, const [0,12]}
       foot_roll_r_ctrl.Rotation.X {type=MovieSceneScriptingFloatChannel, default=0, const [0,12]}
       foot_roll_r_ctrl.Rotation.Y {type=MovieSceneScriptingFloatChannel, default=0, const [0,12]}
       foot_roll_r_ctrl.Rotation.Z {type=MovieSceneScriptingFloatChannel, default=0, const [0,12]}
       foot_roll_r_ctrl.Scale.X {type=MovieSceneScriptingFloatChannel, default=1, const [1,12]}
       foot_roll_r_ctrl.Scale.Y {type=MovieSceneScriptingFloatChannel, default=1, const [1,12]}
       foot_roll_r_ctrl.Scale.Z {type=MovieSceneScriptingFloatChannel, default=1, const [1,12]}
       roll_blend_r_ctrl {type=MovieSceneScriptingFloatChannel, default=1, const [1,12]}
       tip_r_ctrl.Location.X {type=MovieSceneScriptingFloatChannel, default=0, const [0,12]}
       tip_r_ctrl.Location.Y {type=MovieSceneScriptingFloatChannel, default=0, const [0,12]}
       tip_r_ctrl.Location.Z {type=MovieSceneScriptingFloatChannel, default=0, const [0,12]}
       tip_r_ctrl.Rotation.X {type=MovieSceneScriptingFloatChannel, default=0, const [0,12]}
       tip_r_ctrl.Rotation.Y {type=MovieSceneScriptingFloatChannel, default=0, const [0,12]}
       tip_r_ctrl.Rotation.Z {type=MovieSceneScriptingFloatChannel, default=0, const [0,12]}
       tip_r_ctrl.Scale.X {type=MovieSceneScriptingFloatChannel, default=1, const [1,12]}
       tip_r_ctrl.Scale.Y {type=MovieSceneScriptingFloatChannel, default=1, const [1,12]}
       tip_r_ctrl.Scale.Z {type=MovieSceneScriptingFloatChannel, default=1, const [1,12]}
       heel_r_ctrl.Location.X {type=MovieSceneScriptingFloatChannel, default=0, const [0,12]}
       heel_r_ctrl.Location.Y {type=MovieSceneScriptingFloatChannel, default=0, const [0,12]}
       heel_r_ctrl.Location.Z {type=MovieSceneScriptingFloatChannel, default=0, const [0,12]}
       heel_r_ctrl.Rotation.X {type=MovieSceneScriptingFloatChannel, default=0, const [0,12]}
       heel_r_ctrl.Rotation.Y {type=MovieSceneScriptingFloatChannel, default=0, const [0,12]}
       heel_r_ctrl.Rotation.Z {type=MovieSceneScriptingFloatChannel, default=0, const [0,12]}
       heel_r_ctrl.Scale.X {type=MovieSceneScriptingFloatChannel, default=1, const [1,12]}
       heel_r_ctrl.Scale.Y {type=MovieSceneScriptingFloatChannel, default=1, const [1,12]}
       heel_r_ctrl.Scale.Z {type=MovieSceneScriptingFloatChannel, default=1, const [1,12]}
       leg_r_softness {type=MovieSceneScriptingFloatChannel, default=0, const [0,12]}
       leg_r_pv_ik_ctrl.Location.X {
        type=MovieSceneScriptingFloatChannel
        default=-0.2259
        thinned_from=227
        keys [[1780,-1170.2192],[1806,-1186.9717],[1833,-1184.5415],[1874,-1200.1949],[1881,-1207.3431],[1917,-1232.3824],[1931,-1239.5167],[1941,-1247.0183],[1966,-1272.5055],[2021,-1307.0646],[2046,-1332.4546],[2101,-1367.1633],[2126,-1391.9083],[2153,-1408.9478],[2193,-1576.4485],[2201,-1635.9502],[2231,-1898.8269],[2251,-2093.2722],[2286,-2444.4534],[2309,-2628.1858],[2317,-2679.5762],[2336,-2780.7019],[2337,-2789.5754],[2341,-2790.7505],[2366,-2826.0269],[2421,-2818.1665],[2466,-2823.7778],[2468,-2863.0408],[2470,-2819.0261],[2484,-2762.0195],[2487,-2805.8552],[2496,-2753.196],[2501,-2756.9695],[2502,-2765.2437],[2507,-2774.2542],[2509,-2767.4758],[2513,-2779.6826],[2515,-2798.3711],[2518,-2816.7012],[2519,-2830.4163],[2524,-2819.531],[2533,-2825.7876],[2758,-2893.4316],[2762,-2912.4058],[2770,-2908.4941],[2785,-2916.4592],[2927,-2908.0359],[2938,-3101.2083],[2986,-3261.1177],[3058,-3261.1177],[3070,-2792.9702],[3083,-2792.9702],[3099,-3357.9495],[3185,-3357.9495],[3215,-2905.7881],[3241,-3165.7708],[3268,-3405.5054],[3309,-3720.1597],[3316,-3760.2402],[3346,-3869.936],[3376,-3943.531],[3401,-3995.7715],[3456,-3983.1934],[3481,-3999.6699],[3536,-3983.3318],[3561,-3998.397],[3629,-3991.5933],[3666,-3996.6072],[3672,-3992.3149],[3686,-4002.2498],[3696,-4039.2263],[3721,-4071.2351],[3744,-4053.5535],[3776,-3992.158],[3801,-4002.9905],[3856,-3979.6208],[3925,-4002.9905],[3980,-3979.6208],[4049,-4002.9905],[4104,-3979.6208],[4174,-4002.9905],[4229,-3979.6208],[4300,-4002.9905],[4355,-3979.6208],[4360,-4395.9619],[4364,-5026.2046]]
       }
       leg_r_pv_ik_ctrl.Location.Y {
        type=MovieSceneScriptingFloatChannel
        default=7.9236
        thinned_from=220
        keys [[1780,-6317.3579],[1806,-6300.9746],[1874,-6299.5029],[1881,-6287.5918],[1911,-6149.8838],[1917,-6106.6743],[1931,-5982.8853],[1941,-5875.2646],[1966,-5549.6045],[1989,-5193.4106],[1997,-5057.958],[2021,-4625.5444],[2101,-3119.2744],[2126,-2662.2153],[2153,-2190.8923],[2194,-1531.4758],[2201,-1430.3119],[2231,-1037.5824],[2237,-967.496],[2250,-825.6688],[2261,-725.4003],[2286,-591.0543],[2366,-476.463],[2389,-473.4519],[2397,-473.183],[2421,-474.8452],[2446,-475.3619],[2466,-471.2656],[2468,-432.4271],[2470,-471.5169],[2484,-406.7473],[2487,-364.6921],[2496,-407.0265],[2501,-392.1503],[2502,-387.5023],[2506,-381.1143],[2507,-382.1163],[2508,-396.8534],[2509,-419.1385],[2511,-441.8673],[2513,-431.2759],[2515,-442.9962],[2518,-463.1989],[2519,-476.2954],[2521,-471.2655],[2523,-470.7146],[2527,-470.8588],[2532,-466.8421],[2533,-467.2199],[2538,-464.9396],[2758,-493.8587],[2762,-508.324],[2769,-518.4316],[2785,-520.6159],[2927,-476.4626],[2938,-738.6293],[2986,-418.0791],[3058,-418.0791],[3070,-402.7572],[3083,-402.7572],[3099,-197.837],[3185,-197.837],[3215,-530.4311],[3216,-526.4206],[3241,-397.3172],[3264,-263.6003],[3268,-241.9448],[3308,-101.3212],[3309,-100.0595],[3315,-96.4239],[3345,-87.2479],[3346,-86.6961],[3352,-84.9649],[3365,-82.7836],[3366,-82.96],[3376,-85.7536],[3401,-96.0864],[3424,-88.4051],[3432,-85.2039],[3451,-83.8986],[3481,-96.0852],[3504,-88.6796],[3512,-85.3603],[3531,-84.2796],[3536,-85.8987],[3561,-96.155],[3584,-87.9191],[3586,-86.9953],[3588,-86.3148],[3628,-92.5202],[3629,-93.0157],[3635,-94.8586],[3636,-95.0318],[3665,-87.8238],[3666,-87.247],[3672,-85.1797],[3685,-74.9774],[3686,-72.6347],[3696,-47.0093],[3721,-80.6589],[3744,-136.553],[3752,-163.5431],[3771,-240.6912],[3776,-262.9149],[3824,-482.2176],[3832,-517.0658],[3851,-588.5038],[3856,-602.7851],[3881,-628.6771],[3895,-240.6912],[3900,-262.9149],[3948,-482.2176],[3956,-517.0658],[3975,-588.5038],[3980,-602.7851],[4005,-628.6771],[4019,-240.6912],[4024,-262.9149],[4072,-482.2176],[4080,-517.0658],[4099,-588.5038],[4104,-602.7851],[4129,-628.6771],[4144,-240.6912],[4149,-262.9149],[4197,-482.2176],[4205,-517.0658],[4224,-588.5038],[4229,-602.7851],[4254,-628.6771],[4270,-240.6912],[4275,-262.9149],[4323,-482.2176],[4331,-517.0658],[4350,-588.5038],[4355,-602.7851],[4360,-210.4225],[4364,-86.6695]]
       }
       leg_r_pv_ik_ctrl.Location.Z {
        type=MovieSceneScriptingFloatChannel
        default=9.3819
        thinned_from=222
        keys [[1780,337.8889],[1806,330.5858],[1829,331.6788],[1833,332.7832],[1873,332.7538],[1881,329.2324],[1911,322.0186],[1917,321.3821],[1931,317.5876],[1941,311.7724],[1966,288.6703],[1989,272.9805],[1997,268.7339],[2017,254.5998],[2021,250.7635],[2046,221.5706],[2069,202.3542],[2077,197.3931],[2097,182.0257],[2101,178.0786],[2126,149.1631],[2149,131.3143],[2153,129.0798],[2193,86.3882],[2201,73.5136],[2230,33.7636],[2237,25.8526],[2250,10.6443],[2251,9.8725],[2261,7.4381],[2286,-4.2304],[2309,-75.5323],[2317,-97.4173],[2366,-119.7696],[2389,-122.224],[2397,-124.3854],[2416,-127.2334],[2421,-127.1436],[2446,-122.4515],[2461,-123.6521],[2466,-126.2548],[2468,-121.4821],[2470,-127.6708],[2484,-118.3654],[2487,-114.7543],[2496,-119.3753],[2501,-115.5792],[2502,-115.1886],[2506,-116.7369],[2507,-118.1878],[2508,-122.1759],[2509,-126.4888],[2511,-128.1529],[2513,-131.9129],[2515,-137.4087],[2518,-144.8374],[2519,-146.0977],[2521,-146.7725],[2523,-150.668],[2524,-151.8087],[2527,-154.3143],[2530,-156.0647],[2532,-156.1771],[2533,-156.6354],[2538,-157.2086],[2758,-213.7327],[2759,-212.4413],[2762,-205.8834],[2769,-193.2988],[2770,-192.6236],[2777,-189.533],[2783,-188.471],[2927,-191.9935],[2938,-474.3886],[2986,-258.7983],[3058,-258.7983],[3070,-154.346],[3083,-154.346],[3099,-149.0973],[3185,-149.0973],[3215,-229.1929],[3216,-229.9268],[3241,-238.7625],[3264,-263.7078],[3268,-269.6517],[3308,-310.4809],[3316,-310.4568],[3345,-319.6841],[3352,-325.2283],[3366,-333.7026],[3376,-335.9406],[3401,-328.2422],[3424,-329.5204],[3432,-333.3241],[3452,-339.7603],[3456,-339.9416],[3481,-329.9562],[3504,-331.4693],[3512,-335.0818],[3532,-339.8889],[3536,-339.6453],[3561,-328.4477],[3584,-329.3348],[3588,-330.894],[3628,-336.5126],[3636,-332.2548],[3665,-331.7139],[3672,-335.0567],[3685,-337.5846],[3686,-337.0892],[3696,-328.0249],[3721,-307.0324],[3744,-300.0639],[3752,-301.3682],[3772,-301.4331],[3776,-300.1133],[3801,-282.3281],[3824,-280.5346],[3832,-283.9809],[3852,-288.318],[3856,-287.9143],[3881,-266.1932],[3895,-301.6861],[3900,-300.1133],[3925,-282.3281],[3948,-280.5346],[3956,-283.9809],[3976,-288.318],[3980,-287.9143],[4005,-266.1932],[4019,-301.6861],[4024,-300.1133],[4049,-282.3281],[4072,-280.5346],[4080,-283.9809],[4100,-288.318],[4104,-287.9143],[4129,-266.1932],[4144,-301.6861],[4149,-300.1133],[4174,-282.3281],[4197,-280.5346],[4205,-283.9809],[4225,-288.318],[4229,-287.9143],[4254,-266.1932],[4270,-301.6861],[4275,-300.1133],[4300,-282.3281],[4323,-280.5346],[4331,-283.9809],[4351,-288.318],[4355,-287.9143],[4360,-371.5992],[4364,722.9337]]
       }
       leg_r_pv_ik_ctrl.Rotation.X {
        type=MovieSceneScriptingFloatChannel
        default=-12.0486
        thinned_from=226
        keys [[1780,-144.1974],[1806,-168.8752],[1829,-171.5352],[1833,-169.7392],[1873,-162.3892],[1874,-162.8323],[1881,-167.6683],[1910,-172.2244],[1917,-168.4497],[1930,-162.2535],[1941,-159.2571],[1966,-171.2101],[1989,-171.1378],[1997,-166.8175],[2017,-159.9398],[2021,-159.4819],[2046,-172.1454],[2069,-172.1663],[2077,-167.929],[2097,-160.1034],[2101,-159.1952],[2126,-170.8713],[2149,-171.45],[2153,-169.762],[2193,-162.3892],[2201,-167.3754],[2230,-171.828],[2237,-167.9026],[2250,-159.3526],[2251,-156.6586],[2261,-126.1728],[2286,-104.4336],[2309,-105.7155],[2317,-111.4902],[2337,55.1227],[2339,74.2227],[2340,77.8439],[2341,84.509],[2366,112.2971],[2389,104.5928],[2397,91.8859],[2416,82.2121],[2421,83.5695],[2446,104.0022],[2461,103.6506],[2465,96.2048],[2466,94.4751],[2468,-20.9081],[2470,87.5917],[2484,-261.0044],[2487,-344.4832],[2496,-242.7044],[2501,-229.5721],[2502,-228.2982],[2506,130.3329],[2507,128.5923],[2508,116.4974],[2509,100.8568],[2511,90.0611],[2513,84.2635],[2515,63.7968],[2518,43.3727],[2519,29.3577],[2521,40.1096],[2523,43.3451],[2524,43.4544],[2526,45.364],[2527,46.9459],[2530,52.4814],[2532,54.8942],[2533,55.5511],[2538,60.5221],[2758,70.3869],[2759,69.8883],[2762,73.4045],[2769,87.2441],[2770,87.9682],[2777,92.137],[2783,92.9535],[2785,92.8101],[2927,91.9871],[2938,44.1907],[2986,64.1074],[3058,64.1074],[3070,-86.2482],[3083,-86.2482],[3099,60.3069],[3185,60.3069],[3215,107.0468],[3216,106.508],[3241,105.8596],[3264,106.5249],[3268,104.9871],[3308,97.274],[3309,97.6378],[3316,101.0384],[3345,104.3161],[3352,101.076],[3365,95.9202],[3366,95.7177],[3376,95.1704],[3401,105.0752],[3424,106.4924],[3432,102.7639],[3451,95.4226],[3456,94.9041],[3481,103.4866],[3504,104.6918],[3512,101.1151],[3531,95.219],[3536,95.2213],[3561,105.0357],[3584,106.4805],[3588,105.0034],[3628,97.2739],[3636,101.0377],[3665,104.311],[3672,101.1333],[3685,92.4099],[3696,80.6756],[3721,82.2859],[3744,80.5805],[3752,77.5113],[3771,72.0752],[3776,71.7234],[3801,76.7907],[3824,78.8016],[3832,76.9652],[3851,74.4049],[3856,74.5713],[3881,94.2847],[3895,72.0752],[3900,71.7234],[3925,76.7907],[3948,78.8016],[3956,76.9652],[3975,74.4049],[3980,74.5713],[4005,94.2847],[4019,72.0752],[4024,71.7234],[4049,76.7907],[4072,78.8016],[4080,76.9652],[4099,74.4049],[4104,74.5713],[4129,94.2847],[4144,72.0752],[4149,71.7234],[4174,76.7907],[4197,78.8016],[4205,76.9652],[4224,74.4049],[4229,74.5713],[4254,94.2847],[4270,72.0752],[4275,71.7234],[4300,76.7907],[4323,78.8016],[4331,76.9652],[4350,74.4049],[4355,74.5713],[4360,69.6006],[4364,71.0928]]
       }
       leg_r_pv_ik_ctrl.Rotation.Y {
        type=MovieSceneScriptingFloatChannel
        default=0.1485
        thinned_from=227
        keys [[1780,69.5328],[1806,81.8594],[1829,81.8098],[1833,81.223],[1873,79.703],[1881,80.9328],[1910,80.9759],[1917,79.9091],[1930,78.2716],[1941,78.4055],[1966,82.3019],[1989,81.8414],[1997,80.5594],[2016,78.283],[2021,78.3323],[2046,81.5882],[2069,81.0983],[2077,79.8815],[2096,78.2016],[2101,78.4235],[2126,82.2374],[2149,81.8244],[2153,81.2198],[2193,79.703],[2201,80.8852],[2230,80.9295],[2250,77.961],[2251,77.5975],[2261,67.21],[2286,0.9637],[2309,-67.6524],[2317,-83.1121],[2336,52.6265],[2337,64.0048],[2339,-37.8993],[2340,-71.0222],[2341,-83.3123],[2366,-85.0725],[2389,-84.6105],[2397,-83.4534],[2417,-80.0904],[2421,-79.7295],[2446,-81.5662],[2461,-81.4891],[2466,-80.1009],[2468,-84.0661],[2470,-79.3516],[2484,-79.3421],[2487,-87.4198],[2496,-75.3529],[2501,-72.7781],[2502,-70.4401],[2506,-71.2046],[2507,-70.9972],[2508,-70.3564],[2509,-70.9986],[2511,-74.308],[2513,-73.0836],[2515,-74.1038],[2519,-69.6434],[2521,-69.8621],[2523,-67.4762],[2524,-66.9937],[2527,-66.7003],[2530,-66.9527],[2533,-67.5695],[2538,-69.0199],[2758,-65.3654],[2762,-68.3387],[2769,-69.9217],[2777,-69.8465],[2785,-70.0812],[2927,-68.7318],[2938,-63.6234],[2986,-26.8841],[3058,-26.8841],[3070,-46.0077],[3083,-46.0077],[3099,-29.971],[3185,-29.971],[3215,-68.1837],[3216,-67.9888],[3264,-69.0658],[3268,-68.761],[3308,-66.7794],[3316,-67.8888],[3345,-69.1566],[3352,-68.4751],[3366,-66.7875],[3376,-65.9982],[3401,-68.5106],[3424,-68.897],[3432,-68.2306],[3452,-66.2021],[3456,-66.005],[3481,-68.7515],[3504,-69.1721],[3512,-68.453],[3532,-66.231],[3536,-65.9917],[3561,-68.5152],[3584,-69.0388],[3628,-66.7794],[3636,-67.9804],[3665,-69.1535],[3672,-68.4606],[3685,-66.3487],[3696,-63.9021],[3721,-66.934],[3744,-67.0443],[3771,-63.167],[3776,-62.9589],[3801,-66.7186],[3824,-66.9905],[3832,-65.8597],[3851,-63.4254],[3856,-63.2248],[3881,-65.5333],[3895,-63.167],[3900,-62.9589],[3925,-66.7186],[3948,-66.9905],[3956,-65.8597],[3975,-63.4254],[3980,-63.2248],[4005,-65.5333],[4019,-63.167],[4024,-62.9589],[4049,-66.7186],[4072,-66.9905],[4080,-65.8597],[4099,-63.4254],[4104,-63.2248],[4129,-65.5333],[4144,-63.167],[4149,-62.9589],[4174,-66.7186],[4197,-66.9905],[4205,-65.8597],[4224,-63.4254],[4229,-63.2248],[4254,-65.5333],[4270,-63.167],[4275,-62.9589],[4300,-66.7186],[4323,-66.9905],[4331,-65.8597],[4350,-63.4254],[4355,-63.2248],[4360,-59.2024],[4364,-24.1277]]
       }
       leg_r_pv_ik_ctrl.Rotation.Z {
        type=MovieSceneScriptingFloatChannel
        default=0.3682
        thinned_from=227
        keys [[1780,-149.9951],[1806,-159.0666],[1829,-164.5426],[1833,-162.7221],[1873,-150.9861],[1874,-151.5978],[1881,-157.3412],[1910,-164.6876],[1911,-164.2541],[1917,-160.3894],[1931,-152.557],[1941,-148.3367],[1966,-160.792],[1989,-163.8131],[1997,-159.1355],[2017,-149.6559],[2021,-148.4731],[2046,-162.5392],[2069,-165.1991],[2077,-160.5572],[2097,-149.7617],[2101,-148.2784],[2126,-161.2032],[2149,-164.5204],[2153,-162.7245],[2193,-150.9861],[2194,-151.5019],[2201,-157.5875],[2230,-165.0213],[2237,-160.9316],[2250,-149.942],[2251,-147.2063],[2261,-115.1678],[2286,-91.5665],[2309,-82.9779],[2317,-77.1923],[2336,-70.1716],[2337,75.2912],[2339,92.1438],[2340,91.348],[2341,85.5663],[2366,58.905],[2389,69.3318],[2397,81.7891],[2416,88.9145],[2421,86.9091],[2446,67.3766],[2461,69.889],[2465,77.4441],[2466,79.1722],[2468,209.9735],[2470,84.9534],[2484,92.6963],[2487,196.0077],[2496,95.3643],[2501,98.3314],[2502,103.813],[2506,111.3405],[2508,112.3011],[2509,109.9341],[2511,100.6852],[2513,118.5425],[2515,139.4197],[2518,148.8008],[2519,150.0942],[2521,142.4945],[2523,138.2269],[2527,136.3856],[2533,132.5002],[2538,131.1672],[2758,121.5383],[2759,119.9076],[2762,111.0634],[2769,93.2093],[2770,92.321],[2777,87.2188],[2783,85.8804],[2927,92.8689],[2938,102.371],[2986,96.3283],[3058,96.3283],[3070,-42.7768],[3083,-42.7768],[3099,53.5031],[3185,53.5031],[3215,100.6576],[3216,100.3044],[3241,86.679],[3264,87.9503],[3268,89.2447],[3308,93.1204],[3309,92.8604],[3316,89.7922],[3345,90.0415],[3352,92.9858],[3365,96.8802],[3366,96.9059],[3376,95.9952],[3401,86.7094],[3424,87.6213],[3432,91.1274],[3451,96.4347],[3456,96.1867],[3481,88.4245],[3504,89.5748],[3512,92.9585],[3531,96.6108],[3536,95.9447],[3561,86.7334],[3584,87.8575],[3588,89.2826],[3628,93.1204],[3636,90.2709],[3665,90.0299],[3672,93.0122],[3685,95.5519],[3686,95.1502],[3696,89.3506],[3721,80.6788],[3744,81.3574],[3752,83.5274],[3771,85.5978],[3776,85.1194],[3801,81.4516],[3824,83.429],[3832,85.68],[3851,86.9791],[3856,86.2245],[3881,68.8023],[3895,85.5978],[3900,85.1194],[3925,81.4516],[3948,83.429],[3956,85.68],[3975,86.9791],[3980,86.2245],[4005,68.8023],[4019,85.5978],[4024,85.1194],[4049,81.4516],[4072,83.429],[4080,85.68],[4099,86.9791],[4104,86.2245],[4129,68.8023],[4144,85.5978],[4149,85.1194],[4174,81.4516],[4197,83.429],[4205,85.68],[4224,86.9791],[4229,86.2245],[4254,68.8023],[4270,85.5978],[4275,85.1194],[4300,81.4516],[4323,83.429],[4331,85.68],[4350,86.9791],[4355,86.2245],[4360,87.9942],[4364,26.6314]]
       }
       leg_r_pv_ik_ctrl.Scale.X {
        type=MovieSceneScriptingFloatChannel
        default=1
        thinned_from=227
        keys [[1780,1.25],[2519,1.25],[2521,1.2669],[2538,1.555],[2758,2.834],[2762,2.9243],[2769,3],[2785,3],[2927,2.9216],[2938,2.9491],[2986,2.9334],[3058,2.9334],[3070,2.9433],[3083,2.9433],[3099,2.9295],[3185,2.9295],[3216,3],[4355,3],[4360,3.0279],[4364,3.0187]]
       }
       leg_r_pv_ik_ctrl.Scale.Y {
        type=MovieSceneScriptingFloatChannel
        default=1
        thinned_from=227
        keys [[1780,1.25],[2519,1.25],[2521,1.2669],[2538,1.555],[2758,2.834],[2762,2.9243],[2769,3],[2785,3],[2927,2.9216],[2938,2.9491],[2986,2.9334],[3058,2.9334],[3070,2.9433],[3083,2.9433],[3099,2.9295],[3185,2.9295],[3216,3],[4355,3],[4360,3.0279],[4364,3.0187]]
       }
       leg_r_pv_ik_ctrl.Scale.Z {
        type=MovieSceneScriptingFloatChannel
        default=1
        thinned_from=227
        keys [[1780,1.25],[2519,1.25],[2521,1.2669],[2538,1.555],[2758,2.834],[2762,2.9243],[2769,3],[2785,3],[2927,2.9216],[2938,2.9491],[2986,2.9334],[3058,2.9334],[3070,2.9433],[3083,2.9433],[3099,2.9295],[3185,2.9295],[3216,3],[4355,3],[4360,3.0279],[4364,3.0187]]
       }
       upperarm_l_fk_ctrl.Location.X {type=MovieSceneScriptingFloatChannel, default=0, thinned_from=1211, keys [[1790,0],[4368,0]]}
       upperarm_l_fk_ctrl.Location.Y {type=MovieSceneScriptingFloatChannel, default=0, thinned_from=1165, keys [[1790,0],[4368,0]]}
       upperarm_l_fk_ctrl.Location.Z {type=MovieSceneScriptingFloatChannel, default=0, thinned_from=1029, keys [[1790,0],[4368,0]]}
       upperarm_l_fk_ctrl.Rotation.X {
        type=MovieSceneScriptingFloatChannel
        default=0
        thinned_from=66
        keys [[1790,-14.8961],[1822,-4.328],[1864,-15.6929],[1902,-4.5904],[1944,-15.6243],[1982,-4.3333],[2024,-15.6906],[2063,-4.5864],[2104,-15.6222],[2142,-4.328],[2184,-15.6929],[2222,-4.5904],[2264,-15.6243],[2302,-4.3333],[2344,-15.6906],[2383,-4.5864],[2424,-15.6222],[2462,-4.328],[2471,-7.2989],[2484,19.1311],[2501,12.8696],[2504,14.2698],[2511,8.7043],[2512,13.6838],[2518,-47.0153],[2519,-62.0042],[2521,-67.2815],[2526,-60.6637],[2757,-55.4892],[2763,-54.7898],[2779,-27.0956],[2786,-19.6256],[2938,-19.0842],[2948,-15.8356],[2986,9.7549],[3058,9.7549],[3070,-31.2108],[3099,-31.2108],[3185,-39.1211],[3208,-35.6803],[3219,-15.6222],[3257,-4.328],[3299,-15.6928],[3337,-4.5904],[3379,-15.6244],[3417,-4.3333],[3459,-15.6906],[3498,-4.5864],[3539,-15.6222],[3577,-4.328],[3619,-15.6928],[3657,-4.5904],[3699,-15.6244],[3889,-13.6196],[3932,-12.8697],[4033,-10.7734],[4360,-5.619],[4363,-5.619],[4364,-10.6107],[4365,48.1324],[4366,-38.78],[4367,-26.0449],[4368,-27.5839],[4375,-27.1942]]
       }
       upperarm_l_fk_ctrl.Rotation.Y {
        type=MovieSceneScriptingFloatChannel
        default=0
        thinned_from=67
        keys [[1794,13.3644],[1815,8.6932],[1850,17.6063],[1895,9.3597],[1929,18.0517],[1975,8.6891],[2010,17.5581],[2055,9.3542],[2089,18.0464],[2135,8.6932],[2170,17.6063],[2215,9.3597],[2249,18.0517],[2295,8.6891],[2330,17.5581],[2375,9.3542],[2409,18.0464],[2451,8.6932],[2466,11.7797],[2496,-15.037],[2511,-0.227],[2512,-8.1679],[2515,26.7983],[2519,11.54],[2521,4.3502],[2526,10.6427],[2529,9.6582],[2531,9.9872],[2532,9.8652],[2538,9.3393],[2757,7.6897],[2765,13.5666],[2790,8.1459],[2814,1.1874],[2927,7.8373],[2938,7.8466],[2948,10.6591],[2986,32.8839],[3058,32.8839],[3120,31.7877],[3200,33.6773],[3207,33.0395],[3232,6.5226],[3246,8.6932],[3281,17.6063],[3326,9.3597],[3360,18.0517],[3406,8.6891],[3441,17.5581],[3486,9.3542],[3520,18.0464],[3566,8.6932],[3601,17.6063],[3646,9.3597],[3680,18.0517],[3889,16.2562],[3932,28.9527],[4032,13.823],[4360,8.8784],[4363,8.8784],[4364,7.9108],[4365,-34.0553],[4366,-0.5428],[4367,2.6855],[4368,6.0373],[4375,-1.5688]]
       }
       upperarm_l_fk_ctrl.Rotation.Z {
        type=MovieSceneScriptingFloatChannel
        default=0
        thinned_from=64
        keys [[1790,-23.9475],[1825,-17.8433],[1865,-24.278],[1906,-17.8091],[1944,-24.2609],[1986,-17.8084],[2024,-24.289],[2066,-17.8015],[2104,-24.2607],[2145,-17.8433],[2185,-24.278],[2226,-17.8091],[2264,-24.2609],[2306,-17.8084],[2344,-24.289],[2386,-17.8015],[2424,-24.2607],[2465,-17.8433],[2470,-18.2923],[2501,105.6491],[2504,103.0049],[2511,117.8659],[2515,17.0242],[2521,-21.1001],[2528,-22.9201],[2532,-22.1741],[2757,-40.4183],[2773,-36.0918],[2917,-34.3809],[2938,-34.3398],[2948,-32.448],[2986,28.0757],[3058,28.0757],[3063,16.5818],[3069,-55.6015],[3088,-61.5241],[3108,-61.5241],[3185,-57.8925],[3209,-57.7695],[3224,-18.6858],[3238,-4.3042],[3243,-3.8418],[3265,-11.1872],[3274,-13.3725],[3302,-7.3125],[3310,-6.1397],[3379,-24.2609],[3421,-17.8084],[3459,-24.289],[3501,-17.8015],[3539,-24.2607],[3580,-17.8433],[3620,-24.278],[3661,-17.8091],[3889,-18.1019],[4033,-18.4285],[4363,-18.9812],[4364,-31.3962],[4366,-85.5438],[4367,-98.437],[4368,-78.8308],[4375,-77.4674]]
       }
       upperarm_l_fk_ctrl.Scale.X {type=MovieSceneScriptingFloatChannel, default=1, const [1,17]}
       upperarm_l_fk_ctrl.Scale.Y {type=MovieSceneScriptingFloatChannel, default=1, const [1,17]}
       upperarm_l_fk_ctrl.Scale.Z {type=MovieSceneScriptingFloatChannel, default=1, const [1,17]}
       lowerarm_l_fk_ctrl.Location.X {type=MovieSceneScriptingFloatChannel, default=0, thinned_from=612, keys [[1790,0],[2512,0],[2513,-0.0003],[2515,-0.0001],[4368,0]]}
       lowerarm_l_fk_ctrl.Location.Y {type=MovieSceneScriptingFloatChannel, default=0, thinned_from=622, keys [[1790,0],[2513,0.0001],[4368,0]]}
       lowerarm_l_fk_ctrl.Location.Z {type=MovieSceneScriptingFloatChannel, default=0, thinned_from=623, keys [[1790,0],[4368,0]]}
       lowerarm_l_fk_ctrl.Rotation.X {
        type=MovieSceneScriptingFloatChannel
        default=0
        thinned_from=64
        keys [[1790,0.0063],[1812,0.0866],[1849,-0.1028],[1892,0.0866],[1929,-0.1028],[1972,0.0866],[2009,-0.1028],[2052,0.0866],[2089,-0.1028],[2132,0.0866],[2169,-0.1028],[2212,0.0866],[2249,-0.1028],[2292,0.0866],[2329,-0.1028],[2372,0.0866],[2409,-0.1028],[2452,0.0866],[2463,0.0579],[2468,-0.1608],[2496,-3.1469],[2501,-3.1761],[2504,-2.9359],[2506,-2.9581],[2507,-2.9383],[2511,-4.6927],[2515,-1.4969],[2521,-1.0845],[2525,-1.134],[2533,-1.109],[2536,-1.1109],[2757,-4.2172],[2762,-4.2377],[2768,-11.7511],[2774,-22.1866],[2778,-16.4009],[2782,-4.1114],[2784,-1.4526],[2791,-0.1701],[2802,-0.1701],[2938,-0.1727],[2948,-0.1296],[2986,0.2115],[3058,0.2115],[3070,0.3475],[3099,0.3475],[3185,0.2031],[3201,0.1731],[3219,0.142],[3247,0.0866],[3284,-0.1028],[3327,0.0866],[3364,-0.1028],[4367,-0.1028],[4368,0]]
       }
       lowerarm_l_fk_ctrl.Rotation.Y {
        type=MovieSceneScriptingFloatChannel
        default=0
        thinned_from=62
        keys [[1790,0.0243],[1812,0.4179],[1849,-0.3423],[1892,0.4179],[1930,-0.3424],[1972,0.4179],[2009,-0.3424],[2052,0.4179],[2089,-0.3424],[2132,0.4179],[2169,-0.3423],[2212,0.4179],[2250,-0.3424],[2292,0.4179],[2329,-0.3424],[2372,0.4179],[2409,-0.3424],[2452,0.4179],[2471,-0.1626],[2502,-3.1973],[2504,-3.1279],[2506,-3.1372],[2507,-3.1292],[2510,-3.3455],[2512,-2.9855],[2513,-0.5398],[2515,1.1963],[2521,0.7328],[2525,1.1142],[2533,0.9963],[2536,1.0068],[2757,0.6556],[2759,1.0558],[2775,-0.5342],[2779,-0.4051],[2791,-0.5232],[2927,-0.5262],[2938,-0.5271],[2948,-0.7224],[2986,-2.2639],[3058,-2.2639],[3070,0.0899],[3099,0.0899],[3185,0.2959],[3201,0.3421],[3219,0.3843],[3247,0.4179],[3284,-0.3423],[3327,0.4179],[3365,0.0705],[4367,0.0705],[4368,0]]
       }
       lowerarm_l_fk_ctrl.Rotation.Z {
        type=MovieSceneScriptingFloatChannel
        default=0
        thinned_from=67
        keys [[1790,-0.3175],[1812,-5.4617],[1849,4.5811],[1892,-5.4617],[1929,4.5817],[1972,-5.4617],[2009,4.5817],[2052,-5.4617],[2089,4.5817],[2132,-5.4617],[2169,4.5811],[2212,-5.4617],[2249,4.5817],[2292,-5.4617],[2329,4.5817],[2372,-5.4617],[2409,4.5817],[2452,-5.4617],[2467,-1.8662],[2502,60.5997],[2504,57.4109],[2506,57.8337],[2507,57.4685],[2511,80.2937],[2513,30.4832],[2515,21.9145],[2518,47.2055],[2519,62.4094],[2521,69.6393],[2525,56.3036],[2533,60.4733],[2536,60.106],[2757,0.9383],[2778,5.5432],[2789,7.052],[2940,7.0541],[2947,10.3344],[2954,11.5094],[2977,-2.958],[2986,-5.1461],[3057,-5.1461],[3065,10.9004],[3069,15.1533],[3074,-5.8198],[3080,-11.6496],[3091,-12.9434],[3201,-14.4757],[3228,-7.478],[3243,5.8638],[3283,-10.2837],[3327,-5.4617],[3365,0.208],[4367,0.208],[4368,0]]
       }
       lowerarm_l_fk_ctrl.Scale.X {type=MovieSceneScriptingFloatChannel, default=1, const [1,15]}
       lowerarm_l_fk_ctrl.Scale.Y {type=MovieSceneScriptingFloatChannel, default=1, const [1,15]}
       lowerarm_l_fk_ctrl.Scale.Z {type=MovieSceneScriptingFloatChannel, default=1, const [1,15]}
       hand_l_fk_ctrl.Location.X {
        type=MovieSceneScriptingFloatChannel
        default=0
        thinned_from=38
        keys [[1790,0],[2484,-0.0002],[2512,0],[2513,-0.0011],[2515,-0.0001],[4364,0]]
       }
       hand_l_fk_ctrl.Location.Y {type=MovieSceneScriptingFloatChannel, default=0, thinned_from=185, keys [[1790,0],[2512,0],[2513,-0.0002],[2515,0],[4364,0]]}
       hand_l_fk_ctrl.Location.Z {type=MovieSceneScriptingFloatChannel, default=0, thinned_from=932, keys [[1790,0],[2512,0],[2513,-0.0003],[2515,0],[4364,0]]}
       hand_l_fk_ctrl.Rotation.X {
        type=MovieSceneScriptingFloatChannel
        default=0.2503
        thinned_from=64
        keys [[1790,0.2503],[1830,1.0588],[1868,0.2342],[1906,1.0589],[1948,0.2344],[1986,1.0589],[2028,0.2344],[2066,1.0589],[2108,0.2342],[2150,1.0588],[2188,0.2342],[2226,1.0589],[2268,0.2344],[2306,1.0589],[2348,0.2344],[2386,1.0589],[2428,0.2342],[2484,-60.3028],[2501,-68.5938],[2506,-69.1069],[2508,-69.0943],[2513,-59.4935],[2518,-76.3448],[2521,-80.3384],[2765,-51.7951],[2776,16.0099],[2789,1.4913],[2927,0.478],[2938,0.4547],[2986,34.2796],[3052,24.9245],[3124,-17.7277],[3185,-19.83],[3223,0.2342],[3265,1.0588],[3303,0.2342],[3341,1.0589],[3383,0.2344],[3421,1.0589],[3463,0.2344],[3501,1.0589],[3543,0.2342],[3585,1.0588],[3623,0.2342],[3661,1.0589],[3703,0.2344],[3741,1.0589],[3783,0.2344],[3821,1.0589],[3863,0.2342],[3907,0.2344],[3945,1.0589],[3987,0.2342],[4031,0.2344],[4069,1.0589],[4111,0.2342],[4156,0.2344],[4194,1.0589],[4236,0.2342],[4282,0.2344],[4320,1.0589],[4360,0.791],[4364,0.791]]
       }
       hand_l_fk_ctrl.Rotation.Y {
        type=MovieSceneScriptingFloatChannel
        default=6.226
        thinned_from=121
        keys [[1790,6.226],[1826,6.7878],[1830,6.7885],[1868,6.2157],[1900,6.7878],[1906,6.7877],[1948,6.2143],[1980,6.7878],[1986,6.7877],[2028,6.2143],[2060,6.7878],[2066,6.7877],[2108,6.2157],[2146,6.7878],[2150,6.7885],[2188,6.2157],[2220,6.7878],[2226,6.7877],[2268,6.2143],[2300,6.7878],[2306,6.7877],[2348,6.2143],[2380,6.7878],[2386,6.7877],[2428,6.2157],[2467,6.7992],[2506,24.5199],[2511,24.5159],[2512,27.6779],[2519,-3.7774],[2528,6.2201],[2533,2.8726],[2759,4.2715],[2777,14.7825],[2789,13.4184],[2938,13.4462],[2986,11.4022],[3059,4.5752],[3064,42.7148],[3089,0.286],[3148,5.2286],[3223,6.2157],[3261,6.7878],[3265,6.7885],[3303,6.2157],[3335,6.7878],[3341,6.7877],[3383,6.2143],[3415,6.7878],[3421,6.7877],[3463,6.2143],[3495,6.7878],[3501,6.7877],[3543,6.2157],[3581,6.7878],[3585,6.7885],[3623,6.2157],[3655,6.7878],[3661,6.7877],[3703,6.2143],[3735,6.7878],[3741,6.7877],[3783,6.2143],[3815,6.7878],[3821,6.7877],[3863,6.2157],[3907,6.2143],[3939,6.7878],[3945,6.7877],[3987,6.2157],[4031,6.2143],[4063,6.7878],[4069,6.7877],[4111,6.2157],[4156,6.2143],[4188,6.7878],[4194,6.7877],[4236,6.2157],[4282,6.2143],[4314,6.7878],[4320,6.7877],[4360,6.6527],[4364,6.6527]]
       }
       hand_l_fk_ctrl.Rotation.Z {
        type=MovieSceneScriptingFloatChannel
        default=2.1646
        thinned_from=79
        keys [[1790,2.1646],[1804,-1.8891],[1829,-2.8156],[1830,-2.8156],[1868,2.2627],[1906,-2.8149],[1948,2.2611],[1986,-2.8149],[2028,2.2611],[2066,-2.8149],[2108,2.2627],[2149,-2.8156],[2150,-2.8156],[2188,2.2627],[2226,-2.8149],[2268,2.2611],[2306,-2.8149],[2348,2.2611],[2386,-2.8149],[2428,2.2627],[2467,-2.8087],[2506,4.6275],[2511,4.6246],[2512,9.9675],[2513,8.2828],[2515,-5.5731],[2518,-11.2708],[2521,5.0405],[2528,-9.4356],[2767,-7.0567],[2777,-0.3082],[2789,-1.5939],[2927,-1.6067],[2938,-1.6242],[2986,-36.7128],[3058,-36.7128],[3064,-15.7535],[3083,-44.6029],[3099,-44.6029],[3185,-72.229],[3223,2.2627],[3264,-2.8156],[3265,-2.8156],[3303,2.2627],[3341,-2.8149],[3383,2.2611],[3421,-2.8149],[3463,2.2611],[3501,-2.8149],[3543,2.2627],[3584,-2.8156],[3585,-2.8156],[3623,2.2627],[3661,-2.8149],[3703,2.2611],[3741,-2.8149],[3783,2.2611],[3821,-2.8149],[3863,2.2627],[3907,2.2611],[3945,-2.8149],[3987,2.2627],[4031,2.2611],[4069,-2.8149],[4111,2.2627],[4156,2.2611],[4194,-2.8149],[4236,2.2627],[4282,2.2611],[4320,-2.8149],[4360,-1.2738],[4364,-1.2738]]
       }
       hand_l_fk_ctrl.Scale.X {type=MovieSceneScriptingFloatChannel, default=1, const [1,11]}
       hand_l_fk_ctrl.Scale.Y {type=MovieSceneScriptingFloatChannel, default=1, const [1,11]}
       hand_l_fk_ctrl.Scale.Z {type=MovieSceneScriptingFloatChannel, default=1, const [1,11]}
       thumb_01_l_ctrl.Location.X {type=MovieSceneScriptingFloatChannel, default=0, thinned_from=30, keys [[1790,0],[4436,0]]}
       thumb_01_l_ctrl.Location.Y {type=MovieSceneScriptingFloatChannel, default=0, thinned_from=31, keys [[1790,0],[4436,0]]}
       thumb_01_l_ctrl.Location.Z {type=MovieSceneScriptingFloatChannel, default=0, thinned_from=31, keys [[1790,0],[4436,0]]}
       thumb_01_l_ctrl.Rotation.X {
        type=MovieSceneScriptingFloatChannel
        default=-13.482
        thinned_from=106
        keys [[1790,-13.482],[1808,-19.4248],[1826,-20.6586],[1857,-13.4975],[1862,-13.387],[1865,-13.403],[1906,-20.6586],[1937,-13.4975],[1942,-13.387],[1986,-20.6586],[2022,-13.4045],[2025,-13.403],[2066,-20.6586],[2097,-13.4975],[2102,-13.387],[2105,-13.403],[2146,-20.6586],[2177,-13.4975],[2182,-13.387],[2185,-13.403],[2226,-20.6586],[2257,-13.4975],[2262,-13.387],[2306,-20.6586],[2342,-13.4045],[2345,-13.403],[2386,-20.6586],[2417,-13.4975],[2422,-13.387],[2425,-13.403],[2466,-20.6584],[2503,0.796],[2506,0.796],[2507,3.4806],[2509,11.6395],[2512,14.6174],[2515,14.6173],[2520,7.721],[2522,12.5735],[2524,9.3345],[2527,13.4329],[2530,13.8846],[2759,5.7205],[2761,0.3869],[2771,-8.3323],[2780,-18.0053],[2803,-7.1067],[2927,-6.5271],[2938,-6.5055],[2986,11.3587],[3058,11.3587],[3070,-3.0309],[3185,-3.0309],[3218,-13.387],[3220,-13.403],[3261,-20.6586],[3292,-13.4975],[3297,-13.387],[3300,-13.403],[3341,-20.6586],[3372,-13.4975],[3377,-13.387],[3421,-20.6586],[3445,-13.863],[3460,-13.403],[3501,-20.6586],[3532,-13.4975],[3538,-13.387],[3540,-13.403],[3581,-20.6586],[3612,-13.4975],[3617,-13.387],[3620,-13.403],[3661,-20.6586],[3692,-13.4975],[3698,-13.387],[3741,-20.6586],[3765,-13.863],[3780,-13.403],[3821,-20.6586],[3857,-13.387],[3860,-13.403],[3904,-13.403],[3945,-20.6586],[3981,-13.387],[3984,-13.403],[4028,-13.403],[4069,-20.6586],[4105,-13.387],[4108,-13.403],[4153,-13.403],[4194,-20.6586],[4230,-13.387],[4233,-13.403],[4279,-13.403],[4320,-20.6586],[4356,-13.387],[4360,-19.1777],[4364,-19.0358],[4401,-13.4975],[4407,-13.387],[4409,-13.403],[4435,-19.9248],[4436,0]]
       }
       thumb_01_l_ctrl.Rotation.Y {
        type=MovieSceneScriptingFloatChannel
        default=0.9219
        thinned_from=391
        keys [[1790,0.9219],[1813,0.5148],[1814,0.5225],[1819,0.4909],[1821,0.4953],[1823,0.486],[1824,0.4892],[1826,0.484],[1854,0.915],[1855,0.9071],[1857,0.9206],[1860,0.9047],[1862,0.9261],[1865,0.9171],[1870,0.9219],[1893,0.5148],[1894,0.5225],[1899,0.4909],[1901,0.4953],[1903,0.486],[1904,0.4892],[1906,0.484],[1912,0.4782],[1934,0.915],[1935,0.9071],[1937,0.9206],[1940,0.9047],[1942,0.9261],[1950,0.9057],[1951,0.9082],[1953,0.8717],[1954,0.8804],[1973,0.5148],[1974,0.5225],[1979,0.4909],[1981,0.4953],[1983,0.486],[1984,0.4892],[1986,0.484],[1992,0.4822],[2009,0.8867],[2010,0.8838],[2014,0.9144],[2016,0.9097],[2018,0.9173],[2019,0.9136],[2021,0.9199],[2022,0.9161],[2025,0.9171],[2030,0.9219],[2053,0.5148],[2054,0.5225],[2059,0.4909],[2061,0.4953],[2063,0.486],[2064,0.4892],[2066,0.484],[2072,0.4782],[2094,0.915],[2095,0.9071],[2097,0.9206],[2100,0.9047],[2102,0.9261],[2105,0.9171],[2110,0.9219],[2133,0.5148],[2134,0.5225],[2139,0.4909],[2141,0.4953],[2143,0.486],[2144,0.4892],[2146,0.484],[2174,0.915],[2175,0.9071],[2177,0.9206],[2180,0.9047],[2182,0.9261],[2185,0.9171],[2190,0.9219],[2213,0.5148],[2214,0.5225],[2219,0.4909],[2221,0.4953],[2223,0.486],[2224,0.4892],[2226,0.484],[2232,0.4782],[2254,0.915],[2255,0.9071],[2257,0.9206],[2260,0.9047],[2262,0.9261],[2270,0.9057],[2271,0.9082],[2273,0.8717],[2274,0.8804],[2293,0.5148],[2294,0.5225],[2299,0.4909],[2301,0.4953],[2303,0.486],[2304,0.4892],[2306,0.484],[2312,0.4822],[2329,0.8867],[2330,0.8838],[2334,0.9144],[2336,0.9097],[2338,0.9173],[2339,0.9136],[2341,0.9199],[2342,0.9161],[2345,0.9171],[2350,0.9219],[2373,0.5148],[2374,0.5225],[2379,0.4909],[2381,0.4953],[2383,0.486],[2384,0.4892],[2386,0.484],[2392,0.4782],[2414,0.915],[2415,0.9071],[2418,0.9206],[2420,0.9047],[2422,0.9261],[2425,0.9171],[2430,0.9219],[2453,0.5148],[2454,0.5225],[2459,0.4909],[2461,0.4953],[2463,0.486],[2464,0.4892],[2503,-6.9844],[2507,-6.2967],[2509,-4.1788],[2512,-3.4031],[2515,-3.4032],[2518,-5.3194],[2521,-15.4723],[2523,-14.0291],[2530,-15.9151],[2760,-11.2665],[2765,-8.9727],[2785,0.55],[2927,0.962],[2938,0.9654],[2986,2.6307],[3058,2.6307],[3070,3.3498],[3185,3.3498],[3215,0.9047],[3218,0.9261],[3220,0.9171],[3225,0.9219],[3248,0.5148],[3249,0.5225],[3254,0.4909],[3256,0.4953],[3258,0.486],[3259,0.4892],[3262,0.484],[3289,0.915],[3290,0.9071],[3293,0.9206],[3295,0.9047],[3297,0.9261],[3300,0.9171],[3305,0.9219],[3328,0.5148],[3329,0.5225],[3334,0.4909],[3336,0.4953],[3338,0.486],[3339,0.4891],[3341,0.484],[3347,0.4782],[3369,0.915],[3370,0.9071],[3373,0.9206],[3375,0.9047],[3377,0.9261],[3385,0.9057],[3386,0.9081],[3388,0.8717],[3389,0.8804],[3408,0.5148],[3409,0.5225],[3414,0.4909],[3416,0.4953],[3418,0.486],[3419,0.4891],[3422,0.484],[3427,0.4822],[3444,0.8867],[3445,0.8838],[3449,0.9144],[3451,0.9097],[3453,0.9173],[3454,0.9136],[3456,0.9199],[3458,0.9161],[3465,0.9219],[3488,0.5148],[3489,0.5225],[3494,0.4909],[3496,0.4953],[3498,0.486],[3499,0.4891],[3502,0.484],[3507,0.4782],[3529,0.915],[3530,0.9071],[3533,0.9206],[3535,0.9047],[3537,0.9261],[3540,0.9171],[3545,0.9219],[3568,0.5148],[3569,0.5225],[3574,0.4909],[3576,0.4953],[3578,0.486],[3579,0.4892],[3581,0.484],[3609,0.915],[3610,0.9071],[3613,0.9206],[3615,0.9047],[3617,0.9261],[3620,0.9171],[3625,0.9219],[3648,0.5148],[3649,0.5225],[3654,0.4909],[3656,0.4953],[3658,0.486],[3659,0.4891],[3662,0.484],[3667,0.4782],[3689,0.915],[3690,0.9071],[3692,0.9206],[3695,0.9047],[3698,0.9261],[3705,0.9057],[3706,0.9081],[3708,0.8717],[3709,0.8804],[3728,0.5148],[3729,0.5225],[3734,0.4909],[3736,0.4953],[3738,0.486],[3739,0.4891],[3741,0.484],[3747,0.4822],[3764,0.8867],[3765,0.8838],[3769,0.9144],[3771,0.9097],[3773,0.9173],[3774,0.9136],[3776,0.9199],[3778,0.9161],[3785,0.9219],[3808,0.5148],[3809,0.5225],[3814,0.4909],[3816,0.4953],[3818,0.486],[3819,0.4891],[3822,0.484],[3827,0.4782],[3849,0.915],[3850,0.9071],[3853,0.9206],[3855,0.9047],[3857,0.9261],[3860,0.9171],[3865,0.9219],[3888,0.5148],[3893,0.9144],[3895,0.9097],[3897,0.9173],[3898,0.9136],[3900,0.9199],[3902,0.9161],[3909,0.9219],[3932,0.5148],[3933,0.5225],[3938,0.4909],[3940,0.4953],[3942,0.486],[3943,0.4891],[3946,0.484],[3951,0.4782],[3973,0.915],[3974,0.9071],[3977,0.9206],[3979,0.9047],[3981,0.9261],[3984,0.9171],[3989,0.9219],[4012,0.5148],[4017,0.9144],[4019,0.9097],[4021,0.9173],[4022,0.9136],[4024,0.9199],[4026,0.9161],[4033,0.9219],[4056,0.5148],[4057,0.5225],[4062,0.4909],[4064,0.4953],[4066,0.486],[4067,0.4891],[4070,0.484],[4075,0.4782],[4097,0.915],[4098,0.9071],[4101,0.9206],[4103,0.9047],[4105,0.9261],[4108,0.9171],[4113,0.9219],[4136,0.5148],[4142,0.9144],[4144,0.9097],[4146,0.9173],[4147,0.9136],[4149,0.9199],[4151,0.9161],[4158,0.9219],[4181,0.5148],[4182,0.5225],[4187,0.4909],[4189,0.4953],[4191,0.486],[4192,0.4891],[4195,0.484],[4200,0.4782],[4222,0.915],[4223,0.9071],[4226,0.9206],[4228,0.9047],[4230,0.9261],[4233,0.9171],[4238,0.9219],[4261,0.5148],[4268,0.9144],[4270,0.9097],[4272,0.9173],[4273,0.9136],[4275,0.9199],[4277,0.9161],[4284,0.9219],[4307,0.5148],[4308,0.5225],[4313,0.4909],[4315,0.4953],[4317,0.486],[4318,0.4891],[4321,0.484],[4326,0.4782],[4348,0.915],[4349,0.9071],[4352,0.9206],[4354,0.9047],[4356,0.9261],[4360,0.5225],[4364,0.5347],[4398,0.915],[4399,0.9071],[4401,0.9206],[4404,0.9047],[4407,0.9261],[4410,0.9171],[4414,0.9219],[4436,0]]
       }
       thumb_01_l_ctrl.Rotation.Z {
        type=MovieSceneScriptingFloatChannel
        default=-6.0704
        thinned_from=251
        keys [[1790,-6.0704],[1823,-9.072],[1824,-9.0664],[1826,-9.0884],[1854,-6.1246],[1855,-6.1372],[1857,-6.0776],[1860,-6.1028],[1862,-6.0307],[1865,-6.0614],[1869,-6.056],[1903,-9.072],[1904,-9.0664],[1906,-9.0884],[1934,-6.1246],[1935,-6.1372],[1937,-6.0776],[1940,-6.1028],[1942,-6.0307],[1983,-9.072],[1984,-9.0664],[1986,-9.0884],[2018,-6.0804],[2019,-6.0863],[2021,-6.0585],[2029,-6.056],[2063,-9.072],[2064,-9.0664],[2066,-9.0884],[2094,-6.1246],[2095,-6.1372],[2097,-6.0776],[2100,-6.1028],[2102,-6.0307],[2105,-6.0614],[2109,-6.056],[2143,-9.072],[2144,-9.0664],[2146,-9.0884],[2174,-6.1246],[2175,-6.1372],[2177,-6.0776],[2180,-6.1028],[2182,-6.0307],[2185,-6.0614],[2189,-6.056],[2223,-9.072],[2224,-9.0664],[2226,-9.0884],[2254,-6.1246],[2255,-6.1372],[2257,-6.0776],[2260,-6.1028],[2262,-6.0307],[2303,-9.072],[2304,-9.0664],[2306,-9.0884],[2338,-6.0804],[2339,-6.0863],[2341,-6.0585],[2349,-6.056],[2383,-9.072],[2384,-9.0664],[2386,-9.0884],[2414,-6.1246],[2415,-6.1372],[2418,-6.0776],[2420,-6.1028],[2422,-6.0307],[2425,-6.0614],[2429,-6.056],[2463,-9.072],[2464,-9.0664],[2466,-9.0883],[2471,-7.7394],[2503,2.2897],[2506,2.2897],[2512,-0.9339],[2515,-0.9339],[2520,4.9404],[2522,1.8464],[2524,3.9061],[2530,1.003],[2763,10.9673],[2774,0.1535],[2778,-5.7705],[2782,-9.4109],[2794,-11.3247],[2938,-10.8746],[2986,13.1906],[3058,13.1906],[3070,4.7017],[3185,4.7017],[3215,-6.1028],[3218,-6.0307],[3220,-6.0614],[3224,-6.056],[3258,-9.072],[3259,-9.0664],[3261,-9.0884],[3289,-6.1246],[3290,-6.1371],[3293,-6.0776],[3295,-6.1028],[3297,-6.0307],[3300,-6.0614],[3304,-6.056],[3338,-9.072],[3339,-9.0664],[3341,-9.0884],[3369,-6.1246],[3370,-6.1371],[3373,-6.0776],[3375,-6.1028],[3377,-6.0307],[3418,-9.072],[3419,-9.0664],[3421,-9.0884],[3453,-6.0804],[3454,-6.0863],[3456,-6.0585],[3464,-6.056],[3498,-9.072],[3499,-9.0664],[3501,-9.0884],[3529,-6.1246],[3530,-6.1371],[3533,-6.0776],[3535,-6.1028],[3537,-6.0307],[3540,-6.0614],[3544,-6.056],[3578,-9.072],[3579,-9.0664],[3581,-9.0884],[3609,-6.1246],[3610,-6.1371],[3613,-6.0776],[3615,-6.1028],[3617,-6.0307],[3620,-6.0614],[3624,-6.056],[3658,-9.072],[3659,-9.0664],[3661,-9.0884],[3689,-6.1246],[3690,-6.1371],[3692,-6.0776],[3695,-6.1028],[3698,-6.0307],[3738,-9.072],[3739,-9.0664],[3741,-9.0884],[3773,-6.0804],[3774,-6.0863],[3776,-6.0585],[3784,-6.056],[3818,-9.072],[3819,-9.0664],[3822,-9.0884],[3849,-6.1246],[3850,-6.1371],[3853,-6.0776],[3855,-6.1028],[3857,-6.0307],[3860,-6.0614],[3864,-6.056],[3898,-6.0863],[3900,-6.0585],[3908,-6.056],[3942,-9.072],[3943,-9.0664],[3946,-9.0884],[3973,-6.1246],[3974,-6.1371],[3977,-6.0776],[3979,-6.1028],[3981,-6.0307],[3984,-6.0614],[3988,-6.056],[4022,-6.0863],[4024,-6.0585],[4032,-6.056],[4066,-9.072],[4067,-9.0664],[4070,-9.0884],[4097,-6.1246],[4098,-6.1371],[4101,-6.0776],[4103,-6.1028],[4105,-6.0307],[4108,-6.0614],[4112,-6.056],[4147,-6.0863],[4149,-6.0585],[4157,-6.056],[4191,-9.072],[4192,-9.0664],[4195,-9.0884],[4222,-6.1246],[4223,-6.1371],[4226,-6.0776],[4228,-6.1028],[4230,-6.0307],[4233,-6.0614],[4237,-6.056],[4273,-6.0863],[4275,-6.0585],[4283,-6.056],[4317,-9.072],[4318,-9.0664],[4321,-9.0884],[4348,-6.1246],[4349,-6.1371],[4352,-6.0776],[4354,-6.1028],[4356,-6.0307],[4360,-8.5499],[4364,-8.4749],[4398,-6.1246],[4399,-6.1372],[4401,-6.0776],[4404,-6.1028],[4407,-6.0307],[4410,-6.0614],[4413,-6.056],[4435,-8.7881],[4436,0]]
       }
       thumb_01_l_ctrl.Scale.X {type=MovieSceneScriptingFloatChannel, default=1, const [1,12]}
       thumb_01_l_ctrl.Scale.Y {type=MovieSceneScriptingFloatChannel, default=1, const [1,12]}
       thumb_01_l_ctrl.Scale.Z {type=MovieSceneScriptingFloatChannel, default=1, const [1,12]}
       thumb_02_l_ctrl.Location.X {type=MovieSceneScriptingFloatChannel, default=0, thinned_from=668, keys [[1790,0],[2512,0],[2513,-0.0002],[2515,0],[4436,-0.0001]]}
       thumb_02_l_ctrl.Location.Y {type=MovieSceneScriptingFloatChannel, default=0, thinned_from=843, keys [[1790,0],[2512,0],[2513,-0.0002],[2515,0],[4435,0],[4436,-0.0002]]}
       thumb_02_l_ctrl.Location.Z {type=MovieSceneScriptingFloatChannel, default=0, thinned_from=413, keys [[1790,0],[2512,0],[2513,0.0001],[2528,0],[4435,0],[4436,-0.0001]]}
       thumb_02_l_ctrl.Rotation.X {
        type=MovieSceneScriptingFloatChannel
        default=-1.713
        thinned_from=505
        keys [[1790,-1.713],[1803,-1.8882],[1804,-1.8882],[1810,-1.9721],[1826,-1.9714],[1832,-1.9889],[1853,-1.714],[1865,-1.713],[1890,-1.9532],[1891,-1.9532],[1894,-1.9708],[1906,-1.9714],[1913,-1.9879],[1934,-1.7229],[1945,-1.713],[1949,-1.713],[1963,-1.8882],[1964,-1.8882],[1966,-1.9296],[1967,-1.9296],[1970,-1.9721],[1986,-1.9714],[1993,-1.9879],[2009,-1.7405],[2010,-1.7405],[2014,-1.714],[2030,-1.713],[2050,-1.9532],[2051,-1.9532],[2054,-1.9708],[2055,-1.9708],[2062,-1.9889],[2066,-1.9889],[2094,-1.714],[2110,-1.713],[2123,-1.8882],[2124,-1.8882],[2130,-1.9721],[2146,-1.9714],[2152,-1.9889],[2174,-1.714],[2189,-1.713],[2210,-1.9532],[2211,-1.9532],[2214,-1.9708],[2226,-1.9714],[2234,-1.9879],[2253,-1.7229],[2255,-1.7229],[2262,-1.714],[2269,-1.713],[2290,-1.9721],[2306,-1.9714],[2313,-1.9879],[2334,-1.714],[2350,-1.713],[2370,-1.9532],[2371,-1.9532],[2374,-1.9708],[2375,-1.9708],[2382,-1.9889],[2394,-1.9879],[2409,-1.7405],[2414,-1.714],[2430,-1.713],[2443,-1.8882],[2444,-1.8882],[2450,-1.9721],[2465,-1.9714],[2503,0.7911],[2512,0.2786],[2518,0.2833],[2522,1.0502],[2524,0.7818],[2527,1.152],[2531,1.1902],[2758,0.7207],[2760,0.1451],[2771,-0.8775],[2780,-1.7296],[2802,-0.6236],[2927,-0.5775],[2938,-0.5762],[2986,0.8194],[3058,0.8194],[3070,-0.0637],[3185,-0.0637],[3215,-1.714],[3225,-1.713],[3245,-1.9721],[3263,-1.9714],[3267,-1.9889],[3284,-1.7405],[3288,-1.714],[3301,-1.713],[3332,-1.9714],[3341,-1.9714],[3348,-1.9879],[3369,-1.7229],[3377,-1.714],[3384,-1.713],[3402,-1.9296],[3403,-1.9296],[3405,-1.9721],[3425,-1.9714],[3428,-1.9879],[3449,-1.714],[3465,-1.713],[3485,-1.9532],[3486,-1.9532],[3492,-1.9709],[3493,-1.9709],[3497,-1.9889],[3502,-1.9889],[3529,-1.714],[3541,-1.713],[3562,-1.9296],[3563,-1.9296],[3565,-1.9721],[3586,-1.9714],[3587,-1.9889],[3604,-1.7405],[3605,-1.7405],[3609,-1.714],[3621,-1.713],[3645,-1.9532],[3646,-1.9532],[3652,-1.9714],[3662,-1.9714],[3668,-1.9879],[3689,-1.7229],[3700,-1.713],[3704,-1.713],[3725,-1.9721],[3741,-1.9714],[3748,-1.9879],[3769,-1.714],[3780,-1.713],[3805,-1.9532],[3806,-1.9532],[3812,-1.9709],[3813,-1.9709],[3816,-1.9889],[3825,-1.9889],[3844,-1.7405],[3845,-1.7405],[3849,-1.714],[3864,-1.713],[3885,-1.9721],[3888,-1.9708],[3893,-1.714],[3904,-1.713],[3929,-1.9532],[3930,-1.9532],[3936,-1.9709],[3937,-1.9709],[3940,-1.9889],[3949,-1.9889],[3968,-1.7405],[3969,-1.7405],[3973,-1.714],[3988,-1.713],[4009,-1.9721],[4012,-1.9708],[4017,-1.714],[4028,-1.713],[4053,-1.9532],[4054,-1.9532],[4060,-1.9709],[4061,-1.9709],[4064,-1.9889],[4073,-1.9889],[4092,-1.7405],[4093,-1.7405],[4097,-1.714],[4112,-1.713],[4133,-1.9721],[4136,-1.9708],[4142,-1.714],[4153,-1.713],[4178,-1.9532],[4179,-1.9532],[4185,-1.9709],[4186,-1.9709],[4189,-1.9889],[4198,-1.9889],[4217,-1.7405],[4218,-1.7405],[4222,-1.714],[4237,-1.713],[4258,-1.9721],[4261,-1.9708],[4268,-1.714],[4279,-1.713],[4304,-1.9532],[4305,-1.9532],[4311,-1.9709],[4312,-1.9709],[4315,-1.9889],[4324,-1.9889],[4343,-1.7405],[4344,-1.7405],[4348,-1.714],[4356,-1.714],[4360,-1.9708],[4367,-1.9714],[4377,-1.9889],[4399,-1.714],[4410,-1.713],[4435,-1.9532],[4436,-0.0003]]
       }
       thumb_02_l_ctrl.Rotation.Y {
        type=MovieSceneScriptingFloatChannel
        default=4.0271
        thinned_from=220
        keys [[1790,4.0271],[1803,4.4305],[1811,4.6458],[1832,4.6762],[1849,4.0978],[1852,4.0267],[1865,4.0271],[1890,4.5996],[1891,4.5996],[1894,4.6458],[1913,4.6736],[1932,4.0505],[1933,4.0506],[1938,4.0268],[1945,4.0271],[1963,4.4305],[1966,4.5381],[1968,4.5381],[1971,4.6458],[1993,4.6736],[2009,4.0978],[2012,4.0267],[2025,4.0271],[2054,4.6458],[2057,4.6483],[2061,4.6762],[2073,4.6736],[2089,4.0978],[2092,4.0267],[2105,4.0271],[2123,4.4305],[2126,4.5381],[2128,4.5381],[2131,4.6458],[2152,4.6762],[2169,4.0978],[2172,4.0267],[2185,4.0271],[2214,4.6458],[2233,4.6736],[2252,4.0505],[2253,4.0506],[2258,4.0268],[2265,4.0271],[2291,4.6458],[2313,4.6736],[2329,4.0978],[2332,4.0267],[2345,4.0271],[2374,4.6458],[2378,4.6483],[2381,4.6762],[2393,4.6736],[2409,4.0978],[2412,4.0267],[2425,4.0271],[2443,4.4305],[2446,4.5381],[2447,4.5381],[2451,4.6458],[2457,4.6506],[2471,4.4527],[2503,1.0291],[2512,1.7884],[2513,1.7884],[2519,2.1396],[2522,1.6658],[2524,1.9409],[2530,1.551],[2760,2.5231],[2766,3.0488],[2779,4.2221],[2784,3.7302],[2802,1.4595],[2938,1.4607],[2986,2.521],[3058,2.521],[3070,2.4511],[3185,2.4511],[3220,4.0271],[3247,4.6458],[3268,4.6762],[3284,4.0978],[3287,4.0267],[3300,4.0271],[3329,4.6458],[3348,4.6736],[3367,4.0505],[3368,4.0506],[3373,4.0268],[3380,4.0271],[3402,4.5381],[3406,4.6458],[3428,4.6736],[3447,4.0267],[3460,4.0271],[3489,4.6458],[3492,4.6483],[3496,4.6762],[3508,4.6736],[3524,4.0978],[3527,4.0267],[3540,4.0271],[3566,4.6458],[3588,4.6762],[3604,4.0978],[3607,4.0267],[3620,4.0271],[3649,4.6458],[3668,4.6736],[3687,4.0505],[3688,4.0506],[3693,4.0268],[3700,4.0271],[3726,4.6458],[3748,4.6736],[3767,4.0267],[3780,4.0271],[3809,4.6458],[3816,4.6762],[3844,4.0978],[3847,4.0267],[3860,4.0271],[3886,4.6458],[3891,4.0267],[3904,4.0271],[3933,4.6458],[3940,4.6762],[3968,4.0978],[3971,4.0267],[3984,4.0271],[4010,4.6458],[4015,4.0267],[4028,4.0271],[4057,4.6458],[4064,4.6762],[4092,4.0978],[4095,4.0267],[4108,4.0271],[4134,4.6458],[4140,4.0267],[4153,4.0271],[4182,4.6458],[4189,4.6762],[4217,4.0978],[4220,4.0267],[4233,4.0271],[4259,4.6458],[4266,4.0267],[4279,4.0271],[4308,4.6458],[4315,4.6762],[4343,4.0978],[4346,4.0267],[4347,4.027],[4360,4.6482],[4367,4.6506],[4377,4.6762],[4396,4.0267],[4409,4.0271],[4435,4.5996],[4436,0.0033]]
       }
       thumb_02_l_ctrl.Rotation.Z {
        type=MovieSceneScriptingFloatChannel
        default=31.5262
        thinned_from=292
        keys [[1790,31.5262],[1806,32.1282],[1809,32.1258],[1811,32.2546],[1832,32.29],[1836,32.196],[1849,31.6038],[1853,31.5261],[1865,31.5262],[1894,32.2546],[1914,32.287],[1932,31.5518],[1945,31.5262],[1963,32.0018],[1966,32.1282],[1969,32.1258],[1971,32.2546],[1993,32.287],[2009,31.6038],[2013,31.5261],[2025,31.5262],[2054,32.2546],[2074,32.287],[2089,31.6038],[2093,31.5261],[2105,31.5262],[2123,32.0018],[2126,32.1282],[2129,32.1258],[2131,32.2546],[2152,32.29],[2156,32.196],[2169,31.6038],[2173,31.5261],[2185,31.5262],[2214,32.2546],[2234,32.287],[2252,31.5518],[2265,31.5262],[2286,32.1282],[2289,32.1258],[2291,32.2546],[2313,32.287],[2329,31.6038],[2333,31.5261],[2345,31.5262],[2374,32.2546],[2394,32.287],[2409,31.6038],[2413,31.5261],[2425,31.5262],[2443,32.0018],[2446,32.1282],[2449,32.1258],[2452,32.2546],[2466,32.2568],[2471,32.0016],[2503,33.3529],[2505,33.3529],[2510,11.9801],[2513,10.0258],[2518,14.741],[2519,25.8524],[2521,37.0575],[2523,34.4147],[2530,37.5391],[2764,27.2193],[2777,30.3118],[2801,11.0329],[2938,11.0395],[2986,86.8163],[3058,86.8163],[3070,18.3788],[3185,18.3788],[3220,31.5262],[3242,32.1282],[3244,32.1258],[3247,32.2546],[3267,32.29],[3271,32.196],[3287,31.5254],[3300,31.5262],[3325,32.2005],[3348,32.287],[3367,31.5518],[3380,31.5262],[3401,32.1282],[3404,32.1258],[3407,32.2546],[3428,32.287],[3447,31.5254],[3460,31.5262],[3489,32.2546],[3508,32.287],[3527,31.5254],[3540,31.5262],[3562,32.1282],[3564,32.1258],[3566,32.2546],[3587,32.29],[3591,32.196],[3607,31.5254],[3620,31.5262],[3647,32.2],[3652,32.2569],[3668,32.287],[3687,31.5518],[3700,31.5262],[3721,32.1282],[3724,32.1258],[3726,32.2546],[3748,32.287],[3767,31.5254],[3780,31.5262],[3809,32.2546],[3822,32.29],[3848,31.5261],[3860,31.5262],[3882,32.1282],[3884,32.1258],[3886,32.2546],[3891,31.5254],[3904,31.5262],[3933,32.2546],[3946,32.29],[3972,31.5261],[3984,31.5262],[4006,32.1282],[4008,32.1258],[4010,32.2546],[4015,31.5254],[4028,31.5262],[4057,32.2546],[4070,32.29],[4096,31.5261],[4108,31.5262],[4130,32.1282],[4132,32.1258],[4134,32.2546],[4140,31.5254],[4153,31.5262],[4182,32.2546],[4195,32.29],[4221,31.5261],[4233,31.5262],[4255,32.1282],[4257,32.1258],[4259,32.2546],[4266,31.5254],[4279,31.5262],[4308,32.2546],[4321,32.29],[4347,31.5261],[4360,32.2558],[4377,32.29],[4396,31.5254],[4409,31.5262],[4435,32.2004],[4436,0.0083]]
       }
       thumb_02_l_ctrl.Scale.X {type=MovieSceneScriptingFloatChannel, default=1, const [1,12]}
       thumb_02_l_ctrl.Scale.Y {type=MovieSceneScriptingFloatChannel, default=1, const [1,12]}
       thumb_02_l_ctrl.Scale.Z {type=MovieSceneScriptingFloatChannel, default=1, const [1,12]}
       thumb_03_l_ctrl.Location.X {type=MovieSceneScriptingFloatChannel, default=0, thinned_from=233, keys [[1790,0],[2512,0],[2513,-0.0001],[2529,0],[4436,0]]}
       thumb_03_l_ctrl.Location.Y {type=MovieSceneScriptingFloatChannel, default=0, thinned_from=898, keys [[1790,0],[4436,0]]}
       thumb_03_l_ctrl.Location.Z {type=MovieSceneScriptingFloatChannel, default=0, thinned_from=232, keys [[1790,0],[4436,0]]}
       thumb_03_l_ctrl.Rotation.X {
        type=MovieSceneScriptingFloatChannel
        default=0.8579
        thinned_from=86
        keys [[1790,0.8579],[1828,0.0067],[1866,0.8706],[1906,0.0108],[1909,0.0019],[1947,0.8713],[1988,0.0067],[1989,0.0067],[2025,0.872],[2068,0.0067],[2069,0.0067],[2105,0.872],[2148,0.0067],[2149,0.0067],[2186,0.8706],[2226,0.0108],[2229,0.0019],[2267,0.8713],[2308,0.0067],[2309,0.0067],[2345,0.872],[2389,0.0067],[2426,0.872],[2466,0.0104],[2470,0.0468],[2503,1.3385],[2513,1.7878],[2519,1.2006],[2522,1.8509],[2524,1.4488],[2531,2.0068],[2764,0.0755],[2771,0.0518],[2798,4.5983],[2938,4.5705],[2986,0.4044],[3058,0.4044],[3070,1.5304],[3185,1.5304],[3220,0.872],[3264,0.0067],[3302,0.8706],[3341,0.0108],[3345,0.0019],[3382,0.8713],[3423,0.0067],[3425,0.0067],[3460,0.872],[3504,0.0067],[3541,0.872],[3584,0.0067],[3586,0.0067],[3622,0.8706],[3665,0.0019],[3702,0.8713],[3744,0.0067],[3780,0.872],[3823,0.0067],[3824,0.0067],[3860,0.872],[3904,0.872],[3947,0.0067],[3948,0.0067],[3984,0.872],[4028,0.872],[4071,0.0067],[4072,0.0067],[4108,0.872],[4153,0.872],[4196,0.0067],[4197,0.0067],[4233,0.872],[4279,0.872],[4322,0.0067],[4323,0.0067],[4360,0.2222],[4364,0.1663],[4372,0.0067],[4373,0.0067],[4411,0.8706],[4436,0]]
       }
       thumb_03_l_ctrl.Rotation.Y {
        type=MovieSceneScriptingFloatChannel
        default=-1.2625
        thinned_from=189
        keys [[1790,-1.2625],[1828,-0.249],[1859,-1.2689],[1860,-1.2685],[1862,-1.2752],[1866,-1.2743],[1906,-0.2497],[1909,-0.2501],[1939,-1.2689],[1940,-1.2685],[1942,-1.2752],[1944,-1.2746],[1947,-1.2774],[1989,-0.249],[2019,-1.2689],[2020,-1.2685],[2022,-1.2752],[2023,-1.2748],[2025,-1.2805],[2069,-0.249],[2099,-1.2689],[2100,-1.2685],[2102,-1.2752],[2103,-1.2748],[2105,-1.2805],[2149,-0.249],[2179,-1.2689],[2180,-1.2685],[2182,-1.2752],[2186,-1.2743],[2226,-0.2497],[2229,-0.2501],[2259,-1.2689],[2260,-1.2685],[2262,-1.2752],[2264,-1.2746],[2267,-1.2774],[2309,-0.249],[2339,-1.2689],[2340,-1.2685],[2342,-1.2752],[2343,-1.2748],[2345,-1.2805],[2389,-0.249],[2419,-1.2689],[2420,-1.2685],[2422,-1.2752],[2423,-1.2748],[2425,-1.2805],[2466,-0.252],[2469,-0.2874],[2470,-0.2871],[2503,-0.0966],[2512,-0.6119],[2515,-0.6109],[2519,-0.727],[2522,-0.2488],[2524,-0.4849],[2532,-0.143],[2764,-1.3111],[2776,-0.7262],[2800,-6.3979],[2938,-6.3627],[2986,-2.0407],[3058,-2.0407],[3070,-0.7031],[3185,-0.7031],[3215,-1.2685],[3217,-1.2752],[3218,-1.2748],[3220,-1.2805],[3264,-0.249],[3294,-1.2689],[3295,-1.2685],[3297,-1.2752],[3301,-1.2743],[3341,-0.2497],[3344,-0.2501],[3374,-1.2689],[3375,-1.2685],[3377,-1.2752],[3379,-1.2746],[3382,-1.2774],[3425,-0.249],[3454,-1.2689],[3455,-1.2685],[3457,-1.2752],[3458,-1.2748],[3460,-1.2805],[3504,-0.249],[3534,-1.2689],[3535,-1.2685],[3537,-1.2752],[3538,-1.2748],[3540,-1.2805],[3584,-0.249],[3586,-0.249],[3614,-1.2689],[3615,-1.2685],[3617,-1.2752],[3621,-1.2743],[3661,-0.2497],[3664,-0.2501],[3694,-1.2689],[3695,-1.2685],[3697,-1.2752],[3699,-1.2746],[3702,-1.2774],[3744,-0.249],[3774,-1.2689],[3775,-1.2685],[3777,-1.2752],[3778,-1.2748],[3780,-1.2805],[3824,-0.249],[3854,-1.2689],[3855,-1.2685],[3857,-1.2752],[3858,-1.2748],[3860,-1.2805],[3899,-1.2685],[3901,-1.2752],[3902,-1.2748],[3904,-1.2805],[3948,-0.249],[3978,-1.2689],[3979,-1.2685],[3981,-1.2752],[3982,-1.2748],[3984,-1.2805],[4023,-1.2685],[4025,-1.2752],[4026,-1.2748],[4028,-1.2805],[4072,-0.249],[4102,-1.2689],[4103,-1.2685],[4105,-1.2752],[4106,-1.2748],[4108,-1.2805],[4148,-1.2685],[4150,-1.2752],[4151,-1.2748],[4153,-1.2805],[4197,-0.249],[4227,-1.2689],[4228,-1.2685],[4230,-1.2752],[4231,-1.2748],[4233,-1.2805],[4274,-1.2685],[4276,-1.2752],[4277,-1.2748],[4279,-1.2805],[4323,-0.249],[4353,-1.2689],[4354,-1.2685],[4356,-1.2752],[4357,-1.2748],[4360,-0.5269],[4364,-0.4027],[4373,-0.249],[4403,-1.2689],[4404,-1.2685],[4406,-1.2752],[4410,-1.2743],[4436,0]]
       }
       thumb_03_l_ctrl.Rotation.Z {
        type=MovieSceneScriptingFloatChannel
        default=-25.1363
        thinned_from=136
        keys [[1790,-25.1363],[1809,-7.9155],[1828,-5.0134],[1847,-21.9692],[1866,-25.2703],[1885,-11.0141],[1905,-5.0853],[1909,-5.0852],[1928,-22.818],[1947,-25.2703],[1968,-8.4928],[1989,-5.0134],[2007,-21.9714],[2025,-25.2703],[2047,-9.1717],[2069,-5.0134],[2087,-21.9714],[2105,-25.2703],[2127,-9.2253],[2149,-5.0134],[2167,-21.9692],[2186,-25.2703],[2205,-11.0141],[2225,-5.0853],[2229,-5.0852],[2248,-22.818],[2267,-25.2703],[2288,-8.4928],[2309,-5.0134],[2327,-21.9714],[2345,-25.2703],[2367,-9.1717],[2389,-5.0134],[2407,-21.9714],[2425,-25.2703],[2446,-10.0594],[2467,-5.0555],[2469,-5.091],[2484,-18.996],[2501,-37.6757],[2503,-40.2936],[2512,-28.7302],[2515,-28.73],[2519,-12.7955],[2522,-37.4622],[2523,-26.0821],[2524,-23.3933],[2527,-41.2847],[2531,-43.1447],[2759,-7.5307],[2761,15.733],[2764,25.4149],[2779,-4.2066],[2795,-27.4003],[2938,-27.3804],[2986,26.7856],[3058,26.7856],[3070,-19.8093],[3185,-19.8093],[3220,-25.2703],[3242,-9.2262],[3264,-5.0134],[3282,-21.968],[3301,-25.2703],[3320,-11.0154],[3340,-5.0853],[3344,-5.0852],[3363,-22.8171],[3382,-25.2703],[3403,-8.4937],[3425,-5.0134],[3442,-21.97],[3460,-25.2703],[3482,-9.1727],[3504,-5.0134],[3522,-21.97],[3540,-25.2703],[3562,-9.2264],[3584,-5.0134],[3586,-5.0134],[3603,-22.8171],[3621,-25.2703],[3640,-11.0153],[3660,-5.0853],[3664,-5.0852],[3683,-22.817],[3702,-25.2703],[3723,-8.4936],[3744,-5.0134],[3762,-21.9698],[3780,-25.2703],[3802,-9.173],[3824,-5.0134],[3842,-21.9702],[3860,-25.2703],[3882,-9.2266],[3904,-25.2703],[3926,-9.173],[3948,-5.0134],[3966,-21.9702],[3984,-25.2703],[4006,-9.2266],[4028,-25.2703],[4050,-9.173],[4072,-5.0134],[4090,-21.9702],[4108,-25.2703],[4130,-9.2266],[4153,-25.2703],[4175,-9.173],[4197,-5.0134],[4215,-21.9702],[4233,-25.2703],[4255,-9.2266],[4279,-25.2703],[4301,-9.173],[4323,-5.0134],[4341,-21.9702],[4360,-7.2836],[4364,-6.2689],[4373,-5.0134],[4391,-21.9677],[4410,-25.2703],[4423,-17.6815],[4432,-8.4523],[4435,-7.0128],[4436,0]]
       }
       thumb_03_l_ctrl.Scale.X {type=MovieSceneScriptingFloatChannel, default=1, const [1,12]}
       thumb_03_l_ctrl.Scale.Y {type=MovieSceneScriptingFloatChannel, default=1, const [1,12]}
       thumb_03_l_ctrl.Scale.Z {type=MovieSceneScriptingFloatChannel, default=1, const [1,12]}
       index_metacarpal_l_ctrl.Location.X {type=MovieSceneScriptingFloatChannel, default=0, thinned_from=35, keys [[1790,0],[4436,0]]}
       index_metacarpal_l_ctrl.Location.Y {type=MovieSceneScriptingFloatChannel, default=0, thinned_from=28, keys [[1790,0],[4436,0]]}
       index_metacarpal_l_ctrl.Location.Z {type=MovieSceneScriptingFloatChannel, default=0, thinned_from=28, keys [[1790,0],[4436,0]]}
       index_metacarpal_l_ctrl.Rotation.X {
        type=MovieSceneScriptingFloatChannel
        default=-6.3877
        thinned_from=87
        keys [[1790,-6.3877],[1821,-6.9726],[1826,-6.9854],[1866,-6.3796],[1901,-6.9726],[1906,-6.9854],[1946,-6.3796],[1981,-6.9726],[1986,-6.9854],[2025,-6.3788],[2061,-6.9726],[2066,-6.9854],[2105,-6.3788],[2141,-6.9726],[2146,-6.9854],[2186,-6.3796],[2221,-6.9726],[2226,-6.9854],[2266,-6.3796],[2301,-6.9726],[2306,-6.9854],[2345,-6.3788],[2381,-6.9726],[2386,-6.9854],[2425,-6.3788],[2461,-6.9726],[2466,-6.9853],[2503,-0.2591],[2513,-0.0468],[2520,-1.3045],[2522,-0.5891],[2524,-1.0622],[2530,-0.3943],[2757,-0.7261],[2759,-1.595],[2766,-3.0488],[2774,-4.5588],[2777,-4.02],[2781,-1.4428],[2788,0],[3185,0],[3220,-6.3788],[3262,-6.9854],[3301,-6.3796],[3341,-6.9854],[3381,-6.3796],[3416,-6.9726],[3422,-6.9854],[3460,-6.3788],[3502,-6.9854],[3540,-6.3788],[3581,-6.9854],[3621,-6.3796],[3662,-6.9854],[3701,-6.3796],[3741,-6.9854],[3780,-6.3788],[3816,-6.9726],[3822,-6.9854],[3860,-6.3788],[3904,-6.3788],[3940,-6.9726],[3946,-6.9854],[3984,-6.3788],[4028,-6.3788],[4064,-6.9726],[4070,-6.9854],[4108,-6.3788],[4153,-6.3788],[4189,-6.9726],[4195,-6.9854],[4233,-6.3788],[4279,-6.3788],[4315,-6.9726],[4321,-6.9854],[4360,-6.8599],[4364,-6.8348],[4410,-6.3796],[4435,-6.9209],[4436,0]]
       }
       index_metacarpal_l_ctrl.Rotation.Y {
        type=MovieSceneScriptingFloatChannel
        default=4.3708
        thinned_from=173
        keys [[1790,4.3708],[1804,2.6292],[1818,1.8879],[1819,1.8881],[1821,1.8578],[1826,1.8275],[1835,2.0032],[1845,3.5012],[1865,4.3897],[1866,4.3896],[1898,1.8879],[1899,1.8881],[1901,1.8578],[1906,1.8275],[1915,2.011],[1925,3.4908],[1945,4.3897],[1946,4.3896],[1978,1.8879],[1979,1.8881],[1981,1.8578],[1986,1.8275],[1995,2.0124],[2005,3.4875],[2025,4.3907],[2058,1.8879],[2059,1.8881],[2061,1.8578],[2066,1.8275],[2075,2.0124],[2085,3.4875],[2105,4.3907],[2138,1.8879],[2139,1.8881],[2146,1.8275],[2155,2.0032],[2165,3.5012],[2185,4.3897],[2186,4.3896],[2218,1.8879],[2219,1.8881],[2226,1.8275],[2235,2.011],[2245,3.4908],[2265,4.3897],[2266,4.3896],[2298,1.8879],[2299,1.8881],[2306,1.8275],[2315,2.0124],[2325,3.4875],[2345,4.3907],[2378,1.8879],[2379,1.8881],[2386,1.8275],[2395,2.0124],[2405,3.4875],[2425,4.3907],[2458,1.8879],[2459,1.8881],[2503,-1.712],[2506,-1.712],[2507,-0.9522],[2509,1.3699],[2512,2.2128],[2515,2.2128],[2520,-2.8849],[2522,-2.6092],[2524,-2.7985],[2530,-2.5361],[2763,-3.4364],[2780,0.2666],[2788,0],[3185,0],[3220,4.3907],[3253,1.8879],[3254,1.8881],[3262,1.8275],[3271,2.1147],[3281,3.6427],[3300,4.3897],[3301,4.3896],[3333,1.8879],[3334,1.8881],[3341,1.8275],[3350,2.0109],[3360,3.4906],[3380,4.3897],[3381,4.3896],[3413,1.8879],[3414,1.8881],[3422,1.8275],[3431,2.1235],[3441,3.6295],[3460,4.3907],[3493,1.8879],[3494,1.8881],[3502,1.8275],[3511,2.1235],[3521,3.6296],[3540,4.3907],[3573,1.8879],[3574,1.8881],[3581,1.8275],[3590,2.0031],[3600,3.5009],[3620,4.3897],[3621,4.3896],[3653,1.8879],[3654,1.8881],[3662,1.8275],[3671,2.114],[3681,3.6384],[3700,4.3897],[3701,4.3896],[3733,1.8879],[3734,1.8881],[3742,1.8275],[3751,2.1235],[3761,3.6295],[3780,4.3907],[3813,1.8879],[3814,1.8881],[3822,1.8275],[3831,2.1235],[3841,3.6296],[3860,4.3907],[3904,4.3907],[3937,1.8879],[3938,1.8881],[3946,1.8275],[3955,2.1235],[3965,3.6296],[3984,4.3907],[4028,4.3907],[4061,1.8879],[4062,1.8881],[4070,1.8275],[4079,2.1235],[4089,3.6296],[4108,4.3907],[4153,4.3907],[4186,1.8879],[4187,1.8881],[4195,1.8275],[4204,2.1235],[4214,3.6296],[4233,4.3907],[4279,4.3907],[4312,1.8879],[4313,1.8881],[4321,1.8275],[4330,2.1235],[4340,3.6296],[4360,1.9893],[4364,1.9902],[4379,2.003],[4389,3.501],[4409,4.3897],[4410,4.3896],[4423,3.3661],[4429,2.4927],[4432,2.1882],[4434,2.0513],[4435,2.021],[4436,0]]
       }
       index_metacarpal_l_ctrl.Rotation.Z {
        type=MovieSceneScriptingFloatChannel
        default=2.9036
        thinned_from=221
        keys [[1790,2.9036],[1805,1.2751],[1821,0.614],[1824,0.581],[1826,0.5831],[1858,2.9179],[1859,2.917],[1861,2.9352],[1869,2.9367],[1885,1.2751],[1901,0.614],[1904,0.581],[1906,0.5831],[1938,2.9179],[1939,2.917],[1941,2.9352],[1949,2.9367],[1965,1.2751],[1981,0.614],[1984,0.581],[1986,0.5831],[2018,2.9179],[2019,2.917],[2021,2.9352],[2023,2.9333],[2025,2.9526],[2061,0.614],[2062,0.614],[2064,0.581],[2066,0.5831],[2098,2.9179],[2099,2.917],[2101,2.9352],[2103,2.9333],[2105,2.9526],[2144,0.581],[2146,0.5831],[2178,2.9179],[2179,2.917],[2181,2.9352],[2189,2.9367],[2206,1.1733],[2224,0.581],[2226,0.5831],[2258,2.9179],[2259,2.917],[2261,2.9352],[2269,2.9367],[2286,1.1733],[2304,0.581],[2306,0.5831],[2338,2.9179],[2339,2.917],[2341,2.9352],[2343,2.9333],[2345,2.9526],[2384,0.581],[2386,0.5831],[2418,2.9179],[2419,2.917],[2421,2.9352],[2423,2.9333],[2426,2.9526],[2464,0.581],[2471,0.6904],[2503,5.1041],[2505,5.1041],[2510,-1.0063],[2513,-1.5897],[2518,-0.4236],[2521,8.0947],[2522,7.9484],[2523,6.506],[2526,8.3248],[2530,8.8823],[2759,4.4029],[2762,0.6785],[2765,0.2338],[2774,0.379],[2788,0],[3185,0],[3216,2.9352],[3218,2.9333],[3220,2.9526],[3259,0.581],[3261,0.5831],[3293,2.9179],[3294,2.917],[3296,2.9352],[3304,2.9367],[3321,1.1735],[3339,0.581],[3341,0.5831],[3373,2.9179],[3374,2.917],[3376,2.9352],[3384,2.9367],[3401,1.1735],[3419,0.581],[3421,0.5831],[3453,2.9179],[3454,2.917],[3456,2.9352],[3458,2.9333],[3460,2.9526],[3499,0.581],[3501,0.5831],[3533,2.9179],[3534,2.917],[3536,2.9352],[3538,2.9333],[3541,2.9526],[3579,0.581],[3581,0.5831],[3613,2.9179],[3614,2.917],[3616,2.9352],[3624,2.9367],[3641,1.1735],[3659,0.581],[3661,0.5831],[3693,2.9179],[3694,2.917],[3696,2.9352],[3704,2.9367],[3721,1.1735],[3739,0.581],[3741,0.5831],[3773,2.9179],[3774,2.917],[3776,2.9352],[3778,2.9333],[3780,2.9526],[3819,0.581],[3822,0.5831],[3853,2.9179],[3854,2.917],[3856,2.9352],[3858,2.9333],[3860,2.9526],[3898,2.917],[3900,2.9352],[3902,2.9333],[3904,2.9526],[3943,0.581],[3946,0.5831],[3977,2.9179],[3978,2.917],[3980,2.9352],[3982,2.9333],[3984,2.9526],[4022,2.917],[4024,2.9352],[4026,2.9333],[4028,2.9526],[4067,0.581],[4070,0.5831],[4101,2.9179],[4102,2.917],[4104,2.9352],[4106,2.9333],[4108,2.9526],[4147,2.917],[4149,2.9352],[4151,2.9333],[4153,2.9526],[4192,0.581],[4195,0.5831],[4226,2.9179],[4227,2.917],[4229,2.9352],[4231,2.9333],[4233,2.9526],[4273,2.917],[4275,2.9352],[4277,2.9333],[4279,2.9526],[4318,0.581],[4321,0.5831],[4352,2.9179],[4353,2.917],[4355,2.9352],[4357,2.9333],[4360,0.9688],[4364,1.0185],[4402,2.9179],[4403,2.917],[4405,2.9352],[4413,2.9367],[4430,1.1735],[4434,0.8664],[4436,0]]
       }
       index_metacarpal_l_ctrl.Scale.X {type=MovieSceneScriptingFloatChannel, default=1, const [1,12]}
       index_metacarpal_l_ctrl.Scale.Y {type=MovieSceneScriptingFloatChannel, default=1, const [1,12]}
       index_metacarpal_l_ctrl.Scale.Z {type=MovieSceneScriptingFloatChannel, default=1, const [1,12]}
       index_01_l_ctrl.Location.X {
        type=MovieSceneScriptingFloatChannel
        default=0
        thinned_from=25
        keys [[1790,0],[2511,0],[2513,-0.0005],[2522,0],[2768,0],[2796,0.0648],[2927,0.0663],[2938,0.0663],[2986,0.0804],[3058,0.0804],[3070,-0.0543],[3185,-0.0543],[3306,0],[4436,-0.0001]]
       }
       index_01_l_ctrl.Location.Y {
        type=MovieSceneScriptingFloatChannel
        default=0
        thinned_from=469
        keys [[1790,0],[2769,0],[2780,0.0222],[2803,-0.2816],[2927,-0.2872],[2938,-0.2874],[2986,-0.4651],[3058,-0.4651],[3070,0.282],[3185,0.282],[3217,0],[4436,-0.0001]]
       }
       index_01_l_ctrl.Location.Z {
        type=MovieSceneScriptingFloatChannel
        default=0
        thinned_from=712
        keys [[1790,0],[2513,0.0001],[2769,0],[2776,0.0537],[2783,0.2923],[2797,0.4179],[2938,0.4158],[2986,0.146],[3058,0.146],[3070,0.0472],[3185,0.0472],[3216,0],[4436,0]]
       }
       index_01_l_ctrl.Rotation.X {
        type=MovieSceneScriptingFloatChannel
        default=3.5021
        thinned_from=115
        keys [[1790,3.5021],[1825,3.2449],[1830,3.243],[1867,3.5063],[1901,3.2481],[1904,3.2481],[1906,3.2424],[1947,3.5063],[1981,3.2481],[1984,3.2481],[1986,3.2424],[2027,3.5063],[2065,3.2449],[2070,3.243],[2105,3.5064],[2145,3.2449],[2150,3.243],[2187,3.5063],[2221,3.2481],[2224,3.2481],[2226,3.2424],[2268,3.5063],[2301,3.2481],[2304,3.2481],[2306,3.2424],[2347,3.5063],[2385,3.2449],[2390,3.243],[2425,3.5064],[2469,3.2336],[2503,-1.0488],[2506,-1.0488],[2509,3.1436],[2518,4.4625],[2521,2.2246],[2523,2.6037],[2530,1.9893],[2531,1.9902],[2764,4.2422],[2774,3.3382],[2784,-1.8616],[2804,1.7701],[2927,1.7674],[2938,1.7648],[2986,-20.6088],[3058,-20.6088],[3070,-3.6451],[3185,-3.6451],[3220,3.5064],[3260,3.2449],[3265,3.243],[3302,3.5063],[3337,3.2481],[3341,3.2424],[3383,3.5063],[3416,3.2481],[3422,3.2424],[3463,3.5063],[3500,3.2449],[3505,3.243],[3540,3.5064],[3580,3.2449],[3586,3.243],[3623,3.5063],[3657,3.2481],[3662,3.2424],[3703,3.5063],[3736,3.2481],[3742,3.2424],[3782,3.5063],[3820,3.2449],[3825,3.243],[3860,3.5064],[3906,3.5063],[3944,3.2449],[3949,3.243],[3984,3.5064],[4030,3.5063],[4068,3.2449],[4073,3.243],[4108,3.5064],[4155,3.5063],[4193,3.2449],[4198,3.243],[4233,3.5064],[4281,3.5063],[4319,3.2449],[4324,3.243],[4360,3.2933],[4364,3.3042],[4411,3.5063],[4423,3.4001],[4429,3.3165],[4432,3.285],[4435,3.2707],[4436,0.0011]]
       }
       index_01_l_ctrl.Rotation.Y {
        type=MovieSceneScriptingFloatChannel
        default=2.4789
        thinned_from=274
        keys [[1790,2.4789],[1823,2.9409],[1826,2.9439],[1854,2.4825],[1855,2.4835],[1857,2.4751],[1860,2.4771],[1862,2.4678],[1869,2.4714],[1898,2.931],[1904,2.9449],[1906,2.9431],[1938,2.4751],[1949,2.4714],[1978,2.931],[1984,2.9449],[1986,2.9431],[2014,2.4825],[2015,2.4835],[2017,2.4751],[2020,2.4771],[2022,2.4678],[2029,2.4714],[2063,2.9409],[2066,2.9439],[2098,2.4751],[2105,2.4679],[2143,2.9409],[2146,2.9439],[2174,2.4825],[2175,2.4835],[2177,2.4751],[2180,2.4771],[2182,2.4678],[2189,2.4714],[2218,2.931],[2224,2.9449],[2226,2.9431],[2258,2.4751],[2269,2.4714],[2298,2.931],[2304,2.9449],[2306,2.9431],[2334,2.4825],[2335,2.4835],[2337,2.4751],[2340,2.4771],[2342,2.4678],[2349,2.4714],[2383,2.9409],[2386,2.9439],[2418,2.4751],[2425,2.4679],[2463,2.9409],[2466,2.9438],[2471,1.7672],[2501,-4.0229],[2503,-3.438],[2506,-3.438],[2509,-1.0139],[2513,-0.3177],[2518,-3.3934],[2519,-9.5101],[2520,-10.941],[2521,-7.8094],[2522,-7.2448],[2524,-9.7295],[2528,-6.3779],[2532,-6.2482],[2758,-10.0282],[2763,-18.193],[2772,-9.655],[2782,2.6588],[2803,-0.5122],[2927,-0.6285],[2938,-0.633],[2986,-9.5071],[3058,-9.5071],[3070,-2.505],[3185,-2.505],[3216,2.4712],[3218,2.4722],[3220,2.4679],[3258,2.9409],[3261,2.9439],[3289,2.4825],[3290,2.4835],[3292,2.4751],[3295,2.4771],[3297,2.4678],[3304,2.4714],[3333,2.931],[3339,2.9449],[3341,2.9431],[3373,2.4751],[3384,2.4714],[3413,2.931],[3419,2.9449],[3421,2.9431],[3449,2.4825],[3450,2.4835],[3452,2.4751],[3455,2.4771],[3457,2.4678],[3464,2.4714],[3498,2.9409],[3501,2.9439],[3533,2.4751],[3541,2.4679],[3578,2.9409],[3581,2.9439],[3609,2.4825],[3610,2.4835],[3612,2.4751],[3615,2.4771],[3617,2.4678],[3624,2.4714],[3653,2.931],[3659,2.9449],[3661,2.9431],[3693,2.4751],[3704,2.4714],[3733,2.931],[3739,2.9449],[3742,2.9431],[3769,2.4825],[3770,2.4835],[3772,2.4751],[3775,2.4771],[3777,2.4678],[3784,2.4714],[3818,2.9409],[3821,2.9439],[3853,2.4751],[3860,2.4679],[3894,2.4835],[3896,2.4751],[3899,2.4771],[3901,2.4678],[3908,2.4714],[3942,2.9409],[3945,2.9439],[3977,2.4751],[3984,2.4679],[4018,2.4835],[4020,2.4751],[4023,2.4771],[4025,2.4678],[4032,2.4714],[4066,2.9409],[4069,2.9439],[4101,2.4751],[4108,2.4679],[4143,2.4835],[4145,2.4751],[4148,2.4771],[4150,2.4678],[4157,2.4714],[4191,2.9409],[4194,2.9439],[4226,2.4751],[4233,2.4679],[4269,2.4835],[4271,2.4751],[4274,2.4771],[4276,2.4678],[4283,2.4714],[4317,2.9409],[4320,2.9439],[4352,2.4751],[4357,2.4722],[4360,2.8739],[4364,2.8618],[4398,2.4825],[4399,2.4835],[4401,2.4751],[4404,2.4771],[4406,2.4678],[4413,2.4714],[4435,2.8958],[4436,0.0012]]
       }
       index_01_l_ctrl.Rotation.Z {
        type=MovieSceneScriptingFloatChannel
        default=6.7702
        thinned_from=187
        keys [[1790,6.7702],[1824,10.6086],[1830,10.6085],[1857,6.8011],[1858,6.8012],[1862,6.7447],[1867,6.745],[1898,10.5135],[1899,10.5133],[1901,10.5739],[1906,10.6345],[1943,6.7449],[1947,6.745],[1978,10.5135],[1979,10.5133],[1981,10.5739],[1987,10.6345],[2017,6.8011],[2018,6.8012],[2022,6.7447],[2027,6.745],[2064,10.6086],[2070,10.6085],[2103,6.7449],[2105,6.7449],[2144,10.6086],[2150,10.6085],[2177,6.8011],[2178,6.8012],[2182,6.7447],[2187,6.745],[2218,10.5135],[2219,10.5133],[2227,10.6345],[2263,6.7449],[2267,6.745],[2298,10.5135],[2299,10.5133],[2307,10.6345],[2337,6.8011],[2338,6.8012],[2342,6.7447],[2384,10.6086],[2390,10.6085],[2423,6.7449],[2425,6.7449],[2464,10.6086],[2468,10.5411],[2503,15.6919],[2505,15.6919],[2510,-31.896],[2513,-36.2843],[2518,-21.5861],[2519,7.8327],[2520,21.2768],[2522,18.0012],[2524,20.1925],[2530,17.1219],[2531,17.1221],[2763,27.6806],[2794,-25.78],[2938,-25.5005],[2986,53.4956],[3058,53.4956],[3070,8.7195],[3185,8.7195],[3218,6.7449],[3220,6.7449],[3259,10.6086],[3265,10.6085],[3292,6.8011],[3293,6.8012],[3297,6.7447],[3333,10.5135],[3334,10.5133],[3342,10.6345],[3378,6.7449],[3382,6.745],[3413,10.5135],[3414,10.5133],[3422,10.6345],[3452,6.8011],[3453,6.8012],[3457,6.7447],[3462,6.745],[3499,10.6086],[3505,10.6085],[3538,6.7449],[3540,6.7449],[3579,10.6086],[3585,10.6085],[3612,6.8011],[3613,6.8012],[3617,6.7447],[3622,6.745],[3653,10.5135],[3654,10.5133],[3662,10.6345],[3698,6.7449],[3702,6.745],[3733,10.5135],[3734,10.5133],[3742,10.6345],[3772,6.8011],[3773,6.8012],[3777,6.7447],[3819,10.6086],[3825,10.6085],[3858,6.7449],[3897,6.8012],[3901,6.7447],[3943,10.6086],[3949,10.6085],[3982,6.7449],[4021,6.8012],[4025,6.7447],[4067,10.6086],[4073,10.6085],[4106,6.7449],[4146,6.8012],[4150,6.7447],[4192,10.6086],[4198,10.6085],[4231,6.7449],[4272,6.8012],[4276,6.7447],[4318,10.6086],[4324,10.6085],[4357,6.7449],[4360,9.9768],[4364,9.892],[4401,6.8011],[4402,6.8012],[4406,6.7447],[4435,10.238],[4436,0.0025]]
       }
       index_01_l_ctrl.Scale.X {type=MovieSceneScriptingFloatChannel, default=1, const [1,12]}
       index_01_l_ctrl.Scale.Y {type=MovieSceneScriptingFloatChannel, default=1, const [1,12]}
       index_01_l_ctrl.Scale.Z {type=MovieSceneScriptingFloatChannel, default=1, const [1,12]}
       index_02_l_ctrl.Location.X {type=MovieSceneScriptingFloatChannel, default=0, thinned_from=458, keys [[1790,0],[2512,0],[2513,-0.0002],[2529,0],[4436,0]]}
       index_02_l_ctrl.Location.Y {type=MovieSceneScriptingFloatChannel, default=0, thinned_from=687, keys [[1790,0],[4436,0]]}
       index_02_l_ctrl.Location.Z {type=MovieSceneScriptingFloatChannel, default=0, thinned_from=197, keys [[1790,0],[4436,0]]}
       index_02_l_ctrl.Rotation.X {
        type=MovieSceneScriptingFloatChannel
        default=5.2264
        thinned_from=162
        keys [[1790,5.2264],[1827,5.7154],[1865,5.2222],[1869,5.2211],[1894,5.6768],[1907,5.7154],[1945,5.2166],[1987,5.7154],[2024,5.2222],[2029,5.2211],[2054,5.6768],[2065,5.7159],[2066,5.7128],[2104,5.2222],[2109,5.2211],[2148,5.7154],[2185,5.2222],[2189,5.2211],[2214,5.6768],[2227,5.7154],[2265,5.2166],[2307,5.7154],[2344,5.2222],[2349,5.2211],[2374,5.6768],[2385,5.7159],[2386,5.7128],[2424,5.2222],[2429,5.2211],[2466,5.712],[2471,5.3583],[2503,-0.2135],[2513,-0.3575],[2519,0.4074],[2522,-0.1958],[2524,0.1872],[2531,-0.3625],[2532,-0.3625],[2760,0.9957],[2763,1.5212],[2771,2.9723],[2780,5.014],[2803,1.6744],[2927,1.6306],[2938,1.6288],[2986,0.3205],[3058,0.3205],[3070,1.474],[3185,1.474],[3219,5.2222],[3224,5.2211],[3263,5.7154],[3300,5.2222],[3304,5.2211],[3329,5.6768],[3343,5.7154],[3380,5.2166],[3422,5.7154],[3459,5.2222],[3464,5.2211],[3489,5.6768],[3500,5.7159],[3501,5.7128],[3539,5.2222],[3544,5.2211],[3583,5.7154],[3620,5.2222],[3624,5.2211],[3649,5.6768],[3662,5.7154],[3700,5.2166],[3743,5.7154],[3779,5.2222],[3784,5.2211],[3809,5.6768],[3820,5.7159],[3822,5.7128],[3859,5.2222],[3908,5.2211],[3933,5.6768],[3944,5.7159],[3946,5.7128],[3983,5.2222],[4032,5.2211],[4057,5.6768],[4068,5.7159],[4070,5.7128],[4107,5.2222],[4157,5.2211],[4182,5.6768],[4193,5.7159],[4195,5.7128],[4232,5.2222],[4283,5.2211],[4308,5.6768],[4319,5.7159],[4360,5.5815],[4364,5.6217],[4371,5.7154],[4410,5.2222],[4413,5.2211],[4435,5.6625],[4436,0]]
       }
       index_02_l_ctrl.Rotation.Y {
        type=MovieSceneScriptingFloatChannel
        default=-6.1831
        thinned_from=118
        keys [[1790,-6.1831],[1825,-6.7595],[1827,-6.7586],[1865,-6.1752],[1905,-6.7595],[1907,-6.7586],[1945,-6.1764],[1949,-6.1752],[1985,-6.7595],[1987,-6.7586],[2025,-6.1752],[2058,-6.7446],[2066,-6.7636],[2105,-6.1752],[2145,-6.7595],[2151,-6.7586],[2185,-6.1752],[2225,-6.7595],[2227,-6.7586],[2265,-6.1764],[2269,-6.1752],[2305,-6.7595],[2307,-6.7586],[2345,-6.1752],[2378,-6.7446],[2386,-6.7636],[2425,-6.1752],[2465,-6.7595],[2470,-6.5267],[2503,0.0527],[2509,0.0559],[2510,0.0555],[2513,0.0574],[2518,-0.2865],[2519,-1.0111],[2522,-0.0923],[2524,-0.7052],[2526,-0.0838],[2530,0.1489],[2532,0.1486],[2757,-0.2657],[2760,-1.9561],[2767,-3.3153],[2780,-5.9936],[2785,-5.1385],[2802,-2.109],[2927,-2.039],[2938,-2.0367],[2986,-0.9249],[3058,-0.9249],[3070,-1.7692],[3185,-1.7692],[3220,-6.1752],[3260,-6.7595],[3266,-6.7586],[3300,-6.1752],[3340,-6.7595],[3342,-6.7586],[3380,-6.1764],[3384,-6.1752],[3420,-6.7595],[3422,-6.7586],[3460,-6.1752],[3497,-6.7541],[3501,-6.7636],[3540,-6.1752],[3580,-6.7595],[3582,-6.7586],[3620,-6.1752],[3660,-6.7595],[3662,-6.7586],[3700,-6.1764],[3704,-6.1752],[3740,-6.7595],[3742,-6.7586],[3780,-6.1752],[3816,-6.7541],[3822,-6.7636],[3860,-6.1752],[3904,-6.1752],[3940,-6.7541],[3946,-6.7636],[3984,-6.1752],[4028,-6.1752],[4064,-6.7541],[4070,-6.7636],[4108,-6.1752],[4153,-6.1752],[4189,-6.7541],[4195,-6.7636],[4233,-6.1752],[4279,-6.1752],[4315,-6.7541],[4321,-6.7636],[4360,-6.6512],[4364,-6.6835],[4371,-6.7586],[4409,-6.1752],[4435,-6.7022],[4436,0]]
       }
       index_02_l_ctrl.Rotation.Z {
        type=MovieSceneScriptingFloatChannel
        default=15.0095
        thinned_from=181
        keys [[1790,15.0095],[1824,22.0717],[1827,22.0713],[1864,14.9168],[1867,14.9169],[1904,22.0717],[1907,22.0713],[1944,14.917],[1949,14.9169],[1984,22.0717],[1987,22.0713],[2019,14.9655],[2020,14.9658],[2022,14.9165],[2027,14.9169],[2061,22.0077],[2064,22.0073],[2067,22.1196],[2099,14.9655],[2100,14.9658],[2102,14.9165],[2104,14.9169],[2144,22.0717],[2147,22.0713],[2184,14.9168],[2187,14.9169],[2224,22.0717],[2227,22.0713],[2264,14.917],[2269,14.9169],[2304,22.0717],[2307,22.0713],[2339,14.9655],[2340,14.9658],[2342,14.9165],[2348,14.9169],[2382,22.0077],[2384,22.0073],[2387,22.1196],[2419,14.9655],[2420,14.9658],[2422,14.9165],[2424,14.9169],[2464,22.0717],[2469,21.7542],[2503,-9.4557],[2506,-9.4557],[2509,-20.1683],[2513,-23.08],[2518,-13.9245],[2519,4.3526],[2520,7.0864],[2521,-5.6799],[2522,-7.9814],[2524,2.1336],[2527,-10.7146],[2531,-12.0568],[2764,37.1052],[2794,-16.1878],[2938,-15.9506],[2986,56.7606],[3058,56.7606],[3070,9.5501],[3185,9.5501],[3215,14.9658],[3217,14.9165],[3219,14.9169],[3259,22.0717],[3262,22.0713],[3299,14.9168],[3302,14.9169],[3339,22.0717],[3342,22.0713],[3379,14.917],[3384,14.9169],[3419,22.0717],[3422,22.0713],[3454,14.9655],[3455,14.9658],[3457,14.9165],[3462,14.9169],[3497,22.0077],[3499,22.0073],[3502,22.1196],[3534,14.9655],[3535,14.9658],[3537,14.9165],[3541,14.9169],[3579,22.0717],[3582,22.0713],[3619,14.9168],[3622,14.9169],[3659,22.0717],[3662,22.0713],[3699,14.917],[3704,14.9169],[3739,22.0717],[3742,22.0713],[3774,14.9655],[3775,14.9658],[3777,14.9165],[3782,14.9169],[3818,22.0077],[3819,22.0073],[3822,22.1196],[3854,14.9655],[3855,14.9658],[3857,14.9165],[3899,14.9658],[3901,14.9165],[3906,14.9169],[3942,22.0077],[3943,22.0073],[3946,22.1196],[3978,14.9655],[3979,14.9658],[3981,14.9165],[4023,14.9658],[4025,14.9165],[4030,14.9169],[4066,22.0077],[4067,22.0073],[4070,22.1196],[4102,14.9655],[4103,14.9658],[4105,14.9165],[4148,14.9658],[4150,14.9165],[4155,14.9169],[4191,22.0077],[4192,22.0073],[4195,22.1196],[4227,14.9655],[4228,14.9658],[4230,14.9165],[4274,14.9658],[4276,14.9165],[4281,14.9169],[4317,22.0077],[4318,22.0073],[4321,22.1196],[4353,14.9655],[4354,14.9658],[4356,14.9165],[4360,20.9542],[4364,21.5367],[4371,22.0713],[4408,14.9168],[4409,14.9169],[4435,21.3817],[4436,0]]
       }
       index_02_l_ctrl.Scale.X {type=MovieSceneScriptingFloatChannel, default=1, const [1,12]}
       index_02_l_ctrl.Scale.Y {type=MovieSceneScriptingFloatChannel, default=1, const [1,12]}
       index_02_l_ctrl.Scale.Z {type=MovieSceneScriptingFloatChannel, default=1, const [1,12]}
       index_03_l_ctrl.Location.X {type=MovieSceneScriptingFloatChannel, default=0, thinned_from=609, keys [[1790,0],[2512,0],[2513,-0.0001],[2528,0],[4435,0],[4436,-0.0003]]}
       index_03_l_ctrl.Location.Y {type=MovieSceneScriptingFloatChannel, default=0, thinned_from=893, keys [[1790,0],[4436,0]]}
       index_03_l_ctrl.Location.Z {type=MovieSceneScriptingFloatChannel, default=0, thinned_from=498, keys [[1790,0],[4436,0]]}
       index_03_l_ctrl.Rotation.X {
        type=MovieSceneScriptingFloatChannel
        default=0.3373
        thinned_from=401
        keys [[1790,0.3373],[1808,0.3714],[1810,0.3711],[1814,0.374],[1815,0.3734],[1818,0.3754],[1819,0.3754],[1824,0.3784],[1826,0.3766],[1832,0.3766],[1850,0.339],[1851,0.3395],[1855,0.3383],[1856,0.3391],[1858,0.3369],[1860,0.3384],[1864,0.3367],[1867,0.3372],[1869,0.3369],[1907,0.3774],[1912,0.3777],[1930,0.3391],[1931,0.341],[1933,0.3377],[1935,0.3377],[1937,0.3414],[1942,0.3362],[1946,0.3362],[1952,0.3407],[1953,0.3404],[1968,0.3714],[1970,0.3711],[1974,0.374],[1975,0.3734],[1988,0.3774],[1992,0.3777],[2015,0.3383],[2016,0.3385],[2027,0.3362],[2030,0.3385],[2031,0.3384],[2048,0.3714],[2050,0.3711],[2054,0.374],[2055,0.3734],[2072,0.3777],[2090,0.339],[2091,0.3395],[2095,0.3383],[2096,0.3391],[2098,0.3369],[2100,0.3384],[2102,0.3384],[2104,0.3367],[2105,0.3372],[2109,0.3369],[2128,0.3714],[2130,0.3711],[2134,0.374],[2135,0.3734],[2141,0.376],[2144,0.3784],[2146,0.3766],[2152,0.3766],[2170,0.339],[2171,0.3395],[2175,0.3383],[2176,0.3391],[2178,0.3369],[2180,0.3384],[2184,0.3367],[2187,0.3372],[2189,0.3369],[2228,0.3774],[2232,0.3777],[2250,0.3391],[2251,0.341],[2253,0.3377],[2257,0.3414],[2261,0.3362],[2266,0.3362],[2272,0.3407],[2273,0.3404],[2288,0.3714],[2290,0.3711],[2294,0.374],[2295,0.3734],[2308,0.3774],[2312,0.3777],[2335,0.3383],[2336,0.3385],[2347,0.3362],[2350,0.3385],[2351,0.3384],[2368,0.3714],[2370,0.3711],[2374,0.374],[2375,0.3734],[2392,0.3777],[2410,0.339],[2411,0.3395],[2415,0.3383],[2416,0.3391],[2418,0.3369],[2420,0.3384],[2422,0.3384],[2424,0.3367],[2426,0.3372],[2429,0.3369],[2448,0.3714],[2450,0.3711],[2454,0.374],[2455,0.3734],[2458,0.3754],[2462,0.376],[2464,0.3784],[2471,0.3552],[2503,-0.4901],[2506,-0.4901],[2509,-0.6215],[2513,-0.6579],[2518,-0.5437],[2519,-0.3135],[2520,-0.2857],[2521,-0.4669],[2522,-0.5016],[2524,-0.3579],[2527,-0.5417],[2530,-0.5591],[2757,-0.4621],[2761,0.047],[2778,0.3282],[2927,0.2427],[2938,0.2405],[2986,0.359],[3058,0.359],[3070,-0.2992],[3185,-0.2992],[3215,0.3384],[3217,0.3384],[3219,0.3367],[3220,0.3372],[3224,0.3369],[3243,0.3714],[3245,0.3711],[3249,0.374],[3250,0.3734],[3259,0.3784],[3262,0.3766],[3266,0.3766],[3285,0.339],[3286,0.3395],[3290,0.3383],[3291,0.3391],[3293,0.3369],[3295,0.3384],[3297,0.3384],[3299,0.3367],[3300,0.3372],[3304,0.3369],[3343,0.3774],[3347,0.3777],[3365,0.3391],[3366,0.3409],[3369,0.3377],[3372,0.3414],[3375,0.3362],[3381,0.3362],[3387,0.3407],[3388,0.3404],[3403,0.3714],[3405,0.3711],[3409,0.374],[3410,0.3734],[3423,0.3774],[3427,0.3777],[3450,0.3383],[3451,0.3385],[3462,0.3362],[3465,0.3385],[3466,0.3384],[3483,0.3714],[3485,0.3711],[3489,0.374],[3490,0.3734],[3507,0.3777],[3525,0.339],[3526,0.3395],[3530,0.3383],[3531,0.3391],[3533,0.3369],[3535,0.3384],[3539,0.3367],[3541,0.3372],[3544,0.3369],[3563,0.3714],[3565,0.3711],[3569,0.374],[3570,0.3734],[3573,0.3754],[3574,0.3754],[3579,0.3784],[3581,0.3766],[3587,0.3766],[3605,0.339],[3606,0.3395],[3610,0.3383],[3611,0.3391],[3613,0.3369],[3615,0.3384],[3619,0.3367],[3622,0.3372],[3624,0.3369],[3663,0.3774],[3667,0.3777],[3685,0.3391],[3686,0.3409],[3689,0.3377],[3692,0.3414],[3695,0.3362],[3704,0.3362],[3707,0.3407],[3708,0.3404],[3723,0.3714],[3725,0.3711],[3729,0.374],[3730,0.3734],[3743,0.3774],[3747,0.3777],[3770,0.3383],[3771,0.3385],[3782,0.3362],[3785,0.3385],[3786,0.3384],[3803,0.3714],[3805,0.3711],[3809,0.374],[3810,0.3734],[3827,0.3777],[3845,0.339],[3846,0.3395],[3850,0.3383],[3851,0.3391],[3853,0.3369],[3855,0.3384],[3859,0.3367],[3860,0.3372],[3864,0.3369],[3883,0.3714],[3885,0.3711],[3894,0.3383],[3895,0.3385],[3906,0.3362],[3909,0.3385],[3910,0.3384],[3927,0.3714],[3929,0.3711],[3933,0.374],[3934,0.3734],[3951,0.3777],[3969,0.339],[3970,0.3395],[3974,0.3383],[3975,0.3391],[3977,0.3369],[3979,0.3384],[3983,0.3367],[3984,0.3372],[3988,0.3369],[4007,0.3714],[4009,0.3711],[4018,0.3383],[4019,0.3385],[4030,0.3362],[4033,0.3385],[4034,0.3384],[4051,0.3714],[4053,0.3711],[4057,0.374],[4058,0.3734],[4075,0.3777],[4093,0.339],[4094,0.3395],[4098,0.3383],[4099,0.3391],[4101,0.3369],[4103,0.3384],[4107,0.3367],[4108,0.3372],[4112,0.3369],[4131,0.3714],[4133,0.3711],[4143,0.3383],[4144,0.3385],[4155,0.3362],[4158,0.3385],[4159,0.3384],[4176,0.3714],[4178,0.3711],[4182,0.374],[4183,0.3734],[4200,0.3777],[4218,0.339],[4219,0.3395],[4223,0.3383],[4224,0.3391],[4226,0.3369],[4228,0.3384],[4232,0.3367],[4233,0.3372],[4237,0.3369],[4256,0.3714],[4258,0.3711],[4269,0.3383],[4270,0.3385],[4281,0.3362],[4284,0.3385],[4285,0.3384],[4302,0.3714],[4304,0.3711],[4308,0.374],[4309,0.3734],[4326,0.3777],[4344,0.339],[4345,0.3395],[4349,0.3383],[4350,0.3391],[4352,0.3369],[4354,0.3384],[4360,0.374],[4365,0.3734],[4376,0.3766],[4394,0.339],[4395,0.3395],[4399,0.3383],[4400,0.3391],[4402,0.3369],[4404,0.3384],[4408,0.3367],[4411,0.3372],[4413,0.3369],[4435,0.3727],[4436,0.0028]]
       }
       index_03_l_ctrl.Rotation.Y {
        type=MovieSceneScriptingFloatChannel
        default=-0.6976
        thinned_from=252
        keys [[1790,-0.6976],[1821,-0.7438],[1830,-0.7446],[1857,-0.6976],[1867,-0.697],[1905,-0.7443],[1907,-0.7442],[1932,-0.6983],[1935,-0.6987],[1938,-0.6967],[1947,-0.6971],[1985,-0.7443],[1987,-0.7442],[2023,-0.697],[2029,-0.697],[2065,-0.7443],[2070,-0.7442],[2097,-0.6976],[2105,-0.697],[2138,-0.7431],[2152,-0.7446],[2177,-0.6976],[2187,-0.697],[2225,-0.7443],[2227,-0.7442],[2252,-0.6983],[2255,-0.6987],[2258,-0.6967],[2267,-0.6971],[2305,-0.7443],[2307,-0.7442],[2343,-0.697],[2349,-0.697],[2385,-0.7443],[2390,-0.7442],[2417,-0.6976],[2426,-0.697],[2458,-0.7431],[2466,-0.7446],[2502,0.0936],[2503,0.0936],[2513,0.3908],[2519,0.3756],[2522,0.5665],[2524,0.4719],[2530,0.6048],[2761,0.2086],[2764,0.1306],[2782,-0.6521],[2803,-0.3203],[2927,-0.3057],[2938,-0.3052],[2986,0.1328],[3058,0.1328],[3070,0.0221],[3185,0.0221],[3217,-0.6968],[3220,-0.697],[3253,-0.7431],[3267,-0.7446],[3292,-0.6976],[3300,-0.697],[3340,-0.7443],[3342,-0.7442],[3367,-0.6983],[3369,-0.6987],[3373,-0.6967],[3382,-0.6971],[3420,-0.7443],[3422,-0.7442],[3458,-0.697],[3464,-0.697],[3500,-0.7443],[3506,-0.7442],[3532,-0.6976],[3541,-0.697],[3573,-0.7431],[3585,-0.7446],[3612,-0.6976],[3622,-0.697],[3660,-0.7443],[3662,-0.7442],[3687,-0.6983],[3689,-0.6987],[3693,-0.6967],[3702,-0.6971],[3740,-0.7443],[3742,-0.7442],[3778,-0.697],[3784,-0.697],[3820,-0.7443],[3826,-0.7442],[3852,-0.6976],[3908,-0.697],[3944,-0.7443],[3950,-0.7442],[3976,-0.6976],[4032,-0.697],[4068,-0.7443],[4074,-0.7442],[4100,-0.6976],[4157,-0.697],[4193,-0.7443],[4199,-0.7442],[4225,-0.6976],[4283,-0.697],[4319,-0.7443],[4325,-0.7442],[4351,-0.6976],[4356,-0.6968],[4360,-0.7412],[4371,-0.7446],[4401,-0.6976],[4411,-0.697],[4435,-0.7396],[4436,0.0015]]
       }
       index_03_l_ctrl.Rotation.Z {
        type=MovieSceneScriptingFloatChannel
        default=12.1825
        thinned_from=162
        keys [[1790,12.1825],[1818,20.4439],[1821,20.5784],[1827,20.7129],[1856,12.1976],[1860,12.0716],[1863,12.0716],[1905,20.6553],[1907,20.6553],[1933,12.3655],[1939,12.0716],[1949,12.0716],[1984,20.6553],[1987,20.6553],[2022,12.0716],[2027,12.0716],[2064,20.6553],[2067,20.6553],[2097,12.1976],[2100,12.0716],[2103,12.0716],[2141,20.5784],[2147,20.7129],[2177,12.1976],[2180,12.0716],[2183,12.0716],[2225,20.6553],[2227,20.6553],[2253,12.3655],[2254,12.3655],[2259,12.0716],[2262,12.0716],[2305,20.6553],[2307,20.6553],[2342,12.0716],[2348,12.0716],[2385,20.6553],[2387,20.6553],[2420,12.0716],[2423,12.0716],[2461,20.5784],[2466,20.7124],[2503,-39.4499],[2513,-21.33],[2515,-21.6186],[2520,5.6953],[2522,-1.7242],[2524,3.2119],[2531,-3.7356],[2757,-0.3484],[2763,20.3234],[2769,20.5161],[2772,20.4907],[2798,-36.4053],[2938,-36.3804],[2986,61.4204],[3058,61.4204],[3070,-9.7746],[3185,-9.7746],[3216,12.0716],[3220,12.0716],[3257,20.5784],[3262,20.7129],[3266,20.7129],[3292,12.1976],[3295,12.0716],[3298,12.0716],[3339,20.6553],[3342,20.6553],[3368,12.3655],[3374,12.0716],[3377,12.0716],[3420,20.6553],[3422,20.6553],[3458,12.0716],[3462,12.0716],[3500,20.6553],[3502,20.6553],[3535,12.0716],[3541,12.0716],[3577,20.5784],[3582,20.7129],[3612,12.1976],[3616,12.0716],[3618,12.0716],[3660,20.6553],[3662,20.6553],[3688,12.3655],[3694,12.0716],[3704,12.0716],[3739,20.6553],[3742,20.6553],[3778,12.0716],[3782,12.0716],[3821,20.6553],[3855,12.0716],[3906,12.0716],[3945,20.6553],[3979,12.0716],[4030,12.0716],[4069,20.6553],[4103,12.0716],[4155,12.0716],[4194,20.6553],[4228,12.0716],[4281,12.0716],[4320,20.6553],[4354,12.0716],[4356,12.0716],[4360,18.7418],[4364,19.7696],[4371,20.7129],[4404,12.0716],[4407,12.0716],[4435,19.8046],[4436,0.0106]]
       }
       index_03_l_ctrl.Scale.X {type=MovieSceneScriptingFloatChannel, default=1, const [1,12]}
       index_03_l_ctrl.Scale.Y {type=MovieSceneScriptingFloatChannel, default=1, const [1,12]}
       index_03_l_ctrl.Scale.Z {type=MovieSceneScriptingFloatChannel, default=1, const [1,12]}
       middle_metacarpal_l_ctrl.Location.X {type=MovieSceneScriptingFloatChannel, default=0, thinned_from=31, keys [[1790,0],[2512,0],[2513,-0.0001],[2521,0],[4436,0]]}
       middle_metacarpal_l_ctrl.Location.Y {type=MovieSceneScriptingFloatChannel, default=0, thinned_from=28, keys [[1790,0],[4436,0]]}
       middle_metacarpal_l_ctrl.Location.Z {type=MovieSceneScriptingFloatChannel, default=0, thinned_from=28, keys [[1790,0],[4436,0]]}
       middle_metacarpal_l_ctrl.Rotation.X {
        type=MovieSceneScriptingFloatChannel
        default=-1.044
        thinned_from=147
        keys [[1790,-1.044],[1809,-0.3637],[1829,-0.2583],[1838,-0.4572],[1848,-0.9661],[1868,-1.0559],[1888,-0.3852],[1909,-0.2583],[1918,-0.4559],[1928,-0.967],[1947,-1.0566],[1957,-0.8212],[1968,-0.3852],[1990,-0.2583],[1999,-0.511],[2009,-0.9911],[2028,-1.0559],[2049,-0.3633],[2070,-0.2583],[2079,-0.5104],[2089,-0.9906],[2108,-1.0559],[2129,-0.3637],[2150,-0.2583],[2159,-0.5108],[2169,-0.9906],[2188,-1.0559],[2209,-0.3639],[2230,-0.2583],[2239,-0.5104],[2248,-0.967],[2267,-1.0566],[2277,-0.8212],[2288,-0.3852],[2310,-0.2583],[2319,-0.511],[2329,-0.9911],[2348,-1.0559],[2369,-0.3633],[2390,-0.2583],[2399,-0.5104],[2409,-0.9906],[2428,-1.0559],[2447,-0.4125],[2466,-0.2617],[2469,-0.2748],[2471,-0.2551],[2503,0],[2507,-0.0001],[2509,0.0001],[2511,-0.0001],[2513,0.0001],[2520,-0.0393],[2522,-0.0081],[2524,-0.0292],[2529,0],[2530,0],[2774,-0.1687],[2787,0],[3185,0],[3223,-1.0559],[3244,-0.3637],[3265,-0.2583],[3274,-0.5107],[3284,-0.9906],[3303,-1.0559],[3324,-0.364],[3345,-0.2583],[3354,-0.5103],[3364,-0.9906],[3383,-1.0566],[3404,-0.364],[3425,-0.2583],[3434,-0.5109],[3444,-0.9911],[3463,-1.0559],[3484,-0.3634],[3505,-0.2583],[3514,-0.5103],[3524,-0.9906],[3543,-1.0559],[3564,-0.3637],[3585,-0.2583],[3594,-0.5107],[3604,-0.9906],[3623,-1.0559],[3644,-0.364],[3665,-0.2583],[3674,-0.5103],[3684,-0.9906],[3703,-1.0566],[3724,-0.364],[3745,-0.2583],[3754,-0.5109],[3764,-0.9911],[3783,-1.0559],[3804,-0.3634],[3825,-0.2583],[3834,-0.5103],[3844,-0.9906],[3863,-1.0559],[3883,-0.3852],[3907,-1.0559],[3928,-0.3634],[3949,-0.2583],[3958,-0.5103],[3968,-0.9906],[3987,-1.0559],[4007,-0.3852],[4031,-1.0559],[4052,-0.3634],[4073,-0.2583],[4082,-0.5103],[4092,-0.9906],[4111,-1.0559],[4131,-0.3852],[4156,-1.0559],[4177,-0.3634],[4198,-0.2583],[4207,-0.5103],[4217,-0.9906],[4236,-1.0559],[4256,-0.3852],[4282,-1.0559],[4303,-0.3634],[4324,-0.2583],[4333,-0.5103],[4343,-0.9906],[4360,-0.3323],[4364,-0.2992],[4373,-0.2583],[4382,-0.4571],[4392,-0.9661],[4412,-1.0559],[4424,-0.6902],[4430,-0.4449],[4433,-0.364],[4434,-0.3498],[4435,-0.3423],[4436,0]]
       }
       middle_metacarpal_l_ctrl.Rotation.Y {
        type=MovieSceneScriptingFloatChannel
        default=1.7776
        thinned_from=179
        keys [[1790,1.7776],[1824,1.8303],[1829,1.8306],[1864,1.7771],[1868,1.777],[1904,1.8303],[1909,1.8306],[1944,1.7772],[1947,1.777],[1984,1.8303],[1989,1.8306],[2028,1.777],[2064,1.8303],[2069,1.8306],[2104,1.7771],[2108,1.777],[2144,1.8303],[2149,1.8306],[2184,1.7771],[2188,1.777],[2224,1.8303],[2229,1.8306],[2264,1.7772],[2267,1.777],[2304,1.8303],[2309,1.8306],[2348,1.777],[2384,1.8303],[2389,1.8306],[2424,1.7771],[2428,1.777],[2464,1.8303],[2466,1.8304],[2471,1.6667],[2502,-0.3365],[2506,-0.3365],[2507,-0.1341],[2509,0.4822],[2512,0.7072],[2515,0.7072],[2518,0.4422],[2519,-0.1194],[2521,-0.4852],[2523,-0.446],[2530,-0.4908],[2760,-0.3888],[2768,0.0366],[2776,0.8095],[2781,0.3414],[2787,0],[3185,0],[3218,1.777],[3223,1.777],[3259,1.8303],[3264,1.8306],[3299,1.7771],[3303,1.777],[3339,1.8303],[3344,1.8306],[3379,1.7772],[3382,1.777],[3419,1.8303],[3424,1.8306],[3463,1.777],[3499,1.8303],[3504,1.8306],[3539,1.7771],[3543,1.777],[3579,1.8303],[3584,1.8306],[3619,1.7771],[3623,1.777],[3659,1.8303],[3664,1.8306],[3699,1.7772],[3702,1.777],[3739,1.8303],[3744,1.8306],[3783,1.777],[3819,1.8303],[3824,1.8306],[3859,1.7771],[3907,1.777],[3943,1.8303],[3948,1.8306],[3983,1.7771],[4031,1.777],[4067,1.8303],[4072,1.8306],[4107,1.7771],[4156,1.777],[4192,1.8303],[4197,1.8306],[4232,1.7771],[4282,1.777],[4318,1.8303],[4323,1.8306],[4357,1.777],[4360,1.8215],[4373,1.8306],[4408,1.7771],[4412,1.777],[4435,1.8252],[4436,0]]
       }
       middle_metacarpal_l_ctrl.Rotation.Z {
        type=MovieSceneScriptingFloatChannel
        default=1.784
        thinned_from=282
        keys [[1790,1.784],[1807,0.2998],[1825,-0.0905],[1826,-0.0904],[1828,-0.0959],[1829,-0.0958],[1837,0.249],[1845,1.2811],[1849,1.6378],[1862,1.8007],[1866,1.8059],[1875,1.4883],[1885,0.4675],[1905,-0.0905],[1906,-0.0904],[1908,-0.0959],[1909,-0.0958],[1917,0.2491],[1925,1.2816],[1929,1.6378],[1942,1.8007],[1943,1.8006],[1945,1.8075],[1947,1.8074],[1956,1.3986],[1966,0.3794],[1985,-0.0905],[1986,-0.0904],[1988,-0.0959],[1989,-0.0958],[1998,0.3651],[2007,1.4918],[2026,1.8059],[2035,1.4883],[2045,0.4675],[2065,-0.0905],[2066,-0.0904],[2068,-0.0959],[2069,-0.0958],[2077,0.2491],[2085,1.2816],[2089,1.6378],[2102,1.8007],[2106,1.8059],[2115,1.4883],[2125,0.4675],[2145,-0.0905],[2146,-0.0904],[2148,-0.0959],[2149,-0.0958],[2157,0.249],[2165,1.2811],[2169,1.6378],[2182,1.8007],[2186,1.8059],[2195,1.4883],[2205,0.4675],[2225,-0.0905],[2226,-0.0904],[2228,-0.0959],[2229,-0.0958],[2237,0.2491],[2245,1.2816],[2249,1.6378],[2262,1.8007],[2263,1.8006],[2265,1.8075],[2267,1.8074],[2276,1.3986],[2286,0.3794],[2305,-0.0905],[2306,-0.0904],[2308,-0.0959],[2309,-0.0958],[2318,0.3651],[2327,1.4918],[2346,1.8059],[2355,1.4883],[2365,0.4675],[2385,-0.0905],[2386,-0.0904],[2388,-0.0959],[2389,-0.0958],[2397,0.2491],[2405,1.2816],[2409,1.6378],[2422,1.8007],[2426,1.8059],[2435,1.4883],[2445,0.4675],[2465,-0.0905],[2466,-0.0904],[2469,-0.1022],[2470,-0.097],[2502,0.0002],[2512,-0.0001],[2520,-0.0148],[2522,-0.0029],[2524,-0.0108],[2530,0.0004],[2774,-0.0625],[2787,0],[2938,-0.0006],[2986,-0.0126],[3058,-0.0126],[3070,-0.0373],[3185,-0.0373],[3217,1.8007],[3221,1.8059],[3230,1.4885],[3240,0.4676],[3260,-0.0905],[3261,-0.0904],[3263,-0.0959],[3264,-0.0958],[3272,0.2489],[3280,1.2809],[3284,1.6378],[3297,1.8007],[3301,1.8059],[3310,1.4885],[3320,0.4676],[3340,-0.0905],[3341,-0.0904],[3343,-0.0959],[3344,-0.0958],[3352,0.249],[3360,1.2814],[3364,1.6378],[3377,1.8007],[3378,1.8006],[3380,1.8075],[3383,1.8074],[3401,0.3795],[3420,-0.0905],[3421,-0.0904],[3423,-0.0959],[3425,-0.0958],[3434,0.49],[3443,1.5746],[3461,1.8059],[3470,1.4885],[3480,0.4676],[3500,-0.0905],[3501,-0.0904],[3503,-0.0959],[3504,-0.0958],[3512,0.249],[3520,1.2814],[3524,1.6378],[3537,1.8007],[3541,1.8059],[3550,1.4884],[3560,0.4676],[3580,-0.0905],[3581,-0.0904],[3583,-0.0959],[3586,-0.0958],[3593,0.3644],[3601,1.3934],[3617,1.8007],[3621,1.8059],[3630,1.4885],[3640,0.4676],[3660,-0.0905],[3661,-0.0904],[3663,-0.0959],[3664,-0.0958],[3672,0.249],[3680,1.2814],[3684,1.6378],[3697,1.8007],[3698,1.8006],[3700,1.8075],[3702,1.8074],[3711,1.3988],[3721,0.3795],[3740,-0.0905],[3741,-0.0904],[3743,-0.0959],[3744,-0.0958],[3753,0.365],[3762,1.4917],[3781,1.8059],[3790,1.4885],[3800,0.4676],[3820,-0.0905],[3821,-0.0904],[3823,-0.0959],[3824,-0.0958],[3832,0.249],[3840,1.2815],[3844,1.6378],[3857,1.8007],[3861,1.8059],[3870,1.4884],[3880,0.4676],[3905,1.8059],[3914,1.4885],[3924,0.4676],[3944,-0.0905],[3945,-0.0904],[3947,-0.0959],[3948,-0.0958],[3956,0.249],[3964,1.2815],[3968,1.6378],[3981,1.8007],[3985,1.8059],[3994,1.4884],[4004,0.4676],[4029,1.8059],[4038,1.4885],[4048,0.4676],[4068,-0.0905],[4069,-0.0904],[4071,-0.0959],[4072,-0.0958],[4080,0.249],[4088,1.2815],[4092,1.6378],[4105,1.8007],[4109,1.8059],[4118,1.4884],[4128,0.4676],[4154,1.8059],[4163,1.4885],[4173,0.4676],[4193,-0.0905],[4194,-0.0904],[4196,-0.0959],[4197,-0.0958],[4205,0.249],[4213,1.2815],[4217,1.6378],[4230,1.8007],[4234,1.8059],[4243,1.4884],[4253,0.4676],[4280,1.8059],[4289,1.4885],[4299,0.4676],[4319,-0.0905],[4320,-0.0904],[4322,-0.0959],[4323,-0.0958],[4331,0.249],[4339,1.2815],[4343,1.6378],[4356,1.8007],[4357,1.8006],[4360,0.1163],[4364,0.0142],[4372,-0.0959],[4373,-0.0958],[4381,0.2489],[4389,1.2809],[4393,1.6377],[4406,1.8007],[4410,1.8059],[4423,1.093],[4436,0]]
       }
       middle_metacarpal_l_ctrl.Scale.X {type=MovieSceneScriptingFloatChannel, default=1, const [1,12]}
       middle_metacarpal_l_ctrl.Scale.Y {type=MovieSceneScriptingFloatChannel, default=1, const [1,12]}
       middle_metacarpal_l_ctrl.Scale.Z {type=MovieSceneScriptingFloatChannel, default=1, const [1,12]}
       middle_01_l_ctrl.Location.X {
        type=MovieSceneScriptingFloatChannel
        default=0
        thinned_from=26
        keys [[1790,0],[2511,0],[2513,-0.0004],[2524,0],[2771,0],[2798,0.0238],[2938,0.0237],[2986,0.0078],[3058,0.0078],[3070,0.0006],[4436,-0.0001]]
       }
       middle_01_l_ctrl.Location.Y {
        type=MovieSceneScriptingFloatChannel
        default=0
        thinned_from=212
        keys [[1790,0],[2768,0],[2803,-0.2427],[2938,-0.2422],[2986,-0.0813],[3058,-0.0813],[3070,-0.002],[3185,-0.002],[3215,0],[4436,0]]
       }
       middle_01_l_ctrl.Location.Z {
        type=MovieSceneScriptingFloatChannel
        default=0
        thinned_from=203
        keys [[1790,0],[2767,0],[2773,0.008],[2776,0.0422],[2781,0.1722],[2786,0.2122],[2800,0.185],[2938,0.1824],[2986,0.1525],[3058,0.1525],[3070,0.073],[3185,0.073],[3224,0],[4436,0]]
       }
       middle_01_l_ctrl.Rotation.X {
        type=MovieSceneScriptingFloatChannel
        default=-0.2841
        thinned_from=142
        keys [[1790,-0.2841],[1827,-0.3197],[1830,-0.3198],[1866,-0.2835],[1907,-0.3197],[1910,-0.3198],[1944,-0.2836],[1947,-0.2835],[1989,-0.3198],[1990,-0.3197],[2024,-0.2835],[2028,-0.2835],[2067,-0.3197],[2070,-0.3198],[2106,-0.2835],[2108,-0.2835],[2147,-0.3197],[2150,-0.3198],[2186,-0.2835],[2227,-0.3197],[2230,-0.3198],[2264,-0.2836],[2267,-0.2835],[2309,-0.3198],[2310,-0.3197],[2344,-0.2835],[2348,-0.2835],[2387,-0.3197],[2390,-0.3198],[2426,-0.2835],[2428,-0.2835],[2466,-0.3196],[2503,0.1496],[2513,0.8313],[2515,0.8252],[2520,1.9286],[2522,1.1604],[2524,1.6771],[2526,1.1517],[2531,0.9528],[2763,3.4437],[2776,0.7837],[2781,-1.3388],[2799,-4.1631],[2938,-4.1461],[2986,-1.2336],[3058,-1.2336],[3070,-0.6631],[3185,-0.6631],[3221,-0.2835],[3222,-0.2835],[3262,-0.3197],[3265,-0.3198],[3301,-0.2835],[3342,-0.3197],[3345,-0.3198],[3379,-0.2836],[3383,-0.2835],[3424,-0.3198],[3426,-0.3197],[3459,-0.2835],[3463,-0.2835],[3502,-0.3197],[3505,-0.3198],[3541,-0.2835],[3582,-0.3197],[3586,-0.3198],[3621,-0.2835],[3622,-0.2835],[3662,-0.3197],[3665,-0.3198],[3699,-0.2836],[3703,-0.2835],[3744,-0.3198],[3746,-0.3197],[3779,-0.2835],[3782,-0.2835],[3822,-0.3197],[3825,-0.3198],[3861,-0.2835],[3906,-0.2835],[3946,-0.3197],[3949,-0.3198],[3985,-0.2835],[4030,-0.2835],[4070,-0.3197],[4073,-0.3198],[4109,-0.2835],[4155,-0.2835],[4195,-0.3197],[4198,-0.3198],[4234,-0.2835],[4281,-0.2835],[4321,-0.3197],[4324,-0.3198],[4360,-0.3111],[4364,-0.3137],[4371,-0.3197],[4372,-0.3197],[4410,-0.2835],[4411,-0.2835],[4435,-0.316],[4436,-0.0001]]
       }
       middle_01_l_ctrl.Rotation.Y {
        type=MovieSceneScriptingFloatChannel
        default=5.3913
        thinned_from=100
        keys [[1790,5.3913],[1830,5.9021],[1866,5.3855],[1867,5.3856],[1910,5.9021],[1946,5.3856],[1947,5.3856],[1990,5.902],[2026,5.3856],[2027,5.3856],[2070,5.9021],[2106,5.3855],[2107,5.3856],[2150,5.9021],[2186,5.3855],[2187,5.3856],[2230,5.9021],[2266,5.3856],[2267,5.3856],[2310,5.902],[2346,5.3856],[2347,5.3856],[2390,5.9021],[2426,5.3855],[2427,5.3856],[2466,5.9009],[2470,5.5802],[2503,0.6948],[2507,0.0899],[2509,-1.6916],[2512,-2.3425],[2515,-2.3425],[2522,1.2318],[2524,0.2705],[2527,1.6248],[2531,1.7769],[2757,0.9902],[2763,-3.7384],[2772,0.1733],[2782,6.271],[2803,-1.1095],[2927,-1.1065],[2938,-1.1048],[2986,0.5603],[3058,0.5603],[3070,3.0459],[3185,3.0459],[3221,5.3855],[3222,5.3856],[3265,5.9021],[3301,5.3855],[3302,5.3856],[3345,5.9021],[3381,5.3856],[3382,5.3856],[3425,5.902],[3461,5.3856],[3462,5.3856],[3505,5.9021],[3541,5.3855],[3542,5.3856],[3585,5.9021],[3621,5.3855],[3622,5.3856],[3665,5.9021],[3701,5.3856],[3702,5.3856],[3745,5.902],[3781,5.3856],[3782,5.3856],[3825,5.9021],[3861,5.3855],[3906,5.3856],[3949,5.9021],[3985,5.3855],[4030,5.3856],[4073,5.9021],[4109,5.3855],[4155,5.3856],[4198,5.9021],[4234,5.3855],[4281,5.3856],[4324,5.9021],[4360,5.7409],[4410,5.3855],[4411,5.3856],[4435,5.8498],[4436,0.0016]]
       }
       middle_01_l_ctrl.Rotation.Z {
        type=MovieSceneScriptingFloatChannel
        default=1.0937
        thinned_from=87
        keys [[1790,1.0937],[1829,8.2227],[1830,8.2227],[1867,1.011],[1909,8.2227],[1910,8.2227],[1946,1.011],[1947,1.011],[1990,8.2251],[2026,1.011],[2069,8.2227],[2070,8.2227],[2107,1.011],[2149,8.2227],[2150,8.2227],[2187,1.011],[2229,8.2227],[2230,8.2227],[2266,1.011],[2267,1.011],[2310,8.2251],[2346,1.011],[2389,8.2227],[2390,8.2227],[2427,1.011],[2466,8.2092],[2468,8.1407],[2503,8.5268],[2505,8.5268],[2510,-51.119],[2513,-56.6069],[2515,-56.6062],[2518,-35.2409],[2519,7.7782],[2520,29.7769],[2521,30.1672],[2523,29.748],[2531,30.2144],[2766,27.1493],[2786,-13.8755],[2927,-13.1491],[2938,-13.1225],[2986,63.2997],[3058,63.2997],[3070,13.0088],[3185,13.0088],[3222,1.011],[3264,8.2227],[3302,1.011],[3344,8.2227],[3345,8.2227],[3381,1.011],[3382,1.011],[3425,8.2251],[3461,1.011],[3504,8.2227],[3542,1.011],[3584,8.2227],[3585,8.2227],[3622,1.011],[3664,8.2227],[3665,8.2227],[3701,1.011],[3702,1.011],[3746,8.2251],[3781,1.011],[3824,8.2227],[3862,1.011],[3905,1.011],[3948,8.2227],[3986,1.011],[4029,1.011],[4072,8.2227],[4110,1.011],[4154,1.011],[4197,8.2227],[4235,1.011],[4280,1.011],[4323,8.2227],[4360,6.1201],[4364,6.5948],[4373,8.2227],[4411,1.011],[4435,7.4949],[4436,0.002]]
       }
       middle_01_l_ctrl.Scale.X {type=MovieSceneScriptingFloatChannel, default=1, const [1,12]}
       middle_01_l_ctrl.Scale.Y {type=MovieSceneScriptingFloatChannel, default=1, const [1,12]}
       middle_01_l_ctrl.Scale.Z {type=MovieSceneScriptingFloatChannel, default=1, const [1,12]}
       middle_02_l_ctrl.Location.X {
        type=MovieSceneScriptingFloatChannel
        default=-0.0452
        thinned_from=455
        keys [[1790,-0.0452],[2471,-0.0452],[2504,-0.045],[2509,-0.0453],[2511,-0.0452],[2513,-0.0454],[2531,-0.0452],[2938,-0.0452],[2986,-0.0447],[3058,-0.0447],[3070,-0.0451],[4435,-0.0452],[4436,-0.0001]]
       }
       middle_02_l_ctrl.Location.Y {
        type=MovieSceneScriptingFloatChannel
        default=0.0507
        thinned_from=500
        keys [[1790,0.0507],[2471,0.0507],[2504,0.0513],[2509,0.0507],[2938,0.0507],[2986,0.0524],[3058,0.0524],[3070,0.0509],[3185,0.0509],[4435,0.0507],[4436,-0.0001]]
       }
       middle_02_l_ctrl.Location.Z {
        type=MovieSceneScriptingFloatChannel
        default=0.1146
        thinned_from=765
        keys [[1790,0.1146],[2468,0.1146],[2496,0.0988],[2507,0.0988],[2509,0.1146],[2797,0.1146],[2938,0.1143],[2986,0.0804],[3058,0.0804],[3070,0.1108],[3185,0.1108],[3215,0.1146],[4435,0.1146],[4436,0]]
       }
       middle_02_l_ctrl.Rotation.X {
        type=MovieSceneScriptingFloatChannel
        default=-0.8796
        thinned_from=209
        keys [[1790,-0.8796],[1827,-0.1821],[1830,-0.1821],[1855,-0.8728],[1856,-0.8727],[1858,-0.8831],[1860,-0.883],[1865,-0.8934],[1866,-0.8934],[1906,-0.1821],[1910,-0.1821],[1945,-0.889],[1947,-0.889],[1988,-0.1822],[1990,-0.1822],[2015,-0.8728],[2016,-0.8727],[2018,-0.8831],[2020,-0.883],[2024,-0.8934],[2027,-0.8934],[2069,-0.1822],[2070,-0.1822],[2095,-0.8728],[2096,-0.8727],[2098,-0.8831],[2100,-0.883],[2105,-0.8934],[2106,-0.8934],[2146,-0.1821],[2149,-0.1821],[2175,-0.8728],[2176,-0.8727],[2178,-0.8831],[2180,-0.883],[2185,-0.8934],[2226,-0.1821],[2230,-0.1821],[2265,-0.889],[2267,-0.889],[2308,-0.1822],[2310,-0.1822],[2335,-0.8728],[2336,-0.8727],[2338,-0.8831],[2340,-0.883],[2345,-0.8934],[2347,-0.8934],[2388,-0.1822],[2390,-0.1822],[2415,-0.8728],[2416,-0.8727],[2418,-0.8831],[2420,-0.883],[2425,-0.8934],[2426,-0.8934],[2466,-0.1821],[2503,-6.1021],[2506,-6.1021],[2509,-7.0113],[2513,-7.2775],[2515,-7.2677],[2519,-5.4732],[2520,-5.4913],[2521,-6.8067],[2522,-7.0259],[2524,-6.0243],[2528,-7.35],[2532,-7.401],[2757,-6.7531],[2760,-3.655],[2763,-1.8985],[2766,-1.7125],[2778,-0.7354],[2796,-2.706],[2927,-3.1609],[2938,-3.1688],[2986,-2.5647],[3058,-2.5647],[3070,-4.5251],[3185,-4.5251],[3215,-0.883],[3220,-0.8934],[3222,-0.8934],[3261,-0.1821],[3263,-0.1821],[3290,-0.8728],[3291,-0.8727],[3293,-0.8831],[3295,-0.883],[3300,-0.8934],[3302,-0.8934],[3341,-0.1821],[3345,-0.1821],[3380,-0.889],[3382,-0.889],[3425,-0.1822],[3450,-0.8728],[3451,-0.8727],[3453,-0.8831],[3455,-0.883],[3460,-0.8934],[3503,-0.1822],[3530,-0.8728],[3531,-0.8727],[3533,-0.8831],[3535,-0.883],[3542,-0.8934],[3581,-0.1821],[3586,-0.1821],[3610,-0.8728],[3611,-0.8727],[3613,-0.8831],[3615,-0.883],[3622,-0.8934],[3661,-0.1821],[3665,-0.1821],[3700,-0.889],[3702,-0.889],[3744,-0.1822],[3746,-0.1822],[3770,-0.8728],[3771,-0.8727],[3773,-0.8831],[3775,-0.883],[3780,-0.8934],[3823,-0.1822],[3850,-0.8728],[3851,-0.8727],[3853,-0.8831],[3855,-0.883],[3860,-0.8934],[3862,-0.8934],[3895,-0.8727],[3897,-0.8831],[3899,-0.883],[3904,-0.8934],[3947,-0.1822],[3974,-0.8728],[3975,-0.8727],[3977,-0.8831],[3979,-0.883],[3984,-0.8934],[3986,-0.8934],[4019,-0.8727],[4021,-0.8831],[4023,-0.883],[4028,-0.8934],[4071,-0.1822],[4098,-0.8728],[4099,-0.8727],[4101,-0.8831],[4103,-0.883],[4108,-0.8934],[4110,-0.8934],[4144,-0.8727],[4146,-0.8831],[4148,-0.883],[4153,-0.8934],[4196,-0.1822],[4223,-0.8728],[4224,-0.8727],[4226,-0.8831],[4228,-0.883],[4233,-0.8934],[4235,-0.8934],[4270,-0.8727],[4272,-0.8831],[4274,-0.883],[4279,-0.8934],[4322,-0.1822],[4349,-0.8728],[4350,-0.8727],[4352,-0.8831],[4354,-0.883],[4360,-0.3427],[4364,-0.259],[4371,-0.1821],[4372,-0.1821],[4399,-0.8728],[4400,-0.8727],[4402,-0.8831],[4404,-0.883],[4410,-0.8934],[4436,0.0007]]
       }
       middle_02_l_ctrl.Rotation.Y {
        type=MovieSceneScriptingFloatChannel
        default=-0.1219
        thinned_from=366
        keys [[1790,-0.1219],[1814,-0.0631],[1815,-0.0637],[1821,-0.0625],[1822,-0.0625],[1824,-0.0607],[1827,-0.0623],[1830,-0.0623],[1860,-0.1234],[1862,-0.1215],[1864,-0.1239],[1868,-0.1239],[1893,-0.0637],[1894,-0.0639],[1896,-0.0629],[1902,-0.0625],[1904,-0.0607],[1906,-0.0623],[1910,-0.0623],[1942,-0.1237],[1944,-0.1229],[1949,-0.1231],[1974,-0.0631],[1975,-0.0637],[1979,-0.0623],[1980,-0.0627],[1984,-0.062],[1985,-0.0624],[1987,-0.0616],[1990,-0.0616],[2020,-0.1234],[2022,-0.1215],[2025,-0.1239],[2028,-0.1239],[2053,-0.0637],[2054,-0.0639],[2056,-0.0629],[2057,-0.0629],[2059,-0.0623],[2060,-0.0627],[2064,-0.062],[2065,-0.0624],[2067,-0.0616],[2070,-0.0616],[2100,-0.1234],[2102,-0.1215],[2104,-0.1239],[2105,-0.1239],[2134,-0.0631],[2135,-0.0637],[2138,-0.0628],[2142,-0.0625],[2144,-0.0607],[2146,-0.0623],[2151,-0.0623],[2180,-0.1234],[2182,-0.1215],[2184,-0.1239],[2188,-0.1239],[2213,-0.0637],[2214,-0.0639],[2216,-0.0629],[2222,-0.0625],[2224,-0.0607],[2226,-0.0623],[2230,-0.0623],[2262,-0.1237],[2264,-0.1229],[2269,-0.1231],[2294,-0.0631],[2295,-0.0637],[2299,-0.0623],[2300,-0.0627],[2304,-0.062],[2305,-0.0624],[2307,-0.0616],[2310,-0.0616],[2340,-0.1234],[2342,-0.1216],[2345,-0.1239],[2373,-0.0637],[2374,-0.0639],[2376,-0.0629],[2377,-0.0629],[2379,-0.0623],[2380,-0.0627],[2384,-0.062],[2385,-0.0624],[2387,-0.0616],[2390,-0.0616],[2420,-0.1234],[2422,-0.1215],[2424,-0.1239],[2428,-0.1239],[2454,-0.0631],[2455,-0.0637],[2461,-0.0625],[2462,-0.0625],[2464,-0.0607],[2466,-0.0623],[2496,2.5187],[2503,2.0955],[2511,1.93],[2513,1.933],[2515,1.9299],[2520,3.3509],[2522,2.5286],[2524,3.0803],[2527,2.3688],[2531,2.287],[2763,4.3677],[2774,1.9928],[2778,0.701],[2782,-0.0529],[2794,-0.6603],[2927,-0.6252],[2938,-0.6216],[2986,6.9336],[3058,6.9336],[3070,2.5948],[3185,2.5948],[3215,-0.1234],[3217,-0.1216],[3219,-0.1239],[3223,-0.1239],[3249,-0.0631],[3250,-0.0637],[3253,-0.0628],[3254,-0.0628],[3259,-0.0607],[3261,-0.0623],[3266,-0.0623],[3295,-0.1234],[3297,-0.1216],[3300,-0.1239],[3303,-0.1239],[3328,-0.0637],[3329,-0.0639],[3331,-0.0629],[3332,-0.0629],[3339,-0.0607],[3341,-0.0623],[3345,-0.0623],[3377,-0.1237],[3379,-0.1229],[3383,-0.1229],[3384,-0.1231],[3409,-0.0631],[3410,-0.0637],[3414,-0.0623],[3415,-0.0627],[3419,-0.062],[3420,-0.0624],[3423,-0.0616],[3455,-0.1234],[3457,-0.1216],[3461,-0.1239],[3463,-0.1239],[3488,-0.0637],[3489,-0.0639],[3491,-0.0629],[3492,-0.0629],[3494,-0.0623],[3495,-0.0627],[3499,-0.062],[3500,-0.0624],[3502,-0.0616],[3503,-0.0616],[3535,-0.1234],[3537,-0.1216],[3540,-0.1239],[3543,-0.1239],[3569,-0.0631],[3570,-0.0637],[3573,-0.0628],[3574,-0.0628],[3579,-0.0607],[3581,-0.0623],[3585,-0.0623],[3615,-0.1234],[3617,-0.1216],[3620,-0.1239],[3622,-0.1239],[3648,-0.0637],[3649,-0.0639],[3651,-0.0629],[3654,-0.0628],[3659,-0.0607],[3661,-0.0623],[3665,-0.0623],[3697,-0.1237],[3699,-0.1229],[3704,-0.1231],[3729,-0.0631],[3730,-0.0637],[3734,-0.0623],[3735,-0.0627],[3739,-0.062],[3740,-0.0624],[3742,-0.0616],[3775,-0.1234],[3777,-0.1216],[3781,-0.1239],[3783,-0.1239],[3808,-0.0637],[3809,-0.0639],[3811,-0.0629],[3812,-0.0629],[3814,-0.0623],[3815,-0.0627],[3819,-0.062],[3820,-0.0624],[3822,-0.0616],[3826,-0.0616],[3855,-0.1234],[3857,-0.1216],[3859,-0.1239],[3899,-0.1234],[3901,-0.1216],[3905,-0.1239],[3907,-0.1239],[3932,-0.0637],[3933,-0.0639],[3935,-0.0629],[3936,-0.0629],[3938,-0.0623],[3939,-0.0627],[3943,-0.062],[3944,-0.0624],[3946,-0.0616],[3950,-0.0616],[3979,-0.1234],[3981,-0.1216],[3983,-0.1239],[4023,-0.1234],[4025,-0.1216],[4029,-0.1239],[4031,-0.1239],[4056,-0.0637],[4057,-0.0639],[4059,-0.0629],[4060,-0.0629],[4062,-0.0623],[4063,-0.0627],[4067,-0.062],[4068,-0.0624],[4070,-0.0616],[4074,-0.0616],[4103,-0.1234],[4105,-0.1216],[4107,-0.1239],[4148,-0.1234],[4150,-0.1216],[4154,-0.1239],[4156,-0.1239],[4181,-0.0637],[4182,-0.0639],[4184,-0.0629],[4185,-0.0629],[4187,-0.0623],[4188,-0.0627],[4192,-0.062],[4193,-0.0624],[4195,-0.0616],[4199,-0.0616],[4228,-0.1234],[4230,-0.1216],[4232,-0.1239],[4274,-0.1234],[4276,-0.1216],[4280,-0.1239],[4282,-0.1239],[4307,-0.0637],[4308,-0.0639],[4310,-0.0629],[4311,-0.0629],[4313,-0.0623],[4314,-0.0627],[4318,-0.062],[4319,-0.0624],[4321,-0.0616],[4325,-0.0616],[4354,-0.1234],[4356,-0.1216],[4360,-0.0631],[4365,-0.0637],[4371,-0.0623],[4372,-0.0623],[4404,-0.1234],[4406,-0.1216],[4409,-0.1239],[4436,0.0008]]
       }
       middle_02_l_ctrl.Rotation.Z {
        type=MovieSceneScriptingFloatChannel
        default=14.1284
        thinned_from=194
        keys [[1790,14.1284],[1818,23.0368],[1819,23.0368],[1821,23.1804],[1823,23.1804],[1826,23.3239],[1829,23.3239],[1856,14.2033],[1862,14.0678],[1869,14.0678],[1898,23.0368],[1899,23.0368],[1901,23.1804],[1904,23.1804],[1907,23.3239],[1909,23.3239],[1939,14.131],[1940,14.131],[1942,14.0678],[1944,14.0678],[1984,23.2624],[1990,23.2624],[2017,14.2033],[2022,14.0678],[2029,14.0678],[2064,23.2624],[2097,14.2033],[2102,14.0678],[2109,14.0678],[2141,23.1804],[2144,23.1804],[2146,23.3239],[2148,23.3239],[2177,14.2033],[2182,14.0678],[2189,14.0678],[2218,23.0368],[2221,23.1804],[2224,23.1804],[2227,23.3239],[2259,14.131],[2260,14.131],[2262,14.0678],[2264,14.0678],[2304,23.2624],[2310,23.2624],[2337,14.2033],[2342,14.0678],[2349,14.0678],[2384,23.2624],[2389,23.2624],[2417,14.2033],[2422,14.0678],[2429,14.0678],[2458,23.0368],[2459,23.0368],[2462,23.1804],[2463,23.1804],[2466,23.3236],[2470,22.9816],[2503,-9.2722],[2513,-14.2268],[2520,7.1491],[2521,-5.7342],[2522,-8.0382],[2524,2.1176],[2527,-10.785],[2531,-12.1526],[2764,37.4704],[2794,-14.2355],[2938,-14.0837],[2986,60.9935],[3058,60.9935],[3070,-1.9594],[3185,-1.9594],[3217,14.0678],[3224,14.0678],[3257,23.1804],[3259,23.1804],[3262,23.3239],[3264,23.3239],[3292,14.2033],[3297,14.0678],[3304,14.0678],[3337,23.1804],[3341,23.3239],[3345,23.3239],[3377,14.0678],[3379,14.0678],[3420,23.2624],[3426,23.2624],[3457,14.0678],[3464,14.0678],[3501,23.2624],[3504,23.2624],[3532,14.2033],[3537,14.0678],[3544,14.0678],[3577,23.1804],[3578,23.1804],[3581,23.3239],[3585,23.3239],[3612,14.2033],[3617,14.0678],[3619,14.0678],[3657,23.1804],[3659,23.1804],[3662,23.3239],[3665,23.3239],[3699,14.0678],[3739,23.2624],[3743,23.2624],[3772,14.2033],[3777,14.0678],[3784,14.0678],[3820,23.2624],[3824,23.2624],[3857,14.0678],[3864,14.0678],[3896,14.2033],[3901,14.0678],[3908,14.0678],[3944,23.2624],[3948,23.2624],[3981,14.0678],[3988,14.0678],[4020,14.2033],[4025,14.0678],[4032,14.0678],[4068,23.2624],[4072,23.2624],[4105,14.0678],[4112,14.0678],[4145,14.2033],[4150,14.0678],[4157,14.0678],[4193,23.2624],[4197,23.2624],[4230,14.0678],[4237,14.0678],[4271,14.2033],[4276,14.0678],[4283,14.0678],[4319,23.2624],[4323,23.2624],[4356,14.0678],[4360,21.5862],[4364,21.4256],[4406,14.0678],[4408,14.0678],[4435,22.3821],[4436,0.0058]]
       }
       middle_02_l_ctrl.Scale.X {type=MovieSceneScriptingFloatChannel, default=1, const [1,12]}
       middle_02_l_ctrl.Scale.Y {type=MovieSceneScriptingFloatChannel, default=1, const [1,12]}
       middle_02_l_ctrl.Scale.Z {type=MovieSceneScriptingFloatChannel, default=1, const [1,12]}
       middle_03_l_ctrl.Location.X {
        type=MovieSceneScriptingFloatChannel
        default=0.0206
        thinned_from=498
        keys [[1790,0.0206],[2511,0.0206],[2513,0.0205],[2518,0.0206],[4435,0.0206],[4436,-0.0001]]
       }
       middle_03_l_ctrl.Location.Y {type=MovieSceneScriptingFloatChannel, default=-0.0583, thinned_from=557, keys [[1790,-0.0583],[4435,-0.0583],[4436,-0.0002]]}
       middle_03_l_ctrl.Location.Z {type=MovieSceneScriptingFloatChannel, default=0.1324, thinned_from=149, keys [[1790,0.1324],[4435,0.1324],[4436,0.0001]]}
       middle_03_l_ctrl.Rotation.X {
        type=MovieSceneScriptingFloatChannel
        default=-0.2928
        thinned_from=145
        keys [[1790,-0.2928],[1826,-0.0864],[1862,-0.2949],[1866,-0.2951],[1905,-0.0864],[1936,-0.2924],[1937,-0.2924],[1941,-0.2959],[1944,-0.2959],[1946,-0.2942],[1987,-0.0862],[2019,-0.2943],[2025,-0.2959],[2028,-0.2959],[2063,-0.0867],[2066,-0.0862],[2068,-0.0862],[2102,-0.2949],[2104,-0.2954],[2107,-0.2942],[2146,-0.0864],[2185,-0.2951],[2188,-0.2951],[2226,-0.0864],[2256,-0.2924],[2260,-0.2959],[2264,-0.2959],[2266,-0.2942],[2307,-0.0862],[2342,-0.2949],[2343,-0.2949],[2345,-0.2959],[2348,-0.2959],[2383,-0.0867],[2386,-0.0862],[2422,-0.2949],[2424,-0.2954],[2465,-0.0864],[2466,-0.0864],[2484,-1.0619],[2501,-2.1506],[2503,-2.301],[2513,-1.5442],[2515,-1.561],[2519,0.1354],[2520,0.4375],[2522,-0.1219],[2524,0.2477],[2528,-0.2549],[2532,-0.276],[2759,0.6661],[2764,1.5023],[2779,-0.0149],[2794,-0.9639],[2938,-0.9578],[2986,2.819],[3058,2.819],[3070,-0.3717],[3185,-0.3717],[3219,-0.2954],[3261,-0.0864],[3297,-0.2949],[3300,-0.2951],[3341,-0.0864],[3342,-0.0864],[3371,-0.2924],[3372,-0.2924],[3376,-0.2959],[3379,-0.2959],[3382,-0.2942],[3383,-0.2942],[3418,-0.0867],[3454,-0.2943],[3462,-0.2959],[3502,-0.0862],[3539,-0.2954],[3542,-0.2942],[3581,-0.0864],[3584,-0.0864],[3617,-0.2949],[3623,-0.2951],[3661,-0.0864],[3663,-0.0864],[3697,-0.2959],[3703,-0.2942],[3745,-0.0862],[3780,-0.2959],[3782,-0.2959],[3822,-0.0862],[3857,-0.2949],[3859,-0.2954],[3862,-0.2942],[3906,-0.2959],[3946,-0.0862],[3981,-0.2949],[3983,-0.2954],[3986,-0.2942],[4030,-0.2959],[4070,-0.0862],[4105,-0.2949],[4107,-0.2954],[4110,-0.2942],[4155,-0.2959],[4195,-0.0862],[4230,-0.2949],[4232,-0.2954],[4235,-0.2942],[4281,-0.2959],[4321,-0.0862],[4356,-0.2949],[4360,-0.1355],[4364,-0.1385],[4409,-0.2951],[4411,-0.2951],[4436,0.0002]]
       }
       middle_03_l_ctrl.Rotation.Y {
        type=MovieSceneScriptingFloatChannel
        default=0.3933
        thinned_from=158
        keys [[1790,0.3933],[1824,0.0371],[1859,0.3929],[1862,0.3954],[1865,0.3954],[1904,0.0371],[1912,0.0366],[1936,0.3902],[1938,0.3901],[1940,0.3952],[1946,0.3955],[1978,0.0464],[1979,0.0465],[1981,0.0405],[1983,0.0406],[2019,0.3929],[2022,0.3954],[2025,0.3952],[2058,0.0464],[2059,0.0465],[2061,0.0405],[2063,0.0406],[2066,0.0346],[2099,0.3929],[2102,0.3954],[2106,0.3955],[2144,0.0371],[2182,0.3954],[2185,0.3954],[2224,0.0371],[2232,0.0366],[2256,0.3902],[2258,0.3901],[2260,0.3952],[2266,0.3955],[2298,0.0464],[2299,0.0465],[2301,0.0405],[2303,0.0406],[2342,0.3954],[2345,0.3952],[2378,0.0464],[2379,0.0465],[2381,0.0405],[2386,0.0346],[2419,0.3929],[2422,0.3954],[2426,0.3955],[2464,0.0371],[2503,3.2125],[2513,4.2876],[2520,3.8887],[2522,4.4924],[2524,4.0829],[2529,4.6602],[2531,4.6597],[2758,4.0435],[2760,3.2206],[2771,1.7165],[2780,0.4522],[2797,1.2794],[2927,1.4796],[2938,1.4822],[2986,2.7564],[3058,2.7564],[3185,2.7527],[3218,0.3954],[3221,0.3955],[3259,0.0371],[3297,0.3954],[3300,0.3954],[3339,0.0371],[3347,0.0366],[3371,0.3902],[3373,0.3901],[3376,0.3952],[3381,0.3955],[3413,0.0464],[3414,0.0465],[3416,0.0405],[3422,0.0346],[3454,0.3929],[3458,0.3954],[3460,0.3952],[3493,0.0464],[3494,0.0465],[3496,0.0405],[3498,0.0406],[3537,0.3954],[3541,0.3955],[3579,0.0371],[3617,0.3954],[3620,0.3954],[3659,0.0371],[3667,0.0366],[3692,0.3902],[3693,0.3901],[3695,0.3952],[3701,0.3955],[3733,0.0464],[3734,0.0465],[3736,0.0405],[3738,0.0406],[3742,0.0346],[3778,0.3954],[3780,0.3952],[3813,0.0464],[3814,0.0465],[3816,0.0405],[3857,0.3954],[3904,0.3952],[3937,0.0464],[3938,0.0465],[3940,0.0405],[3981,0.3954],[4028,0.3952],[4061,0.0464],[4062,0.0465],[4064,0.0405],[4105,0.3954],[4153,0.3952],[4186,0.0464],[4187,0.0465],[4189,0.0405],[4230,0.3954],[4279,0.3952],[4312,0.0464],[4313,0.0465],[4315,0.0405],[4356,0.3954],[4360,0.0985],[4364,0.1054],[4403,0.3929],[4407,0.3954],[4409,0.3954],[4436,0.0004]]
       }
       middle_03_l_ctrl.Rotation.Z {
        type=MovieSceneScriptingFloatChannel
        default=2.1528
        thinned_from=168
        keys [[1790,2.1528],[1825,10.6011],[1830,10.6011],[1859,2.1552],[1862,2.097],[1904,10.6011],[1912,10.613],[1936,2.2218],[1940,2.097],[1946,2.097],[1981,10.5249],[1983,10.5249],[1988,10.6582],[1990,10.6582],[2019,2.1552],[2022,2.097],[2025,2.097],[2061,10.5249],[2063,10.5249],[2066,10.6582],[2070,10.6582],[2099,2.1552],[2102,2.097],[2106,2.097],[2144,10.6011],[2151,10.6011],[2182,2.097],[2225,10.6011],[2232,10.613],[2256,2.2218],[2261,2.097],[2266,2.097],[2303,10.5249],[2308,10.6582],[2310,10.6582],[2342,2.097],[2345,2.097],[2381,10.5249],[2384,10.5249],[2386,10.6582],[2390,10.6582],[2419,2.1552],[2422,2.097],[2426,2.097],[2465,10.6011],[2503,-37.4591],[2513,-31.8824],[2515,-32.2913],[2520,9.4121],[2522,3.9664],[2524,7.6669],[2532,2.4206],[2764,20.2034],[2927,14.7853],[2938,14.6876],[2986,62.5587],[3058,62.5587],[3070,-6.934],[3185,-6.934],[3219,2.097],[3221,2.097],[3259,10.6011],[3265,10.6011],[3297,2.097],[3340,10.6011],[3347,10.6129],[3371,2.2218],[3376,2.097],[3381,2.097],[3416,10.5249],[3419,10.5249],[3422,10.6582],[3424,10.6582],[3454,2.1552],[3460,2.097],[3503,10.6582],[3506,10.6582],[3539,2.097],[3541,2.097],[3579,10.6011],[3585,10.6011],[3617,2.097],[3660,10.6011],[3667,10.6129],[3696,2.097],[3701,2.097],[3736,10.5249],[3742,10.6582],[3743,10.6582],[3780,2.097],[3818,10.5249],[3819,10.5249],[3823,10.6582],[3825,10.6582],[3857,2.097],[3904,2.097],[3942,10.5249],[3943,10.5249],[3947,10.6582],[3949,10.6582],[3981,2.097],[4028,2.097],[4066,10.5249],[4067,10.5249],[4071,10.6582],[4073,10.6582],[4105,2.097],[4153,2.097],[4191,10.5249],[4192,10.5249],[4196,10.6582],[4198,10.6582],[4230,2.097],[4279,2.097],[4317,10.5249],[4318,10.5249],[4322,10.6582],[4324,10.6582],[4356,2.097],[4360,9.1443],[4364,8.9815],[4403,2.1552],[4407,2.097],[4435,9.7804],[4436,0.0024]]
       }
       middle_03_l_ctrl.Scale.X {type=MovieSceneScriptingFloatChannel, default=1, const [1,12]}
       middle_03_l_ctrl.Scale.Y {type=MovieSceneScriptingFloatChannel, default=1, const [1,12]}
       middle_03_l_ctrl.Scale.Z {type=MovieSceneScriptingFloatChannel, default=1, const [1,12]}
       ring_metacarpal_l_ctrl.Location.X {type=MovieSceneScriptingFloatChannel, default=0, thinned_from=31, keys [[1790,0],[2512,0],[2513,-0.0001],[2521,0],[4436,0]]}
       ring_metacarpal_l_ctrl.Location.Y {type=MovieSceneScriptingFloatChannel, default=0, thinned_from=31, keys [[1790,0],[4436,0]]}
       ring_metacarpal_l_ctrl.Location.Z {type=MovieSceneScriptingFloatChannel, default=0, thinned_from=31, keys [[1790,0],[4436,0]]}
       ring_metacarpal_l_ctrl.Rotation.X {
        type=MovieSceneScriptingFloatChannel
        default=-14.2731
        thinned_from=134
        keys [[1790,-14.2731],[1807,-12.4127],[1825,-11.8029],[1828,-11.7532],[1848,-13.7961],[1868,-14.311],[1886,-12.4576],[1905,-11.8029],[1908,-11.7532],[1928,-13.7892],[1948,-14.3295],[1963,-12.7082],[1990,-11.7532],[2009,-13.8056],[2028,-14.3295],[2046,-12.4576],[2065,-11.8029],[2068,-11.7532],[2088,-13.7931],[2108,-14.3295],[2126,-12.4576],[2145,-11.8029],[2148,-11.7532],[2168,-13.7961],[2188,-14.311],[2206,-12.4576],[2225,-11.8029],[2228,-11.7532],[2248,-13.7892],[2268,-14.3295],[2283,-12.7082],[2310,-11.7532],[2329,-13.8056],[2348,-14.3295],[2366,-12.4576],[2385,-11.8029],[2389,-11.7532],[2408,-13.7931],[2428,-14.3295],[2442,-12.8254],[2457,-12.076],[2496,-0.6509],[2503,-0.0001],[2507,0.0063],[2509,-0.0082],[2513,0.0022],[2518,-0.5996],[2519,-1.7865],[2522,-0.3686],[2524,-1.321],[2526,-0.3712],[2532,0.0118],[2759,-2.3634],[2761,-3.934],[2767,-5.4832],[2774,-7.668],[2777,-6.7666],[2781,-2.4135],[2788,0],[3185,0],[3223,-14.3295],[3238,-12.7083],[3264,-11.7532],[3283,-13.7961],[3303,-14.311],[3318,-12.7083],[3343,-11.7532],[3363,-13.7891],[3383,-14.3295],[3404,-12.3664],[3426,-11.7532],[3444,-13.8055],[3463,-14.3295],[3478,-12.7083],[3504,-11.7532],[3523,-13.7931],[3543,-14.3295],[3558,-12.7083],[3583,-11.7532],[3603,-13.7961],[3623,-14.311],[3638,-12.7083],[3663,-11.7532],[3683,-13.7891],[3703,-14.3295],[3718,-12.7083],[3745,-11.7532],[3764,-13.8055],[3783,-14.3295],[3798,-12.7083],[3823,-11.7532],[3843,-13.7931],[3863,-14.3295],[3878,-12.7083],[3907,-14.3295],[3922,-12.7083],[3947,-11.7532],[3967,-13.7931],[3987,-14.3295],[4002,-12.7083],[4031,-14.3295],[4046,-12.7083],[4071,-11.7532],[4091,-13.7931],[4111,-14.3295],[4126,-12.7083],[4156,-14.3295],[4171,-12.7083],[4196,-11.7532],[4216,-13.7931],[4236,-14.3295],[4251,-12.7083],[4282,-14.3295],[4297,-12.7083],[4322,-11.7532],[4342,-13.7931],[4360,-12.1186],[4364,-11.9427],[4372,-11.7532],[4392,-13.7961],[4412,-14.311],[4424,-13.091],[4430,-12.4577],[4435,-12.3207],[4436,0]]
       }
       ring_metacarpal_l_ctrl.Rotation.Y {
        type=MovieSceneScriptingFloatChannel
        default=3.9875
        thinned_from=115
        keys [[1790,3.9875],[1825,3.7749],[1828,3.772],[1865,3.9913],[1868,3.9903],[1905,3.7749],[1908,3.772],[1942,3.9896],[1946,3.9925],[1990,3.772],[2026,3.9925],[2065,3.7749],[2068,3.772],[2106,3.9925],[2145,3.7749],[2148,3.772],[2185,3.9913],[2188,3.9903],[2225,3.7749],[2228,3.772],[2262,3.9896],[2266,3.9925],[2310,3.772],[2346,3.9925],[2385,3.7749],[2388,3.772],[2426,3.9925],[2457,3.791],[2503,0.6009],[2507,0.403],[2509,-0.1742],[2512,-0.386],[2515,-0.386],[2520,1.3412],[2522,0.9634],[2524,1.2113],[2531,0.8635],[2763,2.0619],[2773,2.7541],[2776,2.5676],[2780,1.1872],[2788,0],[3185,0],[3221,3.9925],[3260,3.7749],[3263,3.772],[3300,3.9913],[3303,3.9903],[3340,3.7749],[3343,3.772],[3377,3.9896],[3381,3.9925],[3425,3.772],[3461,3.9925],[3500,3.7749],[3503,3.772],[3541,3.9925],[3580,3.7749],[3583,3.772],[3620,3.9913],[3623,3.9903],[3660,3.7749],[3663,3.772],[3697,3.9896],[3701,3.9925],[3745,3.772],[3781,3.9925],[3820,3.7749],[3823,3.772],[3861,3.9925],[3905,3.9925],[3944,3.7749],[3947,3.772],[3985,3.9925],[4029,3.9925],[4068,3.7749],[4071,3.772],[4109,3.9925],[4154,3.9925],[4193,3.7749],[4196,3.772],[4234,3.9925],[4280,3.9925],[4319,3.7749],[4322,3.772],[4360,3.8251],[4364,3.8113],[4372,3.772],[4409,3.9913],[4412,3.9903],[4424,3.8977],[4430,3.8362],[4435,3.8114],[4436,0]]
       }
       ring_metacarpal_l_ctrl.Rotation.Z {
        type=MovieSceneScriptingFloatChannel
        default=-0.4343
        thinned_from=178
        keys [[1790,-0.4343],[1807,-2.0679],[1825,-2.6307],[1828,-2.661],[1837,-2.252],[1846,-0.9901],[1865,-0.4018],[1867,-0.4044],[1868,-0.4044],[1886,-2.001],[1905,-2.6307],[1908,-2.661],[1916,-2.3685],[1925,-1.0965],[1942,-0.4199],[1943,-0.4213],[1946,-0.3861],[1968,-2.1184],[1990,-2.661],[2008,-0.8404],[2026,-0.3861],[2036,-0.895],[2047,-2.0679],[2069,-2.661],[2078,-2.1077],[2087,-0.9129],[2106,-0.3861],[2116,-0.895],[2127,-2.0679],[2149,-2.661],[2158,-2.1114],[2167,-0.9001],[2185,-0.4018],[2187,-0.4044],[2188,-0.4044],[2208,-2.1184],[2229,-2.661],[2237,-2.244],[2245,-1.0965],[2262,-0.4199],[2263,-0.4213],[2266,-0.3861],[2288,-2.1184],[2310,-2.661],[2328,-0.8404],[2346,-0.3861],[2356,-0.895],[2367,-2.0679],[2389,-2.661],[2398,-2.1077],[2407,-0.9129],[2426,-0.3861],[2435,-0.791],[2445,-1.9177],[2465,-2.6307],[2466,-2.6307],[2468,-2.6529],[2471,-2.5205],[2496,-0.1488],[2503,-0.002],[2506,-0.002],[2507,-0.228],[2510,-1.0591],[2513,-1.1529],[2518,-1.0357],[2521,-0.0964],[2523,-0.2792],[2532,-0.0003],[2761,-0.8903],[2764,-1.0515],[2774,-1.7239],[2777,-1.482],[2781,-0.4912],[2788,0],[3185,0],[3221,-0.3861],[3231,-0.8948],[3242,-2.0678],[3264,-2.661],[3273,-2.1116],[3282,-0.9001],[3300,-0.4018],[3302,-0.4044],[3303,-0.4044],[3323,-2.1183],[3344,-2.661],[3352,-2.2442],[3360,-1.0966],[3377,-0.4199],[3378,-0.4213],[3381,-0.3861],[3403,-2.1183],[3425,-2.661],[3443,-0.8405],[3461,-0.3861],[3471,-0.8948],[3482,-2.0678],[3504,-2.661],[3513,-2.1078],[3522,-0.913],[3541,-0.3861],[3551,-0.8948],[3562,-2.0678],[3584,-2.661],[3593,-2.1116],[3602,-0.9001],[3620,-0.4018],[3622,-0.4044],[3623,-0.4044],[3643,-2.1183],[3664,-2.661],[3672,-2.2442],[3680,-1.0966],[3697,-0.4199],[3698,-0.4213],[3701,-0.3861],[3723,-2.1183],[3746,-2.661],[3763,-0.8405],[3781,-0.3861],[3791,-0.8948],[3802,-2.0678],[3824,-2.661],[3833,-2.1079],[3842,-0.913],[3861,-0.3861],[3871,-0.8948],[3882,-2.0678],[3905,-0.3861],[3915,-0.8948],[3926,-2.0678],[3948,-2.661],[3957,-2.1079],[3966,-0.913],[3985,-0.3861],[3995,-0.8948],[4006,-2.0678],[4029,-0.3861],[4039,-0.8948],[4050,-2.0678],[4072,-2.661],[4081,-2.1079],[4090,-0.913],[4109,-0.3861],[4119,-0.8948],[4130,-2.0678],[4154,-0.3861],[4164,-0.8948],[4175,-2.0678],[4197,-2.661],[4206,-2.1079],[4215,-0.913],[4234,-0.3861],[4244,-0.8948],[4255,-2.0678],[4280,-0.3861],[4290,-0.8948],[4301,-2.0678],[4323,-2.661],[4332,-2.1079],[4341,-0.913],[4360,-2.3413],[4364,-2.4842],[4373,-2.661],[4382,-2.1116],[4391,-0.9002],[4409,-0.4018],[4411,-0.4044],[4412,-0.4044],[4435,-2.255],[4436,0]]
       }
       ring_metacarpal_l_ctrl.Scale.X {type=MovieSceneScriptingFloatChannel, default=1, const [1,12]}
       ring_metacarpal_l_ctrl.Scale.Y {type=MovieSceneScriptingFloatChannel, default=1, const [1,12]}
       ring_metacarpal_l_ctrl.Scale.Z {type=MovieSceneScriptingFloatChannel, default=1, const [1,12]}
       ring_01_l_ctrl.Location.X {
        type=MovieSceneScriptingFloatChannel
        default=0
        thinned_from=34
        keys [[1790,0],[2511,0],[2513,-0.0004],[2528,0],[2768,0],[2775,0.007],[2782,0.0629],[2796,0.1188],[2938,0.118],[2986,0.0534],[3058,0.0534],[3070,0.0245],[3185,0.0245],[3215,0],[4436,-0.0001]]
       }
       ring_01_l_ctrl.Location.Y {
        type=MovieSceneScriptingFloatChannel
        default=0
        thinned_from=324
        keys [[1790,0],[2768,0],[2775,-0.0315],[2782,-0.2673],[2796,-0.4691],[2938,-0.4655],[2986,-0.1824],[3058,-0.1824],[3070,-0.1026],[3185,-0.1026],[3238,0],[4434,0],[4436,-0.0001]]
       }
       ring_01_l_ctrl.Location.Z {
        type=MovieSceneScriptingFloatChannel
        default=0
        thinned_from=733
        keys [[1790,0],[2768,0],[2773,0.012],[2776,0.0619],[2780,0.2116],[2784,0.2622],[2805,0.0443],[2938,0.0435],[2986,-0.0223],[3058,-0.0223],[3070,0.1703],[3185,0.1703],[3220,0],[4436,0]]
       }
       ring_01_l_ctrl.Rotation.X {
        type=MovieSceneScriptingFloatChannel
        default=12.9291
        thinned_from=114
        keys [[1790,12.9291],[1804,14.7609],[1829,15.6815],[1830,15.6811],[1844,13.736],[1868,12.8937],[1909,15.6815],[1910,15.6811],[1923,13.8855],[1947,12.899],[1989,15.6815],[1990,15.6811],[2003,13.8832],[2027,12.899],[2069,15.6815],[2070,15.6811],[2083,13.8832],[2107,12.899],[2149,15.6815],[2150,15.6811],[2164,13.736],[2188,12.8937],[2229,15.6815],[2230,15.6811],[2243,13.8856],[2267,12.899],[2309,15.6815],[2310,15.6811],[2323,13.8832],[2347,12.899],[2389,15.6815],[2390,15.6811],[2403,13.8833],[2427,12.899],[2466,15.6438],[2503,0.7267],[2513,-1.3455],[2518,-0.2886],[2520,3.0241],[2522,1.3131],[2524,2.4678],[2531,0.8426],[2759,3.7386],[2775,10.562],[2781,4.6546],[2787,0.9503],[2800,-0.5969],[2938,-0.5991],[2986,7.862],[3058,7.862],[3070,1.2403],[3185,1.2403],[3222,12.899],[3264,15.6815],[3265,15.6811],[3279,13.7362],[3303,12.8937],[3344,15.6815],[3345,15.6811],[3359,13.7335],[3383,12.899],[3424,15.6815],[3425,15.6811],[3439,13.7349],[3463,12.899],[3504,15.6815],[3505,15.6811],[3518,13.8835],[3542,12.899],[3584,15.6815],[3585,15.6811],[3599,13.7362],[3623,12.8937],[3664,15.6815],[3665,15.6811],[3679,13.7335],[3703,12.899],[3744,15.6815],[3745,15.6811],[3758,13.8835],[3782,12.899],[3824,15.6815],[3825,15.6811],[3839,13.7349],[3863,12.899],[3906,12.899],[3948,15.6815],[3949,15.6811],[3963,13.7349],[3987,12.899],[4030,12.899],[4072,15.6815],[4073,15.6811],[4087,13.7349],[4111,12.899],[4155,12.899],[4197,15.6815],[4198,15.6811],[4212,13.7349],[4236,12.899],[4281,12.899],[4323,15.6815],[4324,15.6811],[4338,13.7349],[4360,14.8367],[4364,15.1715],[4373,15.6815],[4388,13.7362],[4412,12.8937],[4435,15.0675],[4436,0.0042]]
       }
       ring_01_l_ctrl.Rotation.Y {
        type=MovieSceneScriptingFloatChannel
        default=-1.2291
        thinned_from=901
        keys [[1790,-1.2291],[1791,-1.2253],[1793,-1.2239],[1797,-1.2321],[1800,-1.2232],[1802,-1.2351],[1805,-1.2319],[1808,-1.2395],[1811,-1.2261],[1813,-1.2347],[1815,-1.2285],[1817,-1.2338],[1818,-1.2297],[1824,-1.2366],[1826,-1.231],[1828,-1.2376],[1830,-1.2289],[1835,-1.2341],[1837,-1.2319],[1838,-1.2345],[1840,-1.2306],[1843,-1.2319],[1847,-1.2279],[1849,-1.2301],[1852,-1.2265],[1854,-1.2291],[1860,-1.2256],[1861,-1.2233],[1863,-1.2275],[1864,-1.2265],[1866,-1.2297],[1867,-1.2279],[1872,-1.2303],[1875,-1.2284],[1879,-1.2312],[1880,-1.2258],[1882,-1.2351],[1885,-1.2319],[1888,-1.2395],[1891,-1.2261],[1893,-1.2347],[1895,-1.2285],[1897,-1.2338],[1898,-1.2297],[1904,-1.2366],[1906,-1.231],[1908,-1.2376],[1910,-1.2289],[1914,-1.2337],[1916,-1.2308],[1918,-1.233],[1924,-1.2283],[1927,-1.2315],[1932,-1.2265],[1934,-1.2291],[1940,-1.2256],[1941,-1.2233],[1944,-1.2313],[1946,-1.2267],[1949,-1.2303],[1951,-1.2253],[1953,-1.2239],[1957,-1.2321],[1960,-1.2232],[1962,-1.2351],[1965,-1.2319],[1968,-1.2395],[1971,-1.2261],[1973,-1.2347],[1975,-1.2285],[1977,-1.2338],[1978,-1.2297],[1984,-1.2366],[1986,-1.231],[1988,-1.2376],[1990,-1.2289],[1993,-1.2358],[1996,-1.2276],[1999,-1.2315],[2001,-1.2278],[2003,-1.2316],[2008,-1.2305],[2009,-1.2288],[2011,-1.23],[2012,-1.228],[2014,-1.23],[2018,-1.2289],[2020,-1.2308],[2022,-1.2293],[2024,-1.2336],[2026,-1.2267],[2029,-1.2303],[2031,-1.2253],[2033,-1.2239],[2037,-1.2321],[2040,-1.2232],[2042,-1.2351],[2045,-1.2319],[2048,-1.2395],[2051,-1.2261],[2053,-1.2347],[2055,-1.2285],[2057,-1.2338],[2058,-1.2297],[2064,-1.2366],[2066,-1.231],[2068,-1.2376],[2070,-1.2289],[2073,-1.2358],[2076,-1.2276],[2079,-1.2315],[2081,-1.2278],[2083,-1.2316],[2086,-1.2301],[2088,-1.2311],[2092,-1.2265],[2094,-1.2291],[2100,-1.2256],[2101,-1.2233],[2104,-1.2313],[2106,-1.2267],[2109,-1.2303],[2111,-1.2253],[2113,-1.2239],[2117,-1.2321],[2120,-1.2232],[2122,-1.2351],[2125,-1.2319],[2128,-1.2395],[2131,-1.2261],[2133,-1.2347],[2135,-1.2285],[2137,-1.2338],[2138,-1.2297],[2144,-1.2366],[2146,-1.231],[2148,-1.2376],[2150,-1.2289],[2155,-1.2341],[2157,-1.2319],[2158,-1.2345],[2160,-1.2306],[2163,-1.2319],[2167,-1.2279],[2169,-1.2301],[2172,-1.2265],[2174,-1.2291],[2180,-1.2256],[2181,-1.2233],[2183,-1.2275],[2184,-1.2265],[2186,-1.2297],[2187,-1.2279],[2192,-1.2303],[2195,-1.2284],[2199,-1.2312],[2200,-1.2258],[2202,-1.2351],[2205,-1.2319],[2208,-1.2395],[2211,-1.2261],[2213,-1.2347],[2215,-1.2285],[2217,-1.2338],[2218,-1.2297],[2224,-1.2366],[2226,-1.231],[2228,-1.2376],[2230,-1.2289],[2234,-1.2337],[2236,-1.2308],[2238,-1.233],[2244,-1.2283],[2247,-1.2315],[2252,-1.2265],[2254,-1.2291],[2260,-1.2256],[2261,-1.2233],[2264,-1.2313],[2266,-1.2267],[2269,-1.2303],[2271,-1.2253],[2273,-1.2239],[2277,-1.2321],[2280,-1.2232],[2282,-1.2351],[2285,-1.2319],[2288,-1.2395],[2291,-1.2261],[2293,-1.2347],[2295,-1.2285],[2297,-1.2338],[2298,-1.2297],[2304,-1.2366],[2306,-1.231],[2308,-1.2376],[2310,-1.2289],[2313,-1.2358],[2316,-1.2275],[2319,-1.2315],[2321,-1.2278],[2323,-1.2316],[2328,-1.2305],[2329,-1.2288],[2331,-1.23],[2332,-1.228],[2334,-1.23],[2338,-1.2289],[2340,-1.2308],[2342,-1.2293],[2344,-1.2336],[2346,-1.2267],[2349,-1.2303],[2351,-1.2253],[2353,-1.2239],[2357,-1.2321],[2360,-1.2232],[2362,-1.2351],[2365,-1.2319],[2368,-1.2395],[2371,-1.2261],[2373,-1.2347],[2375,-1.2285],[2377,-1.2338],[2378,-1.2297],[2384,-1.2366],[2386,-1.231],[2388,-1.2376],[2390,-1.2289],[2393,-1.2358],[2396,-1.2275],[2399,-1.2315],[2401,-1.2278],[2403,-1.2316],[2406,-1.2301],[2408,-1.2311],[2412,-1.2265],[2414,-1.2291],[2420,-1.2256],[2421,-1.2233],[2424,-1.2313],[2426,-1.2267],[2429,-1.2303],[2431,-1.2253],[2433,-1.2239],[2437,-1.2321],[2440,-1.2232],[2442,-1.2351],[2445,-1.2319],[2448,-1.2395],[2451,-1.2261],[2453,-1.2347],[2455,-1.2285],[2457,-1.2338],[2458,-1.2297],[2464,-1.2366],[2466,-1.2309],[2467,-1.2323],[2503,0.5985],[2505,0.5985],[2513,-1.3169],[2518,0.1134],[2519,2.5051],[2520,3.2295],[2522,2.6603],[2524,3.0451],[2530,2.4987],[2763,4.3403],[2777,0.329],[2782,-2.1638],[2792,-5.6608],[2802,-6.1639],[2938,-6.1487],[2986,-4.136],[3058,-4.136],[3070,1.136],[3185,1.136],[3215,-1.2256],[3216,-1.2233],[3219,-1.2313],[3221,-1.2267],[3224,-1.2303],[3226,-1.2253],[3228,-1.224],[3232,-1.2321],[3235,-1.2232],[3237,-1.2351],[3240,-1.2319],[3243,-1.2395],[3246,-1.2261],[3248,-1.2347],[3250,-1.2285],[3252,-1.2339],[3253,-1.2297],[3259,-1.2366],[3261,-1.231],[3263,-1.2376],[3265,-1.2289],[3270,-1.2342],[3272,-1.2319],[3273,-1.2345],[3275,-1.2306],[3278,-1.2319],[3282,-1.2279],[3284,-1.2301],[3287,-1.2265],[3289,-1.2291],[3295,-1.2256],[3296,-1.2233],[3298,-1.2275],[3299,-1.2265],[3301,-1.2297],[3302,-1.2279],[3307,-1.2303],[3310,-1.2284],[3314,-1.2312],[3315,-1.2258],[3317,-1.2351],[3320,-1.2319],[3323,-1.2395],[3326,-1.2261],[3328,-1.2347],[3330,-1.2285],[3332,-1.2339],[3333,-1.2297],[3339,-1.2366],[3341,-1.231],[3343,-1.2376],[3345,-1.2289],[3349,-1.2337],[3351,-1.2308],[3353,-1.233],[3359,-1.2283],[3362,-1.2315],[3367,-1.2265],[3369,-1.2291],[3375,-1.2256],[3376,-1.2233],[3379,-1.2313],[3381,-1.2267],[3384,-1.2303],[3386,-1.2253],[3388,-1.224],[3392,-1.2321],[3395,-1.2232],[3397,-1.2351],[3400,-1.2319],[3403,-1.2395],[3406,-1.2261],[3408,-1.2347],[3410,-1.2285],[3412,-1.2339],[3413,-1.2297],[3419,-1.2366],[3421,-1.231],[3423,-1.2376],[3425,-1.2289],[3428,-1.2358],[3431,-1.2276],[3434,-1.2315],[3436,-1.2278],[3438,-1.2316],[3443,-1.2305],[3444,-1.2288],[3446,-1.23],[3447,-1.228],[3449,-1.23],[3453,-1.2289],[3455,-1.2308],[3457,-1.2293],[3459,-1.2336],[3461,-1.2267],[3464,-1.2303],[3466,-1.2253],[3468,-1.224],[3472,-1.2321],[3475,-1.2232],[3477,-1.2351],[3480,-1.2319],[3483,-1.2395],[3486,-1.2261],[3488,-1.2347],[3490,-1.2285],[3492,-1.2339],[3493,-1.2297],[3499,-1.2366],[3501,-1.231],[3503,-1.2376],[3505,-1.2289],[3508,-1.2358],[3511,-1.2276],[3514,-1.2315],[3516,-1.2278],[3518,-1.2316],[3521,-1.2301],[3523,-1.2311],[3527,-1.2265],[3529,-1.2291],[3535,-1.2256],[3536,-1.2233],[3539,-1.2313],[3541,-1.2267],[3544,-1.2303],[3546,-1.2253],[3548,-1.224],[3552,-1.2321],[3555,-1.2232],[3557,-1.2351],[3560,-1.2319],[3563,-1.2395],[3566,-1.2261],[3568,-1.2347],[3570,-1.2285],[3572,-1.2339],[3573,-1.2297],[3579,-1.2366],[3581,-1.231],[3583,-1.2376],[3585,-1.2289],[3590,-1.2342],[3592,-1.2319],[3593,-1.2345],[3595,-1.2306],[3598,-1.2319],[3602,-1.2279],[3604,-1.2301],[3607,-1.2265],[3609,-1.2291],[3615,-1.2256],[3616,-1.2233],[3618,-1.2275],[3619,-1.2265],[3621,-1.2297],[3622,-1.2279],[3627,-1.2303],[3630,-1.2284],[3634,-1.2312],[3635,-1.2258],[3637,-1.2351],[3640,-1.2319],[3643,-1.2395],[3646,-1.2261],[3648,-1.2347],[3650,-1.2285],[3652,-1.2339],[3653,-1.2297],[3659,-1.2366],[3661,-1.231],[3663,-1.2376],[3665,-1.2289],[3669,-1.2337],[3671,-1.2308],[3673,-1.233],[3679,-1.2283],[3682,-1.2315],[3687,-1.2265],[3689,-1.2291],[3695,-1.2256],[3696,-1.2233],[3699,-1.2313],[3701,-1.2267],[3704,-1.2303],[3706,-1.2253],[3708,-1.224],[3712,-1.2321],[3715,-1.2232],[3717,-1.2351],[3720,-1.2319],[3723,-1.2395],[3726,-1.2261],[3728,-1.2347],[3730,-1.2285],[3732,-1.2339],[3733,-1.2297],[3739,-1.2366],[3741,-1.231],[3743,-1.2376],[3746,-1.2289],[3748,-1.2358],[3751,-1.2276],[3754,-1.2315],[3756,-1.2279],[3758,-1.2316],[3763,-1.2305],[3764,-1.2288],[3766,-1.23],[3767,-1.228],[3769,-1.23],[3773,-1.2289],[3775,-1.2308],[3777,-1.2293],[3779,-1.2336],[3781,-1.2267],[3784,-1.2303],[3786,-1.2253],[3788,-1.2239],[3792,-1.2321],[3795,-1.2232],[3797,-1.2351],[3800,-1.2319],[3803,-1.2395],[3806,-1.2261],[3808,-1.2347],[3810,-1.2285],[3812,-1.2339],[3813,-1.2297],[3819,-1.2366],[3821,-1.231],[3823,-1.2376],[3825,-1.2289],[3828,-1.2358],[3831,-1.2276],[3834,-1.2315],[3836,-1.2278],[3838,-1.2316],[3841,-1.2301],[3843,-1.2311],[3847,-1.2265],[3849,-1.2291],[3855,-1.2256],[3856,-1.2233],[3859,-1.2313],[3861,-1.2267],[3864,-1.2303],[3866,-1.2253],[3868,-1.224],[3872,-1.2321],[3875,-1.2232],[3877,-1.2351],[3880,-1.2318],[3883,-1.2395],[3886,-1.2261],[3888,-1.2347],[3891,-1.228],[3893,-1.23],[3897,-1.2289],[3899,-1.2308],[3901,-1.2293],[3903,-1.2336],[3905,-1.2267],[3908,-1.2303],[3910,-1.2253],[3912,-1.2239],[3916,-1.2321],[3919,-1.2232],[3921,-1.2351],[3924,-1.2319],[3927,-1.2395],[3930,-1.2261],[3932,-1.2347],[3934,-1.2285],[3936,-1.2339],[3937,-1.2297],[3943,-1.2366],[3945,-1.231],[3947,-1.2376],[3949,-1.2289],[3952,-1.2358],[3955,-1.2276],[3958,-1.2315],[3960,-1.2278],[3962,-1.2316],[3965,-1.2301],[3967,-1.2311],[3971,-1.2265],[3973,-1.2291],[3979,-1.2256],[3980,-1.2233],[3983,-1.2313],[3985,-1.2267],[3988,-1.2303],[3990,-1.2253],[3992,-1.224],[3996,-1.2321],[3999,-1.2232],[4001,-1.2351],[4004,-1.2318],[4007,-1.2395],[4010,-1.2261],[4012,-1.2347],[4015,-1.228],[4017,-1.23],[4021,-1.2289],[4023,-1.2308],[4025,-1.2293],[4027,-1.2336],[4029,-1.2267],[4032,-1.2303],[4034,-1.2253],[4036,-1.2239],[4040,-1.2321],[4043,-1.2232],[4045,-1.2351],[4048,-1.2319],[4051,-1.2395],[4054,-1.2261],[4056,-1.2347],[4058,-1.2285],[4060,-1.2339],[4061,-1.2297],[4067,-1.2366],[4069,-1.231],[4071,-1.2376],[4073,-1.2289],[4076,-1.2358],[4079,-1.2276],[4082,-1.2315],[4084,-1.2278],[4086,-1.2316],[4089,-1.2301],[4091,-1.2311],[4095,-1.2265],[4097,-1.2291],[4103,-1.2256],[4104,-1.2233],[4107,-1.2313],[4109,-1.2267],[4112,-1.2303],[4114,-1.2253],[4116,-1.224],[4120,-1.2321],[4123,-1.2232],[4125,-1.2351],[4128,-1.2318],[4131,-1.2395],[4134,-1.2261],[4136,-1.2347],[4140,-1.228],[4142,-1.23],[4146,-1.2289],[4148,-1.2308],[4150,-1.2293],[4152,-1.2336],[4154,-1.2267],[4157,-1.2303],[4159,-1.2253],[4161,-1.2239],[4165,-1.2321],[4168,-1.2232],[4170,-1.2351],[4173,-1.2319],[4176,-1.2395],[4179,-1.2261],[4181,-1.2347],[4183,-1.2285],[4185,-1.2339],[4186,-1.2297],[4192,-1.2366],[4194,-1.231],[4196,-1.2376],[4198,-1.2289],[4201,-1.2358],[4204,-1.2276],[4207,-1.2315],[4209,-1.2278],[4211,-1.2316],[4214,-1.2301],[4216,-1.2311],[4220,-1.2265],[4222,-1.2291],[4228,-1.2256],[4229,-1.2233],[4232,-1.2313],[4234,-1.2267],[4237,-1.2303],[4239,-1.2253],[4241,-1.224],[4245,-1.2321],[4248,-1.2232],[4250,-1.2351],[4253,-1.2318],[4256,-1.2395],[4259,-1.2261],[4261,-1.2347],[4266,-1.228],[4268,-1.23],[4272,-1.2289],[4274,-1.2308],[4276,-1.2293],[4278,-1.2336],[4280,-1.2267],[4283,-1.2303],[4285,-1.2253],[4287,-1.2239],[4291,-1.2321],[4294,-1.2232],[4296,-1.2351],[4299,-1.2319],[4302,-1.2395],[4305,-1.2261],[4307,-1.2347],[4309,-1.2285],[4311,-1.2339],[4312,-1.2297],[4318,-1.2366],[4320,-1.231],[4322,-1.2376],[4324,-1.2289],[4327,-1.2358],[4330,-1.2276],[4333,-1.2315],[4335,-1.2278],[4337,-1.2316],[4340,-1.2301],[4342,-1.2311],[4346,-1.2265],[4348,-1.2291],[4354,-1.2256],[4355,-1.2233],[4360,-1.2316],[4365,-1.2285],[4367,-1.2339],[4372,-1.2376],[4381,-1.2319],[4382,-1.2345],[4384,-1.2306],[4387,-1.2319],[4391,-1.2279],[4393,-1.2301],[4396,-1.2265],[4398,-1.2291],[4404,-1.2256],[4405,-1.2233],[4407,-1.2275],[4408,-1.2265],[4410,-1.2297],[4411,-1.2279],[4416,-1.2303],[4419,-1.2284],[4423,-1.2312],[4424,-1.2258],[4426,-1.2351],[4429,-1.2319],[4432,-1.2395],[4435,-1.2261],[4436,0.0007]]
       }
       ring_01_l_ctrl.Rotation.Z {
        type=MovieSceneScriptingFloatChannel
        default=-0.9643
        thinned_from=92
        keys [[1790,-0.9643],[1830,8.2885],[1867,-1.0856],[1910,8.2885],[1946,-1.0858],[1947,-1.0855],[1990,8.2885],[2026,-1.0858],[2027,-1.0855],[2070,8.2885],[2106,-1.0858],[2107,-1.0855],[2150,8.2885],[2187,-1.0856],[2230,8.2885],[2266,-1.0858],[2267,-1.0855],[2310,8.2885],[2346,-1.0858],[2347,-1.0855],[2390,8.2885],[2426,-1.0858],[2427,-1.0855],[2467,8.2408],[2468,8.2071],[2503,3.7156],[2505,3.7156],[2510,-48.8799],[2513,-53.6898],[2515,-53.6893],[2518,-33.0488],[2520,27.7995],[2522,25.0352],[2524,26.8596],[2530,24.297],[2763,33.2083],[2794,-35.1328],[2938,-34.6564],[2986,55.1119],[3058,55.1119],[3070,8.7454],[3185,8.7454],[3221,-1.0858],[3222,-1.0855],[3265,8.2885],[3302,-1.0857],[3345,8.2885],[3381,-1.0858],[3382,-1.0855],[3425,8.2885],[3461,-1.0858],[3462,-1.0855],[3505,8.2885],[3541,-1.0858],[3542,-1.0855],[3585,8.2885],[3622,-1.0856],[3665,8.2885],[3701,-1.0858],[3702,-1.0855],[3745,8.2885],[3781,-1.0858],[3782,-1.0855],[3825,8.2885],[3861,-1.0858],[3906,-1.0855],[3949,8.2885],[3985,-1.0858],[4030,-1.0855],[4073,8.2885],[4109,-1.0858],[4155,-1.0855],[4198,8.2885],[4234,-1.0858],[4281,-1.0855],[4324,8.2885],[4360,5.3622],[4364,4.8352],[4411,-1.0856],[4435,6.6273],[4436,0.0015]]
       }
       ring_01_l_ctrl.Scale.X {type=MovieSceneScriptingFloatChannel, default=1, const [1,12]}
       ring_01_l_ctrl.Scale.Y {type=MovieSceneScriptingFloatChannel, default=1, const [1,12]}
       ring_01_l_ctrl.Scale.Z {type=MovieSceneScriptingFloatChannel, default=1, const [1,12]}
       ring_02_l_ctrl.Location.X {type=MovieSceneScriptingFloatChannel, default=0, thinned_from=252, keys [[1790,0],[2512,0],[2513,-0.0002],[2529,0],[4436,0]]}
       ring_02_l_ctrl.Location.Y {type=MovieSceneScriptingFloatChannel, default=0, thinned_from=309, keys [[1790,0],[2512,0],[2513,-0.0001],[2529,0],[4436,0]]}
       ring_02_l_ctrl.Location.Z {type=MovieSceneScriptingFloatChannel, default=0, thinned_from=1332, keys [[1790,0],[4436,0]]}
       ring_02_l_ctrl.Rotation.X {
        type=MovieSceneScriptingFloatChannel
        default=-1.088
        thinned_from=121
        keys [[1790,-1.088],[1805,-1.2076],[1830,-1.2594],[1844,-1.1383],[1868,-1.0853],[1908,-1.2586],[1911,-1.2604],[1924,-1.1391],[1948,-1.0853],[1988,-1.2586],[1991,-1.2604],[2004,-1.1376],[2028,-1.0853],[2070,-1.2594],[2084,-1.1376],[2109,-1.0853],[2124,-1.2017],[2150,-1.2594],[2164,-1.1383],[2188,-1.0853],[2228,-1.2586],[2231,-1.2604],[2244,-1.1391],[2268,-1.0853],[2308,-1.2586],[2311,-1.2604],[2324,-1.1376],[2348,-1.0853],[2390,-1.2594],[2404,-1.1376],[2429,-1.0853],[2466,-1.2566],[2471,-1.185],[2503,0],[2515,0],[2519,-0.1815],[2522,-0.0392],[2524,-0.1335],[2526,-0.0366],[2532,-0.0001],[2760,-0.3294],[2765,-0.4918],[2780,-1.1129],[2788,-0.8752],[2794,-0.6811],[2927,-0.6219],[2938,-0.6211],[2986,-0.3573],[3058,-0.3573],[3070,-0.3649],[3185,-0.3649],[3222,-1.0854],[3224,-1.0853],[3239,-1.2017],[3265,-1.2594],[3279,-1.1383],[3303,-1.0853],[3343,-1.2586],[3346,-1.2604],[3359,-1.1391],[3383,-1.0853],[3423,-1.2586],[3426,-1.2604],[3439,-1.1376],[3463,-1.0853],[3505,-1.2594],[3519,-1.1376],[3544,-1.0853],[3559,-1.2017],[3585,-1.2594],[3599,-1.1383],[3623,-1.0853],[3663,-1.2586],[3666,-1.2604],[3679,-1.1391],[3703,-1.0853],[3743,-1.2586],[3746,-1.2604],[3759,-1.1376],[3783,-1.0853],[3825,-1.2594],[3839,-1.1376],[3864,-1.0853],[3879,-1.2017],[3907,-1.0853],[3949,-1.2594],[3963,-1.1376],[3988,-1.0853],[4003,-1.2017],[4031,-1.0853],[4073,-1.2594],[4087,-1.1376],[4112,-1.0853],[4127,-1.2017],[4156,-1.0853],[4198,-1.2594],[4212,-1.1376],[4237,-1.0853],[4252,-1.2017],[4282,-1.0853],[4324,-1.2594],[4338,-1.1376],[4360,-1.2334],[4364,-1.2297],[4388,-1.1383],[4412,-1.0853],[4435,-1.2212],[4436,0]]
       }
       ring_02_l_ctrl.Rotation.Y {
        type=MovieSceneScriptingFloatChannel
        default=0.1657
        thinned_from=222
        keys [[1790,0.1657],[1828,0.1925],[1831,0.1927],[1863,0.166],[1868,0.166],[1897,0.1905],[1898,0.1905],[1904,0.1924],[1906,0.1922],[1908,0.1928],[1912,0.192],[1913,0.1921],[1944,0.1659],[1947,0.1659],[1977,0.1905],[1978,0.1905],[1984,0.1924],[1986,0.1922],[1988,0.1928],[1991,0.1923],[1992,0.1924],[2022,0.1662],[2027,0.1659],[2068,0.1925],[2071,0.1927],[2105,0.1657],[2108,0.1662],[2110,0.1657],[2148,0.1925],[2151,0.1927],[2183,0.166],[2188,0.166],[2217,0.1905],[2218,0.1905],[2224,0.1924],[2226,0.1922],[2228,0.1928],[2232,0.192],[2233,0.1921],[2264,0.1659],[2267,0.1659],[2297,0.1905],[2298,0.1905],[2304,0.1924],[2306,0.1922],[2308,0.1928],[2311,0.1923],[2312,0.1924],[2342,0.1662],[2347,0.1659],[2388,0.1925],[2391,0.1927],[2425,0.1657],[2428,0.1662],[2430,0.1657],[2466,0.1924],[2471,0.1855],[2503,0],[2515,0],[2519,0.066],[2520,0.0662],[2521,0.02],[2522,0.0135],[2524,0.0511],[2527,0.005],[2530,0.0002],[2532,0.0002],[2759,0.0903],[2762,0.1696],[2765,0.1798],[2776,0.189],[2790,0.1181],[2791,0.1186],[2796,0.0713],[2801,0.0562],[2802,0.0575],[2938,0.0574],[2986,0.1393],[3058,0.1393],[3070,0.0523],[3185,0.0523],[3220,0.1657],[3222,0.1662],[3225,0.1657],[3263,0.1925],[3266,0.1927],[3298,0.166],[3303,0.166],[3332,0.1905],[3333,0.1905],[3339,0.1924],[3341,0.1922],[3343,0.1928],[3347,0.192],[3348,0.1921],[3379,0.1659],[3382,0.1659],[3412,0.1905],[3413,0.1905],[3419,0.1924],[3421,0.1922],[3423,0.1928],[3426,0.1923],[3427,0.1924],[3457,0.1662],[3462,0.1659],[3503,0.1925],[3506,0.1927],[3540,0.1657],[3542,0.1662],[3545,0.1657],[3583,0.1925],[3586,0.1927],[3618,0.166],[3623,0.166],[3652,0.1905],[3653,0.1905],[3659,0.1924],[3661,0.1922],[3663,0.1928],[3667,0.192],[3668,0.1921],[3699,0.1659],[3702,0.1659],[3732,0.1905],[3733,0.1905],[3739,0.1924],[3741,0.1922],[3743,0.1928],[3746,0.1923],[3747,0.1924],[3777,0.1662],[3782,0.1659],[3823,0.1925],[3826,0.1927],[3860,0.1657],[3863,0.1662],[3865,0.1657],[3902,0.1662],[3906,0.1659],[3947,0.1925],[3950,0.1927],[3984,0.1657],[3987,0.1662],[3989,0.1657],[4026,0.1662],[4030,0.1659],[4071,0.1925],[4074,0.1927],[4108,0.1657],[4111,0.1662],[4113,0.1657],[4151,0.1662],[4155,0.1659],[4196,0.1925],[4199,0.1927],[4233,0.1657],[4236,0.1662],[4238,0.1657],[4277,0.1662],[4281,0.1659],[4322,0.1925],[4325,0.1927],[4360,0.1843],[4364,0.1864],[4372,0.1925],[4375,0.1927],[4407,0.166],[4412,0.166],[4435,0.188],[4436,0]]
       }
       ring_02_l_ctrl.Rotation.Z {
        type=MovieSceneScriptingFloatChannel
        default=4.442
        thinned_from=92
        keys [[1790,4.442],[1831,15.7744],[1867,4.3677],[1909,15.7363],[1945,4.3677],[1946,4.3677],[1989,15.7363],[2025,4.3677],[2026,4.3677],[2071,15.7744],[2105,4.3677],[2106,4.3677],[2151,15.7744],[2187,4.3677],[2229,15.7363],[2265,4.3677],[2266,4.3677],[2309,15.7363],[2345,4.3677],[2346,4.3677],[2391,15.7744],[2425,4.3677],[2426,4.3677],[2467,15.6571],[2469,15.432],[2471,15.5709],[2503,-15.2949],[2512,-21.0269],[2515,-21.0267],[2519,-0.2246],[2520,0.2663],[2521,-12.5549],[2522,-14.8475],[2524,-4.6898],[2527,-17.5773],[2530,-18.9352],[2764,30.4443],[2795,-24.6004],[2938,-24.4138],[2986,40.0869],[3058,40.0869],[3070,-3.9714],[3185,-3.9714],[3220,4.3677],[3221,4.3677],[3266,15.7744],[3302,4.3677],[3344,15.7363],[3380,4.3677],[3424,15.7363],[3460,4.3677],[3461,4.3677],[3506,15.7744],[3540,4.3677],[3586,15.7744],[3622,4.3677],[3664,15.7363],[3700,4.3677],[3701,4.3677],[3744,15.7363],[3780,4.3677],[3781,4.3677],[3826,15.7744],[3860,4.3677],[3905,4.3677],[3950,15.7744],[3984,4.3677],[4029,4.3677],[4074,15.7744],[4108,4.3677],[4154,4.3677],[4199,15.7744],[4233,4.3677],[4280,4.3677],[4325,15.7744],[4360,12.1206],[4364,12.7615],[4375,15.7744],[4411,4.3677],[4435,13.7529],[4436,0]]
       }
       ring_02_l_ctrl.Scale.X {type=MovieSceneScriptingFloatChannel, default=1, const [1,12]}
       ring_02_l_ctrl.Scale.Y {type=MovieSceneScriptingFloatChannel, default=1, const [1,12]}
       ring_02_l_ctrl.Scale.Z {type=MovieSceneScriptingFloatChannel, default=1, const [1,12]}
       ring_03_l_ctrl.Location.X {type=MovieSceneScriptingFloatChannel, default=0, thinned_from=68, keys [[1790,0],[2512,0],[2513,-0.0001],[2529,0],[4435,0],[4436,-0.0001]]}
       ring_03_l_ctrl.Location.Y {type=MovieSceneScriptingFloatChannel, default=0, thinned_from=265, keys [[1790,0],[4436,0]]}
       ring_03_l_ctrl.Location.Z {type=MovieSceneScriptingFloatChannel, default=0, thinned_from=1426, keys [[1790,0],[4436,0]]}
       ring_03_l_ctrl.Rotation.X {
        type=MovieSceneScriptingFloatChannel
        default=-0.4798
        thinned_from=147
        keys [[1790,-0.4798],[1828,-0.5551],[1831,-0.5558],[1844,-0.5015],[1868,-0.4787],[1908,-0.5551],[1911,-0.5558],[1925,-0.4988],[1949,-0.4787],[1983,-0.5527],[1987,-0.5542],[1989,-0.5558],[1990,-0.5558],[2004,-0.5015],[2028,-0.4787],[2063,-0.5527],[2064,-0.5527],[2069,-0.5558],[2070,-0.5558],[2084,-0.5015],[2108,-0.4787],[2148,-0.5551],[2151,-0.5558],[2164,-0.5015],[2188,-0.4787],[2228,-0.5551],[2231,-0.5558],[2245,-0.4988],[2269,-0.4787],[2303,-0.5527],[2307,-0.5542],[2309,-0.5558],[2310,-0.5558],[2324,-0.5015],[2348,-0.4787],[2383,-0.5527],[2384,-0.5527],[2389,-0.5558],[2390,-0.5558],[2404,-0.5015],[2428,-0.4787],[2466,-0.5544],[2503,0],[2515,0],[2519,-0.0807],[2522,-0.0168],[2524,-0.0597],[2529,0],[2759,-0.1041],[2761,-0.177],[2769,-0.2857],[2780,-0.4907],[2788,-0.386],[2794,-0.3003],[2927,-0.274],[2938,-0.2734],[2986,-0.3969],[3058,-0.3969],[3070,-0.1589],[3185,-0.1589],[3223,-0.4787],[3263,-0.5551],[3266,-0.5558],[3279,-0.5015],[3302,-0.4787],[3304,-0.4787],[3343,-0.5551],[3346,-0.5558],[3359,-0.5016],[3383,-0.4787],[3418,-0.5527],[3419,-0.5527],[3424,-0.5558],[3440,-0.4982],[3463,-0.4787],[3498,-0.5527],[3499,-0.5527],[3504,-0.5558],[3505,-0.5558],[3519,-0.5015],[3543,-0.4787],[3583,-0.5551],[3586,-0.5558],[3599,-0.5015],[3623,-0.4787],[3663,-0.5551],[3666,-0.5558],[3680,-0.4988],[3704,-0.4787],[3738,-0.5527],[3739,-0.5527],[3744,-0.5558],[3760,-0.4982],[3783,-0.4787],[3818,-0.5527],[3819,-0.5527],[3824,-0.5558],[3840,-0.4982],[3863,-0.4787],[3907,-0.4787],[3942,-0.5527],[3943,-0.5527],[3948,-0.5558],[3964,-0.4982],[3987,-0.4787],[4031,-0.4787],[4066,-0.5527],[4067,-0.5527],[4072,-0.5558],[4088,-0.4982],[4111,-0.4787],[4156,-0.4787],[4191,-0.5527],[4192,-0.5527],[4197,-0.5558],[4213,-0.4982],[4236,-0.4787],[4282,-0.4787],[4317,-0.5527],[4318,-0.5527],[4323,-0.5558],[4339,-0.4982],[4360,-0.5335],[4364,-0.5436],[4372,-0.5551],[4375,-0.5558],[4388,-0.5015],[4412,-0.4787],[4435,-0.5386],[4436,0.0001]]
       }
       ring_03_l_ctrl.Rotation.Y {
        type=MovieSceneScriptingFloatChannel
        default=0.0664
        thinned_from=338
        keys [[1790,0.0664],[1818,0.0762],[1819,0.0761],[1826,0.0768],[1827,0.0767],[1830,0.0769],[1852,0.0674],[1853,0.0675],[1856,0.0664],[1857,0.0666],[1860,0.0662],[1861,0.0665],[1865,0.0662],[1869,0.0663],[1898,0.0762],[1899,0.0761],[1906,0.0768],[1907,0.0767],[1912,0.077],[1913,0.0768],[1914,0.077],[1936,0.0667],[1937,0.0668],[1941,0.0664],[1943,0.0664],[1945,0.0662],[1948,0.0663],[1986,0.0768],[1988,0.0765],[1990,0.077],[2012,0.0674],[2013,0.0675],[2016,0.0664],[2017,0.0666],[2020,0.0662],[2021,0.0665],[2025,0.0661],[2047,0.074],[2048,0.074],[2054,0.0759],[2055,0.0757],[2067,0.0768],[2068,0.0765],[2070,0.077],[2092,0.0674],[2093,0.0675],[2096,0.0664],[2097,0.0666],[2100,0.0662],[2101,0.0665],[2105,0.0661],[2138,0.0762],[2139,0.0761],[2146,0.0768],[2147,0.0767],[2172,0.0674],[2173,0.0675],[2176,0.0664],[2177,0.0666],[2180,0.0662],[2181,0.0665],[2185,0.0662],[2189,0.0663],[2218,0.0762],[2219,0.0761],[2226,0.0768],[2227,0.0767],[2232,0.077],[2233,0.0768],[2234,0.077],[2256,0.0667],[2257,0.0668],[2261,0.0664],[2263,0.0664],[2265,0.0662],[2268,0.0663],[2306,0.0768],[2308,0.0765],[2310,0.077],[2332,0.0674],[2333,0.0675],[2336,0.0664],[2337,0.0666],[2340,0.0662],[2341,0.0665],[2345,0.0661],[2367,0.074],[2368,0.074],[2374,0.0759],[2375,0.0757],[2387,0.0768],[2388,0.0765],[2390,0.077],[2412,0.0674],[2413,0.0675],[2416,0.0664],[2417,0.0666],[2420,0.0662],[2421,0.0665],[2425,0.0661],[2458,0.0762],[2459,0.0761],[2466,0.0768],[2503,0],[2513,0],[2518,0.0072],[2519,0.0269],[2520,0.029],[2521,0.0091],[2522,0.006],[2523,0.0189],[2524,0.0221],[2527,0.0018],[2530,0],[2757,0.0103],[2760,0.0538],[2762,0.072],[2765,0.0763],[2766,0.0759],[2768,0.0766],[2771,0.0763],[2774,0.0767],[2789,0.0465],[2791,0.0473],[2795,0.0298],[2800,0.0214],[2802,0.0229],[2938,0.0229],[2986,0.1027],[3058,0.1027],[3070,0.0236],[3185,0.0236],[3215,0.0662],[3216,0.0665],[3220,0.0661],[3253,0.0762],[3254,0.0761],[3261,0.0768],[3262,0.0767],[3287,0.0674],[3288,0.0675],[3291,0.0664],[3292,0.0666],[3295,0.0662],[3296,0.0665],[3300,0.0662],[3302,0.0663],[3333,0.0762],[3334,0.0761],[3341,0.0768],[3342,0.0767],[3347,0.077],[3348,0.0768],[3349,0.077],[3371,0.0667],[3372,0.0668],[3376,0.0664],[3378,0.0664],[3380,0.0662],[3382,0.0663],[3421,0.0768],[3423,0.0765],[3425,0.077],[3447,0.0674],[3448,0.0675],[3451,0.0664],[3452,0.0666],[3455,0.0662],[3456,0.0665],[3460,0.0661],[3482,0.074],[3483,0.074],[3489,0.0759],[3490,0.0757],[3501,0.0768],[3503,0.0765],[3505,0.077],[3527,0.0674],[3528,0.0675],[3531,0.0664],[3532,0.0666],[3535,0.0662],[3536,0.0665],[3540,0.0661],[3573,0.0762],[3574,0.0761],[3581,0.0768],[3582,0.0767],[3585,0.0769],[3607,0.0674],[3608,0.0675],[3611,0.0664],[3612,0.0666],[3615,0.0662],[3616,0.0665],[3620,0.0662],[3653,0.0762],[3654,0.0761],[3661,0.0768],[3662,0.0767],[3667,0.077],[3668,0.0768],[3669,0.077],[3691,0.0667],[3692,0.0668],[3696,0.0664],[3698,0.0664],[3700,0.0662],[3702,0.0663],[3742,0.0768],[3743,0.0765],[3746,0.077],[3767,0.0674],[3768,0.0675],[3771,0.0664],[3772,0.0666],[3775,0.0662],[3776,0.0665],[3780,0.0661],[3802,0.074],[3803,0.074],[3809,0.0759],[3810,0.0757],[3821,0.0768],[3823,0.0765],[3825,0.077],[3847,0.0674],[3848,0.0675],[3851,0.0664],[3852,0.0666],[3855,0.0662],[3856,0.0665],[3860,0.0661],[3892,0.0675],[3895,0.0664],[3896,0.0666],[3899,0.0662],[3900,0.0665],[3904,0.0661],[3926,0.074],[3927,0.074],[3933,0.0759],[3934,0.0757],[3945,0.0768],[3947,0.0765],[3949,0.077],[3971,0.0674],[3972,0.0675],[3975,0.0664],[3976,0.0666],[3979,0.0662],[3980,0.0665],[3984,0.0661],[4016,0.0675],[4019,0.0664],[4020,0.0666],[4023,0.0662],[4024,0.0665],[4028,0.0661],[4050,0.074],[4051,0.074],[4057,0.0759],[4058,0.0757],[4069,0.0768],[4071,0.0765],[4073,0.077],[4095,0.0674],[4096,0.0675],[4099,0.0664],[4100,0.0666],[4103,0.0662],[4104,0.0665],[4108,0.0661],[4141,0.0675],[4144,0.0664],[4145,0.0666],[4148,0.0662],[4149,0.0665],[4153,0.0661],[4175,0.074],[4176,0.074],[4182,0.0759],[4183,0.0757],[4194,0.0768],[4196,0.0765],[4198,0.077],[4220,0.0674],[4221,0.0675],[4224,0.0664],[4225,0.0666],[4228,0.0662],[4229,0.0665],[4233,0.0661],[4267,0.0675],[4270,0.0664],[4271,0.0666],[4274,0.0662],[4275,0.0665],[4279,0.0661],[4301,0.074],[4302,0.074],[4308,0.0759],[4309,0.0757],[4320,0.0768],[4322,0.0765],[4324,0.077],[4346,0.0674],[4347,0.0675],[4350,0.0664],[4351,0.0666],[4354,0.0662],[4355,0.0665],[4360,0.0758],[4364,0.0762],[4371,0.0767],[4396,0.0674],[4397,0.0675],[4400,0.0664],[4401,0.0666],[4404,0.0662],[4405,0.0665],[4409,0.0662],[4435,0.075],[4436,0.0003]]
       }
       ring_03_l_ctrl.Rotation.Z {
        type=MovieSceneScriptingFloatChannel
        default=4.4894
        thinned_from=95
        keys [[1790,4.4894],[1829,16.7422],[1860,4.5332],[1861,4.5332],[1865,4.3311],[1909,16.7422],[1912,16.7568],[1945,4.3311],[1991,16.8202],[2020,4.5332],[2025,4.3311],[2026,4.3311],[2067,16.653],[2071,16.8202],[2100,4.5332],[2105,4.3311],[2106,4.3311],[2150,16.7422],[2185,4.3311],[2229,16.7422],[2232,16.7568],[2266,4.3311],[2311,16.8202],[2340,4.5332],[2345,4.3311],[2346,4.3311],[2387,16.653],[2391,16.8202],[2420,4.5332],[2426,4.3311],[2466,16.6247],[2503,-33.0584],[2513,-28.8666],[2515,-29.1357],[2520,12.3296],[2522,6.4549],[2524,10.4038],[2531,4.8082],[2764,24.0226],[2927,16.9101],[2938,16.791],[2986,52.1659],[3058,52.1659],[3070,-7.5067],[3185,-7.5067],[3220,4.3311],[3221,4.3311],[3264,16.7422],[3301,4.3311],[3345,16.7422],[3347,16.7568],[3381,4.3311],[3426,16.8202],[3460,4.3311],[3461,4.3311],[3505,16.8202],[3541,4.3311],[3585,16.7422],[3621,4.3311],[3665,16.7422],[3667,16.7568],[3700,4.3311],[3746,16.8202],[3780,4.3311],[3781,4.3311],[3825,16.8202],[3860,4.3311],[3905,4.3311],[3949,16.8202],[3984,4.3311],[4029,4.3311],[4073,16.8202],[4108,4.3311],[4154,4.3311],[4198,16.8202],[4233,4.3311],[4280,4.3311],[4324,16.8202],[4360,13.265],[4364,14.05],[4373,16.7422],[4404,4.5332],[4410,4.3311],[4435,14.5735],[4436,0.0037]]
       }
       ring_03_l_ctrl.Scale.X {type=MovieSceneScriptingFloatChannel, default=1, const [1,12]}
       ring_03_l_ctrl.Scale.Y {type=MovieSceneScriptingFloatChannel, default=1, const [1,12]}
       ring_03_l_ctrl.Scale.Z {type=MovieSceneScriptingFloatChannel, default=1, const [1,12]}
       pinky_metacarpal_l_ctrl.Location.X {type=MovieSceneScriptingFloatChannel, default=0, thinned_from=32, keys [[1790,0],[4436,0]]}
       pinky_metacarpal_l_ctrl.Location.Y {type=MovieSceneScriptingFloatChannel, default=0, thinned_from=32, keys [[1790,0],[4436,0]]}
       pinky_metacarpal_l_ctrl.Location.Z {type=MovieSceneScriptingFloatChannel, default=0, thinned_from=33, keys [[1790,0],[4436,0]]}
       pinky_metacarpal_l_ctrl.Rotation.X {
        type=MovieSceneScriptingFloatChannel
        default=-18.9529
        thinned_from=122
        keys [[1790,-18.9529],[1825,-16.6678],[1828,-16.6226],[1867,-18.9881],[1905,-16.6678],[1908,-16.6226],[1947,-18.9881],[1985,-16.6678],[1988,-16.6226],[2024,-18.9407],[2028,-19.0053],[2065,-16.6678],[2068,-16.6226],[2088,-18.5141],[2089,-18.5136],[2108,-19.0053],[2145,-16.6678],[2148,-16.6226],[2187,-18.9881],[2225,-16.6678],[2228,-16.6226],[2267,-18.9881],[2308,-16.6226],[2344,-18.9407],[2348,-19.0053],[2385,-16.6678],[2389,-16.6226],[2408,-18.5141],[2409,-18.5136],[2428,-19.0053],[2457,-16.9169],[2503,-10.3781],[2506,-10.3781],[2507,-8.7068],[2510,-2.5567],[2513,-1.8322],[2518,-3.6533],[2519,-8.1505],[2521,-14.8028],[2523,-13.0012],[2527,-15.2751],[2532,-15.5447],[2763,-6.5718],[2774,-10.8531],[2777,-9.5765],[2781,-3.4147],[2788,0],[3185,0],[3223,-19.0053],[3264,-16.6226],[3302,-18.9881],[3340,-16.6678],[3344,-16.6226],[3383,-18.9881],[3423,-16.6226],[3459,-18.9407],[3463,-19.0053],[3504,-16.6226],[3523,-18.5141],[3524,-18.5136],[3543,-19.0053],[3580,-16.6678],[3584,-16.6226],[3623,-18.9881],[3663,-16.6226],[3703,-18.9881],[3740,-16.6678],[3744,-16.6226],[3779,-18.9407],[3783,-19.0053],[3823,-16.6226],[3843,-18.5141],[3844,-18.5136],[3863,-19.0053],[3904,-18.9382],[3907,-19.0053],[3947,-16.6226],[3967,-18.5141],[3968,-18.5136],[3987,-19.0053],[4028,-18.9382],[4031,-19.0053],[4071,-16.6226],[4091,-18.5141],[4092,-18.5136],[4111,-19.0053],[4153,-18.9382],[4156,-19.0053],[4196,-16.6226],[4216,-18.5141],[4217,-18.5136],[4236,-19.0053],[4279,-18.9382],[4282,-19.0053],[4322,-16.6226],[4342,-18.5141],[4343,-18.5136],[4360,-17.2939],[4364,-16.9707],[4372,-16.6226],[4412,-18.9881],[4424,-17.8648],[4430,-17.2667],[4435,-17.1462],[4436,0]]
       }
       pinky_metacarpal_l_ctrl.Rotation.Y {
        type=MovieSceneScriptingFloatChannel
        default=1.5773
        thinned_from=242
        keys [[1790,1.5773],[1805,2.2894],[1825,2.634],[1828,2.6485],[1856,1.6119],[1857,1.6135],[1859,1.5855],[1860,1.587],[1862,1.5592],[1867,1.5632],[1905,2.634],[1908,2.6485],[1936,1.6119],[1937,1.6135],[1939,1.5855],[1940,1.587],[1942,1.5592],[1947,1.5632],[1985,2.634],[1988,2.6485],[2016,1.6119],[2017,1.6135],[2019,1.5855],[2020,1.587],[2022,1.5592],[2023,1.5607],[2025,1.5528],[2027,1.5564],[2065,2.634],[2068,2.6485],[2096,1.6119],[2097,1.6135],[2099,1.5855],[2100,1.587],[2102,1.5592],[2105,1.5695],[2107,1.5555],[2145,2.634],[2148,2.6485],[2176,1.6119],[2177,1.6135],[2179,1.5855],[2180,1.587],[2182,1.5592],[2187,1.5632],[2225,2.634],[2228,2.6485],[2256,1.6119],[2257,1.6135],[2259,1.5855],[2260,1.587],[2262,1.5592],[2267,1.5632],[2305,2.634],[2308,2.6485],[2336,1.6119],[2337,1.6135],[2339,1.5855],[2340,1.587],[2342,1.5592],[2343,1.5607],[2345,1.5528],[2347,1.5564],[2385,2.634],[2388,2.6485],[2416,1.6119],[2417,1.6135],[2419,1.5855],[2420,1.587],[2422,1.5592],[2425,1.5695],[2427,1.5555],[2465,2.634],[2467,2.6392],[2469,2.6366],[2471,2.6935],[2501,1.8352],[2503,1.7234],[2505,1.7234],[2510,-1.6795],[2513,-1.9612],[2515,-1.9612],[2518,-0.5876],[2520,3.2828],[2522,2.7008],[2524,3.0817],[2531,2.5404],[2764,4.3893],[2776,2.7601],[2782,0.48],[2788,0],[3185,0],[3215,1.587],[3217,1.5592],[3220,1.5695],[3222,1.5555],[3264,2.6485],[3291,1.6119],[3292,1.6135],[3294,1.5855],[3295,1.587],[3297,1.5592],[3302,1.5632],[3340,2.634],[3343,2.6485],[3371,1.6119],[3372,1.6135],[3374,1.5855],[3375,1.587],[3377,1.5592],[3382,1.5632],[3423,2.6485],[3451,1.6119],[3452,1.6134],[3454,1.5855],[3455,1.587],[3457,1.5592],[3458,1.5607],[3460,1.5528],[3463,1.5564],[3504,2.6485],[3531,1.6119],[3532,1.6135],[3534,1.5855],[3535,1.587],[3537,1.5592],[3540,1.5695],[3542,1.5555],[3580,2.634],[3583,2.6485],[3611,1.6119],[3612,1.6135],[3614,1.5855],[3615,1.587],[3617,1.5592],[3622,1.5632],[3663,2.6485],[3691,1.6119],[3692,1.6134],[3694,1.5855],[3695,1.587],[3697,1.5592],[3702,1.5632],[3740,2.634],[3744,2.6485],[3771,1.6119],[3772,1.6135],[3774,1.5855],[3775,1.587],[3777,1.5592],[3778,1.5607],[3780,1.5528],[3782,1.5564],[3823,2.6485],[3851,1.612],[3852,1.6135],[3854,1.5855],[3855,1.587],[3857,1.5592],[3860,1.5695],[3862,1.5555],[3896,1.6135],[3898,1.5855],[3899,1.587],[3901,1.5592],[3902,1.5607],[3904,1.5528],[3906,1.5564],[3947,2.6485],[3975,1.612],[3976,1.6135],[3978,1.5855],[3979,1.587],[3981,1.5592],[3984,1.5695],[3986,1.5555],[4020,1.6135],[4022,1.5855],[4023,1.587],[4025,1.5592],[4026,1.5607],[4028,1.5528],[4030,1.5564],[4071,2.6485],[4099,1.612],[4100,1.6135],[4102,1.5855],[4103,1.587],[4105,1.5592],[4108,1.5695],[4110,1.5555],[4145,1.6135],[4147,1.5855],[4148,1.587],[4150,1.5592],[4151,1.5607],[4153,1.5528],[4155,1.5564],[4196,2.6485],[4224,1.612],[4225,1.6135],[4227,1.5855],[4228,1.587],[4230,1.5592],[4233,1.5695],[4235,1.5555],[4271,1.6135],[4273,1.5855],[4274,1.587],[4276,1.5592],[4277,1.5607],[4279,1.5528],[4281,1.5564],[4322,2.6485],[4350,1.612],[4351,1.6135],[4353,1.5855],[4354,1.587],[4356,1.5592],[4360,2.3749],[4364,2.5067],[4372,2.6485],[4400,1.6119],[4401,1.6134],[4403,1.5855],[4404,1.587],[4406,1.5592],[4411,1.5632],[4435,2.4542],[4436,0]]
       }
       pinky_metacarpal_l_ctrl.Rotation.Z {
        type=MovieSceneScriptingFloatChannel
        default=-2.1577
        thinned_from=233
        keys [[1790,-2.1577],[1825,-3.7941],[1828,-3.816],[1856,-2.2114],[1857,-2.214],[1859,-2.17],[1860,-2.1726],[1862,-2.1286],[1867,-2.1354],[1905,-3.7941],[1908,-3.816],[1936,-2.2114],[1937,-2.214],[1939,-2.17],[1940,-2.1726],[1942,-2.1286],[1947,-2.1354],[1985,-3.7941],[1988,-3.816],[2016,-2.2114],[2017,-2.214],[2019,-2.17],[2020,-2.1726],[2022,-2.1286],[2023,-2.1312],[2025,-2.1183],[2028,-2.1244],[2065,-3.7941],[2068,-3.816],[2096,-2.2114],[2097,-2.214],[2099,-2.17],[2100,-2.1726],[2102,-2.1286],[2105,-2.1451],[2107,-2.1229],[2145,-3.7941],[2148,-3.816],[2176,-2.2114],[2177,-2.214],[2179,-2.17],[2180,-2.1726],[2182,-2.1286],[2187,-2.1354],[2228,-3.816],[2256,-2.2114],[2257,-2.214],[2259,-2.17],[2260,-2.1726],[2262,-2.1286],[2267,-2.1354],[2308,-3.816],[2336,-2.2114],[2337,-2.214],[2339,-2.17],[2340,-2.1726],[2342,-2.1286],[2343,-2.1312],[2345,-2.1183],[2348,-2.1244],[2385,-3.7941],[2389,-3.816],[2416,-2.2114],[2417,-2.214],[2419,-2.17],[2420,-2.1726],[2422,-2.1286],[2425,-2.1451],[2427,-2.1229],[2469,-3.8358],[2470,-3.8295],[2503,-7.3783],[2506,-7.3783],[2507,-6.6471],[2510,-3.9978],[2513,-3.6926],[2519,-4.9675],[2520,-7.0569],[2521,-9.3269],[2522,-9.3958],[2523,-7.9325],[2524,-7.8779],[2527,-10.0137],[2531,-10.2481],[2759,-5.7141],[2761,-2.7436],[2764,-1.5141],[2774,-2.4703],[2781,-0.7179],[2788,0],[3185,0],[3215,-2.1726],[3217,-2.1286],[3220,-2.1451],[3222,-2.1229],[3264,-3.816],[3291,-2.2114],[3292,-2.214],[3294,-2.17],[3295,-2.1726],[3297,-2.1286],[3302,-2.1354],[3344,-3.816],[3371,-2.2114],[3372,-2.214],[3374,-2.17],[3375,-2.1726],[3377,-2.1286],[3382,-2.1354],[3423,-3.816],[3451,-2.2114],[3452,-2.214],[3454,-2.17],[3455,-2.1726],[3457,-2.1286],[3458,-2.1312],[3460,-2.1183],[3463,-2.1244],[3504,-3.816],[3531,-2.2114],[3532,-2.214],[3534,-2.17],[3535,-2.1726],[3537,-2.1286],[3540,-2.1451],[3542,-2.1229],[3584,-3.816],[3611,-2.2114],[3612,-2.214],[3614,-2.17],[3615,-2.1726],[3617,-2.1286],[3622,-2.1354],[3663,-3.816],[3691,-2.2114],[3692,-2.214],[3694,-2.17],[3695,-2.1726],[3697,-2.1286],[3702,-2.1354],[3740,-3.7941],[3744,-3.816],[3771,-2.2114],[3772,-2.214],[3774,-2.17],[3775,-2.1726],[3777,-2.1286],[3778,-2.1312],[3780,-2.1183],[3823,-3.816],[3851,-2.2114],[3852,-2.214],[3854,-2.17],[3855,-2.1726],[3857,-2.1286],[3860,-2.1451],[3862,-2.1229],[3896,-2.214],[3898,-2.17],[3899,-2.1726],[3901,-2.1286],[3902,-2.1312],[3904,-2.1183],[3947,-3.816],[3975,-2.2114],[3976,-2.214],[3978,-2.17],[3979,-2.1726],[3981,-2.1286],[3984,-2.1451],[3986,-2.1229],[4020,-2.214],[4022,-2.17],[4023,-2.1726],[4025,-2.1286],[4026,-2.1312],[4028,-2.1183],[4071,-3.816],[4099,-2.2114],[4100,-2.214],[4102,-2.17],[4103,-2.1726],[4105,-2.1286],[4108,-2.1451],[4110,-2.1229],[4145,-2.214],[4147,-2.17],[4148,-2.1726],[4150,-2.1286],[4151,-2.1312],[4153,-2.1183],[4196,-3.816],[4224,-2.2114],[4225,-2.214],[4227,-2.17],[4228,-2.1726],[4230,-2.1286],[4233,-2.1451],[4235,-2.1229],[4271,-2.214],[4273,-2.17],[4274,-2.1726],[4276,-2.1286],[4277,-2.1312],[4279,-2.1183],[4322,-3.816],[4350,-2.2114],[4351,-2.214],[4353,-2.17],[4354,-2.1726],[4356,-2.1286],[4360,-3.3586],[4364,-3.5789],[4372,-3.816],[4400,-2.2114],[4401,-2.214],[4403,-2.17],[4404,-2.1726],[4406,-2.1286],[4411,-2.1354],[4435,-3.5149],[4436,0]]
       }
       pinky_metacarpal_l_ctrl.Scale.X {type=MovieSceneScriptingFloatChannel, default=1, const [1,12]}
       pinky_metacarpal_l_ctrl.Scale.Y {type=MovieSceneScriptingFloatChannel, default=1, const [1,12]}
       pinky_metacarpal_l_ctrl.Scale.Z {type=MovieSceneScriptingFloatChannel, default=1, const [1,12]}
       pinky_01_l_ctrl.Location.X {
        type=MovieSceneScriptingFloatChannel
        default=0
        thinned_from=88
        keys [[1790,0],[2511,0],[2513,-0.0004],[2524,0],[2768,0],[2775,0.0187],[2782,0.1634],[2796,0.2868],[2938,0.2844],[2986,0.1306],[3058,0.1306],[3070,0.1653],[3185,0.1653],[3215,0],[4436,-0.0001]]
       }
       pinky_01_l_ctrl.Location.Y {
        type=MovieSceneScriptingFloatChannel
        default=0
        thinned_from=609
        keys [[1790,0],[2768,0],[2775,-0.0504],[2782,-0.4533],[2796,-0.8988],[2938,-0.8909],[2986,-0.399],[3058,-0.399],[3070,-0.4454],[3185,-0.4454],[3215,0],[4430,0],[4436,-0.0001]]
       }
       pinky_01_l_ctrl.Location.Z {
        type=MovieSceneScriptingFloatChannel
        default=0
        thinned_from=592
        keys [[1790,0],[2768,0],[2774,0.0142],[2781,0.1114],[2803,-0.2979],[2938,-0.2979],[2986,-0.2967],[3058,-0.2967],[3070,0.1163],[3185,0.1163],[3215,0],[4436,0.0001]]
       }
       pinky_01_l_ctrl.Rotation.X {
        type=MovieSceneScriptingFloatChannel
        default=10.7439
        thinned_from=95
        keys [[1790,10.7439],[1829,14.2964],[1830,14.2955],[1867,10.7053],[1909,14.2964],[1910,14.2955],[1948,10.6991],[1989,14.2964],[1990,14.2955],[2027,10.7053],[2069,14.2964],[2070,14.2955],[2108,10.6991],[2149,14.2964],[2150,14.2955],[2187,10.7053],[2229,14.2964],[2230,14.2955],[2268,10.6991],[2309,14.2964],[2310,14.2955],[2347,10.7053],[2389,14.2964],[2390,14.2955],[2428,10.6991],[2466,14.2478],[2471,13.2608],[2503,-3.9884],[2513,-2.5706],[2515,-2.5886],[2520,0.5955],[2522,-1.2008],[2524,0.0155],[2531,-1.6802],[2759,1.3756],[2774,8.3681],[2781,0.1368],[2788,-3.4322],[2791,-3.2365],[2797,-3.3445],[2802,-3.2574],[2927,-3.2711],[2938,-3.2776],[2986,17.2775],[3058,17.2775],[3070,-7.8213],[3185,-7.8213],[3223,10.6991],[3264,14.2964],[3265,14.2955],[3302,10.7053],[3344,14.2964],[3345,14.2955],[3383,10.6991],[3424,14.2964],[3425,14.2955],[3463,10.7053],[3504,14.2964],[3505,14.2955],[3543,10.6991],[3584,14.2964],[3585,14.2955],[3623,10.7053],[3664,14.2964],[3665,14.2955],[3703,10.6991],[3744,14.2964],[3745,14.2955],[3782,10.7053],[3824,14.2964],[3825,14.2955],[3863,10.6991],[3906,10.7053],[3948,14.2964],[3949,14.2955],[3987,10.6991],[4030,10.7053],[4072,14.2964],[4073,14.2955],[4111,10.6991],[4155,10.7053],[4197,14.2964],[4198,14.2955],[4236,10.6991],[4281,10.7053],[4323,14.2964],[4324,14.2955],[4360,13.2043],[4364,13.4508],[4373,14.2964],[4412,10.7053],[4435,13.4973],[4436,0.0034]]
       }
       pinky_01_l_ctrl.Rotation.Y {
        type=MovieSceneScriptingFloatChannel
        default=5.5444
        thinned_from=101
        keys [[1790,5.5444],[1830,6.598],[1864,5.536],[1867,5.5318],[1910,6.598],[1948,5.5299],[1990,6.598],[2024,5.536],[2027,5.5318],[2070,6.598],[2108,5.5299],[2150,6.598],[2184,5.536],[2187,5.5318],[2230,6.598],[2268,5.5299],[2310,6.598],[2344,5.536],[2347,5.5318],[2390,6.598],[2428,5.5299],[2457,6.5051],[2501,7.6648],[2503,7.1094],[2507,5.8783],[2509,2.2136],[2512,0.8749],[2513,0.8755],[2515,0.8737],[2520,15.8665],[2522,13.2966],[2524,15.002],[2530,12.6088],[2763,20.8891],[2772,15.1171],[2777,9.0029],[2782,0.5177],[2802,-7.9138],[2938,-7.8806],[2986,10.2762],[3058,10.2762],[3070,7.2932],[3185,7.2932],[3223,5.5299],[3265,6.598],[3299,5.536],[3302,5.5318],[3345,6.598],[3383,5.5299],[3425,6.598],[3459,5.536],[3462,5.5318],[3505,6.598],[3543,5.5299],[3585,6.598],[3619,5.536],[3622,5.5318],[3665,6.598],[3703,5.5299],[3746,6.598],[3779,5.536],[3782,5.5318],[3825,6.598],[3863,5.5299],[3906,5.5318],[3949,6.598],[3987,5.5299],[4030,5.5318],[4073,6.598],[4111,5.5299],[4155,5.5318],[4198,6.598],[4236,5.5299],[4281,5.5318],[4324,6.598],[4360,6.2511],[4408,5.536],[4411,5.5318],[4435,6.4027],[4436,0.0024]]
       }
       pinky_01_l_ctrl.Rotation.Z {
        type=MovieSceneScriptingFloatChannel
        default=-4.2625
        thinned_from=72
        keys [[1790,-4.2625],[1830,5.3656],[1867,-4.3857],[1910,5.3656],[1948,-4.3861],[1990,5.3656],[2027,-4.3857],[2070,5.3656],[2108,-4.3861],[2150,5.3656],[2187,-4.3857],[2230,5.3656],[2268,-4.3861],[2310,5.3656],[2347,-4.3857],[2390,5.3656],[2428,-4.3861],[2467,5.2848],[2468,5.2398],[2503,-8.1467],[2505,-8.1467],[2510,-46.6417],[2513,-50.1758],[2518,-31.0227],[2519,7.3907],[2520,25.3573],[2522,21.8904],[2524,24.2183],[2530,20.9461],[2763,32.2184],[2794,-43.4996],[2938,-43.2494],[2986,41.2398],[3058,41.2398],[3070,-11.1746],[3185,-11.1746],[3223,-4.3861],[3265,5.3656],[3302,-4.3857],[3345,5.3656],[3383,-4.3861],[3425,5.3656],[3462,-4.3857],[3505,5.3656],[3543,-4.3861],[3585,5.3656],[3622,-4.3857],[3665,5.3656],[3703,-4.3861],[3745,5.3656],[3782,-4.3857],[3825,5.3656],[3863,-4.3861],[3906,-4.3857],[3949,5.3656],[3987,-4.3861],[4030,-4.3857],[4073,5.3656],[4111,-4.3861],[4155,-4.3857],[4198,5.3656],[4236,-4.3861],[4281,-4.3857],[4324,5.3656],[4360,2.1982],[4364,1.6348],[4411,-4.3857],[4435,3.6549],[4436,0.0003]]
       }
       pinky_01_l_ctrl.Scale.X {type=MovieSceneScriptingFloatChannel, default=1, const [1,12]}
       pinky_01_l_ctrl.Scale.Y {type=MovieSceneScriptingFloatChannel, default=1, const [1,12]}
       pinky_01_l_ctrl.Scale.Z {type=MovieSceneScriptingFloatChannel, default=1, const [1,12]}
       pinky_02_l_ctrl.Location.X {type=MovieSceneScriptingFloatChannel, default=0, thinned_from=88, keys [[1790,0],[2512,0],[2513,-0.0001],[2527,0],[4436,0]]}
       pinky_02_l_ctrl.Location.Y {type=MovieSceneScriptingFloatChannel, default=0, thinned_from=231, keys [[1790,0],[4436,0]]}
       pinky_02_l_ctrl.Location.Z {type=MovieSceneScriptingFloatChannel, default=0, thinned_from=1337, keys [[1790,0],[4436,0]]}
       pinky_02_l_ctrl.Rotation.X {
        type=MovieSceneScriptingFloatChannel
        default=-2.6202
        thinned_from=136
        keys [[1790,-2.6202],[1828,-3.0315],[1831,-3.0355],[1844,-2.7414],[1868,-2.6142],[1908,-3.0315],[1911,-3.0355],[1924,-2.7402],[1948,-2.6142],[1988,-3.0315],[1991,-3.0355],[2004,-2.7402],[2028,-2.6142],[2068,-3.0315],[2071,-3.0355],[2084,-2.7412],[2108,-2.614],[2148,-3.0315],[2151,-3.0355],[2164,-2.7414],[2188,-2.6142],[2228,-3.0315],[2231,-3.0355],[2244,-2.7402],[2268,-2.6142],[2308,-3.0315],[2311,-3.0355],[2324,-2.7402],[2348,-2.6142],[2388,-3.0315],[2391,-3.0355],[2404,-2.7412],[2428,-2.614],[2466,-3.0276],[2471,-2.8441],[2503,0.2265],[2512,0.3115],[2515,0.3104],[2519,-0.142],[2522,0.2378],[2524,-0.0123],[2530,0.3385],[2762,-0.8183],[2771,-1.6241],[2780,-2.6655],[2787,-2.1791],[2803,-0.8813],[2927,-0.8386],[2938,-0.8371],[2986,0.0716],[3058,0.0716],[3070,-0.5441],[3185,-0.5441],[3223,-2.614],[3263,-3.0315],[3266,-3.0355],[3279,-2.7414],[3303,-2.6142],[3343,-3.0315],[3346,-3.0355],[3359,-2.7403],[3383,-2.6142],[3423,-3.0315],[3426,-3.0355],[3439,-2.7403],[3463,-2.6142],[3503,-3.0315],[3506,-3.0355],[3520,-2.7235],[3544,-2.614],[3583,-3.0315],[3586,-3.0355],[3599,-2.7414],[3623,-2.6142],[3663,-3.0315],[3666,-3.0355],[3679,-2.7403],[3703,-2.6142],[3743,-3.0315],[3746,-3.0355],[3759,-2.7403],[3783,-2.6142],[3823,-3.0315],[3826,-3.0355],[3839,-2.7412],[3863,-2.614],[3907,-2.6142],[3947,-3.0315],[3950,-3.0355],[3963,-2.7412],[3987,-2.614],[4031,-2.6142],[4071,-3.0315],[4074,-3.0355],[4087,-2.7412],[4111,-2.614],[4156,-2.6142],[4196,-3.0315],[4199,-3.0355],[4212,-2.7412],[4236,-2.614],[4282,-2.6142],[4322,-3.0315],[4325,-3.0355],[4338,-2.7412],[4360,-2.9139],[4364,-2.9653],[4372,-3.0315],[4375,-3.0355],[4388,-2.7414],[4412,-2.6142],[4435,-2.9411],[4436,0]]
       }
       pinky_02_l_ctrl.Rotation.Y {
        type=MovieSceneScriptingFloatChannel
        default=0.0629
        thinned_from=825
        keys [[1790,0.0629],[1792,0.065],[1794,0.0647],[1798,0.0694],[1800,0.0689],[1804,0.0714],[1805,0.0711],[1809,0.0732],[1810,0.0732],[1811,0.0734],[1812,0.0734],[1814,0.0737],[1815,0.0732],[1817,0.0744],[1818,0.0735],[1822,0.0761],[1823,0.0753],[1824,0.0754],[1826,0.0746],[1828,0.0758],[1831,0.0745],[1833,0.075],[1850,0.0659],[1852,0.0662],[1854,0.0653],[1858,0.0653],[1860,0.0647],[1861,0.0649],[1863,0.0646],[1865,0.0651],[1869,0.0647],[1876,0.067],[1877,0.067],[1880,0.069],[1881,0.0688],[1883,0.0715],[1884,0.0711],[1886,0.0725],[1887,0.0724],[1889,0.074],[1891,0.0732],[1894,0.0744],[1895,0.0743],[1896,0.0748],[1898,0.0735],[1902,0.0761],[1903,0.0753],[1904,0.0754],[1906,0.0746],[1908,0.0758],[1912,0.0735],[1916,0.0745],[1921,0.0692],[1923,0.0697],[1925,0.0662],[1927,0.0674],[1929,0.065],[1932,0.0659],[1934,0.064],[1935,0.065],[1936,0.0637],[1938,0.0648],[1939,0.0635],[1941,0.0653],[1943,0.0643],[1944,0.0644],[1945,0.0637],[1948,0.065],[1951,0.0643],[1954,0.0663],[1955,0.066],[1960,0.0696],[1961,0.0688],[1963,0.0715],[1964,0.0711],[1966,0.0725],[1967,0.0724],[1969,0.074],[1971,0.0732],[1974,0.0744],[1975,0.0743],[1976,0.0748],[1978,0.0735],[1982,0.0761],[1983,0.0753],[1984,0.0754],[1986,0.0746],[1988,0.0758],[1992,0.0735],[1996,0.0745],[2001,0.0692],[2003,0.0697],[2005,0.0662],[2007,0.0674],[2009,0.065],[2012,0.0659],[2014,0.064],[2015,0.065],[2016,0.0637],[2018,0.0648],[2019,0.0635],[2021,0.0653],[2023,0.0643],[2024,0.0644],[2025,0.0637],[2028,0.065],[2031,0.0643],[2034,0.0663],[2035,0.066],[2040,0.0696],[2041,0.0688],[2043,0.0715],[2044,0.0711],[2046,0.0725],[2047,0.0724],[2049,0.074],[2051,0.0732],[2054,0.0744],[2055,0.0743],[2056,0.0748],[2058,0.0735],[2062,0.0761],[2063,0.0753],[2064,0.0754],[2066,0.0746],[2068,0.0758],[2071,0.0745],[2073,0.0751],[2098,0.065],[2100,0.065],[2102,0.0647],[2105,0.0647],[2106,0.0655],[2110,0.0629],[2112,0.065],[2114,0.0647],[2118,0.0694],[2120,0.0689],[2124,0.0714],[2125,0.0711],[2129,0.0732],[2130,0.0732],[2131,0.0734],[2132,0.0734],[2134,0.0737],[2135,0.0732],[2137,0.0744],[2138,0.0735],[2142,0.0761],[2143,0.0753],[2144,0.0754],[2146,0.0746],[2148,0.0758],[2151,0.0745],[2153,0.075],[2170,0.0659],[2172,0.0662],[2174,0.0653],[2178,0.0653],[2180,0.0647],[2181,0.0649],[2183,0.0646],[2185,0.0651],[2189,0.0647],[2196,0.067],[2197,0.067],[2200,0.069],[2201,0.0688],[2203,0.0715],[2204,0.0711],[2206,0.0725],[2207,0.0724],[2209,0.074],[2211,0.0732],[2214,0.0744],[2215,0.0743],[2216,0.0748],[2218,0.0735],[2222,0.0761],[2223,0.0753],[2224,0.0754],[2226,0.0746],[2228,0.0758],[2232,0.0735],[2236,0.0745],[2241,0.0692],[2243,0.0697],[2245,0.0662],[2247,0.0674],[2249,0.065],[2252,0.0659],[2254,0.064],[2255,0.065],[2256,0.0637],[2258,0.0648],[2259,0.0635],[2261,0.0653],[2263,0.0643],[2264,0.0644],[2265,0.0637],[2268,0.065],[2271,0.0643],[2274,0.0663],[2275,0.066],[2280,0.0696],[2281,0.0688],[2283,0.0715],[2284,0.0711],[2286,0.0725],[2287,0.0724],[2289,0.074],[2291,0.0732],[2294,0.0744],[2295,0.0743],[2296,0.0748],[2298,0.0735],[2302,0.0761],[2303,0.0753],[2304,0.0754],[2306,0.0746],[2308,0.0758],[2312,0.0735],[2316,0.0745],[2321,0.0692],[2323,0.0697],[2325,0.0662],[2327,0.0674],[2329,0.065],[2332,0.0659],[2334,0.064],[2335,0.065],[2336,0.0637],[2338,0.0648],[2339,0.0635],[2341,0.0653],[2343,0.0643],[2344,0.0644],[2345,0.0637],[2348,0.065],[2351,0.0643],[2354,0.0663],[2355,0.066],[2360,0.0696],[2361,0.0688],[2363,0.0715],[2364,0.0711],[2366,0.0725],[2367,0.0724],[2369,0.074],[2371,0.0732],[2374,0.0744],[2375,0.0743],[2376,0.0748],[2378,0.0735],[2382,0.0761],[2383,0.0753],[2384,0.0754],[2386,0.0746],[2388,0.0758],[2391,0.0745],[2393,0.0751],[2418,0.065],[2420,0.065],[2422,0.0647],[2425,0.0647],[2426,0.0655],[2430,0.0629],[2432,0.065],[2434,0.0647],[2438,0.0694],[2440,0.0689],[2444,0.0714],[2445,0.0711],[2449,0.0732],[2450,0.0732],[2451,0.0734],[2452,0.0734],[2454,0.0737],[2455,0.0732],[2457,0.0744],[2458,0.0735],[2462,0.0761],[2463,0.0753],[2464,0.0754],[2466,0.0746],[2467,0.0754],[2469,0.0725],[2471,0.0852],[2502,0.1919],[2503,0.1919],[2507,0.2089],[2509,0.2607],[2512,0.2804],[2519,0.2564],[2522,0.2797],[2523,0.2707],[2531,0.2875],[2762,0.2357],[2770,0.1885],[2780,0.0808],[2786,0.0165],[2791,0.0446],[2797,-0.0172],[2802,0.0178],[2927,0.0256],[2938,0.0259],[2986,-0.1302],[3058,-0.1302],[3070,0.1403],[3185,0.1403],[3215,0.065],[3217,0.0647],[3220,0.0647],[3221,0.0655],[3225,0.0629],[3227,0.065],[3229,0.0647],[3233,0.0694],[3235,0.0689],[3239,0.0714],[3240,0.0711],[3244,0.0732],[3245,0.0732],[3246,0.0734],[3247,0.0734],[3249,0.0737],[3250,0.0732],[3252,0.0744],[3253,0.0735],[3257,0.0761],[3258,0.0753],[3259,0.0754],[3261,0.0746],[3263,0.0758],[3266,0.0745],[3268,0.075],[3285,0.0659],[3287,0.0662],[3289,0.0653],[3293,0.0653],[3295,0.0647],[3296,0.0649],[3298,0.0646],[3300,0.0651],[3304,0.0647],[3311,0.067],[3312,0.067],[3315,0.069],[3316,0.0688],[3318,0.0715],[3319,0.0711],[3321,0.0725],[3322,0.0724],[3324,0.074],[3326,0.0732],[3329,0.0744],[3330,0.0743],[3331,0.0748],[3333,0.0735],[3337,0.0761],[3338,0.0753],[3339,0.0754],[3341,0.0746],[3343,0.0758],[3347,0.0735],[3351,0.0745],[3356,0.0692],[3358,0.0697],[3360,0.0662],[3362,0.0674],[3364,0.065],[3367,0.0659],[3369,0.064],[3370,0.065],[3371,0.0637],[3373,0.0648],[3374,0.0635],[3376,0.0653],[3378,0.0643],[3379,0.0644],[3380,0.0637],[3383,0.065],[3386,0.0643],[3389,0.0663],[3390,0.066],[3395,0.0696],[3396,0.0688],[3398,0.0715],[3399,0.0711],[3401,0.0725],[3402,0.0724],[3404,0.074],[3406,0.0732],[3409,0.0744],[3410,0.0743],[3411,0.0748],[3413,0.0735],[3417,0.0761],[3418,0.0753],[3419,0.0754],[3421,0.0746],[3423,0.0758],[3427,0.0735],[3431,0.0745],[3436,0.0692],[3438,0.0697],[3440,0.0662],[3442,0.0674],[3444,0.065],[3447,0.0659],[3449,0.064],[3450,0.065],[3451,0.0637],[3453,0.0648],[3454,0.0635],[3456,0.0653],[3458,0.0643],[3459,0.0644],[3460,0.0637],[3463,0.065],[3466,0.0643],[3469,0.0663],[3470,0.066],[3475,0.0696],[3476,0.0688],[3478,0.0715],[3479,0.0711],[3481,0.0725],[3482,0.0724],[3484,0.074],[3486,0.0732],[3489,0.0744],[3490,0.0743],[3491,0.0748],[3493,0.0735],[3497,0.0761],[3498,0.0753],[3499,0.0754],[3501,0.0746],[3503,0.0758],[3506,0.0745],[3508,0.0751],[3533,0.065],[3535,0.065],[3537,0.0647],[3540,0.0647],[3541,0.0655],[3543,0.0638],[3544,0.0638],[3545,0.0629],[3547,0.065],[3549,0.0647],[3553,0.0694],[3555,0.0689],[3559,0.0714],[3560,0.0711],[3564,0.0732],[3565,0.0732],[3566,0.0734],[3567,0.0734],[3569,0.0737],[3570,0.0732],[3572,0.0744],[3573,0.0735],[3577,0.0761],[3578,0.0753],[3579,0.0754],[3581,0.0746],[3583,0.0758],[3586,0.0745],[3588,0.075],[3605,0.0659],[3607,0.0662],[3609,0.0653],[3613,0.0653],[3615,0.0647],[3616,0.0649],[3618,0.0646],[3620,0.0651],[3624,0.0647],[3631,0.067],[3632,0.067],[3635,0.069],[3636,0.0688],[3638,0.0715],[3639,0.0711],[3641,0.0725],[3642,0.0724],[3644,0.074],[3646,0.0732],[3649,0.0744],[3650,0.0743],[3651,0.0748],[3653,0.0735],[3657,0.0761],[3658,0.0753],[3659,0.0754],[3661,0.0746],[3663,0.0758],[3667,0.0735],[3671,0.0745],[3676,0.0692],[3678,0.0697],[3680,0.0662],[3682,0.0674],[3684,0.065],[3687,0.0659],[3689,0.064],[3690,0.065],[3691,0.0637],[3693,0.0648],[3694,0.0635],[3696,0.0653],[3698,0.0643],[3699,0.0644],[3700,0.0637],[3703,0.065],[3706,0.0643],[3709,0.0663],[3710,0.066],[3715,0.0696],[3716,0.0688],[3718,0.0715],[3719,0.0711],[3721,0.0725],[3722,0.0724],[3724,0.074],[3726,0.0732],[3729,0.0744],[3730,0.0743],[3731,0.0748],[3733,0.0735],[3737,0.0761],[3738,0.0753],[3739,0.0754],[3741,0.0746],[3743,0.0758],[3747,0.0735],[3751,0.0745],[3756,0.0692],[3758,0.0697],[3760,0.0662],[3762,0.0674],[3764,0.065],[3767,0.0659],[3769,0.064],[3770,0.065],[3771,0.0637],[3773,0.0648],[3774,0.0635],[3776,0.0653],[3778,0.0643],[3779,0.0644],[3780,0.0637],[3783,0.065],[3786,0.0643],[3789,0.0663],[3790,0.066],[3795,0.0696],[3796,0.0688],[3798,0.0715],[3799,0.0711],[3801,0.0725],[3802,0.0724],[3804,0.074],[3806,0.0732],[3809,0.0744],[3810,0.0743],[3811,0.0748],[3813,0.0735],[3817,0.0761],[3818,0.0753],[3819,0.0754],[3821,0.0746],[3823,0.0758],[3826,0.0745],[3828,0.0751],[3853,0.065],[3855,0.065],[3857,0.0647],[3860,0.0647],[3861,0.0655],[3865,0.0629],[3867,0.065],[3869,0.0647],[3873,0.0694],[3875,0.0689],[3879,0.0714],[3880,0.0711],[3884,0.0732],[3885,0.0732],[3886,0.0734],[3887,0.0734],[3891,0.0659],[3893,0.064],[3894,0.065],[3895,0.0637],[3897,0.0648],[3898,0.0635],[3900,0.0653],[3902,0.0643],[3903,0.0644],[3904,0.0637],[3907,0.065],[3910,0.0643],[3913,0.0663],[3914,0.066],[3919,0.0696],[3920,0.0688],[3922,0.0715],[3923,0.0711],[3925,0.0725],[3926,0.0724],[3928,0.074],[3930,0.0732],[3933,0.0744],[3934,0.0743],[3935,0.0748],[3937,0.0735],[3941,0.0761],[3942,0.0753],[3943,0.0754],[3945,0.0746],[3947,0.0758],[3950,0.0745],[3952,0.0751],[3977,0.065],[3979,0.065],[3981,0.0647],[3984,0.0647],[3985,0.0655],[3989,0.0629],[3991,0.065],[3993,0.0647],[3997,0.0694],[3999,0.0689],[4003,0.0714],[4004,0.0711],[4008,0.0732],[4009,0.0732],[4010,0.0734],[4011,0.0734],[4015,0.0659],[4017,0.064],[4018,0.065],[4019,0.0637],[4021,0.0648],[4022,0.0635],[4024,0.0653],[4026,0.0643],[4027,0.0644],[4028,0.0637],[4031,0.065],[4034,0.0643],[4037,0.0663],[4038,0.066],[4043,0.0696],[4044,0.0688],[4046,0.0715],[4047,0.0711],[4049,0.0725],[4050,0.0724],[4052,0.074],[4054,0.0732],[4057,0.0744],[4058,0.0743],[4059,0.0748],[4061,0.0735],[4065,0.0761],[4066,0.0753],[4067,0.0754],[4069,0.0746],[4071,0.0758],[4074,0.0745],[4076,0.0751],[4101,0.065],[4103,0.065],[4105,0.0647],[4108,0.0647],[4109,0.0655],[4113,0.0629],[4115,0.065],[4117,0.0647],[4121,0.0694],[4123,0.0689],[4127,0.0714],[4128,0.0711],[4132,0.0732],[4133,0.0732],[4134,0.0734],[4135,0.0734],[4140,0.0659],[4142,0.064],[4143,0.065],[4144,0.0637],[4146,0.0648],[4147,0.0635],[4149,0.0653],[4151,0.0643],[4152,0.0644],[4153,0.0637],[4156,0.065],[4159,0.0643],[4162,0.0663],[4163,0.066],[4168,0.0696],[4169,0.0688],[4171,0.0715],[4172,0.0711],[4174,0.0725],[4175,0.0724],[4177,0.074],[4179,0.0732],[4182,0.0744],[4183,0.0743],[4184,0.0748],[4186,0.0735],[4190,0.0761],[4191,0.0753],[4192,0.0754],[4194,0.0746],[4196,0.0758],[4199,0.0745],[4201,0.0751],[4226,0.065],[4228,0.065],[4230,0.0647],[4233,0.0647],[4234,0.0655],[4238,0.0629],[4240,0.065],[4242,0.0647],[4246,0.0694],[4248,0.0689],[4252,0.0714],[4253,0.0711],[4257,0.0732],[4258,0.0732],[4259,0.0734],[4260,0.0734],[4266,0.0659],[4268,0.064],[4269,0.065],[4270,0.0637],[4272,0.0648],[4273,0.0635],[4275,0.0653],[4277,0.0643],[4278,0.0644],[4279,0.0637],[4282,0.065],[4285,0.0643],[4288,0.0663],[4289,0.066],[4294,0.0696],[4295,0.0688],[4297,0.0715],[4298,0.0711],[4300,0.0725],[4301,0.0724],[4303,0.074],[4305,0.0732],[4308,0.0744],[4309,0.0743],[4310,0.0748],[4312,0.0735],[4316,0.0761],[4317,0.0753],[4318,0.0754],[4320,0.0746],[4322,0.0758],[4325,0.0745],[4327,0.0751],[4352,0.065],[4354,0.065],[4356,0.0647],[4357,0.0648],[4360,0.0737],[4365,0.0732],[4367,0.0744],[4372,0.0758],[4375,0.0745],[4377,0.075],[4394,0.0659],[4396,0.0662],[4398,0.0653],[4402,0.0653],[4404,0.0647],[4405,0.0649],[4407,0.0646],[4409,0.0651],[4413,0.0647],[4420,0.067],[4421,0.067],[4424,0.069],[4425,0.0688],[4427,0.0715],[4428,0.0711],[4430,0.0725],[4431,0.0724],[4433,0.074],[4435,0.0732],[4436,0]]
       }
       pinky_02_l_ctrl.Rotation.Z {
        type=MovieSceneScriptingFloatChannel
        default=6.6309
        thinned_from=105
        keys [[1790,6.6309],[1829,17.705],[1867,6.558],[1909,17.705],[1940,6.7408],[1941,6.7408],[1945,6.558],[1948,6.558],[1989,17.705],[2025,6.558],[2028,6.558],[2069,17.705],[2105,6.558],[2149,17.705],[2187,6.558],[2229,17.705],[2265,6.558],[2268,6.558],[2309,17.705],[2345,6.558],[2348,6.558],[2389,17.705],[2426,6.558],[2428,6.5581],[2467,17.6453],[2469,17.4916],[2503,-9.2352],[2507,-9.2352],[2512,-9.1303],[2515,-9.1301],[2520,10.5689],[2522,-3.3895],[2524,5.9987],[2527,-5.9049],[2530,-7.1488],[2764,38.3437],[2794,-19.5767],[2938,-19.0287],[2986,56.2097],[3058,56.2097],[3070,7.8137],[3185,7.8137],[3220,6.558],[3224,6.5581],[3264,17.705],[3302,6.558],[3344,17.705],[3380,6.558],[3383,6.558],[3424,17.705],[3460,6.558],[3463,6.558],[3504,17.705],[3540,6.558],[3584,17.705],[3622,6.558],[3664,17.705],[3700,6.558],[3703,6.558],[3744,17.705],[3780,6.558],[3783,6.558],[3824,17.705],[3860,6.558],[3907,6.558],[3948,17.705],[3984,6.558],[4031,6.558],[4072,17.705],[4108,6.558],[4156,6.558],[4197,17.705],[4233,6.558],[4282,6.558],[4323,17.705],[4360,14.3208],[4364,15.0848],[4373,17.705],[4411,6.558],[4435,15.7319],[4436,0]]
       }
       pinky_02_l_ctrl.Scale.X {type=MovieSceneScriptingFloatChannel, default=1, const [1,12]}
       pinky_02_l_ctrl.Scale.Y {type=MovieSceneScriptingFloatChannel, default=1, const [1,12]}
       pinky_02_l_ctrl.Scale.Z {type=MovieSceneScriptingFloatChannel, default=1, const [1,12]}
       pinky_03_l_ctrl.Location.X {type=MovieSceneScriptingFloatChannel, default=0, thinned_from=146, keys [[1790,0],[2512,0],[2513,-0.0001],[2525,0],[4436,0]]}
       pinky_03_l_ctrl.Location.Y {type=MovieSceneScriptingFloatChannel, default=0, thinned_from=706, keys [[1790,0],[4436,0]]}
       pinky_03_l_ctrl.Location.Z {type=MovieSceneScriptingFloatChannel, default=0, thinned_from=38, keys [[1790,0],[4436,0]]}
       pinky_03_l_ctrl.Rotation.X {
        type=MovieSceneScriptingFloatChannel
        default=-1.1107
        thinned_from=110
        keys [[1790,-1.1107],[1828,-1.2852],[1832,-1.2855],[1845,-1.1539],[1868,-1.1081],[1908,-1.2852],[1925,-1.1542],[1949,-1.108],[1988,-1.2852],[2028,-1.1081],[2044,-1.2288],[2071,-1.286],[2084,-1.1616],[2108,-1.1082],[2148,-1.2852],[2152,-1.2855],[2165,-1.1539],[2188,-1.1081],[2228,-1.2852],[2245,-1.1542],[2269,-1.108],[2308,-1.2852],[2348,-1.1081],[2364,-1.2288],[2391,-1.286],[2404,-1.1616],[2428,-1.1082],[2467,-1.2828],[2503,-0.5319],[2506,-0.5319],[2510,-0.7316],[2513,-0.7499],[2518,-0.6137],[2520,-0.2257],[2522,-0.2655],[2524,-0.2402],[2529,-0.2766],[2530,-0.2766],[2763,-0.1533],[2775,-0.8475],[2780,-1.1168],[2785,-0.9745],[2803,-0.3713],[2938,-0.3724],[2986,-0.2785],[3058,-0.2785],[3070,-0.7518],[3185,-0.7518],[3223,-1.1082],[3263,-1.2852],[3267,-1.2855],[3280,-1.1539],[3303,-1.1081],[3343,-1.2852],[3360,-1.1542],[3384,-1.108],[3423,-1.2852],[3463,-1.1081],[3479,-1.2288],[3506,-1.286],[3519,-1.1616],[3543,-1.1082],[3583,-1.2852],[3587,-1.2855],[3600,-1.1539],[3623,-1.1081],[3663,-1.2852],[3680,-1.1542],[3704,-1.108],[3743,-1.2852],[3783,-1.1081],[3799,-1.2288],[3826,-1.286],[3839,-1.1616],[3863,-1.1082],[3907,-1.1081],[3923,-1.2288],[3950,-1.286],[3963,-1.1616],[3987,-1.1082],[4031,-1.1081],[4047,-1.2288],[4074,-1.286],[4087,-1.1616],[4111,-1.1082],[4156,-1.1081],[4172,-1.2288],[4199,-1.286],[4212,-1.1616],[4236,-1.1082],[4282,-1.1081],[4298,-1.2288],[4325,-1.286],[4338,-1.1616],[4360,-1.2353],[4364,-1.2572],[4372,-1.2852],[4376,-1.2855],[4389,-1.1539],[4412,-1.1081],[4435,-1.2468],[4436,0]]
       }
       pinky_03_l_ctrl.Rotation.Y {
        type=MovieSceneScriptingFloatChannel
        default=-0.7516
        thinned_from=114
        keys [[1790,-0.7516],[1828,-0.8699],[1832,-0.87],[1862,-0.7501],[1868,-0.7499],[1908,-0.8699],[1910,-0.8699],[1949,-0.7498],[1988,-0.8699],[1990,-0.8699],[2028,-0.7499],[2071,-0.8703],[2102,-0.7501],[2108,-0.7499],[2148,-0.8699],[2152,-0.87],[2182,-0.7501],[2188,-0.7499],[2228,-0.8699],[2230,-0.8699],[2269,-0.7498],[2308,-0.8699],[2310,-0.8699],[2348,-0.7499],[2391,-0.8703],[2422,-0.7501],[2428,-0.7499],[2466,-0.8688],[2503,0.5319],[2513,0.757],[2519,0.7213],[2522,1.0294],[2524,0.8717],[2529,1.0918],[2763,0.3356],[2772,-0.1569],[2782,-0.7463],[2803,-0.2568],[2927,-0.2361],[2938,-0.2352],[2986,0.9054],[3058,0.9054],[3070,0.2542],[3185,0.2542],[3217,-0.7501],[3223,-0.7499],[3264,-0.8699],[3267,-0.87],[3297,-0.7501],[3303,-0.7499],[3344,-0.8699],[3345,-0.8699],[3384,-0.7498],[3423,-0.8699],[3425,-0.8699],[3463,-0.7499],[3506,-0.8703],[3537,-0.7501],[3543,-0.7499],[3584,-0.8699],[3587,-0.87],[3617,-0.7501],[3623,-0.7499],[3663,-0.8699],[3665,-0.8699],[3704,-0.7498],[3744,-0.8699],[3783,-0.7499],[3826,-0.8703],[3857,-0.7501],[3907,-0.7499],[3950,-0.8703],[3981,-0.7501],[4031,-0.7499],[4074,-0.8703],[4105,-0.7501],[4156,-0.7499],[4199,-0.8703],[4230,-0.7501],[4282,-0.7499],[4325,-0.8703],[4356,-0.7501],[4357,-0.7505],[4360,-0.8361],[4364,-0.8523],[4372,-0.8699],[4376,-0.87],[4406,-0.7501],[4412,-0.7499],[4435,-0.8492],[4436,0]]
       }
       pinky_03_l_ctrl.Rotation.Z {
        type=MovieSceneScriptingFloatChannel
        default=-10.8715
        thinned_from=97
        keys [[1790,-10.8715],[1828,5.9059],[1832,5.9091],[1867,-11.091],[1908,5.9059],[1910,5.9059],[1945,-11.091],[1947,-11.091],[1989,5.9059],[1990,5.9059],[2027,-11.091],[2071,5.9628],[2105,-11.091],[2107,-11.091],[2149,5.9059],[2152,5.9091],[2187,-11.091],[2229,5.9059],[2230,5.9059],[2265,-11.091],[2267,-11.091],[2309,5.9059],[2310,5.9059],[2347,-11.091],[2391,5.9628],[2425,-11.091],[2427,-11.091],[2467,5.8027],[2503,-40.503],[2513,-36.078],[2515,-36.3137],[2520,2.6214],[2522,-3.2009],[2524,0.6524],[2531,-4.7032],[2764,14.0282],[2927,10.2078],[2938,10.1247],[2986,29.5535],[3058,29.5535],[3070,-16.9797],[3185,-16.9797],[3220,-11.091],[3222,-11.091],[3264,5.9059],[3267,5.9093],[3302,-11.091],[3344,5.9059],[3345,5.9059],[3380,-11.091],[3382,-11.091],[3424,5.9059],[3425,5.9059],[3463,-11.091],[3506,5.9628],[3541,-11.091],[3584,5.9059],[3587,5.9093],[3600,-5.8356],[3623,-11.091],[3664,5.9059],[3665,5.9059],[3700,-11.091],[3702,-11.091],[3745,5.9059],[3782,-11.091],[3826,5.9628],[3860,-11.091],[3906,-11.091],[3950,5.9628],[3984,-11.091],[4030,-11.091],[4074,5.9628],[4108,-11.091],[4155,-11.091],[4199,5.9628],[4233,-11.091],[4281,-11.091],[4325,5.9628],[4360,0.9505],[4364,2.0678],[4373,5.9059],[4376,5.9093],[4389,-5.8358],[4412,-11.091],[4435,2.9318],[4436,0]]
       }
       pinky_03_l_ctrl.Scale.X {type=MovieSceneScriptingFloatChannel, default=1, const [1,12]}
       pinky_03_l_ctrl.Scale.Y {type=MovieSceneScriptingFloatChannel, default=1, const [1,12]}
       pinky_03_l_ctrl.Scale.Z {type=MovieSceneScriptingFloatChannel, default=1, const [1,12]}
       hand_l_ik_ctrl.Location.X {
        type=MovieSceneScriptingFloatChannel
        default=-11.978
        keys [[1790,-11.978],[1815,-19.1961],[1857,-10.4486],[1896,-19.3286],[1937,-10.6466],[1975,-19.1977],[2017,-10.4519],[2056,-19.3264],[2097,-10.6511],[2135,-19.1961],[2177,-10.4486],[2216,-19.3286],[2257,-10.6466],[2295,-19.1977],[2337,-10.4519],[2376,-19.3264],[2417,-10.6511],[2455,-19.1961],[2496,19.8128],[2502,11.5485],[2504,13.2738],[2505,13.1564],[2507,13.396],[2511,2.9748],[2513,25.1717],[2518,-19.6193],[2520,-24.7189],[2524,-19.5156],[2759,-22.663],[2763,-24.2294],[2772,-20.4035],[2778,-21.4776],[2927,-191.6958],[2938,-106.9152],[2986,-128.6872],[3058,-128.6872],[3070,11.7464],[3083,11.7464],[3099,-204.797],[3185,-204.797],[3250,-19.1961],[3292,-10.4486],[3331,-19.3286],[3372,-10.6466],[3410,-19.1976],[3452,-10.4519],[3491,-19.3264],[3532,-10.651],[3570,-19.1961],[3612,-10.4486],[3651,-19.3286],[3692,-10.6466],[3730,-19.1976],[3772,-10.4519],[3811,-19.3264],[3852,-10.6511],[3896,-10.4519],[3935,-19.3264],[3976,-10.6511],[4020,-10.4519],[4059,-19.3264],[4100,-10.6511],[4145,-10.4519],[4184,-19.3264],[4225,-10.6511],[4271,-10.4519],[4310,-19.3264],[4351,-10.6511],[4360,-60.9103],[4364,-32.8735],[4365,-19.1961],[4401,-10.4486],[4435,-18.9118],[4436,-0.0028]]
       }
       hand_l_ik_ctrl.Location.Y {
        type=MovieSceneScriptingFloatChannel
        default=-17.115
        keys [[1790,-17.115],[1812,-14.1385],[1850,-18.4071],[1891,-14.7764],[1928,-18.7625],[1972,-14.1346],[2010,-18.3901],[2052,-14.7716],[2088,-18.7583],[2132,-14.1385],[2170,-18.4071],[2211,-14.7764],[2248,-18.7625],[2292,-14.1346],[2330,-18.3901],[2372,-14.7716],[2408,-18.7583],[2452,-14.1385],[2471,-18.4027],[2484,24.0508],[2496,34.4176],[2502,31.8348],[2503,31.9254],[2506,31.6907],[2507,31.7496],[2511,22.1889],[2512,25.2552],[2513,5.6144],[2515,-22.8753],[2518,-24.0296],[2520,-19.2414],[2763,-26.1604],[2777,-13.5822],[2927,137.0095],[2938,82.1625],[2986,116.04],[3058,116.04],[3070,49.3573],[3083,49.3573],[3099,142.7913],[3185,142.7913],[3247,-14.1385],[3285,-18.4071],[3326,-14.7765],[3363,-18.7625],[3407,-14.1346],[3445,-18.3901],[3487,-14.7716],[3523,-18.7584],[3567,-14.1385],[3605,-18.4071],[3646,-14.7765],[3683,-18.7625],[3727,-14.1346],[3765,-18.39],[3807,-14.7716],[3843,-18.7584],[3887,-14.1385],[3931,-14.7716],[3967,-18.7584],[4011,-14.1385],[4055,-14.7716],[4091,-18.7584],[4135,-14.1385],[4180,-14.7716],[4216,-18.7584],[4260,-14.1385],[4306,-14.7716],[4342,-18.7584],[4360,8.4426],[4364,11.1604],[4394,-18.4071],[4415,-17.266],[4425,-15.9333],[4430,-15.1685],[4433,-14.8511],[4435,-14.7765],[4436,0.0001]]
       }
       hand_l_ik_ctrl.Location.Z {
        type=MovieSceneScriptingFloatChannel
        default=-17.4264
        thinned_from=75
        keys [[1790,-17.4264],[1795,-17.7636],[1834,-12.3404],[1875,-17.863],[1914,-12.1632],[1955,-17.7665],[1994,-12.2349],[2035,-17.8657],[2074,-12.1581],[2115,-17.7636],[2154,-12.3404],[2195,-17.8631],[2234,-12.1632],[2275,-17.7665],[2314,-12.2349],[2355,-17.8657],[2394,-12.1581],[2435,-17.7636],[2484,38.241],[2501,52.138],[2505,51.6623],[2508,52.4287],[2511,47.022],[2512,48.6503],[2515,18.9376],[2520,-10.1254],[2525,-12.9603],[2758,-9.8552],[2762,-9.1851],[2777,-23.3892],[2927,-1154.2893],[2938,-1044.0977],[2986,-955.2029],[3058,-955.2029],[3070,-852.8777],[3083,-852.8777],[3099,-942.4043],[3185,-942.4043],[3230,-17.7636],[3269,-12.3404],[3310,-17.8631],[3349,-12.1633],[3390,-17.7664],[3429,-12.2349],[3470,-17.8656],[3509,-12.1581],[3550,-17.7636],[3589,-12.3404],[3630,-17.8631],[3669,-12.1633],[3710,-17.7664],[3749,-12.2349],[3790,-17.8656],[3829,-12.1581],[3870,-17.7636],[3914,-17.8656],[3953,-12.1581],[3994,-17.7636],[4038,-17.8656],[4077,-12.1581],[4118,-17.7636],[4163,-17.8656],[4202,-12.1581],[4243,-17.7636],[4289,-17.8656],[4328,-12.1581],[4360,97.0886],[4364,54.2562],[4378,-12.3404],[4419,-17.8631],[4427,-17.0645],[4433,-16.0306],[4435,-15.654],[4436,0]]
       }
       hand_l_ik_ctrl.Rotation.X {
        type=MovieSceneScriptingFloatChannel
        default=0.562
        thinned_from=73
        keys [[1790,0.562],[1791,0.5591],[1837,8.0165],[1870,0.6054],[1916,8.2141],[1951,0.5598],[1996,8.1394],[2030,0.6056],[2076,8.2186],[2111,0.5591],[2157,8.0165],[2190,0.6055],[2236,8.2141],[2271,0.5598],[2316,8.1394],[2350,0.6056],[2396,8.2186],[2431,0.5591],[2469,7.4866],[2501,423.4865],[2505,419.674],[2507,419.4963],[2511,442.9581],[2519,225.372],[2523,228.3294],[2778,363.4235],[2791,360.567],[2802,360.5677],[2927,414.8188],[2938,246.8788],[2986,553.1932],[3058,553.1932],[3070,850.6121],[3083,850.6121],[3099,728.1816],[3185,728.1816],[3226,720.5591],[3272,728.0165],[3305,720.6055],[3351,728.2141],[3386,720.5598],[3431,728.1394],[3465,720.6057],[3511,728.2185],[3546,720.5591],[3592,728.0165],[3625,720.6055],[3671,728.2141],[3706,720.5598],[3751,728.1394],[3785,720.6056],[3831,728.2186],[3866,720.5591],[3909,720.6056],[3955,728.2186],[3990,720.5591],[4033,720.6056],[4079,728.2186],[4114,720.5591],[4158,720.6056],[4204,728.2186],[4239,720.5591],[4284,720.6056],[4330,728.2186],[4360,761.0862],[4364,786.2089],[4381,728.0165],[4414,720.6055],[4435,724.2867],[4436,720.0015]]
       }
       hand_l_ik_ctrl.Rotation.Y {
        type=MovieSceneScriptingFloatChannel
        default=-12.2477
        thinned_from=77
        keys [[1790,-12.2477],[1808,-15.4651],[1845,-4.6853],[1888,-15.4473],[1925,-4.7242],[1968,-15.4511],[2005,-4.7372],[2048,-15.4464],[2085,-4.7242],[2128,-15.4651],[2165,-4.6853],[2208,-15.4473],[2245,-4.7242],[2288,-15.4511],[2325,-4.7372],[2368,-15.4464],[2405,-4.7242],[2448,-15.4651],[2496,40.4928],[2502,31.5595],[2505,36.2001],[2506,36.1549],[2507,36.4049],[2511,-3.498],[2512,20.0122],[2513,68.4933],[2515,44.1488],[2519,20.2801],[2521,29.0088],[2528,21.1213],[2778,-12.5227],[2786,-11.7185],[2802,-11.7236],[2927,-1.5874],[2938,24.0005],[2986,-31.3575],[3058,-31.3575],[3070,88.1772],[3083,88.1772],[3099,16.8641],[3185,16.8641],[3243,-15.465],[3280,-4.6855],[3323,-15.4472],[3360,-4.7242],[3403,-15.4511],[3440,-4.7373],[3483,-15.4463],[3520,-4.7242],[3563,-15.465],[3600,-4.6855],[3643,-15.4472],[3680,-4.7242],[3723,-15.451],[3760,-4.7373],[3803,-15.4463],[3840,-4.7242],[3883,-15.465],[3927,-15.4463],[3964,-4.7242],[4007,-15.465],[4051,-15.4463],[4088,-4.7242],[4131,-15.465],[4176,-15.4463],[4213,-4.7242],[4256,-15.465],[4302,-15.4463],[4339,-4.7242],[4360,46.7153],[4364,-6.38],[4389,-4.6855],[4432,-15.4472],[4435,-15.2677],[4436,0.0018]]
       }
       hand_l_ik_ctrl.Rotation.Z {
        type=MovieSceneScriptingFloatChannel
        default=-16.1788
        thinned_from=168
        keys [[1790,-16.1788],[1791,-16.1717],[1798,-16.5015],[1815,-13.9146],[1841,-17.1678],[1846,-16.9937],[1850,-17.0621],[1870,-16.2998],[1878,-16.7288],[1896,-14.06],[1921,-17.3819],[1927,-17.1738],[1930,-17.1994],[1951,-16.1735],[1958,-16.4989],[1976,-14.0653],[2001,-17.1665],[2006,-17.0092],[2010,-17.0773],[2030,-16.3006],[2038,-16.7278],[2056,-14.0578],[2081,-17.3807],[2087,-17.1747],[2090,-17.2018],[2111,-16.1717],[2118,-16.5015],[2135,-13.9146],[2161,-17.1678],[2166,-16.9937],[2170,-17.0621],[2190,-16.2998],[2198,-16.7288],[2216,-14.06],[2241,-17.3819],[2247,-17.1738],[2250,-17.1994],[2271,-16.1735],[2278,-16.4989],[2296,-14.0653],[2321,-17.1665],[2326,-17.0092],[2330,-17.0773],[2350,-16.3006],[2358,-16.7277],[2376,-14.0578],[2401,-17.3807],[2407,-17.1747],[2410,-17.2018],[2431,-16.1717],[2438,-16.5015],[2455,-13.9146],[2470,-17.018],[2484,-249.2179],[2501,-195.6613],[2505,-202.3857],[2507,-202.6387],[2511,-184.9181],[2512,-198.4943],[2513,-300.3838],[2515,-413.3709],[2519,-445.7118],[2523,-441.5555],[2528,-446.6042],[2778,-378.778],[2791,-380.0445],[2801,-380.0441],[2927,-540.9192],[2938,-421.6702],[2986,-350.969],[3058,-350.969],[3070,-441.5762],[3083,-441.5762],[3099,-490.5191],[3185,-490.5191],[3226,-16.1716],[3233,-16.5014],[3250,-13.9146],[3276,-17.1678],[3281,-16.9937],[3285,-17.0621],[3305,-16.2998],[3313,-16.7288],[3331,-14.06],[3356,-17.3819],[3362,-17.1738],[3365,-17.1995],[3386,-16.1735],[3393,-16.4989],[3411,-14.0653],[3436,-17.1665],[3441,-17.0093],[3445,-17.0773],[3465,-16.3006],[3473,-16.7277],[3491,-14.0578],[3516,-17.3807],[3522,-17.1747],[3525,-17.2018],[3546,-16.1716],[3553,-16.5014],[3570,-13.9146],[3596,-17.1678],[3601,-16.9937],[3605,-17.0621],[3625,-16.2997],[3633,-16.7288],[3651,-14.06],[3676,-17.3819],[3682,-17.1738],[3685,-17.1995],[3706,-16.1735],[3713,-16.4989],[3731,-14.0653],[3756,-17.1665],[3761,-17.0093],[3765,-17.0773],[3785,-16.3006],[3793,-16.7278],[3811,-14.0578],[3836,-17.3807],[3842,-17.1746],[3845,-17.2018],[3866,-16.1716],[3873,-16.5014],[3909,-16.3006],[3917,-16.7278],[3935,-14.0578],[3960,-17.3807],[3966,-17.1746],[3969,-17.2018],[3990,-16.1716],[3997,-16.5014],[4033,-16.3006],[4041,-16.7278],[4059,-14.0578],[4084,-17.3807],[4090,-17.1746],[4093,-17.2018],[4114,-16.1716],[4121,-16.5014],[4158,-16.3006],[4166,-16.7278],[4184,-14.0578],[4209,-17.3807],[4215,-17.1746],[4218,-17.2018],[4239,-16.1716],[4246,-16.5014],[4284,-16.3006],[4292,-16.7278],[4310,-14.0578],[4335,-17.3807],[4341,-17.1746],[4344,-17.2018],[4360,146.2591],[4364,31.6457],[4365,-13.9146],[4385,-17.1678],[4390,-16.9937],[4394,-17.0621],[4414,-16.2998],[4422,-16.7288],[4436,0.0011]]
       }
       hand_l_ik_ctrl.Scale.X {
        type=MovieSceneScriptingFloatChannel
        default=1
        thinned_from=12
        keys [[1790,1],[2927,1.0493],[2938,1.042],[3058,1.0418],[3070,1.0356],[3083,1.0356],[3099,1.0443],[3185,1.0443],[4364,0.9972],[4436,1]]
       }
       hand_l_ik_ctrl.Scale.Y {
        type=MovieSceneScriptingFloatChannel
        default=1
        thinned_from=12
        keys [[1790,1],[2927,1.0493],[2938,1.042],[3058,1.0418],[3070,1.0356],[3083,1.0356],[3099,1.0443],[3185,1.0443],[4364,0.9972],[4436,1]]
       }
       hand_l_ik_ctrl.Scale.Z {
        type=MovieSceneScriptingFloatChannel
        default=1
        thinned_from=12
        keys [[1790,1],[2927,1.0493],[2938,1.042],[3058,1.0418],[3070,1.0356],[3083,1.0356],[3099,1.0443],[3185,1.0443],[4364,0.9972],[4436,1]]
       }
       Softness {type=MovieSceneScriptingFloatChannel, default=0, const [0,12]}
       hand_l_ikorient_ctrl.Rotation.Roll {type=MovieSceneScriptingFloatChannel, default=0, const [0,12]}
       hand_l_ikorient_ctrl.Rotation.Pitch {type=MovieSceneScriptingFloatChannel, default=0, const [0,12]}
       hand_l_ikorient_ctrl.Rotation.Yaw {type=MovieSceneScriptingFloatChannel, default=0, const [0,12]}
       arm_l_pv_ik_ctrl.Location.X {
        type=MovieSceneScriptingFloatChannel
        default=4.279
        keys [[1790,4.279],[1809,2.5563],[1844,7.4943],[1889,3.1115],[1922,7.8432],[1969,2.5539],[2003,7.4515],[2049,3.1122],[2082,7.8381],[2129,2.5563],[2164,7.4943],[2209,3.1115],[2242,7.8432],[2289,2.5539],[2323,7.4515],[2369,3.1122],[2402,7.8381],[2449,2.5563],[2471,6.125],[2496,-6.3512],[2501,-7.5108],[2507,-4.6891],[2510,-3.4819],[2512,1.6661],[2513,12.169],[2520,-15.6365],[2527,-10.8248],[2528,-10.9779],[2531,-10.3748],[2776,-1.2573],[2787,-4.5016],[2927,-23.4774],[2938,41.3649],[2986,-50.8771],[3058,-50.8771],[3070,93.247],[3083,93.247],[3099,-53.2921],[3185,-53.2921],[3244,2.5563],[3279,7.4943],[3324,3.1115],[3357,7.8432],[3404,2.5539],[3438,7.4515],[3484,3.1122],[3517,7.8381],[3564,2.5563],[3599,7.4943],[3644,3.1115],[3677,7.8432],[3724,2.5539],[3758,7.4515],[3804,3.1122],[3837,7.8381],[3884,2.5563],[3928,3.1122],[3961,7.8381],[4008,2.5563],[4052,3.1122],[4085,7.8381],[4132,2.5563],[4177,3.1122],[4210,7.8381],[4257,2.5563],[4303,3.1122],[4336,7.8381],[4360,-47.6433],[4364,-46.0505],[4388,7.4943],[4433,3.1115],[4435,3.1857],[4436,0.0007]]
       }
       arm_l_pv_ik_ctrl.Location.Y {
        type=MovieSceneScriptingFloatChannel
        default=-15.588
        keys [[1790,-15.588],[1803,-16.0987],[1842,-14.0107],[1883,-16.0543],[1921,-13.98],[1963,-16.1008],[2001,-13.9453],[2043,-16.0546],[2080,-13.9802],[2123,-16.0987],[2162,-14.0107],[2203,-16.0543],[2241,-13.98],[2283,-16.1008],[2321,-13.9453],[2363,-16.0546],[2400,-13.9802],[2443,-16.0987],[2484,51.4698],[2501,71.0566],[2505,67.5363],[2511,82.0706],[2513,41.5049],[2515,12.375],[2518,-4.7134],[2520,-8.339],[2527,-12.9503],[2772,-22.9376],[2927,-1125.1108],[2938,-979.1877],[2986,-915.748],[3058,-915.748],[3070,-789.0583],[3083,-789.0583],[3099,-928.2495],[3185,-928.2495],[3238,-16.0987],[3277,-14.0107],[3318,-16.0543],[3356,-13.98],[3398,-16.1008],[3436,-13.9453],[3478,-16.0546],[3515,-13.9801],[3558,-16.0987],[3597,-14.0107],[3638,-16.0543],[3676,-13.98],[3718,-16.1008],[3756,-13.9453],[3798,-16.0546],[3835,-13.9801],[3878,-16.0987],[3922,-16.0546],[3959,-13.9801],[4002,-16.0987],[4046,-16.0546],[4083,-13.9801],[4126,-16.0987],[4171,-16.0546],[4208,-13.9801],[4251,-16.0987],[4297,-16.0546],[4334,-13.9801],[4360,49.8972],[4364,27.779],[4386,-14.0107],[4427,-16.0543],[4431,-16.0051],[4433,-15.9537],[4435,-15.8627],[4436,-0.0008]]
       }
       arm_l_pv_ik_ctrl.Location.Z {
        type=MovieSceneScriptingFloatChannel
        default=13.5194
        keys [[1790,13.5194],[1820,21.5739],[1864,12.8742],[1900,22.0296],[1944,13.0773],[1980,21.5735],[2024,12.8827],[2060,22.0276],[2104,13.081],[2140,21.5739],[2184,12.8742],[2220,22.0295],[2264,13.0772],[2300,21.5735],[2344,12.8827],[2380,22.0276],[2424,13.081],[2460,21.5739],[2469,20.7736],[2470,20.7778],[2505,-41.7673],[2508,-42.0681],[2511,-39.8346],[2512,-43.4815],[2515,-4.3817],[2519,-13.4785],[2520,-13.1906],[2523,-14.9882],[2757,-7.502],[2777,24.8853],[2786,23.2683],[2800,23.5659],[2927,-368.2859],[2938,-389.3984],[2986,-351.6665],[3058,-351.6665],[3070,-435.3859],[3083,-435.3859],[3099,-281.6357],[3185,-281.6357],[3219,13.0811],[3255,21.5739],[3299,12.8742],[3335,22.0295],[3379,13.0772],[3415,21.5735],[3459,12.8827],[3495,22.0276],[3539,13.081],[3575,21.5739],[3619,12.8742],[3655,22.0295],[3699,13.0772],[3735,21.5735],[3779,12.8827],[3815,22.0276],[3859,13.0811],[3903,12.8827],[3939,22.0276],[3983,13.0811],[4027,12.8827],[4063,22.0276],[4107,13.0811],[4152,12.8827],[4188,22.0276],[4232,13.0811],[4278,12.8827],[4314,22.0276],[4360,19.576],[4364,16.959],[4408,12.8742],[4435,20.9552],[4436,0.0014]]
       }
       arm_l_pv_ik_ctrl.Rotation.X {
        type=MovieSceneScriptingFloatChannel
        default=11.5537
        thinned_from=79
        keys [[1790,11.5537],[1815,13.4023],[1865,11.4323],[1895,13.4023],[1945,11.4431],[1975,13.4029],[2025,11.4325],[2055,13.4028],[2105,11.4434],[2135,13.4023],[2185,11.4323],[2215,13.4023],[2265,11.4431],[2295,13.4029],[2345,11.4325],[2375,13.4028],[2425,11.4434],[2455,13.4023],[2467,13.2092],[2468,13.2121],[2501,-42.2583],[2506,-37.5],[2511,-59.9359],[2513,-18.8687],[2518,-33.0589],[2522,-21.0187],[2525,-23.0028],[2528,-21.3875],[2536,-22.294],[2765,-8.1995],[2777,27.0818],[2789,21.8114],[2802,21.8134],[2927,22.3199],[2938,10.5085],[2986,-0.6417],[3058,-0.6417],[3070,-105.3255],[3083,-105.3255],[3099,15.2574],[3185,15.2574],[3220,11.4434],[3250,13.4023],[3300,11.4324],[3330,13.4023],[3380,11.4431],[3410,13.4029],[3460,11.4325],[3490,13.4028],[3540,11.4433],[3570,13.4023],[3620,11.4324],[3650,13.4024],[3700,11.4431],[3730,13.4029],[3780,11.4325],[3810,13.4028],[3860,11.4434],[3904,11.4325],[3934,13.4028],[3984,11.4434],[4028,11.4325],[4058,13.4028],[4108,11.4434],[4153,11.4325],[4183,13.4028],[4233,11.4434],[4279,11.4325],[4309,13.4028],[4360,-113.1778],[4364,13.3538],[4365,13.4023],[4409,11.4324],[4435,13.3409],[4436,-0.0009]]
       }
       arm_l_pv_ik_ctrl.Rotation.Y {
        type=MovieSceneScriptingFloatChannel
        default=17.2199
        thinned_from=84
        keys [[1790,17.2199],[1813,11.1122],[1850,18.8678],[1894,11.1081],[1930,18.964],[1973,11.109],[2010,18.8385],[2054,11.105],[2090,18.9658],[2133,11.1122],[2170,18.8678],[2214,11.1081],[2250,18.964],[2293,11.109],[2330,18.8385],[2374,11.105],[2410,18.9658],[2453,11.1122],[2466,12.745],[2471,17.0796],[2502,3.8113],[2505,3.0693],[2506,3.0715],[2507,3.0535],[2511,18.5319],[2512,10.6272],[2513,10.447],[2515,20.3068],[2518,9.8414],[2520,-1.1292],[2529,7.0359],[2767,26.2412],[2777,16.9387],[2788,18.3896],[2795,18.3903],[2927,15.7652],[2938,15.2511],[2986,6.5301],[3058,6.5301],[3070,42.0003],[3083,42.0003],[3099,56.2897],[3185,56.2897],[3248,11.1122],[3285,18.8678],[3329,11.1081],[3365,18.964],[3408,11.1091],[3445,18.8385],[3489,11.105],[3525,18.9658],[3568,11.1122],[3605,18.8678],[3649,11.1081],[3685,18.9641],[3728,11.1091],[3765,18.8385],[3809,11.105],[3845,18.9658],[3888,11.1122],[3933,11.105],[3969,18.9658],[4012,11.1122],[4057,11.105],[4093,18.9658],[4136,11.1122],[4182,11.105],[4218,18.9658],[4261,11.1122],[4308,11.105],[4344,18.9658],[4360,-9.8811],[4364,-11.6723],[4394,18.8678],[4415,17.1638],[4425,14.3299],[4430,12.4161],[4433,11.5707],[4434,11.3842],[4435,11.2479],[4436,-0.0013]]
       }
       arm_l_pv_ik_ctrl.Rotation.Z {
        type=MovieSceneScriptingFloatChannel
        default=-13.6877
        thinned_from=73
        keys [[1790,-13.6877],[1828,-5.308],[1869,-13.7114],[1909,-5.1812],[1949,-13.6949],[1989,-5.1702],[2029,-13.7113],[2069,-5.1778],[2109,-13.6948],[2148,-5.308],[2189,-13.7114],[2229,-5.1812],[2269,-13.6949],[2309,-5.1702],[2349,-13.7113],[2389,-5.1778],[2429,-13.6948],[2467,-5.3249],[2501,50.0943],[2506,48.2635],[2507,48.2935],[2511,40.9109],[2512,42.7619],[2518,-45.5628],[2521,-64.7554],[2527,-59.358],[2530,-59.2724],[2776,-12.8318],[2791,-20.8573],[2802,-20.8568],[2927,-11.9218],[2938,-38.5673],[2986,32.0167],[3058,32.0167],[3070,-181.2422],[3083,-181.2422],[3099,-38.6225],[3185,-38.6225],[3224,-13.6948],[3263,-5.3079],[3304,-13.7113],[3344,-5.1812],[3384,-13.6949],[3424,-5.1702],[3464,-13.7114],[3504,-5.1778],[3544,-13.6948],[3583,-5.3079],[3624,-13.7113],[3664,-5.1812],[3704,-13.6949],[3744,-5.1702],[3784,-13.7114],[3824,-5.1778],[3864,-13.6948],[3908,-13.7114],[3948,-5.1778],[3988,-13.6948],[4032,-13.7114],[4072,-5.1778],[4112,-13.6948],[4157,-13.7114],[4197,-5.1778],[4237,-13.6948],[4283,-13.7114],[4323,-5.1778],[4360,-100.0005],[4364,-43.7478],[4372,-5.3079],[4413,-13.7113],[4436,0.0021]]
       }
       arm_l_pv_ik_ctrl.Scale.X {type=MovieSceneScriptingFloatChannel, default=1, const [1,12]}
       arm_l_pv_ik_ctrl.Scale.Y {type=MovieSceneScriptingFloatChannel, default=1, const [1,12]}
       arm_l_pv_ik_ctrl.Scale.Z {type=MovieSceneScriptingFloatChannel, default=1, const [1,12]}
       upperarm_r_fk_ctrl.Location.X {type=MovieSceneScriptingFloatChannel, default=0, thinned_from=1236, keys [[1790,0],[4366,0]]}
       upperarm_r_fk_ctrl.Location.Y {type=MovieSceneScriptingFloatChannel, default=0, thinned_from=1176, keys [[1790,0],[2512,0],[2513,-0.0001],[2536,0],[4367,0]]}
       upperarm_r_fk_ctrl.Location.Z {type=MovieSceneScriptingFloatChannel, default=0, thinned_from=984, keys [[1790,0],[2512,0],[2513,-0.0001],[2536,0],[4366,0]]}
       upperarm_r_fk_ctrl.Rotation.X {
        type=MovieSceneScriptingFloatChannel
        default=-8.95
        thinned_from=77
        keys [[1790,-8.95],[1792,-9.0336],[1830,-1.1218],[1872,-8.9826],[1909,-0.9809],[1952,-9.0343],[1990,-1.1154],[2032,-8.9859],[2069,-0.9865],[2112,-9.0336],[2150,-1.1218],[2192,-8.9826],[2229,-0.981],[2272,-9.0343],[2310,-1.1154],[2352,-8.9859],[2389,-0.9865],[2432,-9.0336],[2468,-1.1601],[2471,-2.1406],[2502,30.5128],[2503,30.2603],[2511,33.185],[2512,28.183],[2513,-7.2633],[2515,-43.8377],[2519,-59.4444],[2523,-52.1781],[2529,-54.4589],[2531,-54.1131],[2532,-54.179],[2538,-53.6671],[2757,-42.4048],[2766,-40.8782],[2777,-20.4808],[2790,-4.6194],[2806,1.9369],[2882,-9.4631],[2910,-6.5047],[2938,-17.7576],[2957,-20.2281],[2965.5525,-2.5868],[2968.9288,-14.4157],[2977.7188,-110.3981],[2983.6637,-114.2196],[2999,-115.2558],[3058,-115.9678],[3065.225,-165.1177],[3066,-162.7931],[3067,-56.0972],[3067.4362,28.4332],[3068.435,21.2899],[3070,16.1439],[3099,16.1439],[3185,13.3082],[3199,37.6318],[3209,61.9151],[3216,52.3906],[3227,5.8126],[3237,-7.3847],[3265,-1.1218],[3307,-8.9825],[3344,-0.981],[3387,-9.0343],[3425,-1.1154],[3467,-8.9859],[3504,-0.9866],[3547,-9.0335],[3585,-1.1217],[3627,-8.9825],[3664,-0.981],[3707,-9.0342],[4364,-4.1432],[4365,-4.1485],[4366,-133.753]]
       }
       upperarm_r_fk_ctrl.Rotation.Y {
        type=MovieSceneScriptingFloatChannel
        default=15.3543
        keys [[1790,15.3543],[1817,10.6635],[1852,16.1964],[1898,9.9608],[1933,16.0504],[1978,10.6634],[2012,16.179],[2058,9.9611],[2093,16.0582],[2137,10.6635],[2172,16.1964],[2218,9.9609],[2253,16.0504],[2298,10.6634],[2332,16.179],[2378,9.9611],[2413,16.0582],[2457,10.6635],[2467,11.2876],[2472,12.6018],[2496,15.5179],[2501,15.4633],[2507,19.6099],[2511,15.6202],[2513,38.2368],[2520,5.2112],[2524,13.8467],[2526,13.3351],[2528,13.8021],[2538,11.3146],[2757,12.9708],[2763,17.8685],[2785,8.948],[2810,4.4587],[2871,8.4215],[2920,2.9195],[2946,6.442],[2986,7.7036],[3058,7.7036],[3065,12.5],[3066,10],[3067,46.3357],[3068,18.2446],[3070,15.2172],[3083,24.7319],[3099,24.7319],[3185,35.938],[3197,35.6861],[3205,31.4577],[3218,18.6989],[3244,2.4207],[3284,16.1964],[3330,9.9609],[3365,16.0504],[3410,10.6634],[3444,16.1789],[3490,9.9611],[3525,16.0582],[3569,10.6635],[3604,16.1964],[3650,9.9609],[3685,16.0504],[3730,10.6634],[4364,10.699],[4365,10.6635],[4366,10.6635]]
       }
       upperarm_r_fk_ctrl.Rotation.Z {
        type=MovieSceneScriptingFloatChannel
        default=-20.6216
        keys [[1790,-20.6216],[1793,-20.7229],[1833,-15.4727],[1873,-20.6957],[1913,-15.4091],[1953,-20.7245],[1993,-15.4693],[2033,-20.7045],[2073,-15.4054],[2113,-20.7229],[2153,-15.4727],[2193,-20.6957],[2233,-15.409],[2273,-20.7245],[2313,-15.4693],[2353,-20.7046],[2393,-15.4054],[2433,-20.7229],[2462,-16.4196],[2469,-15.3263],[2502,119.9729],[2503,119.0543],[2511,134.5184],[2515,8.1594],[2520,-23.9583],[2521,-22.4184],[2522,-22.5509],[2523,-22.3948],[2524,-22.4683],[2532,-17.9869],[2536,-18.1156],[2757,-38.9229],[2773,-33.1917],[2927,-32.5027],[2938,-32.5027],[2946,-33.363],[2986,-99.6175],[3058,-99.6175],[3065,-92.1057],[3066,-95],[3067,21.1725],[3068,35.1839],[3069,40.5856],[3071,37.1842],[3083,44.0472],[3099,44.0472],[3185,34.1728],[3200,44.0472],[3203,42.4289],[3212,10.6767],[3219,-15.4564],[3228,-24.1158],[3249,-10.0996],[3264,-15.4728],[3273,-20.5],[3308,-20.6957],[3348,-15.4091],[3388,-20.7245],[3428,-15.4693],[3468,-20.7046],[3508,-15.4054],[3548,-20.7228],[3588,-15.4727],[3628,-20.6957],[3668,-15.4091],[3708,-20.7245],[4364,-17.8026],[4365,-17.9011],[4366,-17.9011]]
       }
       upperarm_r_fk_ctrl.Scale.X {type=MovieSceneScriptingFloatChannel, default=1, const [1,14]}
       upperarm_r_fk_ctrl.Scale.Y {type=MovieSceneScriptingFloatChannel, default=1, const [1,14]}
       upperarm_r_fk_ctrl.Scale.Z {type=MovieSceneScriptingFloatChannel, default=1, const [1,14]}
       lowerarm_r_fk_ctrl.Location.X {type=MovieSceneScriptingFloatChannel, default=0, thinned_from=1206, keys [[1790,0],[2512,0],[2513,-0.0006],[2515,0],[4364,0]]}
       lowerarm_r_fk_ctrl.Location.Y {type=MovieSceneScriptingFloatChannel, default=0, thinned_from=1233, keys [[1790,0],[2511,0],[2513,0.0001],[2515,0],[4364,0]]}
       lowerarm_r_fk_ctrl.Location.Z {type=MovieSceneScriptingFloatChannel, default=0, thinned_from=1213, keys [[1790,0],[4364,0]]}
       lowerarm_r_fk_ctrl.Rotation.X {
        type=MovieSceneScriptingFloatChannel
        default=-0.1266
        thinned_from=74
        keys [[1790,-0.1266],[1822,0.0291],[1862,-0.1282],[1902,0.0291],[1942,-0.1282],[1982,0.0291],[2032,-0.1543],[2062,0.0203],[2102,-0.1282],[2142,0.0291],[2182,-0.1282],[2222,0.0291],[2263,-0.1282],[2302,0.0291],[2352,-0.1543],[2382,0.0203],[2422,-0.1282],[2462,0.0291],[2467,-0.0206],[2501,-3.4505],[2504,-3.1538],[2506,-3.183],[2507,-3.1578],[2511,-3.6723],[2515,-1.2692],[2520,-1.0779],[2528,-1.1263],[2764,-6.1141],[2769,-13.7455],[2774,-22.181],[2777,-19.2796],[2781,-6.2714],[2783,-2.5408],[2789,-0.1817],[2927,-0.22],[2938,-0.2215],[2986,-1.1585],[3058,-0.0952],[3064,-1.6215],[3067,-0.9518],[3070,0.0003],[3099,0.0003],[3185,0.0195],[3217,-0.1282],[3257,0.0291],[3297,-0.1282],[3337,0.0291],[3378,-0.1282],[3417,0.0291],[3467,-0.1543],[3497,0.0203],[3537,-0.1282],[3577,0.0291],[3617,-0.1282],[3657,0.0291],[3697,-0.1282],[3737,0.0291],[3787,-0.1543],[3817,0.0203],[3857,-0.1282],[3911,-0.1543],[3941,0.0203],[3981,-0.1282],[4035,-0.1543],[4065,0.0203],[4105,-0.1282],[4160,-0.1543],[4190,0.0203],[4230,-0.1282],[4286,-0.1543],[4316,0.0203],[4356,-0.1282],[4364,0.0128]]
       }
       lowerarm_r_fk_ctrl.Rotation.Y {
        type=MovieSceneScriptingFloatChannel
        default=-0.4102
        thinned_from=75
        keys [[1790,-0.4102],[1822,0.117],[1863,-0.4144],[1902,0.117],[1942,-0.4144],[1982,0.117],[2032,-0.4844],[2062,0.0792],[2103,-0.4144],[2142,0.117],[2183,-0.4144],[2222,0.117],[2262,-0.4144],[2302,0.117],[2352,-0.4844],[2382,0.0792],[2423,-0.4144],[2462,0.117],[2468,0.0288],[2471,-0.4851],[2496,-3.5084],[2504,-3.1939],[2507,-3.194],[2509,-3.3434],[2511,-4.4716],[2512,-3.0649],[2515,0.8203],[2518,0.6583],[2519,0.8144],[2520,0.7936],[2528,1.1423],[2759,0.9995],[2775,-0.5375],[2779,-0.3966],[2927,-0.5559],[2938,-0.5591],[2986,-3.3915],[3058,-3.3901],[3064,-0.7898],[3067,0.2147],[3070,0.1174],[3099,0.1174],[3185,0.1158],[3218,-0.4144],[3257,0.117],[3298,-0.4144],[3337,0.117],[3378,-0.4144],[3417,0.117],[3467,-0.4844],[3497,0.0792],[3538,-0.4144],[3577,0.117],[3618,-0.4144],[3657,0.117],[3698,-0.4144],[3737,0.117],[3787,-0.4844],[3817,0.0792],[3858,-0.4144],[3911,-0.4844],[3941,0.0792],[3982,-0.4144],[4035,-0.4844],[4065,0.0792],[4106,-0.4144],[4160,-0.4844],[4190,0.0792],[4231,-0.4144],[4286,-0.4844],[4316,0.0792],[4357,-0.4144],[4364,0.0591]]
       }
       lowerarm_r_fk_ctrl.Rotation.Z {
        type=MovieSceneScriptingFloatChannel
        default=5.5311
        keys [[1790,5.5311],[1822,-1.5157],[1862,5.5892],[1902,-1.5149],[1943,5.5891],[1982,-1.5149],[2032,6.5458],[2062,-1.0165],[2103,5.5892],[2142,-1.5157],[2183,5.5892],[2222,-1.5149],[2262,5.5891],[2302,-1.5149],[2352,6.5458],[2382,-1.0165],[2422,5.5892],[2462,-1.5157],[2471,1.5725],[2502,63.5573],[2504,60.3655],[2506,60.7933],[2507,60.4268],[2511,65.3444],[2513,40.618],[2515,38.9807],[2518,61.6626],[2520,68.5735],[2528,56.3682],[2778,5.5232],[2927,9.4347],[2952,30.8099],[2961,31.8912],[3058,17.9655],[3065.0775,51.5711],[3067,68.2686],[3068,-0.1215],[3069,-12.1507],[3076,-14.7316],[3099,-14.7316],[3185,-5.3086],[3204,1.1582],[3213,14.3283],[3232,-6.9313],[3237,-9.5634],[3251,2.912],[3268,-2.7495],[3287,6.5558],[3298,5.5892],[3337,-1.5149],[3377,5.5891],[3417,-1.5149],[3467,6.5458],[3497,-1.0165],[3537,5.5892],[3577,-1.5157],[3617,5.5892],[3657,-1.5149],[3697,5.5891],[3737,-1.5149],[3787,6.5459],[3817,-1.0165],[3858,5.5892],[3911,6.5459],[4106,5.5892],[4160,6.5459],[4190,-1.0165],[4231,5.5892],[4286,6.5459],[4316,-1.0165],[4357,5.5892],[4364,6.2751],[4366,-13.6697]]
       }
       lowerarm_r_fk_ctrl.Scale.X {type=MovieSceneScriptingFloatChannel, default=1, const [1,10]}
       lowerarm_r_fk_ctrl.Scale.Y {type=MovieSceneScriptingFloatChannel, default=1, const [1,10]}
       lowerarm_r_fk_ctrl.Scale.Z {type=MovieSceneScriptingFloatChannel, default=1, const [1,10]}
       hand_r_fk_ctrl.Location.X {
        type=MovieSceneScriptingFloatChannel
        default=0
        thinned_from=61
        keys [[1790,0],[2512,-0.0001],[2513,-0.0012],[2515,-0.0001],[4435,0],[4436,-0.0003]]
       }
       hand_r_fk_ctrl.Location.Y {type=MovieSceneScriptingFloatChannel, default=0, thinned_from=123, keys [[1790,0],[2512,0],[2513,-0.0003],[2515,0],[4436,0]]}
       hand_r_fk_ctrl.Location.Z {type=MovieSceneScriptingFloatChannel, default=0, thinned_from=988, keys [[1790,0],[2512,0],[2513,-0.0004],[2515,0],[4436,0]]}
       hand_r_fk_ctrl.Rotation.X {
        type=MovieSceneScriptingFloatChannel
        default=0.2502
        thinned_from=87
        keys [[1790,0.2502],[1829,1.0595],[1868,0.234],[1900,1.0585],[1904,1.0595],[1948,0.234],[1986,1.0594],[2028,0.234],[2066,1.0594],[2108,0.234],[2149,1.0595],[2188,0.234],[2220,1.0585],[2224,1.0595],[2268,0.234],[2306,1.0594],[2348,0.234],[2386,1.0594],[2428,0.234],[2451,0.9991],[2463,1.0553],[2469,1.4851],[2506,-81.7396],[2511,-81.7247],[2515,-67.3564],[2522,-88.1747],[2527,-90.4498],[2760,-80.8148],[2768,-41.796],[2777,16.25],[2788,1.5538],[2927,1.5547],[2938,1.5499],[2963.7125,-36.6382],[2978.61,-76.4364],[3022.6288,-81.143],[3058,-81.148],[3063.8425,-70.6385],[3065,-34.7476],[3067,9.6953],[3070,14.9329],[3079,14.9329],[3085,15.4452],[3089,14.9329],[3185,17.1611],[3223,0.2341],[3264,1.0595],[3303,0.2341],[3335,1.0585],[3339,1.0595],[3383,0.2341],[3421,1.0594],[3463,0.2341],[3501,1.0594],[3543,0.2341],[3584,1.0595],[3623,0.2341],[3655,1.0585],[3659,1.0595],[3703,0.2341],[3741,1.0594],[3783,0.2341],[3821,1.0594],[3863,0.2341],[3907,0.2341],[3945,1.0594],[3987,0.2341],[4031,0.2341],[4069,1.0594],[4111,0.2341],[4156,0.2341],[4194,1.0594],[4236,0.2341],[4282,0.2341],[4320,1.0594],[4360,0.8089],[4373,1.0595],[4412,0.2341],[4435,1.0384],[4436,0.0006]]
       }
       hand_r_fk_ctrl.Rotation.Y {
        type=MovieSceneScriptingFloatChannel
        default=6.2188
        thinned_from=139
        keys [[1790,6.2188],[1827,6.7818],[1831,6.7828],[1868,6.2085],[1896,6.7784],[1904,6.7798],[1948,6.2085],[1980,6.7807],[1986,6.7806],[2028,6.2085],[2060,6.7807],[2066,6.7806],[2108,6.2085],[2147,6.7818],[2151,6.7828],[2188,6.2085],[2216,6.7784],[2224,6.7798],[2268,6.2085],[2300,6.7807],[2306,6.7806],[2348,6.2085],[2380,6.7807],[2386,6.7806],[2428,6.2085],[2452,6.745],[2464,6.7794],[2470,6.9344],[2506,18.2146],[2510,18.2102],[2512,18.268],[2518,7.4428],[2519,8.6351],[2523,-0.4386],[2528,6.6458],[2530,6.3361],[2532,6.8667],[2533,6.8392],[2538,7.0042],[2760,3.2589],[2777,14.8309],[2927,15.6761],[2938,15.6819],[2963.695,14.6678],[2986,2.0838],[3012,-16.927],[3058,-17.0461],[3065,-2.2458],[3067,13.8445],[3069,16.1558],[3079,-4.0369],[3085,-0.6095],[3089,-4.0369],[3185,27.6071],[3223,6.2086],[3262,6.7818],[3266,6.7828],[3303,6.2085],[3331,6.7784],[3339,6.7798],[3383,6.2086],[3415,6.7807],[3421,6.7806],[3463,6.2086],[3495,6.7807],[3501,6.7806],[3543,6.2086],[3582,6.7818],[3586,6.7828],[3623,6.2086],[3651,6.7784],[3659,6.7798],[3703,6.2086],[3735,6.7807],[3741,6.7806],[3783,6.2086],[3815,6.7807],[3821,6.7806],[3863,6.2086],[3907,6.2086],[3939,6.7807],[3945,6.7806],[3987,6.2086],[4031,6.2086],[4063,6.7807],[4069,6.7806],[4111,6.2086],[4156,6.2086],[4188,6.7807],[4194,6.7806],[4236,6.2086],[4282,6.2086],[4314,6.7807],[4320,6.7806],[4360,6.6332],[4371,6.7818],[4375,6.7828],[4412,6.2086],[4435,6.7659],[4436,0.0018]]
       }
       hand_r_fk_ctrl.Rotation.Z {
        type=MovieSceneScriptingFloatChannel
        default=2.1652
        thinned_from=80
        keys [[1790,2.1652],[1831,-2.8161],[1868,2.2633],[1904,-2.8136],[1948,2.2633],[1986,-2.8143],[2028,2.2633],[2066,-2.8143],[2108,2.2633],[2151,-2.8161],[2188,2.2633],[2224,-2.8136],[2268,2.2633],[2306,-2.8143],[2348,2.2633],[2386,-2.8143],[2428,2.2633],[2466,-2.8079],[2471,-2.524],[2496,2.5556],[2506,2.731],[2510,2.728],[2512,5.9669],[2515,-7.7559],[2518,-7.4907],[2519,-7.5179],[2521,-6.8847],[2523,-11.4527],[2528,-5.3164],[2768,-7.5997],[2777,-0.2302],[2927,-2.3339],[2946.77,-2.4119],[2965.7638,7.2961],[2971.955,10.1328],[2982.2725,-20.4905],[2995.8025,-22.0399],[3058,-22.5837],[3065,-25.0104],[3067,-43.8836],[3074,-47.9427],[3080,-47.6244],[3087,-60.2189],[3183,-69.4232],[3194,-70.8393],[3223,2.2633],[3266,-2.8161],[3303,2.2633],[3339,-2.8136],[3383,2.2633],[3421,-2.8143],[3463,2.2633],[3501,-2.8143],[3543,2.2633],[3586,-2.8161],[3623,2.2633],[3659,-2.8136],[3703,2.2633],[3741,-2.8143],[3783,2.2633],[3821,-2.8143],[3863,2.2633],[3907,2.2633],[3945,-2.8143],[3987,2.2633],[4031,2.2633],[4069,-2.8143],[4111,2.2633],[4156,2.2633],[4194,-2.8143],[4236,2.2633],[4282,2.2633],[4320,-2.8143],[4360,-1.0622],[4375,-2.8161],[4412,2.2633],[4435,-2.6834],[4436,-0.0007]]
       }
       hand_r_fk_ctrl.Scale.X {type=MovieSceneScriptingFloatChannel, default=1, const [1,12]}
       hand_r_fk_ctrl.Scale.Y {type=MovieSceneScriptingFloatChannel, default=1, const [1,12]}
       hand_r_fk_ctrl.Scale.Z {type=MovieSceneScriptingFloatChannel, default=1, const [1,12]}
       thumb_01_r_ctrl.Location.X {
        type=MovieSceneScriptingFloatChannel
        default=0
        thinned_from=37
        keys [[1790,0],[2938,0],[2986,0.0002],[3099,0.0002],[3185,-0.0007],[4436,0]]
       }
       thumb_01_r_ctrl.Location.Y {type=MovieSceneScriptingFloatChannel, default=0, thinned_from=30, keys [[1790,0],[2938,0],[2986,-0.0006],[3185,-0.0007],[4436,0]]}
       thumb_01_r_ctrl.Location.Z {
        type=MovieSceneScriptingFloatChannel
        default=0
        thinned_from=29
        keys [[1790,0],[2769,0],[2788,0.0002],[2938,0.0001],[2986,-0.0008],[3099,-0.0008],[3185,-0.0002],[4436,0]]
       }
       thumb_01_r_ctrl.Rotation.X {
        type=MovieSceneScriptingFloatChannel
        default=-13.514
        thinned_from=148
        keys [[1790,-13.514],[1809,-19.7011],[1828,-20.6343],[1831,-20.6911],[1848,-14.2686],[1865,-13.3952],[1886,-18.9439],[1908,-20.6343],[1911,-20.6911],[1926,-14.9635],[1942,-13.4045],[1945,-13.387],[1966,-18.9439],[1988,-20.6343],[1991,-20.6911],[2006,-14.9636],[2022,-13.4045],[2025,-13.387],[2046,-18.9439],[2068,-20.6343],[2071,-20.6911],[2086,-14.9636],[2102,-13.4045],[2105,-13.4035],[2126,-18.9439],[2148,-20.6343],[2151,-20.6911],[2168,-14.2686],[2185,-13.3952],[2206,-18.9439],[2228,-20.6343],[2231,-20.6911],[2246,-14.9636],[2262,-13.4045],[2265,-13.387],[2286,-18.9439],[2308,-20.6343],[2311,-20.6911],[2326,-14.9636],[2342,-13.4045],[2345,-13.387],[2368,-19.5042],[2391,-20.6911],[2406,-14.9636],[2422,-13.4045],[2425,-13.4035],[2446,-18.9439],[2467,-20.6038],[2504,2.9569],[2508,11.3803],[2512,14.6135],[2515,14.6135],[2520,7.7358],[2522,12.5735],[2524,9.3345],[2527,13.4329],[2530,13.8846],[2759,5.7205],[2761,0.3869],[2771,-8.3323],[2780,-18.0059],[2794,-11.8505],[2927,-10.135],[2938,-10.1214],[2986,0.9412],[3099,0.9412],[3185,8.1678],[3220,-13.4035],[3243,-19.5039],[3266,-20.6911],[3283,-14.269],[3300,-13.3952],[3323,-19.5039],[3346,-20.6911],[3361,-14.9641],[3377,-13.4045],[3381,-13.387],[3402,-19.25],[3423,-20.6343],[3426,-20.6911],[3443,-14.283],[3460,-13.387],[3483,-19.5039],[3506,-20.6911],[3523,-14.283],[3540,-13.4035],[3563,-19.5039],[3586,-20.6911],[3603,-14.2689],[3620,-13.3952],[3641,-18.9435],[3663,-20.6343],[3666,-20.6911],[3683,-14.283],[3700,-13.387],[3723,-19.5039],[3746,-20.6911],[3763,-14.283],[3780,-13.387],[3801,-18.9435],[3823,-20.6343],[3826,-20.6911],[3841,-14.964],[3857,-13.4045],[3860,-13.4035],[3881,-18.9434],[3904,-13.387],[3925,-18.9435],[3947,-20.6343],[3950,-20.6911],[3965,-14.964],[3981,-13.4045],[3984,-13.4035],[4005,-18.9434],[4028,-13.387],[4049,-18.9435],[4071,-20.6343],[4074,-20.6911],[4089,-14.964],[4105,-13.4045],[4108,-13.4035],[4129,-18.9434],[4153,-13.387],[4174,-18.9435],[4196,-20.6343],[4199,-20.6911],[4214,-14.964],[4230,-13.4045],[4233,-13.4035],[4254,-18.9434],[4279,-13.387],[4300,-18.9435],[4322,-20.6343],[4325,-20.6911],[4340,-14.964],[4356,-13.4045],[4360,-19.7263],[4364,-20.1129],[4372,-20.6343],[4375,-20.6911],[4392,-14.269],[4409,-13.3952],[4410,-13.3952],[4435,-19.9151],[4436,0]]
       }
       thumb_01_r_ctrl.Rotation.Y {
        type=MovieSceneScriptingFloatChannel
        default=0.9212
        thinned_from=499
        keys [[1790,0.9212],[1806,0.5742],[1808,0.5867],[1811,0.5071],[1813,0.5254],[1819,0.489],[1820,0.5018],[1825,0.4815],[1826,0.487],[1828,0.4793],[1831,0.4903],[1832,0.4862],[1859,0.9182],[1860,0.917],[1864,0.9253],[1865,0.9214],[1886,0.5742],[1888,0.5867],[1891,0.5071],[1893,0.5254],[1899,0.489],[1900,0.5018],[1905,0.4815],[1906,0.487],[1908,0.4793],[1911,0.4903],[1913,0.4794],[1929,0.8867],[1930,0.8838],[1934,0.9144],[1936,0.9097],[1938,0.9173],[1939,0.9136],[1941,0.9199],[1942,0.9161],[1945,0.9261],[1954,0.8476],[1955,0.8539],[1966,0.5742],[1968,0.5867],[1971,0.5071],[1973,0.5254],[1979,0.489],[1980,0.5018],[1985,0.4815],[1986,0.487],[1988,0.4793],[1991,0.4903],[1993,0.4794],[2009,0.8867],[2010,0.8838],[2014,0.9144],[2016,0.9097],[2018,0.9173],[2019,0.9136],[2021,0.9199],[2022,0.9161],[2025,0.9261],[2034,0.8476],[2035,0.8539],[2046,0.5742],[2048,0.5867],[2051,0.5071],[2053,0.5254],[2059,0.489],[2060,0.5018],[2065,0.4815],[2066,0.487],[2068,0.4793],[2071,0.4903],[2073,0.4794],[2089,0.8867],[2090,0.8838],[2094,0.9144],[2096,0.9097],[2098,0.9173],[2099,0.9136],[2101,0.9199],[2102,0.9161],[2105,0.9167],[2110,0.9212],[2126,0.5742],[2128,0.5867],[2131,0.5071],[2133,0.5254],[2139,0.489],[2140,0.5018],[2145,0.4815],[2146,0.487],[2148,0.4793],[2151,0.4903],[2152,0.4862],[2179,0.9182],[2180,0.917],[2184,0.9253],[2185,0.9214],[2206,0.5742],[2208,0.5867],[2211,0.5071],[2213,0.5254],[2219,0.489],[2220,0.5018],[2225,0.4815],[2226,0.487],[2228,0.4793],[2231,0.4903],[2233,0.4795],[2249,0.8867],[2250,0.8838],[2254,0.9144],[2256,0.9097],[2258,0.9173],[2259,0.9136],[2261,0.9199],[2262,0.9161],[2265,0.9261],[2274,0.8476],[2275,0.8539],[2286,0.5742],[2288,0.5867],[2291,0.5071],[2293,0.5254],[2299,0.489],[2300,0.5018],[2305,0.4815],[2306,0.487],[2308,0.4793],[2311,0.4903],[2313,0.4794],[2329,0.8867],[2330,0.8838],[2334,0.9144],[2336,0.9097],[2338,0.9173],[2339,0.9136],[2341,0.9199],[2342,0.9161],[2345,0.9261],[2354,0.8476],[2355,0.8539],[2366,0.5742],[2368,0.5867],[2371,0.5071],[2373,0.5254],[2379,0.489],[2380,0.5018],[2385,0.4815],[2386,0.487],[2389,0.4793],[2391,0.4903],[2393,0.4794],[2409,0.8867],[2410,0.8838],[2414,0.9144],[2416,0.9097],[2418,0.9173],[2419,0.9136],[2421,0.9199],[2422,0.9161],[2425,0.9167],[2430,0.9212],[2446,0.5742],[2448,0.5867],[2451,0.5071],[2453,0.5254],[2459,0.489],[2460,0.5018],[2465,0.4815],[2466,0.487],[2496,-7.1352],[2503,-6.8262],[2512,-3.4106],[2515,-3.4107],[2518,-5.3267],[2521,-15.4723],[2523,-14.0291],[2530,-15.9151],[2759,-12.5257],[2761,-10.2953],[2764,-9.2391],[2784,0.5398],[2927,0.0953],[2938,0.0889],[2986,-4.2037],[3099,-4.2037],[3185,-11.0513],[3216,0.9199],[3218,0.9161],[3225,0.9212],[3241,0.5742],[3243,0.5867],[3246,0.5071],[3248,0.5254],[3254,0.489],[3255,0.5018],[3260,0.4815],[3261,0.487],[3263,0.4793],[3266,0.4903],[3267,0.4862],[3294,0.9182],[3295,0.917],[3299,0.9253],[3300,0.9214],[3321,0.5742],[3323,0.5867],[3326,0.5071],[3328,0.5254],[3334,0.489],[3335,0.5018],[3340,0.4815],[3341,0.487],[3343,0.4793],[3346,0.4903],[3348,0.4794],[3364,0.8867],[3365,0.8838],[3369,0.9144],[3371,0.9097],[3373,0.9173],[3374,0.9136],[3376,0.9199],[3377,0.9161],[3380,0.9261],[3389,0.8476],[3390,0.8538],[3401,0.5742],[3403,0.5867],[3406,0.5071],[3408,0.5254],[3414,0.489],[3415,0.5018],[3420,0.4815],[3421,0.487],[3423,0.4793],[3426,0.4903],[3428,0.4794],[3444,0.8867],[3445,0.8838],[3449,0.9144],[3451,0.9097],[3453,0.9173],[3454,0.9136],[3456,0.9199],[3458,0.9161],[3460,0.9261],[3469,0.8476],[3470,0.8538],[3481,0.5742],[3483,0.5867],[3486,0.5071],[3488,0.5254],[3494,0.489],[3495,0.5018],[3500,0.4815],[3501,0.487],[3503,0.4793],[3506,0.4903],[3508,0.4794],[3524,0.8867],[3525,0.8838],[3529,0.9144],[3531,0.9097],[3533,0.9173],[3534,0.9136],[3536,0.9199],[3538,0.9161],[3545,0.9212],[3561,0.5742],[3563,0.5867],[3566,0.5071],[3568,0.5254],[3574,0.489],[3575,0.5018],[3580,0.4815],[3581,0.487],[3583,0.4793],[3586,0.4903],[3587,0.4862],[3614,0.9182],[3615,0.917],[3619,0.9253],[3620,0.9214],[3641,0.5742],[3643,0.5867],[3646,0.5071],[3648,0.5254],[3654,0.489],[3655,0.5018],[3660,0.4815],[3661,0.487],[3663,0.4793],[3666,0.4903],[3668,0.4794],[3684,0.8867],[3685,0.8838],[3689,0.9144],[3691,0.9097],[3693,0.9173],[3694,0.9136],[3696,0.9199],[3698,0.9161],[3700,0.9261],[3709,0.8476],[3710,0.8538],[3721,0.5742],[3723,0.5867],[3726,0.5071],[3728,0.5254],[3734,0.489],[3735,0.5018],[3740,0.4815],[3741,0.487],[3743,0.4793],[3746,0.4903],[3748,0.4794],[3764,0.8867],[3765,0.8838],[3769,0.9144],[3771,0.9097],[3773,0.9173],[3774,0.9136],[3776,0.9199],[3778,0.9161],[3780,0.9261],[3789,0.8476],[3790,0.8538],[3801,0.5742],[3803,0.5867],[3806,0.5071],[3808,0.5254],[3814,0.489],[3815,0.5018],[3820,0.4815],[3821,0.487],[3823,0.4793],[3826,0.4903],[3828,0.4794],[3844,0.8867],[3845,0.8838],[3849,0.9144],[3851,0.9097],[3853,0.9173],[3854,0.9136],[3856,0.9199],[3857,0.9161],[3860,0.9167],[3865,0.9212],[3881,0.5742],[3883,0.5867],[3886,0.5071],[3888,0.5254],[3893,0.9144],[3895,0.9097],[3897,0.9173],[3898,0.9136],[3900,0.9199],[3902,0.9161],[3904,0.9261],[3913,0.8476],[3914,0.8538],[3925,0.5742],[3927,0.5867],[3930,0.5071],[3932,0.5254],[3938,0.489],[3939,0.5018],[3944,0.4815],[3945,0.487],[3947,0.4793],[3950,0.4903],[3952,0.4794],[3968,0.8867],[3969,0.8838],[3973,0.9144],[3975,0.9097],[3977,0.9173],[3978,0.9136],[3980,0.9199],[3981,0.9161],[3984,0.9167],[3989,0.9212],[4005,0.5742],[4007,0.5867],[4010,0.5071],[4012,0.5254],[4017,0.9144],[4019,0.9097],[4021,0.9173],[4022,0.9136],[4024,0.9199],[4026,0.9161],[4028,0.9261],[4037,0.8476],[4038,0.8538],[4049,0.5742],[4051,0.5867],[4054,0.5071],[4056,0.5254],[4062,0.489],[4063,0.5018],[4068,0.4815],[4069,0.487],[4071,0.4793],[4074,0.4903],[4076,0.4794],[4092,0.8867],[4093,0.8838],[4097,0.9144],[4099,0.9097],[4101,0.9173],[4102,0.9136],[4104,0.9199],[4105,0.9161],[4108,0.9167],[4113,0.9212],[4129,0.5742],[4131,0.5867],[4134,0.5071],[4136,0.5254],[4142,0.9144],[4144,0.9097],[4146,0.9173],[4147,0.9136],[4149,0.9199],[4151,0.9161],[4153,0.9261],[4162,0.8476],[4163,0.8538],[4174,0.5742],[4176,0.5867],[4179,0.5071],[4181,0.5254],[4187,0.489],[4188,0.5018],[4193,0.4815],[4194,0.487],[4196,0.4793],[4199,0.4903],[4201,0.4794],[4217,0.8867],[4218,0.8838],[4222,0.9144],[4224,0.9097],[4226,0.9173],[4227,0.9136],[4229,0.9199],[4230,0.9161],[4233,0.9167],[4238,0.9212],[4254,0.5742],[4256,0.5867],[4259,0.5071],[4261,0.5254],[4268,0.9144],[4270,0.9097],[4272,0.9173],[4273,0.9136],[4275,0.9199],[4277,0.9161],[4279,0.9261],[4288,0.8476],[4289,0.8538],[4300,0.5742],[4302,0.5867],[4305,0.5071],[4307,0.5254],[4313,0.489],[4314,0.5018],[4319,0.4815],[4320,0.487],[4322,0.4793],[4325,0.4903],[4327,0.4794],[4343,0.8867],[4344,0.8838],[4348,0.9144],[4350,0.9097],[4352,0.9173],[4353,0.9136],[4355,0.9199],[4356,0.9161],[4360,0.5227],[4364,0.5018],[4372,0.4793],[4375,0.4903],[4376,0.4862],[4403,0.9182],[4404,0.917],[4408,0.9253],[4410,0.9214],[4430,0.5742],[4432,0.5867],[4436,0]]
       }
       thumb_01_r_ctrl.Rotation.Z {
        type=MovieSceneScriptingFloatChannel
        default=-6.0805
        thinned_from=211
        keys [[1790,-6.0805],[1825,-9.0796],[1826,-9.0698],[1828,-9.0968],[1832,-9.0784],[1865,-6.0464],[1905,-9.0796],[1906,-9.0698],[1908,-9.0968],[1938,-6.0804],[1939,-6.0863],[1941,-6.0585],[1942,-6.0643],[1945,-6.0307],[1985,-9.0796],[1986,-9.0698],[1988,-9.0968],[2018,-6.0804],[2019,-6.0863],[2021,-6.0585],[2022,-6.0643],[2025,-6.0307],[2065,-9.0796],[2066,-9.0698],[2068,-9.0968],[2098,-6.0804],[2099,-6.0863],[2101,-6.0585],[2109,-6.0565],[2145,-9.0796],[2146,-9.0698],[2148,-9.0968],[2152,-9.0784],[2185,-6.0464],[2225,-9.0796],[2226,-9.0698],[2228,-9.0968],[2258,-6.0804],[2259,-6.0863],[2261,-6.0585],[2262,-6.0643],[2265,-6.0307],[2305,-9.0796],[2306,-9.0698],[2308,-9.0968],[2338,-6.0804],[2339,-6.0863],[2341,-6.0585],[2342,-6.0643],[2345,-6.0307],[2385,-9.0796],[2386,-9.0698],[2388,-9.0968],[2418,-6.0804],[2419,-6.0863],[2421,-6.0585],[2429,-6.0565],[2465,-9.0796],[2496,2.6166],[2512,-0.9424],[2515,-0.9424],[2520,4.9424],[2522,1.8464],[2524,3.9062],[2530,1.003],[2763,10.9673],[2774,0.1535],[2778,-5.7673],[2793,-11.1919],[2927,-10.4026],[2938,-10.3798],[2986,35.4284],[3099,35.4284],[3185,5.0011],[3216,-6.0585],[3224,-6.0565],[3260,-9.0796],[3261,-9.0699],[3264,-9.0968],[3267,-9.0784],[3300,-6.0464],[3340,-9.0796],[3341,-9.0699],[3344,-9.0968],[3373,-6.0804],[3374,-6.0863],[3376,-6.0585],[3377,-6.0643],[3380,-6.0307],[3420,-9.0796],[3421,-9.0699],[3423,-9.0968],[3453,-6.0804],[3454,-6.0863],[3456,-6.0585],[3457,-6.0643],[3460,-6.0307],[3500,-9.0796],[3501,-9.0699],[3504,-9.0968],[3533,-6.0804],[3534,-6.0863],[3536,-6.0585],[3544,-6.0565],[3580,-9.0796],[3581,-9.0699],[3584,-9.0968],[3587,-9.0784],[3620,-6.0464],[3660,-9.0796],[3661,-9.0699],[3663,-9.0968],[3693,-6.0804],[3694,-6.0863],[3696,-6.0585],[3697,-6.0643],[3700,-6.0307],[3740,-9.0796],[3741,-9.0699],[3744,-9.0968],[3773,-6.0804],[3774,-6.0863],[3776,-6.0585],[3777,-6.0643],[3780,-6.0307],[3820,-9.0796],[3821,-9.0699],[3823,-9.0968],[3853,-6.0804],[3854,-6.0863],[3856,-6.0585],[3864,-6.0565],[3898,-6.0863],[3900,-6.0585],[3901,-6.0643],[3904,-6.0307],[3944,-9.0796],[3945,-9.0699],[3947,-9.0968],[3977,-6.0804],[3978,-6.0863],[3980,-6.0585],[3988,-6.0565],[4022,-6.0863],[4024,-6.0585],[4025,-6.0643],[4028,-6.0307],[4068,-9.0796],[4069,-9.0699],[4071,-9.0968],[4101,-6.0804],[4102,-6.0863],[4104,-6.0585],[4112,-6.0565],[4147,-6.0863],[4149,-6.0585],[4150,-6.0643],[4153,-6.0307],[4193,-9.0796],[4194,-9.0699],[4196,-9.0968],[4226,-6.0804],[4227,-6.0863],[4229,-6.0585],[4273,-6.0863],[4275,-6.0585],[4276,-6.0643],[4279,-6.0307],[4319,-9.0796],[4320,-9.0699],[4322,-9.0968],[4352,-6.0804],[4353,-6.0863],[4355,-6.0585],[4356,-6.0643],[4360,-8.4054],[4364,-8.7383],[4372,-9.0968],[4376,-9.0784],[4410,-6.0464],[4435,-8.7656],[4436,0]]
       }
       thumb_01_r_ctrl.Scale.X {type=MovieSceneScriptingFloatChannel, default=1, const [1,12]}
       thumb_01_r_ctrl.Scale.Y {type=MovieSceneScriptingFloatChannel, default=1, const [1,12]}
       thumb_01_r_ctrl.Scale.Z {type=MovieSceneScriptingFloatChannel, default=1, const [1,12]}
       thumb_02_r_ctrl.Location.X {type=MovieSceneScriptingFloatChannel, default=0, thinned_from=686, keys [[1790,0],[2512,0],[2513,-0.0003],[2515,0],[4436,-0.0001]]}
       thumb_02_r_ctrl.Location.Y {type=MovieSceneScriptingFloatChannel, default=0, thinned_from=864, keys [[1790,0],[2512,0],[2513,-0.0001],[2515,0],[4435,0],[4436,-0.0001]]}
       thumb_02_r_ctrl.Location.Z {type=MovieSceneScriptingFloatChannel, default=0, thinned_from=445, keys [[1790,0],[2512,0],[2513,0.0002],[2515,0],[4436,-0.0001]]}
       thumb_02_r_ctrl.Rotation.X {
        type=MovieSceneScriptingFloatChannel
        default=-1.713
        thinned_from=584
        keys [[1790,-1.713],[1810,-1.9532],[1811,-1.9532],[1818,-1.9714],[1831,-1.9714],[1832,-1.9889],[1853,-1.714],[1869,-1.713],[1873,-1.7526],[1874,-1.7526],[1887,-1.9296],[1888,-1.9296],[1890,-1.9721],[1911,-1.9714],[1913,-1.9879],[1933,-1.714],[1951,-1.713],[1953,-1.7526],[1954,-1.7526],[1963,-1.8882],[1964,-1.8882],[1970,-1.9721],[1991,-1.9714],[1993,-1.9879],[2009,-1.7405],[2010,-1.7405],[2014,-1.714],[2030,-1.713],[2050,-1.9532],[2051,-1.9532],[2058,-1.9714],[2071,-1.9714],[2073,-1.9879],[2094,-1.714],[2110,-1.713],[2130,-1.9532],[2131,-1.9532],[2138,-1.9714],[2151,-1.9714],[2152,-1.9889],[2169,-1.7405],[2173,-1.714],[2189,-1.713],[2203,-1.8882],[2207,-1.9296],[2210,-1.9721],[2231,-1.9714],[2234,-1.9879],[2254,-1.714],[2270,-1.713],[2273,-1.7526],[2274,-1.7526],[2287,-1.9296],[2288,-1.9296],[2290,-1.9721],[2311,-1.9714],[2314,-1.9879],[2334,-1.714],[2350,-1.713],[2370,-1.9532],[2371,-1.9532],[2377,-1.9714],[2391,-1.9714],[2394,-1.9879],[2414,-1.714],[2430,-1.713],[2450,-1.9532],[2451,-1.9532],[2454,-1.9708],[2463,-1.9714],[2502,0.7948],[2512,0.2817],[2515,0.2817],[2522,1.052],[2524,0.7833],[2529,1.1902],[2757,0.9767],[2759,0.426],[2780,-1.7273],[2802,-0.628],[2927,-0.3267],[2938,-0.3178],[2986,1.1429],[3099,1.1429],[3185,3.1267],[3216,-1.714],[3224,-1.713],[3249,-1.9708],[3265,-1.9714],[3267,-1.9889],[3288,-1.714],[3306,-1.713],[3325,-1.9721],[3346,-1.9714],[3348,-1.9879],[3364,-1.7405],[3365,-1.7405],[3369,-1.714],[3384,-1.713],[3401,-1.9296],[3402,-1.9296],[3405,-1.9721],[3426,-1.9714],[3428,-1.9879],[3449,-1.714],[3465,-1.713],[3468,-1.7526],[3485,-1.9532],[3486,-1.9532],[3489,-1.9708],[3506,-1.9714],[3509,-1.9879],[3529,-1.714],[3544,-1.713],[3565,-1.9532],[3566,-1.9532],[3573,-1.9714],[3586,-1.9714],[3587,-1.9889],[3610,-1.714],[3626,-1.713],[3645,-1.9721],[3666,-1.9714],[3668,-1.9879],[3689,-1.714],[3705,-1.713],[3708,-1.7526],[3709,-1.7526],[3721,-1.9296],[3722,-1.9296],[3725,-1.9721],[3746,-1.9714],[3748,-1.9879],[3768,-1.714],[3785,-1.713],[3805,-1.9532],[3806,-1.9532],[3809,-1.9708],[3825,-1.9714],[3828,-1.9879],[3849,-1.714],[3864,-1.713],[3885,-1.9532],[3886,-1.9532],[3892,-1.714],[3909,-1.713],[3929,-1.9532],[3930,-1.9532],[3933,-1.9708],[3949,-1.9714],[3952,-1.9879],[3973,-1.714],[3988,-1.713],[4009,-1.9532],[4010,-1.9532],[4016,-1.714],[4033,-1.713],[4053,-1.9532],[4054,-1.9532],[4057,-1.9708],[4073,-1.9714],[4076,-1.9879],[4097,-1.714],[4112,-1.713],[4133,-1.9532],[4134,-1.9532],[4141,-1.714],[4158,-1.713],[4178,-1.9532],[4179,-1.9532],[4182,-1.9708],[4198,-1.9714],[4201,-1.9879],[4222,-1.714],[4237,-1.713],[4258,-1.9532],[4259,-1.9532],[4267,-1.714],[4284,-1.713],[4304,-1.9532],[4305,-1.9532],[4308,-1.9708],[4324,-1.9714],[4327,-1.9879],[4348,-1.714],[4356,-1.714],[4360,-1.9708],[4375,-1.9714],[4377,-1.9889],[4397,-1.714],[4414,-1.713],[4430,-1.9296],[4432,-1.9296],[4434,-1.9721],[4435,-1.9708],[4436,-0.0003]]
       }
       thumb_02_r_ctrl.Rotation.Y {
        type=MovieSceneScriptingFloatChannel
        default=4.0271
        thinned_from=490
        keys [[1790,4.0271],[1810,4.5996],[1814,4.6458],[1831,4.6506],[1832,4.6762],[1849,4.0978],[1852,4.0267],[1865,4.0271],[1873,4.1155],[1875,4.115],[1886,4.5381],[1891,4.6458],[1911,4.6506],[1913,4.6736],[1929,4.0978],[1932,4.0267],[1945,4.0271],[1953,4.1155],[1955,4.115],[1966,4.5381],[1971,4.6458],[1991,4.6506],[1994,4.6736],[2009,4.0978],[2012,4.0267],[2025,4.0271],[2033,4.1155],[2035,4.115],[2054,4.6458],[2071,4.6506],[2073,4.6736],[2089,4.0978],[2092,4.0267],[2105,4.0271],[2134,4.6458],[2148,4.6506],[2152,4.6762],[2169,4.0978],[2172,4.0267],[2185,4.0271],[2193,4.1155],[2195,4.115],[2206,4.5381],[2211,4.6458],[2231,4.6506],[2233,4.6736],[2249,4.0978],[2252,4.0267],[2265,4.0271],[2273,4.1155],[2275,4.115],[2291,4.6458],[2311,4.6506],[2313,4.6736],[2329,4.0978],[2332,4.0267],[2345,4.0271],[2353,4.1155],[2355,4.115],[2374,4.6458],[2391,4.6506],[2393,4.6736],[2409,4.0978],[2412,4.0267],[2425,4.0271],[2454,4.6458],[2461,4.6506],[2502,1.0344],[2512,1.7864],[2519,2.1429],[2522,1.6604],[2524,1.9435],[2530,1.5439],[2779,4.2294],[2798,1.9792],[2927,0.9663],[2938,0.9515],[2986,-10.8019],[3099,-10.8019],[3185,-2.2619],[3216,4.027],[3220,4.0271],[3249,4.6458],[3265,4.6506],[3267,4.6762],[3284,4.0978],[3287,4.0267],[3300,4.0271],[3309,4.1155],[3310,4.115],[3322,4.5381],[3326,4.6458],[3345,4.6506],[3348,4.6736],[3364,4.0978],[3367,4.0267],[3380,4.0271],[3389,4.1155],[3390,4.115],[3401,4.5381],[3406,4.6458],[3425,4.6506],[3428,4.6736],[3447,4.0267],[3460,4.0271],[3468,4.1155],[3470,4.115],[3489,4.6458],[3505,4.6506],[3508,4.6736],[3524,4.0978],[3527,4.0267],[3540,4.0271],[3569,4.6458],[3585,4.6506],[3587,4.6762],[3604,4.0978],[3607,4.0267],[3620,4.0271],[3629,4.1155],[3630,4.115],[3641,4.5381],[3646,4.6458],[3665,4.6506],[3668,4.6736],[3687,4.0267],[3700,4.0271],[3708,4.1155],[3710,4.115],[3721,4.5381],[3726,4.6458],[3745,4.6506],[3748,4.6736],[3767,4.0267],[3780,4.0271],[3789,4.1155],[3790,4.115],[3809,4.6458],[3825,4.6506],[3828,4.6736],[3844,4.0978],[3847,4.0267],[3904,4.0271],[3913,4.1155],[3914,4.115],[3933,4.6458],[3949,4.6506],[3952,4.6736],[3968,4.0978],[3971,4.0267],[4028,4.0271],[4037,4.1155],[4038,4.115],[4057,4.6458],[4073,4.6506],[4076,4.6736],[4092,4.0978],[4095,4.0267],[4153,4.0271],[4162,4.1155],[4163,4.115],[4182,4.6458],[4198,4.6506],[4201,4.6736],[4217,4.0978],[4220,4.0267],[4279,4.0271],[4288,4.1155],[4289,4.115],[4308,4.6458],[4324,4.6506],[4327,4.6736],[4343,4.0978],[4346,4.0267],[4355,4.027],[4360,4.6458],[4372,4.6506],[4377,4.6762],[4396,4.0267],[4409,4.0271],[4417,4.1155],[4419,4.115],[4430,4.5381],[4435,4.6458],[4436,0.0029]]
       }
       thumb_02_r_ctrl.Rotation.Z {
        type=MovieSceneScriptingFloatChannel
        default=31.5262
        thinned_from=278
        keys [[1790,31.5262],[1810,32.2005],[1832,32.29],[1836,32.196],[1849,31.6038],[1853,31.5261],[1865,31.5262],[1875,31.6275],[1886,32.1282],[1889,32.1258],[1891,32.2546],[1913,32.287],[1929,31.6038],[1933,31.5261],[1945,31.5262],[1955,31.6275],[1966,32.1282],[1969,32.1258],[1971,32.2546],[1993,32.287],[2009,31.6038],[2013,31.5261],[2025,31.5262],[2035,31.6275],[2051,32.2005],[2057,32.2569],[2073,32.287],[2089,31.6038],[2093,31.5261],[2105,31.5262],[2134,32.2546],[2152,32.29],[2156,32.196],[2169,31.6038],[2173,31.5261],[2185,31.5262],[2195,31.6275],[2206,32.1282],[2209,32.1258],[2211,32.2546],[2233,32.287],[2249,31.6038],[2253,31.5261],[2265,31.5262],[2275,31.6275],[2286,32.1282],[2289,32.1258],[2291,32.2546],[2313,32.287],[2329,31.6038],[2333,31.5261],[2345,31.5262],[2355,31.6275],[2371,32.2005],[2377,32.2569],[2394,32.287],[2409,31.6038],[2413,31.5261],[2425,31.5262],[2454,32.2546],[2467,32.2449],[2502,33.2623],[2507,18.9204],[2513,10.0259],[2518,14.7227],[2519,25.7909],[2521,37.0764],[2523,34.3913],[2529,37.5373],[2530,37.5372],[2764,27.2193],[2777,30.2925],[2801,11.0443],[2938,11.1036],[2986,53.1033],[3099,53.1033],[3185,31.4204],[3220,31.5262],[3249,32.2546],[3267,32.29],[3271,32.196],[3287,31.5254],[3300,31.5262],[3310,31.6275],[3318,32.0018],[3322,32.1282],[3324,32.1258],[3326,32.2546],[3348,32.287],[3364,31.6038],[3368,31.5261],[3380,31.5262],[3390,31.6275],[3401,32.1282],[3404,32.1258],[3407,32.2546],[3428,32.287],[3447,31.5254],[3460,31.5262],[3470,31.6275],[3486,32.2005],[3492,32.2569],[3509,32.287],[3527,31.5254],[3540,31.5262],[3567,32.2],[3572,32.2569],[3587,32.29],[3591,32.196],[3607,31.5254],[3620,31.5262],[3630,31.6275],[3642,32.1282],[3644,32.1258],[3647,32.2546],[3668,32.287],[3687,31.5254],[3700,31.5262],[3710,31.6275],[3721,32.1282],[3724,32.1258],[3726,32.2546],[3748,32.287],[3767,31.5254],[3780,31.5262],[3790,31.6275],[3806,32.2004],[3812,32.2569],[3828,32.287],[3844,31.6038],[3848,31.5261],[3860,31.5262],[3887,32.2],[3891,31.5254],[3904,31.5262],[3914,31.6275],[3930,32.2004],[3936,32.2569],[3952,32.287],[3968,31.6038],[3972,31.5261],[3984,31.5262],[4011,32.2],[4015,31.5254],[4028,31.5262],[4038,31.6275],[4054,32.2004],[4060,32.2569],[4076,32.287],[4092,31.6038],[4096,31.5261],[4108,31.5262],[4135,32.2],[4140,31.5254],[4153,31.5262],[4163,31.6275],[4179,32.2004],[4185,32.2569],[4201,32.287],[4217,31.6038],[4221,31.5261],[4233,31.5262],[4260,32.2],[4266,31.5254],[4279,31.5262],[4289,31.6275],[4305,32.2004],[4311,32.2569],[4327,32.287],[4343,31.6038],[4347,31.5261],[4360,32.2546],[4377,32.29],[4396,31.5254],[4409,31.5262],[4419,31.6275],[4430,32.1282],[4433,32.1258],[4435,32.2546],[4436,0.0084]]
       }
       thumb_02_r_ctrl.Scale.X {type=MovieSceneScriptingFloatChannel, default=1, const [1,12]}
       thumb_02_r_ctrl.Scale.Y {type=MovieSceneScriptingFloatChannel, default=1, const [1,12]}
       thumb_02_r_ctrl.Scale.Z {type=MovieSceneScriptingFloatChannel, default=1, const [1,12]}
       thumb_03_r_ctrl.Location.X {type=MovieSceneScriptingFloatChannel, default=0, thinned_from=234, keys [[1790,0],[2512,0],[2513,-0.0001],[2524,0],[4436,0]]}
       thumb_03_r_ctrl.Location.Y {type=MovieSceneScriptingFloatChannel, default=0, thinned_from=960, keys [[1790,0],[4436,0]]}
       thumb_03_r_ctrl.Location.Z {type=MovieSceneScriptingFloatChannel, default=0, thinned_from=497, keys [[1790,0],[4436,0]]}
       thumb_03_r_ctrl.Rotation.X {
        type=MovieSceneScriptingFloatChannel
        default=0.8579
        thinned_from=95
        keys [[1790,0.8579],[1826,0.0108],[1827,0.0108],[1829,0.0019],[1831,0.0019],[1866,0.8706],[1906,0.0108],[1907,0.0108],[1909,0.0019],[1911,0.0019],[1945,0.872],[1981,0.0216],[1988,0.0032],[1991,0.0032],[2025,0.872],[2066,0.0108],[2067,0.0108],[2069,0.0019],[2071,0.0019],[2105,0.872],[2150,0.0019],[2186,0.8706],[2226,0.0108],[2227,0.0108],[2229,0.0019],[2231,0.0019],[2266,0.872],[2308,0.0032],[2311,0.0032],[2345,0.872],[2386,0.0108],[2387,0.0108],[2389,0.0019],[2391,0.0019],[2425,0.872],[2466,0.0108],[2470,0.0421],[2496,1.2682],[2513,1.7875],[2519,1.1998],[2522,1.8461],[2524,1.4527],[2530,2.0008],[2762,0.1771],[2771,0.0511],[2798,4.5986],[2938,4.5626],[2986,1.287],[3099,1.287],[3185,0.902],[3220,0.872],[3265,0.0019],[3303,0.8706],[3345,0.0019],[3381,0.872],[3423,0.0032],[3425,0.0032],[3460,0.872],[3505,0.0019],[3541,0.872],[3585,0.0019],[3622,0.8706],[3665,0.0019],[3700,0.872],[3743,0.0032],[3746,0.0032],[3780,0.872],[3825,0.0019],[3860,0.872],[3904,0.872],[3949,0.0019],[3984,0.872],[4028,0.872],[4073,0.0019],[4108,0.872],[4153,0.872],[4198,0.0019],[4233,0.872],[4279,0.872],[4324,0.0019],[4360,0.2536],[4364,0.3001],[4411,0.8706],[4436,0]]
       }
       thumb_03_r_ctrl.Rotation.Y {
        type=MovieSceneScriptingFloatChannel
        default=-1.2625
        thinned_from=207
        keys [[1790,-1.2625],[1826,-0.2497],[1829,-0.2501],[1832,-0.2499],[1859,-1.2689],[1860,-1.2685],[1862,-1.2752],[1866,-1.2743],[1906,-0.2497],[1911,-0.2501],[1939,-1.2689],[1940,-1.2685],[1942,-1.2752],[1943,-1.2748],[1945,-1.2805],[1987,-0.2484],[1991,-0.2484],[2019,-1.2689],[2020,-1.2685],[2022,-1.2752],[2023,-1.2748],[2025,-1.2805],[2066,-0.2497],[2071,-0.2501],[2100,-1.2684],[2103,-1.2812],[2105,-1.2805],[2147,-0.2497],[2149,-0.2501],[2152,-0.2499],[2179,-1.2689],[2180,-1.2685],[2182,-1.2752],[2186,-1.2743],[2226,-0.2497],[2231,-0.2501],[2259,-1.2689],[2260,-1.2685],[2262,-1.2752],[2263,-1.2748],[2265,-1.2805],[2307,-0.2484],[2311,-0.2484],[2339,-1.2689],[2340,-1.2685],[2342,-1.2752],[2343,-1.2748],[2345,-1.2805],[2386,-0.2497],[2391,-0.2501],[2420,-1.2684],[2423,-1.2812],[2425,-1.2805],[2466,-0.2497],[2469,-0.2887],[2470,-0.2886],[2502,-0.0968],[2512,-0.612],[2515,-0.611],[2519,-0.727],[2522,-0.2497],[2524,-0.4842],[2530,-0.1454],[2531,-0.1454],[2764,-1.3108],[2776,-0.7266],[2800,-6.4053],[2938,-6.3697],[2986,-0.817],[3099,-0.817],[3185,-1.8571],[3218,-1.2812],[3220,-1.2805],[3261,-0.2497],[3266,-0.2501],[3267,-0.2499],[3294,-1.2689],[3295,-1.2685],[3297,-1.2752],[3301,-1.2743],[3341,-0.2497],[3345,-0.2501],[3374,-1.2689],[3375,-1.2685],[3377,-1.2752],[3378,-1.2748],[3380,-1.2805],[3416,-0.2634],[3417,-0.2634],[3422,-0.2484],[3425,-0.2484],[3454,-1.2689],[3455,-1.2685],[3457,-1.2752],[3458,-1.2748],[3460,-1.2805],[3501,-0.2497],[3505,-0.2501],[3535,-1.2684],[3536,-1.2684],[3538,-1.2812],[3540,-1.2805],[3581,-0.2497],[3586,-0.2501],[3587,-0.2499],[3614,-1.2689],[3615,-1.2685],[3617,-1.2752],[3621,-1.2743],[3662,-0.2497],[3665,-0.2501],[3694,-1.2689],[3695,-1.2685],[3697,-1.2752],[3698,-1.2748],[3700,-1.2805],[3742,-0.2484],[3745,-0.2484],[3774,-1.2689],[3775,-1.2685],[3777,-1.2752],[3778,-1.2748],[3780,-1.2805],[3822,-0.2497],[3825,-0.2501],[3858,-1.2812],[3899,-1.2685],[3901,-1.2752],[3902,-1.2748],[3904,-1.2805],[3946,-0.2497],[3949,-0.2501],[3982,-1.2812],[4023,-1.2685],[4025,-1.2752],[4026,-1.2748],[4028,-1.2805],[4070,-0.2497],[4073,-0.2501],[4106,-1.2812],[4148,-1.2685],[4150,-1.2752],[4151,-1.2748],[4153,-1.2805],[4195,-0.2497],[4198,-0.2501],[4231,-1.2812],[4274,-1.2685],[4276,-1.2752],[4277,-1.2748],[4279,-1.2805],[4321,-0.2497],[4324,-0.2501],[4357,-1.2812],[4360,-0.4833],[4364,-0.3615],[4371,-0.2497],[4375,-0.2501],[4376,-0.2499],[4403,-1.2689],[4404,-1.2685],[4406,-1.2752],[4410,-1.2743],[4436,0]]
       }
       thumb_03_r_ctrl.Rotation.Z {
        type=MovieSceneScriptingFloatChannel
        default=-25.1363
        thinned_from=192
        keys [[1790,-25.1363],[1807,-9.1717],[1825,-5.0853],[1831,-5.0852],[1848,-22.818],[1866,-25.2703],[1885,-10.9339],[1905,-5.0853],[1911,-5.0852],[1928,-22.818],[1945,-25.2703],[1978,-5.5795],[1979,-5.5795],[1981,-5.265],[1984,-5.2651],[1987,-4.9505],[1991,-4.9505],[2008,-22.818],[2025,-25.2703],[2045,-11.0141],[2065,-5.0853],[2071,-5.0852],[2088,-22.818],[2105,-25.2703],[2125,-11.0141],[2145,-5.0853],[2151,-5.0852],[2168,-22.818],[2186,-25.2703],[2205,-10.934],[2225,-5.0853],[2231,-5.0852],[2248,-22.818],[2265,-25.2703],[2298,-5.5795],[2299,-5.5795],[2302,-5.265],[2304,-5.2651],[2307,-4.9505],[2311,-4.9505],[2328,-22.8179],[2345,-25.2703],[2365,-11.0142],[2385,-5.0853],[2391,-5.0852],[2408,-22.818],[2425,-25.2703],[2445,-11.0141],[2465,-5.0853],[2469,-5.1645],[2484,-33.8156],[2502,-40.2935],[2512,-28.7302],[2515,-28.73],[2519,-12.7955],[2522,-37.4178],[2523,-25.9497],[2524,-23.2403],[2527,-41.2699],[2531,-43.1446],[2758,-21.5947],[2761,15.7331],[2764,25.4149],[2779,-4.2256],[2794,-26.1114],[2938,-26.0633],[2986,-23.8003],[3099,-23.8003],[3185,5.2755],[3220,-25.2703],[3240,-11.0155],[3260,-5.0853],[3266,-5.0852],[3283,-22.817],[3301,-25.2703],[3320,-10.9353],[3340,-5.0853],[3346,-5.0852],[3363,-22.8171],[3380,-25.2703],[3413,-5.5795],[3414,-5.5795],[3417,-5.265],[3419,-5.2651],[3422,-4.9505],[3425,-4.9505],[3442,-21.8961],[3460,-25.2703],[3480,-11.0155],[3500,-5.0853],[3504,-5.0852],[3522,-21.97],[3540,-25.2703],[3560,-11.0154],[3580,-5.0853],[3586,-5.0852],[3603,-22.8171],[3621,-25.2703],[3640,-10.9352],[3660,-5.0853],[3666,-5.0852],[3683,-22.817],[3700,-25.2703],[3733,-5.5795],[3734,-5.5795],[3737,-5.265],[3739,-5.2651],[3742,-4.9505],[3745,-4.9505],[3762,-21.8958],[3780,-25.2703],[3800,-11.0153],[3820,-5.0853],[3824,-5.0852],[3842,-21.9702],[3860,-25.2703],[3880,-11.0156],[3904,-25.2703],[3924,-11.0153],[3944,-5.0853],[3948,-5.0852],[3966,-21.9702],[3984,-25.2703],[4004,-11.0156],[4028,-25.2703],[4048,-11.0153],[4068,-5.0853],[4072,-5.0852],[4090,-21.9702],[4108,-25.2703],[4128,-11.0156],[4153,-25.2703],[4173,-11.0153],[4193,-5.0853],[4197,-5.0852],[4215,-21.9702],[4233,-25.2703],[4253,-11.0156],[4279,-25.2703],[4299,-11.0153],[4319,-5.0853],[4323,-5.0852],[4341,-21.9702],[4360,-7.2829],[4364,-6.3006],[4373,-5.0852],[4375,-5.0852],[4392,-22.8169],[4410,-25.2703],[4423,-17.6815],[4432,-8.4936],[4435,-7.0309],[4436,0]]
       }
       thumb_03_r_ctrl.Scale.X {type=MovieSceneScriptingFloatChannel, default=1, const [1,12]}
       thumb_03_r_ctrl.Scale.Y {type=MovieSceneScriptingFloatChannel, default=1, const [1,12]}
       thumb_03_r_ctrl.Scale.Z {type=MovieSceneScriptingFloatChannel, default=1, const [1,12]}
       index_metacarpal_r_ctrl.Location.X {type=MovieSceneScriptingFloatChannel, default=0, thinned_from=1540, keys [[1790,0],[4436,0]]}
       index_metacarpal_r_ctrl.Location.Y {type=MovieSceneScriptingFloatChannel, default=0, thinned_from=1592, keys [[1790,0],[4436,0]]}
       index_metacarpal_r_ctrl.Location.Z {type=MovieSceneScriptingFloatChannel, default=0, thinned_from=616, keys [[1790,0],[4436,0]]}
       index_metacarpal_r_ctrl.Rotation.X {
        type=MovieSceneScriptingFloatChannel
        default=-6.3877
        thinned_from=88
        keys [[1790,-6.3877],[1821,-6.9726],[1826,-6.9854],[1866,-6.3796],[1901,-6.9726],[1906,-6.9854],[1946,-6.3796],[1981,-6.9726],[1986,-6.9854],[2026,-6.3796],[2061,-6.9726],[2066,-6.9854],[2105,-6.3788],[2141,-6.9726],[2146,-6.9854],[2186,-6.3796],[2221,-6.9726],[2226,-6.9854],[2266,-6.3796],[2301,-6.9726],[2306,-6.9854],[2346,-6.3796],[2381,-6.9726],[2386,-6.9854],[2425,-6.3788],[2461,-6.9726],[2466,-6.9853],[2496,-0.3982],[2504,-0.2279],[2513,-0.0464],[2520,-1.3045],[2522,-0.5892],[2524,-1.0623],[2526,-0.5796],[2531,-0.3943],[2760,-2.0434],[2767,-3.2535],[2774,-4.5588],[2777,-4.02],[2781,-1.4428],[2788,0],[3185,0],[3220,-6.3788],[3262,-6.9854],[3301,-6.3796],[3341,-6.9854],[3381,-6.3796],[3416,-6.9726],[3422,-6.9854],[3461,-6.3796],[3502,-6.9854],[3540,-6.3788],[3581,-6.9854],[3621,-6.3796],[3662,-6.9854],[3701,-6.3796],[3741,-6.9854],[3781,-6.3796],[3816,-6.9726],[3822,-6.9854],[3860,-6.3788],[3905,-6.3796],[3940,-6.9726],[3946,-6.9854],[3984,-6.3788],[4029,-6.3796],[4064,-6.9726],[4070,-6.9854],[4108,-6.3788],[4154,-6.3796],[4189,-6.9726],[4195,-6.9854],[4233,-6.3788],[4280,-6.3796],[4315,-6.9726],[4321,-6.9854],[4360,-6.8599],[4364,-6.8348],[4410,-6.3796],[4435,-6.9209],[4436,0]]
       }
       index_metacarpal_r_ctrl.Rotation.Y {
        type=MovieSceneScriptingFloatChannel
        default=4.3708
        thinned_from=180
        keys [[1790,4.3708],[1804,2.6292],[1818,1.8879],[1819,1.8881],[1821,1.8578],[1826,1.8275],[1835,2.0032],[1845,3.5012],[1865,4.3897],[1866,4.3896],[1898,1.8879],[1899,1.8881],[1901,1.8578],[1906,1.8275],[1915,2.0124],[1925,3.4875],[1945,4.3897],[1946,4.3896],[1978,1.8879],[1979,1.8881],[1981,1.8578],[1986,1.8275],[1995,2.011],[2005,3.4908],[2025,4.3897],[2026,4.3896],[2058,1.8879],[2059,1.8881],[2061,1.8578],[2066,1.8275],[2075,2.0124],[2085,3.4875],[2105,4.3907],[2138,1.8879],[2139,1.8881],[2146,1.8275],[2155,2.0032],[2165,3.5012],[2185,4.3897],[2186,4.3896],[2218,1.8879],[2219,1.8881],[2226,1.8275],[2235,2.0124],[2245,3.4875],[2265,4.3897],[2266,4.3896],[2298,1.8879],[2299,1.8881],[2306,1.8275],[2315,2.011],[2325,3.4908],[2345,4.3897],[2346,4.3896],[2378,1.8879],[2379,1.8881],[2386,1.8275],[2395,2.0124],[2405,3.4875],[2425,4.3907],[2458,1.8879],[2459,1.8881],[2496,-1.804],[2504,-1.1061],[2508,1.2875],[2512,2.213],[2515,2.213],[2520,-2.8848],[2522,-2.6089],[2524,-2.7983],[2531,-2.5362],[2763,-3.4368],[2780,0.2667],[2788,0],[3185,0],[3220,4.3907],[3253,1.8879],[3254,1.8881],[3262,1.8275],[3271,2.1147],[3281,3.6427],[3300,4.3897],[3301,4.3896],[3333,1.8879],[3334,1.8881],[3341,1.8275],[3350,2.0123],[3360,3.4873],[3380,4.3897],[3381,4.3896],[3413,1.8879],[3414,1.8881],[3422,1.8275],[3431,2.114],[3441,3.6384],[3460,4.3897],[3461,4.3896],[3493,1.8879],[3494,1.8881],[3502,1.8275],[3511,2.1235],[3521,3.6296],[3540,4.3907],[3573,1.8879],[3574,1.8881],[3581,1.8275],[3590,2.0031],[3600,3.5009],[3620,4.3897],[3621,4.3896],[3653,1.8879],[3654,1.8881],[3662,1.8275],[3671,2.1235],[3681,3.6295],[3700,4.3897],[3701,4.3896],[3733,1.8879],[3734,1.8881],[3742,1.8275],[3751,2.114],[3761,3.6384],[3780,4.3897],[3781,4.3896],[3813,1.8879],[3814,1.8881],[3822,1.8275],[3831,2.1235],[3841,3.6296],[3860,4.3907],[3905,4.3896],[3937,1.8879],[3938,1.8881],[3946,1.8275],[3955,2.1235],[3965,3.6296],[3984,4.3907],[4029,4.3896],[4061,1.8879],[4062,1.8881],[4070,1.8275],[4079,2.1235],[4089,3.6296],[4108,4.3907],[4154,4.3896],[4186,1.8879],[4187,1.8881],[4195,1.8275],[4204,2.1235],[4214,3.6296],[4233,4.3907],[4280,4.3896],[4312,1.8879],[4313,1.8881],[4321,1.8275],[4330,2.1235],[4340,3.6296],[4360,1.9893],[4364,1.9902],[4379,2.003],[4389,3.501],[4409,4.3897],[4410,4.3896],[4423,3.3661],[4429,2.4927],[4432,2.1882],[4434,2.0513],[4435,2.021],[4436,0]]
       }
       index_metacarpal_r_ctrl.Rotation.Z {
        type=MovieSceneScriptingFloatChannel
        default=2.9036
        thinned_from=240
        keys [[1790,2.9036],[1805,1.2751],[1821,0.614],[1824,0.581],[1826,0.5831],[1858,2.9179],[1859,2.917],[1861,2.9352],[1869,2.9367],[1885,1.2751],[1901,0.614],[1904,0.581],[1906,0.5831],[1938,2.9179],[1939,2.917],[1941,2.9352],[1949,2.9367],[1965,1.2852],[1981,0.614],[1984,0.581],[1986,0.5831],[2018,2.9179],[2019,2.917],[2021,2.9352],[2029,2.9367],[2061,0.614],[2062,0.614],[2064,0.581],[2066,0.5831],[2098,2.9179],[2099,2.917],[2101,2.9352],[2103,2.9333],[2105,2.9526],[2141,0.614],[2144,0.581],[2146,0.5831],[2178,2.9179],[2179,2.917],[2181,2.9352],[2189,2.9367],[2205,1.2752],[2221,0.614],[2224,0.581],[2226,0.5831],[2258,2.9179],[2259,2.917],[2261,2.9352],[2269,2.9367],[2286,1.1716],[2304,0.581],[2306,0.5831],[2338,2.9179],[2339,2.917],[2341,2.9352],[2349,2.9367],[2365,1.2852],[2381,0.614],[2384,0.581],[2386,0.5831],[2418,2.9179],[2419,2.917],[2421,2.9352],[2423,2.9333],[2426,2.9526],[2461,0.614],[2464,0.581],[2471,0.6948],[2502,5.1216],[2507,0.9872],[2513,-1.5793],[2518,-0.4165],[2521,8.0875],[2522,7.9642],[2523,6.5379],[2527,8.6451],[2531,8.8792],[2758,6.161],[2761,1.4888],[2764,0.2245],[2774,0.379],[2788,0],[3185,0],[3216,2.9352],[3218,2.9333],[3220,2.9526],[3259,0.581],[3261,0.5831],[3293,2.9179],[3294,2.917],[3296,2.9352],[3304,2.9367],[3321,1.1735],[3339,0.581],[3341,0.5831],[3373,2.9179],[3374,2.917],[3376,2.9352],[3384,2.9367],[3400,1.2853],[3416,0.614],[3419,0.581],[3422,0.5831],[3453,2.9179],[3454,2.917],[3456,2.9352],[3464,2.9367],[3481,1.1717],[3499,0.581],[3501,0.5831],[3533,2.9179],[3534,2.917],[3536,2.9352],[3538,2.9333],[3541,2.9526],[3579,0.581],[3581,0.5831],[3613,2.9179],[3614,2.917],[3616,2.9352],[3624,2.9367],[3641,1.1735],[3659,0.581],[3662,0.5831],[3693,2.9179],[3694,2.917],[3696,2.9352],[3704,2.9367],[3721,1.1717],[3739,0.581],[3741,0.5831],[3773,2.9179],[3774,2.917],[3776,2.9352],[3784,2.9367],[3801,1.1717],[3819,0.581],[3822,0.5831],[3853,2.9179],[3854,2.917],[3856,2.9352],[3858,2.9333],[3860,2.9526],[3898,2.917],[3900,2.9352],[3908,2.9367],[3925,1.1717],[3943,0.581],[3946,0.5831],[3977,2.9179],[3978,2.917],[3980,2.9352],[3982,2.9333],[3984,2.9526],[4022,2.917],[4024,2.9352],[4032,2.9367],[4049,1.1717],[4067,0.581],[4070,0.5831],[4101,2.9179],[4102,2.917],[4104,2.9352],[4106,2.9333],[4108,2.9526],[4147,2.917],[4149,2.9352],[4157,2.9367],[4174,1.1717],[4192,0.581],[4195,0.5831],[4226,2.9179],[4227,2.917],[4229,2.9352],[4231,2.9333],[4233,2.9526],[4273,2.917],[4275,2.9352],[4283,2.9367],[4300,1.1717],[4318,0.581],[4321,0.5831],[4352,2.9179],[4353,2.917],[4355,2.9352],[4357,2.9333],[4360,0.9688],[4364,1.0185],[4402,2.9179],[4403,2.917],[4405,2.9352],[4413,2.9367],[4430,1.1735],[4434,0.8664],[4436,0]]
       }
       index_metacarpal_r_ctrl.Scale.X {type=MovieSceneScriptingFloatChannel, default=1, const [1,12]}
       index_metacarpal_r_ctrl.Scale.Y {type=MovieSceneScriptingFloatChannel, default=1, const [1,12]}
       index_metacarpal_r_ctrl.Scale.Z {type=MovieSceneScriptingFloatChannel, default=1, const [1,12]}
       index_01_r_ctrl.Location.X {
        type=MovieSceneScriptingFloatChannel
        default=0
        thinned_from=30
        keys [[1790,0],[2511,0],[2513,-0.0005],[2519,0],[2768,0],[2927,-0.0046],[2938,-0.0047],[2986,-0.0679],[3099,-0.0679],[3185,-0.2702],[3687,0],[4436,-0.0001]]
       }
       index_01_r_ctrl.Location.Y {
        type=MovieSceneScriptingFloatChannel
        default=0
        thinned_from=401
        keys [[1790,0],[2513,0.0001],[2769,0],[2780,0.0225],[2927,0.0753],[2938,0.0756],[2986,0.2304],[3099,0.2304],[3185,0.8711],[3219,0],[4436,-0.0001]]
       }
       index_01_r_ctrl.Location.Z {
        type=MovieSceneScriptingFloatChannel
        default=0
        thinned_from=738
        keys [[1790,0],[2769,0],[2775,0.0338],[2782,0.2696],[2796,0.4129],[2938,0.4053],[2986,-0.0007],[3099,-0.0007],[3185,-0.2312],[3216,0],[4436,0]]
       }
       index_01_r_ctrl.Rotation.X {
        type=MovieSceneScriptingFloatChannel
        default=3.5021
        thinned_from=159
        keys [[1790,3.5021],[1821,3.2481],[1824,3.2481],[1826,3.2424],[1855,3.4968],[1866,3.5064],[1901,3.2481],[1904,3.2481],[1906,3.2424],[1945,3.5064],[1949,3.5064],[1981,3.2481],[1987,3.2424],[2025,3.5064],[2029,3.5064],[2061,3.2481],[2067,3.2424],[2105,3.5064],[2141,3.2481],[2147,3.2424],[2175,3.4968],[2186,3.5064],[2221,3.2481],[2227,3.2424],[2265,3.5064],[2269,3.5064],[2301,3.2481],[2307,3.2424],[2345,3.5064],[2349,3.5064],[2381,3.2481],[2387,3.2424],[2425,3.5064],[2461,3.2481],[2464,3.2481],[2469,3.232],[2484,-0.5051],[2502,-1.0491],[2508,3.0317],[2511,4.1899],[2518,4.3123],[2521,0.2321],[2522,0.2388],[2523,0.9812],[2527,-0.0887],[2531,-0.2056],[2757,0.4043],[2761,3.5959],[2764,4.2412],[2774,3.3376],[2779,0.1812],[2784,-1.8605],[2791,-0.5767],[2804,1.7695],[2938,1.7646],[2986,-29.4504],[3099,-29.4504],[3185,-15.8319],[3220,3.5064],[3257,3.2481],[3262,3.2424],[3290,3.4968],[3301,3.5064],[3337,3.2481],[3342,3.2424],[3380,3.5064],[3381,3.5064],[3416,3.2481],[3422,3.2424],[3460,3.5064],[3461,3.5064],[3497,3.2481],[3502,3.2424],[3540,3.5064],[3577,3.2481],[3582,3.2424],[3610,3.4968],[3621,3.5064],[3657,3.2481],[3662,3.2424],[3700,3.5064],[3701,3.5064],[3736,3.2481],[3742,3.2424],[3780,3.5064],[3781,3.5064],[3816,3.2481],[3822,3.2424],[3860,3.5064],[3905,3.5064],[3940,3.2481],[3946,3.2424],[3984,3.5064],[4029,3.5064],[4064,3.2481],[4070,3.2424],[4108,3.5064],[4154,3.5064],[4189,3.2481],[4195,3.2424],[4233,3.5064],[4280,3.5064],[4315,3.2481],[4321,3.2424],[4360,3.2791],[4371,3.2424],[4399,3.4968],[4410,3.5064],[4423,3.3998],[4429,3.3161],[4432,3.2854],[4435,3.2708],[4436,0.001]]
       }
       index_01_r_ctrl.Rotation.Y {
        type=MovieSceneScriptingFloatChannel
        default=2.4789
        thinned_from=303
        keys [[1790,2.4789],[1818,2.931],[1824,2.9449],[1826,2.9431],[1852,2.4849],[1856,2.485],[1858,2.4677],[1865,2.4679],[1898,2.931],[1904,2.9449],[1906,2.9431],[1934,2.4825],[1935,2.4835],[1937,2.4751],[1940,2.4771],[1942,2.4678],[1945,2.4679],[1978,2.931],[1984,2.9449],[1986,2.9431],[2014,2.4825],[2015,2.4835],[2017,2.4751],[2020,2.4771],[2022,2.4678],[2025,2.4679],[2058,2.931],[2064,2.9449],[2066,2.9431],[2094,2.4825],[2095,2.4835],[2097,2.4751],[2100,2.4771],[2102,2.4678],[2105,2.4679],[2138,2.931],[2142,2.9371],[2144,2.9449],[2146,2.9431],[2172,2.4849],[2176,2.485],[2179,2.4677],[2185,2.4679],[2218,2.931],[2222,2.9371],[2224,2.9449],[2226,2.9431],[2254,2.4825],[2255,2.4835],[2257,2.4751],[2260,2.4771],[2262,2.4678],[2265,2.4679],[2298,2.931],[2302,2.9371],[2304,2.9449],[2306,2.9431],[2334,2.4825],[2335,2.4835],[2337,2.4751],[2340,2.4771],[2342,2.4678],[2345,2.4679],[2378,2.931],[2382,2.9371],[2384,2.9449],[2386,2.9431],[2414,2.4825],[2415,2.4835],[2417,2.4751],[2420,2.4771],[2422,2.4678],[2426,2.4679],[2458,2.931],[2462,2.9371],[2464,2.9449],[2469,2.6513],[2496,-4.2933],[2504,-2.9662],[2513,-0.3174],[2518,-3.4567],[2519,-9.6755],[2520,-11.137],[2521,-8.0837],[2522,-7.5385],[2524,-9.9345],[2527,-6.8987],[2530,-6.5716],[2531,-6.5716],[2758,-10.2536],[2763,-18.2073],[2774,-6.6773],[2782,2.6601],[2801,-0.3354],[2927,-0.639],[2938,-0.6457],[2986,-6.1772],[3099,-6.1772],[3185,-10.0937],[3215,2.4771],[3217,2.4678],[3220,2.4679],[3253,2.931],[3259,2.9449],[3261,2.9431],[3287,2.4849],[3289,2.485],[3293,2.4677],[3300,2.4679],[3333,2.931],[3339,2.9449],[3341,2.9431],[3369,2.4825],[3370,2.4835],[3372,2.4751],[3375,2.4771],[3377,2.4678],[3380,2.4679],[3413,2.931],[3417,2.9371],[3419,2.9449],[3421,2.9431],[3449,2.4825],[3450,2.4835],[3452,2.4751],[3455,2.4771],[3457,2.4678],[3460,2.4679],[3493,2.931],[3499,2.9449],[3501,2.9431],[3529,2.4825],[3530,2.4835],[3532,2.4751],[3535,2.4771],[3537,2.4678],[3541,2.4679],[3573,2.931],[3579,2.9449],[3581,2.9431],[3607,2.4849],[3610,2.485],[3613,2.4677],[3620,2.4679],[3653,2.931],[3659,2.9449],[3661,2.9431],[3689,2.4825],[3690,2.4835],[3692,2.4751],[3695,2.4771],[3697,2.4678],[3700,2.4679],[3733,2.931],[3737,2.9371],[3739,2.9449],[3741,2.9431],[3769,2.4825],[3770,2.4835],[3772,2.4751],[3775,2.4771],[3777,2.4678],[3783,2.4679],[3813,2.931],[3817,2.9371],[3819,2.9449],[3821,2.9431],[3849,2.4825],[3850,2.4835],[3852,2.4751],[3855,2.4771],[3857,2.4678],[3894,2.4835],[3896,2.4751],[3899,2.4771],[3901,2.4678],[3907,2.4679],[3937,2.931],[3941,2.9371],[3943,2.9449],[3945,2.9431],[3973,2.4825],[3974,2.4835],[3976,2.4751],[3979,2.4771],[3981,2.4678],[4018,2.4835],[4020,2.4751],[4023,2.4771],[4025,2.4678],[4031,2.4679],[4061,2.931],[4065,2.9371],[4067,2.9449],[4069,2.9431],[4097,2.4825],[4098,2.4835],[4100,2.4751],[4103,2.4771],[4105,2.4678],[4143,2.4835],[4145,2.4751],[4148,2.4771],[4150,2.4678],[4156,2.4679],[4186,2.931],[4190,2.9371],[4192,2.9449],[4194,2.9431],[4222,2.4825],[4223,2.4835],[4225,2.4751],[4228,2.4771],[4230,2.4678],[4269,2.4835],[4271,2.4751],[4274,2.4771],[4276,2.4678],[4282,2.4679],[4312,2.931],[4316,2.9371],[4318,2.9449],[4320,2.9431],[4348,2.4825],[4349,2.4835],[4351,2.4751],[4354,2.4771],[4356,2.4678],[4360,2.9123],[4364,2.8976],[4396,2.4849],[4399,2.485],[4402,2.4677],[4409,2.4679],[4435,2.8953],[4436,0.0012]]
       }
       index_01_r_ctrl.Rotation.Z {
        type=MovieSceneScriptingFloatChannel
        default=6.7702
        thinned_from=198
        keys [[1790,6.7702],[1818,10.5135],[1819,10.5133],[1821,10.5739],[1826,10.6345],[1852,6.8764],[1854,6.8766],[1858,6.7445],[1865,6.7449],[1898,10.5135],[1899,10.5133],[1901,10.5739],[1906,10.6345],[1937,6.8011],[1938,6.8012],[1942,6.7447],[1945,6.7449],[1978,10.5135],[1979,10.5133],[1981,10.5739],[1987,10.6345],[2017,6.8011],[2018,6.8012],[2022,6.7447],[2025,6.7449],[2058,10.5135],[2059,10.5133],[2061,10.5739],[2067,10.6345],[2097,6.8011],[2098,6.8012],[2102,6.7447],[2105,6.7449],[2138,10.5135],[2139,10.5133],[2147,10.6345],[2172,6.8764],[2174,6.8766],[2179,6.7445],[2185,6.7449],[2218,10.5135],[2219,10.5133],[2227,10.6345],[2257,6.8011],[2258,6.8012],[2262,6.7447],[2265,6.7449],[2298,10.5135],[2299,10.5133],[2307,10.6345],[2337,6.8011],[2338,6.8012],[2342,6.7447],[2345,6.7449],[2378,10.5135],[2379,10.5133],[2387,10.6345],[2417,6.8011],[2418,6.8012],[2422,6.7447],[2425,6.7449],[2458,10.5135],[2459,10.5133],[2466,10.6343],[2468,10.5668],[2504,7.5013],[2508,-24.0893],[2513,-36.2842],[2518,-20.1078],[2519,14.0029],[2521,36.3446],[2523,33.9082],[2530,36.6397],[2775,16.7196],[2785,-7.4989],[2927,-5.7135],[2938,-5.6696],[2986,91.9544],[3099,91.9544],[3185,71.1],[3217,6.7447],[3220,6.7449],[3253,10.5135],[3254,10.5133],[3262,10.6345],[3287,6.8764],[3289,6.8766],[3294,6.7445],[3300,6.7449],[3333,10.5135],[3334,10.5133],[3342,10.6345],[3372,6.8011],[3373,6.8012],[3377,6.7447],[3380,6.7449],[3413,10.5135],[3414,10.5133],[3422,10.6345],[3452,6.8011],[3453,6.8012],[3457,6.7447],[3460,6.7449],[3493,10.5135],[3494,10.5133],[3502,10.6345],[3532,6.8011],[3533,6.8012],[3537,6.7447],[3540,6.7449],[3573,10.5135],[3574,10.5133],[3582,10.6345],[3607,6.8764],[3609,6.8766],[3614,6.7445],[3620,6.7449],[3653,10.5135],[3654,10.5133],[3662,10.6345],[3692,6.8011],[3693,6.8012],[3697,6.7447],[3700,6.7449],[3733,10.5135],[3734,10.5133],[3742,10.6345],[3772,6.8011],[3773,6.8012],[3777,6.7447],[3780,6.7449],[3813,10.5135],[3814,10.5133],[3822,10.6345],[3852,6.8011],[3853,6.8012],[3857,6.7447],[3897,6.8012],[3901,6.7447],[3904,6.7449],[3937,10.5135],[3938,10.5133],[3946,10.6345],[3976,6.8011],[3977,6.8012],[3981,6.7447],[4021,6.8012],[4025,6.7447],[4028,6.7449],[4061,10.5135],[4062,10.5133],[4070,10.6345],[4100,6.8011],[4101,6.8012],[4105,6.7447],[4146,6.8012],[4150,6.7447],[4153,6.7449],[4186,10.5135],[4187,10.5133],[4195,10.6345],[4225,6.8011],[4226,6.8012],[4230,6.7447],[4272,6.8012],[4276,6.7447],[4279,6.7449],[4312,10.5135],[4313,10.5133],[4321,10.6345],[4351,6.8011],[4352,6.8012],[4356,6.7447],[4360,10.3608],[4364,10.5035],[4371,10.6345],[4396,6.8764],[4398,6.8766],[4403,6.7445],[4409,6.7449],[4435,10.2346],[4436,0.0026]]
       }
       index_01_r_ctrl.Scale.X {type=MovieSceneScriptingFloatChannel, default=1, const [1,12]}
       index_01_r_ctrl.Scale.Y {type=MovieSceneScriptingFloatChannel, default=1, const [1,12]}
       index_01_r_ctrl.Scale.Z {type=MovieSceneScriptingFloatChannel, default=1, const [1,12]}
       index_02_r_ctrl.Location.X {type=MovieSceneScriptingFloatChannel, default=0, thinned_from=465, keys [[1790,0],[2512,0],[2513,-0.0002],[2530,0],[4436,0]]}
       index_02_r_ctrl.Location.Y {type=MovieSceneScriptingFloatChannel, default=0, thinned_from=702, keys [[1790,0],[2512,0],[2513,-0.0001],[2530,0],[4436,0]]}
       index_02_r_ctrl.Location.Z {type=MovieSceneScriptingFloatChannel, default=0, thinned_from=210, keys [[1790,0],[4436,0]]}
       index_02_r_ctrl.Rotation.X {
        type=MovieSceneScriptingFloatChannel
        default=5.2264
        thinned_from=207
        keys [[1790,5.2264],[1814,5.6768],[1827,5.7154],[1853,5.2404],[1862,5.2285],[1865,5.2166],[1894,5.6768],[1907,5.7154],[1945,5.2166],[1974,5.6768],[1985,5.7159],[1987,5.7128],[2025,5.2166],[2054,5.6768],[2067,5.7154],[2093,5.2404],[2109,5.2211],[2134,5.6767],[2148,5.7154],[2173,5.2404],[2182,5.2285],[2185,5.2166],[2214,5.6768],[2227,5.7154],[2265,5.2166],[2294,5.6767],[2305,5.7159],[2307,5.7128],[2345,5.2166],[2374,5.6768],[2387,5.7154],[2413,5.2404],[2429,5.2211],[2454,5.6768],[2466,5.712],[2471,5.3603],[2496,-0.1057],[2504,-0.2358],[2513,-0.3576],[2519,0.4073],[2522,-0.1958],[2524,0.1876],[2531,-0.3624],[2760,0.9963],[2763,1.5209],[2771,2.9708],[2780,5.0135],[2799,2.37],[2927,2.1796],[2938,2.1763],[2986,1.6113],[3099,1.6113],[3185,1.1671],[3215,5.2297],[3224,5.2211],[3249,5.6768],[3263,5.7154],[3288,5.2404],[3301,5.2166],[3329,5.6768],[3343,5.7154],[3380,5.2166],[3409,5.6768],[3420,5.7159],[3422,5.7128],[3460,5.2166],[3489,5.6768],[3503,5.7154],[3528,5.2404],[3544,5.2211],[3569,5.6768],[3583,5.7154],[3608,5.2404],[3621,5.2166],[3649,5.6768],[3662,5.7154],[3700,5.2166],[3729,5.6768],[3740,5.7159],[3742,5.7128],[3780,5.2166],[3809,5.6768],[3823,5.7154],[3848,5.2404],[3864,5.2211],[3904,5.2166],[3933,5.6768],[3947,5.7154],[3972,5.2404],[3988,5.2211],[4028,5.2166],[4057,5.6768],[4071,5.7154],[4096,5.2404],[4112,5.2211],[4153,5.2166],[4182,5.6768],[4196,5.7154],[4221,5.2404],[4237,5.2211],[4279,5.2166],[4308,5.6768],[4322,5.7154],[4347,5.2404],[4356,5.2285],[4360,5.6768],[4365,5.6767],[4371,5.7154],[4397,5.2404],[4410,5.2166],[4435,5.6625],[4436,0]]
       }
       index_02_r_ctrl.Rotation.Y {
        type=MovieSceneScriptingFloatChannel
        default=-6.1831
        thinned_from=180
        keys [[1790,-6.1831],[1825,-6.7595],[1827,-6.7586],[1857,-6.1824],[1869,-6.1752],[1905,-6.7595],[1907,-6.7586],[1944,-6.1758],[1949,-6.1752],[1978,-6.7446],[1986,-6.7636],[2024,-6.1758],[2029,-6.1752],[2065,-6.7595],[2067,-6.7586],[2097,-6.1824],[2105,-6.1752],[2145,-6.7595],[2147,-6.7586],[2177,-6.1824],[2189,-6.1752],[2225,-6.7595],[2227,-6.7586],[2264,-6.1758],[2269,-6.1752],[2298,-6.7446],[2306,-6.7636],[2344,-6.1758],[2349,-6.1752],[2385,-6.7595],[2387,-6.7586],[2417,-6.1824],[2425,-6.1752],[2465,-6.7595],[2470,-6.5279],[2502,0.0526],[2503,0.0526],[2505,0.054],[2506,0.0539],[2512,0.0574],[2515,0.0574],[2519,-1.0111],[2522,-0.0936],[2524,-0.7067],[2527,0.0684],[2530,0.1486],[2532,0.1486],[2757,-0.2686],[2760,-1.954],[2767,-3.3144],[2780,-5.9937],[2784,-5.3344],[2799,-2.6],[2927,-2.3975],[2938,-2.3939],[2986,-2.0705],[3099,-2.0705],[3185,-1.0922],[3217,-6.174],[3220,-6.1752],[3260,-6.7595],[3262,-6.7586],[3292,-6.1824],[3304,-6.1752],[3340,-6.7595],[3342,-6.7586],[3379,-6.1758],[3384,-6.1752],[3416,-6.7541],[3422,-6.7636],[3459,-6.1758],[3464,-6.1752],[3500,-6.7595],[3502,-6.7586],[3532,-6.1824],[3540,-6.1752],[3580,-6.7595],[3582,-6.7586],[3612,-6.1824],[3624,-6.1752],[3660,-6.7595],[3662,-6.7586],[3699,-6.1758],[3704,-6.1752],[3736,-6.7541],[3739,-6.753],[3741,-6.7636],[3779,-6.1758],[3784,-6.1752],[3820,-6.7595],[3822,-6.7586],[3852,-6.1824],[3908,-6.1752],[3944,-6.7595],[3946,-6.7586],[3976,-6.1824],[4032,-6.1752],[4068,-6.7595],[4070,-6.7586],[4100,-6.1824],[4157,-6.1752],[4193,-6.7595],[4195,-6.7586],[4225,-6.1824],[4233,-6.1752],[4283,-6.1752],[4319,-6.7595],[4321,-6.7586],[4351,-6.1824],[4356,-6.174],[4360,-6.6526],[4364,-6.7079],[4371,-6.7586],[4401,-6.1824],[4413,-6.1752],[4435,-6.7022],[4436,0]]
       }
       index_02_r_ctrl.Rotation.Z {
        type=MovieSceneScriptingFloatChannel
        default=15.0095
        thinned_from=207
        keys [[1790,15.0095],[1824,22.0717],[1827,22.0713],[1856,15.0214],[1858,15.0219],[1860,14.9162],[1869,14.9169],[1904,22.0717],[1907,22.0713],[1939,14.9655],[1940,14.9658],[1942,14.9165],[1949,14.9169],[1978,21.8958],[1981,22.0077],[1984,22.0073],[1987,22.1196],[2019,14.9655],[2020,14.9658],[2022,14.9165],[2029,14.9169],[2064,22.0717],[2067,22.0713],[2096,15.0214],[2098,15.0219],[2100,14.9162],[2105,14.9169],[2144,22.0717],[2151,22.0713],[2176,15.0214],[2178,15.0219],[2180,14.9162],[2189,14.9169],[2224,22.0717],[2227,22.0713],[2259,14.9655],[2260,14.9658],[2262,14.9165],[2269,14.9169],[2302,22.0077],[2304,22.0073],[2307,22.1196],[2339,14.9655],[2340,14.9658],[2342,14.9165],[2349,14.9169],[2384,22.0717],[2387,22.0713],[2416,15.0214],[2418,15.0219],[2420,14.9162],[2425,14.9169],[2464,22.0717],[2469,21.7726],[2471,21.8567],[2496,-6.8515],[2504,-11.6497],[2512,-23.08],[2515,-23.0798],[2519,4.3526],[2520,7.0788],[2521,-5.7211],[2522,-7.9991],[2524,2.0916],[2528,-11.5841],[2532,-12.1241],[2759,13.4324],[2764,37.1284],[2794,-16.1879],[2938,-16.0398],[2986,54.136],[3099,54.136],[3185,36.3821],[3215,14.9162],[3220,14.9169],[3259,22.0717],[3262,22.0713],[3291,15.0214],[3293,15.0219],[3295,14.9162],[3304,14.9169],[3339,22.0717],[3342,22.0713],[3374,14.9655],[3375,14.9658],[3377,14.9165],[3384,14.9169],[3417,22.0077],[3419,22.0073],[3422,22.1196],[3454,14.9655],[3455,14.9658],[3457,14.9165],[3464,14.9169],[3499,22.0717],[3502,22.0713],[3531,15.0214],[3533,15.0219],[3535,14.9162],[3540,14.9169],[3579,22.0717],[3582,22.0713],[3611,15.0214],[3613,15.0219],[3615,14.9162],[3624,14.9169],[3659,22.0717],[3662,22.0713],[3694,14.9655],[3695,14.9658],[3697,14.9165],[3704,14.9169],[3737,22.0077],[3739,22.0073],[3742,22.1196],[3774,14.9655],[3775,14.9658],[3777,14.9165],[3784,14.9169],[3819,22.0717],[3822,22.0713],[3851,15.0215],[3853,15.0219],[3855,14.9162],[3899,14.9658],[3901,14.9165],[3908,14.9169],[3943,22.0717],[3946,22.0713],[3975,15.0215],[3977,15.0219],[3979,14.9162],[4023,14.9658],[4025,14.9165],[4032,14.9169],[4067,22.0717],[4070,22.0713],[4099,15.0215],[4101,15.0219],[4103,14.9162],[4148,14.9658],[4150,14.9165],[4157,14.9169],[4192,22.0717],[4195,22.0713],[4224,15.0215],[4226,15.0219],[4228,14.9162],[4274,14.9658],[4276,14.9165],[4283,14.9169],[4318,22.0717],[4321,22.0713],[4350,15.0215],[4352,15.0219],[4354,14.9162],[4356,14.9166],[4360,20.9019],[4364,21.5117],[4371,22.0713],[4400,15.0214],[4402,15.0219],[4404,14.9162],[4413,14.9169],[4435,21.3817],[4436,0]]
       }
       index_02_r_ctrl.Scale.X {type=MovieSceneScriptingFloatChannel, default=1, const [1,12]}
       index_02_r_ctrl.Scale.Y {type=MovieSceneScriptingFloatChannel, default=1, const [1,12]}
       index_02_r_ctrl.Scale.Z {type=MovieSceneScriptingFloatChannel, default=1, const [1,12]}
       index_03_r_ctrl.Location.X {type=MovieSceneScriptingFloatChannel, default=0, thinned_from=550, keys [[1790,0],[2512,0],[2513,-0.0001],[2526,0],[4435,0],[4436,-0.0002]]}
       index_03_r_ctrl.Location.Y {type=MovieSceneScriptingFloatChannel, default=0, thinned_from=807, keys [[1790,0],[4436,-0.0001]]}
       index_03_r_ctrl.Location.Z {type=MovieSceneScriptingFloatChannel, default=0, thinned_from=444, keys [[1790,0],[4436,0]]}
       index_03_r_ctrl.Rotation.X {
        type=MovieSceneScriptingFloatChannel
        default=0.4479
        thinned_from=482
        keys [[1790,0.4479],[1791,0.4479],[1809,0.4827],[1810,0.4824],[1814,0.4854],[1815,0.4848],[1818,0.4868],[1822,0.4874],[1824,0.4898],[1826,0.4881],[1832,0.4887],[1850,0.4487],[1851,0.4491],[1855,0.4478],[1856,0.4485],[1858,0.4463],[1860,0.4479],[1862,0.4478],[1864,0.4456],[1866,0.4456],[1872,0.4503],[1873,0.4501],[1889,0.4827],[1890,0.4824],[1894,0.4854],[1895,0.4848],[1898,0.4868],[1902,0.4874],[1904,0.4898],[1906,0.4881],[1930,0.4488],[1931,0.4505],[1935,0.4472],[1937,0.4508],[1939,0.4456],[1949,0.4456],[1952,0.4503],[1953,0.4501],[1969,0.4827],[1970,0.4824],[1974,0.4854],[1975,0.4848],[1988,0.4888],[2010,0.4487],[2011,0.4491],[2015,0.4478],[2016,0.4485],[2018,0.4463],[2020,0.4479],[2024,0.4456],[2029,0.4456],[2032,0.4503],[2033,0.4501],[2047,0.4814],[2048,0.4808],[2052,0.4857],[2053,0.4838],[2064,0.4898],[2066,0.4881],[2067,0.4881],[2090,0.4487],[2091,0.4491],[2095,0.4478],[2096,0.4485],[2098,0.4463],[2100,0.4479],[2102,0.4478],[2104,0.4456],[2111,0.4479],[2129,0.4827],[2130,0.4824],[2134,0.4854],[2135,0.4848],[2142,0.4874],[2144,0.4898],[2146,0.4881],[2151,0.4881],[2152,0.4887],[2170,0.4487],[2171,0.4491],[2175,0.4478],[2176,0.4485],[2178,0.4463],[2180,0.4479],[2181,0.4478],[2184,0.4456],[2189,0.4456],[2192,0.4503],[2193,0.4501],[2209,0.4827],[2210,0.4824],[2214,0.4854],[2215,0.4848],[2222,0.4874],[2224,0.4898],[2226,0.4881],[2227,0.4881],[2250,0.4488],[2251,0.4505],[2255,0.4472],[2257,0.4508],[2260,0.4456],[2269,0.4456],[2272,0.4503],[2273,0.4501],[2289,0.4827],[2290,0.4824],[2294,0.4854],[2295,0.4848],[2308,0.4888],[2330,0.4487],[2331,0.4491],[2335,0.4478],[2336,0.4485],[2338,0.4463],[2340,0.4479],[2344,0.4456],[2349,0.4456],[2352,0.4503],[2353,0.4501],[2367,0.4814],[2368,0.4808],[2372,0.4857],[2373,0.4838],[2382,0.4874],[2384,0.4898],[2386,0.4881],[2387,0.4881],[2410,0.4487],[2411,0.4491],[2415,0.4478],[2416,0.4485],[2418,0.4463],[2420,0.4479],[2422,0.4478],[2424,0.4456],[2426,0.4456],[2430,0.4479],[2431,0.4479],[2449,0.4827],[2450,0.4824],[2454,0.4854],[2455,0.4848],[2462,0.4874],[2464,0.4898],[2471,0.4659],[2496,-0.3994],[2504,-0.4614],[2512,-0.5783],[2515,-0.5786],[2519,-0.2096],[2520,-0.1782],[2521,-0.3645],[2522,-0.4005],[2524,-0.2548],[2526,-0.4046],[2529,-0.4611],[2531,-0.4611],[2757,-0.3606],[2760,0.05],[2763,0.2505],[2766,0.2752],[2778,0.4388],[2927,0.3797],[2938,0.3789],[2986,0.3457],[3099,0.3457],[3185,0.3537],[3215,0.4479],[3217,0.4478],[3219,0.4456],[3226,0.4479],[3244,0.4827],[3245,0.4824],[3249,0.4854],[3250,0.4848],[3259,0.4898],[3261,0.4881],[3266,0.4881],[3267,0.4887],[3285,0.4487],[3286,0.4491],[3290,0.4478],[3291,0.4485],[3293,0.4463],[3295,0.4479],[3301,0.4456],[3307,0.4503],[3308,0.4501],[3324,0.4827],[3325,0.4824],[3329,0.4854],[3330,0.4848],[3339,0.4898],[3341,0.4881],[3342,0.4881],[3365,0.4488],[3366,0.4505],[3369,0.4472],[3372,0.4508],[3375,0.4456],[3381,0.4456],[3387,0.4503],[3388,0.4501],[3404,0.4827],[3405,0.4824],[3409,0.4854],[3410,0.4848],[3423,0.4888],[3445,0.4487],[3446,0.4491],[3450,0.4478],[3451,0.4485],[3453,0.4463],[3455,0.4479],[3461,0.4456],[3467,0.4503],[3468,0.4501],[3482,0.4814],[3483,0.4808],[3487,0.4857],[3488,0.4838],[3499,0.4898],[3501,0.4881],[3506,0.4881],[3525,0.4487],[3526,0.4491],[3530,0.4478],[3531,0.4485],[3533,0.4463],[3535,0.4479],[3537,0.4478],[3539,0.4456],[3541,0.4456],[3545,0.4479],[3546,0.4479],[3564,0.4827],[3565,0.4824],[3569,0.4854],[3570,0.4848],[3579,0.4898],[3581,0.4881],[3587,0.4887],[3605,0.4487],[3606,0.4491],[3610,0.4478],[3611,0.4485],[3613,0.4463],[3615,0.4479],[3621,0.4456],[3627,0.4503],[3628,0.4501],[3644,0.4827],[3645,0.4824],[3649,0.4854],[3650,0.4848],[3659,0.4898],[3661,0.4881],[3662,0.4881],[3685,0.4488],[3686,0.4505],[3688,0.4472],[3690,0.4472],[3692,0.4508],[3695,0.4456],[3701,0.4456],[3707,0.4503],[3708,0.4501],[3724,0.4827],[3725,0.4824],[3729,0.4854],[3730,0.4848],[3743,0.4888],[3765,0.4487],[3766,0.4491],[3770,0.4478],[3771,0.4485],[3773,0.4463],[3775,0.4479],[3781,0.4456],[3787,0.4503],[3788,0.4501],[3802,0.4814],[3803,0.4808],[3807,0.4857],[3808,0.4838],[3817,0.4874],[3819,0.4898],[3822,0.4881],[3845,0.4487],[3846,0.4491],[3850,0.4478],[3851,0.4485],[3853,0.4463],[3855,0.4479],[3860,0.4456],[3866,0.4479],[3884,0.4827],[3885,0.4824],[3890,0.4491],[3894,0.4478],[3895,0.4485],[3897,0.4463],[3899,0.4479],[3905,0.4456],[3911,0.4503],[3912,0.4501],[3926,0.4814],[3927,0.4808],[3931,0.4857],[3932,0.4838],[3941,0.4874],[3943,0.4898],[3946,0.4881],[3969,0.4487],[3970,0.4491],[3974,0.4478],[3975,0.4485],[3977,0.4463],[3979,0.4479],[3984,0.4456],[3990,0.4479],[4008,0.4827],[4009,0.4824],[4014,0.4491],[4018,0.4478],[4019,0.4485],[4021,0.4463],[4023,0.4479],[4029,0.4456],[4035,0.4503],[4036,0.4501],[4050,0.4814],[4051,0.4808],[4055,0.4857],[4056,0.4838],[4065,0.4874],[4067,0.4898],[4070,0.4881],[4093,0.4487],[4094,0.4491],[4098,0.4478],[4099,0.4485],[4101,0.4463],[4103,0.4479],[4108,0.4456],[4114,0.4479],[4132,0.4827],[4133,0.4824],[4139,0.4491],[4143,0.4478],[4144,0.4485],[4146,0.4463],[4148,0.4479],[4154,0.4456],[4160,0.4503],[4161,0.4501],[4175,0.4814],[4176,0.4808],[4180,0.4857],[4181,0.4838],[4190,0.4874],[4192,0.4898],[4195,0.4881],[4218,0.4487],[4219,0.4491],[4223,0.4478],[4224,0.4485],[4226,0.4463],[4228,0.4479],[4233,0.4456],[4239,0.4479],[4257,0.4827],[4258,0.4824],[4265,0.4491],[4269,0.4478],[4270,0.4485],[4272,0.4463],[4274,0.4479],[4280,0.4456],[4286,0.4503],[4287,0.4501],[4301,0.4814],[4302,0.4808],[4306,0.4857],[4307,0.4838],[4316,0.4874],[4318,0.4898],[4321,0.4881],[4344,0.4487],[4345,0.4491],[4349,0.4478],[4350,0.4485],[4352,0.4463],[4354,0.4479],[4360,0.4854],[4365,0.4848],[4371,0.4881],[4376,0.4887],[4394,0.4487],[4395,0.4491],[4399,0.4478],[4400,0.4485],[4402,0.4463],[4404,0.4479],[4410,0.4456],[4416,0.4503],[4417,0.4501],[4433,0.4827],[4435,0.483],[4436,0.0004]]
       }
       index_03_r_ctrl.Rotation.Y {
        type=MovieSceneScriptingFloatChannel
        default=-0.6753
        thinned_from=240
        keys [[1790,-0.6753],[1818,-0.7373],[1827,-0.7394],[1857,-0.676],[1864,-0.6752],[1898,-0.7373],[1907,-0.7394],[1932,-0.677],[1935,-0.6774],[1938,-0.6748],[1945,-0.6752],[1985,-0.7389],[1992,-0.739],[2017,-0.676],[2025,-0.6752],[2061,-0.7383],[2067,-0.7394],[2097,-0.676],[2109,-0.675],[2141,-0.7383],[2151,-0.7394],[2177,-0.676],[2184,-0.6752],[2218,-0.7373],[2227,-0.7394],[2252,-0.677],[2255,-0.6774],[2258,-0.6748],[2265,-0.6752],[2305,-0.7389],[2312,-0.739],[2337,-0.676],[2345,-0.6752],[2381,-0.7383],[2387,-0.7394],[2417,-0.676],[2429,-0.675],[2461,-0.7383],[2464,-0.7382],[2466,-0.7393],[2513,0.4682],[2519,0.4177],[2522,0.6129],[2524,0.5068],[2529,0.6553],[2531,0.6553],[2760,0.2967],[2763,0.1516],[2771,-0.1577],[2780,-0.622],[2781,-0.6178],[2782,-0.6207],[2801,-0.248],[2927,-0.2291],[2938,-0.2288],[2986,-0.274],[3099,-0.274],[3185,0.3201],[3216,-0.6749],[3224,-0.675],[3257,-0.7383],[3266,-0.7394],[3292,-0.676],[3300,-0.6752],[3336,-0.7383],[3342,-0.7394],[3367,-0.677],[3368,-0.6774],[3373,-0.6748],[3384,-0.6752],[3420,-0.7389],[3427,-0.739],[3452,-0.676],[3456,-0.6749],[3460,-0.6752],[3497,-0.7383],[3502,-0.7394],[3532,-0.676],[3536,-0.6749],[3544,-0.675],[3576,-0.7383],[3582,-0.7394],[3612,-0.676],[3620,-0.6752],[3657,-0.7383],[3662,-0.7394],[3687,-0.677],[3690,-0.6774],[3693,-0.6748],[3700,-0.6752],[3740,-0.7389],[3747,-0.739],[3772,-0.676],[3776,-0.6749],[3780,-0.6752],[3816,-0.7383],[3826,-0.7394],[3852,-0.676],[3864,-0.675],[3897,-0.6761],[3900,-0.6749],[3904,-0.6752],[3940,-0.7383],[3950,-0.7394],[3976,-0.676],[3988,-0.675],[4021,-0.6761],[4024,-0.6749],[4028,-0.6752],[4064,-0.7383],[4074,-0.7394],[4100,-0.676],[4112,-0.675],[4146,-0.6761],[4149,-0.6749],[4153,-0.6752],[4189,-0.7383],[4199,-0.7394],[4225,-0.676],[4237,-0.675],[4272,-0.6761],[4275,-0.6749],[4279,-0.6752],[4315,-0.7383],[4325,-0.7394],[4351,-0.676],[4356,-0.6749],[4360,-0.7306],[4364,-0.7352],[4371,-0.7394],[4401,-0.676],[4409,-0.6752],[4435,-0.7328],[4436,0.0002]]
       }
       index_03_r_ctrl.Rotation.Z {
        type=MovieSceneScriptingFloatChannel
        default=12.1267
        thinned_from=170
        keys [[1790,12.1267],[1818,20.4425],[1821,20.577],[1826,20.7115],[1856,12.1963],[1860,12.0703],[1865,12.0703],[1898,20.4425],[1901,20.577],[1906,20.7115],[1933,12.3642],[1935,12.3642],[1939,12.0703],[1945,12.0703],[1984,20.6539],[1992,20.6659],[2020,12.0703],[2025,12.0703],[2055,20.308],[2061,20.577],[2067,20.7115],[2100,12.0703],[2104,12.0703],[2142,20.577],[2147,20.7115],[2180,12.0703],[2185,12.0703],[2222,20.577],[2227,20.7115],[2253,12.3642],[2259,12.0703],[2265,12.0703],[2304,20.6539],[2312,20.6659],[2340,12.0703],[2345,12.0703],[2381,20.577],[2387,20.7115],[2420,12.0703],[2424,12.0703],[2462,20.577],[2463,20.577],[2466,20.711],[2502,-39.4496],[2513,-21.44],[2515,-21.6327],[2520,5.6347],[2522,-1.6909],[2524,3.3239],[2531,-3.7345],[2764,20.5193],[2772,20.4904],[2798,-36.4056],[2938,-36.3881],[2986,26.2994],[3099,26.2994],[3185,32.2257],[3216,12.0703],[3219,12.0703],[3262,20.7115],[3295,12.0703],[3299,12.0703],[3333,20.4425],[3342,20.7115],[3368,12.3642],[3374,12.0703],[3380,12.0703],[3420,20.6539],[3427,20.6658],[3456,12.0703],[3460,12.0703],[3490,20.308],[3502,20.7115],[3535,12.0703],[3539,12.0703],[3577,20.577],[3582,20.7115],[3616,12.0703],[3619,12.0703],[3658,20.577],[3662,20.7115],[3688,12.3642],[3689,12.3642],[3694,12.0703],[3700,12.0703],[3740,20.6539],[3747,20.6658],[3776,12.0703],[3779,12.0703],[3813,20.4425],[3822,20.7115],[3855,12.0703],[3903,12.0703],[3937,20.4425],[3946,20.7115],[3979,12.0703],[4027,12.0703],[4061,20.4425],[4070,20.7115],[4103,12.0703],[4152,12.0703],[4186,20.4425],[4195,20.7115],[4228,12.0703],[4278,12.0703],[4312,20.4425],[4321,20.7115],[4354,12.0703],[4355,12.0703],[4360,19.4035],[4364,20.0855],[4371,20.7115],[4404,12.0703],[4408,12.0703],[4435,19.8254],[4436,0.0049]]
       }
       index_03_r_ctrl.Scale.X {type=MovieSceneScriptingFloatChannel, default=1, const [1,12]}
       index_03_r_ctrl.Scale.Y {type=MovieSceneScriptingFloatChannel, default=1, const [1,12]}
       index_03_r_ctrl.Scale.Z {type=MovieSceneScriptingFloatChannel, default=1, const [1,12]}
       middle_metacarpal_r_ctrl.Location.X {type=MovieSceneScriptingFloatChannel, default=0, thinned_from=37, keys [[1790,0],[2512,0],[2513,-0.0001],[2519,0],[4436,0]]}
       middle_metacarpal_r_ctrl.Location.Y {type=MovieSceneScriptingFloatChannel, default=0, thinned_from=36, keys [[1790,0],[4436,0]]}
       middle_metacarpal_r_ctrl.Location.Z {type=MovieSceneScriptingFloatChannel, default=0, thinned_from=33, keys [[1790,0],[4436,0]]}
       middle_metacarpal_r_ctrl.Rotation.X {
        type=MovieSceneScriptingFloatChannel
        default=-1.0443
        thinned_from=155
        keys [[1790,-1.0443],[1810,-0.3491],[1830,-0.2583],[1839,-0.5113],[1848,-0.9668],[1867,-1.0566],[1877,-0.8212],[1888,-0.3854],[1910,-0.2583],[1919,-0.5104],[1929,-0.991],[1948,-1.0559],[1969,-0.3633],[1990,-0.2583],[1999,-0.5104],[2008,-0.967],[2027,-1.0566],[2037,-0.8212],[2048,-0.3854],[2070,-0.2583],[2079,-0.5105],[2088,-0.9663],[2107,-1.0566],[2117,-0.8212],[2128,-0.3854],[2150,-0.2583],[2159,-0.5113],[2168,-0.9668],[2187,-1.0566],[2197,-0.8212],[2208,-0.3854],[2230,-0.2583],[2239,-0.5104],[2249,-0.991],[2268,-1.0559],[2289,-0.3633],[2310,-0.2583],[2319,-0.5104],[2328,-0.967],[2347,-1.0566],[2357,-0.8212],[2368,-0.3854],[2390,-0.2583],[2399,-0.5105],[2408,-0.9663],[2427,-1.0566],[2436,-0.8637],[2446,-0.4451],[2466,-0.2624],[2469,-0.2756],[2471,-0.2518],[2502,0],[2513,-0.0001],[2519,-0.0392],[2522,-0.0084],[2524,-0.029],[2530,0],[2531,0],[2774,-0.1685],[2787,0],[3185,0],[3222,-1.0566],[3232,-0.8213],[3243,-0.3854],[3265,-0.2583],[3274,-0.5112],[3283,-0.9668],[3302,-1.0566],[3324,-0.3634],[3346,-0.2583],[3364,-0.991],[3383,-1.0559],[3403,-0.3854],[3424,-0.2583],[3433,-0.4559],[3443,-0.967],[3463,-1.0566],[3484,-0.3634],[3505,-0.2583],[3514,-0.5105],[3523,-0.9663],[3542,-1.0566],[3564,-0.3634],[3586,-0.2583],[3604,-0.9911],[3623,-1.0566],[3644,-0.3634],[3666,-0.2583],[3684,-0.991],[3703,-1.0559],[3724,-0.3634],[3745,-0.2583],[3754,-0.5103],[3763,-0.967],[3782,-1.0566],[3792,-0.8213],[3803,-0.3854],[3825,-0.2583],[3834,-0.5105],[3844,-0.991],[3863,-1.0566],[3884,-0.3634],[3906,-1.0566],[3916,-0.8213],[3927,-0.3854],[3949,-0.2583],[3958,-0.5105],[3968,-0.991],[3987,-1.0566],[4008,-0.3634],[4030,-1.0566],[4040,-0.8213],[4051,-0.3854],[4073,-0.2583],[4082,-0.5105],[4092,-0.991],[4111,-1.0566],[4132,-0.3634],[4155,-1.0566],[4165,-0.8213],[4176,-0.3854],[4198,-0.2583],[4207,-0.5105],[4217,-0.991],[4236,-1.0566],[4257,-0.3634],[4281,-1.0566],[4291,-0.8213],[4302,-0.3854],[4324,-0.2583],[4333,-0.5105],[4343,-0.991],[4360,-0.3266],[4364,-0.3344],[4383,-0.5112],[4393,-0.991],[4412,-1.0566],[4424,-0.6902],[4430,-0.4452],[4433,-0.3634],[4434,-0.3491],[4435,-0.342],[4436,0]]
       }
       middle_metacarpal_r_ctrl.Rotation.Y {
        type=MovieSceneScriptingFloatChannel
        default=1.7777
        thinned_from=181
        keys [[1790,1.7777],[1830,1.8306],[1864,1.7772],[1867,1.777],[1910,1.8306],[1948,1.777],[1984,1.8303],[1989,1.8306],[2024,1.7772],[2027,1.777],[2064,1.8303],[2069,1.8306],[2104,1.7772],[2107,1.777],[2150,1.8306],[2184,1.7772],[2187,1.777],[2230,1.8306],[2268,1.777],[2304,1.8303],[2309,1.8306],[2344,1.7772],[2347,1.777],[2384,1.8303],[2389,1.8306],[2424,1.7772],[2427,1.777],[2466,1.8305],[2471,1.6665],[2502,-0.3349],[2507,0.3069],[2512,0.705],[2515,0.7049],[2518,0.4392],[2519,-0.1212],[2521,-0.4854],[2523,-0.4454],[2531,-0.49],[2765,-0.2724],[2776,0.8097],[2781,0.3415],[2787,0],[3185,0],[3218,1.7771],[3222,1.777],[3265,1.8306],[3299,1.7772],[3302,1.777],[3345,1.8306],[3383,1.777],[3419,1.8303],[3424,1.8306],[3459,1.7772],[3462,1.777],[3499,1.8303],[3504,1.8306],[3539,1.7772],[3541,1.7769],[3585,1.8306],[3619,1.7772],[3622,1.777],[3665,1.8306],[3703,1.777],[3739,1.8303],[3744,1.8306],[3779,1.7772],[3782,1.777],[3819,1.8303],[3824,1.8306],[3859,1.7772],[3906,1.777],[3943,1.8303],[3948,1.8306],[3983,1.7772],[4030,1.777],[4067,1.8303],[4072,1.8306],[4107,1.7772],[4155,1.777],[4192,1.8303],[4197,1.8306],[4232,1.7772],[4281,1.777],[4318,1.8303],[4323,1.8306],[4357,1.7771],[4360,1.8153],[4364,1.8222],[4373,1.8306],[4408,1.7772],[4411,1.777],[4435,1.8252],[4436,0]]
       }
       middle_metacarpal_r_ctrl.Rotation.Z {
        type=MovieSceneScriptingFloatChannel
        default=1.783
        thinned_from=236
        keys [[1790,1.783],[1809,0.172],[1829,-0.0959],[1830,-0.0958],[1838,0.3638],[1847,1.493],[1865,1.8075],[1867,1.8074],[1877,1.3009],[1888,0.2297],[1909,-0.0959],[1910,-0.0958],[1919,0.4903],[1928,1.5755],[1946,1.8059],[1955,1.4883],[1965,0.4674],[1970,0.1274],[1985,-0.0905],[1986,-0.0904],[1988,-0.0959],[1989,-0.0958],[1998,0.3649],[2007,1.4941],[2025,1.8075],[2027,1.8074],[2036,1.3986],[2046,0.3783],[2065,-0.0905],[2066,-0.0904],[2068,-0.0959],[2069,-0.0958],[2078,0.364],[2087,1.4926],[2105,1.8075],[2107,1.8074],[2117,1.3009],[2128,0.2297],[2149,-0.0959],[2150,-0.0958],[2158,0.3638],[2167,1.493],[2185,1.8075],[2187,1.8074],[2197,1.3009],[2208,0.2297],[2229,-0.0959],[2230,-0.0958],[2239,0.4903],[2248,1.5755],[2266,1.8059],[2275,1.4884],[2285,0.4675],[2290,0.1274],[2305,-0.0905],[2306,-0.0904],[2308,-0.0959],[2309,-0.0958],[2318,0.3649],[2327,1.4941],[2345,1.8075],[2347,1.8074],[2356,1.3986],[2366,0.3783],[2385,-0.0905],[2386,-0.0904],[2388,-0.0959],[2389,-0.0958],[2398,0.364],[2407,1.4926],[2425,1.8075],[2427,1.8074],[2437,1.3009],[2448,0.2297],[2469,-0.1022],[2470,-0.097],[2502,0],[2512,0],[2519,-0.0146],[2522,-0.0031],[2524,-0.0107],[2531,0],[2774,-0.0625],[2787,0],[2938,-0.0006],[2986,-0.0126],[3099,-0.0126],[3185,-0.0373],[3220,1.8075],[3222,1.8074],[3232,1.3011],[3243,0.2297],[3264,-0.0959],[3273,0.3637],[3282,1.4929],[3300,1.8075],[3303,1.8074],[3313,1.2003],[3323,0.2298],[3344,-0.0959],[3345,-0.0958],[3354,0.4901],[3363,1.5754],[3381,1.8059],[3390,1.4885],[3400,0.4676],[3405,0.1274],[3420,-0.0905],[3421,-0.0904],[3423,-0.0959],[3425,-0.0958],[3433,0.3647],[3442,1.4939],[3460,1.8075],[3470,1.4904],[3480,0.4676],[3485,0.1274],[3500,-0.0905],[3501,-0.0904],[3503,-0.0959],[3504,-0.0958],[3513,0.3638],[3522,1.4925],[3540,1.8075],[3543,1.8074],[3553,1.2003],[3563,0.2298],[3584,-0.0959],[3586,-0.0958],[3594,0.489],[3603,1.5749],[3620,1.8075],[3622,1.8074],[3632,1.301],[3643,0.2298],[3664,-0.0959],[3665,-0.0958],[3674,0.4901],[3683,1.5754],[3701,1.8059],[3710,1.4885],[3720,0.4676],[3725,0.1274],[3740,-0.0905],[3741,-0.0904],[3743,-0.0959],[3744,-0.0958],[3753,0.3647],[3762,1.4939],[3780,1.8075],[3782,1.8074],[3791,1.3988],[3801,0.3784],[3820,-0.0905],[3821,-0.0904],[3823,-0.0959],[3824,-0.0958],[3833,0.3638],[3842,1.4925],[3860,1.8075],[3862,1.8074],[3872,1.301],[3883,0.2297],[3904,1.8075],[3906,1.8074],[3915,1.3988],[3925,0.3784],[3944,-0.0905],[3945,-0.0904],[3947,-0.0959],[3948,-0.0958],[3957,0.3638],[3966,1.4925],[3984,1.8075],[3986,1.8074],[3996,1.301],[4007,0.2297],[4028,1.8075],[4030,1.8074],[4039,1.3988],[4049,0.3784],[4068,-0.0905],[4069,-0.0904],[4071,-0.0959],[4072,-0.0958],[4081,0.3638],[4090,1.4925],[4108,1.8075],[4110,1.8074],[4120,1.301],[4131,0.2297],[4153,1.8075],[4155,1.8074],[4164,1.3988],[4174,0.3784],[4193,-0.0905],[4194,-0.0904],[4196,-0.0959],[4197,-0.0958],[4206,0.3638],[4215,1.4925],[4233,1.8075],[4235,1.8074],[4245,1.301],[4256,0.2297],[4279,1.8075],[4281,1.8074],[4290,1.3988],[4300,0.3784],[4319,-0.0905],[4320,-0.0904],[4322,-0.0959],[4323,-0.0958],[4332,0.3638],[4341,1.4925],[4360,0.094],[4364,0.0091],[4373,-0.0959],[4382,0.3637],[4391,1.4929],[4409,1.8075],[4422,1.2003],[4436,0]]
       }
       middle_metacarpal_r_ctrl.Scale.X {type=MovieSceneScriptingFloatChannel, default=1, const [1,12]}
       middle_metacarpal_r_ctrl.Scale.Y {type=MovieSceneScriptingFloatChannel, default=1, const [1,12]}
       middle_metacarpal_r_ctrl.Scale.Z {type=MovieSceneScriptingFloatChannel, default=1, const [1,12]}
       middle_01_r_ctrl.Location.X {type=MovieSceneScriptingFloatChannel, default=0, thinned_from=25, keys [[1790,0],[2511,0],[2513,-0.0005],[2519,0],[4436,0]]}
       middle_01_r_ctrl.Location.Y {
        type=MovieSceneScriptingFloatChannel
        default=0
        thinned_from=151
        keys [[1790,0],[2512,0],[2513,0.0001],[2768,0],[2938,-0.0002],[2986,-0.0019],[3099,-0.0019],[3185,-0.0014],[3216,0],[4436,0]]
       }
       middle_01_r_ctrl.Location.Z {
        type=MovieSceneScriptingFloatChannel
        default=0
        thinned_from=241
        keys [[1790,0],[2767,0],[2776,0.0421],[2781,0.1718],[2786,0.2114],[2938,0.2046],[2986,-0.0306],[3099,-0.0306],[3185,-0.0501],[3222,0],[4436,0]]
       }
       middle_01_r_ctrl.Rotation.X {
        type=MovieSceneScriptingFloatChannel
        default=-0.2841
        thinned_from=144
        keys [[1790,-0.2841],[1827,-0.3197],[1830,-0.3198],[1866,-0.2835],[1867,-0.2835],[1909,-0.3198],[1910,-0.3197],[1946,-0.2835],[1948,-0.2835],[1989,-0.3198],[1990,-0.3197],[2026,-0.2835],[2027,-0.2835],[2067,-0.3197],[2070,-0.3198],[2106,-0.2835],[2108,-0.2835],[2147,-0.3197],[2150,-0.3198],[2186,-0.2835],[2187,-0.2835],[2229,-0.3198],[2230,-0.3197],[2266,-0.2835],[2309,-0.3198],[2310,-0.3197],[2346,-0.2835],[2348,-0.2835],[2387,-0.3197],[2390,-0.3198],[2426,-0.2835],[2427,-0.2835],[2466,-0.3196],[2484,0.4866],[2502,0.1495],[2513,0.8296],[2515,0.8247],[2520,1.9282],[2522,1.1608],[2524,1.6761],[2526,1.1512],[2532,0.9528],[2763,3.4433],[2776,0.7835],[2781,-1.3389],[2799,-4.163],[2938,-4.1429],[2986,-10.3427],[3099,-10.3427],[3185,-3.6201],[3221,-0.2835],[3262,-0.3197],[3265,-0.3198],[3301,-0.2835],[3303,-0.2835],[3344,-0.3198],[3345,-0.3197],[3381,-0.2835],[3383,-0.2835],[3424,-0.3198],[3425,-0.3197],[3461,-0.2835],[3502,-0.3197],[3505,-0.3198],[3541,-0.2835],[3543,-0.2835],[3582,-0.3197],[3585,-0.3198],[3621,-0.2835],[3664,-0.3198],[3665,-0.3197],[3701,-0.2835],[3703,-0.2835],[3744,-0.3198],[3745,-0.3197],[3781,-0.2835],[3783,-0.2835],[3822,-0.3197],[3825,-0.3198],[3861,-0.2835],[3907,-0.2835],[3946,-0.3197],[3949,-0.3198],[3985,-0.2835],[4031,-0.2835],[4070,-0.3197],[4073,-0.3198],[4109,-0.2835],[4156,-0.2835],[4195,-0.3197],[4198,-0.3198],[4234,-0.2835],[4282,-0.2835],[4321,-0.3197],[4324,-0.3198],[4360,-0.3103],[4364,-0.3132],[4371,-0.3197],[4372,-0.3197],[4410,-0.2835],[4412,-0.2835],[4435,-0.316],[4436,0]]
       }
       middle_01_r_ctrl.Rotation.Y {
        type=MovieSceneScriptingFloatChannel
        default=5.3913
        thinned_from=98
        keys [[1790,5.3913],[1830,5.9021],[1866,5.3855],[1867,5.3856],[1910,5.902],[1946,5.3855],[1947,5.3856],[1990,5.902],[2026,5.3855],[2027,5.3856],[2070,5.9021],[2106,5.3855],[2107,5.3856],[2150,5.9021],[2186,5.3855],[2187,5.3856],[2230,5.902],[2266,5.3855],[2267,5.3856],[2310,5.902],[2346,5.3855],[2347,5.3856],[2390,5.9021],[2426,5.3855],[2427,5.3856],[2466,5.9009],[2470,5.5801],[2502,0.6949],[2512,-2.3425],[2515,-2.3426],[2522,1.2326],[2524,0.2705],[2527,1.6247],[2531,1.777],[2757,0.9891],[2763,-3.7399],[2772,0.173],[2782,6.2711],[2803,-1.1095],[2938,-1.1093],[2986,-0.429],[3099,-0.429],[3185,-0.9246],[3221,5.3855],[3222,5.3856],[3265,5.9021],[3301,5.3855],[3302,5.3856],[3345,5.902],[3381,5.3855],[3382,5.3856],[3425,5.902],[3461,5.3855],[3462,5.3856],[3505,5.9021],[3541,5.3855],[3542,5.3856],[3585,5.9021],[3621,5.3855],[3622,5.3856],[3665,5.902],[3701,5.3855],[3702,5.3856],[3745,5.902],[3781,5.3855],[3782,5.3856],[3825,5.9021],[3861,5.3855],[3906,5.3856],[3949,5.9021],[3985,5.3855],[4030,5.3856],[4073,5.9021],[4109,5.3855],[4155,5.3856],[4198,5.9021],[4234,5.3855],[4281,5.3856],[4324,5.9021],[4360,5.7409],[4410,5.3855],[4411,5.3856],[4435,5.8498],[4436,0]]
       }
       middle_01_r_ctrl.Rotation.Z {
        type=MovieSceneScriptingFloatChannel
        default=1.0937
        thinned_from=78
        keys [[1790,1.0937],[1829,8.2227],[1830,8.2227],[1867,1.011],[1910,8.2251],[1947,1.011],[1990,8.2251],[2027,1.011],[2069,8.2227],[2070,8.2227],[2107,1.011],[2149,8.2227],[2187,1.011],[2230,8.2251],[2268,1.011],[2310,8.2251],[2347,1.011],[2389,8.2227],[2390,8.2227],[2427,1.011],[2466,8.2092],[2468,8.1407],[2496,9.7616],[2504,-1.72],[2508,-41.346],[2513,-56.6068],[2515,-56.6062],[2518,-35.24],[2519,7.7804],[2520,29.7784],[2521,30.167],[2523,29.7483],[2531,30.2137],[2765,27.984],[2785,-11.4966],[2927,-11.0471],[2938,-11.0307],[2986,67.7075],[3099,67.7075],[3185,50.5635],[3222,1.011],[3264,8.2227],[3302,1.011],[3346,8.2251],[3382,1.011],[3426,8.2251],[3462,1.011],[3504,8.2227],[3542,1.011],[3584,8.2227],[3585,8.2227],[3622,1.011],[3666,8.2251],[3702,1.011],[3746,8.2251],[3782,1.011],[3824,8.2227],[3862,1.011],[3906,1.011],[3948,8.2227],[3986,1.011],[4030,1.011],[4072,8.2227],[4110,1.011],[4155,1.011],[4197,8.2227],[4235,1.011],[4281,1.011],[4323,8.2227],[4360,6.1201],[4364,6.5948],[4373,8.2227],[4411,1.011],[4435,7.4949],[4436,0]]
       }
       middle_01_r_ctrl.Scale.X {type=MovieSceneScriptingFloatChannel, default=1, const [1,12]}
       middle_01_r_ctrl.Scale.Y {type=MovieSceneScriptingFloatChannel, default=1, const [1,12]}
       middle_01_r_ctrl.Scale.Z {type=MovieSceneScriptingFloatChannel, default=1, const [1,12]}
       middle_02_r_ctrl.Location.X {
        type=MovieSceneScriptingFloatChannel
        default=0.0007
        thinned_from=364
        keys [[1790,0.0007],[2467,0.0007],[2484,0.0005],[2512,0.0008],[2513,0.0005],[2515,0.0008],[4435,0.0007],[4436,0]]
       }
       middle_02_r_ctrl.Location.Y {
        type=MovieSceneScriptingFloatChannel
        default=0.0008
        thinned_from=363
        keys [[1790,0.0008],[2467,0.0008],[2484,0.0006],[2503,0.0006],[2512,0.0008],[4435,0.0008],[4436,0]]
       }
       middle_02_r_ctrl.Location.Z {
        type=MovieSceneScriptingFloatChannel
        default=0.0459
        thinned_from=841
        keys [[1790,0.0459],[2471,0.0459],[2496,0.0317],[2503,0.0317],[2505,0.0369],[2509,0.0421],[2510,0.0421],[2513,0.0473],[2521,0.0475],[2769,0.0476],[2790,0.0459],[4435,0.0459],[4436,0]]
       }
       middle_02_r_ctrl.Rotation.X {
        type=MovieSceneScriptingFloatChannel
        default=1.9259
        thinned_from=384
        keys [[1790,1.9259],[1809,2.1302],[1810,2.1297],[1814,2.1455],[1815,2.1433],[1818,2.1521],[1821,2.1564],[1824,2.1683],[1826,2.1608],[1830,2.1608],[1850,1.9326],[1851,1.9339],[1855,1.926],[1856,1.9292],[1858,1.9188],[1860,1.9252],[1864,1.9148],[1867,1.9148],[1889,2.1302],[1890,2.1297],[1894,2.1455],[1895,2.1433],[1898,2.1521],[1901,2.1564],[1904,2.1683],[1906,2.1608],[1912,2.1638],[1930,1.9326],[1931,1.9339],[1935,1.926],[1936,1.9292],[1938,1.9188],[1940,1.9252],[1947,1.9148],[1969,2.1302],[1970,2.1297],[1974,2.1455],[1975,2.1433],[1978,2.1521],[1981,2.1564],[1984,2.1683],[1986,2.1608],[1992,2.1638],[2015,1.9262],[2027,1.9148],[2049,2.1302],[2050,2.1297],[2054,2.1455],[2055,2.1433],[2068,2.164],[2072,2.1654],[2090,1.9326],[2091,1.9339],[2095,1.926],[2096,1.9292],[2098,1.9188],[2100,1.9252],[2107,1.9148],[2129,2.1302],[2130,2.1297],[2134,2.1455],[2135,2.1433],[2138,2.1521],[2141,2.1564],[2144,2.1683],[2146,2.1608],[2150,2.1608],[2170,1.9326],[2171,1.9339],[2175,1.926],[2176,1.9292],[2178,1.9188],[2180,1.9252],[2187,1.9148],[2209,2.1302],[2210,2.1297],[2214,2.1455],[2215,2.1433],[2221,2.1564],[2224,2.1683],[2226,2.1608],[2232,2.1638],[2250,1.9326],[2251,1.9339],[2255,1.926],[2256,1.9292],[2258,1.9188],[2260,1.9252],[2267,1.9148],[2289,2.1302],[2290,2.1297],[2294,2.1455],[2295,2.1433],[2304,2.1683],[2306,2.1608],[2312,2.1638],[2335,1.9262],[2347,1.9148],[2369,2.1302],[2370,2.1297],[2374,2.1455],[2375,2.1433],[2392,2.1654],[2410,1.9326],[2411,1.9339],[2415,1.926],[2416,1.9292],[2418,1.9188],[2420,1.9252],[2427,1.9148],[2449,2.1302],[2450,2.1297],[2454,2.1455],[2455,2.1433],[2458,2.1521],[2461,2.1564],[2464,2.1683],[2470,2.0295],[2496,-2.3534],[2504,-2.6366],[2513,-3.5419],[2515,-3.5325],[2519,-2.3466],[2522,-3.4224],[2524,-2.7272],[2528,-3.6654],[2532,-3.7052],[2757,-3.2247],[2760,-1.3045],[2763,-0.3623],[2767,-0.0132],[2781,1.8519],[2800,0.8404],[2927,0.7128],[2938,0.7112],[2986,0.5609],[3099,0.5609],[3185,-0.7782],[3215,1.9252],[3220,1.9148],[3222,1.9148],[3244,2.1302],[3245,2.1297],[3249,2.1455],[3250,2.1433],[3259,2.1683],[3263,2.1608],[3265,2.1608],[3285,1.9326],[3286,1.9339],[3290,1.926],[3291,1.9292],[3293,1.9188],[3295,1.9252],[3300,1.9148],[3302,1.9148],[3324,2.1302],[3325,2.1297],[3329,2.1455],[3330,2.1433],[3339,2.1683],[3341,2.1608],[3347,2.1637],[3365,1.9326],[3366,1.9339],[3370,1.926],[3371,1.9292],[3373,1.9188],[3375,1.9252],[3379,1.9148],[3382,1.9148],[3404,2.1302],[3405,2.1297],[3409,2.1455],[3410,2.1433],[3416,2.1564],[3419,2.1683],[3422,2.1608],[3427,2.1637],[3450,1.9262],[3462,1.9148],[3484,2.1302],[3485,2.1297],[3489,2.1455],[3490,2.1433],[3503,2.164],[3507,2.1654],[3525,1.9326],[3526,1.9339],[3530,1.926],[3531,1.9292],[3533,1.9188],[3535,1.9252],[3541,1.9148],[3564,2.1302],[3565,2.1297],[3569,2.1455],[3570,2.1433],[3579,2.1683],[3581,2.1608],[3605,1.9326],[3606,1.9339],[3610,1.926],[3611,1.9292],[3613,1.9188],[3615,1.9252],[3620,1.9148],[3622,1.9148],[3644,2.1302],[3645,2.1297],[3649,2.1455],[3650,2.1433],[3659,2.1683],[3662,2.1608],[3667,2.1637],[3685,1.9326],[3686,1.9339],[3690,1.926],[3691,1.9292],[3693,1.9188],[3695,1.9252],[3702,1.9148],[3724,2.1302],[3725,2.1297],[3729,2.1455],[3730,2.1433],[3739,2.1683],[3741,2.1608],[3747,2.1637],[3770,1.9262],[3782,1.9148],[3804,2.1302],[3805,2.1297],[3809,2.1455],[3810,2.1433],[3823,2.164],[3827,2.1654],[3845,1.9326],[3846,1.9339],[3850,1.926],[3851,1.9292],[3853,1.9188],[3855,1.9252],[3860,1.9148],[3863,1.9148],[3884,2.1302],[3885,2.1297],[3894,1.9262],[3906,1.9148],[3928,2.1302],[3929,2.1297],[3933,2.1455],[3934,2.1433],[3947,2.164],[3951,2.1654],[3969,1.9326],[3970,1.9339],[3974,1.926],[3975,1.9292],[3977,1.9188],[3979,1.9252],[3984,1.9148],[3987,1.9148],[4008,2.1302],[4009,2.1297],[4018,1.9262],[4030,1.9148],[4052,2.1302],[4053,2.1297],[4057,2.1455],[4058,2.1433],[4071,2.164],[4075,2.1654],[4093,1.9326],[4094,1.9339],[4098,1.926],[4099,1.9292],[4101,1.9188],[4103,1.9252],[4108,1.9148],[4111,1.9148],[4132,2.1302],[4133,2.1297],[4143,1.9262],[4155,1.9148],[4177,2.1302],[4178,2.1297],[4182,2.1455],[4183,2.1433],[4196,2.164],[4200,2.1654],[4218,1.9326],[4219,1.9339],[4223,1.926],[4224,1.9292],[4226,1.9188],[4228,1.9252],[4233,1.9148],[4236,1.9148],[4257,2.1302],[4258,2.1297],[4269,1.9262],[4281,1.9148],[4303,2.1302],[4304,2.1297],[4308,2.1455],[4309,2.1433],[4322,2.164],[4326,2.1654],[4344,1.9326],[4345,1.9339],[4349,1.926],[4350,1.9292],[4352,1.9188],[4354,1.9252],[4360,2.1455],[4365,2.1433],[4373,2.1608],[4394,1.9326],[4395,1.9339],[4399,1.926],[4400,1.9292],[4402,1.9188],[4404,1.9252],[4409,1.9148],[4411,1.9148],[4433,2.1302],[4435,2.1331],[4436,0]]
       }
       middle_02_r_ctrl.Rotation.Y {
        type=MovieSceneScriptingFloatChannel
        default=-2.7903
        thinned_from=283
        keys [[1790,-2.7903],[1818,-3.1333],[1821,-3.139],[1824,-3.1371],[1830,-3.1447],[1856,-2.7939],[1858,-2.7944],[1862,-2.787],[1869,-2.7883],[1898,-3.1333],[1901,-3.139],[1904,-3.1371],[1910,-3.1447],[1936,-2.7939],[1938,-2.7944],[1942,-2.787],[1949,-2.7883],[1978,-3.1333],[1981,-3.139],[1984,-3.1371],[1990,-3.1447],[2019,-2.7909],[2029,-2.7883],[2065,-3.1422],[2070,-3.1414],[2096,-2.7939],[2098,-2.7944],[2102,-2.787],[2109,-2.7883],[2138,-3.1333],[2141,-3.139],[2144,-3.1371],[2150,-3.1447],[2176,-2.7939],[2178,-2.7944],[2182,-2.787],[2189,-2.7883],[2218,-3.1333],[2221,-3.139],[2224,-3.1371],[2230,-3.1447],[2256,-2.7939],[2258,-2.7944],[2262,-2.787],[2269,-2.7883],[2302,-3.139],[2304,-3.1371],[2310,-3.1447],[2339,-2.7909],[2349,-2.7883],[2385,-3.1422],[2390,-3.1414],[2416,-2.7939],[2418,-2.7944],[2422,-2.787],[2429,-2.7883],[2461,-3.139],[2464,-3.1371],[2466,-3.1446],[2502,0.7573],[2513,0.9105],[2515,0.9079],[2521,1.1207],[2524,1.0485],[2529,1.1327],[2531,1.1305],[2767,0.3228],[2781,-2.6756],[2800,-1.1474],[2927,-1.0428],[2938,-1.0408],[2986,0.1896],[3099,0.1896],[3185,2.0283],[3217,-2.787],[3224,-2.7883],[3257,-3.139],[3259,-3.1371],[3265,-3.1447],[3291,-2.7939],[3293,-2.7944],[3297,-2.787],[3304,-2.7883],[3337,-3.139],[3339,-3.1371],[3344,-3.1447],[3371,-2.7939],[3373,-2.7944],[3377,-2.787],[3384,-2.7883],[3417,-3.139],[3419,-3.1371],[3424,-3.1447],[3454,-2.7909],[3464,-2.7883],[3500,-3.1422],[3505,-3.1414],[3531,-2.7939],[3533,-2.7944],[3537,-2.787],[3544,-2.7883],[3577,-3.139],[3579,-3.1371],[3585,-3.1447],[3611,-2.7939],[3613,-2.7944],[3617,-2.787],[3624,-2.7883],[3657,-3.139],[3659,-3.1371],[3665,-3.1447],[3691,-2.7939],[3693,-2.7944],[3697,-2.787],[3704,-2.7883],[3736,-3.139],[3739,-3.1371],[3745,-3.1447],[3774,-2.7909],[3784,-2.7883],[3820,-3.1422],[3825,-3.1414],[3851,-2.7939],[3853,-2.7944],[3857,-2.787],[3908,-2.7883],[3944,-3.1422],[3949,-3.1414],[3975,-2.7939],[3977,-2.7944],[3981,-2.787],[4032,-2.7883],[4068,-3.1422],[4073,-3.1414],[4099,-2.7939],[4101,-2.7944],[4105,-2.787],[4157,-2.7883],[4193,-3.1422],[4198,-3.1414],[4224,-2.7939],[4226,-2.7944],[4230,-2.787],[4283,-2.7883],[4319,-3.1422],[4324,-3.1414],[4350,-2.7939],[4352,-2.7944],[4356,-2.787],[4360,-3.0872],[4364,-3.1171],[4371,-3.1447],[4373,-3.1447],[4400,-2.7939],[4402,-2.7944],[4406,-2.787],[4413,-2.7883],[4435,-3.1078],[4436,0]]
       }
       middle_02_r_ctrl.Rotation.Z {
        type=MovieSceneScriptingFloatChannel
        default=14.0985
        thinned_from=157
        keys [[1790,14.0985],[1818,23.0124],[1821,23.1561],[1826,23.2998],[1860,14.0378],[1864,14.0378],[1898,23.0124],[1901,23.1561],[1902,23.1561],[1906,23.2998],[1940,14.0378],[1944,14.0378],[1978,23.0124],[1981,23.1561],[1982,23.1561],[1986,23.2998],[2022,14.0378],[2025,14.0378],[2064,23.2382],[2067,23.2382],[2100,14.0378],[2104,14.0378],[2138,23.0124],[2141,23.1561],[2146,23.2998],[2180,14.0378],[2184,14.0378],[2221,23.1561],[2226,23.2998],[2260,14.0378],[2264,14.0378],[2301,23.1561],[2306,23.2998],[2342,14.0378],[2345,14.0378],[2384,23.2382],[2387,23.2382],[2420,14.0378],[2424,14.0378],[2458,23.0124],[2461,23.1561],[2466,23.2995],[2469,22.9962],[2470,22.997],[2496,-6.5004],[2504,-9.8824],[2513,-14.0926],[2520,7.3096],[2521,-5.5752],[2522,-7.8823],[2524,2.2779],[2527,-10.6318],[2531,-12.0009],[2759,13.6393],[2764,37.6258],[2794,-14.254],[2938,-14.1206],[2986,54.9787],[3099,54.9787],[3185,36.736],[3215,14.0379],[3219,14.0378],[3262,23.2998],[3295,14.0378],[3299,14.0378],[3341,23.2998],[3375,14.0379],[3379,14.0378],[3416,23.1561],[3422,23.2998],[3457,14.0378],[3460,14.0378],[3499,23.2382],[3502,23.2382],[3535,14.0378],[3539,14.0378],[3581,23.2998],[3615,14.0379],[3619,14.0378],[3662,23.2998],[3695,14.0379],[3699,14.0378],[3736,23.1561],[3742,23.2998],[3777,14.0378],[3780,14.0378],[3819,23.2382],[3822,23.2382],[3855,14.0378],[3904,14.0378],[3943,23.2382],[3946,23.2382],[3979,14.0378],[4028,14.0378],[4067,23.2382],[4070,23.2382],[4103,14.0378],[4153,14.0378],[4192,23.2382],[4195,23.2382],[4228,14.0378],[4279,14.0378],[4318,23.2382],[4321,23.2382],[4354,14.0378],[4356,14.0379],[4360,21.4636],[4364,21.2906],[4404,14.0378],[4408,14.0378],[4435,22.3489],[4436,0]]
       }
       middle_02_r_ctrl.Scale.X {type=MovieSceneScriptingFloatChannel, default=1, const [1,12]}
       middle_02_r_ctrl.Scale.Y {type=MovieSceneScriptingFloatChannel, default=1, const [1,12]}
       middle_02_r_ctrl.Scale.Z {type=MovieSceneScriptingFloatChannel, default=1, const [1,12]}
       middle_03_r_ctrl.Location.X {type=MovieSceneScriptingFloatChannel, default=0, thinned_from=451, keys [[1790,0],[2512,0],[2513,-0.0001],[2518,0],[4435,0],[4436,-0.0003]]}
       middle_03_r_ctrl.Location.Y {type=MovieSceneScriptingFloatChannel, default=0, thinned_from=849, keys [[1790,0],[4436,0.0001]]}
       middle_03_r_ctrl.Location.Z {type=MovieSceneScriptingFloatChannel, default=0, thinned_from=101, keys [[1790,0],[4436,0]]}
       middle_03_r_ctrl.Rotation.X {
        type=MovieSceneScriptingFloatChannel
        default=-0.9514
        thinned_from=455
        keys [[1790,-0.9514],[1812,-1.0566],[1813,-1.0562],[1818,-1.0635],[1819,-1.0628],[1821,-1.0664],[1823,-1.0649],[1827,-1.0693],[1830,-1.0693],[1851,-0.9548],[1852,-0.9552],[1856,-0.9513],[1858,-0.9524],[1860,-0.9498],[1869,-0.9481],[1892,-1.0566],[1893,-1.0562],[1898,-1.0635],[1899,-1.0628],[1901,-1.0664],[1903,-1.0649],[1906,-1.0693],[1912,-1.0694],[1931,-0.9548],[1932,-0.9552],[1937,-0.9513],[1938,-0.9524],[1941,-0.9498],[1947,-0.9498],[1968,-1.0534],[1970,-1.0517],[1975,-1.0612],[1978,-1.0635],[1979,-1.0628],[1981,-1.0664],[1983,-1.0649],[1987,-1.0693],[1992,-1.0694],[2011,-0.9548],[2012,-0.9552],[2016,-0.9513],[2018,-0.9524],[2027,-0.9498],[2048,-1.0534],[2050,-1.0517],[2055,-1.0612],[2058,-1.0635],[2059,-1.0628],[2061,-1.0664],[2063,-1.0649],[2068,-1.0693],[2072,-1.0694],[2093,-0.9531],[2095,-0.9557],[2098,-0.9498],[2109,-0.9498],[2132,-1.0566],[2133,-1.0562],[2138,-1.0635],[2139,-1.0628],[2141,-1.0664],[2144,-1.0649],[2150,-1.0693],[2171,-0.9548],[2172,-0.9552],[2177,-0.9513],[2178,-0.9524],[2180,-0.9498],[2187,-0.9481],[2212,-1.0566],[2213,-1.0562],[2218,-1.0635],[2219,-1.0628],[2221,-1.0664],[2223,-1.0649],[2226,-1.0693],[2232,-1.0694],[2251,-0.9548],[2252,-0.9552],[2256,-0.9513],[2258,-0.9524],[2261,-0.9498],[2267,-0.9498],[2288,-1.0534],[2290,-1.0517],[2295,-1.0612],[2298,-1.0635],[2299,-1.0628],[2301,-1.0664],[2303,-1.0649],[2307,-1.0693],[2312,-1.0694],[2331,-0.9548],[2332,-0.9552],[2336,-0.9513],[2338,-0.9524],[2342,-0.9498],[2348,-0.9498],[2368,-1.0534],[2370,-1.0517],[2375,-1.0612],[2378,-1.0635],[2379,-1.0628],[2381,-1.0664],[2383,-1.0649],[2386,-1.0693],[2392,-1.0694],[2412,-0.9531],[2414,-0.9531],[2415,-0.9557],[2418,-0.9498],[2427,-0.9498],[2452,-1.0566],[2453,-1.0562],[2458,-1.0635],[2459,-1.0628],[2461,-1.0664],[2463,-1.0649],[2466,-1.0693],[2484,-1.2107],[2502,-1.3666],[2507,-1.0292],[2513,-0.8214],[2515,-0.8234],[2519,-0.4604],[2522,-0.8493],[2524,-0.6224],[2526,-0.856],[2528,-0.934],[2532,-0.9429],[2758,-0.5875],[2760,-0.139],[2763,0.1801],[2775,-0.5882],[2780,-0.9131],[2802,-0.3322],[2927,-0.3339],[2938,-0.3346],[2986,0.0563],[3099,0.0563],[3185,0.5433],[3215,-0.9498],[3221,-0.9498],[3247,-1.0566],[3248,-1.0562],[3253,-1.0635],[3254,-1.0628],[3256,-1.0664],[3259,-1.0649],[3262,-1.0693],[3265,-1.0693],[3286,-0.9548],[3287,-0.9552],[3292,-0.9513],[3293,-0.9524],[3296,-0.9498],[3302,-0.9481],[3327,-1.0566],[3328,-1.0562],[3333,-1.0635],[3334,-1.0628],[3336,-1.0664],[3338,-1.0649],[3341,-1.0693],[3347,-1.0694],[3366,-0.9548],[3367,-0.9552],[3371,-0.9513],[3373,-0.9524],[3378,-0.9498],[3382,-0.9498],[3403,-1.0534],[3405,-1.0517],[3410,-1.0612],[3413,-1.0635],[3414,-1.0628],[3416,-1.0664],[3419,-1.0649],[3422,-1.0693],[3427,-1.0694],[3446,-0.9548],[3447,-0.9552],[3452,-0.9513],[3453,-0.9524],[3461,-0.9498],[3483,-1.0534],[3485,-1.0517],[3490,-1.0612],[3493,-1.0635],[3494,-1.0628],[3496,-1.0664],[3498,-1.0649],[3502,-1.0693],[3507,-1.0694],[3527,-0.9531],[3530,-0.9557],[3533,-0.9498],[3543,-0.9498],[3567,-1.0566],[3568,-1.0562],[3573,-1.0635],[3574,-1.0628],[3576,-1.0664],[3579,-1.0649],[3581,-1.0693],[3606,-0.9548],[3607,-0.9552],[3611,-0.9513],[3613,-0.9524],[3622,-0.9481],[3647,-1.0566],[3648,-1.0562],[3653,-1.0635],[3654,-1.0628],[3656,-1.0664],[3659,-1.0649],[3662,-1.0693],[3667,-1.0694],[3686,-0.9548],[3687,-0.9552],[3692,-0.9513],[3693,-0.9524],[3696,-0.9498],[3700,-0.9498],[3723,-1.0534],[3725,-1.0517],[3730,-1.0612],[3733,-1.0635],[3734,-1.0628],[3736,-1.0664],[3739,-1.0649],[3742,-1.0693],[3747,-1.0694],[3766,-0.9548],[3767,-0.9552],[3772,-0.9513],[3773,-0.9524],[3776,-0.9498],[3781,-0.9498],[3803,-1.0534],[3805,-1.0517],[3810,-1.0612],[3813,-1.0635],[3814,-1.0628],[3816,-1.0664],[3819,-1.0649],[3822,-1.0693],[3827,-1.0694],[3848,-0.9531],[3850,-0.9557],[3853,-0.9498],[3863,-0.9498],[3887,-1.0566],[3888,-1.0562],[3890,-0.9548],[3891,-0.9552],[3896,-0.9513],[3897,-0.9524],[3900,-0.9498],[3905,-0.9498],[3927,-1.0534],[3929,-1.0517],[3934,-1.0612],[3937,-1.0635],[3938,-1.0628],[3940,-1.0664],[3943,-1.0649],[3946,-1.0693],[3951,-1.0694],[3972,-0.9531],[3974,-0.9557],[3977,-0.9498],[3987,-0.9498],[4011,-1.0566],[4012,-1.0562],[4014,-0.9548],[4015,-0.9552],[4020,-0.9513],[4021,-0.9524],[4024,-0.9498],[4029,-0.9498],[4051,-1.0534],[4053,-1.0517],[4058,-1.0612],[4061,-1.0635],[4062,-1.0628],[4064,-1.0664],[4067,-1.0649],[4070,-1.0693],[4075,-1.0694],[4096,-0.9531],[4098,-0.9557],[4101,-0.9498],[4111,-0.9498],[4135,-1.0566],[4136,-1.0562],[4139,-0.9548],[4140,-0.9552],[4145,-0.9513],[4146,-0.9524],[4149,-0.9498],[4154,-0.9498],[4176,-1.0534],[4178,-1.0517],[4183,-1.0612],[4186,-1.0635],[4187,-1.0628],[4189,-1.0664],[4192,-1.0649],[4195,-1.0693],[4200,-1.0694],[4221,-0.9531],[4223,-0.9557],[4226,-0.9498],[4236,-0.9498],[4260,-1.0566],[4261,-1.0562],[4265,-0.9548],[4266,-0.9552],[4271,-0.9513],[4272,-0.9524],[4275,-0.9498],[4280,-0.9498],[4302,-1.0534],[4304,-1.0517],[4309,-1.0612],[4312,-1.0635],[4313,-1.0628],[4315,-1.0664],[4318,-1.0649],[4321,-1.0693],[4326,-1.0694],[4347,-0.9531],[4349,-0.9557],[4352,-0.9498],[4357,-0.9498],[4360,-1.057],[4364,-1.0625],[4373,-1.0693],[4395,-0.9548],[4396,-0.9552],[4400,-0.9513],[4402,-0.9524],[4405,-0.9498],[4411,-0.9481],[4435,-1.0561],[4436,0.0012]]
       }
       middle_03_r_ctrl.Rotation.Y {
        type=MovieSceneScriptingFloatChannel
        default=-1.8667
        thinned_from=360
        keys [[1790,-1.8667],[1818,-2.0943],[1826,-2.1016],[1830,-2.1016],[1856,-1.8671],[1858,-1.8672],[1860,-1.8635],[1867,-1.8632],[1898,-2.0943],[1908,-2.1016],[1936,-1.8671],[1938,-1.8672],[1940,-1.8635],[1947,-1.8635],[1975,-2.0907],[1988,-2.1016],[2017,-1.8671],[2029,-1.8635],[2055,-2.0907],[2067,-2.1016],[2093,-1.8718],[2095,-1.8722],[2097,-1.8635],[2109,-1.8635],[2138,-2.0943],[2146,-2.1016],[2150,-2.1016],[2176,-1.8671],[2178,-1.8672],[2180,-1.8635],[2187,-1.8632],[2218,-2.0943],[2228,-2.1016],[2257,-1.8671],[2267,-1.8635],[2295,-2.0907],[2308,-2.1016],[2337,-1.8671],[2349,-1.8635],[2375,-2.0907],[2388,-2.1016],[2412,-1.8718],[2415,-1.8722],[2417,-1.8635],[2429,-1.8635],[2458,-2.0943],[2466,-2.1016],[2513,2.0451],[2519,1.6449],[2522,2.2558],[2524,1.8979],[2529,2.4073],[2531,2.4044],[2760,1.1675],[2765,0.5578],[2781,-1.7875],[2800,-0.798],[2927,-0.6591],[2938,-0.657],[2986,0.8054],[3099,0.8054],[3185,1.6225],[3215,-1.8635],[3224,-1.8635],[3253,-2.0943],[3266,-2.1016],[3291,-1.8671],[3301,-1.8632],[3333,-2.0943],[3342,-2.1016],[3371,-1.8671],[3373,-1.8672],[3382,-1.8635],[3410,-2.0907],[3416,-2.098],[3419,-2.0978],[3421,-2.1016],[3451,-1.8671],[3453,-1.8672],[3455,-1.8635],[3464,-1.8635],[3490,-2.0907],[3504,-2.1016],[3527,-1.8718],[3530,-1.8722],[3532,-1.8635],[3543,-1.8635],[3573,-2.0943],[3585,-2.1016],[3611,-1.8671],[3613,-1.8672],[3621,-1.8632],[3653,-2.0943],[3665,-2.1016],[3692,-1.8671],[3704,-1.8635],[3730,-2.0907],[3741,-2.1016],[3745,-2.1016],[3771,-1.8671],[3773,-1.8672],[3775,-1.8635],[3781,-1.8635],[3810,-2.0907],[3821,-2.1016],[3847,-1.8718],[3850,-1.8722],[3852,-1.8635],[3897,-1.8672],[3899,-1.8635],[3905,-1.8635],[3934,-2.0907],[3945,-2.1016],[3971,-1.8718],[3974,-1.8722],[3976,-1.8635],[4021,-1.8672],[4023,-1.8635],[4029,-1.8635],[4058,-2.0907],[4069,-2.1016],[4095,-1.8718],[4098,-1.8722],[4100,-1.8635],[4146,-1.8672],[4148,-1.8635],[4154,-1.8635],[4183,-2.0907],[4194,-2.1016],[4220,-1.8718],[4223,-1.8722],[4225,-1.8635],[4272,-1.8672],[4274,-1.8635],[4280,-1.8635],[4309,-2.0907],[4320,-2.1016],[4346,-1.8718],[4349,-1.8722],[4351,-1.8635],[4357,-1.8635],[4360,-2.0831],[4364,-2.0771],[4400,-1.8671],[4402,-1.8672],[4404,-1.8635],[4410,-1.8632],[4435,-2.0777],[4436,0.0003]]
       }
       middle_03_r_ctrl.Rotation.Z {
        type=MovieSceneScriptingFloatChannel
        default=2.2405
        thinned_from=231
        keys [[1790,2.2405],[1818,10.4288],[1821,10.5622],[1823,10.5622],[1826,10.6956],[1860,2.125],[1866,2.125],[1898,10.4288],[1899,10.4288],[1902,10.5622],[1904,10.5622],[1906,10.6956],[1910,10.6956],[1940,2.125],[1948,2.125],[1978,10.4288],[1979,10.4288],[1981,10.5622],[1983,10.5622],[1988,10.6956],[1990,10.6956],[2016,2.25],[2022,2.125],[2024,2.125],[2058,10.4288],[2059,10.4288],[2061,10.5622],[2064,10.5622],[2066,10.6956],[2067,10.6956],[2093,2.4165],[2094,2.4165],[2098,2.125],[2104,2.125],[2138,10.4288],[2139,10.4288],[2141,10.5622],[2142,10.5622],[2147,10.6956],[2148,10.6956],[2182,2.125],[2186,2.125],[2221,10.5622],[2223,10.5622],[2226,10.6956],[2230,10.6956],[2261,2.125],[2266,2.125],[2302,10.5622],[2303,10.5622],[2308,10.6956],[2309,10.6956],[2336,2.25],[2342,2.125],[2344,2.125],[2378,10.4288],[2379,10.4288],[2381,10.5622],[2383,10.5622],[2386,10.6956],[2387,10.6956],[2413,2.4165],[2414,2.4165],[2418,2.125],[2424,2.125],[2461,10.5622],[2464,10.5622],[2466,10.6952],[2484,-31.9925],[2502,-37.4045],[2513,-32.129],[2515,-32.3331],[2520,9.3859],[2522,3.9432],[2524,7.6412],[2532,2.3992],[2764,20.1858],[2927,13.7202],[2938,13.6134],[2986,36.3424],[3099,36.3424],[3185,33.3438],[3215,2.125],[3218,2.125],[3257,10.5622],[3263,10.6956],[3264,10.6956],[3296,2.125],[3301,2.125],[3337,10.5622],[3342,10.6956],[3376,2.125],[3383,2.125],[3416,10.5622],[3418,10.5622],[3422,10.6956],[3425,10.6956],[3457,2.125],[3459,2.125],[3497,10.5622],[3498,10.5622],[3503,10.6956],[3504,10.6956],[3528,2.4165],[3529,2.4165],[3533,2.125],[3538,2.125],[3573,10.4288],[3577,10.5622],[3582,10.6956],[3616,2.125],[3621,2.125],[3657,10.5622],[3658,10.5622],[3662,10.6956],[3696,2.125],[3703,2.125],[3736,10.5622],[3737,10.5622],[3742,10.6956],[3777,2.125],[3779,2.125],[3813,10.4288],[3816,10.5622],[3818,10.5622],[3823,10.6956],[3824,10.6956],[3848,2.4165],[3849,2.4165],[3853,2.125],[3903,2.125],[3937,10.4288],[3940,10.5622],[3942,10.5622],[3947,10.6956],[3948,10.6956],[3972,2.4165],[3973,2.4165],[3977,2.125],[4027,2.125],[4061,10.4288],[4064,10.5622],[4066,10.5622],[4071,10.6956],[4072,10.6956],[4096,2.4165],[4097,2.4165],[4101,2.125],[4152,2.125],[4186,10.4288],[4189,10.5622],[4191,10.5622],[4196,10.6956],[4197,10.6956],[4221,2.4165],[4222,2.4165],[4226,2.125],[4278,2.125],[4312,10.4288],[4315,10.5622],[4317,10.5622],[4322,10.6956],[4323,10.6956],[4347,2.4165],[4348,2.4165],[4352,2.125],[4357,2.125],[4360,8.8932],[4364,9.833],[4371,10.6956],[4405,2.125],[4410,2.125],[4435,9.8167],[4436,0.0084]]
       }
       middle_03_r_ctrl.Scale.X {type=MovieSceneScriptingFloatChannel, default=1, const [1,12]}
       middle_03_r_ctrl.Scale.Y {type=MovieSceneScriptingFloatChannel, default=1, const [1,12]}
       middle_03_r_ctrl.Scale.Z {type=MovieSceneScriptingFloatChannel, default=1, const [1,12]}
       ring_metacarpal_r_ctrl.Location.X {type=MovieSceneScriptingFloatChannel, default=0, thinned_from=31, keys [[1790,0],[2512,0],[2513,-0.0001],[2521,0],[4436,0]]}
       ring_metacarpal_r_ctrl.Location.Y {type=MovieSceneScriptingFloatChannel, default=0, thinned_from=30, keys [[1790,0],[4436,0]]}
       ring_metacarpal_r_ctrl.Location.Z {type=MovieSceneScriptingFloatChannel, default=0, thinned_from=29, keys [[1790,0],[4436,0]]}
       ring_metacarpal_r_ctrl.Rotation.X {
        type=MovieSceneScriptingFloatChannel
        default=-14.2731
        thinned_from=129
        keys [[1790,-14.2731],[1800,-13.0908],[1830,-11.7532],[1849,-13.8056],[1868,-14.311],[1886,-12.4576],[1905,-11.8029],[1908,-11.7532],[1928,-13.8009],[1948,-14.3295],[1963,-12.7082],[1990,-11.7532],[2009,-13.8129],[2028,-14.3295],[2046,-12.4576],[2065,-11.8029],[2068,-11.7532],[2088,-13.8009],[2108,-14.3295],[2123,-12.7082],[2150,-11.7532],[2169,-13.8056],[2188,-14.311],[2206,-12.4576],[2225,-11.8029],[2228,-11.7532],[2248,-13.8009],[2268,-14.3295],[2283,-12.7082],[2310,-11.7532],[2329,-13.8129],[2348,-14.3295],[2366,-12.4576],[2385,-11.8029],[2389,-11.7532],[2408,-13.8009],[2428,-14.3295],[2443,-12.7082],[2458,-12.031],[2496,-0.256],[2504,0.0053],[2506,-0.0028],[2507,0.0038],[2509,-0.0036],[2513,0.0025],[2520,-1.7833],[2522,-0.3623],[2524,-1.3172],[2531,0.0113],[2761,-3.934],[2764,-4.6397],[2774,-7.6585],[2777,-6.7526],[2781,-2.4114],[2788,0],[3185,0],[3223,-14.3295],[3238,-12.7083],[3265,-11.7532],[3284,-13.8055],[3303,-14.311],[3318,-12.7083],[3343,-11.7532],[3363,-13.8009],[3383,-14.3295],[3398,-12.7083],[3425,-11.7532],[3444,-13.8129],[3463,-14.3295],[3478,-12.7083],[3504,-11.7532],[3523,-13.8009],[3543,-14.3295],[3558,-12.7083],[3585,-11.7532],[3604,-13.8055],[3623,-14.311],[3638,-12.7083],[3663,-11.7532],[3683,-13.8009],[3703,-14.3295],[3718,-12.7083],[3745,-11.7532],[3764,-13.8129],[3783,-14.3295],[3798,-12.7083],[3823,-11.7532],[3843,-13.8009],[3863,-14.3295],[3878,-12.7083],[3907,-14.3295],[3922,-12.7083],[3947,-11.7532],[3967,-13.8009],[3987,-14.3295],[4002,-12.7083],[4031,-14.3295],[4046,-12.7083],[4071,-11.7532],[4091,-13.8009],[4111,-14.3295],[4126,-12.7083],[4156,-14.3295],[4171,-12.7083],[4196,-11.7532],[4216,-13.8009],[4236,-14.3295],[4251,-12.7083],[4282,-14.3295],[4297,-12.7083],[4322,-11.7532],[4342,-13.8009],[4360,-12.1569],[4364,-12.2067],[4393,-13.8055],[4412,-14.311],[4424,-13.091],[4430,-12.4577],[4435,-12.3207],[4436,0]]
       }
       ring_metacarpal_r_ctrl.Rotation.Y {
        type=MovieSceneScriptingFloatChannel
        default=3.9875
        thinned_from=145
        keys [[1790,3.9875],[1830,3.772],[1865,3.9913],[1868,3.9903],[1905,3.7749],[1908,3.772],[1942,3.9896],[1946,3.9925],[1990,3.772],[2022,3.9896],[2026,3.9925],[2065,3.7749],[2068,3.772],[2102,3.9896],[2106,3.9925],[2150,3.772],[2185,3.9913],[2188,3.9903],[2225,3.7749],[2228,3.772],[2262,3.9896],[2266,3.9925],[2310,3.772],[2342,3.9896],[2346,3.9925],[2385,3.7749],[2388,3.772],[2422,3.9896],[2426,3.9925],[2456,3.7927],[2471,3.61],[2496,0.6927],[2512,-0.4005],[2515,-0.4005],[2520,1.3367],[2522,0.9591],[2524,1.2133],[2530,0.8581],[2759,1.4908],[2774,2.7512],[2777,2.3575],[2781,0.8395],[2788,0],[3185,0],[3217,3.9896],[3221,3.9925],[3265,3.772],[3300,3.9913],[3303,3.9903],[3340,3.7749],[3343,3.772],[3377,3.9896],[3381,3.9925],[3425,3.772],[3457,3.9896],[3461,3.9925],[3500,3.7749],[3503,3.772],[3537,3.9896],[3541,3.9925],[3585,3.772],[3620,3.9913],[3623,3.9903],[3660,3.7749],[3663,3.772],[3697,3.9896],[3701,3.9925],[3745,3.772],[3777,3.9896],[3781,3.9925],[3820,3.7749],[3823,3.772],[3857,3.9896],[3905,3.9925],[3944,3.7749],[3947,3.772],[3981,3.9896],[4029,3.9925],[4068,3.7749],[4071,3.772],[4105,3.9896],[4154,3.9925],[4193,3.7749],[4196,3.772],[4230,3.9896],[4280,3.9925],[4319,3.7749],[4322,3.772],[4356,3.9896],[4360,3.8383],[4364,3.8412],[4409,3.9913],[4412,3.9903],[4424,3.8977],[4430,3.8362],[4435,3.8114],[4436,0]]
       }
       ring_metacarpal_r_ctrl.Rotation.Z {
        type=MovieSceneScriptingFloatChannel
        default=-0.4343
        thinned_from=228
        keys [[1790,-0.4343],[1808,-2.1184],[1827,-2.6468],[1828,-2.6457],[1830,-2.661],[1838,-2.1114],[1847,-0.9001],[1865,-0.4018],[1867,-0.4044],[1868,-0.4044],[1886,-2.001],[1905,-2.6307],[1908,-2.661],[1916,-2.3632],[1925,-1.1018],[1942,-0.4199],[1943,-0.4213],[1946,-0.3861],[1956,-0.895],[1966,-2.001],[1987,-2.6468],[1988,-2.6457],[1991,-2.661],[2006,-0.997],[2022,-0.4199],[2023,-0.4213],[2026,-0.3861],[2036,-0.895],[2047,-2.0679],[2069,-2.661],[2077,-2.2509],[2085,-1.1018],[2102,-0.4199],[2103,-0.4213],[2106,-0.3861],[2116,-0.895],[2126,-2.001],[2147,-2.6468],[2148,-2.6457],[2150,-2.661],[2158,-2.1114],[2167,-0.9001],[2185,-0.4018],[2187,-0.4044],[2188,-0.4044],[2208,-2.1184],[2229,-2.661],[2237,-2.2509],[2245,-1.1018],[2262,-0.4199],[2263,-0.4213],[2266,-0.3861],[2276,-0.895],[2286,-2.001],[2307,-2.6468],[2308,-2.6457],[2311,-2.661],[2326,-0.997],[2342,-0.4199],[2343,-0.4213],[2346,-0.3861],[2356,-0.895],[2367,-2.0679],[2389,-2.661],[2397,-2.2509],[2405,-1.1018],[2422,-0.4199],[2423,-0.4213],[2426,-0.3861],[2436,-0.895],[2446,-2.001],[2467,-2.646],[2471,-2.5139],[2502,0.0034],[2507,-0.7163],[2513,-1.1597],[2518,-1.0362],[2521,-0.0951],[2523,-0.2839],[2530,0.0037],[2757,-0.1488],[2759,-0.5332],[2762,-0.9854],[2766,-1.1618],[2774,-1.7235],[2777,-1.4797],[2781,-0.4907],[2788,0],[3185,0],[3217,-0.4199],[3218,-0.4213],[3221,-0.3861],[3231,-0.8948],[3241,-2.0009],[3262,-2.6468],[3263,-2.6457],[3265,-2.661],[3273,-2.1116],[3282,-0.9001],[3300,-0.4017],[3302,-0.4044],[3303,-0.4044],[3323,-2.1183],[3344,-2.661],[3352,-2.251],[3360,-1.102],[3377,-0.4199],[3378,-0.4213],[3381,-0.3861],[3391,-0.8948],[3401,-2.0009],[3422,-2.6468],[3423,-2.6457],[3426,-2.661],[3441,-0.9971],[3457,-0.4199],[3458,-0.4213],[3461,-0.3861],[3471,-0.8948],[3482,-2.0678],[3504,-2.661],[3512,-2.251],[3520,-1.102],[3537,-0.4199],[3538,-0.4213],[3541,-0.3861],[3551,-0.8948],[3561,-2.0009],[3582,-2.6468],[3583,-2.6457],[3586,-2.661],[3603,-0.8379],[3620,-0.4017],[3622,-0.4044],[3623,-0.4044],[3643,-2.1183],[3664,-2.661],[3672,-2.251],[3680,-1.102],[3697,-0.4199],[3698,-0.4213],[3701,-0.3861],[3711,-0.8948],[3721,-2.0009],[3742,-2.6468],[3743,-2.6457],[3746,-2.661],[3761,-0.9972],[3777,-0.4199],[3778,-0.4213],[3781,-0.3861],[3791,-0.8948],[3802,-2.0678],[3824,-2.661],[3832,-2.2511],[3840,-1.102],[3857,-0.4199],[3858,-0.4213],[3861,-0.3861],[3871,-0.8948],[3881,-2.0009],[3901,-0.4199],[3902,-0.4213],[3905,-0.3861],[3915,-0.8948],[3926,-2.0678],[3948,-2.661],[3956,-2.2511],[3964,-1.102],[3981,-0.4199],[3982,-0.4213],[3985,-0.3861],[3995,-0.8948],[4005,-2.0009],[4025,-0.4199],[4026,-0.4213],[4029,-0.3861],[4039,-0.8948],[4050,-2.0678],[4072,-2.661],[4080,-2.2511],[4088,-1.102],[4105,-0.4199],[4106,-0.4213],[4109,-0.3861],[4119,-0.8948],[4129,-2.0009],[4150,-0.4199],[4151,-0.4213],[4154,-0.3861],[4164,-0.8948],[4175,-2.0678],[4197,-2.661],[4205,-2.2511],[4213,-1.102],[4230,-0.4199],[4231,-0.4213],[4234,-0.3861],[4244,-0.8948],[4254,-2.0009],[4276,-0.4199],[4277,-0.4213],[4280,-0.3861],[4290,-0.8948],[4301,-2.0678],[4323,-2.661],[4331,-2.2511],[4339,-1.102],[4356,-0.4199],[4357,-0.4213],[4360,-2.3521],[4364,-2.5058],[4371,-2.6468],[4372,-2.6457],[4375,-2.661],[4392,-0.8379],[4409,-0.4017],[4411,-0.4044],[4412,-0.4044],[4435,-2.255],[4436,0]]
       }
       ring_metacarpal_r_ctrl.Scale.X {type=MovieSceneScriptingFloatChannel, default=1, const [1,12]}
       ring_metacarpal_r_ctrl.Scale.Y {type=MovieSceneScriptingFloatChannel, default=1, const [1,12]}
       ring_metacarpal_r_ctrl.Scale.Z {type=MovieSceneScriptingFloatChannel, default=1, const [1,12]}
       ring_01_r_ctrl.Location.X {
        type=MovieSceneScriptingFloatChannel
        default=0
        thinned_from=22
        keys [[1790,0],[2511,0],[2513,-0.0005],[2519,0],[2768,0],[2796,0.1197],[2938,0.1179],[2986,0.0018],[3099,0.0018],[3185,0.001],[4436,0]]
       }
       ring_01_r_ctrl.Location.Y {
        type=MovieSceneScriptingFloatChannel
        default=0
        thinned_from=269
        keys [[1790,0],[2768,0],[2775,-0.0316],[2782,-0.267],[2796,-0.4717],[2938,-0.4645],[2986,-0.0155],[3099,-0.0155],[3185,-0.0102],[3237,0],[4436,0]]
       }
       ring_01_r_ctrl.Location.Z {
        type=MovieSceneScriptingFloatChannel
        default=0
        thinned_from=616
        keys [[1790,0],[2768,0],[2773,0.012],[2776,0.0619],[2780,0.2115],[2784,0.2623],[2804,0.0539],[2938,0.054],[2986,0.0743],[3099,0.0743],[3185,0.0944],[3227,0],[4436,0]]
       }
       ring_01_r_ctrl.Rotation.X {
        type=MovieSceneScriptingFloatChannel
        default=12.9291
        thinned_from=115
        keys [[1790,12.9291],[1804,14.7609],[1829,15.6815],[1830,15.6811],[1844,13.7396],[1868,12.8937],[1909,15.6815],[1910,15.6811],[1923,13.8873],[1947,12.899],[1989,15.6815],[1990,15.6811],[2003,13.8873],[2027,12.899],[2069,15.6815],[2070,15.6811],[2083,13.8873],[2107,12.899],[2149,15.6815],[2150,15.6811],[2164,13.7396],[2188,12.8937],[2229,15.6815],[2230,15.6811],[2243,13.8873],[2267,12.899],[2309,15.6815],[2310,15.6811],[2323,13.8873],[2347,12.899],[2389,15.6815],[2390,15.6811],[2403,13.8873],[2427,12.899],[2466,15.6438],[2471,14.7728],[2496,1.0484],[2504,0.3959],[2513,-1.3455],[2518,-0.2885],[2520,3.0203],[2522,1.3221],[2524,2.4802],[2530,0.8739],[2761,5.6506],[2764,6.5369],[2775,10.562],[2782,3.5701],[2790,0.5966],[2805,-1.2827],[2938,-1.2813],[2986,0.3177],[3099,0.3177],[3185,1.524],[3222,12.899],[3264,15.6815],[3265,15.6811],[3279,13.7398],[3303,12.8937],[3344,15.6815],[3345,15.6811],[3359,13.7284],[3383,12.899],[3424,15.6815],[3426,15.6811],[3439,13.7284],[3463,12.899],[3504,15.6815],[3505,15.6811],[3518,13.8876],[3542,12.899],[3584,15.6815],[3586,15.6811],[3599,13.7398],[3623,12.8937],[3664,15.6815],[3665,15.6811],[3679,13.7284],[3703,12.899],[3744,15.6815],[3746,15.6811],[3759,13.7284],[3782,12.899],[3824,15.6815],[3838,13.8875],[3863,12.899],[3906,12.899],[3948,15.6815],[3962,13.8875],[3987,12.899],[4030,12.899],[4072,15.6815],[4086,13.8875],[4111,12.899],[4155,12.899],[4197,15.6815],[4211,13.8875],[4236,12.899],[4281,12.899],[4323,15.6815],[4337,13.8875],[4360,14.8367],[4364,15.1461],[4373,15.6815],[4388,13.7398],[4412,12.8937],[4435,15.0675],[4436,0]]
       }
       ring_01_r_ctrl.Rotation.Y {
        type=MovieSceneScriptingFloatChannel
        default=-1.2291
        thinned_from=932
        keys [[1790,-1.2291],[1791,-1.2253],[1793,-1.2239],[1797,-1.2321],[1800,-1.2232],[1802,-1.2351],[1805,-1.2319],[1808,-1.2395],[1811,-1.2261],[1813,-1.2347],[1815,-1.2285],[1817,-1.2339],[1818,-1.2297],[1824,-1.2366],[1826,-1.231],[1828,-1.2376],[1830,-1.2289],[1832,-1.2388],[1833,-1.2329],[1835,-1.2398],[1838,-1.2261],[1840,-1.2398],[1841,-1.2352],[1842,-1.2402],[1844,-1.2325],[1845,-1.2332],[1848,-1.2284],[1849,-1.2301],[1852,-1.2265],[1854,-1.2291],[1860,-1.2256],[1861,-1.2233],[1863,-1.2275],[1864,-1.2265],[1866,-1.2297],[1867,-1.2279],[1872,-1.2303],[1875,-1.2284],[1879,-1.2312],[1880,-1.2258],[1882,-1.2351],[1885,-1.2319],[1888,-1.2395],[1891,-1.2261],[1893,-1.2347],[1895,-1.2285],[1897,-1.2339],[1898,-1.2297],[1904,-1.2366],[1906,-1.231],[1908,-1.2376],[1910,-1.2289],[1912,-1.2254],[1916,-1.2392],[1918,-1.2308],[1919,-1.2317],[1921,-1.2257],[1923,-1.2353],[1925,-1.2246],[1927,-1.2328],[1928,-1.23],[1929,-1.2363],[1932,-1.2247],[1933,-1.2306],[1935,-1.2156],[1937,-1.2377],[1939,-1.2296],[1941,-1.2367],[1943,-1.2261],[1945,-1.2305],[1946,-1.2267],[1949,-1.2303],[1951,-1.2253],[1953,-1.2239],[1957,-1.2321],[1960,-1.2232],[1962,-1.2351],[1965,-1.2319],[1968,-1.2395],[1971,-1.2261],[1973,-1.2347],[1975,-1.2285],[1977,-1.2339],[1978,-1.2297],[1984,-1.2366],[1986,-1.231],[1988,-1.2376],[1990,-1.2289],[1992,-1.2254],[1996,-1.2392],[1998,-1.2307],[1999,-1.2317],[2001,-1.2257],[2003,-1.2353],[2005,-1.2246],[2007,-1.2328],[2008,-1.23],[2009,-1.2363],[2012,-1.2247],[2013,-1.2306],[2015,-1.2156],[2017,-1.2377],[2019,-1.2296],[2021,-1.2367],[2023,-1.2261],[2025,-1.2305],[2026,-1.2267],[2029,-1.2303],[2031,-1.2253],[2033,-1.2239],[2037,-1.2321],[2040,-1.2232],[2042,-1.2351],[2045,-1.2319],[2048,-1.2395],[2051,-1.2261],[2053,-1.2347],[2055,-1.2285],[2057,-1.2339],[2058,-1.2297],[2064,-1.2366],[2066,-1.231],[2068,-1.2376],[2070,-1.2289],[2072,-1.2254],[2076,-1.2392],[2078,-1.2307],[2079,-1.2317],[2081,-1.2257],[2083,-1.2353],[2085,-1.2246],[2087,-1.2328],[2088,-1.2284],[2089,-1.2301],[2092,-1.2265],[2094,-1.2291],[2100,-1.2256],[2101,-1.2233],[2104,-1.2313],[2106,-1.2267],[2109,-1.2303],[2111,-1.2253],[2113,-1.2239],[2117,-1.2321],[2120,-1.2232],[2122,-1.2351],[2125,-1.2319],[2128,-1.2395],[2131,-1.2261],[2133,-1.2347],[2135,-1.2285],[2137,-1.2339],[2138,-1.2297],[2144,-1.2366],[2146,-1.231],[2148,-1.2376],[2150,-1.2289],[2152,-1.2388],[2153,-1.2329],[2155,-1.2397],[2158,-1.2261],[2160,-1.2398],[2161,-1.2352],[2162,-1.2402],[2164,-1.2325],[2165,-1.2332],[2168,-1.2284],[2169,-1.2301],[2172,-1.2265],[2174,-1.2291],[2180,-1.2256],[2181,-1.2233],[2183,-1.2275],[2184,-1.2265],[2186,-1.2297],[2187,-1.2279],[2192,-1.2303],[2195,-1.2284],[2199,-1.2312],[2200,-1.2258],[2202,-1.2351],[2205,-1.2319],[2208,-1.2395],[2211,-1.2261],[2213,-1.2347],[2215,-1.2285],[2217,-1.2339],[2218,-1.2297],[2224,-1.2366],[2226,-1.231],[2228,-1.2376],[2230,-1.2289],[2232,-1.2254],[2236,-1.2392],[2238,-1.2307],[2239,-1.2317],[2241,-1.2257],[2243,-1.2353],[2245,-1.2246],[2247,-1.2328],[2248,-1.23],[2249,-1.2363],[2252,-1.2247],[2253,-1.2306],[2255,-1.2156],[2257,-1.2377],[2259,-1.2296],[2261,-1.2367],[2263,-1.2261],[2265,-1.2305],[2266,-1.2267],[2269,-1.2303],[2271,-1.2253],[2273,-1.2239],[2277,-1.2321],[2280,-1.2232],[2282,-1.2351],[2285,-1.2319],[2288,-1.2395],[2291,-1.2261],[2293,-1.2347],[2295,-1.2285],[2297,-1.2339],[2298,-1.2297],[2304,-1.2366],[2306,-1.231],[2308,-1.2376],[2310,-1.2289],[2312,-1.2254],[2316,-1.2392],[2318,-1.2307],[2319,-1.2317],[2321,-1.2257],[2323,-1.2353],[2325,-1.2246],[2327,-1.2328],[2328,-1.23],[2329,-1.2363],[2332,-1.2247],[2333,-1.2306],[2335,-1.2156],[2337,-1.2377],[2339,-1.2296],[2341,-1.2367],[2343,-1.2261],[2345,-1.2305],[2346,-1.2267],[2349,-1.2303],[2351,-1.2253],[2353,-1.2239],[2357,-1.2321],[2360,-1.2232],[2362,-1.2351],[2365,-1.2319],[2368,-1.2395],[2371,-1.2261],[2373,-1.2347],[2375,-1.2285],[2377,-1.2339],[2378,-1.2297],[2384,-1.2366],[2386,-1.231],[2388,-1.2376],[2390,-1.2289],[2392,-1.2254],[2396,-1.2392],[2398,-1.2307],[2399,-1.2317],[2401,-1.2257],[2403,-1.2353],[2405,-1.2246],[2407,-1.2328],[2408,-1.2284],[2409,-1.2301],[2412,-1.2265],[2414,-1.2291],[2420,-1.2256],[2421,-1.2233],[2424,-1.2313],[2426,-1.2267],[2429,-1.2303],[2431,-1.2253],[2433,-1.2239],[2437,-1.2321],[2440,-1.2232],[2442,-1.2351],[2445,-1.2319],[2448,-1.2395],[2451,-1.2261],[2453,-1.2347],[2455,-1.2285],[2457,-1.2339],[2458,-1.2297],[2464,-1.2366],[2466,-1.2309],[2467,-1.2323],[2496,0.8035],[2504,0.301],[2513,-1.3169],[2518,0.1135],[2519,2.5051],[2520,3.2287],[2522,2.6559],[2524,3.0394],[2530,2.4988],[2763,4.3403],[2777,0.33],[2782,-2.1614],[2792,-5.661],[2802,-6.1628],[2938,-6.1563],[2986,4.3251],[3099,4.3251],[3185,3.7799],[3215,-1.2256],[3216,-1.2233],[3219,-1.2313],[3221,-1.2267],[3224,-1.2303],[3226,-1.2253],[3228,-1.224],[3232,-1.2321],[3235,-1.2232],[3237,-1.2351],[3240,-1.2319],[3243,-1.2395],[3246,-1.2261],[3248,-1.2347],[3250,-1.2285],[3252,-1.2339],[3253,-1.2297],[3259,-1.2366],[3261,-1.231],[3263,-1.2376],[3265,-1.2289],[3267,-1.2388],[3268,-1.2329],[3270,-1.2398],[3273,-1.2261],[3275,-1.2398],[3276,-1.2352],[3277,-1.2402],[3279,-1.2325],[3280,-1.2332],[3283,-1.2284],[3284,-1.2301],[3287,-1.2265],[3289,-1.2291],[3295,-1.2256],[3296,-1.2233],[3298,-1.2275],[3299,-1.2265],[3301,-1.2297],[3302,-1.2279],[3307,-1.2303],[3310,-1.2284],[3314,-1.2312],[3315,-1.2258],[3317,-1.2351],[3320,-1.2319],[3323,-1.2395],[3326,-1.2261],[3328,-1.2347],[3330,-1.2285],[3332,-1.2339],[3333,-1.2297],[3339,-1.2366],[3341,-1.231],[3343,-1.2376],[3345,-1.2289],[3347,-1.2254],[3351,-1.2392],[3353,-1.2308],[3354,-1.2317],[3356,-1.2257],[3358,-1.2353],[3360,-1.2246],[3362,-1.2328],[3363,-1.23],[3364,-1.2363],[3367,-1.2247],[3368,-1.2306],[3370,-1.2156],[3372,-1.2377],[3374,-1.2296],[3376,-1.2367],[3378,-1.2261],[3380,-1.2305],[3381,-1.2267],[3384,-1.2303],[3386,-1.2253],[3388,-1.224],[3392,-1.2321],[3395,-1.2232],[3397,-1.2351],[3400,-1.2319],[3403,-1.2395],[3406,-1.2261],[3408,-1.2347],[3410,-1.2285],[3412,-1.2339],[3413,-1.2297],[3419,-1.2366],[3421,-1.231],[3423,-1.2376],[3427,-1.2254],[3431,-1.2392],[3433,-1.2308],[3434,-1.2317],[3436,-1.2257],[3438,-1.2353],[3440,-1.2246],[3442,-1.2328],[3443,-1.23],[3444,-1.2363],[3447,-1.2247],[3448,-1.2306],[3450,-1.2156],[3452,-1.2377],[3454,-1.2296],[3456,-1.2367],[3458,-1.2261],[3460,-1.2305],[3461,-1.2267],[3464,-1.2303],[3466,-1.2253],[3468,-1.224],[3472,-1.2321],[3475,-1.2232],[3477,-1.2351],[3480,-1.2319],[3483,-1.2395],[3486,-1.2261],[3488,-1.2347],[3490,-1.2285],[3492,-1.2339],[3493,-1.2297],[3499,-1.2366],[3501,-1.231],[3503,-1.2376],[3507,-1.2254],[3511,-1.2392],[3513,-1.2307],[3514,-1.2317],[3516,-1.2257],[3518,-1.2353],[3520,-1.2246],[3522,-1.2328],[3523,-1.2284],[3524,-1.2301],[3527,-1.2265],[3529,-1.2291],[3535,-1.2256],[3536,-1.2233],[3539,-1.2313],[3541,-1.2267],[3544,-1.2303],[3546,-1.2253],[3548,-1.224],[3552,-1.2321],[3555,-1.2232],[3557,-1.2351],[3560,-1.2319],[3563,-1.2395],[3566,-1.2261],[3568,-1.2347],[3570,-1.2285],[3572,-1.2339],[3573,-1.2297],[3579,-1.2366],[3581,-1.231],[3583,-1.2376],[3586,-1.2289],[3587,-1.2388],[3588,-1.2329],[3590,-1.2398],[3593,-1.2261],[3595,-1.2398],[3596,-1.2352],[3597,-1.2402],[3599,-1.2325],[3600,-1.2332],[3603,-1.2284],[3604,-1.2301],[3607,-1.2265],[3609,-1.2291],[3615,-1.2256],[3616,-1.2233],[3618,-1.2275],[3619,-1.2265],[3621,-1.2297],[3622,-1.2279],[3627,-1.2303],[3630,-1.2284],[3634,-1.2312],[3635,-1.2258],[3637,-1.2351],[3640,-1.2319],[3643,-1.2395],[3646,-1.2261],[3648,-1.2347],[3650,-1.2285],[3652,-1.2339],[3653,-1.2297],[3659,-1.2366],[3661,-1.231],[3663,-1.2376],[3665,-1.2289],[3667,-1.2254],[3671,-1.2392],[3673,-1.2308],[3674,-1.2317],[3676,-1.2257],[3678,-1.2353],[3680,-1.2246],[3682,-1.2328],[3683,-1.23],[3684,-1.2363],[3687,-1.2247],[3688,-1.2306],[3690,-1.2156],[3692,-1.2377],[3694,-1.2296],[3696,-1.2367],[3698,-1.2261],[3700,-1.2305],[3701,-1.2267],[3704,-1.2303],[3706,-1.2253],[3708,-1.224],[3712,-1.2321],[3715,-1.2232],[3717,-1.2351],[3720,-1.2319],[3723,-1.2395],[3726,-1.2261],[3728,-1.2347],[3730,-1.2285],[3732,-1.2339],[3733,-1.2297],[3739,-1.2366],[3741,-1.231],[3743,-1.2376],[3747,-1.2254],[3751,-1.2392],[3753,-1.2307],[3754,-1.2317],[3756,-1.2257],[3758,-1.2353],[3760,-1.2246],[3762,-1.2328],[3763,-1.23],[3764,-1.2363],[3767,-1.2247],[3768,-1.2306],[3770,-1.2156],[3772,-1.2377],[3774,-1.2296],[3776,-1.2367],[3778,-1.2261],[3780,-1.2305],[3781,-1.2267],[3784,-1.2303],[3786,-1.2253],[3788,-1.2239],[3792,-1.2321],[3795,-1.2232],[3797,-1.2351],[3800,-1.2319],[3803,-1.2395],[3806,-1.2261],[3808,-1.2347],[3810,-1.2285],[3812,-1.2339],[3813,-1.2297],[3819,-1.2366],[3821,-1.231],[3823,-1.2376],[3827,-1.2254],[3831,-1.2392],[3833,-1.2308],[3834,-1.2317],[3836,-1.2257],[3838,-1.2353],[3840,-1.2246],[3842,-1.2328],[3843,-1.2284],[3844,-1.2301],[3847,-1.2265],[3849,-1.2291],[3855,-1.2256],[3856,-1.2233],[3859,-1.2313],[3861,-1.2267],[3864,-1.2303],[3866,-1.2253],[3868,-1.224],[3872,-1.2321],[3875,-1.2232],[3877,-1.2351],[3880,-1.2319],[3883,-1.2395],[3886,-1.2261],[3888,-1.2347],[3891,-1.2247],[3892,-1.2306],[3894,-1.2156],[3896,-1.2377],[3898,-1.2296],[3900,-1.2367],[3902,-1.2261],[3904,-1.2305],[3905,-1.2267],[3908,-1.2303],[3910,-1.2253],[3912,-1.2239],[3916,-1.2321],[3919,-1.2232],[3921,-1.2351],[3924,-1.2319],[3927,-1.2395],[3930,-1.2261],[3932,-1.2347],[3934,-1.2285],[3936,-1.2339],[3937,-1.2297],[3943,-1.2366],[3945,-1.231],[3947,-1.2376],[3951,-1.2254],[3955,-1.2392],[3957,-1.2308],[3958,-1.2317],[3960,-1.2257],[3962,-1.2353],[3964,-1.2246],[3966,-1.2328],[3967,-1.2284],[3968,-1.2301],[3971,-1.2265],[3973,-1.2291],[3979,-1.2256],[3980,-1.2233],[3983,-1.2313],[3985,-1.2267],[3988,-1.2303],[3990,-1.2253],[3992,-1.224],[3996,-1.2321],[3999,-1.2232],[4001,-1.2351],[4004,-1.2319],[4007,-1.2395],[4010,-1.2261],[4012,-1.2347],[4015,-1.2247],[4016,-1.2306],[4018,-1.2156],[4020,-1.2377],[4022,-1.2296],[4024,-1.2367],[4026,-1.2261],[4028,-1.2305],[4029,-1.2267],[4032,-1.2303],[4034,-1.2253],[4036,-1.2239],[4040,-1.2321],[4043,-1.2232],[4045,-1.2351],[4048,-1.2319],[4051,-1.2395],[4054,-1.2261],[4056,-1.2347],[4058,-1.2285],[4060,-1.2339],[4061,-1.2297],[4067,-1.2366],[4069,-1.231],[4071,-1.2376],[4075,-1.2254],[4079,-1.2392],[4081,-1.2308],[4082,-1.2317],[4084,-1.2257],[4086,-1.2353],[4088,-1.2246],[4090,-1.2328],[4091,-1.2284],[4092,-1.2301],[4095,-1.2265],[4097,-1.2291],[4103,-1.2256],[4104,-1.2233],[4107,-1.2313],[4109,-1.2267],[4112,-1.2303],[4114,-1.2253],[4116,-1.224],[4120,-1.2321],[4123,-1.2232],[4125,-1.2351],[4128,-1.2319],[4131,-1.2395],[4134,-1.2261],[4136,-1.2347],[4140,-1.2247],[4141,-1.2306],[4143,-1.2156],[4145,-1.2377],[4147,-1.2296],[4149,-1.2367],[4151,-1.2261],[4153,-1.2305],[4154,-1.2267],[4157,-1.2303],[4159,-1.2253],[4161,-1.2239],[4165,-1.2321],[4168,-1.2232],[4170,-1.2351],[4173,-1.2319],[4176,-1.2395],[4179,-1.2261],[4181,-1.2347],[4183,-1.2285],[4185,-1.2339],[4186,-1.2297],[4192,-1.2366],[4194,-1.231],[4196,-1.2376],[4200,-1.2254],[4204,-1.2392],[4206,-1.2308],[4207,-1.2317],[4209,-1.2257],[4211,-1.2353],[4213,-1.2246],[4215,-1.2328],[4216,-1.2284],[4217,-1.2301],[4220,-1.2265],[4222,-1.2291],[4228,-1.2256],[4229,-1.2233],[4232,-1.2313],[4234,-1.2267],[4237,-1.2303],[4239,-1.2253],[4241,-1.224],[4245,-1.2321],[4248,-1.2232],[4250,-1.2351],[4253,-1.2319],[4256,-1.2395],[4259,-1.2261],[4261,-1.2347],[4266,-1.2247],[4267,-1.2306],[4269,-1.2156],[4271,-1.2377],[4273,-1.2296],[4275,-1.2367],[4277,-1.2261],[4279,-1.2305],[4280,-1.2267],[4283,-1.2303],[4285,-1.2253],[4287,-1.2239],[4291,-1.2321],[4294,-1.2232],[4296,-1.2351],[4299,-1.2319],[4302,-1.2395],[4305,-1.2261],[4307,-1.2347],[4309,-1.2285],[4311,-1.2339],[4312,-1.2297],[4318,-1.2366],[4320,-1.231],[4322,-1.2376],[4326,-1.2254],[4330,-1.2392],[4332,-1.2308],[4333,-1.2317],[4335,-1.2257],[4337,-1.2353],[4339,-1.2246],[4341,-1.2328],[4342,-1.2284],[4343,-1.2301],[4346,-1.2265],[4348,-1.2291],[4354,-1.2256],[4355,-1.2233],[4360,-1.2316],[4365,-1.2285],[4367,-1.2339],[4376,-1.2388],[4377,-1.2329],[4379,-1.2398],[4382,-1.2261],[4384,-1.2398],[4385,-1.2352],[4386,-1.2402],[4388,-1.2325],[4389,-1.2332],[4392,-1.2284],[4393,-1.2301],[4396,-1.2265],[4398,-1.2291],[4404,-1.2256],[4405,-1.2233],[4407,-1.2275],[4408,-1.2265],[4410,-1.2297],[4411,-1.2279],[4416,-1.2303],[4419,-1.2284],[4423,-1.2312],[4424,-1.2258],[4426,-1.2351],[4429,-1.2319],[4432,-1.2395],[4435,-1.2261],[4436,0]]
       }
       ring_01_r_ctrl.Rotation.Z {
        type=MovieSceneScriptingFloatChannel
        default=-0.9643
        thinned_from=125
        keys [[1790,-0.9643],[1830,8.2885],[1832,8.2892],[1867,-1.0856],[1910,8.2885],[1943,-1.0519],[1944,-1.0391],[1946,-1.0858],[1947,-1.0855],[1990,8.2885],[2023,-1.0519],[2024,-1.0391],[2026,-1.0858],[2027,-1.0855],[2070,8.2885],[2106,-1.0858],[2107,-1.0855],[2150,8.2885],[2152,8.2892],[2187,-1.0856],[2230,8.2885],[2263,-1.0519],[2264,-1.0391],[2266,-1.0858],[2267,-1.0855],[2310,8.2885],[2343,-1.0519],[2344,-1.0391],[2346,-1.0858],[2347,-1.0855],[2390,8.2885],[2426,-1.0858],[2427,-1.0855],[2467,8.2408],[2468,8.2071],[2496,5.5128],[2503,1.5307],[2508,-40.2725],[2513,-53.6898],[2515,-53.6893],[2518,-33.0488],[2520,27.7913],[2522,25.0503],[2524,26.9035],[2530,24.3008],[2763,33.2083],[2794,-35.1101],[2938,-34.6897],[2986,58.8009],[3099,58.8009],[3185,43.7134],[3221,-1.0858],[3222,-1.0855],[3265,8.2885],[3267,8.2893],[3302,-1.0857],[3345,8.2885],[3378,-1.0519],[3379,-1.0391],[3381,-1.0858],[3382,-1.0855],[3425,8.2885],[3458,-1.0519],[3459,-1.0391],[3461,-1.0858],[3462,-1.0855],[3505,8.2885],[3541,-1.0858],[3542,-1.0855],[3585,8.2885],[3587,8.2893],[3622,-1.0856],[3665,8.2885],[3698,-1.0519],[3699,-1.0391],[3701,-1.0858],[3702,-1.0855],[3745,8.2885],[3778,-1.0519],[3779,-1.0391],[3781,-1.0858],[3782,-1.0855],[3825,8.2885],[3861,-1.0858],[3863,-1.0855],[3902,-1.0519],[3903,-1.0391],[3905,-1.0858],[3906,-1.0855],[3949,8.2885],[3985,-1.0858],[3987,-1.0855],[4026,-1.0519],[4027,-1.0391],[4029,-1.0858],[4030,-1.0855],[4073,8.2885],[4109,-1.0858],[4111,-1.0855],[4151,-1.0519],[4152,-1.0391],[4154,-1.0858],[4155,-1.0855],[4198,8.2885],[4234,-1.0858],[4236,-1.0855],[4277,-1.0519],[4278,-1.0391],[4280,-1.0858],[4281,-1.0855],[4324,8.2885],[4360,5.2399],[4364,5.7163],[4376,8.2893],[4411,-1.0856],[4435,6.6273],[4436,0]]
       }
       ring_01_r_ctrl.Scale.X {type=MovieSceneScriptingFloatChannel, default=1, const [1,12]}
       ring_01_r_ctrl.Scale.Y {type=MovieSceneScriptingFloatChannel, default=1, const [1,12]}
       ring_01_r_ctrl.Scale.Z {type=MovieSceneScriptingFloatChannel, default=1, const [1,12]}
       ring_02_r_ctrl.Location.X {type=MovieSceneScriptingFloatChannel, default=0, thinned_from=266, keys [[1790,0],[2512,0],[2513,-0.0002],[2515,0],[4436,0]]}
       ring_02_r_ctrl.Location.Y {type=MovieSceneScriptingFloatChannel, default=0, thinned_from=332, keys [[1790,0],[2512,0],[2513,-0.0001],[2528,0],[4436,0]]}
       ring_02_r_ctrl.Location.Z {type=MovieSceneScriptingFloatChannel, default=0, thinned_from=1276, keys [[1790,0],[4436,0]]}
       ring_02_r_ctrl.Rotation.X {
        type=MovieSceneScriptingFloatChannel
        default=-1.0879
        thinned_from=157
        keys [[1790,-1.0879],[1828,-1.2586],[1831,-1.2604],[1844,-1.1378],[1868,-1.0853],[1903,-1.2532],[1907,-1.2567],[1909,-1.2604],[1910,-1.2602],[1924,-1.1376],[1948,-1.0853],[1988,-1.2586],[1991,-1.2604],[2004,-1.1376],[2028,-1.0853],[2063,-1.2532],[2067,-1.2567],[2069,-1.2604],[2070,-1.2602],[2084,-1.1376],[2108,-1.0853],[2148,-1.2586],[2151,-1.2604],[2164,-1.1378],[2188,-1.0853],[2223,-1.2532],[2227,-1.2567],[2229,-1.2604],[2230,-1.2602],[2244,-1.1376],[2268,-1.0853],[2308,-1.2586],[2311,-1.2604],[2324,-1.1376],[2348,-1.0853],[2383,-1.2532],[2387,-1.2567],[2389,-1.2604],[2390,-1.2602],[2404,-1.1376],[2428,-1.0853],[2466,-1.257],[2471,-1.1861],[2503,-0.0001],[2515,-0.0001],[2519,-0.1815],[2522,-0.0392],[2524,-0.1335],[2527,-0.0124],[2531,-0.0001],[2760,-0.3292],[2765,-0.4919],[2780,-1.1137],[2785,-0.9741],[2803,-0.3669],[2927,-0.3549],[2938,-0.3545],[2986,-0.3035],[3099,-0.3035],[3185,-0.1979],[3223,-1.0853],[3263,-1.2586],[3266,-1.2604],[3279,-1.1378],[3303,-1.0853],[3338,-1.2532],[3342,-1.2567],[3344,-1.2604],[3345,-1.2602],[3359,-1.1376],[3383,-1.0853],[3423,-1.2586],[3426,-1.2604],[3439,-1.1376],[3463,-1.0853],[3498,-1.2532],[3505,-1.2602],[3519,-1.1376],[3543,-1.0853],[3583,-1.2586],[3586,-1.2604],[3599,-1.1378],[3623,-1.0853],[3658,-1.2532],[3665,-1.2602],[3679,-1.1376],[3703,-1.0853],[3743,-1.2586],[3746,-1.2604],[3759,-1.1376],[3783,-1.0853],[3818,-1.2532],[3824,-1.2604],[3840,-1.1304],[3863,-1.0853],[3907,-1.0853],[3942,-1.2532],[3948,-1.2604],[3964,-1.1304],[3987,-1.0853],[4031,-1.0853],[4066,-1.2532],[4072,-1.2604],[4088,-1.1304],[4111,-1.0853],[4156,-1.0853],[4191,-1.2532],[4197,-1.2604],[4213,-1.1304],[4236,-1.0853],[4282,-1.0853],[4317,-1.2532],[4323,-1.2604],[4339,-1.1304],[4360,-1.2098],[4364,-1.2325],[4372,-1.2586],[4375,-1.2604],[4388,-1.1378],[4412,-1.0853],[4435,-1.2209],[4436,0]]
       }
       ring_02_r_ctrl.Rotation.Y {
        type=MovieSceneScriptingFloatChannel
        default=0.1664
        thinned_from=370
        keys [[1790,0.1664],[1812,0.1884],[1813,0.1882],[1817,0.1905],[1818,0.1905],[1824,0.1924],[1826,0.1922],[1828,0.1928],[1832,0.1924],[1862,0.1662],[1867,0.1659],[1892,0.1884],[1893,0.1882],[1895,0.1901],[1896,0.19],[1903,0.1923],[1904,0.1919],[1907,0.1924],[1908,0.1921],[1910,0.1928],[1936,0.1665],[1938,0.1668],[1940,0.166],[1941,0.1663],[1943,0.166],[1948,0.166],[1977,0.1905],[1978,0.1905],[1984,0.1924],[1986,0.1922],[1988,0.1928],[1991,0.1923],[1992,0.1924],[2022,0.1662],[2027,0.1659],[2052,0.1884],[2053,0.1882],[2055,0.1901],[2056,0.19],[2063,0.1923],[2064,0.1919],[2067,0.1924],[2068,0.1921],[2070,0.1928],[2102,0.1662],[2107,0.1659],[2132,0.1884],[2133,0.1882],[2137,0.1905],[2138,0.1905],[2144,0.1924],[2146,0.1922],[2148,0.1928],[2152,0.1924],[2182,0.1662],[2187,0.1659],[2212,0.1884],[2213,0.1882],[2215,0.1901],[2216,0.19],[2223,0.1923],[2224,0.1919],[2227,0.1924],[2228,0.1921],[2230,0.1928],[2256,0.1665],[2258,0.1668],[2260,0.166],[2261,0.1663],[2263,0.166],[2268,0.166],[2297,0.1905],[2298,0.1905],[2304,0.1924],[2306,0.1922],[2308,0.1928],[2311,0.1923],[2312,0.1924],[2342,0.1662],[2347,0.1659],[2372,0.1884],[2373,0.1882],[2375,0.1901],[2376,0.19],[2383,0.1923],[2384,0.1919],[2387,0.1924],[2388,0.1921],[2390,0.1928],[2422,0.1662],[2427,0.1659],[2452,0.1884],[2453,0.1882],[2457,0.1905],[2458,0.1905],[2464,0.1924],[2466,0.1922],[2467,0.1923],[2501,0.0025],[2507,0.0002],[2515,0.0002],[2519,0.066],[2520,0.0663],[2521,0.02],[2522,0.0135],[2524,0.0511],[2527,0.005],[2530,0.0002],[2758,0.0544],[2761,0.1527],[2764,0.1793],[2775,0.1885],[2790,0.1177],[2791,0.1181],[2796,0.071],[2801,0.056],[2802,0.0573],[2938,0.0571],[2986,0.1303],[3099,0.1303],[3185,0.0305],[3217,0.1662],[3222,0.1659],[3247,0.1884],[3248,0.1882],[3252,0.1905],[3253,0.1905],[3259,0.1924],[3261,0.1922],[3263,0.1928],[3267,0.1924],[3297,0.1662],[3302,0.1659],[3327,0.1884],[3328,0.1882],[3330,0.1901],[3331,0.19],[3338,0.1923],[3339,0.1919],[3342,0.1924],[3343,0.1921],[3346,0.1928],[3371,0.1665],[3373,0.1668],[3375,0.166],[3376,0.1663],[3378,0.166],[3383,0.166],[3412,0.1905],[3413,0.1905],[3419,0.1924],[3421,0.1922],[3423,0.1928],[3426,0.1923],[3427,0.1924],[3457,0.1662],[3462,0.1659],[3487,0.1884],[3488,0.1882],[3490,0.1901],[3491,0.19],[3498,0.1923],[3499,0.1919],[3501,0.1924],[3503,0.1921],[3505,0.1928],[3537,0.1662],[3542,0.1659],[3567,0.1884],[3568,0.1882],[3572,0.1905],[3573,0.1905],[3579,0.1924],[3581,0.1922],[3583,0.1928],[3587,0.1924],[3617,0.1662],[3622,0.1659],[3647,0.1884],[3648,0.1882],[3650,0.1901],[3651,0.19],[3658,0.1923],[3659,0.1919],[3661,0.1924],[3663,0.1921],[3665,0.1928],[3691,0.1665],[3693,0.1668],[3695,0.166],[3696,0.1663],[3698,0.166],[3703,0.166],[3732,0.1905],[3733,0.1905],[3739,0.1924],[3741,0.1922],[3743,0.1928],[3746,0.1923],[3747,0.1924],[3777,0.1662],[3782,0.1659],[3807,0.1884],[3808,0.1882],[3810,0.1901],[3811,0.19],[3818,0.1923],[3819,0.1919],[3821,0.1924],[3823,0.1921],[3826,0.1928],[3857,0.1662],[3862,0.1659],[3887,0.1884],[3888,0.1882],[3901,0.1662],[3906,0.1659],[3931,0.1884],[3932,0.1882],[3934,0.1901],[3935,0.19],[3942,0.1923],[3943,0.1919],[3945,0.1924],[3947,0.1921],[3950,0.1928],[3981,0.1662],[3986,0.1659],[4011,0.1884],[4012,0.1882],[4025,0.1662],[4030,0.1659],[4055,0.1884],[4056,0.1882],[4058,0.1901],[4059,0.19],[4066,0.1923],[4067,0.1919],[4069,0.1924],[4071,0.1921],[4074,0.1928],[4105,0.1662],[4110,0.1659],[4135,0.1884],[4136,0.1882],[4150,0.1662],[4155,0.1659],[4180,0.1884],[4181,0.1882],[4183,0.1901],[4184,0.19],[4191,0.1923],[4192,0.1919],[4194,0.1924],[4196,0.1921],[4199,0.1928],[4230,0.1662],[4235,0.1659],[4260,0.1884],[4261,0.1882],[4276,0.1662],[4281,0.1659],[4306,0.1884],[4307,0.1882],[4309,0.1901],[4310,0.19],[4317,0.1923],[4318,0.1919],[4320,0.1924],[4322,0.1921],[4325,0.1928],[4356,0.1662],[4357,0.1662],[4360,0.1886],[4367,0.1905],[4372,0.1928],[4376,0.1924],[4406,0.1662],[4411,0.1659],[4435,0.188],[4436,0]]
       }
       ring_02_r_ctrl.Rotation.Z {
        type=MovieSceneScriptingFloatChannel
        default=4.5139
        thinned_from=111
        keys [[1790,4.5139],[1829,15.7363],[1865,4.3677],[1866,4.3677],[1906,15.6519],[1908,15.6519],[1910,15.8101],[1940,4.5543],[1941,4.5544],[1947,4.3677],[1989,15.7363],[2025,4.3677],[2026,4.3677],[2067,15.6519],[2068,15.6519],[2071,15.8101],[2105,4.3677],[2106,4.3677],[2149,15.7363],[2185,4.3677],[2186,4.3677],[2227,15.6519],[2228,15.6519],[2231,15.8101],[2260,4.5543],[2261,4.5544],[2267,4.3677],[2309,15.7363],[2345,4.3677],[2346,4.3677],[2387,15.6519],[2388,15.6519],[2391,15.8101],[2425,4.3677],[2426,4.3677],[2467,15.6714],[2469,15.4454],[2470,15.4636],[2496,-12.7454],[2504,-16.1479],[2512,-21.0269],[2515,-21.0267],[2519,-0.2587],[2520,0.253],[2521,-12.5549],[2522,-14.8475],[2524,-4.6898],[2527,-17.5773],[2530,-18.9352],[2764,30.432],[2794,-22.3582],[2938,-22.0047],[2986,54.0387],[3099,54.0387],[3185,31.8755],[3220,4.3677],[3221,4.3677],[3264,15.7363],[3300,4.3677],[3342,15.6519],[3343,15.6519],[3346,15.8101],[3382,4.3677],[3424,15.7363],[3460,4.3677],[3461,4.3677],[3502,15.6519],[3503,15.6519],[3505,15.8101],[3540,4.3677],[3584,15.7363],[3620,4.3677],[3662,15.6519],[3663,15.6519],[3666,15.8101],[3695,4.5544],[3702,4.3677],[3744,15.7363],[3780,4.3677],[3781,4.3677],[3825,15.8101],[3860,4.3677],[3905,4.3677],[3949,15.8101],[3984,4.3677],[4029,4.3677],[4073,15.8101],[4108,4.3677],[4154,4.3677],[4198,15.8101],[4233,4.3677],[4280,4.3677],[4324,15.8101],[4360,12.5512],[4364,13.2703],[4373,15.7363],[4410,4.3677],[4435,13.7517],[4436,0]]
       }
       ring_02_r_ctrl.Scale.X {type=MovieSceneScriptingFloatChannel, default=1, const [1,12]}
       ring_02_r_ctrl.Scale.Y {type=MovieSceneScriptingFloatChannel, default=1, const [1,12]}
       ring_02_r_ctrl.Scale.Z {type=MovieSceneScriptingFloatChannel, default=1, const [1,12]}
       ring_03_r_ctrl.Location.X {type=MovieSceneScriptingFloatChannel, default=0, thinned_from=65, keys [[1790,0],[2512,0],[2513,-0.0001],[2528,0],[4435,0],[4436,-0.0001]]}
       ring_03_r_ctrl.Location.Y {type=MovieSceneScriptingFloatChannel, default=0, thinned_from=301, keys [[1790,0],[4436,0]]}
       ring_03_r_ctrl.Location.Z {type=MovieSceneScriptingFloatChannel, default=0, thinned_from=1572, keys [[1790,0],[4436,0]]}
       ring_03_r_ctrl.Rotation.X {
        type=MovieSceneScriptingFloatChannel
        default=-0.4798
        thinned_from=155
        keys [[1790,-0.4798],[1828,-0.5551],[1831,-0.5558],[1844,-0.502],[1868,-0.4787],[1903,-0.5527],[1910,-0.5558],[1924,-0.5015],[1948,-0.4787],[1988,-0.5551],[1991,-0.5558],[2004,-0.5016],[2028,-0.4787],[2063,-0.5527],[2064,-0.5527],[2069,-0.5558],[2070,-0.5558],[2085,-0.4987],[2107,-0.4787],[2148,-0.5551],[2151,-0.5558],[2164,-0.502],[2188,-0.4787],[2223,-0.5527],[2224,-0.5527],[2229,-0.5558],[2230,-0.5558],[2244,-0.5015],[2268,-0.4787],[2308,-0.5551],[2311,-0.5558],[2324,-0.5016],[2348,-0.4787],[2383,-0.5527],[2390,-0.5558],[2405,-0.4987],[2427,-0.4787],[2429,-0.4787],[2466,-0.5544],[2471,-0.5233],[2503,0],[2513,0],[2518,-0.0275],[2519,-0.0807],[2522,-0.0168],[2524,-0.0597],[2529,0],[2759,-0.1041],[2761,-0.1771],[2769,-0.2857],[2780,-0.4907],[2785,-0.4298],[2802,-0.1743],[2927,-0.1699],[2938,-0.1698],[2986,-0.2908],[3099,-0.2908],[3185,-0.1881],[3222,-0.4787],[3224,-0.4787],[3263,-0.5551],[3266,-0.5558],[3279,-0.502],[3303,-0.4787],[3338,-0.5527],[3342,-0.5542],[3344,-0.5558],[3345,-0.5558],[3359,-0.5015],[3383,-0.4787],[3423,-0.5551],[3426,-0.5558],[3439,-0.5016],[3463,-0.4787],[3498,-0.5527],[3499,-0.5527],[3504,-0.5558],[3505,-0.5558],[3520,-0.4987],[3542,-0.4787],[3544,-0.4787],[3583,-0.5551],[3586,-0.5558],[3599,-0.502],[3623,-0.4787],[3658,-0.5527],[3659,-0.5527],[3664,-0.5558],[3665,-0.5558],[3679,-0.5015],[3703,-0.4787],[3743,-0.5551],[3746,-0.5558],[3759,-0.5016],[3783,-0.4787],[3818,-0.5527],[3819,-0.5527],[3824,-0.5558],[3839,-0.502],[3864,-0.4787],[3907,-0.4787],[3942,-0.5527],[3943,-0.5527],[3948,-0.5558],[3963,-0.502],[3988,-0.4787],[4031,-0.4787],[4066,-0.5527],[4067,-0.5527],[4072,-0.5558],[4087,-0.502],[4112,-0.4787],[4156,-0.4787],[4191,-0.5527],[4192,-0.5527],[4197,-0.5558],[4212,-0.502],[4237,-0.4787],[4282,-0.4787],[4317,-0.5527],[4318,-0.5527],[4323,-0.5558],[4338,-0.502],[4360,-0.5326],[4364,-0.5421],[4372,-0.5551],[4375,-0.5558],[4388,-0.502],[4412,-0.4787],[4435,-0.5385],[4436,0]]
       }
       ring_03_r_ctrl.Rotation.Y {
        type=MovieSceneScriptingFloatChannel
        default=0.0663
        thinned_from=350
        keys [[1790,0.0663],[1805,0.0737],[1806,0.0734],[1809,0.075],[1810,0.0745],[1818,0.0762],[1819,0.0761],[1826,0.0768],[1827,0.0767],[1830,0.0769],[1856,0.0667],[1857,0.0668],[1861,0.0664],[1863,0.0664],[1865,0.0661],[1887,0.074],[1888,0.074],[1894,0.0759],[1895,0.0757],[1906,0.0768],[1908,0.0765],[1910,0.077],[1932,0.0674],[1933,0.0675],[1936,0.0664],[1937,0.0666],[1940,0.0662],[1941,0.0665],[1945,0.0661],[1967,0.074],[1968,0.074],[1974,0.0759],[1975,0.0757],[1978,0.0762],[1979,0.0761],[1986,0.0768],[1987,0.0767],[1992,0.077],[1993,0.0768],[1994,0.077],[2016,0.0667],[2017,0.0668],[2021,0.0664],[2023,0.0664],[2025,0.0661],[2070,0.077],[2092,0.0674],[2093,0.0675],[2096,0.0664],[2097,0.0666],[2100,0.0662],[2101,0.0665],[2105,0.0662],[2107,0.0663],[2125,0.0737],[2126,0.0734],[2129,0.075],[2130,0.0745],[2138,0.0762],[2139,0.0761],[2146,0.0768],[2147,0.0767],[2150,0.0769],[2176,0.0667],[2177,0.0668],[2181,0.0664],[2183,0.0664],[2185,0.0661],[2207,0.074],[2208,0.074],[2214,0.0759],[2215,0.0757],[2226,0.0768],[2228,0.0765],[2230,0.077],[2252,0.0674],[2253,0.0675],[2256,0.0664],[2257,0.0666],[2260,0.0662],[2261,0.0665],[2265,0.0661],[2287,0.074],[2288,0.074],[2294,0.0759],[2295,0.0757],[2298,0.0762],[2299,0.0761],[2306,0.0768],[2307,0.0767],[2312,0.077],[2313,0.0768],[2314,0.077],[2336,0.0667],[2337,0.0668],[2341,0.0664],[2343,0.0664],[2345,0.0661],[2386,0.0768],[2388,0.0765],[2390,0.077],[2412,0.0674],[2413,0.0675],[2416,0.0664],[2417,0.0666],[2420,0.0662],[2421,0.0665],[2425,0.0662],[2427,0.0663],[2445,0.0737],[2446,0.0734],[2449,0.075],[2450,0.0745],[2458,0.0762],[2459,0.0761],[2466,0.0768],[2503,-0.0001],[2513,-0.0001],[2518,0.0072],[2519,0.0269],[2520,0.029],[2521,0.0091],[2522,0.006],[2524,0.0219],[2527,0.0018],[2530,-0.0004],[2758,0.0232],[2760,0.0534],[2762,0.0721],[2765,0.0764],[2766,0.076],[2768,0.0767],[2771,0.0763],[2775,0.0767],[2789,0.0465],[2791,0.0473],[2795,0.0298],[2800,0.0214],[2802,0.0229],[2938,0.0229],[2986,0.0623],[3099,0.0623],[3185,-0.0129],[3215,0.0662],[3216,0.0665],[3220,0.0662],[3222,0.0663],[3240,0.0737],[3241,0.0734],[3244,0.075],[3245,0.0745],[3253,0.0762],[3254,0.0761],[3261,0.0768],[3262,0.0767],[3265,0.0769],[3291,0.0667],[3292,0.0668],[3296,0.0664],[3298,0.0664],[3300,0.0661],[3322,0.074],[3323,0.074],[3329,0.0759],[3330,0.0757],[3343,0.0765],[3346,0.077],[3367,0.0674],[3368,0.0675],[3371,0.0664],[3372,0.0666],[3375,0.0662],[3376,0.0665],[3380,0.0661],[3402,0.074],[3403,0.074],[3409,0.0759],[3410,0.0757],[3413,0.0762],[3414,0.0761],[3421,0.0768],[3422,0.0767],[3427,0.077],[3428,0.0768],[3429,0.077],[3451,0.0667],[3452,0.0668],[3456,0.0664],[3458,0.0664],[3460,0.0661],[3503,0.0765],[3527,0.0674],[3528,0.0675],[3531,0.0664],[3532,0.0666],[3535,0.0662],[3536,0.0665],[3540,0.0662],[3543,0.0663],[3560,0.0737],[3561,0.0734],[3564,0.075],[3565,0.0745],[3573,0.0762],[3574,0.0761],[3581,0.0768],[3582,0.0767],[3585,0.0769],[3611,0.0667],[3612,0.0668],[3616,0.0664],[3618,0.0664],[3620,0.0661],[3642,0.074],[3643,0.074],[3649,0.0759],[3650,0.0757],[3661,0.0768],[3663,0.0765],[3665,0.077],[3687,0.0674],[3688,0.0675],[3691,0.0664],[3692,0.0666],[3695,0.0662],[3696,0.0665],[3700,0.0661],[3722,0.074],[3723,0.074],[3729,0.0759],[3730,0.0757],[3733,0.0762],[3734,0.0761],[3741,0.0768],[3742,0.0767],[3747,0.077],[3748,0.0768],[3749,0.077],[3771,0.0667],[3772,0.0668],[3776,0.0664],[3778,0.0664],[3780,0.0661],[3825,0.077],[3847,0.0674],[3848,0.0675],[3851,0.0664],[3852,0.0666],[3855,0.0662],[3856,0.0665],[3860,0.0662],[3862,0.0663],[3880,0.0737],[3881,0.0734],[3884,0.075],[3885,0.0745],[3895,0.0667],[3896,0.0668],[3900,0.0664],[3902,0.0664],[3904,0.0661],[3949,0.077],[3971,0.0674],[3972,0.0675],[3975,0.0664],[3976,0.0666],[3979,0.0662],[3980,0.0665],[3984,0.0662],[3986,0.0663],[4004,0.0737],[4005,0.0734],[4008,0.075],[4009,0.0745],[4019,0.0667],[4020,0.0668],[4024,0.0664],[4026,0.0664],[4028,0.0661],[4073,0.077],[4095,0.0674],[4096,0.0675],[4099,0.0664],[4100,0.0666],[4103,0.0662],[4104,0.0665],[4108,0.0662],[4110,0.0663],[4128,0.0737],[4129,0.0734],[4132,0.075],[4133,0.0745],[4144,0.0667],[4145,0.0668],[4149,0.0664],[4151,0.0664],[4153,0.0661],[4198,0.077],[4220,0.0674],[4221,0.0675],[4224,0.0664],[4225,0.0666],[4228,0.0662],[4229,0.0665],[4233,0.0662],[4235,0.0663],[4253,0.0737],[4254,0.0734],[4257,0.075],[4258,0.0745],[4270,0.0667],[4271,0.0668],[4275,0.0664],[4277,0.0664],[4279,0.0661],[4324,0.077],[4346,0.0674],[4347,0.0675],[4350,0.0664],[4351,0.0666],[4354,0.0662],[4355,0.0665],[4360,0.0753],[4364,0.076],[4371,0.0767],[4400,0.0667],[4401,0.0668],[4405,0.0664],[4407,0.0664],[4409,0.0661],[4431,0.074],[4432,0.074],[4435,0.0748],[4436,0.0002]]
       }
       ring_03_r_ctrl.Rotation.Z {
        type=MovieSceneScriptingFloatChannel
        default=4.4115
        thinned_from=92
        keys [[1790,4.4115],[1829,16.7422],[1865,4.3311],[1866,4.3311],[1906,16.653],[1910,16.8202],[1940,4.5332],[1941,4.5332],[1945,4.3311],[1946,4.3311],[1989,16.7422],[1992,16.7568],[2025,4.3311],[2026,4.3311],[2071,16.8202],[2100,4.5332],[2105,4.3311],[2150,16.7422],[2185,4.3311],[2186,4.3311],[2231,16.8202],[2260,4.5332],[2265,4.3311],[2310,16.7422],[2312,16.7568],[2345,4.3311],[2346,4.3311],[2391,16.8202],[2420,4.5332],[2426,4.3311],[2467,16.6248],[2502,-33.0584],[2513,-29.0017],[2515,-29.198],[2520,12.3753],[2522,6.4549],[2524,10.4038],[2531,4.8082],[2764,24.0226],[2927,20.8998],[2938,20.8325],[2986,52.3873],[3099,52.3873],[3185,46.8285],[3220,4.3311],[3264,16.7422],[3301,4.3311],[3341,16.653],[3346,16.8202],[3381,4.3311],[3424,16.7422],[3427,16.7568],[3460,4.3311],[3461,4.3311],[3502,16.653],[3505,16.8202],[3541,4.3311],[3585,16.7422],[3621,4.3311],[3666,16.8202],[3700,4.3311],[3701,4.3311],[3745,16.7422],[3747,16.7568],[3780,4.3311],[3781,4.3311],[3825,16.8202],[3860,4.3311],[3905,4.3311],[3949,16.8202],[3984,4.3311],[4029,4.3311],[4073,16.8202],[4108,4.3311],[4154,4.3311],[4198,16.8202],[4233,4.3311],[4280,4.3311],[4324,16.8202],[4360,13.3984],[4364,14.1533],[4373,16.7422],[4410,4.3311],[4435,14.5577],[4436,0.0037]]
       }
       ring_03_r_ctrl.Scale.X {type=MovieSceneScriptingFloatChannel, default=1, const [1,12]}
       ring_03_r_ctrl.Scale.Y {type=MovieSceneScriptingFloatChannel, default=1, const [1,12]}
       ring_03_r_ctrl.Scale.Z {type=MovieSceneScriptingFloatChannel, default=1, const [1,12]}
       pinky_metacarpal_r_ctrl.Location.X {type=MovieSceneScriptingFloatChannel, default=0, thinned_from=28, keys [[1790,0],[2512,0],[2513,-0.0001],[2520,0],[4436,0]]}
       pinky_metacarpal_r_ctrl.Location.Y {type=MovieSceneScriptingFloatChannel, default=0, thinned_from=30, keys [[1790,0],[4436,0]]}
       pinky_metacarpal_r_ctrl.Location.Z {type=MovieSceneScriptingFloatChannel, default=0, thinned_from=29, keys [[1790,0],[4436,0]]}
       pinky_metacarpal_r_ctrl.Rotation.X {
        type=MovieSceneScriptingFloatChannel
        default=-18.9529
        thinned_from=88
        keys [[1790,-18.9529],[1825,-16.6678],[1828,-16.6226],[1868,-19.0053],[1905,-16.6678],[1908,-16.6226],[1947,-19.0053],[1985,-16.6678],[1988,-16.6226],[2027,-19.0053],[2065,-16.6678],[2068,-16.6226],[2108,-19.0053],[2145,-16.6678],[2148,-16.6226],[2188,-19.0053],[2225,-16.6678],[2228,-16.6226],[2267,-19.0053],[2308,-16.6226],[2347,-19.0053],[2385,-16.6678],[2389,-16.6226],[2428,-19.0053],[2458,-16.8716],[2502,-10.3796],[2507,-5.0887],[2513,-1.8329],[2518,-3.6464],[2519,-8.143],[2521,-14.8028],[2523,-13.0012],[2527,-15.2751],[2532,-15.5447],[2763,-6.5718],[2774,-10.8531],[2777,-9.5765],[2781,-3.4147],[2788,0],[3185,0],[3223,-19.0053],[3264,-16.6226],[3303,-19.0053],[3340,-16.6678],[3344,-16.6226],[3383,-19.0053],[3423,-16.6226],[3463,-19.0053],[3504,-16.6226],[3543,-19.0053],[3580,-16.6678],[3584,-16.6226],[3623,-19.0053],[3663,-16.6226],[3703,-19.0053],[3740,-16.6678],[3744,-16.6226],[3782,-19.0053],[3823,-16.6226],[3863,-19.0053],[3906,-19.0053],[3947,-16.6226],[3987,-19.0053],[4030,-19.0053],[4071,-16.6226],[4111,-19.0053],[4155,-19.0053],[4196,-16.6226],[4236,-19.0053],[4281,-19.0053],[4322,-16.6226],[4360,-17.2939],[4364,-17.1199],[4372,-16.6226],[4412,-19.0053],[4430,-17.2771],[4435,-17.1489],[4436,0]]
       }
       pinky_metacarpal_r_ctrl.Rotation.Y {
        type=MovieSceneScriptingFloatChannel
        default=1.5773
        thinned_from=182
        keys [[1790,1.5773],[1805,2.2894],[1825,2.634],[1828,2.6485],[1861,1.5724],[1862,1.5738],[1864,1.5647],[1865,1.5695],[1867,1.5555],[1905,2.634],[1908,2.6485],[1941,1.5724],[1942,1.5738],[1945,1.5528],[1948,1.5564],[1985,2.634],[1988,2.6485],[2021,1.5724],[2022,1.5738],[2025,1.5528],[2028,1.5564],[2065,2.634],[2068,2.6485],[2101,1.5724],[2102,1.5738],[2104,1.5647],[2105,1.5695],[2107,1.5555],[2145,2.634],[2148,2.6485],[2181,1.5724],[2182,1.5738],[2184,1.5647],[2185,1.5695],[2187,1.5555],[2225,2.634],[2228,2.6485],[2261,1.5724],[2262,1.5738],[2265,1.5528],[2268,1.5564],[2305,2.634],[2308,2.6485],[2341,1.5724],[2342,1.5738],[2345,1.5528],[2348,1.5564],[2385,2.634],[2389,2.6485],[2421,1.5724],[2422,1.5738],[2424,1.5647],[2425,1.5695],[2427,1.5555],[2465,2.634],[2467,2.6392],[2469,2.6367],[2496,1.8865],[2503,1.5879],[2508,-1.1054],[2513,-1.9818],[2515,-1.9818],[2518,-0.6034],[2520,3.2826],[2522,2.7008],[2524,3.0817],[2531,2.5404],[2764,4.3893],[2776,2.7601],[2782,0.48],[2788,0],[3185,0],[3216,1.5724],[3217,1.5738],[3219,1.5647],[3220,1.5695],[3222,1.5555],[3260,2.634],[3264,2.6485],[3296,1.5724],[3297,1.5738],[3299,1.5647],[3300,1.5695],[3302,1.5555],[3340,2.634],[3344,2.6485],[3376,1.5724],[3377,1.5738],[3380,1.5528],[3420,2.634],[3423,2.6485],[3456,1.5724],[3457,1.5738],[3460,1.5528],[3463,1.5564],[3500,2.634],[3504,2.6485],[3536,1.5724],[3537,1.5738],[3539,1.5647],[3540,1.5695],[3542,1.5555],[3580,2.634],[3584,2.6485],[3616,1.5724],[3617,1.5738],[3619,1.5647],[3620,1.5695],[3622,1.5555],[3660,2.634],[3663,2.6485],[3696,1.5724],[3697,1.5738],[3700,1.5528],[3703,1.5564],[3740,2.634],[3744,2.6485],[3776,1.5724],[3777,1.5738],[3780,1.5528],[3783,1.5564],[3823,2.6485],[3856,1.5724],[3857,1.5738],[3859,1.5647],[3860,1.5695],[3862,1.5555],[3901,1.5738],[3904,1.5528],[3907,1.5564],[3947,2.6485],[3980,1.5724],[3981,1.5738],[3983,1.5647],[3984,1.5695],[3986,1.5555],[4025,1.5738],[4028,1.5528],[4031,1.5564],[4071,2.6485],[4104,1.5724],[4105,1.5738],[4107,1.5647],[4108,1.5695],[4110,1.5555],[4150,1.5738],[4153,1.5528],[4156,1.5564],[4196,2.6485],[4229,1.5724],[4230,1.5738],[4232,1.5647],[4233,1.5695],[4235,1.5555],[4276,1.5738],[4279,1.5528],[4282,1.5564],[4322,2.6485],[4355,1.5724],[4356,1.5738],[4360,2.3532],[4364,2.4954],[4372,2.6485],[4405,1.5724],[4406,1.5738],[4408,1.5647],[4409,1.5695],[4411,1.5555],[4435,2.4528],[4436,0]]
       }
       pinky_metacarpal_r_ctrl.Rotation.Z {
        type=MovieSceneScriptingFloatChannel
        default=-2.1577
        thinned_from=169
        keys [[1790,-2.1577],[1825,-3.7941],[1828,-3.816],[1861,-2.1495],[1862,-2.152],[1864,-2.1375],[1865,-2.1451],[1867,-2.1229],[1905,-3.7941],[1908,-3.816],[1941,-2.1495],[1942,-2.152],[1945,-2.1183],[1948,-2.1244],[1985,-3.7941],[1988,-3.816],[2021,-2.1495],[2022,-2.152],[2025,-2.1183],[2028,-2.1244],[2065,-3.7941],[2068,-3.816],[2101,-2.1495],[2102,-2.152],[2104,-2.1375],[2105,-2.1451],[2107,-2.1229],[2145,-3.7941],[2148,-3.816],[2181,-2.1495],[2182,-2.152],[2184,-2.1375],[2185,-2.1451],[2187,-2.1229],[2228,-3.816],[2261,-2.1495],[2262,-2.152],[2265,-2.1183],[2268,-2.1244],[2308,-3.816],[2341,-2.1495],[2342,-2.152],[2345,-2.1183],[2348,-2.1244],[2385,-3.7941],[2389,-3.816],[2421,-2.1495],[2422,-2.1519],[2424,-2.1375],[2425,-2.1451],[2427,-2.1229],[2469,-3.8358],[2470,-3.8295],[2502,-7.3788],[2507,-5.104],[2513,-3.693],[2519,-4.9694],[2520,-7.0605],[2521,-9.3269],[2522,-9.3958],[2523,-7.9325],[2524,-7.8779],[2527,-10.0137],[2531,-10.2481],[2759,-5.7141],[2761,-2.7436],[2764,-1.5141],[2774,-2.4703],[2781,-0.7179],[2788,0],[3185,0],[3216,-2.1495],[3217,-2.152],[3219,-2.1375],[3220,-2.1451],[3222,-2.1229],[3264,-3.816],[3296,-2.1495],[3297,-2.152],[3299,-2.1375],[3300,-2.1451],[3302,-2.1229],[3344,-3.816],[3376,-2.1495],[3377,-2.1519],[3380,-2.1183],[3423,-3.816],[3456,-2.1495],[3457,-2.152],[3460,-2.1183],[3463,-2.1244],[3504,-3.816],[3536,-2.1495],[3537,-2.152],[3539,-2.1375],[3540,-2.1451],[3542,-2.1229],[3584,-3.816],[3616,-2.1495],[3617,-2.1519],[3619,-2.1375],[3620,-2.1451],[3622,-2.1229],[3663,-3.816],[3696,-2.1495],[3697,-2.152],[3700,-2.1183],[3703,-2.1244],[3740,-3.7941],[3744,-3.816],[3776,-2.1495],[3777,-2.1519],[3780,-2.1183],[3823,-3.816],[3856,-2.1495],[3857,-2.1519],[3859,-2.1375],[3860,-2.1451],[3862,-2.1229],[3901,-2.1519],[3904,-2.1183],[3947,-3.816],[3980,-2.1495],[3981,-2.1519],[3983,-2.1375],[3984,-2.1451],[3986,-2.1229],[4025,-2.1519],[4028,-2.1183],[4071,-3.816],[4104,-2.1495],[4105,-2.1519],[4107,-2.1375],[4108,-2.1451],[4110,-2.1229],[4150,-2.1519],[4153,-2.1183],[4196,-3.816],[4229,-2.1495],[4230,-2.1519],[4232,-2.1375],[4233,-2.1451],[4235,-2.1229],[4276,-2.1519],[4279,-2.1183],[4322,-3.816],[4355,-2.1495],[4356,-2.1519],[4360,-3.3586],[4364,-3.5789],[4372,-3.816],[4405,-2.1495],[4406,-2.1519],[4408,-2.1375],[4409,-2.1451],[4411,-2.1229],[4435,-3.5128],[4436,0]]
       }
       pinky_metacarpal_r_ctrl.Scale.X {type=MovieSceneScriptingFloatChannel, default=1, const [1,12]}
       pinky_metacarpal_r_ctrl.Scale.Y {type=MovieSceneScriptingFloatChannel, default=1, const [1,12]}
       pinky_metacarpal_r_ctrl.Scale.Z {type=MovieSceneScriptingFloatChannel, default=1, const [1,12]}
       pinky_01_r_ctrl.Location.X {
        type=MovieSceneScriptingFloatChannel
        default=0
        thinned_from=69
        keys [[1790,0],[2511,0],[2513,-0.0005],[2519,0],[2768,0],[2784,0.1967],[2796,0.2863],[2938,0.2843],[2986,0.1606],[3099,0.1606],[3185,0.2852],[3238,0],[4436,0]]
       }
       pinky_01_r_ctrl.Location.Y {
        type=MovieSceneScriptingFloatChannel
        default=0
        thinned_from=571
        keys [[1790,0],[2768,0],[2775,-0.0505],[2782,-0.4544],[2796,-0.8971],[2938,-0.8898],[2986,-0.4295],[3099,-0.4295],[3185,-0.8865],[3216,0],[4436,0]]
       }
       pinky_01_r_ctrl.Location.Z {
        type=MovieSceneScriptingFloatChannel
        default=0
        thinned_from=610
        keys [[1790,0],[2768,0],[2781,0.1115],[2803,-0.2985],[2938,-0.2974],[2986,0.2198],[3099,0.2198],[3185,0.3058],[3216,0],[4436,0]]
       }
       pinky_01_r_ctrl.Rotation.X {
        type=MovieSceneScriptingFloatChannel
        default=10.7443
        thinned_from=95
        keys [[1790,10.7443],[1829,14.2964],[1830,14.2955],[1868,10.7053],[1909,14.2964],[1910,14.2955],[1948,10.7053],[1989,14.2964],[1990,14.2955],[2028,10.7053],[2069,14.2964],[2070,14.2955],[2108,10.7053],[2149,14.2964],[2150,14.2955],[2188,10.7053],[2229,14.2964],[2230,14.2955],[2268,10.7053],[2309,14.2964],[2310,14.2955],[2348,10.7053],[2389,14.2964],[2390,14.2955],[2428,10.7053],[2466,14.2478],[2502,-3.9885],[2512,-2.598],[2518,-2.6839],[2519,-1.71],[2522,-5.0074],[2524,-2.9432],[2532,-5.8214],[2758,-2.7071],[2761,2.7397],[2774,8.3668],[2781,0.1365],[2788,-3.4286],[2791,-3.2293],[2797,-3.3401],[2802,-3.2564],[2938,-3.2662],[2986,14.8473],[3099,14.8473],[3185,-3.9075],[3222,10.7053],[3264,14.2964],[3265,14.2955],[3303,10.7053],[3344,14.2964],[3345,14.2955],[3383,10.7053],[3424,14.2964],[3425,14.2955],[3463,10.7053],[3504,14.2964],[3505,14.2955],[3543,10.7053],[3584,14.2964],[3585,14.2955],[3623,10.7053],[3664,14.2964],[3665,14.2955],[3703,10.7053],[3744,14.2964],[3745,14.2955],[3783,10.7053],[3824,14.2964],[3825,14.2955],[3863,10.7053],[3907,10.7053],[3948,14.2964],[3949,14.2955],[3987,10.7053],[4031,10.7053],[4072,14.2964],[4073,14.2955],[4111,10.7053],[4156,10.7053],[4197,14.2964],[4198,14.2955],[4236,10.7053],[4282,10.7053],[4323,14.2964],[4324,14.2955],[4360,13.2061],[4364,13.4523],[4373,14.2964],[4412,10.7053],[4435,13.4973],[4436,0]]
       }
       pinky_01_r_ctrl.Rotation.Y {
        type=MovieSceneScriptingFloatChannel
        default=5.5447
        thinned_from=140
        keys [[1790,5.5447],[1830,6.598],[1864,5.536],[1868,5.5318],[1910,6.598],[1912,6.5981],[1944,5.536],[1948,5.5318],[1990,6.598],[1992,6.5981],[2024,5.536],[2028,5.5318],[2070,6.598],[2072,6.5981],[2104,5.536],[2108,5.5318],[2150,6.598],[2184,5.536],[2188,5.5318],[2230,6.598],[2232,6.5981],[2264,5.536],[2268,5.5318],[2310,6.598],[2312,6.5981],[2344,5.536],[2348,5.5318],[2390,6.598],[2392,6.5981],[2424,5.536],[2428,5.5318],[2456,6.4954],[2470,6.9155],[2496,7.9107],[2504,6.1251],[2512,0.8716],[2515,0.8717],[2520,15.1114],[2522,12.1374],[2524,14.1021],[2526,12.0826],[2531,11.3132],[2532,11.3132],[2763,20.8885],[2772,15.109],[2777,9.0029],[2782,0.5128],[2802,-7.9147],[2938,-7.8892],[2986,3.7205],[3099,3.7205],[3185,13.5797],[3219,5.536],[3222,5.5318],[3265,6.598],[3299,5.536],[3303,5.5318],[3345,6.598],[3347,6.5981],[3379,5.536],[3383,5.5318],[3425,6.598],[3427,6.5981],[3459,5.536],[3463,5.5318],[3505,6.598],[3507,6.5981],[3539,5.536],[3543,5.5318],[3585,6.598],[3619,5.536],[3622,5.5318],[3665,6.598],[3667,6.5981],[3699,5.536],[3703,5.5318],[3747,6.5981],[3779,5.536],[3783,5.5318],[3825,6.598],[3827,6.5981],[3859,5.536],[3907,5.5318],[3949,6.598],[3951,6.5981],[3983,5.536],[4031,5.5318],[4073,6.598],[4075,6.5981],[4107,5.536],[4156,5.5318],[4198,6.598],[4200,6.5981],[4232,5.536],[4282,5.5318],[4324,6.598],[4326,6.5981],[4360,6.265],[4408,5.536],[4412,5.5318],[4435,6.4027],[4436,0]]
       }
       pinky_01_r_ctrl.Rotation.Z {
        type=MovieSceneScriptingFloatChannel
        default=-4.2615
        thinned_from=70
        keys [[1790,-4.2615],[1830,5.3656],[1867,-4.3857],[1910,5.3656],[1948,-4.3857],[1990,5.3656],[2028,-4.3857],[2070,5.3656],[2108,-4.3857],[2150,5.3656],[2188,-4.3857],[2230,5.3656],[2268,-4.3857],[2310,5.3656],[2348,-4.3857],[2390,5.3656],[2428,-4.3857],[2467,5.2847],[2468,5.2397],[2496,-5.6862],[2513,-50.1755],[2518,-33.3718],[2519,0.2985],[2520,13.26],[2522,3.7088],[2524,10.0911],[2532,1.1126],[2764,32.1271],[2794,-43.5153],[2938,-43.0814],[2986,51.9313],[3099,51.9313],[3185,21.784],[3222,-4.3857],[3265,5.3656],[3303,-4.3857],[3345,5.3656],[3383,-4.3857],[3425,5.3656],[3463,-4.3857],[3505,5.3656],[3543,-4.3857],[3585,5.3656],[3622,-4.3857],[3665,5.3656],[3703,-4.3857],[3745,5.3656],[3783,-4.3857],[3825,5.3656],[3862,-4.3857],[3907,-4.3857],[3949,5.3656],[3986,-4.3857],[4031,-4.3857],[4073,5.3656],[4110,-4.3857],[4156,-4.3857],[4198,5.3656],[4235,-4.3857],[4282,-4.3857],[4324,5.3656],[4360,2.32],[4364,1.7717],[4411,-4.3857],[4435,3.6549],[4436,0]]
       }
       pinky_01_r_ctrl.Scale.X {type=MovieSceneScriptingFloatChannel, default=1, const [1,12]}
       pinky_01_r_ctrl.Scale.Y {type=MovieSceneScriptingFloatChannel, default=1, const [1,12]}
       pinky_01_r_ctrl.Scale.Z {type=MovieSceneScriptingFloatChannel, default=1, const [1,12]}
       pinky_02_r_ctrl.Location.X {type=MovieSceneScriptingFloatChannel, default=0, thinned_from=79, keys [[1790,0],[2512,0],[2513,-0.0001],[2525,0],[4436,0]]}
       pinky_02_r_ctrl.Location.Y {type=MovieSceneScriptingFloatChannel, default=0, thinned_from=215, keys [[1790,0],[4436,0]]}
       pinky_02_r_ctrl.Location.Z {type=MovieSceneScriptingFloatChannel, default=0, thinned_from=1401, keys [[1790,0],[4436,0]]}
       pinky_02_r_ctrl.Rotation.X {
        type=MovieSceneScriptingFloatChannel
        default=-2.6202
        thinned_from=138
        keys [[1790,-2.6202],[1828,-3.0315],[1831,-3.0355],[1844,-2.7404],[1868,-2.6142],[1908,-3.0315],[1911,-3.0355],[1924,-2.7401],[1948,-2.6142],[1988,-3.0315],[1991,-3.0355],[2004,-2.7402],[2028,-2.6142],[2068,-3.0315],[2071,-3.0355],[2084,-2.7401],[2108,-2.614],[2148,-3.0315],[2151,-3.0355],[2164,-2.7404],[2188,-2.6142],[2228,-3.0315],[2231,-3.0355],[2244,-2.7401],[2268,-2.6142],[2308,-3.0315],[2311,-3.0355],[2324,-2.7402],[2348,-2.6142],[2388,-3.0315],[2391,-3.0355],[2404,-2.7401],[2428,-2.614],[2466,-3.0276],[2471,-2.8434],[2496,0.164],[2504,0.2387],[2513,0.3116],[2518,0.16],[2519,-0.142],[2522,0.2377],[2524,-0.0123],[2530,0.3384],[2532,0.3384],[2760,-0.528],[2765,-0.9591],[2780,-2.6669],[2799,-1.2066],[2927,-1.1371],[2938,-1.1359],[2986,-0.9681],[3099,-0.9681],[3185,-0.3793],[3224,-2.614],[3263,-3.0315],[3266,-3.0355],[3279,-2.7404],[3303,-2.6142],[3343,-3.0315],[3346,-3.0355],[3359,-2.7401],[3383,-2.6142],[3423,-3.0315],[3426,-3.0355],[3439,-2.7403],[3463,-2.6142],[3503,-3.0315],[3506,-3.0355],[3520,-2.7224],[3544,-2.614],[3583,-3.0315],[3586,-3.0355],[3599,-2.7404],[3623,-2.6142],[3663,-3.0315],[3666,-3.0355],[3679,-2.7401],[3703,-2.6142],[3743,-3.0315],[3746,-3.0355],[3759,-2.7403],[3783,-2.6142],[3823,-3.0315],[3826,-3.0355],[3840,-2.7224],[3864,-2.614],[3907,-2.6142],[3947,-3.0315],[3950,-3.0355],[3964,-2.7224],[3988,-2.614],[4031,-2.6142],[4071,-3.0315],[4074,-3.0355],[4088,-2.7224],[4112,-2.614],[4156,-2.6142],[4196,-3.0315],[4199,-3.0355],[4213,-2.7224],[4237,-2.614],[4282,-2.6142],[4322,-3.0315],[4325,-3.0355],[4339,-2.7224],[4360,-2.9087],[4364,-2.9642],[4372,-3.0315],[4375,-3.0355],[4388,-2.7404],[4412,-2.6142],[4435,-2.9411],[4436,0]]
       }
       pinky_02_r_ctrl.Rotation.Y {
        type=MovieSceneScriptingFloatChannel
        default=0.0629
        thinned_from=869
        keys [[1790,0.0629],[1792,0.065],[1794,0.0647],[1798,0.0694],[1801,0.0688],[1803,0.0715],[1804,0.0711],[1806,0.0725],[1807,0.0724],[1809,0.074],[1811,0.0732],[1814,0.0744],[1815,0.0743],[1816,0.0748],[1818,0.0735],[1822,0.0761],[1823,0.0753],[1824,0.0754],[1826,0.0746],[1828,0.0758],[1831,0.0745],[1832,0.075],[1833,0.0738],[1834,0.0743],[1836,0.0716],[1837,0.072],[1839,0.0712],[1841,0.0718],[1845,0.0661],[1847,0.0674],[1850,0.0659],[1852,0.0662],[1854,0.0653],[1858,0.0653],[1860,0.0647],[1861,0.0649],[1865,0.0637],[1868,0.065],[1871,0.0643],[1874,0.0663],[1875,0.066],[1880,0.0696],[1881,0.0688],[1883,0.0715],[1884,0.0711],[1886,0.0725],[1887,0.0724],[1889,0.074],[1891,0.0732],[1894,0.0744],[1895,0.0743],[1896,0.0748],[1898,0.0735],[1902,0.0761],[1903,0.0753],[1904,0.0754],[1906,0.0746],[1908,0.0758],[1911,0.0745],[1912,0.0751],[1922,0.0692],[1924,0.0698],[1928,0.0666],[1929,0.0669],[1933,0.0647],[1935,0.0657],[1936,0.0655],[1937,0.0657],[1940,0.0644],[1943,0.0649],[1945,0.0637],[1948,0.065],[1951,0.0643],[1954,0.0663],[1955,0.066],[1964,0.0714],[1965,0.0711],[1969,0.0732],[1970,0.0732],[1971,0.0734],[1972,0.0734],[1974,0.0737],[1975,0.0732],[1977,0.0744],[1978,0.0735],[1982,0.0761],[1983,0.0753],[1984,0.0754],[1986,0.0746],[1988,0.0758],[1992,0.0735],[1996,0.0745],[2001,0.0692],[2003,0.0697],[2005,0.0662],[2007,0.0674],[2008,0.0666],[2009,0.0669],[2013,0.0647],[2015,0.0657],[2016,0.0655],[2017,0.0657],[2020,0.0644],[2023,0.0649],[2025,0.0637],[2028,0.065],[2031,0.0643],[2034,0.0663],[2035,0.066],[2040,0.0696],[2041,0.0688],[2043,0.0715],[2044,0.0711],[2046,0.0725],[2047,0.0724],[2049,0.074],[2051,0.0732],[2054,0.0744],[2055,0.0743],[2056,0.0748],[2058,0.0735],[2062,0.0761],[2063,0.0753],[2064,0.0754],[2066,0.0746],[2068,0.0758],[2071,0.0745],[2072,0.0751],[2082,0.0692],[2084,0.0698],[2088,0.0666],[2089,0.0669],[2093,0.0647],[2095,0.0657],[2096,0.0655],[2097,0.0657],[2100,0.0644],[2103,0.0649],[2104,0.0645],[2106,0.0655],[2110,0.0629],[2112,0.065],[2114,0.0647],[2118,0.0694],[2121,0.0688],[2123,0.0715],[2124,0.0711],[2126,0.0725],[2127,0.0724],[2129,0.074],[2131,0.0732],[2134,0.0744],[2135,0.0743],[2136,0.0748],[2138,0.0735],[2142,0.0761],[2143,0.0753],[2144,0.0754],[2146,0.0746],[2148,0.0758],[2151,0.0745],[2152,0.075],[2153,0.0738],[2154,0.0743],[2156,0.0716],[2157,0.072],[2159,0.0712],[2161,0.0718],[2165,0.0661],[2167,0.0674],[2170,0.0659],[2172,0.0662],[2174,0.0653],[2178,0.0653],[2180,0.0647],[2181,0.0649],[2185,0.0637],[2188,0.065],[2191,0.0643],[2194,0.0663],[2195,0.066],[2200,0.0696],[2201,0.0688],[2203,0.0715],[2204,0.0711],[2206,0.0725],[2207,0.0724],[2209,0.074],[2211,0.0732],[2214,0.0744],[2215,0.0743],[2216,0.0748],[2218,0.0735],[2222,0.0761],[2223,0.0753],[2224,0.0754],[2226,0.0746],[2228,0.0758],[2231,0.0745],[2232,0.0751],[2242,0.0692],[2244,0.0698],[2248,0.0666],[2249,0.0669],[2253,0.0647],[2255,0.0657],[2256,0.0655],[2257,0.0657],[2260,0.0644],[2263,0.0649],[2265,0.0637],[2268,0.065],[2271,0.0643],[2274,0.0663],[2275,0.066],[2284,0.0714],[2285,0.0711],[2289,0.0732],[2290,0.0732],[2291,0.0734],[2292,0.0734],[2294,0.0737],[2295,0.0732],[2297,0.0744],[2298,0.0735],[2302,0.0761],[2303,0.0753],[2304,0.0754],[2306,0.0746],[2308,0.0758],[2312,0.0735],[2316,0.0745],[2321,0.0692],[2323,0.0697],[2325,0.0662],[2327,0.0674],[2328,0.0666],[2329,0.0669],[2333,0.0647],[2335,0.0657],[2336,0.0655],[2337,0.0657],[2340,0.0644],[2343,0.0649],[2345,0.0637],[2348,0.065],[2351,0.0643],[2354,0.0663],[2355,0.066],[2360,0.0696],[2361,0.0688],[2363,0.0715],[2364,0.0711],[2366,0.0725],[2367,0.0724],[2369,0.074],[2371,0.0732],[2374,0.0744],[2375,0.0743],[2376,0.0748],[2378,0.0735],[2382,0.0761],[2383,0.0753],[2384,0.0754],[2386,0.0746],[2388,0.0758],[2391,0.0745],[2392,0.0751],[2402,0.0692],[2404,0.0698],[2408,0.0666],[2409,0.0669],[2413,0.0647],[2415,0.0657],[2416,0.0655],[2417,0.0657],[2420,0.0644],[2423,0.0649],[2424,0.0645],[2426,0.0655],[2430,0.0629],[2432,0.065],[2434,0.0647],[2438,0.0694],[2441,0.0688],[2443,0.0715],[2444,0.0711],[2446,0.0725],[2447,0.0724],[2449,0.074],[2451,0.0732],[2454,0.0744],[2455,0.0743],[2456,0.0748],[2458,0.0735],[2462,0.0761],[2463,0.0753],[2464,0.0754],[2466,0.0746],[2467,0.0754],[2469,0.0729],[2471,0.0842],[2503,0.1953],[2509,0.2685],[2513,0.2804],[2519,0.2563],[2522,0.2799],[2523,0.2709],[2530,0.2877],[2532,0.2877],[2766,0.2216],[2776,0.1337],[2786,0.0163],[2791,0.0444],[2797,-0.0178],[2802,0.0171],[2927,0.0204],[2938,0.0205],[2986,0.3344],[3099,0.3344],[3185,0.0464],[3215,0.0644],[3218,0.0649],[3219,0.0645],[3221,0.0655],[3223,0.0638],[3224,0.0638],[3225,0.0629],[3227,0.065],[3229,0.0647],[3233,0.0694],[3236,0.0688],[3238,0.0715],[3239,0.0711],[3241,0.0725],[3242,0.0724],[3244,0.074],[3246,0.0732],[3249,0.0744],[3250,0.0743],[3251,0.0748],[3253,0.0735],[3257,0.0761],[3258,0.0753],[3259,0.0754],[3261,0.0746],[3263,0.0758],[3266,0.0745],[3267,0.075],[3268,0.0738],[3269,0.0743],[3271,0.0716],[3272,0.072],[3274,0.0712],[3276,0.0718],[3280,0.0661],[3282,0.0674],[3285,0.0659],[3287,0.0662],[3289,0.0653],[3293,0.0653],[3295,0.0647],[3296,0.0649],[3300,0.0637],[3303,0.065],[3306,0.0643],[3309,0.0663],[3310,0.066],[3315,0.0696],[3316,0.0688],[3318,0.0715],[3319,0.0711],[3321,0.0725],[3322,0.0724],[3324,0.074],[3326,0.0732],[3329,0.0744],[3330,0.0743],[3331,0.0748],[3333,0.0735],[3337,0.0761],[3338,0.0753],[3339,0.0754],[3341,0.0746],[3343,0.0758],[3346,0.0745],[3347,0.0751],[3357,0.0692],[3359,0.0698],[3363,0.0666],[3364,0.0669],[3368,0.0647],[3370,0.0657],[3371,0.0655],[3372,0.0657],[3375,0.0644],[3378,0.0649],[3380,0.0637],[3383,0.065],[3386,0.0643],[3389,0.0663],[3390,0.066],[3399,0.0714],[3400,0.0711],[3404,0.0732],[3405,0.0732],[3406,0.0734],[3407,0.0734],[3409,0.0737],[3410,0.0732],[3412,0.0744],[3413,0.0735],[3417,0.0761],[3418,0.0753],[3419,0.0754],[3421,0.0746],[3423,0.0758],[3427,0.0735],[3431,0.0745],[3436,0.0692],[3438,0.0697],[3440,0.0662],[3442,0.0674],[3443,0.0666],[3444,0.0669],[3448,0.0647],[3450,0.0657],[3451,0.0655],[3452,0.0657],[3455,0.0644],[3458,0.0649],[3460,0.0637],[3463,0.065],[3466,0.0643],[3469,0.0663],[3470,0.066],[3475,0.0696],[3476,0.0688],[3478,0.0715],[3479,0.0711],[3481,0.0725],[3482,0.0724],[3484,0.074],[3486,0.0732],[3489,0.0744],[3490,0.0743],[3491,0.0748],[3493,0.0735],[3497,0.0761],[3498,0.0753],[3499,0.0754],[3501,0.0746],[3503,0.0758],[3506,0.0745],[3507,0.0751],[3517,0.0692],[3519,0.0698],[3523,0.0666],[3524,0.0669],[3528,0.0647],[3530,0.0657],[3531,0.0655],[3532,0.0657],[3535,0.0644],[3538,0.0649],[3539,0.0645],[3541,0.0655],[3543,0.0638],[3544,0.0638],[3545,0.0629],[3547,0.065],[3549,0.0647],[3553,0.0694],[3556,0.0688],[3558,0.0715],[3559,0.0711],[3561,0.0725],[3562,0.0724],[3564,0.074],[3566,0.0732],[3569,0.0744],[3570,0.0743],[3571,0.0748],[3573,0.0735],[3577,0.0761],[3578,0.0753],[3579,0.0754],[3581,0.0746],[3583,0.0758],[3586,0.0745],[3587,0.075],[3588,0.0738],[3589,0.0743],[3591,0.0716],[3592,0.072],[3594,0.0712],[3596,0.0718],[3600,0.0661],[3602,0.0674],[3605,0.0659],[3607,0.0662],[3609,0.0653],[3613,0.0653],[3615,0.0647],[3616,0.0649],[3620,0.0637],[3623,0.065],[3626,0.0643],[3629,0.0663],[3630,0.066],[3635,0.0696],[3636,0.0688],[3638,0.0715],[3639,0.0711],[3641,0.0725],[3642,0.0724],[3644,0.074],[3646,0.0732],[3649,0.0744],[3650,0.0743],[3651,0.0748],[3653,0.0735],[3657,0.0761],[3658,0.0753],[3659,0.0754],[3661,0.0746],[3663,0.0758],[3666,0.0745],[3667,0.0751],[3677,0.0692],[3679,0.0698],[3683,0.0666],[3684,0.0669],[3688,0.0647],[3690,0.0657],[3691,0.0655],[3692,0.0657],[3695,0.0644],[3698,0.0649],[3700,0.0637],[3703,0.065],[3706,0.0643],[3709,0.0663],[3710,0.066],[3719,0.0714],[3720,0.0711],[3724,0.0732],[3725,0.0732],[3726,0.0734],[3727,0.0734],[3729,0.0737],[3730,0.0732],[3732,0.0744],[3733,0.0735],[3737,0.0761],[3738,0.0753],[3739,0.0754],[3741,0.0746],[3743,0.0758],[3747,0.0735],[3751,0.0745],[3756,0.0692],[3758,0.0697],[3760,0.0662],[3762,0.0674],[3763,0.0666],[3764,0.0669],[3768,0.0647],[3770,0.0657],[3771,0.0655],[3772,0.0657],[3775,0.0644],[3778,0.0649],[3780,0.0637],[3783,0.065],[3786,0.0643],[3789,0.0663],[3790,0.066],[3795,0.0696],[3796,0.0688],[3798,0.0715],[3799,0.0711],[3801,0.0725],[3802,0.0724],[3804,0.074],[3806,0.0732],[3809,0.0744],[3810,0.0743],[3811,0.0748],[3813,0.0735],[3817,0.0761],[3818,0.0753],[3819,0.0754],[3821,0.0746],[3823,0.0758],[3826,0.0745],[3827,0.0751],[3837,0.0692],[3839,0.0698],[3843,0.0666],[3844,0.0669],[3848,0.0647],[3850,0.0657],[3851,0.0655],[3852,0.0657],[3855,0.0644],[3858,0.0649],[3859,0.0645],[3861,0.0655],[3865,0.0629],[3867,0.065],[3869,0.0647],[3873,0.0694],[3876,0.0688],[3878,0.0715],[3879,0.0711],[3881,0.0725],[3882,0.0724],[3884,0.074],[3886,0.0732],[3892,0.0647],[3894,0.0657],[3895,0.0655],[3896,0.0657],[3899,0.0644],[3902,0.0649],[3904,0.0637],[3907,0.065],[3910,0.0643],[3913,0.0663],[3914,0.066],[3919,0.0696],[3920,0.0688],[3922,0.0715],[3923,0.0711],[3925,0.0725],[3926,0.0724],[3928,0.074],[3930,0.0732],[3933,0.0744],[3934,0.0743],[3935,0.0748],[3937,0.0735],[3941,0.0761],[3942,0.0753],[3943,0.0754],[3945,0.0746],[3947,0.0758],[3950,0.0745],[3951,0.0751],[3961,0.0692],[3963,0.0698],[3967,0.0666],[3968,0.0669],[3972,0.0647],[3974,0.0657],[3975,0.0655],[3976,0.0657],[3979,0.0644],[3982,0.0649],[3983,0.0645],[3985,0.0655],[3989,0.0629],[3991,0.065],[3993,0.0647],[3997,0.0694],[4000,0.0688],[4002,0.0715],[4003,0.0711],[4005,0.0725],[4006,0.0724],[4008,0.074],[4010,0.0732],[4016,0.0647],[4018,0.0657],[4019,0.0655],[4020,0.0657],[4023,0.0644],[4026,0.0649],[4028,0.0637],[4031,0.065],[4034,0.0643],[4037,0.0663],[4038,0.066],[4043,0.0696],[4044,0.0688],[4046,0.0715],[4047,0.0711],[4049,0.0725],[4050,0.0724],[4052,0.074],[4054,0.0732],[4057,0.0744],[4058,0.0743],[4059,0.0748],[4061,0.0735],[4065,0.0761],[4066,0.0753],[4067,0.0754],[4069,0.0746],[4071,0.0758],[4074,0.0745],[4075,0.0751],[4085,0.0692],[4087,0.0698],[4091,0.0666],[4092,0.0669],[4096,0.0647],[4098,0.0657],[4099,0.0655],[4100,0.0657],[4103,0.0644],[4106,0.0649],[4107,0.0645],[4109,0.0655],[4113,0.0629],[4115,0.065],[4117,0.0647],[4121,0.0694],[4124,0.0688],[4126,0.0715],[4127,0.0711],[4129,0.0725],[4130,0.0724],[4132,0.074],[4134,0.0732],[4141,0.0647],[4143,0.0657],[4144,0.0655],[4145,0.0657],[4148,0.0644],[4151,0.0649],[4153,0.0637],[4156,0.065],[4159,0.0643],[4162,0.0663],[4163,0.066],[4168,0.0696],[4169,0.0688],[4171,0.0715],[4172,0.0711],[4174,0.0725],[4175,0.0724],[4177,0.074],[4179,0.0732],[4182,0.0744],[4183,0.0743],[4184,0.0748],[4186,0.0735],[4190,0.0761],[4191,0.0753],[4192,0.0754],[4194,0.0746],[4196,0.0758],[4199,0.0745],[4200,0.0751],[4210,0.0692],[4212,0.0698],[4216,0.0666],[4217,0.0669],[4221,0.0647],[4223,0.0657],[4224,0.0655],[4225,0.0657],[4228,0.0644],[4231,0.0649],[4232,0.0645],[4234,0.0655],[4238,0.0629],[4240,0.065],[4242,0.0647],[4246,0.0694],[4249,0.0688],[4251,0.0715],[4252,0.0711],[4254,0.0725],[4255,0.0724],[4257,0.074],[4259,0.0732],[4267,0.0647],[4269,0.0657],[4270,0.0655],[4271,0.0657],[4274,0.0644],[4277,0.0649],[4279,0.0637],[4282,0.065],[4285,0.0643],[4288,0.0663],[4289,0.066],[4294,0.0696],[4295,0.0688],[4297,0.0715],[4298,0.0711],[4300,0.0725],[4301,0.0724],[4303,0.074],[4305,0.0732],[4308,0.0744],[4309,0.0743],[4310,0.0748],[4312,0.0735],[4316,0.0761],[4317,0.0753],[4318,0.0754],[4320,0.0746],[4322,0.0758],[4325,0.0745],[4326,0.0751],[4336,0.0692],[4338,0.0698],[4342,0.0666],[4343,0.0669],[4347,0.0647],[4349,0.0657],[4350,0.0655],[4351,0.0657],[4354,0.0644],[4357,0.0649],[4360,0.0744],[4365,0.0743],[4366,0.0748],[4372,0.0758],[4375,0.0745],[4376,0.075],[4377,0.0738],[4378,0.0743],[4380,0.0716],[4381,0.072],[4383,0.0712],[4385,0.0718],[4389,0.0661],[4391,0.0674],[4394,0.0659],[4396,0.0662],[4398,0.0653],[4402,0.0653],[4404,0.0647],[4405,0.0649],[4409,0.0637],[4412,0.065],[4415,0.0643],[4418,0.0663],[4419,0.066],[4424,0.0696],[4425,0.0688],[4427,0.0715],[4428,0.0711],[4430,0.0725],[4431,0.0724],[4433,0.074],[4435,0.0732],[4436,0]]
       }
       pinky_02_r_ctrl.Rotation.Z {
        type=MovieSceneScriptingFloatChannel
        default=6.6309
        thinned_from=113
        keys [[1790,6.6309],[1829,17.705],[1832,17.707],[1865,6.558],[1868,6.558],[1909,17.705],[1945,6.558],[1948,6.558],[1989,17.705],[2025,6.558],[2028,6.558],[2069,17.705],[2105,6.558],[2108,6.5581],[2149,17.705],[2152,17.707],[2185,6.558],[2188,6.558],[2229,17.705],[2265,6.558],[2268,6.558],[2309,17.705],[2345,6.558],[2348,6.558],[2389,17.705],[2426,6.558],[2428,6.5581],[2467,17.6453],[2469,17.5075],[2496,-6.5725],[2504,-9.2352],[2510,-9.1303],[2515,-9.1301],[2520,10.5689],[2522,-3.3896],[2524,5.9987],[2527,-5.9049],[2531,-7.1488],[2764,38.3434],[2794,-19.5767],[2938,-19.387],[2986,54.0391],[3099,54.0391],[3185,39.5411],[3220,6.558],[3264,17.705],[3267,17.7072],[3300,6.558],[3303,6.558],[3344,17.705],[3380,6.558],[3383,6.558],[3424,17.705],[3460,6.558],[3463,6.558],[3504,17.705],[3540,6.558],[3584,17.705],[3587,17.7072],[3620,6.558],[3623,6.558],[3664,17.705],[3700,6.558],[3703,6.558],[3744,17.705],[3780,6.558],[3782,6.558],[3824,17.705],[3860,6.558],[3906,6.558],[3948,17.705],[3984,6.558],[4030,6.558],[4072,17.705],[4108,6.558],[4155,6.558],[4197,17.705],[4233,6.558],[4281,6.558],[4323,17.705],[4360,14.7018],[4364,15.3798],[4373,17.705],[4376,17.7072],[4409,6.558],[4435,15.7319],[4436,0]]
       }
       pinky_02_r_ctrl.Scale.X {type=MovieSceneScriptingFloatChannel, default=1, const [1,12]}
       pinky_02_r_ctrl.Scale.Y {type=MovieSceneScriptingFloatChannel, default=1, const [1,12]}
       pinky_02_r_ctrl.Scale.Z {type=MovieSceneScriptingFloatChannel, default=1, const [1,12]}
       pinky_03_r_ctrl.Location.X {type=MovieSceneScriptingFloatChannel, default=0, thinned_from=189, keys [[1790,0],[2512,0],[2513,-0.0001],[2528,0],[4436,0]]}
       pinky_03_r_ctrl.Location.Y {type=MovieSceneScriptingFloatChannel, default=0, thinned_from=808, keys [[1790,0],[4436,0]]}
       pinky_03_r_ctrl.Location.Z {type=MovieSceneScriptingFloatChannel, default=0, thinned_from=36, keys [[1790,0],[4436,0]]}
       pinky_03_r_ctrl.Rotation.X {
        type=MovieSceneScriptingFloatChannel
        default=-1.1107
        thinned_from=148
        keys [[1790,-1.1107],[1828,-1.2852],[1868,-1.1082],[1903,-1.2797],[1904,-1.2796],[1906,-1.2832],[1908,-1.2832],[1910,-1.2869],[1924,-1.1624],[1948,-1.1082],[1988,-1.2852],[1992,-1.2851],[2005,-1.1537],[2029,-1.108],[2063,-1.2797],[2064,-1.2796],[2066,-1.2832],[2068,-1.2832],[2070,-1.2869],[2084,-1.1624],[2108,-1.1082],[2148,-1.2852],[2188,-1.1082],[2223,-1.2797],[2224,-1.2796],[2226,-1.2832],[2228,-1.2832],[2230,-1.2869],[2244,-1.1624],[2268,-1.1082],[2308,-1.2852],[2312,-1.2851],[2325,-1.1537],[2349,-1.108],[2383,-1.2797],[2384,-1.2796],[2386,-1.2832],[2388,-1.2832],[2390,-1.2869],[2404,-1.1624],[2428,-1.1082],[2467,-1.2828],[2496,-0.5099],[2504,-0.5658],[2512,-0.7499],[2515,-0.7501],[2520,-0.2264],[2522,-0.2655],[2524,-0.2402],[2529,-0.2766],[2532,-0.2766],[2763,-0.1533],[2775,-0.838],[2780,-1.1191],[2785,-0.9746],[2802,-0.4006],[2938,-0.3994],[2986,-0.5483],[3099,-0.5483],[3185,0.0496],[3223,-1.1082],[3263,-1.2852],[3303,-1.1082],[3338,-1.2797],[3343,-1.2832],[3345,-1.2869],[3359,-1.1624],[3383,-1.1082],[3423,-1.2852],[3427,-1.2851],[3440,-1.1537],[3464,-1.108],[3498,-1.2797],[3503,-1.2832],[3505,-1.2869],[3519,-1.1624],[3543,-1.1082],[3583,-1.2852],[3623,-1.1082],[3658,-1.2797],[3663,-1.2832],[3665,-1.2869],[3679,-1.1624],[3703,-1.1082],[3743,-1.2852],[3747,-1.2851],[3760,-1.1537],[3784,-1.108],[3818,-1.2797],[3823,-1.2832],[3825,-1.2869],[3839,-1.1624],[3863,-1.1082],[3908,-1.108],[3942,-1.2797],[3947,-1.2832],[3949,-1.2869],[3963,-1.1624],[3987,-1.1082],[4032,-1.108],[4066,-1.2797],[4071,-1.2832],[4073,-1.2869],[4087,-1.1624],[4111,-1.1082],[4157,-1.108],[4191,-1.2797],[4196,-1.2832],[4198,-1.2869],[4212,-1.1624],[4236,-1.1082],[4283,-1.108],[4317,-1.2797],[4322,-1.2832],[4324,-1.2869],[4338,-1.1624],[4360,-1.2353],[4364,-1.2571],[4372,-1.2852],[4412,-1.1082],[4435,-1.2468],[4436,0]]
       }
       pinky_03_r_ctrl.Rotation.Y {
        type=MovieSceneScriptingFloatChannel
        default=-0.7516
        thinned_from=173
        keys [[1790,-0.7516],[1828,-0.8699],[1862,-0.7501],[1868,-0.7499],[1901,-0.8652],[1902,-0.8648],[1906,-0.8691],[1912,-0.871],[1942,-0.7501],[1948,-0.7499],[1988,-0.8699],[2023,-0.7502],[2029,-0.7498],[2061,-0.8652],[2062,-0.8648],[2066,-0.8691],[2072,-0.871],[2102,-0.7501],[2108,-0.7499],[2148,-0.8699],[2182,-0.7501],[2188,-0.7499],[2221,-0.8652],[2222,-0.8648],[2232,-0.871],[2262,-0.7501],[2268,-0.7499],[2308,-0.8699],[2343,-0.7502],[2349,-0.7498],[2381,-0.8652],[2382,-0.8648],[2392,-0.871],[2422,-0.7501],[2428,-0.7499],[2466,-0.8688],[2513,0.7557],[2519,0.7212],[2522,1.0294],[2524,0.8717],[2529,1.0918],[2763,0.3356],[2772,-0.1566],[2782,-0.7457],[2801,-0.3106],[2927,-0.2617],[2938,-0.261],[2986,-0.3816],[3099,-0.3816],[3185,0.1794],[3217,-0.7501],[3223,-0.7499],[3264,-0.8699],[3297,-0.7501],[3303,-0.7499],[3336,-0.8652],[3337,-0.8648],[3347,-0.871],[3377,-0.7501],[3383,-0.7499],[3423,-0.8699],[3458,-0.7502],[3464,-0.7498],[3496,-0.8652],[3497,-0.8648],[3507,-0.871],[3537,-0.7501],[3543,-0.7499],[3584,-0.8699],[3617,-0.7501],[3623,-0.7499],[3656,-0.8652],[3657,-0.8648],[3661,-0.8691],[3667,-0.871],[3697,-0.7501],[3703,-0.7499],[3743,-0.8699],[3778,-0.7502],[3784,-0.7498],[3816,-0.8652],[3817,-0.8648],[3827,-0.871],[3857,-0.7501],[3908,-0.7498],[3940,-0.8652],[3941,-0.8648],[3951,-0.871],[3981,-0.7501],[4032,-0.7498],[4064,-0.8652],[4065,-0.8648],[4075,-0.871],[4105,-0.7501],[4157,-0.7498],[4189,-0.8652],[4190,-0.8648],[4200,-0.871],[4230,-0.7501],[4283,-0.7498],[4315,-0.8652],[4316,-0.8648],[4326,-0.871],[4356,-0.7501],[4357,-0.7505],[4360,-0.8361],[4364,-0.8524],[4372,-0.8699],[4406,-0.7501],[4412,-0.7499],[4435,-0.8492],[4436,0]]
       }
       pinky_03_r_ctrl.Rotation.Z {
        type=MovieSceneScriptingFloatChannel
        default=-10.8715
        thinned_from=100
        keys [[1790,-10.8715],[1828,5.9059],[1830,5.9059],[1865,-11.091],[1867,-11.091],[1907,5.78],[1908,5.78],[1910,6.0161],[1945,-11.091],[1947,-11.091],[1990,5.9059],[2025,-11.091],[2027,-11.091],[2067,5.78],[2068,5.78],[2070,6.0161],[2105,-11.091],[2107,-11.091],[2150,5.9059],[2185,-11.091],[2187,-11.091],[2227,5.78],[2230,6.0161],[2265,-11.091],[2267,-11.091],[2310,5.9059],[2345,-11.091],[2347,-11.091],[2390,6.0161],[2425,-11.091],[2427,-11.091],[2467,5.8027],[2484,-35.3538],[2502,-40.5029],[2513,-36.2112],[2515,-36.4002],[2520,2.6214],[2522,-3.2009],[2524,0.6524],[2530,-4.7032],[2764,14.0282],[2927,7.7087],[2938,7.5999],[2986,29.9718],[3099,29.9718],[3185,23.2571],[3220,-11.091],[3222,-11.091],[3265,5.9059],[3301,-11.091],[3342,5.78],[3343,5.78],[3346,6.0161],[3381,-11.091],[3382,-11.091],[3425,5.9059],[3460,-11.091],[3462,-11.091],[3502,5.78],[3503,5.78],[3505,6.0161],[3541,-11.091],[3585,5.9059],[3621,-11.091],[3622,-11.091],[3666,6.0161],[3700,-11.091],[3702,-11.091],[3745,5.9059],[3780,-11.091],[3782,-11.091],[3825,6.0161],[3860,-11.091],[3906,-11.091],[3949,6.0161],[3984,-11.091],[4030,-11.091],[4073,6.0161],[4108,-11.091],[4155,-11.091],[4198,6.0161],[4233,-11.091],[4281,-11.091],[4324,6.0161],[4360,0.5974],[4364,-0.3795],[4410,-11.091],[4435,2.9318],[4436,0]]
       }
       pinky_03_r_ctrl.Scale.X {type=MovieSceneScriptingFloatChannel, default=1, const [1,12]}
       pinky_03_r_ctrl.Scale.Y {type=MovieSceneScriptingFloatChannel, default=1, const [1,12]}
       pinky_03_r_ctrl.Scale.Z {type=MovieSceneScriptingFloatChannel, default=1, const [1,12]}
       hand_r_ik_ctrl.Location.X {
        type=MovieSceneScriptingFloatChannel
        default=-15.5886
        keys [[1790,-15.5886],[1815,-19.8185],[1851,-14.9348],[1896,-20.6547],[1935,-15.2269],[1975,-19.8175],[2012,-14.8933],[2056,-20.5811],[2095,-15.2248],[2135,-19.8185],[2171,-14.9347],[2216,-20.6547],[2255,-15.2269],[2295,-19.8175],[2332,-14.8933],[2376,-20.5811],[2415,-15.2248],[2455,-19.8185],[2470,-17.6903],[2502,12.127],[2504,14.1045],[2506,13.9673],[2507,14.0889],[2511,0.642],[2513,21.0289],[2515,-11.118],[2518,-23.2489],[2520,-24.5612],[2525,-22.1222],[2758,-24.9757],[2763,-27.0911],[2772,-21.9958],[2779,-22.9694],[2927,-193.0518],[2938,-122.9127],[2986,-133.3441],[3058,-133.3441],[3070,11.6013],[3083,11.6013],[3099,-166.3281],[3185,-166.3281],[3250,-19.8185],[3286,-14.9348],[3331,-20.6547],[3370,-15.2269],[3410,-19.8175],[3447,-14.8933],[3491,-20.5811],[3530,-15.2248],[3570,-19.8185],[3606,-14.9348],[3651,-20.6547],[3690,-15.2269],[3730,-19.8175],[3767,-14.8933],[3811,-20.5811],[3850,-15.2248],[3891,-14.8933],[3935,-20.5811],[3974,-15.2248],[4015,-14.8933],[4059,-20.5811],[4098,-15.2248],[4140,-14.8933],[4184,-20.5811],[4223,-15.2248],[4266,-14.8933],[4310,-20.5811],[4349,-15.2248],[4360,-21.7668],[4364,-11.1052],[4365,-19.8185],[4395,-14.9348],[4435,-20.331],[4436,-0.0036]]
       }
       hand_r_ik_ctrl.Location.Y {
        type=MovieSceneScriptingFloatChannel
        default=-10.5062
        keys [[1790,-10.5062],[1810,-9.3448],[1842,-11.84],[1891,-9.0794],[1923,-11.8527],[1970,-9.3468],[2002,-11.8284],[2051,-9.0579],[2082,-11.8567],[2130,-9.3448],[2162,-11.84],[2211,-9.0794],[2243,-11.8527],[2290,-9.3468],[2322,-11.8284],[2371,-9.0579],[2402,-11.8567],[2450,-9.3448],[2467,-10.5725],[2496,41.9539],[2506,37.8513],[2511,41.3061],[2512,37.7548],[2515,-26.4316],[2521,-16.8297],[2528,-19.2764],[2531,-19.0314],[2532,-19.0605],[2538,-19.0871],[2763,-23.2471],[2778,-8.7349],[2927,111.1523],[2938,202.7058],[2986,49.2562],[3058,49.2562],[3070,175.8678],[3083,175.8678],[3099,40.1522],[3185,40.1522],[3245,-9.3448],[3277,-11.84],[3326,-9.0794],[3358,-11.8527],[3405,-9.3468],[3437,-11.8284],[3486,-9.0579],[3517,-11.8567],[3565,-9.3448],[3597,-11.84],[3646,-9.0794],[3678,-11.8527],[3725,-9.3468],[3757,-11.8284],[3806,-9.0579],[3837,-11.8567],[3885,-9.3448],[3930,-9.0579],[3961,-11.8567],[4009,-9.3448],[4054,-9.0579],[4085,-11.8567],[4133,-9.3448],[4179,-9.0579],[4210,-11.8567],[4258,-9.3448],[4305,-9.0579],[4336,-11.8567],[4360,-36.3713],[4364,-24.876],[4386,-11.84],[4423,-9.9451],[4432,-9.1504],[4435,-9.0794],[4436,0.0017]]
       }
       hand_r_ik_ctrl.Location.Z {
        type=MovieSceneScriptingFloatChannel
        default=-14.2365
        keys [[1790,-14.2365],[1825,-10.3293],[1865,-14.4438],[1904,-9.8174],[1945,-14.4515],[1984,-10.3556],[2023,-14.1901],[2063,-9.6445],[2105,-14.4497],[2145,-10.3293],[2185,-14.4438],[2224,-9.8174],[2265,-14.4515],[2304,-10.3556],[2343,-14.1901],[2383,-9.6445],[2425,-14.4497],[2465,-10.3293],[2467,-10.361],[2471,-6.2705],[2502,61.3008],[2507,64.2682],[2511,51.3176],[2512,58.5514],[2513,53.8996],[2518,3.0687],[2519,-7.8893],[2522,-12.446],[2523,-12.4077],[2526,-12.7352],[2763,-8.1092],[2776,-21.8141],[2927,-1152.8861],[2938,-1064.9893],[2986,-919.5941],[3058,-919.5941],[3070,-964.6178],[3083,-964.6178],[3099,-1064.795],[3185,-1064.795],[3220,-14.4497],[3260,-10.3294],[3300,-14.4437],[3339,-9.8175],[3380,-14.4515],[3419,-10.3555],[3458,-14.1901],[3498,-9.6445],[3540,-14.4497],[3580,-10.3294],[3620,-14.4437],[3659,-9.8175],[3700,-14.4516],[3739,-10.3555],[3778,-14.19],[3818,-9.6445],[3860,-14.4497],[3902,-14.19],[3942,-9.6445],[3984,-14.4497],[4026,-14.19],[4066,-9.6445],[4108,-14.4497],[4151,-14.19],[4191,-9.6445],[4233,-14.4497],[4277,-14.19],[4317,-9.6445],[4360,59.9373],[4364,49.2126],[4409,-14.4437],[4422,-12.9692],[4429,-11.4297],[4434,-10.6709],[4435,-10.5417],[4436,-0.0001]]
       }
       hand_r_ik_ctrl.Rotation.X {
        type=MovieSceneScriptingFloatChannel
        default=1.5141
        thinned_from=78
        keys [[1790,1.5141],[1791,1.4962],[1836,8.2175],[1871,1.5472],[1916,8.2979],[1951,1.497],[1996,8.2077],[2031,1.545],[2076,8.2882],[2111,1.4961],[2156,8.2174],[2191,1.5472],[2236,8.2979],[2271,1.497],[2316,8.2077],[2351,1.545],[2396,8.2882],[2431,1.4961],[2454,4.78],[2465,7.1414],[2471,10.1423],[2484,43.0445],[2502,69.5113],[2504,67.8069],[2506,67.9973],[2507,67.838],[2511,82.9067],[2512,76.9808],[2513,5.0646],[2515,-114.392],[2521,-136.1249],[2529,-137.9814],[2533,-137.815],[2778,3.9069],[2785,1.2577],[2927,3.143],[2938,-54.5455],[2986,-27.9018],[3058,-27.9018],[3070,17.4125],[3083,17.4125],[3099,2.0356],[3185,2.0356],[3226,1.4961],[3271,8.2174],[3306,1.5472],[3351,8.2979],[3386,1.4969],[3431,8.2076],[3466,1.5451],[3511,8.2881],[3546,1.4961],[3591,8.2174],[3626,1.5472],[3671,8.2978],[3706,1.4969],[3751,8.2076],[3786,1.5451],[3831,8.2881],[3866,1.4961],[3910,1.5451],[3955,8.2881],[3990,1.4961],[4034,1.5451],[4079,8.2881],[4114,1.4961],[4159,1.5451],[4204,8.2881],[4239,1.4961],[4285,1.5451],[4330,8.2881],[4360,-167.8692],[4364,-163.8536],[4380,8.2174],[4415,1.5471],[4435,4.2194],[4436,-0.0004]]
       }
       hand_r_ik_ctrl.Rotation.Y {
        type=MovieSceneScriptingFloatChannel
        default=-6.6752
        keys [[1790,-6.6752],[1818,-9.765],[1847,-5.2051],[1898,-9.7489],[1927,-5.1922],[1978,-9.7473],[2010,-4.8588],[2027,-5.6826],[2039,-5.3192],[2059,-9.2165],[2086,-5.1926],[2138,-9.765],[2167,-5.2051],[2218,-9.7489],[2247,-5.1922],[2298,-9.7473],[2330,-4.8588],[2347,-5.6826],[2359,-5.3191],[2379,-9.2165],[2406,-5.1926],[2458,-9.765],[2466,-8.8024],[2470,-5.6687],[2484,57.2513],[2496,32.754],[2502,17.9263],[2504,23.7371],[2506,23.6517],[2507,23.9808],[2511,-17.5157],[2512,12.5543],[2513,87.2648],[2515,43.2291],[2519,27.7821],[2522,21.6787],[2778,-12.2298],[2927,2.5852],[2938,-18.6676],[2986,9.6627],[3058,9.6627],[3070,-40.5415],[3083,-40.5415],[3099,-12.8523],[3185,-12.8523],[3253,-9.7649],[3282,-5.2052],[3333,-9.7489],[3362,-5.1922],[3413,-9.7473],[3445,-4.8589],[3462,-5.6826],[3474,-5.3192],[3494,-9.2164],[3521,-5.1926],[3573,-9.765],[3602,-5.2052],[3653,-9.7489],[3682,-5.1923],[3733,-9.7473],[3765,-4.8589],[3782,-5.6826],[3794,-5.3192],[3814,-9.2164],[3841,-5.1926],[3906,-5.6826],[3918,-5.3192],[3938,-9.2164],[3965,-5.1926],[4030,-5.6826],[4042,-5.3192],[4062,-9.2164],[4089,-5.1926],[4155,-5.6826],[4167,-5.3192],[4187,-9.2164],[4214,-5.1926],[4281,-5.6826],[4293,-5.3192],[4313,-9.2164],[4340,-5.1926],[4360,35.9492],[4364,25.3743],[4391,-5.2052],[4435,-9.2626],[4436,0.0024]]
       }
       hand_r_ik_ctrl.Rotation.Z {
        type=MovieSceneScriptingFloatChannel
        default=-17.4892
        thinned_from=120
        keys [[1790,-17.4892],[1791,-17.46],[1797,-17.5918],[1818,-14.037],[1857,-18.2888],[1871,-17.5908],[1877,-17.8026],[1899,-14.1215],[1936,-18.3419],[1951,-17.46],[1957,-17.592],[1979,-14.1307],[2018,-18.2293],[2031,-17.4865],[2036,-17.6366],[2059,-14.0378],[2096,-18.341],[2111,-17.46],[2117,-17.5918],[2138,-14.037],[2177,-18.2888],[2191,-17.5908],[2197,-17.8026],[2219,-14.1215],[2256,-18.3419],[2271,-17.46],[2277,-17.592],[2299,-14.1307],[2338,-18.2293],[2351,-17.4865],[2356,-17.6366],[2379,-14.0378],[2416,-18.341],[2431,-17.46],[2437,-17.5918],[2458,-14.037],[2501,155.3443],[2505,151.8534],[2507,151.7173],[2510,159.6158],[2512,149.4012],[2513,51.7289],[2515,-87.0323],[2518,-99.5268],[2778,341.2352],[2927,208.4592],[2938,128.5459],[2986,278.0994],[3058,278.0994],[3070,182.4284],[3083,182.4284],[3099,200.3904],[3185,200.3904],[3226,342.5399],[3232,342.4082],[3253,345.963],[3292,341.7112],[3306,342.4092],[3312,342.1974],[3334,345.8785],[3371,341.6581],[3386,342.5399],[3392,342.408],[3414,345.8694],[3453,341.7707],[3471,342.3634],[3494,345.9622],[3531,341.659],[3546,342.5399],[3552,342.4082],[3573,345.963],[3612,341.7112],[3626,342.4092],[3632,342.1974],[3654,345.8785],[3691,341.6581],[3706,342.5399],[3712,342.408],[3734,345.8694],[3773,341.7707],[3791,342.3634],[3814,345.9622],[3851,341.659],[3866,342.5399],[3897,341.7707],[3915,342.3634],[3938,345.9622],[3975,341.659],[3990,342.5399],[4021,341.7707],[4039,342.3634],[4062,345.9622],[4099,341.659],[4114,342.5399],[4146,341.7707],[4164,342.3634],[4187,345.9622],[4224,341.659],[4239,342.5399],[4272,341.7707],[4290,342.3634],[4313,345.9622],[4350,341.659],[4360,353.7234],[4364,333.4854],[4401,341.7112],[4415,342.4092],[4421,342.1974],[4436,360.0005]]
       }
       hand_r_ik_ctrl.Scale.X {
        type=MovieSceneScriptingFloatChannel
        default=1
        thinned_from=12
        keys [[1790,1],[2927,1.0493],[2938,1.042],[3058,1.0418],[3070,1.0356],[3083,1.0356],[3099,1.0443],[3185,1.0443],[4364,0.9972],[4436,1]]
       }
       hand_r_ik_ctrl.Scale.Y {
        type=MovieSceneScriptingFloatChannel
        default=1
        thinned_from=12
        keys [[1790,1],[2927,1.0493],[2938,1.042],[3058,1.0418],[3070,1.0356],[3083,1.0356],[3099,1.0443],[3185,1.0443],[4364,0.9972],[4436,1]]
       }
       hand_r_ik_ctrl.Scale.Z {
        type=MovieSceneScriptingFloatChannel
        default=1
        thinned_from=12
        keys [[1790,1],[2927,1.0493],[2938,1.042],[3058,1.0418],[3070,1.0356],[3083,1.0356],[3099,1.0443],[3185,1.0443],[4364,0.9972],[4436,1]]
       }
       Softness_2 {type=MovieSceneScriptingFloatChannel, default=0, const [0,12]}
       hand_r_ikorient_ctrl.Rotation.Roll {type=MovieSceneScriptingFloatChannel, default=0, const [0,12]}
       hand_r_ikorient_ctrl.Rotation.Pitch {type=MovieSceneScriptingFloatChannel, default=0, const [0,12]}
       hand_r_ikorient_ctrl.Rotation.Yaw {type=MovieSceneScriptingFloatChannel, default=0, const [0,12]}
       arm_r_pv_ik_ctrl.Location.X {
        type=MovieSceneScriptingFloatChannel
        default=-3.8966
        keys [[1790,-3.8966],[1793,-3.9473],[1838,0.7012],[1874,-4.2012],[1919,0.3861],[1953,-3.9444],[1998,0.6934],[2034,-4.2102],[2079,0.3848],[2113,-3.9473],[2158,0.7012],[2194,-4.2012],[2239,0.3861],[2273,-3.9444],[2318,0.6934],[2354,-4.2102],[2399,0.3848],[2433,-3.9472],[2453,-2.6777],[2463,-0.7853],[2471,2.6111],[2496,-9.5857],[2501,-10.6998],[2506,-8.3574],[2511,-16.8763],[2513,10.9892],[2520,-18.8024],[2524,-12.1269],[2536,-14.5923],[2764,-10.5424],[2766,-10.5885],[2776,-6.3012],[2787,-9.8593],[2927,-4.856],[2938,-78.6373],[2986,47.0074],[3058,47.0074],[3070,-66.9503],[3083,-66.9503],[3099,25.8287],[3185,25.8287],[3228,-3.9473],[3273,0.7012],[3309,-4.2012],[3354,0.3862],[3388,-3.9444],[3433,0.6933],[3469,-4.2102],[3514,0.3848],[3548,-3.9473],[3593,0.7012],[3629,-4.2012],[3674,0.3861],[3708,-3.9444],[3753,0.6934],[3789,-4.2102],[3834,0.3848],[3868,-3.9473],[3913,-4.2102],[3958,0.3848],[3992,-3.9473],[4037,-4.2102],[4082,0.3848],[4116,-3.9473],[4162,-4.2102],[4207,0.3848],[4241,-3.9473],[4288,-4.2102],[4333,0.3848],[4360,5.0703],[4364,11.8948],[4382,0.7012],[4418,-4.2012],[4436,-0.0022]]
       }
       arm_r_pv_ik_ctrl.Location.Y {
        type=MovieSceneScriptingFloatChannel
        default=-16.5167
        thinned_from=70
        keys [[1790,-16.5167],[1825,-13.4417],[1864,-16.6725],[1904,-13.3867],[1944,-16.677],[1984,-13.469],[2024,-16.6913],[2064,-13.3782],[2104,-16.6777],[2145,-13.4417],[2184,-16.6725],[2224,-13.3867],[2264,-16.677],[2304,-13.469],[2344,-16.6913],[2384,-13.3782],[2424,-16.6777],[2465,-13.4417],[2467,-13.4782],[2502,91.7717],[2504,90.7445],[2511,98.5906],[2518,-3.4101],[2523,-16.5653],[2536,-14.511],[2772,-22.638],[2927,-1121.8218],[2938,-993.4785],[2986,-867.4564],[3058,-867.4564],[3070,-853.2388],[3083,-853.2388],[3099,-1026.7515],[3185,-1026.7515],[3219,-16.6777],[3260,-13.4417],[3299,-16.6726],[3339,-13.3868],[3379,-16.677],[3419,-13.469],[3459,-16.6914],[3499,-13.3782],[3539,-16.6777],[3580,-13.4417],[3619,-16.6726],[3659,-13.3867],[3699,-16.677],[3739,-13.469],[3779,-16.6913],[3819,-13.3782],[3859,-16.6777],[3903,-16.6913],[3943,-13.3782],[3983,-16.6777],[4027,-16.6913],[4067,-13.3782],[4107,-16.6777],[4152,-16.6913],[4192,-13.3782],[4232,-16.6777],[4278,-16.6913],[4318,-13.3782],[4360,48.9864],[4364,40.79],[4408,-16.6726],[4422,-15.7879],[4429,-14.9292],[4435,-14.2921],[4436,-0.0018]]
       }
       arm_r_pv_ik_ctrl.Location.Z {
        type=MovieSceneScriptingFloatChannel
        default=15.0396
        keys [[1790,15.0396],[1818,21.7962],[1863,14.5622],[1899,22.6624],[1944,14.652],[1978,21.7963],[2023,14.558],[2059,22.6615],[2104,14.6528],[2138,21.7962],[2183,14.5622],[2219,22.6624],[2264,14.652],[2298,21.7963],[2343,14.558],[2379,22.6615],[2424,14.6528],[2458,21.7962],[2471,20.6091],[2502,-40.8419],[2505,-40.1105],[2511,-32.6349],[2512,-39.7023],[2515,-5.3052],[2519,-10.6686],[2520,-10.2161],[2524,-12.6376],[2777,25.0424],[2784,23.7001],[2927,-356.0189],[2938,-393.3209],[2986,-322.8185],[3058,-322.8185],[3070,-493.8365],[3083,-493.8365],[3099,-333.2554],[3185,-333.2554],[3219,14.6528],[3253,21.7962],[3298,14.5622],[3334,22.6624],[3379,14.652],[3413,21.7963],[3458,14.558],[3494,22.6615],[3539,14.6528],[3573,21.7962],[3618,14.5622],[3654,22.6625],[3699,14.652],[3733,21.7962],[3778,14.558],[3814,22.6615],[3859,14.6528],[3902,14.558],[3938,22.6615],[3983,14.6528],[4026,14.558],[4062,22.6615],[4107,14.6528],[4151,14.558],[4187,22.6615],[4232,14.6528],[4277,14.558],[4313,22.6615],[4360,13.703],[4364,-0.2736],[4407,14.5622],[4435,21.9129],[4436,0.0015]]
       }
       arm_r_pv_ik_ctrl.Rotation.X {
        type=MovieSceneScriptingFloatChannel
        default=11.5684
        thinned_from=73
        keys [[1790,11.5684],[1814,13.234],[1865,11.4307],[1894,13.2351],[1944,11.4392],[1974,13.2325],[2025,11.4314],[2054,13.2332],[2104,11.4402],[2134,13.234],[2185,11.4307],[2214,13.235],[2264,11.4391],[2294,13.2325],[2345,11.4314],[2374,13.2332],[2424,11.4402],[2454,13.234],[2484,-33.5144],[2502,-53.7977],[2505,-50.9182],[2507,-50.8736],[2511,-78.1348],[2513,-32.6207],[2515,-38.6203],[2520,-17.0864],[2528,-23.9505],[2536,-20.7325],[2766,-5.0838],[2777,26.947],[2787,21.7782],[2927,21.1909],[2938,133.2108],[2986,-7.5689],[3058,-7.5689],[3070,-19.7472],[3083,-19.7472],[3099,-28.8693],[3185,-28.8693],[3219,11.4402],[3249,13.234],[3300,11.4307],[3329,13.2351],[3379,11.4391],[3409,13.2324],[3460,11.4314],[3489,13.2331],[3539,11.4402],[3569,13.234],[3620,11.4308],[3649,13.2351],[3699,11.4391],[3729,13.2325],[3780,11.4314],[3809,13.2332],[3859,11.4402],[3904,11.4314],[3933,13.2332],[3983,11.4402],[4028,11.4314],[4057,13.2332],[4107,11.4402],[4153,11.4314],[4182,13.2332],[4232,11.4402],[4279,11.4314],[4308,13.2332],[4360,-16.1519],[4364,4.3008],[4409,11.4307],[4435,13.1953],[4436,-0.0006]]
       }
       arm_r_pv_ik_ctrl.Rotation.Y {
        type=MovieSceneScriptingFloatChannel
        default=19.1751
        keys [[1790,19.1751],[1816,11.9661],[1857,20.1795],[1896,11.9559],[1936,20.2319],[1976,11.9641],[2017,20.1799],[2057,11.9557],[2096,20.2315],[2136,11.9661],[2177,20.1794],[2216,11.9559],[2256,20.2319],[2296,11.9641],[2337,20.1799],[2377,11.9557],[2416,20.2315],[2456,11.9661],[2465,12.7139],[2469,14.3241],[2496,19.9115],[2502,28.339],[2504,27.4058],[2506,27.4677],[2507,27.4177],[2511,39.4158],[2512,33.0374],[2513,27.8497],[2518,10.5125],[2520,4.8078],[2524,10.1933],[2767,26.9666],[2777,17.0306],[2927,21.9341],[2938,33.5427],[2986,37.9731],[3058,37.9731],[3070,-7.7151],[3083,-7.7151],[3099,17.826],[3185,17.826],[3251,11.9661],[3292,20.1794],[3331,11.9559],[3371,20.2319],[3411,11.9641],[3452,20.1798],[3492,11.9556],[3531,20.2315],[3571,11.9661],[3612,20.1794],[3651,11.9559],[3691,20.2319],[3731,11.9641],[3772,20.1798],[3812,11.9556],[3851,20.2315],[3896,20.1798],[3936,11.9556],[3975,20.2315],[4020,20.1798],[4060,11.9556],[4099,20.2315],[4145,20.1798],[4185,11.9556],[4224,20.2315],[4271,20.1798],[4311,11.9556],[4350,20.2315],[4360,7.0023],[4364,5.8303],[4366,11.9661],[4401,20.1794],[4418,18.4743],[4427,15.1985],[4431,13.5504],[4433,12.9119],[4434,12.6562],[4435,12.4427],[4436,-0.0015]]
       }
       arm_r_pv_ik_ctrl.Rotation.Z {
        type=MovieSceneScriptingFloatChannel
        default=-13.8375
        thinned_from=77
        keys [[1790,-13.8375],[1829,-5.129],[1869,-13.8602],[1909,-5.1371],[1949,-13.8433],[1989,-5.1307],[2029,-13.8589],[2069,-5.1383],[2109,-13.8429],[2149,-5.129],[2189,-13.8602],[2229,-5.1372],[2269,-13.8433],[2309,-5.1307],[2349,-13.8589],[2389,-5.1383],[2429,-13.8429],[2460,-6.1055],[2467,-5.1713],[2471,-3.2639],[2496,40.5176],[2506,37.9428],[2507,38.0422],[2511,28.3857],[2512,30.5018],[2515,-39.267],[2520,-64.9848],[2524,-58.1602],[2527,-59.0587],[2528,-59.039],[2531,-59.358],[2536,-59.2956],[2765,-48.1933],[2776,-12.6242],[2787,-20.1482],[2927,-33.111],[2938,115.7128],[2986,-58.468],[3058,-58.468],[3070,44.9855],[3083,44.9855],[3099,34.1705],[3185,34.1705],[3224,-13.8429],[3264,-5.129],[3304,-13.8603],[3344,-5.1371],[3384,-13.8432],[3424,-5.1307],[3464,-13.8589],[3504,-5.1383],[3544,-13.8429],[3584,-5.129],[3624,-13.8603],[3664,-5.1371],[3704,-13.8432],[3744,-5.1308],[3784,-13.8589],[3824,-5.1383],[3864,-13.8429],[3908,-13.8589],[3948,-5.1383],[3988,-13.8429],[4032,-13.8589],[4072,-5.1383],[4112,-13.8429],[4157,-13.8589],[4197,-5.1383],[4237,-13.8429],[4283,-13.8589],[4323,-5.1383],[4360,38.9004],[4364,28.438],[4373,-5.129],[4413,-13.8603],[4436,-0.0002]]
       }
       arm_r_pv_ik_ctrl.Scale.X {type=MovieSceneScriptingFloatChannel, default=1, const [1,12]}
       arm_r_pv_ik_ctrl.Scale.Y {type=MovieSceneScriptingFloatChannel, default=1, const [1,12]}
       arm_r_pv_ik_ctrl.Scale.Z {type=MovieSceneScriptingFloatChannel, default=1, const [1,12]}
       body_ctrl.Location.X {
        type=MovieSceneScriptingFloatChannel
        default=3.2168
        thinned_from=226
        keys [[1780,-95.979],[1800,-95.8895],[1817,-96.571],[1832,-97.3987],[1838,-97.507],[1859,-96.6675],[1868,-96.1483],[1881,-95.7675],[1888,-95.9196],[1912,-97.2746],[1918,-97.3887],[1930,-97.0788],[1948,-96.1562],[1959,-95.8905],[1977,-96.6054],[1992,-97.4034],[1998,-97.5063],[2019,-96.6678],[2028,-96.1491],[2041,-95.7686],[2048,-95.9207],[2072,-97.2756],[2078,-97.3897],[2090,-97.0922],[2108,-96.1869],[2120,-95.8895],[2137,-96.571],[2152,-97.3987],[2158,-97.507],[2179,-96.6675],[2188,-96.1483],[2201,-95.7675],[2208,-95.9196],[2232,-97.2746],[2238,-97.3887],[2250,-96.998],[2260,-95.2961],[2268,-93.6246],[2279,-92.8216],[2290,-94.6193],[2312,-100.5765],[2318,-101.728],[2339,-103.391],[2348,-103.9512],[2361,-104.3906],[2368,-104.2736],[2392,-103.0544],[2398,-102.969],[2410,-103.3153],[2428,-104.281],[2440,-104.6237],[2465,-103.5941],[2471,-103.8004],[2484,-108.1729],[2496,-103.0842],[2501,-102.5608],[2502,-102.1074],[2504,-101.858],[2507,-101.8385],[2511,-102.2963],[2513,-101.8804],[2515,-101.0829],[2521,-99.9368],[2522,-99.8992],[2525,-100.4005],[2529,-101.2974],[2532,-101.2421],[2536,-100.6708],[2762,-105.1595],[2763,-105.7461],[2767,-107.4996],[2774,-108.5099],[2783,-109.0447],[2789,-108.7377],[2938,-110.6712],[2967,-194.0676],[2985,-302.2244],[3003,-333.4589],[3051,-353.0249],[3059,-354.2651],[3099,-354.3017],[3185,-392.0059],[3252,-405.0332],[3267,-404.1841],[3303,-408.2695],[3316,-407.6404],[3333,-405.0269],[3347,-404.187],[3375,-407.7909],[3383,-408.2697],[3394,-407.6853],[3412,-404.9559],[3427,-404.174],[3463,-408.2677],[3476,-407.6378],[3493,-405.0244],[3507,-404.1846],[3535,-407.7264],[3543,-408.2002],[3555,-407.5747],[3572,-405.0332],[3587,-404.1841],[3623,-408.2695],[3636,-407.6404],[3653,-405.0269],[3667,-404.187],[4258,-405.296],[4364,-406.1822],[4366,-1382.8905],[4492,-122248.6953]]
       }
       body_ctrl.Location.Y {
        type=MovieSceneScriptingFloatChannel
        default=-3.5392
        thinned_from=214
        keys [[1780,-6325.2803],[1868,-6325.3442],[1881,-6312.0674],[1898,-6243.8115],[1912,-6165.1641],[1930,-6015.3838],[1948,-5815.5483],[1959,-5672.1089],[1977,-5407.9614],[1998,-5061.4814],[2028,-4516.8369],[2108,-3010.1809],[2137,-2486.033],[2158,-2123.2698],[2173,-1875.7675],[2188,-1640.3773],[2201,-1447.816],[2208,-1349.0763],[2218,-1214.4746],[2232,-1039.4468],[2238,-969.4916],[2250,-839.4124],[2260,-740.8807],[2268,-668.3189],[2279,-578.47],[2290,-502.8835],[2295,-473.9141],[2297,-463.3055],[2310,-408.1261],[2312,-401.7564],[2318,-385.9872],[2333,-368.2158],[2460,-368.0267],[2471,-366.5338],[2484,-339.0817],[2496,-330.3972],[2502,-331.4409],[2507,-333.5667],[2513,-348.3446],[2519,-374.0986],[2521,-376.3812],[2762,-371.0311],[2767,-367.4387],[2774,-365.7048],[2789,-364.0469],[2938,-360.5478],[2945,-359.654],[2968,-165.0058],[2991,101.036],[3018,142.7265],[3052,153.2774],[3057,154.0513],[3060.3925,147.6183],[3063.9125,68.7657],[3071.3925,-190.4319],[3079,-240.1545],[3083,-222.7762],[3092,-121.0144],[3102,-85.2675],[3185,10.664],[3188,8.8781],[3201,-9.4411],[3228,-32.9712],[3303,11.4782],[3315,25.7697],[3316,26.0665],[3323,25.279],[3330,24.6713],[3333,24.8437],[3345,27.5273],[3353,29.8653],[3365,32.5673],[3373,33.0877],[3375,32.9838],[3383,31.4829],[3394,28.0133],[3405,25.577],[3410,25.393],[3412,25.5308],[3425,28.4083],[3433,30.6653],[3448,33.3596],[3453,33.3647],[3454,33.2924],[3463,31.4279],[3476,27.0273],[3483,25.2801],[3490,24.6723],[3493,24.8447],[3505,27.528],[3513,29.866],[3525,32.5781],[3533,33.1088],[3535,33.0069],[3543,31.5079],[3555,27.7024],[3565,25.5462],[3570,25.3623],[3572,25.5018],[3585,28.4005],[3593,30.6649],[3608,33.3593],[3613,33.3644],[3614,33.2921],[3623,31.4272],[3636,27.0263],[3643,25.279],[3650,24.6713],[3653,24.8437],[3752,32.2766],[3887,38.635],[4260,38.635],[4364,40.9735],[4366,-132.352],[4491,-21800.877]]
       }
       body_ctrl.Location.Z {
        type=MovieSceneScriptingFloatChannel
        default=0.0816
        thinned_from=253
        keys [[1780,1125.9796],[1800,1128.9688],[1817,1135.6248],[1853,1125.4495],[1859,1128.2477],[1868,1136.6006],[1898,1169.2327],[1940,1189.5717],[1977,1228.0194],[2019,1248.9216],[2058,1289.8992],[2100,1310.238],[2137,1348.6913],[2152,1356.8896],[2158,1362.9686],[2210,1466.6299],[2291,2505.2976],[2314,2695.7419],[2344,2764.187],[2368,2782.5366],[2378,2784.428],[2420,2763.1985],[2460,2783.4348],[2468,2780.855],[2484,2740.9587],[2496,2729.6809],[2501,2736.0149],[2502,2743.5771],[2507,2749.3906],[2511,2742.4568],[2515,2774.4924],[2519,2788.5457],[2524,2782.2693],[2536,2807.9316],[2762,2973.5149],[2786,2935.134],[2938,2980.4756],[2958,2957.197],[2965,2965.1941],[2984,3218.874],[3017,3300.6162],[3049,3314.7375],[3063.9125,3309.9661],[3080,3280.9092],[3095,3305.5935],[3185,3378.6531],[3315,3842.408],[3345,3960.9907],[3394,4087.9009],[3412,4094.7524],[3454,4070.2368],[3493,4096.5547],[3535,4070.7346],[3572,4094.7522],[3614,4070.2351],[3651,4096.0942],[3656,4093.5422],[3670,4064.8271],[3682,4055.2871],[3689,4057.2456],[3716,4096.5586],[3732,4108.9033],[3745,4100.4209],[3777,4009.707],[3784,4001.8174],[3815,4019.5166],[3851,3998.3479],[3886,4021.2483],[3911,3997.3577],[3939,4019.5166],[3976,3996.7488],[4009,4019.3718],[4032,3997.769],[4063,4019.0483],[4103,3996.7488],[4133,4019.0608],[4160,3998.0166],[4188,4018.584],[4228,3996.7488],[4258,4018.439],[4282,3997.2021],[4312,4019.5166],[4355,3988.6157],[4364,3996.7817],[4366,4749.5112],[4369,7416.6484],[4379,12697.0996],[4385,13452.0264],[4394,12841.0146],[4427,-1048.5585],[4507,-1126.9403],[4522,34169.3477]]
       }
       body_ctrl.Rotation.X {
        type=MovieSceneScriptingFloatChannel
        default=-2.5628
        thinned_from=192
        keys [[1780,-2.5628],[1799,-2.3939],[1800,-2.3583],[1810,-1.9085],[1815,-1.6682],[1817,-1.579],[1830,-1.2541],[1832,-1.2646],[1838,-1.4038],[1853,-2.1127],[1858,-2.3394],[1859,-2.3782],[1868,-2.5638],[1880,-2.2941],[1881,-2.2535],[1888,-1.9322],[1895,-1.6002],[1898,-1.4733],[1910,-1.2037],[1912,-1.2146],[1918,-1.3598],[1930,-1.9407],[1938,-2.3358],[1940,-2.4137],[1948,-2.5699],[1959,-2.3466],[1970,-1.8656],[1975,-1.6385],[1977,-1.5551],[1990,-1.2552],[1992,-1.2657],[1998,-1.4058],[2013,-2.1192],[2018,-2.3474],[2019,-2.3864],[2028,-2.5732],[2040,-2.3016],[2041,-2.2607],[2048,-1.9372],[2055,-1.6028],[2058,-1.475],[2070,-1.2035],[2072,-1.2144],[2078,-1.3596],[2090,-1.9405],[2098,-2.3357],[2100,-2.4136],[2108,-2.5698],[2119,-2.3463],[2120,-2.3089],[2130,-1.865],[2135,-1.6377],[2137,-1.5542],[2150,-1.2541],[2152,-1.2646],[2158,-1.4038],[2173,-2.1127],[2178,-2.3394],[2179,-2.3782],[2188,-2.5638],[2200,-2.2941],[2201,-2.2535],[2208,-1.9322],[2256,-1.6032],[2333,-2.1192],[2338,-2.3474],[2339,-2.3864],[2348,-2.5732],[2360,-2.3016],[2361,-2.2607],[2368,-1.9372],[2375,-1.6028],[2378,-1.475],[2390,-1.2035],[2392,-1.2144],[2398,-1.3596],[2410,-1.9405],[2418,-2.3357],[2420,-2.4136],[2428,-2.5698],[2439,-2.4361],[2440,-2.4131],[2450,-2.123],[2454,-1.9904],[2455,-1.952],[2457,-1.8472],[2460,-1.6271],[2465,-1.1262],[2468,-0.7675],[2469,-0.6412],[2470,-0.5124],[2471,-0.3815],[2484,25.0918],[2496,24.1381],[2503,24.2489],[2512,22.1655],[2515,5.7586],[2518,-6.9651],[2519,-9.891],[2520,-11.221],[2521,-11.1352],[2522,-11.3511],[2524,-11.9678],[2525,-11.8746],[2527,-11.2494],[2529,-10.305],[2530,-9.8128],[2531,-9.3605],[2532,-8.988],[2535,-8.6738],[2536,-8.753],[2762,-3.2537],[2763,-2.7703],[2767,-1.6667],[2769,-1.1296],[2772,-0.3588],[2773,-0.1089],[2774,0.1232],[2782,1.6693],[2783,1.8075],[2785,2.039],[2789,2.3023],[2927,1.8256],[2938,1.7979],[2986,11.171],[3058,11.171],[3070,8.5987],[3185,8.5987],[3316,8.1975],[3331,8.1185],[3365,8.8475],[3373,8.4426],[3375,8.3623],[3383,8.2012],[3394,8.4287],[3412,9.2283],[3425,9.5284],[3427,9.518],[3433,9.3787],[3454,8.391],[3463,8.2013],[3475,7.9136],[3490,8.6192],[3493,8.7477],[3505,9.0228],[3507,9.0124],[3513,8.8683],[3525,8.2832],[3533,7.8783],[3535,7.798],[3543,7.637],[3554,7.8647],[3572,8.6648],[3585,8.965],[3587,8.9547],[3593,8.8161],[3614,7.8346],[3623,7.6464],[3635,7.9212],[3650,8.6218],[3653,8.7494],[3665,9.0227],[3667,9.0123],[3673,8.8683],[3787,8.2297],[4286,8.2297],[4360,4.6707],[4366,4.4234],[4399,-38.3437],[4411,-82.2715],[4495,-82.354]]
       }
       body_ctrl.Rotation.Y {
        type=MovieSceneScriptingFloatChannel
        default=-1.1155
        thinned_from=129
        keys [[1780,-1.1155],[1799,-0.5922],[1800,-0.5003],[1810,0.1149],[1815,0.0196],[1817,-0.0656],[1830,-1.0259],[1832,-1.2034],[1838,-1.7191],[1853,-2.4329],[1858,-2.3293],[1859,-2.2869],[1868,-1.6969],[1880,-0.8113],[1881,-0.7573],[1888,-0.5629],[1895,-0.7176],[1898,-0.8618],[1910,-1.6821],[1912,-1.8299],[1918,-2.238],[1930,-2.6522],[1938,-2.3641],[1940,-2.2193],[1948,-1.474],[1959,-0.3949],[1970,0.1184],[1975,0.0221],[1977,-0.0641],[1990,-1.0348],[1992,-1.2143],[1998,-1.7356],[2013,-2.4572],[2018,-2.3518],[2019,-2.3087],[2028,-1.7085],[2040,-0.8077],[2041,-0.7528],[2048,-0.5549],[2055,-0.7097],[2058,-0.8538],[2070,-1.6741],[2072,-1.8218],[2078,-2.2299],[2090,-2.6441],[2098,-2.3571],[2100,-2.213],[2108,-1.4708],[2119,-0.3963],[2120,-0.3162],[2130,0.1149],[2135,0.0196],[2137,-0.0656],[2150,-1.0259],[2152,-1.2034],[2158,-1.7191],[2173,-2.4329],[2178,-2.3293],[2179,-2.2869],[2188,-1.6969],[2200,-0.8113],[2201,-0.7573],[2208,-0.5629],[2215,-0.7176],[2218,-0.8618],[2230,-1.6821],[2232,-1.8299],[2281,-1.9129],[2333,-2.4572],[2338,-2.3518],[2339,-2.3087],[2348,-1.7085],[2360,-0.8077],[2361,-0.7528],[2368,-0.5549],[2375,-0.7097],[2378,-0.8538],[2390,-1.6741],[2392,-1.8218],[2398,-2.2299],[2410,-2.6441],[2418,-2.3571],[2420,-2.213],[2428,-1.4708],[2439,-0.3963],[2440,-0.3162],[2450,0.1149],[2454,0.044],[2455,0.0068],[2457,-0.0864],[2460,-0.2635],[2465,-0.615],[2468,-1.2166],[2469,-1.5494],[2470,-1.9248],[2474,-2.3903],[2479,-2.1417],[2503,3.6044],[2515,1.8506],[2536,0.8729],[2762,1.3137],[2763,1.1942],[2767,0.6222],[2769,0.306],[2772,-0.1636],[2773,-0.3175],[2774,-0.4516],[2782,-1.0752],[2785,-1.0729],[2789,-1.0634],[2927,-1.0454],[2938,-1.0462],[2986,-4.4823],[3058,-4.4823],[3062,-4.8921],[3075,11.0561],[3083,14.1874],[3185,14.1874],[3234,-1.5893],[4332,-1.6103],[4359,-1.2132],[4364,-0.0221],[4365,12.025],[4366,24.9493],[4369,28.6073],[4393,30.9978]]
       }
       body_ctrl.Rotation.Z {
        type=MovieSceneScriptingFloatChannel
        default=3.2484
        thinned_from=143
        keys [[1780,3.2484],[1799,2.6417],[1800,2.5257],[1810,1.4088],[1815,1.1953],[1817,1.2122],[1830,1.9499],[1832,2.1254],[1838,2.6827],[1853,3.788],[1858,3.8888],[1859,3.883],[1868,3.4043],[1881,2.0583],[1888,1.4453],[1895,1.1922],[1898,1.2298],[1910,1.9484],[1912,2.1243],[1918,2.6828],[1930,3.6458],[1938,3.8913],[1940,3.8685],[1948,3.407],[1959,2.2721],[1970,1.3303],[1975,1.1959],[1977,1.2129],[1990,1.9511],[1992,2.1267],[1998,2.6845],[2013,3.7906],[2018,3.8914],[2019,3.8856],[2028,3.4066],[2041,2.0593],[2048,1.4457],[2055,1.1925],[2058,1.23],[2070,1.9486],[2072,2.1245],[2078,2.683],[2090,3.646],[2098,3.8915],[2100,3.8687],[2108,3.4071],[2119,2.2718],[2120,2.1654],[2130,1.3297],[2135,1.1953],[2137,1.2122],[2150,1.9499],[2152,2.1254],[2158,2.6827],[2173,3.788],[2178,3.8888],[2179,3.883],[2188,3.4043],[2201,2.0583],[2208,1.4453],[2215,1.1922],[2218,1.2298],[2230,1.9484],[2232,2.1243],[2238,2.6828],[2250,4.3812],[2268,33.9194],[2279,65.2819],[2310,160.5405],[2312,165.0638],[2318,176.0394],[2329,185.1976],[2332,185.4905],[2368,183.3115],[2378,183.5489],[2410,187.4056],[2420,188.0749],[2428,187.9691],[2450,186.8626],[2468,188.0142],[2484,187.0737],[2496,185.4843],[2502,185.0528],[2507,186.0548],[2511,184.2187],[2512,184.6306],[2513,185.6483],[2515,187.0697],[2518,188.6931],[2520,191.7377],[2521,192.7552],[2522,193.1669],[2525,191.4942],[2531,193.1666],[2532,193.1359],[2535,193.834],[2762,200.6223],[2938,200.5674],[2986,227.6363],[3058,227.6363],[3063.9125,216.9572],[3070,175.3272],[3185,175.3272],[3200,176.9724],[3243,202.6947],[4366,204.4259],[4395,178.3417],[4397,177.3058],[4400,176.8419]]
       }
       body_ctrl.Scale.X {
        type=MovieSceneScriptingFloatChannel
        default=1
        thinned_from=344
        keys [[1780,1.25],[2520,1.25],[2536,1.5211],[2762,2.9243],[2763,2.9428],[2767,2.993],[2769,3],[4436,3]]
       }
       body_ctrl.Scale.Y {
        type=MovieSceneScriptingFloatChannel
        default=1
        thinned_from=344
        keys [[1780,1.25],[2520,1.25],[2536,1.5211],[2762,2.9243],[2763,2.9428],[2767,2.993],[2769,3],[4436,3]]
       }
       body_ctrl.Scale.Z {
        type=MovieSceneScriptingFloatChannel
        default=1
        thinned_from=344
        keys [[1780,1.25],[2520,1.25],[2536,1.5211],[2762,2.9243],[2763,2.9428],[2767,2.993],[2769,3],[4436,3]]
       }
       hips_ctrl.Location.X {type=MovieSceneScriptingFloatChannel, default=0, thinned_from=1296, keys [[1790,0],[4373,0]]}
       hips_ctrl.Location.Y {type=MovieSceneScriptingFloatChannel, default=0, thinned_from=1220, keys [[1790,0],[4372,0]]}
       hips_ctrl.Location.Z {type=MovieSceneScriptingFloatChannel, default=0, thinned_from=1171, keys [[1790,0],[4372,0]]}
       hips_ctrl.Rotation.X {
        type=MovieSceneScriptingFloatChannel
        default=-4.618
        thinned_from=81
        keys [[1790,-4.618],[1830,-5.8084],[1868,-4.6106],[1910,-5.8106],[1948,-4.6102],[1990,-5.8072],[2028,-4.5997],[2070,-5.8108],[2108,-4.6103],[2150,-5.8084],[2188,-4.6106],[2230,-5.8106],[2268,-4.6102],[2310,-5.8072],[2348,-4.5997],[2390,-5.8108],[2428,-4.6103],[2464,-5.6516],[2496,-15.9141],[2502,-18.5],[2509,-30.9427],[2511,-30.093],[2513,-9.3571],[2518,9.4779],[2520,7.0724],[2524,11.2395],[2531,8.9129],[2766,0.6691],[2777,-2.089],[2800,-6.5241],[2938,-6.4112],[2964,-52.1746],[2972,-36.1872],[2978,-25.362],[2992,-17.8515],[3014,-16.3333],[3058,-15.1213],[3060,-9.923],[3064.8212,11.2982],[3077,-6.0071],[3185,-6.0071],[3205,-5.3791],[3209,-5.2829],[3223,-4.6103],[3265,-5.8084],[3303,-4.6106],[3345,-5.8106],[3383,-4.6103],[3425,-5.8072],[3463,-4.5997],[3505,-5.8108],[3543,-4.6103],[3585,-5.8084],[3623,-4.6106],[3665,-5.8106],[3703,-4.6103],[3745,-5.8072],[3783,-4.5997],[3825,-5.8108],[3863,-4.6103],[3907,-4.5997],[3949,-5.8108],[3987,-4.6103],[4031,-4.5997],[4073,-5.8108],[4111,-4.6103],[4156,-4.5997],[4198,-5.8108],[4236,-4.6103],[4282,-4.5997],[4324,-5.8108],[4364,-5.4199],[4365,-5.4209],[4366,-5.442],[4369,-5.6841],[4372,-6.0611],[4373,-5.8949],[4375,-8.9733]]
       }
       hips_ctrl.Rotation.Y {
        type=MovieSceneScriptingFloatChannel
        default=1.9239
        thinned_from=60
        keys [[1790,1.9239],[1813,0.3382],[1853,2.9165],[1892,0.3375],[1933,2.9349],[1973,0.3382],[2013,2.941],[2052,0.3345],[2093,2.9353],[2133,0.3382],[2173,2.9165],[2212,0.3375],[2253,2.9349],[2293,0.3382],[2333,2.941],[2372,0.3345],[2413,2.9353],[2453,0.3382],[2470,1.0655],[2484,-1.4472],[2501,0.9993],[2507,1.3356],[2515,0.4711],[2518,0.8616],[2521,-0.2866],[2527,0.0283],[2529,-0.0911],[2530,-0.086],[2782,1.0737],[2927,0.7881],[2938,0.7779],[2986,-7.1566],[3058,-7.1566],[3061,-4.102],[3070,-2.6301],[3185,-2.6301],[3201,-1.283],[3248,0.3382],[3327,0.3375],[3408,0.3382],[3487,0.3345],[3568,0.3382],[3647,0.3375],[3728,0.3382],[3807,0.3345],[3888,0.3382],[4012,0.3382],[4055,0.3345],[4136,0.3382],[4180,0.3345],[4261,0.3382],[4364,0.3449],[4365,-27.7051],[4366,1.2246],[4369,5.9855],[4372,-18.4551],[4373,-18.3715],[4375,10.9844]]
       }
       hips_ctrl.Rotation.Z {
        type=MovieSceneScriptingFloatChannel
        default=-0.1554
        keys [[1790,-0.1554],[1812,-0.0299],[1851,-0.2579],[1892,-0.0299],[1931,-0.2596],[1972,-0.0299],[2011,-0.2593],[2052,-0.0296],[2091,-0.2596],[2132,-0.0299],[2171,-0.2579],[2212,-0.0299],[2251,-0.2596],[2292,-0.0299],[2331,-0.2593],[2372,-0.0296],[2411,-0.2596],[2452,-0.0299],[2470,-0.117],[2496,16.7515],[2502,17.0548],[2507,16.9092],[2511,17.0989],[2512,16.5559],[2515,6.1247],[2519,0.0693],[2521,-0.0395],[2527,0.005],[2529,-0.0146],[2530,-0.0136],[2765,0.0037],[2927,-0.242],[2938,-0.2461],[2986,-2.7892],[3040,5.8967],[3060,6.3278],[3061,6.5369],[3082,21.7628],[3099,22.1411],[3185,22.1411],[3205,15.0941],[3209,12.4585],[3247,-0.0299],[4345,-0.2596],[4364,-0.0333],[4365,-20.8835],[4366,-0.1202],[4369,-0.5985],[4372,1.948],[4373,1.7718],[4375,6.2864]]
       }
       hips_ctrl.Scale.X {type=MovieSceneScriptingFloatChannel, default=1, const [1,12]}
       hips_ctrl.Scale.Y {type=MovieSceneScriptingFloatChannel, default=1, const [1,12]}
       hips_ctrl.Scale.Z {type=MovieSceneScriptingFloatChannel, default=1, const [1,12]}
       hips_tan_ctrl.Location.X {
        type=MovieSceneScriptingFloatChannel
        default=0.0172
        keys [[1790,0.0172],[1793,0.02],[1795,0.0131],[1797,0.0164],[1799,0.0098],[1801,0.0113],[1803,0.0063],[1804,0.0073],[1806,0.003],[1809,0.0057],[1811,-0.0001],[1812,0.0001],[1813,-0.0002],[1815,0.0004],[1821,-0.0008],[1823,0.0023],[1826,0.0006],[1828,0.0032],[1829,0.0026],[1850,0.0233],[1851,0.0233],[1856,0.024],[1858,0.0234],[1860,0.0236],[1864,0.0219],[1867,0.0241],[1869,0.0166],[1873,0.0208],[1875,0.0134],[1877,0.0165],[1879,0.01],[1880,0.0108],[1882,0.0083],[1883,0.0084],[1885,0.0049],[1887,0.0056],[1890,0.0015],[1891,0.003],[1898,-0.0001],[1900,0.0011],[1902,0.0004],[1913,0.0078],[1915,0.0075],[1923,0.0202],[1925,0.0173],[1927,0.0218],[1928,0.0218],[1935,0.0235],[1937,0.0231],[1938,0.0232],[1944,0.0216],[1947,0.0236],[1949,0.0163],[1953,0.0193],[1955,0.012],[1957,0.0175],[1959,0.0104],[1961,0.0108],[1963,0.0071],[1964,0.0076],[1966,0.003],[1969,0.0052],[1971,-0.0002],[1972,0.0011],[1978,-0.0001],[1980,0.001],[1982,0.0005],[1989,0.0034],[1990,0.0034],[1996,0.0123],[1997,0.0122],[2003,0.0195],[2004,0.0185],[2015,0.0236],[2026,0.0205],[2028,0.0206],[2045,0.0025],[2047,0.0079],[2050,0.004],[2051,0.0042],[2056,0.0002],[2058,0.0007],[2059,0.0004],[2061,0.0006],[2062,0.0005],[2073,0.0078],[2075,0.0075],[2083,0.0202],[2085,0.0173],[2087,0.0218],[2088,0.0218],[2095,0.0235],[2104,0.0217],[2107,0.0237],[2109,0.0163],[2113,0.02],[2115,0.0131],[2117,0.0164],[2119,0.0098],[2121,0.0113],[2123,0.0063],[2124,0.0073],[2126,0.003],[2129,0.0057],[2131,-0.0001],[2132,0.0001],[2133,-0.0002],[2135,0.0004],[2141,-0.0008],[2143,0.0023],[2146,0.0006],[2148,0.0032],[2149,0.0026],[2170,0.0233],[2171,0.0233],[2176,0.024],[2178,0.0234],[2180,0.0236],[2184,0.0219],[2187,0.0241],[2189,0.0166],[2193,0.0208],[2195,0.0134],[2197,0.0165],[2199,0.01],[2200,0.0108],[2202,0.0083],[2203,0.0084],[2205,0.0049],[2207,0.0056],[2210,0.0015],[2211,0.003],[2218,-0.0001],[2220,0.0011],[2222,0.0004],[2233,0.0078],[2235,0.0075],[2243,0.0202],[2245,0.0173],[2247,0.0218],[2248,0.0218],[2255,0.0235],[2257,0.0231],[2258,0.0232],[2264,0.0216],[2267,0.0236],[2269,0.0163],[2273,0.0193],[2275,0.012],[2277,0.0175],[2279,0.0104],[2281,0.0108],[2283,0.0071],[2284,0.0076],[2286,0.003],[2289,0.0052],[2291,-0.0002],[2292,0.0011],[2298,-0.0001],[2300,0.001],[2302,0.0005],[2309,0.0034],[2310,0.0034],[2316,0.0123],[2317,0.0122],[2323,0.0195],[2324,0.0185],[2335,0.0236],[2346,0.0205],[2348,0.0206],[2365,0.0025],[2367,0.0079],[2370,0.004],[2371,0.0042],[2376,0.0002],[2378,0.0007],[2379,0.0004],[2381,0.0006],[2382,0.0005],[2393,0.0078],[2395,0.0075],[2403,0.0202],[2405,0.0173],[2407,0.0218],[2408,0.0218],[2415,0.0235],[2424,0.0217],[2427,0.0237],[2429,0.0163],[2433,0.02],[2435,0.0131],[2437,0.0164],[2439,0.0098],[2441,0.0113],[2443,0.0063],[2444,0.0073],[2446,0.003],[2449,0.0057],[2451,-0.0001],[2452,0.0001],[2453,-0.0002],[2455,0.0004],[2461,-0.0008],[2463,0.0023],[2466,0.0006],[2471,0.0193],[2502,0.1785],[2508,0.2119],[2511,0.2658],[2513,0.2171],[2515,0.1085],[2518,-0.0183],[2519,-0.0214],[2521,-0.0116],[2526,-0.0259],[2528,-0.0215],[2531,-0.0263],[2532,-0.0241],[2536,-0.0284],[2927,0.2051],[2938,0.6211],[2986,-0.9124],[3058,-0.9124],[3070,-1.251],[3083,-1.251],[3099,0.6342],[3185,0.6342],[3219,0.0217],[3222,0.0237],[3224,0.0163],[3228,0.02],[3230,0.0131],[3232,0.0164],[3234,0.0098],[3236,0.0113],[3238,0.0063],[3239,0.0073],[3241,0.003],[3244,0.0057],[3246,-0.0001],[3247,0.0001],[3248,-0.0002],[3250,0.0004],[3256,-0.0008],[3258,0.0023],[3261,0.0006],[3263,0.0032],[3264,0.0026],[3285,0.0233],[3286,0.0233],[3288,0.0237],[3289,0.0237],[3291,0.024],[3293,0.0234],[3295,0.0236],[3299,0.0219],[3302,0.0241],[3304,0.0166],[3308,0.0208],[3310,0.0134],[3312,0.0165],[3314,0.01],[3315,0.0108],[3317,0.0083],[3318,0.0084],[3320,0.0049],[3322,0.0056],[3325,0.0015],[3326,0.003],[3333,-0.0001],[3335,0.0011],[3337,0.0004],[3348,0.0078],[3350,0.0075],[3358,0.0202],[3360,0.0173],[3362,0.0218],[3363,0.0218],[3370,0.0235],[3372,0.0231],[3373,0.0232],[3379,0.0216],[3382,0.0236],[3384,0.0163],[3388,0.0193],[3390,0.012],[3392,0.0175],[3394,0.0104],[3396,0.0108],[3398,0.0071],[3399,0.0076],[3401,0.003],[3404,0.0052],[3406,-0.0002],[3407,0.0011],[3413,-0.0001],[3415,0.001],[3417,0.0005],[3424,0.0034],[3425,0.0034],[3431,0.0122],[3432,0.0122],[3438,0.0195],[3439,0.0185],[3450,0.0236],[3461,0.0205],[3463,0.0206],[3480,0.0025],[3482,0.0079],[3485,0.004],[3486,0.0042],[3491,0.0002],[3493,0.0007],[3494,0.0004],[3496,0.0006],[3497,0.0005],[3508,0.0078],[3510,0.0075],[3518,0.0202],[3520,0.0173],[3522,0.0218],[3523,0.0218],[3530,0.0235],[3539,0.0217],[3542,0.0237],[3544,0.0163],[3548,0.02],[3550,0.0131],[3552,0.0164],[3554,0.0098],[3556,0.0113],[3558,0.0063],[3559,0.0073],[3561,0.003],[3564,0.0057],[3566,-0.0001],[3567,0.0001],[3568,-0.0002],[3570,0.0004],[3576,-0.0008],[3578,0.0023],[3581,0.0006],[3583,0.0032],[3584,0.0026],[3605,0.0233],[3606,0.0233],[3611,0.024],[3613,0.0234],[3615,0.0236],[3619,0.0219],[3622,0.0241],[3624,0.0166],[3628,0.0208],[3630,0.0134],[3632,0.0165],[3634,0.01],[3635,0.0108],[3637,0.0083],[3638,0.0084],[3640,0.0049],[3642,0.0056],[3645,0.0015],[3646,0.003],[3653,-0.0001],[3655,0.0011],[3657,0.0004],[3668,0.0078],[3670,0.0075],[3678,0.0202],[3680,0.0173],[3682,0.0218],[3683,0.0218],[3690,0.0235],[3692,0.0231],[3693,0.0232],[3699,0.0216],[3702,0.0236],[3704,0.0163],[3708,0.0193],[3710,0.012],[3712,0.0175],[3714,0.0104],[3716,0.0108],[3718,0.0071],[3719,0.0076],[3721,0.003],[3724,0.0052],[3726,-0.0003],[3727,0.0011],[3733,-0.0001],[3735,0.001],[3737,0.0005],[3744,0.0034],[3745,0.0034],[3751,0.0122],[3752,0.0122],[3758,0.0195],[3759,0.0185],[3770,0.0236],[3781,0.0205],[3783,0.0206],[3800,0.0025],[3802,0.0079],[3805,0.004],[3806,0.0042],[3811,0.0002],[3813,0.0007],[3814,0.0004],[3816,0.0006],[3817,0.0005],[3828,0.0078],[3830,0.0075],[3838,0.0202],[3840,0.0173],[3842,0.0218],[3843,0.0218],[3850,0.0235],[3859,0.0217],[3862,0.0237],[3864,0.0163],[3868,0.02],[3870,0.0131],[3872,0.0164],[3874,0.0098],[3876,0.0113],[3878,0.0063],[3879,0.0073],[3881,0.003],[3884,0.0057],[3886,-0.0001],[3887,0.0001],[3888,-0.0002],[3894,0.0236],[3905,0.0205],[3907,0.0206],[3924,0.0025],[3926,0.0079],[3929,0.004],[3930,0.0042],[3935,0.0002],[3937,0.0007],[3938,0.0004],[3940,0.0006],[3941,0.0005],[3952,0.0078],[3954,0.0075],[3962,0.0202],[3964,0.0173],[3966,0.0218],[3967,0.0218],[3974,0.0235],[3983,0.0217],[3986,0.0237],[3988,0.0163],[3992,0.02],[3994,0.0131],[3996,0.0164],[3998,0.0098],[4000,0.0113],[4002,0.0063],[4003,0.0073],[4005,0.003],[4008,0.0057],[4010,-0.0001],[4011,0.0001],[4012,-0.0002],[4018,0.0236],[4029,0.0205],[4031,0.0206],[4048,0.0025],[4050,0.0079],[4053,0.004],[4054,0.0042],[4059,0.0002],[4061,0.0007],[4062,0.0004],[4064,0.0006],[4065,0.0005],[4076,0.0078],[4078,0.0075],[4086,0.0202],[4088,0.0173],[4090,0.0218],[4091,0.0218],[4098,0.0235],[4107,0.0217],[4110,0.0237],[4112,0.0163],[4116,0.02],[4118,0.0131],[4120,0.0164],[4122,0.0098],[4124,0.0113],[4126,0.0063],[4127,0.0073],[4129,0.003],[4132,0.0057],[4134,-0.0001],[4135,0.0001],[4136,-0.0002],[4143,0.0236],[4154,0.0205],[4156,0.0206],[4173,0.0025],[4175,0.0079],[4178,0.004],[4179,0.0042],[4184,0.0002],[4186,0.0007],[4187,0.0004],[4189,0.0006],[4190,0.0005],[4201,0.0078],[4203,0.0075],[4211,0.0202],[4213,0.0173],[4215,0.0218],[4216,0.0218],[4223,0.0235],[4232,0.0217],[4235,0.0237],[4237,0.0163],[4241,0.02],[4243,0.0131],[4245,0.0164],[4247,0.0098],[4249,0.0113],[4251,0.0063],[4252,0.0073],[4254,0.003],[4257,0.0057],[4259,-0.0001],[4260,0.0001],[4261,-0.0002],[4269,0.0236],[4280,0.0205],[4282,0.0206],[4299,0.0025],[4301,0.0079],[4304,0.004],[4305,0.0042],[4310,0.0002],[4312,0.0007],[4313,0.0004],[4315,0.0006],[4316,0.0005],[4327,0.0078],[4329,0.0075],[4337,0.0202],[4339,0.0173],[4341,0.0218],[4342,0.0218],[4349,0.0235],[4360,-0.1818],[4364,-0.4766],[4365,0.0004],[4372,0.0032],[4373,0.0026],[4394,0.0233],[4395,0.0233],[4397,0.0237],[4398,0.0237],[4400,0.024],[4402,0.0234],[4404,0.0236],[4408,0.0219],[4411,0.0241],[4413,0.0166],[4417,0.0208],[4419,0.0134],[4421,0.0165],[4423,0.01],[4424,0.0108],[4426,0.0083],[4427,0.0084],[4429,0.0049],[4431,0.0056],[4434,0.0015],[4435,0.003],[4436,0]]
       }
       hips_tan_ctrl.Location.Y {
        type=MovieSceneScriptingFloatChannel
        default=0.1217
        thinned_from=735
        keys [[1790,0.1217],[1791,0.1217],[1793,0.1201],[1794,0.1205],[1795,0.1199],[1796,0.12],[1798,0.1178],[1799,0.1182],[1800,0.1182],[1801,0.1184],[1803,0.1165],[1804,0.1172],[1809,0.1148],[1811,0.1154],[1814,0.1142],[1816,0.1149],[1818,0.1139],[1820,0.1164],[1823,0.1143],[1825,0.1169],[1827,0.1156],[1829,0.1185],[1832,0.1177],[1844,0.122],[1845,0.1219],[1849,0.1227],[1851,0.1225],[1853,0.123],[1858,0.1231],[1859,0.1227],[1863,0.1228],[1864,0.1226],[1866,0.123],[1868,0.1217],[1869,0.1222],[1870,0.1217],[1871,0.1218],[1873,0.1202],[1874,0.1205],[1875,0.1199],[1876,0.12],[1878,0.1179],[1879,0.1182],[1885,0.1164],[1886,0.1166],[1889,0.1152],[1891,0.116],[1893,0.1144],[1896,0.1148],[1897,0.1144],[1911,0.118],[1912,0.1179],[1914,0.1193],[1916,0.119],[1919,0.1209],[1920,0.1206],[1923,0.1223],[1925,0.1213],[1928,0.1225],[1929,0.1225],[1932,0.123],[1934,0.1228],[1937,0.123],[1939,0.1228],[1940,0.123],[1944,0.1226],[1946,0.1227],[1948,0.1203],[1949,0.1203],[1952,0.1183],[1954,0.1224],[1963,0.117],[1964,0.1171],[1965,0.1164],[1967,0.1167],[1969,0.1146],[1971,0.1161],[1974,0.1142],[1976,0.1149],[1977,0.1144],[1982,0.1158],[1984,0.1154],[1986,0.1163],[1987,0.1162],[1994,0.1193],[1996,0.1189],[1998,0.1211],[2000,0.1206],[2003,0.1224],[2005,0.1214],[2016,0.1232],[2025,0.1227],[2027,0.1229],[2032,0.1211],[2033,0.1214],[2039,0.118],[2040,0.1182],[2045,0.1147],[2047,0.1176],[2054,0.1146],[2057,0.1147],[2059,0.115],[2060,0.115],[2074,0.1193],[2076,0.119],[2078,0.1209],[2080,0.1206],[2083,0.1224],[2085,0.1213],[2088,0.1225],[2094,0.1229],[2098,0.123],[2104,0.1226],[2106,0.123],[2108,0.1217],[2109,0.1222],[2110,0.1217],[2111,0.1217],[2113,0.1201],[2114,0.1205],[2115,0.1199],[2116,0.12],[2118,0.1178],[2119,0.1182],[2120,0.1182],[2121,0.1184],[2123,0.1165],[2124,0.1172],[2129,0.1148],[2131,0.1154],[2134,0.1142],[2136,0.1149],[2138,0.1139],[2140,0.1164],[2143,0.1143],[2145,0.1169],[2147,0.1156],[2149,0.1185],[2152,0.1177],[2164,0.122],[2165,0.1219],[2169,0.1227],[2171,0.1225],[2173,0.123],[2178,0.1231],[2179,0.1227],[2183,0.1228],[2184,0.1226],[2186,0.123],[2188,0.1217],[2189,0.1222],[2190,0.1217],[2191,0.1218],[2193,0.1202],[2194,0.1205],[2195,0.1199],[2196,0.12],[2198,0.1179],[2199,0.1182],[2205,0.1164],[2206,0.1166],[2209,0.1152],[2211,0.116],[2213,0.1144],[2216,0.1148],[2217,0.1144],[2231,0.118],[2232,0.1179],[2234,0.1193],[2236,0.119],[2239,0.1209],[2240,0.1206],[2243,0.1223],[2245,0.1213],[2248,0.1225],[2249,0.1225],[2252,0.123],[2254,0.1228],[2257,0.123],[2259,0.1228],[2260,0.123],[2264,0.1226],[2266,0.1227],[2268,0.1203],[2269,0.1203],[2272,0.1183],[2274,0.1224],[2283,0.117],[2284,0.1171],[2285,0.1164],[2287,0.1167],[2289,0.1146],[2291,0.1161],[2294,0.1142],[2296,0.1149],[2297,0.1144],[2302,0.1158],[2304,0.1154],[2306,0.1163],[2307,0.1162],[2314,0.1193],[2316,0.1189],[2318,0.1211],[2320,0.1206],[2323,0.1224],[2325,0.1214],[2336,0.1232],[2345,0.1227],[2347,0.1229],[2352,0.1211],[2353,0.1214],[2359,0.118],[2360,0.1182],[2365,0.1147],[2367,0.1176],[2374,0.1146],[2377,0.1147],[2379,0.115],[2380,0.115],[2394,0.1193],[2396,0.119],[2398,0.1209],[2400,0.1206],[2403,0.1224],[2405,0.1213],[2408,0.1225],[2414,0.1229],[2418,0.123],[2424,0.1226],[2426,0.123],[2428,0.1217],[2429,0.1222],[2430,0.1217],[2431,0.1217],[2433,0.1201],[2434,0.1205],[2435,0.1199],[2436,0.12],[2438,0.1178],[2439,0.1182],[2440,0.1182],[2441,0.1184],[2443,0.1165],[2444,0.1172],[2449,0.1148],[2451,0.1154],[2454,0.1142],[2456,0.1149],[2458,0.1139],[2460,0.1164],[2463,0.1143],[2465,0.1169],[2467,0.1154],[2502,0.4811],[2507,0.3409],[2509,0.7564],[2511,1.0042],[2513,0.7416],[2515,0.1469],[2518,-0.4513],[2519,-0.4876],[2520,-0.4841],[2523,-0.487],[2767,-0.0935],[2780,-0.003],[2927,2.1012],[2938,0.1636],[2986,-1.3156],[3058,-1.3156],[3070,-0.0827],[3083,-0.0827],[3099,2.4137],[3185,2.4137],[3219,0.1226],[3221,0.123],[3223,0.1217],[3224,0.1222],[3225,0.1217],[3226,0.1217],[3228,0.1201],[3229,0.1205],[3230,0.1199],[3231,0.12],[3233,0.1178],[3234,0.1182],[3235,0.1182],[3236,0.1184],[3238,0.1165],[3239,0.1172],[3244,0.1148],[3246,0.1154],[3249,0.1142],[3251,0.1149],[3253,0.1139],[3255,0.1164],[3258,0.1143],[3260,0.1169],[3262,0.1156],[3264,0.1185],[3267,0.1177],[3279,0.122],[3280,0.1219],[3284,0.1227],[3286,0.1225],[3288,0.123],[3293,0.1231],[3294,0.1227],[3298,0.1228],[3299,0.1226],[3301,0.123],[3303,0.1217],[3304,0.1222],[3305,0.1217],[3306,0.1218],[3308,0.1202],[3309,0.1205],[3310,0.1199],[3311,0.12],[3313,0.1179],[3314,0.1182],[3320,0.1164],[3321,0.1166],[3324,0.1152],[3326,0.116],[3328,0.1144],[3331,0.1148],[3332,0.1144],[3346,0.118],[3347,0.1179],[3349,0.1193],[3351,0.119],[3354,0.1209],[3355,0.1206],[3358,0.1223],[3360,0.1213],[3363,0.1225],[3364,0.1225],[3367,0.123],[3369,0.1228],[3372,0.123],[3374,0.1228],[3375,0.123],[3379,0.1226],[3381,0.1227],[3383,0.1203],[3384,0.1203],[3387,0.1183],[3389,0.1224],[3398,0.117],[3399,0.1171],[3400,0.1164],[3402,0.1167],[3404,0.1146],[3406,0.1161],[3409,0.1142],[3411,0.1149],[3412,0.1144],[3417,0.1158],[3419,0.1154],[3421,0.1163],[3422,0.1162],[3429,0.1193],[3431,0.1189],[3433,0.1211],[3435,0.1206],[3438,0.1224],[3440,0.1214],[3451,0.1232],[3460,0.1227],[3462,0.1229],[3467,0.1211],[3468,0.1214],[3474,0.118],[3475,0.1182],[3480,0.1147],[3482,0.1176],[3489,0.1146],[3492,0.1147],[3494,0.115],[3495,0.115],[3509,0.1193],[3511,0.119],[3513,0.1209],[3515,0.1206],[3518,0.1224],[3520,0.1213],[3523,0.1225],[3529,0.1229],[3533,0.123],[3539,0.1226],[3541,0.123],[3543,0.1217],[3544,0.1222],[3545,0.1217],[3546,0.1217],[3548,0.1201],[3549,0.1205],[3550,0.1199],[3551,0.12],[3553,0.1178],[3554,0.1182],[3555,0.1182],[3556,0.1184],[3558,0.1165],[3559,0.1172],[3564,0.1148],[3566,0.1154],[3569,0.1142],[3571,0.1149],[3573,0.1139],[3575,0.1164],[3578,0.1143],[3580,0.1169],[3582,0.1156],[3584,0.1185],[3587,0.1177],[3599,0.122],[3600,0.1219],[3604,0.1227],[3606,0.1225],[3608,0.123],[3613,0.1231],[3614,0.1227],[3618,0.1228],[3619,0.1226],[3621,0.123],[3623,0.1217],[3624,0.1222],[3625,0.1217],[3626,0.1218],[3628,0.1202],[3629,0.1205],[3630,0.1199],[3631,0.12],[3633,0.1179],[3634,0.1182],[3640,0.1164],[3641,0.1166],[3644,0.1152],[3646,0.116],[3648,0.1144],[3651,0.1148],[3652,0.1144],[3666,0.118],[3667,0.1179],[3669,0.1193],[3671,0.119],[3674,0.1209],[3675,0.1206],[3678,0.1223],[3680,0.1213],[3683,0.1225],[3684,0.1225],[3687,0.123],[3689,0.1228],[3692,0.123],[3694,0.1228],[3695,0.123],[3699,0.1226],[3701,0.1227],[3703,0.1203],[3704,0.1203],[3707,0.1183],[3709,0.1224],[3718,0.117],[3719,0.1171],[3720,0.1164],[3722,0.1167],[3724,0.1146],[3726,0.1161],[3729,0.1142],[3731,0.1149],[3732,0.1144],[3737,0.1158],[3739,0.1154],[3741,0.1163],[3742,0.1162],[3749,0.1193],[3751,0.1189],[3753,0.1211],[3755,0.1206],[3758,0.1224],[3760,0.1214],[3771,0.1232],[3780,0.1227],[3782,0.1229],[3787,0.1211],[3788,0.1214],[3794,0.118],[3795,0.1182],[3800,0.1147],[3802,0.1176],[3809,0.1146],[3812,0.1147],[3814,0.115],[3815,0.115],[3829,0.1193],[3831,0.119],[3833,0.1209],[3835,0.1206],[3838,0.1224],[3840,0.1213],[3843,0.1225],[3849,0.1229],[3853,0.123],[3859,0.1226],[3861,0.123],[3863,0.1217],[3864,0.1222],[3865,0.1217],[3866,0.1217],[3868,0.1201],[3869,0.1205],[3870,0.1199],[3871,0.12],[3873,0.1178],[3874,0.1182],[3875,0.1182],[3876,0.1184],[3878,0.1165],[3879,0.1172],[3884,0.1148],[3886,0.1154],[3895,0.1232],[3904,0.1227],[3906,0.1229],[3911,0.1211],[3912,0.1214],[3918,0.118],[3919,0.1182],[3924,0.1147],[3926,0.1176],[3933,0.1146],[3936,0.1147],[3938,0.115],[3939,0.115],[3953,0.1193],[3955,0.119],[3957,0.1209],[3959,0.1206],[3962,0.1224],[3964,0.1213],[3967,0.1225],[3973,0.1229],[3977,0.123],[3983,0.1226],[3985,0.123],[3987,0.1217],[3988,0.1222],[3989,0.1217],[3990,0.1217],[3992,0.1201],[3993,0.1205],[3994,0.1199],[3995,0.12],[3997,0.1178],[3998,0.1182],[3999,0.1182],[4000,0.1184],[4002,0.1165],[4003,0.1172],[4008,0.1148],[4010,0.1154],[4019,0.1232],[4028,0.1227],[4030,0.1229],[4035,0.1211],[4036,0.1214],[4042,0.118],[4043,0.1182],[4048,0.1147],[4050,0.1176],[4057,0.1146],[4060,0.1147],[4062,0.115],[4063,0.115],[4077,0.1193],[4079,0.119],[4081,0.1209],[4083,0.1206],[4086,0.1224],[4088,0.1213],[4091,0.1225],[4097,0.1229],[4101,0.123],[4107,0.1226],[4109,0.123],[4111,0.1217],[4112,0.1222],[4113,0.1217],[4114,0.1217],[4116,0.1201],[4117,0.1205],[4118,0.1199],[4119,0.12],[4121,0.1178],[4122,0.1182],[4123,0.1182],[4124,0.1184],[4126,0.1165],[4127,0.1172],[4132,0.1148],[4134,0.1154],[4144,0.1232],[4153,0.1227],[4155,0.1229],[4160,0.1211],[4161,0.1214],[4167,0.118],[4168,0.1182],[4173,0.1147],[4175,0.1176],[4182,0.1146],[4185,0.1147],[4187,0.115],[4188,0.115],[4202,0.1193],[4204,0.119],[4206,0.1209],[4208,0.1206],[4211,0.1224],[4213,0.1213],[4216,0.1225],[4222,0.1229],[4226,0.123],[4232,0.1226],[4234,0.123],[4236,0.1217],[4237,0.1222],[4238,0.1217],[4239,0.1217],[4241,0.1201],[4242,0.1205],[4243,0.1199],[4244,0.12],[4246,0.1178],[4247,0.1182],[4248,0.1182],[4249,0.1184],[4251,0.1165],[4252,0.1172],[4257,0.1148],[4259,0.1154],[4270,0.1232],[4279,0.1227],[4281,0.1229],[4286,0.1211],[4287,0.1214],[4293,0.118],[4294,0.1182],[4299,0.1147],[4301,0.1176],[4308,0.1146],[4311,0.1147],[4313,0.115],[4314,0.115],[4328,0.1193],[4330,0.119],[4332,0.1209],[4334,0.1206],[4337,0.1224],[4339,0.1213],[4342,0.1225],[4348,0.1229],[4352,0.123],[4360,-0.1137],[4364,0.1495],[4366,0.1149],[4371,0.1156],[4373,0.1185],[4376,0.1177],[4388,0.122],[4389,0.1219],[4393,0.1227],[4395,0.1225],[4397,0.123],[4402,0.1231],[4403,0.1227],[4407,0.1228],[4408,0.1226],[4410,0.123],[4412,0.1217],[4413,0.1222],[4414,0.1217],[4415,0.1218],[4417,0.1202],[4418,0.1205],[4419,0.1199],[4420,0.12],[4422,0.1179],[4423,0.1182],[4429,0.1164],[4430,0.1166],[4433,0.1152],[4435,0.116],[4436,0.008]]
       }
       hips_tan_ctrl.Location.Z {
        type=MovieSceneScriptingFloatChannel
        default=-0.1335
        thinned_from=394
        keys [[1790,-0.1335],[1794,-0.1336],[1798,-0.1376],[1799,-0.1376],[1805,-0.1398],[1806,-0.1396],[1811,-0.1399],[1816,-0.1354],[1818,-0.1361],[1820,-0.1256],[1822,-0.1266],[1824,-0.1174],[1827,-0.1196],[1829,-0.1095],[1832,-0.11],[1843,-0.105],[1847,-0.1059],[1849,-0.1057],[1851,-0.1091],[1853,-0.1083],[1873,-0.1336],[1874,-0.1336],[1878,-0.1376],[1879,-0.1376],[1883,-0.1389],[1884,-0.1388],[1889,-0.1397],[1919,-0.1055],[1920,-0.1059],[1922,-0.1049],[1925,-0.106],[1927,-0.1055],[1932,-0.1101],[1934,-0.1094],[1948,-0.1327],[1950,-0.1305],[1952,-0.131],[1953,-0.1309],[1962,-0.1386],[1963,-0.1383],[1965,-0.1396],[1970,-0.1397],[1971,-0.14],[1988,-0.1133],[1989,-0.1163],[1998,-0.1056],[2000,-0.1057],[2002,-0.1051],[2005,-0.1057],[2007,-0.1054],[2031,-0.1332],[2032,-0.1332],[2046,-0.1397],[2049,-0.1396],[2053,-0.1373],[2054,-0.1374],[2079,-0.1055],[2080,-0.1059],[2082,-0.1049],[2085,-0.106],[2086,-0.1055],[2113,-0.1336],[2114,-0.1336],[2118,-0.1376],[2119,-0.1376],[2125,-0.1398],[2126,-0.1396],[2131,-0.1399],[2136,-0.1354],[2138,-0.1361],[2140,-0.1256],[2142,-0.1266],[2144,-0.1174],[2147,-0.1196],[2149,-0.1095],[2152,-0.11],[2163,-0.105],[2167,-0.1059],[2169,-0.1057],[2171,-0.1091],[2173,-0.1083],[2193,-0.1336],[2194,-0.1336],[2198,-0.1376],[2199,-0.1376],[2203,-0.1389],[2204,-0.1388],[2209,-0.1397],[2239,-0.1055],[2240,-0.1059],[2242,-0.1049],[2245,-0.106],[2247,-0.1055],[2252,-0.1101],[2254,-0.1094],[2268,-0.1327],[2270,-0.1305],[2272,-0.131],[2273,-0.1309],[2282,-0.1386],[2283,-0.1383],[2285,-0.1396],[2290,-0.1397],[2291,-0.14],[2308,-0.1133],[2309,-0.1163],[2318,-0.1056],[2320,-0.1057],[2322,-0.1051],[2325,-0.1057],[2327,-0.1054],[2351,-0.1332],[2352,-0.1332],[2366,-0.1397],[2369,-0.1396],[2373,-0.1373],[2374,-0.1374],[2399,-0.1055],[2400,-0.1059],[2402,-0.1049],[2405,-0.106],[2406,-0.1055],[2433,-0.1336],[2434,-0.1336],[2438,-0.1376],[2439,-0.1376],[2445,-0.1398],[2446,-0.1396],[2451,-0.1399],[2456,-0.1354],[2458,-0.1361],[2460,-0.1256],[2462,-0.1266],[2464,-0.1174],[2501,-0.011],[2506,0.0008],[2511,-0.116],[2515,0.3006],[2519,-0.1949],[2521,-0.2269],[2525,-0.3201],[2766,0.2523],[2778,0.4853],[2801,0.6913],[2927,0.8857],[2938,0.6456],[2986,-0.284],[3058,-0.284],[3070,0.1411],[3083,0.1411],[3099,-0.0441],[3185,-0.0441],[3228,-0.1336],[3229,-0.1336],[3233,-0.1376],[3234,-0.1376],[3240,-0.1398],[3241,-0.1396],[3246,-0.1399],[3251,-0.1354],[3253,-0.1361],[3255,-0.1256],[3257,-0.1266],[3259,-0.1174],[3262,-0.1196],[3264,-0.1095],[3267,-0.11],[3278,-0.105],[3282,-0.106],[3284,-0.1057],[3286,-0.1091],[3288,-0.1083],[3308,-0.1336],[3309,-0.1336],[3313,-0.1376],[3314,-0.1376],[3318,-0.1389],[3319,-0.1388],[3324,-0.1397],[3354,-0.1055],[3355,-0.1059],[3357,-0.1049],[3360,-0.106],[3362,-0.1055],[3367,-0.1101],[3369,-0.1094],[3383,-0.1327],[3385,-0.1305],[3387,-0.131],[3388,-0.1309],[3397,-0.1386],[3398,-0.1383],[3400,-0.1396],[3405,-0.1397],[3406,-0.14],[3423,-0.1133],[3424,-0.1163],[3433,-0.1056],[3435,-0.1057],[3437,-0.1051],[3440,-0.1057],[3442,-0.1054],[3466,-0.1332],[3467,-0.1332],[3473,-0.1371],[3474,-0.1371],[3481,-0.1397],[3484,-0.1396],[3488,-0.1373],[3489,-0.1374],[3514,-0.1055],[3515,-0.1059],[3517,-0.1049],[3520,-0.106],[3521,-0.1055],[3548,-0.1336],[3549,-0.1336],[3553,-0.1376],[3554,-0.1376],[3560,-0.1398],[3561,-0.1396],[3566,-0.1399],[3571,-0.1354],[3573,-0.1361],[3575,-0.1256],[3577,-0.1266],[3579,-0.1174],[3582,-0.1196],[3584,-0.1095],[3587,-0.11],[3598,-0.105],[3602,-0.106],[3604,-0.1057],[3606,-0.1091],[3608,-0.1083],[3628,-0.1336],[3629,-0.1336],[3633,-0.1376],[3634,-0.1376],[3638,-0.1389],[3639,-0.1388],[3644,-0.1397],[3674,-0.1055],[3675,-0.1059],[3677,-0.1049],[3680,-0.106],[3682,-0.1055],[3687,-0.1101],[3689,-0.1094],[3703,-0.1327],[3705,-0.1305],[3707,-0.131],[3708,-0.1309],[3717,-0.1386],[3718,-0.1383],[3720,-0.1396],[3725,-0.1397],[3726,-0.14],[3743,-0.1133],[3744,-0.1163],[3753,-0.1056],[3755,-0.1057],[3757,-0.1051],[3760,-0.1057],[3762,-0.1054],[3786,-0.1332],[3787,-0.1332],[3793,-0.1371],[3794,-0.1371],[3801,-0.1397],[3804,-0.1396],[3808,-0.1373],[3809,-0.1374],[3834,-0.1055],[3835,-0.1059],[3837,-0.1049],[3840,-0.106],[3841,-0.1055],[3868,-0.1336],[3869,-0.1336],[3873,-0.1376],[3874,-0.1376],[3880,-0.1398],[3881,-0.1396],[3886,-0.1399],[3910,-0.1332],[3911,-0.1332],[3917,-0.1371],[3918,-0.1371],[3925,-0.1397],[3928,-0.1396],[3932,-0.1373],[3933,-0.1374],[3958,-0.1055],[3959,-0.1059],[3961,-0.1049],[3964,-0.106],[3965,-0.1055],[3992,-0.1336],[3993,-0.1336],[3997,-0.1376],[3998,-0.1376],[4004,-0.1398],[4005,-0.1396],[4010,-0.1399],[4034,-0.1332],[4035,-0.1332],[4041,-0.1371],[4042,-0.1371],[4049,-0.1397],[4052,-0.1396],[4056,-0.1373],[4057,-0.1374],[4082,-0.1055],[4083,-0.1059],[4085,-0.1049],[4088,-0.106],[4089,-0.1055],[4116,-0.1336],[4117,-0.1336],[4121,-0.1376],[4122,-0.1376],[4128,-0.1398],[4129,-0.1396],[4134,-0.1399],[4159,-0.1332],[4160,-0.1332],[4166,-0.1371],[4167,-0.1371],[4174,-0.1397],[4177,-0.1396],[4181,-0.1373],[4182,-0.1374],[4207,-0.1055],[4208,-0.1059],[4210,-0.1049],[4213,-0.106],[4214,-0.1055],[4241,-0.1336],[4242,-0.1336],[4246,-0.1376],[4247,-0.1376],[4253,-0.1398],[4254,-0.1396],[4259,-0.1399],[4285,-0.1332],[4286,-0.1332],[4292,-0.1371],[4293,-0.1371],[4300,-0.1397],[4303,-0.1396],[4307,-0.1373],[4308,-0.1374],[4333,-0.1055],[4334,-0.1059],[4336,-0.1049],[4339,-0.106],[4340,-0.1055],[4360,-0.4425],[4364,-0.484],[4366,-0.1354],[4371,-0.1196],[4373,-0.1095],[4376,-0.11],[4387,-0.105],[4391,-0.106],[4393,-0.1057],[4395,-0.1091],[4397,-0.1083],[4417,-0.1336],[4418,-0.1336],[4422,-0.1376],[4423,-0.1376],[4427,-0.1389],[4428,-0.1388],[4433,-0.1397],[4436,-0.0073]]
       }
       hips_tan_ctrl.Rotation.X {type=MovieSceneScriptingFloatChannel, default=0, const [0,12]}
       hips_tan_ctrl.Rotation.Y {type=MovieSceneScriptingFloatChannel, default=0, const [0,12]}
       hips_tan_ctrl.Rotation.Z {type=MovieSceneScriptingFloatChannel, default=0, const [0,12]}
       hips_tan_ctrl.Scale.X {type=MovieSceneScriptingFloatChannel, default=1, const [1,12]}
       hips_tan_ctrl.Scale.Y {type=MovieSceneScriptingFloatChannel, default=1, const [1,12]}
       hips_tan_ctrl.Scale.Z {type=MovieSceneScriptingFloatChannel, default=1, const [1,12]}
       spine_01_ctrl.Location.X {
        type=MovieSceneScriptingFloatChannel
        default=0.0309
        thinned_from=267
        keys [[1790,0.0309],[1793,0.0309],[1795,0.026],[1796,0.026],[1808,0.0118],[1809,0.0118],[1812,0.0077],[1821,0.0077],[1825,0.0099],[1828,0.0121],[1829,0.0121],[1854,0.0399],[1865,0.036],[1867,0.036],[1870,0.0311],[1872,0.0311],[1875,0.0261],[1876,0.0261],[1890,0.0095],[1891,0.0095],[1893,0.0077],[1898,0.0077],[1923,0.0343],[1925,0.0343],[1934,0.0397],[1945,0.0358],[1946,0.0358],[1949,0.0309],[1953,0.0309],[1956,0.026],[1957,0.026],[1963,0.0159],[1964,0.0159],[1969,0.0118],[1972,0.0077],[1977,0.0077],[1981,0.0087],[2014,0.0396],[2053,0.0077],[2056,0.0077],[2083,0.0343],[2084,0.0343],[2094,0.0397],[2112,0.0309],[2115,0.026],[2116,0.026],[2128,0.0118],[2129,0.0118],[2132,0.0077],[2141,0.0077],[2143,0.0099],[2146,0.0099],[2148,0.0121],[2149,0.0121],[2173,0.0399],[2175,0.0399],[2191,0.0311],[2192,0.0311],[2210,0.0095],[2213,0.0077],[2218,0.0077],[2220,0.0087],[2221,0.0087],[2243,0.0343],[2244,0.0343],[2254,0.0397],[2265,0.0358],[2266,0.0358],[2286,0.0118],[2289,0.0118],[2292,0.0077],[2297,0.0077],[2301,0.0087],[2302,0.0087],[2334,0.0396],[2374,0.0077],[2376,0.0077],[2404,0.0343],[2405,0.0343],[2414,0.0397],[2426,0.0358],[2427,0.0358],[2431,0.0309],[2432,0.0309],[2435,0.026],[2436,0.026],[2446,0.0118],[2448,0.0118],[2452,0.0077],[2461,0.0077],[2470,0.0132],[2503,-0.0115],[2509,-0.0115],[2512,-0.0105],[2515,-0.0187],[2520,-0.0026],[2527,-0.0133],[2528,-0.0129],[2927,0.2517],[2938,0.2522],[2986,0.3426],[3058,0.3426],[3070,-0.1933],[3185,-0.1933],[3222,0.0358],[3225,0.0309],[3231,0.026],[3232,0.026],[3249,0.0077],[3255,0.0077],[3258,0.0099],[3259,0.0099],[3288,0.0399],[3290,0.0399],[3301,0.036],[3302,0.036],[3304,0.0311],[3310,0.0261],[3311,0.0261],[3325,0.0095],[3326,0.0095],[3329,0.0077],[3331,0.0077],[3335,0.0087],[3336,0.0087],[3369,0.0397],[3381,0.0358],[3382,0.0358],[3384,0.0309],[3387,0.0309],[3407,0.0077],[3413,0.0077],[3415,0.0087],[3416,0.0087],[3449,0.0396],[3490,0.0077],[3493,0.0082],[3494,0.0082],[3529,0.0397],[3547,0.0309],[3548,0.0309],[3550,0.026],[3552,0.026],[3566,0.0077],[3576,0.0077],[3580,0.0099],[3581,0.0099],[3609,0.0399],[3610,0.0399],[3621,0.036],[3622,0.036],[3624,0.0311],[3628,0.0311],[3648,0.0077],[3653,0.0077],[3655,0.0087],[3657,0.0087],[3678,0.0343],[3680,0.0343],[3689,0.0397],[3701,0.0358],[3705,0.0309],[3708,0.0309],[3724,0.0118],[3726,0.0077],[3733,0.0077],[3769,0.0396],[3808,0.0077],[3811,0.0077],[3849,0.0397],[3893,0.0396],[3932,0.0077],[3935,0.0077],[3973,0.0397],[4017,0.0396],[4056,0.0077],[4059,0.0077],[4097,0.0397],[4142,0.0396],[4181,0.0077],[4184,0.0077],[4222,0.0397],[4268,0.0396],[4307,0.0077],[4310,0.0077],[4348,0.0397],[4360,0.0077],[4367,0.0077],[4397,0.0399],[4410,0.036],[4411,0.036],[4416,0.0311],[4419,0.0261],[4421,0.0261],[4436,0]]
       }
       spine_01_ctrl.Location.Y {
        type=MovieSceneScriptingFloatChannel
        default=0.1406
        thinned_from=354
        keys [[1790,0.1406],[1793,0.1391],[1796,0.1392],[1798,0.1378],[1799,0.1378],[1803,0.1368],[1804,0.1368],[1806,0.136],[1807,0.136],[1809,0.1352],[1814,0.1352],[1816,0.1359],[1818,0.1359],[1822,0.138],[1823,0.1379],[1826,0.1401],[1827,0.1401],[1830,0.1422],[1831,0.1422],[1845,0.1446],[1849,0.1446],[1852,0.1442],[1865,0.1418],[1866,0.1419],[1868,0.1405],[1870,0.1406],[1873,0.1391],[1876,0.1392],[1893,0.1351],[1920,0.1444],[1927,0.1447],[1945,0.1418],[1946,0.1419],[1948,0.1405],[1950,0.1406],[1954,0.1391],[1956,0.1392],[1958,0.1378],[1959,0.1378],[1963,0.1368],[1964,0.1368],[1966,0.136],[1967,0.136],[1970,0.1352],[1974,0.1352],[1976,0.1359],[1977,0.1359],[2006,0.1449],[2053,0.1351],[2080,0.1444],[2087,0.1447],[2105,0.1418],[2106,0.1419],[2108,0.1405],[2110,0.1406],[2113,0.1391],[2116,0.1392],[2123,0.1368],[2124,0.1368],[2126,0.136],[2127,0.136],[2129,0.1352],[2134,0.1352],[2136,0.1359],[2138,0.1359],[2140,0.138],[2143,0.1379],[2145,0.1401],[2146,0.1401],[2150,0.1422],[2151,0.1422],[2165,0.1446],[2169,0.1446],[2172,0.1442],[2185,0.1418],[2186,0.1419],[2188,0.1405],[2190,0.1406],[2193,0.1391],[2196,0.1392],[2213,0.1351],[2240,0.1444],[2247,0.1447],[2265,0.1418],[2266,0.1419],[2268,0.1405],[2271,0.1406],[2274,0.1391],[2276,0.1392],[2278,0.1378],[2279,0.1378],[2283,0.1368],[2284,0.1368],[2294,0.1352],[2296,0.1359],[2297,0.1359],[2326,0.1449],[2373,0.1351],[2400,0.1444],[2407,0.1447],[2425,0.1418],[2426,0.1419],[2428,0.1405],[2430,0.1406],[2434,0.1391],[2436,0.1392],[2443,0.1368],[2444,0.1368],[2454,0.1352],[2456,0.1359],[2457,0.1359],[2462,0.138],[2463,0.1379],[2465,0.1401],[2467,0.1399],[2468,0.1404],[2502,-0.0694],[2511,-0.0694],[2512,-0.0609],[2518,-0.202],[2520,-0.1534],[2524,-0.1824],[2764,-0.0114],[2774,0.0249],[2780,0.0042],[2927,0.0197],[2938,0.0199],[2986,0.1286],[3058,0.1286],[3070,-0.0578],[3185,-0.0578],[3220,0.1418],[3221,0.1419],[3223,0.1405],[3226,0.1406],[3228,0.1391],[3231,0.1392],[3233,0.1378],[3234,0.1378],[3238,0.1368],[3239,0.1368],[3245,0.1352],[3249,0.1352],[3251,0.1359],[3252,0.1359],[3256,0.138],[3258,0.1379],[3261,0.1401],[3262,0.1401],[3264,0.1422],[3266,0.1422],[3280,0.1446],[3284,0.1446],[3288,0.1442],[3300,0.1418],[3301,0.1419],[3303,0.1405],[3305,0.1406],[3309,0.1391],[3311,0.1392],[3328,0.1351],[3355,0.1444],[3362,0.1447],[3380,0.1418],[3381,0.1419],[3383,0.1405],[3385,0.1406],[3390,0.1391],[3391,0.1392],[3398,0.1368],[3399,0.1368],[3401,0.136],[3409,0.1352],[3411,0.1359],[3412,0.1359],[3441,0.1449],[3488,0.1351],[3515,0.1444],[3522,0.1447],[3540,0.1418],[3541,0.1419],[3543,0.1405],[3545,0.1406],[3549,0.1391],[3551,0.1392],[3553,0.1378],[3554,0.1378],[3558,0.1368],[3559,0.1368],[3569,0.1352],[3571,0.1359],[3573,0.1359],[3576,0.138],[3578,0.1379],[3584,0.1422],[3585,0.1422],[3600,0.1446],[3604,0.1446],[3606,0.1442],[3608,0.1442],[3620,0.1418],[3621,0.1419],[3623,0.1405],[3626,0.1406],[3628,0.1391],[3631,0.1392],[3633,0.1378],[3634,0.1378],[3648,0.1351],[3675,0.1444],[3682,0.1447],[3700,0.1418],[3701,0.1419],[3703,0.1405],[3706,0.1406],[3708,0.1391],[3711,0.1392],[3718,0.1368],[3719,0.1368],[3721,0.136],[3722,0.136],[3724,0.1352],[3729,0.1352],[3731,0.1359],[3732,0.1359],[3761,0.1449],[3808,0.1351],[3835,0.1444],[3842,0.1447],[3860,0.1418],[3861,0.1419],[3863,0.1405],[3866,0.1406],[3869,0.1391],[3871,0.1392],[3873,0.1378],[3874,0.1378],[3878,0.1368],[3879,0.1368],[3881,0.136],[3932,0.1351],[3959,0.1444],[3966,0.1447],[3984,0.1418],[3985,0.1419],[3987,0.1405],[3990,0.1406],[3993,0.1391],[3995,0.1392],[3997,0.1378],[3998,0.1378],[4002,0.1368],[4003,0.1368],[4005,0.136],[4056,0.1351],[4083,0.1444],[4090,0.1447],[4108,0.1418],[4109,0.1419],[4111,0.1405],[4114,0.1406],[4117,0.1391],[4119,0.1392],[4121,0.1378],[4122,0.1378],[4126,0.1368],[4127,0.1368],[4129,0.136],[4181,0.1351],[4208,0.1444],[4215,0.1447],[4233,0.1418],[4234,0.1419],[4236,0.1405],[4239,0.1406],[4242,0.1391],[4244,0.1392],[4246,0.1378],[4247,0.1378],[4251,0.1368],[4252,0.1368],[4254,0.136],[4307,0.1351],[4334,0.1444],[4341,0.1447],[4360,0.1352],[4366,0.1359],[4371,0.1401],[4373,0.1422],[4389,0.1446],[4393,0.1446],[4397,0.1442],[4409,0.1418],[4410,0.1419],[4412,0.1405],[4414,0.1406],[4417,0.1391],[4420,0.1392],[4428,0.1366],[4435,0.1352],[4436,0]]
       }
       spine_01_ctrl.Location.Z {
        type=MovieSceneScriptingFloatChannel
        default=-0.1078
        thinned_from=185
        keys [[1790,-0.1078],[1793,-0.1079],[1798,-0.111],[1806,-0.1127],[1810,-0.1127],[1820,-0.1022],[1821,-0.1022],[1826,-0.0958],[1827,-0.0958],[1830,-0.0893],[1831,-0.0893],[1843,-0.0852],[1852,-0.0881],[1853,-0.0881],[1870,-0.1078],[1873,-0.1079],[1878,-0.111],[1879,-0.111],[1890,-0.1127],[1923,-0.085],[1954,-0.1079],[1955,-0.1079],[1958,-0.111],[1959,-0.111],[1969,-0.1127],[2003,-0.0851],[2050,-0.1127],[2083,-0.085],[2113,-0.1079],[2114,-0.1079],[2126,-0.1127],[2130,-0.1127],[2141,-0.1022],[2142,-0.1022],[2146,-0.0958],[2147,-0.0958],[2150,-0.0893],[2151,-0.0893],[2163,-0.0852],[2172,-0.0881],[2173,-0.0881],[2198,-0.111],[2199,-0.111],[2210,-0.1127],[2243,-0.085],[2245,-0.0855],[2270,-0.1078],[2275,-0.1079],[2278,-0.111],[2279,-0.111],[2289,-0.1127],[2323,-0.0851],[2370,-0.1127],[2403,-0.085],[2405,-0.0855],[2430,-0.1078],[2434,-0.1079],[2461,-0.1022],[2462,-0.1022],[2503,0.0041],[2511,0.0041],[2512,0.0044],[2515,0.2271],[2519,-0.0583],[2521,-0.0821],[2525,-0.1598],[2759,-0.0739],[2766,0.2153],[2778,0.3969],[2801,0.556],[2938,0.5539],[2986,-0.0038],[3058,-0.0038],[3070,0.0921],[3185,0.0921],[3229,-0.1079],[3233,-0.111],[3234,-0.111],[3252,-0.1087],[3253,-0.1087],[3256,-0.1022],[3257,-0.1022],[3260,-0.0958],[3262,-0.0958],[3278,-0.0852],[3287,-0.0881],[3288,-0.0881],[3305,-0.1078],[3310,-0.1079],[3325,-0.1127],[3358,-0.085],[3360,-0.0855],[3385,-0.1078],[3390,-0.1079],[3438,-0.0851],[3485,-0.1127],[3518,-0.085],[3520,-0.0855],[3521,-0.0855],[3545,-0.1078],[3549,-0.1079],[3553,-0.111],[3554,-0.111],[3564,-0.1127],[3566,-0.1127],[3572,-0.1087],[3573,-0.1087],[3576,-0.1022],[3577,-0.1022],[3581,-0.0958],[3582,-0.0958],[3598,-0.0852],[3607,-0.0881],[3608,-0.0881],[3625,-0.1078],[3645,-0.1127],[3678,-0.085],[3680,-0.0855],[3681,-0.0855],[3705,-0.1078],[3710,-0.1079],[3724,-0.1127],[3758,-0.0851],[3805,-0.1127],[3838,-0.085],[3840,-0.0855],[3841,-0.0855],[3870,-0.1079],[3884,-0.1127],[3929,-0.1127],[3962,-0.085],[3964,-0.0855],[3965,-0.0855],[3994,-0.1079],[4008,-0.1127],[4053,-0.1127],[4086,-0.085],[4088,-0.0855],[4089,-0.0855],[4118,-0.1079],[4132,-0.1127],[4178,-0.1127],[4211,-0.085],[4213,-0.0855],[4214,-0.0855],[4243,-0.1079],[4257,-0.1127],[4304,-0.1127],[4337,-0.085],[4339,-0.0855],[4340,-0.0855],[4360,-0.1104],[4364,-0.106],[4373,-0.0893],[4387,-0.0852],[4396,-0.0881],[4414,-0.1078],[4419,-0.1079],[4434,-0.1127],[4435,-0.1125],[4436,0]]
       }
       spine_01_ctrl.Rotation.X {
        type=MovieSceneScriptingFloatChannel
        default=2.1135
        thinned_from=79
        keys [[1790,2.1135],[1812,2.0098],[1847,2.1859],[1892,2.0098],[1927,2.1859],[1972,2.0098],[2008,2.1888],[2053,2.0077],[2087,2.1859],[2132,2.0098],[2167,2.1859],[2212,2.0098],[2247,2.1859],[2292,2.0098],[2328,2.1888],[2373,2.0077],[2407,2.1859],[2452,2.0098],[2468,2.0975],[2505,-0.9982],[2509,-0.9982],[2512,-0.8847],[2518,-3.0692],[2520,-2.5757],[2524,-3.0886],[2764,-0.0058],[2774,0.7442],[2780,0.4993],[2927,0.6851],[2938,0.6874],[2986,2.0848],[3058,2.0848],[3061,1.4308],[3065,-11.6769],[3070,-1.0358],[3185,-1.0358],[3247,2.0098],[3282,2.1859],[3327,2.0098],[3362,2.1859],[3407,2.0098],[3443,2.1888],[3488,2.0077],[3522,2.1859],[3567,2.0098],[3602,2.1859],[3647,2.0098],[3682,2.1859],[3727,2.0098],[3763,2.1888],[3808,2.0077],[3842,2.1859],[3887,2.0098],[3932,2.0077],[3967,-3.2211],[3996,3.0599],[4029,-2.7386],[4056,3.1263],[4098,-2.5196],[4128,2.9323],[4164,-1.2464],[4202,2.7307],[4231,-1.205],[4307,2.0077],[4341,2.1859],[4360,2.0115],[4364,2.0194],[4391,2.1859],[4413,2.1179],[4430,2.0282],[4435,2.011],[4436,0]]
       }
       spine_01_ctrl.Rotation.Y {
        type=MovieSceneScriptingFloatChannel
        default=-0.4968
        thinned_from=67
        keys [[1790,-0.4968],[1817,-0.157],[1854,-0.6009],[1897,-0.1577],[1934,-0.5962],[1977,-0.1576],[2015,-0.5935],[2057,-0.1583],[2094,-0.5962],[2137,-0.157],[2174,-0.6009],[2217,-0.1577],[2254,-0.5962],[2297,-0.1576],[2335,-0.5935],[2377,-0.1583],[2414,-0.5962],[2457,-0.157],[2470,-0.2182],[2503,-0.0392],[2510,-0.0392],[2512,-0.0511],[2515,0.0922],[2520,-0.0498],[2927,-4.6285],[2938,-4.636],[2986,-5.4997],[3058,-5.4997],[3070,2.6048],[3185,2.6048],[3252,-0.157],[3289,-0.6009],[3332,-0.1577],[3369,-0.5962],[3412,-0.1576],[3450,-0.5935],[3492,-0.1583],[3529,-0.5962],[3572,-0.157],[3609,-0.6009],[3652,-0.1577],[3689,-0.5962],[3732,-0.1576],[3770,-0.5935],[3812,-0.1583],[3849,-0.5962],[3894,-0.5935],[3936,-0.1583],[3973,-0.5962],[4018,-0.5935],[4060,-0.1583],[4097,-0.5962],[4143,-0.5935],[4185,-0.1583],[4222,-0.5962],[4269,-0.5935],[4311,-0.1583],[4348,-0.5962],[4360,-0.1632],[4364,-0.1585],[4367,-0.157],[4398,-0.6009],[4436,0]]
       }
       spine_01_ctrl.Rotation.Z {
        type=MovieSceneScriptingFloatChannel
        default=-0.4427
        thinned_from=73
        keys [[1790,-0.4427],[1815,-0.9205],[1855,-0.2162],[1895,-0.9206],[1935,-0.2181],[1974,-0.9197],[2014,-0.1767],[2055,-0.9211],[2095,-0.218],[2135,-0.9205],[2175,-0.2162],[2215,-0.9206],[2255,-0.2181],[2294,-0.9197],[2334,-0.1767],[2375,-0.9211],[2415,-0.218],[2455,-0.9205],[2503,-2.2979],[2510,-2.2979],[2512,-2.3179],[2522,-0.5684],[2526,-0.8372],[2529,-0.5945],[2531,-0.7819],[2532,-0.6622],[2536,-0.7273],[2760,-0.7434],[2927,-0.1027],[2938,-0.0942],[2986,4.4927],[3058,4.4927],[3070,-3.3116],[3185,-3.3116],[3250,-0.9205],[3290,-0.2162],[3330,-0.9206],[3370,-0.2181],[3409,-0.9197],[3449,-0.1767],[3490,-0.9211],[3530,-0.218],[3570,-0.9205],[3610,-0.2162],[3650,-0.9206],[3690,-0.2181],[3729,-0.9197],[3769,-0.1767],[3810,-0.9211],[3850,-0.218],[3893,-0.1767],[3934,-0.9211],[3974,-0.218],[4017,-0.1767],[4058,-0.9211],[4098,-0.218],[4142,-0.1767],[4183,-0.9211],[4223,-0.218],[4268,-0.1767],[4309,-0.9211],[4349,-0.218],[4360,-0.9192],[4365,-0.9205],[4399,-0.2162],[4435,-0.9103],[4436,0]]
       }
       spine_01_ctrl.Scale.X {type=MovieSceneScriptingFloatChannel, default=1, const [1,12]}
       spine_01_ctrl.Scale.Y {type=MovieSceneScriptingFloatChannel, default=1, const [1,12]}
       spine_01_ctrl.Scale.Z {type=MovieSceneScriptingFloatChannel, default=1, const [1,12]}
       spine_02_ctrl.Location.X {
        type=MovieSceneScriptingFloatChannel
        default=0.0168
        thinned_from=113
        keys [[1790,0.0168],[1794,0.0157],[1797,0.0147],[1865,0.0171],[1883,0.0113],[1887,0.0093],[1906,0.0064],[1924,0.0156],[1937,0.0168],[1945,0.0168],[1973,0.0072],[1982,0.0065],[1995,0.0101],[2005,0.0154],[2048,0.0073],[2064,0.0065],[2185,0.0171],[2208,0.0093],[2213,0.0072],[2236,0.0102],[2257,0.0168],[2265,0.0168],[2288,0.0092],[2293,0.0072],[2319,0.0127],[2325,0.0154],[2345,0.017],[2396,0.0102],[2400,0.0129],[2419,0.0168],[2425,0.0168],[2464,0.0065],[2470,-0.0072],[2501,0.0056],[2509,0.0056],[2512,0.0059],[2515,-0.0063],[2520,0.0033],[2927,0.2127],[2938,0.213],[2986,0.2611],[3058,0.2611],[3070,-0.1035],[3185,-0.1035],[3220,0.0168],[3259,0.0065],[3296,0.0171],[3300,0.0171],[3323,0.0093],[3342,0.0064],[3351,0.0102],[3360,0.0156],[3373,0.0168],[3380,0.0168],[3403,0.0092],[3408,0.0072],[3460,0.017],[3515,0.0129],[3616,0.0171],[3620,0.0171],[3629,0.016],[3648,0.0072],[3662,0.0064],[3675,0.0129],[3680,0.0156],[3700,0.0168],[3728,0.0072],[3780,0.017],[3803,0.0073],[3820,0.0065],[3835,0.0129],[3840,0.0156],[3854,0.0168],[3904,0.017],[3927,0.0073],[3944,0.0065],[3959,0.0129],[3964,0.0156],[3978,0.0168],[4028,0.017],[4051,0.0073],[4068,0.0065],[4083,0.0129],[4088,0.0156],[4102,0.0168],[4153,0.017],[4176,0.0073],[4193,0.0065],[4208,0.0129],[4213,0.0156],[4227,0.0168],[4279,0.017],[4302,0.0073],[4319,0.0065],[4334,0.0129],[4339,0.0156],[4353,0.0168],[4360,0.0082],[4365,0.0082],[4369,-13.6158]]
       }
       spine_02_ctrl.Location.Y {
        type=MovieSceneScriptingFloatChannel
        default=0.0938
        thinned_from=398
        keys [[1790,0.0938],[1792,0.0938],[1794,0.0921],[1797,0.0923],[1818,0.086],[1823,0.086],[1862,0.0952],[1865,0.0952],[1867,0.0954],[1869,0.0937],[1870,0.0938],[1874,0.0921],[1877,0.0923],[1883,0.089],[1884,0.089],[1899,0.0859],[1901,0.0862],[1902,0.0861],[1913,0.0882],[1914,0.0882],[1916,0.09],[1917,0.09],[1920,0.0918],[1921,0.0918],[1934,0.095],[1943,0.0953],[1944,0.0952],[1950,0.0957],[1963,0.089],[1964,0.0891],[1977,0.0861],[1980,0.0857],[1985,0.0863],[1988,0.0872],[1989,0.087],[1993,0.0883],[1994,0.0882],[1996,0.0902],[1997,0.0901],[2022,0.0955],[2025,0.0954],[2027,0.0948],[2028,0.0949],[2041,0.0904],[2048,0.0863],[2053,0.0863],[2058,0.0859],[2060,0.0862],[2062,0.086],[2073,0.0882],[2074,0.0882],[2076,0.09],[2077,0.09],[2081,0.0918],[2083,0.0918],[2102,0.0953],[2105,0.0952],[2107,0.0954],[2109,0.0937],[2112,0.0938],[2114,0.0921],[2117,0.0923],[2138,0.086],[2143,0.086],[2182,0.0952],[2185,0.0952],[2187,0.0954],[2189,0.0937],[2192,0.0938],[2195,0.0921],[2197,0.0923],[2203,0.089],[2204,0.089],[2208,0.0877],[2219,0.0859],[2221,0.0862],[2222,0.0861],[2233,0.0882],[2234,0.0882],[2236,0.09],[2237,0.09],[2241,0.0918],[2255,0.095],[2259,0.0952],[2260,0.0951],[2263,0.0953],[2265,0.0952],[2270,0.0957],[2275,0.0906],[2279,0.0906],[2283,0.089],[2285,0.0891],[2287,0.0877],[2288,0.0877],[2300,0.0857],[2308,0.0872],[2309,0.087],[2313,0.0883],[2314,0.0882],[2316,0.0902],[2317,0.0901],[2342,0.0955],[2345,0.0954],[2347,0.0948],[2348,0.0949],[2362,0.0904],[2365,0.0904],[2369,0.0863],[2378,0.0859],[2380,0.0862],[2382,0.086],[2393,0.0882],[2394,0.0882],[2396,0.09],[2397,0.09],[2401,0.0918],[2422,0.0953],[2425,0.0952],[2427,0.0954],[2429,0.0937],[2434,0.0921],[2437,0.0923],[2458,0.086],[2463,0.086],[2468,0.0868],[2507,-0.0317],[2512,-0.027],[2518,-0.1372],[2520,-0.1214],[2522,-0.1225],[2764,-0.0118],[2774,0.0133],[2780,-0.001],[2927,0.0116],[2938,0.0117],[2986,0.0275],[3058,0.0275],[3070,-0.0695],[3185,-0.0695],[3217,0.0953],[3220,0.0952],[3222,0.0954],[3224,0.0937],[3225,0.0938],[3230,0.0921],[3232,0.0923],[3253,0.086],[3258,0.086],[3297,0.0952],[3300,0.0952],[3302,0.0954],[3304,0.0937],[3305,0.0938],[3310,0.0921],[3312,0.0923],[3318,0.089],[3319,0.089],[3327,0.0863],[3334,0.0859],[3336,0.0862],[3337,0.0861],[3348,0.0882],[3349,0.0882],[3351,0.09],[3353,0.09],[3355,0.0918],[3358,0.0918],[3371,0.095],[3378,0.0953],[3380,0.0952],[3387,0.0957],[3390,0.0906],[3393,0.0906],[3398,0.089],[3399,0.0891],[3403,0.0877],[3404,0.0877],[3407,0.0863],[3417,0.0857],[3420,0.0863],[3421,0.0863],[3423,0.0872],[3424,0.087],[3428,0.0883],[3429,0.0882],[3431,0.0902],[3436,0.092],[3457,0.0955],[3460,0.0954],[3462,0.0948],[3463,0.0949],[3476,0.0904],[3484,0.0863],[3488,0.0863],[3493,0.0859],[3495,0.0862],[3497,0.086],[3508,0.0882],[3509,0.0882],[3511,0.09],[3512,0.09],[3516,0.0918],[3537,0.0953],[3540,0.0952],[3542,0.0954],[3544,0.0937],[3550,0.0921],[3552,0.0923],[3573,0.086],[3578,0.086],[3617,0.0952],[3620,0.0952],[3622,0.0954],[3624,0.0937],[3630,0.0921],[3632,0.0923],[3638,0.089],[3639,0.089],[3647,0.0863],[3654,0.0859],[3656,0.0862],[3657,0.0861],[3668,0.0882],[3669,0.0882],[3671,0.09],[3676,0.0918],[3690,0.095],[3694,0.0952],[3696,0.0951],[3698,0.0953],[3702,0.0952],[3706,0.0957],[3710,0.0906],[3712,0.0906],[3718,0.089],[3719,0.0891],[3727,0.0863],[3735,0.0857],[3743,0.0872],[3744,0.087],[3748,0.0883],[3749,0.0882],[3751,0.0902],[3752,0.0901],[3756,0.092],[3777,0.0955],[3780,0.0954],[3782,0.0948],[3783,0.0949],[3796,0.0904],[3798,0.0904],[3803,0.0863],[3807,0.0863],[3813,0.0859],[3815,0.0862],[3817,0.086],[3828,0.0882],[3829,0.0882],[3831,0.09],[3832,0.09],[3836,0.0918],[3857,0.0953],[3860,0.0952],[3862,0.0954],[3864,0.0937],[3867,0.0938],[3872,0.0923],[3901,0.0955],[3904,0.0954],[3906,0.0948],[3907,0.0949],[3920,0.0904],[3922,0.0904],[3927,0.0863],[3931,0.0863],[3937,0.0859],[3939,0.0862],[3941,0.086],[3952,0.0882],[3953,0.0882],[3955,0.09],[3956,0.09],[3960,0.0918],[3981,0.0953],[3984,0.0952],[3986,0.0954],[3988,0.0937],[3991,0.0938],[3996,0.0923],[4025,0.0955],[4028,0.0954],[4030,0.0948],[4031,0.0949],[4044,0.0904],[4046,0.0904],[4051,0.0863],[4055,0.0863],[4061,0.0859],[4063,0.0862],[4065,0.086],[4076,0.0882],[4077,0.0882],[4079,0.09],[4080,0.09],[4084,0.0918],[4105,0.0953],[4108,0.0952],[4110,0.0954],[4112,0.0937],[4115,0.0938],[4120,0.0923],[4150,0.0955],[4153,0.0954],[4155,0.0948],[4156,0.0949],[4169,0.0904],[4171,0.0904],[4176,0.0863],[4180,0.0863],[4186,0.0859],[4188,0.0862],[4190,0.086],[4201,0.0882],[4202,0.0882],[4204,0.09],[4205,0.09],[4209,0.0918],[4230,0.0953],[4233,0.0952],[4235,0.0954],[4237,0.0937],[4240,0.0938],[4245,0.0923],[4276,0.0955],[4279,0.0954],[4281,0.0948],[4282,0.0949],[4295,0.0904],[4297,0.0904],[4302,0.0863],[4306,0.0863],[4312,0.0859],[4314,0.0862],[4316,0.086],[4327,0.0882],[4328,0.0882],[4330,0.09],[4331,0.09],[4335,0.0918],[4356,0.0953],[4360,0.0863],[4365,0.0863],[4369,-0.7931]]
       }
       spine_02_ctrl.Location.Z {
        type=MovieSceneScriptingFloatChannel
        default=-0.0983
        thinned_from=197
        keys [[1790,-0.0983],[1794,-0.0984],[1809,-0.1024],[1843,-0.0801],[1870,-0.0983],[1874,-0.0984],[1884,-0.1023],[1890,-0.1024],[1892,-0.1016],[1917,-0.0811],[1920,-0.0801],[1925,-0.0799],[1929,-0.0809],[1932,-0.0809],[1934,-0.0848],[1940,-0.0887],[1945,-0.0934],[1950,-0.1005],[1967,-0.1024],[1970,-0.1024],[1977,-0.0991],[1980,-0.094],[1997,-0.0811],[2005,-0.0799],[2041,-0.1021],[2047,-0.1025],[2077,-0.0811],[2085,-0.0799],[2089,-0.0809],[2090,-0.0809],[2092,-0.0826],[2093,-0.0826],[2110,-0.0983],[2114,-0.0984],[2129,-0.1024],[2163,-0.0801],[2190,-0.0983],[2207,-0.1024],[2210,-0.1024],[2212,-0.1016],[2237,-0.0811],[2240,-0.0801],[2245,-0.0799],[2249,-0.0809],[2252,-0.0809],[2254,-0.0848],[2265,-0.0934],[2270,-0.1005],[2284,-0.1023],[2290,-0.1024],[2297,-0.0991],[2300,-0.094],[2304,-0.0888],[2325,-0.0799],[2367,-0.1025],[2397,-0.0811],[2400,-0.0801],[2405,-0.0799],[2409,-0.0809],[2410,-0.0809],[2412,-0.0826],[2413,-0.0826],[2430,-0.0983],[2434,-0.0984],[2449,-0.1024],[2467,-0.0817],[2501,0.0112],[2510,0.0112],[2512,0.0093],[2515,0.1999],[2519,-0.0399],[2521,-0.0568],[2525,-0.1177],[2757,-0.0732],[2768,0.2082],[2779,0.3355],[2802,0.4503],[2938,0.4491],[2986,0.2248],[3058,0.2248],[3070,0.0791],[3185,0.0791],[3244,-0.1024],[3278,-0.0801],[3305,-0.0983],[3309,-0.0984],[3319,-0.1023],[3325,-0.1024],[3327,-0.1016],[3352,-0.0811],[3360,-0.0799],[3364,-0.0809],[3367,-0.0809],[3369,-0.0848],[3375,-0.0887],[3380,-0.0934],[3390,-0.1009],[3399,-0.1023],[3405,-0.1024],[3407,-0.1016],[3415,-0.094],[3420,-0.0888],[3432,-0.0811],[3440,-0.0799],[3482,-0.1025],[3512,-0.0811],[3520,-0.0799],[3524,-0.0809],[3525,-0.0809],[3527,-0.0826],[3528,-0.0826],[3545,-0.0983],[3564,-0.1024],[3598,-0.0801],[3625,-0.0983],[3645,-0.1024],[3647,-0.1016],[3672,-0.0811],[3680,-0.0799],[3684,-0.0809],[3687,-0.0809],[3695,-0.0887],[3700,-0.0934],[3710,-0.1009],[3722,-0.1024],[3725,-0.1024],[3727,-0.1016],[3755,-0.0801],[3760,-0.0799],[3796,-0.1021],[3803,-0.1025],[3832,-0.0811],[3836,-0.0801],[3840,-0.0799],[3844,-0.0809],[3845,-0.0809],[3847,-0.0826],[3848,-0.0826],[3866,-0.0983],[3869,-0.0984],[3884,-0.1024],[3920,-0.1021],[3927,-0.1025],[3956,-0.0811],[3960,-0.0801],[3964,-0.0799],[3968,-0.0809],[3969,-0.0809],[3971,-0.0826],[3972,-0.0826],[3990,-0.0983],[3993,-0.0984],[4008,-0.1024],[4044,-0.1021],[4051,-0.1025],[4080,-0.0811],[4084,-0.0801],[4088,-0.0799],[4092,-0.0809],[4093,-0.0809],[4095,-0.0826],[4096,-0.0826],[4114,-0.0983],[4117,-0.0984],[4132,-0.1024],[4169,-0.1021],[4176,-0.1025],[4205,-0.0811],[4209,-0.0801],[4213,-0.0799],[4217,-0.0809],[4218,-0.0809],[4220,-0.0826],[4221,-0.0826],[4239,-0.0983],[4242,-0.0984],[4257,-0.1024],[4295,-0.1021],[4302,-0.1025],[4331,-0.0811],[4335,-0.0801],[4339,-0.0799],[4343,-0.0809],[4344,-0.0809],[4346,-0.0826],[4347,-0.0826],[4360,-0.1011],[4365,-0.1011],[4369,11.5953]]
       }
       spine_02_ctrl.Rotation.X {
        type=MovieSceneScriptingFloatChannel
        default=1.6601
        thinned_from=82
        keys [[1790,1.6601],[1819,1.4899],[1859,1.692],[1899,1.4898],[1939,1.6922],[1979,1.4898],[2019,1.6969],[2057,1.4886],[2099,1.692],[2139,1.4899],[2179,1.692],[2219,1.4898],[2259,1.6922],[2299,1.4898],[2339,1.6969],[2377,1.4886],[2419,1.692],[2459,1.4899],[2467,1.5157],[2496,14.0017],[2507,8.6795],[2509,18.0183],[2511,23.7212],[2513,17.2413],[2515,4.4032],[2518,-9.2844],[2519,-10.8417],[2520,-10.8203],[2522,-10.8873],[2757,-13.0169],[2767,-1.8629],[2780,0.3596],[2927,3.8793],[2938,3.8645],[2946,2.6248],[2950,2.1658],[2960,9.328],[2962,8.5292],[2974,-13.7717],[2981,-14.8044],[3058,-14.8044],[3061,-14.8821],[3063,-5.8932],[3070,40.5385],[3099,40.5385],[3166,34.2058],[3202,28.821],[3254,1.4899],[3294,1.692],[3334,1.4898],[3374,1.6922],[3414,1.4898],[3454,1.6969],[3492,1.4886],[3534,1.692],[3574,1.4899],[3614,1.692],[3654,1.4898],[3694,1.6922],[3734,1.4898],[3774,1.6969],[3812,1.4886],[3854,1.692],[3898,1.6969],[3936,1.4886],[3955,1.5953],[3978,1.692],[4022,1.6969],[4060,1.4886],[4102,1.692],[4147,1.6969],[4185,1.4886],[4227,1.692],[4273,1.6969],[4311,1.4886],[4353,1.692],[4360,1.4986],[4365,1.4986],[4369,-0.2905],[4371,14.6969]]
       }
       spine_02_ctrl.Rotation.Y {
        type=MovieSceneScriptingFloatChannel
        default=-0.2827
        thinned_from=71
        keys [[1790,-0.2827],[1825,-0.0879],[1863,-0.2922],[1907,-0.0857],[1944,-0.2857],[1987,-0.0826],[2028,-0.2929],[2067,-0.0861],[2105,-0.2857],[2145,-0.0879],[2183,-0.2922],[2227,-0.0857],[2264,-0.2857],[2307,-0.0826],[2348,-0.2929],[2387,-0.0861],[2425,-0.2857],[2465,-0.0879],[2467,-0.0883],[2471,0.0546],[2496,-2.7102],[2507,-3.8722],[2512,-2.9708],[2513,-3.7759],[2518,-0.2755],[2519,-0.0123],[2520,-0.0811],[2529,0.1406],[2530,0.1389],[2927,-3.8968],[2938,-3.8934],[2986,3.2502],[3058,3.2502],[3067,0],[3072,12.5],[3083,13.1527],[3116,1.4348],[3159,-12.887],[3176,-14.5095],[3224,4.9812],[3298,-0.2921],[3342,-0.0857],[3379,-0.2857],[3422,-0.0826],[3463,-0.2929],[3502,-0.0861],[3540,-0.2857],[3580,-0.0878],[3618,-0.2921],[3662,-0.0857],[3699,-0.2857],[3742,-0.0826],[3783,-0.2929],[3822,-0.0861],[3860,-0.2857],[3907,-0.2929],[3946,-0.0861],[3984,-0.2857],[4031,-0.2929],[4070,-0.0861],[4108,-0.2857],[4156,-0.2929],[4195,-0.0861],[4233,-0.2857],[4282,-0.2929],[4321,-0.0861],[4360,-0.1245],[4365,-0.1245],[4369,42.1385],[4371,-4.315]]
       }
       spine_02_ctrl.Rotation.Z {
        type=MovieSceneScriptingFloatChannel
        default=-0.3678
        thinned_from=68
        keys [[1790,-0.3678],[1815,-0.7625],[1855,-0.1871],[1895,-0.7622],[1935,-0.1888],[1975,-0.7611],[2014,-0.1553],[2055,-0.7624],[2095,-0.1887],[2135,-0.7625],[2175,-0.1871],[2215,-0.7622],[2255,-0.1888],[2295,-0.7611],[2334,-0.1553],[2375,-0.7624],[2415,-0.1887],[2455,-0.7625],[2468,-0.6609],[2495,-1.0315],[2529,-0.4767],[2531,-0.6304],[2532,-0.5333],[2536,-0.5886],[2760,-0.6122],[2927,0.0275],[2938,0.0328],[2986,3.2943],[3058,3.2943],[3065,2.8487],[3070,-17.9387],[3185,-17.9387],[3242,-0.7625],[3282,-0.1871],[3330,-0.7622],[3370,-0.1888],[3410,-0.7611],[3449,-0.1553],[3490,-0.7624],[3530,-0.1887],[3570,-0.7625],[3610,-0.1871],[3650,-0.7622],[3690,-0.1888],[3730,-0.7611],[3769,-0.1553],[3810,-0.7624],[3850,-0.1887],[3893,-0.1553],[3934,-0.7624],[3974,-0.1887],[4017,-0.1553],[4058,-0.7624],[4098,-0.1887],[4142,-0.1553],[4183,-0.7624],[4223,-0.1887],[4268,-0.1553],[4309,-0.7624],[4349,-0.1887],[4360,-0.7614],[4365,-0.7614],[4366,-0.7625],[4369,-5.3929],[4371,-0.9512]]
       }
       spine_02_ctrl.Scale.X {type=MovieSceneScriptingFloatChannel, default=1, const [1,12]}
       spine_02_ctrl.Scale.Y {type=MovieSceneScriptingFloatChannel, default=1, const [1,12]}
       spine_02_ctrl.Scale.Z {type=MovieSceneScriptingFloatChannel, default=1, const [1,12]}
       spine_03_ctrl.Location.X {
        type=MovieSceneScriptingFloatChannel
        default=0.0155
        thinned_from=93
        keys [[1790,0.0155],[1791,0.0161],[1804,0.0119],[1809,0.0091],[1828,0.0006],[1831,-0.0003],[1839,0.0055],[1842,0.0067],[1868,0.0157],[1878,0.0157],[1881,0.0151],[1884,0.0121],[1894,0.0063],[1911,-0.0001],[1964,0.0119],[1970,0.0091],[2031,0.0176],[2071,-0.0003],[2151,-0.0003],[2163,0.0067],[2210,0.0092],[2231,-0.0001],[2285,0.0119],[2352,0.0176],[2391,-0.0003],[2468,0.0007],[2470,0.001],[2496,0.027],[2501,0.0268],[2510,0.0268],[2515,-0.0268],[2521,0.0078],[2527,-0.0182],[2528,-0.0176],[2530,-0.0183],[2532,-0.0173],[2927,0.4204],[2938,0.4213],[2986,0.519],[3058,0.519],[3070,-0.1759],[3185,-0.1759],[3240,0.0119],[3266,-0.0003],[3307,0.0168],[3346,-0.0001],[3358,0.006],[3362,0.0069],[3400,0.0119],[3405,0.0091],[3426,-0.0007],[3437,0.0059],[3506,-0.0003],[3547,0.0161],[3560,0.0119],[3586,-0.0003],[3598,0.0067],[3627,0.0168],[3645,0.0092],[3666,-0.0001],[3682,0.0069],[3720,0.0119],[3725,0.0091],[3746,-0.0007],[3787,0.0176],[3826,-0.0003],[3867,0.0161],[3950,-0.0003],[3991,0.0161],[4009,0.0091],[4035,0.0176],[4074,-0.0003],[4115,0.0161],[4133,0.0091],[4160,0.0176],[4199,-0.0003],[4240,0.0161],[4258,0.0091],[4286,0.0176],[4325,-0.0003],[4360,0.0067],[4363,0.0067],[4364,-0.056],[4366,-0.4098],[4367,-0.6092],[4369,-0.84],[4375,-0.0003]]
       }
       spine_03_ctrl.Location.Y {
        type=MovieSceneScriptingFloatChannel
        default=0.1575
        thinned_from=184
        keys [[1790,0.1575],[1792,0.1575],[1795,0.1533],[1796,0.1535],[1803,0.1445],[1806,0.1445],[1808,0.1404],[1824,0.1322],[1867,0.1614],[1870,0.1575],[1872,0.1575],[1875,0.1533],[1877,0.1535],[1884,0.1445],[1885,0.1445],[1888,0.1404],[1905,0.1324],[1944,0.1612],[1947,0.1596],[1949,0.1597],[1963,0.1445],[1964,0.1445],[1968,0.1404],[1984,0.1322],[2027,0.162],[2035,0.1537],[2037,0.1538],[2065,0.1322],[2107,0.1614],[2111,0.1575],[2113,0.1575],[2115,0.1533],[2116,0.1535],[2125,0.1445],[2126,0.1445],[2129,0.1404],[2144,0.1322],[2188,0.1614],[2191,0.1575],[2195,0.1533],[2196,0.1535],[2205,0.1445],[2225,0.1324],[2264,0.1612],[2267,0.1596],[2269,0.1597],[2284,0.1445],[2289,0.1404],[2304,0.1322],[2347,0.162],[2355,0.1537],[2357,0.1538],[2385,0.1322],[2427,0.1614],[2435,0.1533],[2436,0.1535],[2445,0.1445],[2446,0.1445],[2449,0.1404],[2450,0.1404],[2464,0.1322],[2470,0.1325],[2507,-0.028],[2512,-0.0021],[2515,-0.1774],[2520,-0.2497],[2526,-0.1999],[2530,-0.2007],[2774,0.0337],[2780,0.0114],[2799,0.2307],[2927,0.2466],[2938,0.2466],[2986,0.169],[3058,0.169],[3070,-0.1697],[3185,-0.1697],[3222,0.1614],[3230,0.1533],[3231,0.1535],[3239,0.1445],[3241,0.1445],[3244,0.1404],[3259,0.1322],[3302,0.1614],[3310,0.1533],[3311,0.1535],[3319,0.1445],[3340,0.1324],[3379,0.1612],[3382,0.1596],[3384,0.1597],[3419,0.1322],[3463,0.162],[3470,0.1537],[3471,0.1538],[3500,0.1322],[3542,0.1614],[3547,0.1575],[3550,0.1533],[3551,0.1535],[3559,0.1445],[3560,0.1445],[3564,0.1404],[3579,0.1322],[3623,0.1614],[3626,0.1575],[3630,0.1533],[3631,0.1535],[3644,0.1404],[3645,0.1404],[3660,0.1324],[3699,0.1612],[3702,0.1596],[3704,0.1597],[3719,0.1445],[3724,0.1404],[3739,0.1322],[3782,0.162],[3790,0.1537],[3792,0.1538],[3820,0.1322],[3863,0.1614],[3866,0.1575],[3870,0.1533],[3871,0.1535],[3880,0.1445],[3881,0.1445],[3944,0.1322],[3987,0.1614],[3990,0.1575],[3994,0.1533],[3995,0.1535],[4004,0.1445],[4005,0.1445],[4030,0.162],[4038,0.1537],[4040,0.1538],[4068,0.1322],[4111,0.1614],[4114,0.1575],[4118,0.1533],[4119,0.1535],[4128,0.1445],[4129,0.1445],[4155,0.162],[4163,0.1537],[4165,0.1538],[4193,0.1322],[4236,0.1614],[4239,0.1575],[4243,0.1533],[4244,0.1535],[4253,0.1445],[4254,0.1445],[4281,0.162],[4289,0.1537],[4291,0.1538],[4319,0.1322],[4360,0.1394],[4363,0.1394],[4364,0.1524],[4366,0.2262],[4367,0.2678],[4369,0.3159],[4375,0.3078]]
       }
       spine_03_ctrl.Location.Z {
        type=MovieSceneScriptingFloatChannel
        default=-0.1243
        thinned_from=184
        keys [[1790,-0.1243],[1796,-0.1279],[1803,-0.13],[1810,-0.1305],[1812,-0.1295],[1813,-0.1297],[1843,-0.0997],[1870,-0.1243],[1876,-0.1279],[1883,-0.13],[1886,-0.13],[1888,-0.1305],[1892,-0.1295],[1893,-0.1297],[1922,-0.0998],[1931,-0.1029],[1933,-0.1028],[1963,-0.13],[1968,-0.1305],[1972,-0.1295],[1973,-0.1297],[2002,-0.0998],[2027,-0.1206],[2030,-0.1242],[2036,-0.1278],[2049,-0.1306],[2082,-0.0998],[2084,-0.1001],[2085,-0.1],[2116,-0.1279],[2123,-0.13],[2128,-0.1305],[2132,-0.1295],[2133,-0.1297],[2163,-0.0997],[2191,-0.1243],[2196,-0.1279],[2204,-0.13],[2208,-0.1305],[2212,-0.1295],[2213,-0.1297],[2242,-0.0998],[2251,-0.1029],[2253,-0.1028],[2284,-0.13],[2288,-0.1305],[2292,-0.1295],[2293,-0.1297],[2322,-0.0998],[2347,-0.1206],[2351,-0.1242],[2356,-0.1278],[2369,-0.1306],[2402,-0.0998],[2404,-0.1001],[2405,-0.1],[2431,-0.1243],[2443,-0.13],[2449,-0.1305],[2452,-0.1295],[2453,-0.1297],[2469,-0.0982],[2501,-0.0357],[2506,-0.0217],[2511,-0.072],[2512,-0.0646],[2515,0.2358],[2519,-0.0367],[2521,-0.0622],[2525,-0.1423],[2757,-0.0773],[2768,0.289],[2779,0.455],[2802,0.6221],[2938,0.6205],[2986,0.4727],[3058,0.4727],[3070,0.0974],[3185,0.0974],[3222,-0.1207],[3226,-0.1243],[3231,-0.1279],[3240,-0.13],[3243,-0.1305],[3245,-0.1305],[3247,-0.1295],[3248,-0.1297],[3278,-0.0997],[3306,-0.1243],[3311,-0.1279],[3318,-0.13],[3325,-0.1305],[3327,-0.1295],[3328,-0.1297],[3357,-0.0998],[3366,-0.1029],[3368,-0.1028],[3400,-0.13],[3405,-0.1305],[3407,-0.1295],[3408,-0.1297],[3437,-0.0998],[3462,-0.1206],[3466,-0.1242],[3471,-0.1278],[3484,-0.1306],[3517,-0.0998],[3519,-0.1001],[3520,-0.1],[3546,-0.1243],[3551,-0.1279],[3558,-0.13],[3564,-0.1305],[3567,-0.1295],[3568,-0.1297],[3598,-0.0997],[3622,-0.1207],[3625,-0.1243],[3639,-0.13],[3644,-0.1305],[3647,-0.1295],[3648,-0.1297],[3677,-0.0998],[3686,-0.1029],[3688,-0.1028],[3718,-0.13],[3725,-0.1305],[3727,-0.1295],[3728,-0.1297],[3757,-0.0998],[3785,-0.1242],[3804,-0.1306],[3837,-0.0998],[3839,-0.1001],[3840,-0.1],[3866,-0.1243],[3880,-0.13],[3928,-0.1306],[3961,-0.0998],[3963,-0.1001],[3964,-0.1],[3990,-0.1243],[4004,-0.13],[4007,-0.1305],[4033,-0.1242],[4052,-0.1306],[4085,-0.0998],[4087,-0.1001],[4088,-0.1],[4114,-0.1243],[4128,-0.13],[4131,-0.1305],[4135,-0.1295],[4136,-0.1297],[4158,-0.1242],[4177,-0.1306],[4210,-0.0998],[4212,-0.1001],[4213,-0.1],[4239,-0.1243],[4253,-0.13],[4256,-0.1305],[4260,-0.1295],[4261,-0.1297],[4284,-0.1242],[4303,-0.1306],[4336,-0.0998],[4338,-0.1001],[4339,-0.1],[4360,-0.1296],[4363,-0.1296],[4364,0.3048],[4366,2.7546],[4367,4.1359],[4369,5.7343],[4375,4.2218]]
       }
       spine_03_ctrl.Rotation.X {
        type=MovieSceneScriptingFloatChannel
        default=2.1032
        thinned_from=86
        keys [[1790,2.1032],[1824,1.7002],[1863,2.1446],[1904,1.699],[1943,2.1441],[1984,1.6992],[2023,2.1546],[2064,1.699],[2103,2.1446],[2144,1.7002],[2183,2.1446],[2224,1.699],[2263,2.1441],[2304,1.6992],[2343,2.1546],[2384,1.699],[2423,2.1446],[2464,1.7002],[2470,2.3489],[2494,21.4901],[2506,20.9178],[2511,26.9867],[2513,20.3818],[2515,9.318],[2518,-2.314],[2520,-3.7317],[2527,-3.0752],[2530,-3.0857],[2757,-2.6186],[2775,0.9014],[2780,0.6657],[2927,0.8845],[2938,0.8844],[2944,1.5937],[2963,27.3878],[2970,26.8138],[2982,-10.0575],[2988,-11.8125],[3058,-11.8125],[3060,-12.0209],[3065.055,-21.3902],[3068,9.5874],[3071.9013,-2.4582],[3101,-2.4582],[3179,-6.9102],[3194,-6.4025],[3203,6.3342],[3209,6.8093],[3250,-3.0115],[3263,-4.314],[3307,3.9703],[3379,4.8853],[3419,1.6992],[3458,2.1546],[3499,1.699],[3538,2.1446],[3579,1.7002],[3618,2.1446],[3659,1.699],[3698,2.1441],[3739,1.6992],[3778,2.1546],[3819,1.699],[3858,2.1446],[3943,1.699],[3982,2.1446],[4026,2.1546],[4067,1.699],[4106,2.1446],[4151,2.1546],[4192,1.699],[4231,2.1446],[4277,2.1546],[4318,1.699],[4357,2.1446],[4360,1.7666],[4363,1.7666],[4364,1.7636],[4365,5.3257],[4366,1.7468],[4367,1.7373],[4369,1.7263],[4373,2.0206],[4375,2.114]]
       }
       spine_03_ctrl.Rotation.Y {
        type=MovieSceneScriptingFloatChannel
        default=-0.1916
        thinned_from=84
        keys [[1790,-0.1916],[1796,-0.2127],[1831,0.0122],[1876,-0.221],[1911,0.0134],[1956,-0.2127],[1991,0.0212],[2009,-0.0467],[2011,-0.0463],[2036,-0.2345],[2071,0.0134],[2116,-0.2127],[2151,0.0122],[2196,-0.221],[2231,0.0134],[2276,-0.2127],[2311,0.0212],[2329,-0.0467],[2331,-0.0463],[2356,-0.2345],[2391,0.0134],[2436,-0.2127],[2459,-0.0746],[2470,0.0264],[2507,-6.5562],[2509,-3.2445],[2511,-1.0998],[2518,0.081],[2521,-0.1906],[2528,0.2063],[2757,5.7556],[2766,5.5017],[2812,2.2488],[2927,-0.7446],[2938,-0.7582],[2986,0.741],[3058,0.741],[3070,1.9849],[3185,1.9849],[3231,-0.2127],[3266,0.0122],[3311,-0.221],[3346,0.0134],[3391,-0.2127],[3426,0.0212],[3444,-0.0467],[3446,-0.0463],[3471,-0.2345],[3506,0.0134],[3551,-0.2127],[3586,0.0122],[3631,-0.221],[3666,0.0134],[3711,-0.2127],[3746,0.0212],[3764,-0.0467],[3766,-0.0463],[3791,-0.2345],[3826,0.0134],[3871,-0.2127],[3950,0.0134],[3995,-0.2127],[4039,-0.2345],[4074,0.0134],[4119,-0.2127],[4139,-0.0463],[4164,-0.2345],[4199,0.0134],[4244,-0.2127],[4265,-0.0463],[4290,-0.2345],[4325,0.0134],[4360,-0.0955],[4363,-0.0955],[4364,0.2315],[4365,36.8546],[4366,2.4509],[4367,4.0914],[4369,8.2129],[4373,30.98],[4375,18.5603]]
       }
       spine_03_ctrl.Rotation.Z {
        type=MovieSceneScriptingFloatChannel
        default=-0.5118
        thinned_from=99
        keys [[1790,-0.5118],[1816,-1.0296],[1855,-0.2878],[1895,-1.0292],[1935,-0.2916],[1975,-1.0276],[2015,-0.2464],[2055,-1.0291],[2095,-0.2914],[2136,-1.0296],[2175,-0.2878],[2215,-1.0292],[2255,-0.2916],[2295,-1.0276],[2335,-0.2464],[2375,-1.0291],[2415,-0.2914],[2456,-1.0296],[2468,-0.8958],[2501,-8.6216],[2507,-7.4591],[2510,-8.7493],[2519,-1.0003],[2522,-0.6234],[2526,-0.872],[2529,-0.6155],[2531,-0.82],[2532,-0.6941],[2536,-0.7767],[2757,-1.1947],[2760,-0.8241],[2927,0.2645],[2938,0.2796],[2986,7.0361],[3051,12.7377],[3058.3025,13.6838],[3064,20.0105],[3065,21.8474],[3066,-13.5524],[3069,-14.2554],[3079,-13.3078],[3085,21.6686],[3092,27.1828],[3100,15.2606],[3102,13.6059],[3107,18.5761],[3112,11.9506],[3117,14.9772],[3123,9.3081],[3128,12.1592],[3132,7.3721],[3137,10.4241],[3141,2.5519],[3145,5.3333],[3149,1.9028],[3155,4.4473],[3161,1.0406],[3165,2.6808],[3171,-0.2212],[3176,3.5641],[3182,-0.9782],[3189,5.4323],[3197,0.8537],[3209,-11.6728],[3243,0.6856],[3290,-0.2878],[3330,-1.0292],[3370,-0.2916],[3410,-1.0276],[3450,-0.2464],[3490,-1.0291],[3530,-0.2914],[3571,-1.0296],[3610,-0.2878],[3650,-1.0292],[3690,-0.2916],[3730,-1.0276],[3770,-0.2464],[3810,-1.0291],[3850,-0.2914],[3934,-1.0291],[3974,-0.2914],[4058,-1.0291],[4098,-0.2914],[4143,-0.2464],[4183,-1.0291],[4223,-0.2914],[4269,-0.2464],[4309,-1.0291],[4349,-0.2914],[4360,-1.0245],[4364,-1.0258],[4365,9.4432],[4366,-1.0296],[4367,-0.9895],[4369,-0.7589],[4373,0.291],[4375,0.7905]]
       }
       spine_03_ctrl.Scale.X {type=MovieSceneScriptingFloatChannel, default=1, const [1,16]}
       spine_03_ctrl.Scale.Y {type=MovieSceneScriptingFloatChannel, default=1, const [1,16]}
       spine_03_ctrl.Scale.Z {type=MovieSceneScriptingFloatChannel, default=1, const [1,16]}
       spine_04_ctrl.Location.X {
        type=MovieSceneScriptingFloatChannel
        default=-0.0044
        thinned_from=63
        keys [[1790,-0.0044],[1852,-0.0298],[1879,0.0072],[1880,0.0072],[1932,-0.0318],[2012,-0.032],[2130,0.0052],[2172,-0.0298],[2199,0.0072],[2200,0.0072],[2252,-0.0318],[2332,-0.032],[2502,-0.0034],[2509,-0.0034],[2515,-0.0603],[2521,0.0193],[2527,-0.0365],[2529,-0.033],[2530,-0.0345],[2533,-0.0288],[2759,0.005],[2776,-0.0165],[2927,0.0553],[2938,0.0564],[2971,0.5841],[2986,0.7431],[3058,0.7431],[3070,-0.219],[3185,-0.219],[3237,0.0079],[3257,-0.0054],[3287,-0.0298],[3314,0.0072],[3315,0.0072],[3367,-0.0318],[3447,-0.032],[3607,-0.0298],[3634,0.0072],[3635,0.0072],[3687,-0.0318],[3719,0.0079],[3767,-0.032],[3797,0.0099],[4001,0.0079],[4045,0.0099],[4125,0.0079],[4170,0.0099],[4250,0.0079],[4296,0.0099],[4360,0.0022],[4364,0.0005],[4396,-0.0298],[4420,0.0043],[4423,0.0072],[4424,0.0072],[4436,0]]
       }
       spine_04_ctrl.Location.Y {
        type=MovieSceneScriptingFloatChannel
        default=0.2205
        thinned_from=110
        keys [[1790,0.2205],[1804,0.1896],[1805,0.1896],[1809,0.1801],[1810,0.1801],[1827,0.1558],[1866,0.2257],[1868,0.2222],[1869,0.2223],[1907,0.1559],[1944,0.2253],[1986,0.156],[2025,0.2269],[2067,0.1559],[2106,0.2257],[2108,0.2222],[2109,0.2223],[2130,0.1801],[2131,0.1801],[2147,0.1558],[2186,0.2257],[2188,0.2222],[2189,0.2223],[2227,0.1559],[2264,0.2253],[2306,0.156],[2345,0.2269],[2387,0.1559],[2426,0.2257],[2428,0.2222],[2429,0.2223],[2449,0.1801],[2467,0.1558],[2469,0.1563],[2501,-0.0028],[2505,-0.0024],[2511,-0.0049],[2512,0.0126],[2520,-0.4394],[2525,-0.2898],[2531,-0.321],[2775,0.0633],[2780,0.0327],[2927,0.0642],[2938,0.0644],[2971,0.0967],[2986,0.1064],[3058,0.1064],[3070,-0.3322],[3185,-0.3322],[3221,0.2257],[3223,0.2222],[3224,0.2223],[3244,0.1801],[3257,0.1584],[3262,0.1558],[3301,0.2257],[3303,0.2222],[3304,0.2223],[3342,0.1559],[3379,0.2253],[3421,0.156],[3460,0.2269],[3502,0.1559],[3541,0.2257],[3543,0.2222],[3544,0.2223],[3560,0.1896],[3561,0.1896],[3564,0.1801],[3582,0.1558],[3621,0.2257],[3623,0.2222],[3624,0.2223],[3662,0.1559],[3699,0.2253],[3741,0.156],[3780,0.2269],[3822,0.1559],[3861,0.2257],[3863,0.2222],[3864,0.2223],[3946,0.1559],[3985,0.2257],[3987,0.2222],[4028,0.2269],[4070,0.1559],[4109,0.2257],[4111,0.2222],[4112,0.2223],[4133,0.1801],[4134,0.1801],[4195,0.1559],[4234,0.2257],[4236,0.2222],[4279,0.2269],[4321,0.1559],[4360,0.1779],[4364,0.1712],[4371,0.1558],[4410,0.2257],[4412,0.2222],[4413,0.2223],[4434,0.1787],[4436,0]]
       }
       spine_04_ctrl.Location.Z {
        type=MovieSceneScriptingFloatChannel
        default=-0.1421
        thinned_from=128
        keys [[1790,-0.1421],[1804,-0.1519],[1808,-0.1527],[1811,-0.1527],[1844,-0.1142],[1889,-0.1526],[1925,-0.1142],[1969,-0.1526],[2002,-0.1142],[2005,-0.1141],[2049,-0.1526],[2085,-0.1142],[2088,-0.1147],[2090,-0.1144],[2092,-0.1171],[2093,-0.117],[2121,-0.1496],[2124,-0.1519],[2128,-0.1527],[2164,-0.1142],[2209,-0.1526],[2245,-0.1142],[2289,-0.1526],[2322,-0.1142],[2325,-0.1141],[2369,-0.1526],[2405,-0.1142],[2408,-0.1147],[2410,-0.1144],[2412,-0.1171],[2413,-0.117],[2441,-0.1496],[2445,-0.1519],[2449,-0.1527],[2466,-0.1298],[2504,0.0137],[2510,0.0212],[2512,0.0206],[2515,0.3132],[2519,-0.1122],[2521,-0.1436],[2525,-0.2316],[2757,-0.1506],[2763,0.1674],[2768,0.3521],[2779,0.5658],[2802,0.7895],[2938,0.7875],[2971,0.6451],[2986,0.6022],[3058,0.6022],[3070,0.0979],[3185,0.0979],[3236,-0.1496],[3240,-0.1519],[3244,-0.1527],[3245,-0.1527],[3257,-0.1417],[3279,-0.1142],[3324,-0.1526],[3360,-0.1142],[3404,-0.1526],[3437,-0.1142],[3440,-0.1141],[3484,-0.1526],[3520,-0.1142],[3523,-0.1147],[3525,-0.1144],[3527,-0.1171],[3528,-0.117],[3560,-0.1519],[3564,-0.1527],[3599,-0.1142],[3644,-0.1526],[3680,-0.1142],[3724,-0.1526],[3757,-0.1142],[3760,-0.1141],[3804,-0.1526],[3840,-0.1142],[3843,-0.1147],[3845,-0.1144],[3847,-0.1171],[3964,-0.1142],[3967,-0.1147],[3969,-0.1144],[3971,-0.1171],[3972,-0.117],[4000,-0.1496],[4004,-0.1519],[4052,-0.1526],[4088,-0.1142],[4091,-0.1147],[4093,-0.1144],[4095,-0.1171],[4096,-0.117],[4124,-0.1496],[4128,-0.1519],[4133,-0.1527],[4177,-0.1526],[4213,-0.1142],[4216,-0.1147],[4218,-0.1144],[4220,-0.1171],[4221,-0.117],[4249,-0.1496],[4253,-0.1519],[4303,-0.1526],[4339,-0.1142],[4342,-0.1147],[4344,-0.1144],[4346,-0.1171],[4347,-0.117],[4360,-0.1505],[4364,-0.1485],[4388,-0.1142],[4433,-0.1526],[4435,-0.1522],[4436,0]]
       }
       spine_04_ctrl.Rotation.X {
        type=MovieSceneScriptingFloatChannel
        default=4.5516
        thinned_from=81
        keys [[1790,4.5516],[1827,2.2601],[1865,4.6879],[1907,2.2483],[1944,4.6866],[1987,2.2541],[2024,4.7347],[2067,2.2475],[2104,4.6867],[2147,2.2601],[2185,4.6879],[2227,2.2483],[2264,4.6866],[2307,2.2541],[2344,4.7347],[2387,2.2475],[2424,4.6867],[2467,2.2632],[2503,1.5132],[2510,1.5132],[2512,2.0248],[2520,-12.6904],[2525,-7.0635],[2531,-8.6564],[2758,-6.5788],[2764,-0.7556],[2770,1.7],[2782,5.6329],[2927,5.8705],[2938,5.8694],[2967,8.3817],[2975,8.2158],[2988,4.2137],[2995,3.9802],[3058,3.9802],[3062,3.9094],[3065,-0.1341],[3068,2.5625],[3075,-4.8405],[3083,-12.7699],[3185,-12.7699],[3212,9.7596],[3219,10.5163],[3254,-2.8121],[3268,-3.6442],[3304,4.6131],[3342,2.2483],[3379,4.6866],[3422,2.2541],[3459,4.7347],[3502,2.2475],[3539,4.6867],[3582,2.2601],[3620,4.6879],[3662,2.2483],[3699,4.6866],[3742,2.2541],[3779,4.7347],[3822,2.2475],[3859,4.6867],[3946,2.2475],[3983,4.6867],[4027,4.7347],[4070,2.2475],[4107,4.6867],[4195,2.2475],[4232,4.6867],[4278,4.7347],[4321,2.2475],[4360,2.7914],[4364,2.6317],[4371,2.2601],[4409,4.6879],[4422,3.9889],[4429,3.506],[4434,3.1829],[4435,3.1234],[4436,0]]
       }
       spine_04_ctrl.Rotation.Y {
        type=MovieSceneScriptingFloatChannel
        default=0.767
        thinned_from=75
        keys [[1790,0.767],[1808,-0.1399],[1852,1.9036],[1888,-0.1484],[1932,1.9711],[1968,-0.1399],[2012,1.9829],[2047,-0.1794],[2092,1.9711],[2128,-0.1399],[2172,1.9036],[2208,-0.1484],[2252,1.9711],[2288,-0.1399],[2332,1.9829],[2367,-0.1794],[2412,1.9711],[2448,-0.1399],[2471,0.9045],[2505,-0.1802],[2509,-0.1802],[2512,-0.0402],[2518,1.9419],[2519,0.4564],[2521,-0.994],[2527,0.9766],[2529,0.8518],[2530,0.8541],[2759,-0.7848],[2763,0.0256],[2770,0.5577],[2782,0.9766],[2927,0.6188],[2938,0.612],[2971,-3.1998],[2986,-4.3484],[3058,-4.3484],[3070,-0.1093],[3185,-0.1093],[3243,-0.1399],[3257,0.3491],[3287,1.9036],[3323,-0.1484],[3367,1.9711],[3403,-0.1399],[3447,1.9829],[3482,-0.1794],[3527,1.9711],[3563,-0.1399],[3607,1.9036],[3643,-0.1484],[3687,1.9711],[3723,-0.1399],[3767,1.9829],[3802,-0.1794],[3847,1.9711],[3971,1.9711],[4050,-0.1794],[4095,1.9711],[4131,-0.1399],[4175,-0.1794],[4220,1.9711],[4256,-0.1399],[4301,-0.1794],[4346,1.9711],[4360,-0.0363],[4364,0.0302],[4396,1.9036],[4432,-0.1484],[4436,0]]
       }
       spine_04_ctrl.Rotation.Z {
        type=MovieSceneScriptingFloatChannel
        default=-0.8262
        thinned_from=85
        keys [[1790,-0.8262],[1821,-1.3432],[1861,-0.7103],[1901,-1.3452],[1942,-0.7244],[1980,-1.3349],[2020,-0.6794],[2062,-1.3448],[2102,-0.7244],[2141,-1.3432],[2181,-0.7103],[2221,-1.3452],[2262,-0.7244],[2300,-1.3349],[2340,-0.6794],[2382,-1.3448],[2422,-0.7244],[2461,-1.3432],[2503,-3.3334],[2510,-3.3334],[2512,-3.3415],[2519,-1.3665],[2520,-1.372],[2524,-0.9542],[2526,-1.1498],[2529,-0.7781],[2531,-1.0891],[2532,-0.9351],[2536,-1.1391],[2757,-1.3248],[2768,-1.0013],[2775,-1.0129],[2927,-0.4115],[2938,-0.3987],[2971,6.0852],[2986,8.0385],[3058,8.0385],[3063.955,6.3653],[3065.3562,-22.772],[3065.6837,-22.957],[3070,-4.9592],[3083,-4.9592],[3084,1.3289],[3086,4.7147],[3091,-4.9592],[3099,-4.9592],[3185,10.7668],[3194,10.5491],[3209,-11.4274],[3217,-14.9298],[3256,-1.3432],[3257,-1.3421],[3296,-0.7103],[3336,-1.3452],[3377,-0.7244],[3415,-1.3349],[3455,-0.6794],[3497,-1.3448],[3537,-0.7244],[3576,-1.3432],[3616,-0.7103],[3656,-1.3452],[3697,-0.7244],[3735,-1.3349],[3775,-0.6794],[3817,-1.3448],[3857,-0.7244],[3941,-1.3448],[3981,-0.7244],[4023,-0.6794],[4065,-1.3448],[4105,-0.7244],[4190,-1.3448],[4230,-0.7244],[4316,-1.3448],[4356,-0.7244],[4360,-1.2906],[4364,-1.2776],[4405,-0.7103],[4435,-1.287],[4436,0]]
       }
       spine_04_ctrl.Scale.X {type=MovieSceneScriptingFloatChannel, default=1, const [1,14]}
       spine_04_ctrl.Scale.Y {type=MovieSceneScriptingFloatChannel, default=1, const [1,14]}
       spine_04_ctrl.Scale.Z {type=MovieSceneScriptingFloatChannel, default=1, const [1,14]}
       head_ctrl.Location.X {type=MovieSceneScriptingFloatChannel, default=0, thinned_from=24, keys [[1790,0],[4436,0]]}
       head_ctrl.Location.Y {type=MovieSceneScriptingFloatChannel, default=0, thinned_from=30, keys [[1790,0],[4436,0]]}
       head_ctrl.Location.Z {type=MovieSceneScriptingFloatChannel, default=0, thinned_from=24, keys [[1790,0],[4436,0]]}
       head_ctrl.Rotation.X {
        type=MovieSceneScriptingFloatChannel
        default=34.321
        thinned_from=113
        keys [[1790,34.321],[1806,34.3305],[1835,33.3965],[1870,34.3304],[1886,34.3226],[1915,33.4199],[1950,34.3196],[1966,34.328],[1990,33.5298],[2027,34.7973],[2075,33.42],[2110,34.321],[2126,34.3305],[2155,33.3965],[2190,34.3304],[2206,34.3226],[2235,33.4199],[2266,28.0107],[2289,26.1564],[2310,33.5298],[2347,34.7973],[2395,33.42],[2430,34.321],[2446,34.3305],[2467,33.5737],[2507,-20.6485],[2509,29.2585],[2511,59.9008],[2513,47.0465],[2515,44.6668],[2520,14.1744],[2526,4.9315],[2541,4.9734],[2757,11.5925],[2763,30.8721],[2777,37.8442],[2816,40.6601],[2848,40.6853],[2927,40.2945],[2938,40.2911],[2954,43.5917],[2971,70.8587],[2975,72.5697],[2978,69.7932],[2989,55.0152],[3018,46.8371],[3058,51.1518],[3070,77.1551],[3153,76.9675],[3185,73.7606],[3223,60.9824],[3269,35.8882],[3305,34.3304],[3321,34.3226],[3350,33.4199],[3385,34.3196],[3401,34.328],[3425,33.5298],[3462,34.7973],[3510,33.42],[3545,34.321],[3561,34.3305],[3590,33.3965],[3625,34.3304],[3641,34.3226],[3670,33.4199],[3705,34.3196],[3721,34.328],[3745,33.5298],[3782,34.7973],[3830,33.42],[3865,34.321],[3881,34.3305],[3954,33.42],[3989,34.321],[4005,34.3305],[4078,33.42],[4113,34.321],[4129,34.3305],[4155,34.7973],[4203,33.42],[4238,34.321],[4245,34.3014],[4259,29.4955],[4275,25.6206],[4329,33.42],[4360,34.1565],[4364,34.1565],[4369,27.4501],[4379,33.3965],[4414,34.3304],[4435,34.3017],[4436,0]]
       }
       head_ctrl.Rotation.Y {
        type=MovieSceneScriptingFloatChannel
        default=0.4717
        thinned_from=68
        keys [[1790,0.4717],[1809,-0.7624],[1852,3.1711],[1888,-0.5562],[1931,3.1613],[1969,-0.8859],[2011,3.3978],[2048,-0.5744],[2090,3.1596],[2129,-0.7624],[2172,3.1711],[2208,-0.5562],[2251,3.1613],[2266,-6.1325],[2289,-4.3539],[2331,3.3978],[2368,-0.5744],[2410,3.1596],[2449,-0.7624],[2502,3.1042],[2506,2.0125],[2511,-5.7388],[2518,1.6821],[2521,-0.692],[2523,-0.4775],[2525,-0.4956],[2541,-0.5069],[2760,-0.6055],[2781,0.7746],[2927,0.3759],[2938,0.3617],[2986,-10.338],[3058,-8.9603],[3070,7.8386],[3099,7.8386],[3153,-6.3063],[3185,-8.4526],[3244,-0.7624],[3287,3.1711],[3323,-0.5562],[3366,3.1613],[3404,-0.8859],[3446,3.3978],[3483,-0.5744],[3525,3.1596],[3564,-0.7624],[3607,3.1711],[3643,-0.5562],[3686,3.1613],[3724,-0.8859],[3766,3.3978],[3803,-0.5744],[3845,3.1596],[3927,-0.5744],[3969,3.1596],[4051,-0.5744],[4093,3.1596],[4176,-0.5744],[4218,3.1596],[4302,-0.5744],[4344,3.1596],[4360,-0.6152],[4364,-0.6152],[4369,37.5888],[4396,3.1711],[4432,-0.5562],[4436,0]]
       }
       head_ctrl.Rotation.Z {
        type=MovieSceneScriptingFloatChannel
        default=-1.4396
        thinned_from=104
        keys [[1790,-1.4396],[1817,-3.9415],[1857,-2.2687],[1868,-2.4616],[1880,-2.1259],[1897,-2.944],[1944,-1.2317],[1978,-3.6893],[2027,-0.8278],[2057,-2.9383],[2061,-2.9308],[2065,-2.94],[2104,-1.2396],[2137,-3.9415],[2177,-2.2687],[2188,-2.4616],[2200,-2.1259],[2217,-2.944],[2266,63.3596],[2289,54.7607],[2347,-0.8278],[2377,-2.9383],[2381,-2.9308],[2385,-2.94],[2424,-1.2396],[2457,-3.9415],[2466,-3.839],[2471,-2.7883],[2496,1.9537],[2501,2.0195],[2507,0.2386],[2508,0.6233],[2511,-0.8785],[2513,5.4909],[2522,-3.9345],[2525,-0.7454],[2526,-0.7501],[2529,-0.4248],[2531,-0.7408],[2532,-0.587],[2538,-0.9585],[2541,-1.0078],[2792,-3.7227],[2938,-3.9382],[2986,-18.3718],[3058,-2.0152],[3070,21.6631],[3099,21.6631],[3153,19.742],[3185,22.0141],[3219,-1.2396],[3252,-3.9415],[3292,-2.2687],[3303,-2.4616],[3315,-2.1259],[3332,-2.9441],[3379,-1.2317],[3413,-3.6893],[3462,-0.8278],[3492,-2.9383],[3496,-2.9308],[3500,-2.94],[3539,-1.2396],[3572,-3.9415],[3612,-2.2687],[3623,-2.4616],[3635,-2.1259],[3652,-2.9441],[3699,-1.2317],[3733,-3.6894],[3782,-0.8278],[3812,-2.9383],[3816,-2.9308],[3820,-2.94],[3859,-1.2396],[3936,-2.9383],[3940,-2.9308],[3944,-2.94],[3983,-1.2396],[4060,-2.9383],[4064,-2.9308],[4068,-2.94],[4107,-1.2396],[4155,-0.8278],[4185,-2.9383],[4189,-2.9308],[4193,-2.94],[4232,-1.2396],[4281,-0.8278],[4311,-2.9383],[4315,-2.9308],[4319,-2.94],[4360,-3.8786],[4364,-3.8786],[4367,-3.9415],[4369,-22.6259],[4401,-2.2687],[4412,-2.4616],[4424,-2.1259],[4435,-2.7907],[4436,0]]
       }
       head_ctrl.Scale.X {type=MovieSceneScriptingFloatChannel, default=1, const [1,13]}
       head_ctrl.Scale.Y {type=MovieSceneScriptingFloatChannel, default=1, const [1,13]}
       head_ctrl.Scale.Z {type=MovieSceneScriptingFloatChannel, default=1, const [1,13]}
       clavicle_l_ctrl.Location.X {type=MovieSceneScriptingFloatChannel, default=0, thinned_from=53, keys [[1790,0],[4372,0]]}
       clavicle_l_ctrl.Location.Y {type=MovieSceneScriptingFloatChannel, default=0, thinned_from=610, keys [[1790,0],[4372,0]]}
       clavicle_l_ctrl.Location.Z {type=MovieSceneScriptingFloatChannel, default=0, thinned_from=187, keys [[1790,0],[4372,0]]}
       clavicle_l_ctrl.Rotation.X {
        type=MovieSceneScriptingFloatChannel
        default=-0.0002
        thinned_from=929
        keys [[1790,-0.0002],[1792,0.0003],[1794,0.0001],[1796,0.0003],[1798,0],[1800,0],[1802,0.0002],[1804,-0.0001],[1807,0],[1813,-0.0002],[1816,0.0001],[1818,-0.0001],[1820,-0.0001],[1822,0.0001],[1824,-0.0001],[1826,0],[1828,-0.0002],[1831,0],[1832,-0.0002],[1834,-0.0002],[1838,0.0002],[1840,-0.0001],[1841,0.0001],[1845,0.0001],[1846,0],[1848,0.0002],[1852,0],[1856,0.0002],[1858,0],[1860,0.0001],[1864,-0.0002],[1865,-0.0001],[1867,-0.0001],[1868,0.0001],[1870,0],[1872,0.0002],[1875,-0.0003],[1877,0.0001],[1879,-0.0002],[1881,0],[1883,-0.0004],[1885,-0.0001],[1886,-0.0002],[1888,0.0003],[1889,-0.0001],[1894,-0.0002],[1896,0.0002],[1899,-0.0001],[1902,0],[1903,0.0003],[1905,-0.0004],[1908,0.0003],[1910,-0.0002],[1912,0],[1914,-0.0002],[1916,0],[1918,-0.0002],[1920,0.0001],[1924,0.0001],[1928,-0.0002],[1930,0.0001],[1931,-0.0001],[1937,0.0002],[1939,-0.0001],[1942,0.0002],[1944,-0.0002],[1947,0.0001],[1953,-0.0001],[1955,0.0002],[1960,-0.0001],[1962,0.0002],[1964,-0.0001],[1967,0],[1973,-0.0002],[1976,0.0002],[1978,-0.0002],[1980,0.0002],[1982,-0.0001],[1984,0],[1988,-0.0002],[1990,0],[1994,-0.0001],[1996,-0.0004],[1999,0.0001],[2001,-0.0003],[2005,0.0002],[2006,0],[2008,0.0002],[2012,0],[2015,0.0002],[2018,0],[2020,0.0002],[2024,-0.0002],[2025,-0.0001],[2027,-0.0001],[2028,0.0001],[2030,-0.0001],[2032,0.0002],[2035,-0.0003],[2037,0.0001],[2039,-0.0002],[2042,0.0001],[2048,-0.0002],[2055,0.0001],[2058,-0.0001],[2060,0.0001],[2066,-0.0001],[2068,0.0001],[2071,0],[2072,0.0002],[2074,-0.0002],[2078,0.0001],[2080,-0.0002],[2082,0.0001],[2084,-0.0002],[2087,0.0003],[2098,0],[2099,0.0001],[2101,-0.0001],[2105,0.0002],[2107,-0.0002],[2108,0],[2110,-0.0002],[2112,0.0003],[2114,0.0001],[2116,0.0003],[2118,0],[2120,0],[2122,0.0002],[2124,-0.0001],[2127,0],[2133,-0.0002],[2136,0.0001],[2138,-0.0001],[2140,-0.0001],[2142,0.0001],[2144,-0.0001],[2146,0],[2148,-0.0002],[2151,0],[2152,-0.0002],[2154,-0.0002],[2158,0.0002],[2160,-0.0001],[2161,0.0001],[2165,0.0001],[2167,0],[2168,0.0002],[2172,0],[2176,0.0002],[2178,0],[2180,0.0001],[2184,-0.0002],[2185,-0.0001],[2187,-0.0001],[2188,0.0001],[2190,0],[2192,0.0002],[2195,-0.0003],[2197,0.0001],[2199,-0.0002],[2201,0],[2203,-0.0004],[2205,-0.0001],[2206,-0.0002],[2208,0.0003],[2209,-0.0001],[2214,-0.0002],[2216,0.0002],[2219,-0.0001],[2222,0],[2223,0.0003],[2225,-0.0004],[2228,0.0003],[2230,-0.0002],[2232,0],[2234,-0.0002],[2236,0],[2238,-0.0002],[2240,0.0001],[2244,0.0001],[2246,-0.0002],[2248,-0.0002],[2250,0.0001],[2251,-0.0001],[2257,0.0002],[2259,-0.0001],[2262,0.0002],[2264,-0.0002],[2267,0.0001],[2273,-0.0001],[2275,0.0002],[2280,-0.0001],[2282,0.0002],[2284,-0.0001],[2287,0],[2293,-0.0002],[2296,0.0002],[2298,-0.0002],[2300,0.0002],[2302,-0.0001],[2304,0],[2308,-0.0002],[2310,0],[2314,-0.0001],[2316,-0.0004],[2319,0.0001],[2321,-0.0003],[2325,0.0002],[2326,0],[2328,0.0002],[2332,0],[2335,0.0002],[2338,0],[2340,0.0002],[2344,-0.0002],[2345,-0.0001],[2347,-0.0001],[2348,0.0001],[2350,-0.0001],[2352,0.0002],[2355,-0.0003],[2357,0.0001],[2359,-0.0002],[2362,0.0001],[2368,-0.0002],[2375,0.0001],[2378,-0.0001],[2383,0],[2386,-0.0001],[2388,0.0001],[2391,0],[2392,0.0002],[2394,-0.0002],[2398,0.0001],[2400,-0.0002],[2402,0.0001],[2404,-0.0002],[2407,0.0003],[2418,0],[2419,0.0001],[2421,-0.0001],[2425,0.0002],[2427,-0.0002],[2428,0],[2430,-0.0002],[2432,0.0003],[2434,0.0001],[2436,0.0003],[2438,0],[2440,0],[2442,0.0002],[2444,-0.0001],[2447,0],[2453,-0.0002],[2456,0.0001],[2458,-0.0001],[2460,-0.0001],[2462,0.0001],[2464,-0.0001],[2466,0],[2471,-0.1967],[2496,-14.4886],[2506,-15.6537],[2507,-15.6169],[2511,-15.7669],[2512,-14.0517],[2515,-4.3993],[2521,-2.7552],[2523,-2.7713],[2533,-2.7706],[2778,0.0762],[2788,0.0008],[2801,0.0008],[2927,-0.0004],[2938,-0.0007],[2986,-0.3093],[3058,-0.3157],[3070,-0.2611],[3099,-0.2611],[3185,-1.2894],[3216,-0.0001],[3220,0.0002],[3222,-0.0002],[3223,0],[3225,-0.0002],[3227,0.0003],[3229,0.0001],[3231,0.0003],[3233,0],[3235,-0.0001],[3237,0.0002],[3239,-0.0001],[3242,0],[3248,-0.0002],[3251,0.0001],[3253,-0.0001],[3257,0.0001],[3259,-0.0001],[3261,0.0001],[3263,-0.0002],[3266,0],[3269,-0.0002],[3273,0.0002],[3275,-0.0001],[3276,0.0001],[3280,0.0001],[3281,0],[3283,0.0002],[3287,0],[3291,0.0001],[3293,0],[3295,0.0001],[3299,-0.0002],[3302,-0.0001],[3303,0.0001],[3305,0],[3307,0.0002],[3310,-0.0003],[3312,0.0001],[3314,-0.0002],[3316,0],[3318,-0.0004],[3320,-0.0001],[3321,-0.0002],[3323,0.0003],[3327,-0.0002],[3329,-0.0002],[3331,0.0002],[3334,-0.0001],[3337,0],[3338,0.0003],[3340,-0.0004],[3343,0.0003],[3345,-0.0002],[3347,0],[3349,-0.0002],[3351,0],[3353,-0.0002],[3355,0.0001],[3359,0.0001],[3361,-0.0002],[3363,-0.0002],[3365,0.0001],[3366,-0.0001],[3370,0.0001],[3372,0.0002],[3374,-0.0001],[3377,0.0001],[3379,-0.0002],[3382,0.0001],[3387,-0.0001],[3390,0.0002],[3395,-0.0001],[3397,0.0002],[3399,-0.0001],[3402,0],[3408,-0.0002],[3411,0.0002],[3413,-0.0002],[3415,0.0002],[3417,-0.0001],[3419,0],[3423,-0.0002],[3425,0],[3429,-0.0001],[3431,-0.0004],[3434,0.0001],[3436,-0.0003],[3440,0.0002],[3441,0],[3443,0.0002],[3447,0],[3450,0.0002],[3453,0],[3455,0.0002],[3459,-0.0002],[3460,-0.0001],[3462,-0.0001],[3463,0.0001],[3465,-0.0001],[3467,0.0002],[3470,-0.0003],[3472,0.0001],[3474,-0.0002],[3476,0.0001],[3480,0],[3482,-0.0002],[3490,0],[3493,-0.0001],[3495,0.0001],[3501,-0.0001],[3503,0.0001],[3506,0],[3507,0.0002],[3509,-0.0002],[3513,0.0001],[3515,-0.0002],[3517,0.0001],[3519,-0.0002],[3522,0.0004],[3533,0],[3534,0.0001],[3536,-0.0001],[3540,0.0002],[3542,-0.0002],[3543,0],[3545,-0.0002],[3547,0.0003],[3549,0.0001],[3551,0.0003],[3553,0],[3555,-0.0001],[3557,0.0002],[3559,-0.0001],[3562,0],[3568,-0.0002],[3571,0.0001],[3573,-0.0001],[3577,0.0001],[3579,-0.0001],[3581,0],[3583,-0.0002],[3586,0],[3589,-0.0002],[3593,0.0002],[3595,-0.0001],[3596,0],[3600,0.0001],[3601,0],[3603,0.0002],[3611,0.0001],[3613,0],[3614,0.0001],[3619,-0.0002],[3622,-0.0001],[3623,0.0001],[3625,0],[3627,0.0002],[3630,-0.0003],[3632,0.0001],[3634,-0.0002],[3636,0],[3638,-0.0004],[3640,-0.0001],[3641,-0.0002],[3643,0.0003],[3644,-0.0001],[3649,-0.0002],[3651,0.0002],[3654,-0.0001],[3657,0],[3658,0.0003],[3660,-0.0004],[3663,0.0003],[3665,-0.0002],[3667,0],[3669,-0.0002],[3671,0],[3673,-0.0002],[3675,0.0001],[3679,0.0001],[3681,-0.0002],[3683,-0.0002],[3685,0.0001],[3686,-0.0001],[3692,0.0002],[3694,-0.0001],[3697,0.0001],[3699,-0.0002],[3702,0.0001],[3708,-0.0001],[3710,0.0002],[3715,-0.0001],[3717,0.0002],[3719,-0.0001],[3722,0],[3728,-0.0002],[3731,0.0002],[3733,-0.0002],[3735,0.0002],[3737,-0.0001],[3739,0],[3743,-0.0002],[3746,0],[3749,-0.0001],[3751,-0.0004],[3754,0.0001],[3756,-0.0003],[3760,0.0002],[3761,0],[3763,0.0002],[3767,0],[3769,0.0002],[3773,0],[3775,0.0002],[3779,-0.0002],[3780,-0.0001],[3782,-0.0001],[3783,0.0001],[3785,-0.0001],[3787,0.0002],[3790,-0.0003],[3792,0.0001],[3794,-0.0002],[3796,0.0001],[3803,-0.0002],[3810,0],[3813,-0.0001],[3815,0.0001],[3821,-0.0001],[3823,0.0001],[3826,0],[3827,0.0002],[3829,-0.0002],[3833,0.0001],[3835,-0.0002],[3837,0.0001],[3839,-0.0002],[3842,0.0004],[3853,0],[3854,0.0001],[3856,-0.0001],[3860,0.0002],[3862,-0.0002],[3863,0.0001],[3865,-0.0002],[3867,0.0003],[3869,0.0001],[3871,0.0003],[3873,0],[3875,0],[3877,0.0002],[3879,-0.0001],[3882,0],[3888,-0.0002],[3893,0.0002],[3897,0],[3899,0.0002],[3903,-0.0002],[3904,-0.0001],[3906,-0.0001],[3907,0.0001],[3909,-0.0001],[3911,0.0002],[3914,-0.0003],[3916,0.0001],[3918,-0.0002],[3920,0.0001],[3927,-0.0002],[3934,0],[3937,-0.0001],[3939,0.0001],[3945,-0.0001],[3947,0.0001],[3950,0],[3951,0.0002],[3953,-0.0002],[3957,0.0001],[3959,-0.0002],[3961,0.0001],[3963,-0.0002],[3966,0.0004],[3977,0],[3978,0.0001],[3980,-0.0001],[3984,0.0002],[3986,-0.0002],[3987,0.0001],[3989,-0.0002],[3991,0.0003],[3993,0.0001],[3995,0.0003],[3997,0],[3999,0],[4001,0.0002],[4003,-0.0001],[4006,0],[4012,-0.0002],[4017,0.0002],[4021,0],[4023,0.0002],[4027,-0.0002],[4028,-0.0001],[4030,-0.0001],[4031,0.0001],[4033,-0.0001],[4035,0.0002],[4038,-0.0003],[4040,0.0001],[4042,-0.0002],[4044,0.0001],[4051,-0.0002],[4058,0],[4061,-0.0001],[4063,0.0001],[4069,-0.0001],[4071,0.0001],[4074,0],[4075,0.0002],[4077,-0.0002],[4081,0.0001],[4083,-0.0002],[4085,0.0001],[4087,-0.0002],[4090,0.0004],[4101,0],[4102,0.0001],[4104,-0.0001],[4108,0.0002],[4110,-0.0002],[4111,0.0001],[4113,-0.0002],[4115,0.0003],[4117,0.0001],[4119,0.0003],[4121,0],[4123,0],[4125,0.0002],[4127,-0.0001],[4130,0],[4136,-0.0002],[4142,0.0002],[4146,0],[4148,0.0002],[4152,-0.0002],[4153,-0.0001],[4155,-0.0001],[4156,0.0001],[4158,-0.0001],[4160,0.0002],[4163,-0.0003],[4165,0.0001],[4167,-0.0002],[4169,0.0001],[4176,-0.0002],[4183,0],[4186,-0.0001],[4188,0.0001],[4194,-0.0001],[4196,0.0001],[4199,0],[4200,0.0002],[4202,-0.0002],[4206,0.0001],[4208,-0.0002],[4210,0.0001],[4212,-0.0002],[4215,0.0004],[4226,0],[4227,0.0001],[4229,-0.0001],[4233,0.0002],[4235,-0.0002],[4236,0.0001],[4238,-0.0002],[4240,0.0003],[4242,0.0001],[4244,0.0003],[4246,0],[4248,0],[4250,0.0002],[4252,-0.0001],[4255,0],[4261,-0.0002],[4268,0.0002],[4272,0],[4274,0.0002],[4278,-0.0002],[4279,-0.0001],[4281,-0.0001],[4282,0.0001],[4284,-0.0001],[4286,0.0002],[4289,-0.0003],[4291,0.0001],[4293,-0.0002],[4295,0.0001],[4302,-0.0002],[4309,0],[4312,-0.0001],[4314,0.0001],[4320,-0.0001],[4322,0.0001],[4325,0],[4326,0.0002],[4328,-0.0002],[4332,0.0001],[4334,-0.0002],[4336,0.0001],[4338,-0.0002],[4341,0.0004],[4352,0],[4353,0.0001],[4360,-0.0001],[4366,0.0001],[4372,-0.0002]]
       }
       clavicle_l_ctrl.Rotation.Y {
        type=MovieSceneScriptingFloatChannel
        default=11.4668
        thinned_from=70
        keys [[1790,11.4668],[1821,16.3276],[1865,11.2533],[1901,16.3211],[1945,11.2535],[1981,16.3265],[2025,11.2533],[2061,16.3207],[2105,11.2517],[2141,16.3276],[2185,11.2533],[2221,16.3211],[2265,11.2535],[2301,16.3265],[2345,11.2533],[2381,16.3207],[2425,11.2517],[2461,16.3276],[2468,16.0714],[2484,31.0212],[2501,36.4059],[2506,37.0862],[2507,37.0676],[2511,41.8119],[2513,23.5707],[2518,13.2958],[2520,20.8712],[2523,16.6945],[2760,16.9169],[2778,15.6404],[2788,15.6734],[2938,15.5849],[2986,-6.9463],[3058,-8.4183],[3063,-7.9728],[3083,7.329],[3099,7.329],[3185,7.2199],[3214,8.4918],[3269,24.1503],[3300,11.2533],[3346,21.6957],[3380,11.1083],[3420,19.6674],[3460,11.2533],[3497,17.9185],[3540,11.2517],[3576,17.7802],[3620,11.2534],[3656,16.3211],[3700,11.2535],[3736,16.3265],[3780,11.2533],[3816,16.3207],[3860,11.2517],[3904,11.2533],[3940,16.3207],[3984,11.2517],[4028,11.2533],[4064,16.3207],[4108,11.2517],[4153,11.2533],[4189,16.3207],[4233,11.2517],[4279,11.2533],[4315,16.3207],[4360,15.8265],[4364,15.5895],[4372,15.0253]]
       }
       clavicle_l_ctrl.Rotation.Z {
        type=MovieSceneScriptingFloatChannel
        default=-14.4937
        thinned_from=941
        keys [[1790,-14.4937],[2466,-14.4936],[2470,-14.5255],[2501,35.0517],[2506,31.4043],[2511,33.4404],[2515,-8.9243],[2519,-25.956],[2522,-16.6734],[2525,-15.2386],[2532,-16.0865],[2760,-15.6255],[2778,-14.465],[2788,-14.493],[2927,-13.9733],[2938,-13.9562],[2986,-1.3811],[3058,-1.418],[3070,-19.7535],[3099,-19.7535],[3185,-27.8611],[3216,-14.4937],[4372,-14.4938]]
       }
       clavicle_l_ctrl.Scale.X {type=MovieSceneScriptingFloatChannel, default=1, const [1,12]}
       clavicle_l_ctrl.Scale.Y {type=MovieSceneScriptingFloatChannel, default=1, const [1,12]}
       clavicle_l_ctrl.Scale.Z {type=MovieSceneScriptingFloatChannel, default=1, const [1,12]}
       clavicle_r_ctrl.Location.X {type=MovieSceneScriptingFloatChannel, default=0, thinned_from=35, keys [[1790,0],[4436,0]]}
       clavicle_r_ctrl.Location.Y {type=MovieSceneScriptingFloatChannel, default=0, thinned_from=648, keys [[1790,0],[4436,0]]}
       clavicle_r_ctrl.Location.Z {type=MovieSceneScriptingFloatChannel, default=0, thinned_from=193, keys [[1790,0],[4436,0]]}
       clavicle_r_ctrl.Rotation.X {
        type=MovieSceneScriptingFloatChannel
        default=-0.0005
        thinned_from=926
        keys [[1790,-0.0005],[1792,-0.0008],[1795,0.0004],[1798,-0.0002],[1800,0.0002],[1802,-0.0001],[1803,0.0001],[1805,0.0002],[1813,-0.0003],[1815,-0.0001],[1817,-0.0009],[1818,-0.0001],[1820,-0.0004],[1821,-0.0003],[1822,-0.0006],[1824,-0.0003],[1826,-0.0008],[1828,-0.0001],[1830,-0.0007],[1832,-0.0001],[1833,-0.0006],[1835,0.0009],[1838,-0.0008],[1841,0.0003],[1842,0.0003],[1844,0.0006],[1846,-0.0006],[1847,0.0007],[1851,-0.0005],[1853,0.0004],[1855,-0.0003],[1857,0.001],[1859,-0.0003],[1861,0.0006],[1864,-0.0002],[1865,-0.0002],[1869,-0.0009],[1870,-0.0009],[1871,-0.001],[1876,0.0008],[1877,0.0001],[1879,0.0009],[1881,-0.0001],[1883,0.0003],[1888,-0.0007],[1889,-0.0004],[1891,-0.0014],[1893,0],[1895,-0.0012],[1898,-0.0002],[1899,-0.0002],[1901,0.0007],[1903,0],[1905,0.0007],[1907,-0.0003],[1909,0.0001],[1911,-0.0007],[1914,-0.0001],[1915,-0.0003],[1917,0.0001],[1919,-0.0004],[1921,0.0001],[1922,-0.0001],[1924,0.0001],[1925,-0.0001],[1929,0.0008],[1931,0.0007],[1934,0.0003],[1935,0.0004],[1937,-0.0001],[1938,0.0001],[1939,-0.0002],[1941,0.0007],[1942,0.0005],[1943,0.0008],[1945,-0.0003],[1946,0.0005],[1949,-0.0008],[1950,-0.0005],[1951,-0.0006],[1953,0.0008],[1955,-0.0011],[1957,-0.0001],[1958,-0.0008],[1960,-0.0003],[1961,-0.0006],[1962,0.001],[1964,-0.0011],[1967,0.0006],[1969,-0.0023],[1971,-0.0008],[1974,-0.002],[1976,0.0002],[1978,-0.0009],[1980,-0.0004],[1983,-0.001],[1985,0.0001],[1986,-0.0005],[1987,-0.0003],[1989,-0.0008],[1991,0.0006],[1992,0.0002],[1994,0.0008],[1996,0.0004],[1997,0.0007],[1998,0.0004],[2001,0.0012],[2003,0.0005],[2005,0.0005],[2007,-0.0002],[2008,-0.0001],[2010,-0.0004],[2012,0.0005],[2013,-0.0003],[2015,0.0003],[2016,-0.0004],[2019,0.0003],[2020,0.0001],[2021,0.0001],[2023,-0.0004],[2025,0.0005],[2027,-0.0007],[2030,0.0005],[2033,-0.0001],[2035,0.0006],[2036,0],[2038,0.0004],[2040,-0.0008],[2043,0.0003],[2044,0.0002],[2045,0.0008],[2048,-0.0008],[2049,-0.0005],[2050,-0.0007],[2052,0.0004],[2055,-0.0012],[2060,0.0004],[2063,-0.0001],[2066,0.0003],[2068,-0.0004],[2071,0.0006],[2073,-0.0008],[2076,0.0003],[2079,-0.0012],[2083,0.0007],[2084,-0.0002],[2087,0.0007],[2092,-0.0006],[2095,0.0004],[2097,-0.0001],[2098,0.0001],[2099,-0.0002],[2101,0.0007],[2102,-0.0001],[2104,0.001],[2107,-0.0003],[2109,0],[2112,-0.0008],[2115,0.0004],[2118,-0.0002],[2120,0.0002],[2122,-0.0001],[2123,0.0001],[2126,0.0002],[2130,-0.0002],[2131,-0.0001],[2133,-0.0003],[2135,-0.0001],[2137,-0.0009],[2138,-0.0001],[2140,-0.0004],[2141,-0.0003],[2142,-0.0006],[2144,-0.0002],[2146,-0.0008],[2148,-0.0001],[2150,-0.0007],[2152,-0.0002],[2153,-0.0006],[2155,0.0009],[2159,-0.0008],[2161,0.0003],[2162,0.0003],[2164,0.0006],[2166,-0.0006],[2167,0.0007],[2171,-0.0005],[2173,0.0004],[2175,-0.0004],[2177,0.001],[2179,-0.0003],[2181,0.0006],[2184,-0.0002],[2185,-0.0002],[2189,-0.001],[2190,-0.0008],[2191,-0.001],[2196,0.0008],[2197,0.0001],[2199,0.0009],[2201,-0.0001],[2203,0.0003],[2208,-0.0007],[2209,-0.0004],[2211,-0.0014],[2213,0],[2215,-0.0012],[2221,0.0007],[2223,0],[2225,0.0007],[2227,-0.0003],[2229,0.0001],[2231,-0.0007],[2234,0],[2235,-0.0003],[2237,0.0001],[2239,-0.0004],[2241,0.0001],[2242,-0.0001],[2244,0.0001],[2245,-0.0001],[2249,0.0008],[2251,0.0007],[2254,0.0003],[2255,0.0004],[2257,-0.0001],[2258,0.0001],[2259,-0.0002],[2261,0.0007],[2262,0.0005],[2263,0.0008],[2265,-0.0003],[2266,0.0005],[2269,-0.0008],[2270,-0.0005],[2271,-0.0005],[2273,0.0008],[2275,-0.0011],[2277,-0.0001],[2278,-0.0009],[2280,-0.0003],[2281,-0.0006],[2282,0.001],[2284,-0.0011],[2287,0.0006],[2289,-0.0023],[2291,-0.0008],[2294,-0.002],[2296,0.0002],[2298,-0.0009],[2300,-0.0004],[2303,-0.001],[2305,0.0001],[2306,-0.0005],[2307,-0.0003],[2309,-0.0008],[2311,0.0006],[2312,0.0002],[2314,0.0009],[2316,0.0004],[2317,0.0007],[2318,0.0004],[2321,0.0012],[2323,0.0005],[2324,0.0005],[2327,-0.0002],[2328,-0.0001],[2330,-0.0005],[2332,0.0005],[2333,-0.0003],[2335,0.0002],[2336,-0.0004],[2339,0.0002],[2341,0.0001],[2343,-0.0004],[2345,0.0005],[2347,-0.0007],[2350,0.0006],[2353,-0.0001],[2355,0.0006],[2356,0],[2358,0.0004],[2360,-0.0008],[2363,0.0003],[2364,0.0002],[2365,0.0008],[2368,-0.0008],[2369,-0.0005],[2370,-0.0007],[2372,0.0004],[2375,-0.0012],[2380,0.0004],[2383,-0.0001],[2386,0.0003],[2388,-0.0005],[2391,0.0006],[2393,-0.0008],[2396,0.0003],[2399,-0.0012],[2403,0.0007],[2404,-0.0002],[2407,0.0007],[2412,-0.0006],[2415,0.0004],[2417,-0.0001],[2418,0.0001],[2419,-0.0002],[2421,0.0007],[2422,-0.0001],[2424,0.001],[2427,-0.0003],[2429,0],[2432,-0.0007],[2435,0.0004],[2438,-0.0002],[2440,0.0002],[2442,-0.0001],[2443,0.0001],[2446,0.0002],[2450,-0.0002],[2451,-0.0001],[2453,-0.0003],[2455,-0.0001],[2457,-0.0009],[2458,-0.0001],[2460,-0.0004],[2461,-0.0003],[2462,-0.0006],[2464,-0.0002],[2484,6.7097],[2501,7.9384],[2506,10.0199],[2511,8.5254],[2512,8.954],[2515,3.9551],[2519,-3.4348],[2523,-3.4083],[2531,-3.411],[2778,0.0929],[2927,0.0073],[2938,0.0061],[2986,-1.3943],[3058,-1.3943],[3070,5.2487],[3099,5.2487],[3185,4.8404],[3216,0.0007],[3217,-0.0001],[3219,0.001],[3222,-0.0003],[3224,0],[3227,-0.0007],[3230,0.0004],[3233,-0.0002],[3235,0.0002],[3237,-0.0001],[3238,0.0001],[3241,0.0002],[3248,-0.0003],[3250,-0.0001],[3252,-0.0009],[3254,-0.0002],[3255,-0.0004],[3256,-0.0003],[3257,-0.0006],[3259,-0.0003],[3261,-0.0008],[3263,-0.0001],[3265,-0.0007],[3267,-0.0001],[3268,-0.0006],[3270,0.0009],[3274,-0.0008],[3279,0.0006],[3281,-0.0006],[3282,0.0007],[3286,-0.0005],[3288,0.0004],[3290,-0.0004],[3292,0.001],[3294,-0.0003],[3296,0.0006],[3304,-0.0009],[3305,-0.0008],[3306,-0.001],[3311,0.0008],[3312,0.0001],[3314,0.0008],[3316,-0.0001],[3318,0.0003],[3323,-0.0007],[3324,-0.0004],[3326,-0.0014],[3328,0],[3330,-0.0012],[3333,-0.0002],[3334,-0.0002],[3336,0.0007],[3338,0],[3340,0.0007],[3342,-0.0004],[3344,0.0001],[3346,-0.0008],[3349,0],[3350,-0.0003],[3352,0.0001],[3354,-0.0004],[3356,0.0001],[3357,-0.0002],[3359,0.0001],[3360,-0.0001],[3364,0.0008],[3370,0.0004],[3372,0],[3373,0.0001],[3374,-0.0002],[3376,0.0007],[3377,0.0005],[3378,0.0008],[3380,-0.0003],[3381,0.0005],[3383,-0.0008],[3388,0.0009],[3390,-0.0011],[3392,-0.0002],[3393,-0.0008],[3395,-0.0004],[3396,-0.0007],[3397,0.001],[3399,-0.0011],[3402,0.0006],[3404,-0.0023],[3406,-0.0007],[3409,-0.0021],[3411,0.0002],[3413,-0.0009],[3415,-0.0004],[3418,-0.001],[3420,0.0001],[3421,-0.0005],[3422,-0.0004],[3424,-0.0008],[3426,0.0006],[3427,0.0002],[3429,0.0009],[3431,0.0004],[3432,0.0007],[3433,0.0004],[3436,0.0012],[3439,0.0005],[3440,0.0006],[3442,-0.0001],[3443,0],[3445,-0.0005],[3447,0.0004],[3448,-0.0003],[3450,0.0003],[3451,-0.0004],[3454,0.0003],[3458,-0.0004],[3460,0.0005],[3462,-0.0007],[3465,0.0005],[3468,-0.0001],[3470,0.0005],[3471,0],[3473,0.0004],[3475,-0.0008],[3478,0.0003],[3479,0.0002],[3480,0.0008],[3483,-0.0008],[3484,-0.0005],[3485,-0.0007],[3487,0.0003],[3490,-0.0012],[3495,0.0004],[3499,-0.0001],[3501,0.0002],[3503,-0.0005],[3506,0.0005],[3508,-0.0008],[3511,0.0003],[3514,-0.0012],[3518,0.0007],[3519,-0.0002],[3522,0.0007],[3527,-0.0007],[3530,0.0004],[3532,0],[3533,0.0001],[3534,-0.0002],[3536,0.0007],[3537,-0.0001],[3539,0.001],[3542,-0.0003],[3544,0],[3547,-0.0007],[3550,0.0004],[3553,-0.0002],[3555,0.0002],[3557,-0.0001],[3558,0.0001],[3559,0],[3560,0.0002],[3565,-0.0002],[3566,-0.0001],[3568,-0.0003],[3570,-0.0001],[3572,-0.0009],[3574,-0.0002],[3575,-0.0004],[3576,-0.0003],[3577,-0.0006],[3579,-0.0003],[3581,-0.0007],[3583,-0.0001],[3585,-0.0007],[3587,-0.0001],[3588,-0.0006],[3590,0.0009],[3594,-0.0008],[3596,0.0003],[3597,0.0003],[3599,0.0006],[3601,-0.0006],[3602,0.0007],[3606,-0.0005],[3608,0.0004],[3610,-0.0003],[3612,0.001],[3614,-0.0003],[3616,0.0006],[3624,-0.0009],[3625,-0.0008],[3626,-0.001],[3631,0.0008],[3632,0.0001],[3634,0.0008],[3636,-0.0001],[3638,0.0003],[3643,-0.0007],[3644,-0.0004],[3646,-0.0014],[3648,0],[3650,-0.0012],[3653,-0.0002],[3654,-0.0002],[3656,0.0007],[3658,0],[3660,0.0006],[3662,-0.0004],[3664,0.0001],[3666,-0.0008],[3669,0],[3670,-0.0003],[3672,0.0001],[3674,-0.0004],[3676,0.0001],[3677,-0.0002],[3679,0.0001],[3680,-0.0001],[3684,0.0008],[3689,0.0003],[3690,0.0004],[3692,0],[3693,0.0001],[3694,-0.0002],[3696,0.0007],[3697,0.0005],[3698,0.0008],[3700,-0.0003],[3701,0.0005],[3704,-0.0008],[3708,0.0008],[3710,-0.0011],[3712,-0.0002],[3713,-0.0008],[3715,-0.0004],[3716,-0.0006],[3717,0.001],[3719,-0.0011],[3722,0.0006],[3724,-0.0023],[3726,-0.0008],[3729,-0.0021],[3731,0.0002],[3733,-0.0009],[3735,-0.0004],[3738,-0.001],[3740,0.0001],[3741,-0.0005],[3742,-0.0004],[3744,-0.0008],[3746,0.0006],[3747,0.0002],[3749,0.0009],[3751,0.0004],[3752,0.0007],[3753,0.0004],[3756,0.0012],[3758,0.0005],[3760,0.0006],[3762,-0.0002],[3763,0],[3765,-0.0005],[3767,0.0004],[3768,-0.0003],[3770,0.0002],[3771,-0.0004],[3774,0.0003],[3775,0],[3776,0.0001],[3778,-0.0004],[3780,0.0005],[3782,-0.0007],[3785,0.0005],[3788,0],[3790,0.0005],[3791,0],[3793,0.0004],[3795,-0.0008],[3798,0.0003],[3799,0.0002],[3800,0.0008],[3803,-0.0008],[3804,-0.0005],[3805,-0.0007],[3807,0.0003],[3810,-0.0012],[3815,0.0003],[3818,-0.0001],[3821,0.0002],[3823,-0.0004],[3826,0.0006],[3828,-0.0008],[3831,0.0003],[3834,-0.0012],[3838,0.0007],[3839,-0.0002],[3842,0.0007],[3847,-0.0007],[3850,0.0004],[3852,0],[3853,0.0001],[3854,-0.0002],[3856,0.0007],[3857,-0.0001],[3859,0.001],[3862,-0.0003],[3864,0],[3867,-0.0007],[3870,0.0003],[3873,-0.0002],[3875,0.0002],[3877,-0.0001],[3878,0.0001],[3881,0.0002],[3885,-0.0002],[3886,-0.0001],[3888,-0.0003],[3891,0.0004],[3892,-0.0003],[3894,0.0002],[3895,-0.0004],[3898,0.0003],[3899,0],[3900,0.0001],[3902,-0.0004],[3904,0.0005],[3906,-0.0007],[3909,0.0005],[3912,0],[3914,0.0005],[3915,0],[3917,0.0004],[3919,-0.0008],[3922,0.0003],[3923,0.0002],[3924,0.0008],[3927,-0.0008],[3928,-0.0005],[3929,-0.0007],[3931,0.0003],[3934,-0.0012],[3939,0.0003],[3942,-0.0001],[3945,0.0002],[3947,-0.0004],[3950,0.0006],[3952,-0.0008],[3955,0.0003],[3958,-0.0012],[3962,0.0007],[3963,-0.0002],[3966,0.0007],[3971,-0.0007],[3974,0.0004],[3976,0],[3977,0.0001],[3978,-0.0002],[3980,0.0007],[3981,-0.0001],[3983,0.001],[3986,-0.0003],[3988,0],[3991,-0.0007],[3994,0.0003],[3997,-0.0002],[3999,0.0002],[4001,-0.0001],[4002,0.0001],[4005,0.0002],[4009,-0.0002],[4010,-0.0001],[4012,-0.0003],[4015,0.0004],[4016,-0.0003],[4018,0.0002],[4019,-0.0004],[4022,0.0003],[4023,0],[4024,0.0001],[4026,-0.0004],[4028,0.0005],[4030,-0.0007],[4033,0.0005],[4036,0],[4038,0.0005],[4039,0],[4041,0.0004],[4043,-0.0008],[4046,0.0003],[4047,0.0002],[4048,0.0008],[4051,-0.0008],[4052,-0.0005],[4053,-0.0007],[4055,0.0003],[4058,-0.0012],[4063,0.0003],[4066,-0.0001],[4069,0.0002],[4071,-0.0004],[4074,0.0006],[4076,-0.0008],[4079,0.0003],[4082,-0.0012],[4086,0.0007],[4087,-0.0002],[4090,0.0007],[4095,-0.0007],[4098,0.0004],[4100,0],[4101,0.0001],[4102,-0.0002],[4104,0.0007],[4105,-0.0001],[4107,0.001],[4110,-0.0003],[4112,0],[4115,-0.0007],[4118,0.0003],[4121,-0.0002],[4123,0.0002],[4125,-0.0001],[4126,0.0001],[4129,0.0002],[4133,-0.0002],[4134,-0.0001],[4136,-0.0003],[4140,0.0004],[4141,-0.0003],[4143,0.0002],[4144,-0.0004],[4147,0.0003],[4148,0],[4149,0.0001],[4151,-0.0004],[4153,0.0005],[4155,-0.0007],[4158,0.0005],[4161,0],[4163,0.0005],[4164,0],[4166,0.0004],[4168,-0.0008],[4171,0.0003],[4172,0.0002],[4173,0.0008],[4176,-0.0008],[4177,-0.0005],[4178,-0.0007],[4180,0.0003],[4183,-0.0012],[4188,0.0003],[4191,-0.0001],[4194,0.0002],[4196,-0.0004],[4199,0.0006],[4201,-0.0008],[4204,0.0003],[4207,-0.0012],[4211,0.0007],[4212,-0.0002],[4215,0.0007],[4220,-0.0007],[4223,0.0004],[4225,0],[4226,0.0001],[4227,-0.0002],[4229,0.0007],[4230,-0.0001],[4232,0.001],[4235,-0.0003],[4237,0],[4240,-0.0007],[4243,0.0003],[4246,-0.0002],[4248,0.0002],[4250,-0.0001],[4251,0.0001],[4254,0.0002],[4258,-0.0002],[4259,-0.0001],[4261,-0.0003],[4266,0.0004],[4267,-0.0003],[4269,0.0002],[4270,-0.0004],[4273,0.0003],[4274,0],[4275,0.0001],[4277,-0.0004],[4279,0.0005],[4281,-0.0007],[4284,0.0005],[4287,0],[4289,0.0005],[4290,0],[4292,0.0004],[4294,-0.0008],[4297,0.0003],[4298,0.0002],[4299,0.0008],[4302,-0.0008],[4303,-0.0005],[4304,-0.0007],[4306,0.0003],[4309,-0.0012],[4314,0.0003],[4317,-0.0001],[4320,0.0002],[4322,-0.0004],[4325,0.0006],[4327,-0.0008],[4330,0.0003],[4333,-0.0012],[4337,0.0007],[4338,-0.0002],[4341,0.0007],[4346,-0.0007],[4349,0.0004],[4351,0],[4352,0.0001],[4353,-0.0002],[4355,0.0007],[4356,-0.0001],[4365,-0.0001],[4367,-0.0009],[4372,-0.0001],[4376,-0.0001],[4377,-0.0006],[4379,0.0009],[4383,-0.0008],[4385,0.0004],[4388,0.0006],[4390,-0.0006],[4391,0.0007],[4395,-0.0005],[4397,0.0004],[4399,-0.0003],[4401,0.001],[4403,-0.0003],[4405,0.0006],[4413,-0.001],[4414,-0.0008],[4415,-0.001],[4420,0.0008],[4421,0.0001],[4423,0.0009],[4425,-0.0002],[4427,0.0003],[4432,-0.0007],[4433,-0.0004],[4435,-0.0014],[4436,0]]
       }
       clavicle_r_ctrl.Rotation.Y {
        type=MovieSceneScriptingFloatChannel
        default=11.4657
        thinned_from=71
        keys [[1790,11.4657],[1821,16.3269],[1865,11.2531],[1901,16.3207],[1945,11.2531],[1981,16.3269],[2025,11.2532],[2061,16.3205],[2105,11.2533],[2141,16.3269],[2185,11.2531],[2221,16.3207],[2265,11.2531],[2301,16.3269],[2345,11.2532],[2381,16.3205],[2425,11.2533],[2461,16.3269],[2470,14.9016],[2484,25.5829],[2501,30.0376],[2506,31.5496],[2507,31.5033],[2511,35.6041],[2515,11.9001],[2518,10.8488],[2520,14.6375],[2529,21.6376],[2760,19.9167],[2778,15.5689],[2938,15.8451],[2986,31.1948],[3058,31.1948],[3070,27.5903],[3099,27.5903],[3185,16.0863],[3214,8.4934],[3269,24.1496],[3300,11.2531],[3346,21.6953],[3380,11.1078],[3420,19.6679],[3460,11.2532],[3497,17.9184],[3540,11.2534],[3576,17.7795],[3620,11.2531],[3656,16.3207],[3700,11.2531],[3736,16.3269],[3780,11.2532],[3816,16.3205],[3860,11.2533],[3904,11.2532],[3940,16.3205],[3984,11.2533],[4028,11.2532],[4064,16.3205],[4108,11.2533],[4153,11.2532],[4189,16.3205],[4233,11.2533],[4279,11.2532],[4315,16.3205],[4360,15.826],[4364,15.684],[4409,11.2531],[4435,15.5518],[4436,0]]
       }
       clavicle_r_ctrl.Rotation.Z {
        type=MovieSceneScriptingFloatChannel
        default=-14.4941
        thinned_from=932
        keys [[1790,-14.4941],[2464,-14.4938],[2501,25.9573],[2506,25.4453],[2511,26.997],[2519,-22.6633],[2521,-17.1675],[2527,-21.3787],[2759,-22.0728],[2778,-14.2917],[2784,-14.4402],[2938,-14.407],[2986,-13.7683],[3058,-13.7683],[3062,-20],[3066,-40],[3069,-10],[3088,-3.0374],[3101,0.6841],[3114,2.1074],[3185,1.1007],[3216,-14.493],[4435,-14.4949],[4436,0]]
       }
       clavicle_r_ctrl.Scale.X {type=MovieSceneScriptingFloatChannel, default=1, const [1,12]}
       clavicle_r_ctrl.Scale.Y {type=MovieSceneScriptingFloatChannel, default=1, const [1,12]}
       clavicle_r_ctrl.Scale.Z {type=MovieSceneScriptingFloatChannel, default=1, const [1,12]}
       neck_01_ctrl.Location.X {type=MovieSceneScriptingFloatChannel, default=0, thinned_from=239, keys [[1790,0],[4436,0]]}
       neck_01_ctrl.Location.Y {type=MovieSceneScriptingFloatChannel, default=0, thinned_from=955, keys [[1790,0],[4436,0]]}
       neck_01_ctrl.Location.Z {type=MovieSceneScriptingFloatChannel, default=0, thinned_from=182, keys [[1790,0],[4436,0]]}
       neck_01_ctrl.Rotation.X {
        type=MovieSceneScriptingFloatChannel
        default=12.7734
        thinned_from=69
        keys [[1790,12.7734],[1821,15.0542],[1861,12.4194],[1901,15.057],[1941,12.4225],[1982,15.0548],[2021,12.8901],[2061,15.056],[2101,12.4224],[2141,15.0542],[2181,12.4194],[2221,15.057],[2261,12.4225],[2302,15.0548],[2341,12.8901],[2381,15.056],[2421,12.4224],[2461,15.0542],[2468,14.7819],[2501,12.144],[2506,12.144],[2511,27.9382],[2515,15.0747],[2519,20.8965],[2524,10.4909],[2529,8.3097],[2757,4.4272],[2760,13.4903],[2766,14.1292],[2778,14.666],[2938,14.536],[2986,12.5516],[3058,12.5516],[3070,13.2085],[3099,13.2085],[3185,37.1768],[3216,12.4224],[3217,12.4021],[3296,12.4194],[3336,15.057],[3376,12.4225],[3417,15.0548],[3456,12.8901],[3496,15.056],[3536,12.4224],[3576,15.0542],[3616,12.4194],[3656,15.057],[3696,12.4225],[3737,15.0548],[3776,12.8901],[3816,15.056],[3856,12.4224],[3940,15.056],[3980,12.4224],[4064,15.056],[4104,12.4224],[4149,12.8901],[4189,15.056],[4229,12.4224],[4283,4.7094],[4340,4.1921],[4345,6.6104],[4349,6.5564],[4390,4.1891],[4420,6.5336],[4436,0]]
       }
       neck_01_ctrl.Rotation.Y {
        type=MovieSceneScriptingFloatChannel
        default=0.6555
        thinned_from=70
        keys [[1790,0.6555],[1819,-0.4616],[1853,1.8943],[1899,-0.0589],[1929,1.8743],[1978,-0.5316],[2011,2.0828],[2059,-0.0578],[2089,1.8717],[2139,-0.4616],[2173,1.8943],[2219,-0.0589],[2249,1.8743],[2298,-0.5316],[2331,2.0828],[2379,-0.0578],[2409,1.8717],[2459,-0.4616],[2470,0.9194],[2484,20.9793],[2501,18.7923],[2508,18.7921],[2512,18.2964],[2518,-1.07],[2519,-1.6774],[2521,0.549],[2525,-0.9596],[2529,-0.4857],[2530,-0.5354],[2532,-0.4539],[2791,0.1369],[2802,0.1367],[2927,0.1054],[2938,0.1001],[2986,-6.306],[3058,-6.306],[3070,7.4648],[3099,7.4648],[3185,18.6189],[3217,-6.1749],[3288,1.8943],[3334,-0.0589],[3364,1.8743],[3413,-0.5316],[3446,2.0828],[3494,-0.0578],[3524,1.8717],[3574,-0.4616],[3608,1.8943],[3654,-0.0589],[3684,1.8743],[3733,-0.5316],[3766,2.0828],[3814,-0.0578],[3844,1.8717],[3938,-0.0578],[3968,1.8717],[4014,2.0828],[4062,-0.0578],[4092,1.8717],[4139,2.0828],[4187,-0.0578],[4217,1.8717],[4313,-0.0578],[4343,1.8717],[4360,-0.3962],[4364,-0.3217],[4397,1.8943],[4436,0]]
       }
       neck_01_ctrl.Rotation.Z {
        type=MovieSceneScriptingFloatChannel
        default=0.5376
        thinned_from=103
        keys [[1790,0.5376],[1793,0.5947],[1843,-1.5067],[1863,-1.2396],[1884,1.2595],[1928,-0.9905],[1954,0.8253],[2004,-1.225],[2041,1.2753],[2042,1.2745],[2047,1.2844],[2088,-1.0036],[2113,0.5947],[2163,-1.5067],[2183,-1.2396],[2204,1.2595],[2248,-0.9905],[2274,0.8253],[2324,-1.225],[2361,1.2753],[2362,1.2745],[2367,1.2844],[2408,-1.0036],[2433,0.5947],[2468,-0.2499],[2501,6.6175],[2511,6.6176],[2513,7.8079],[2519,5.5106],[2522,0.6601],[2526,3.2051],[2529,1.7263],[2530,1.9887],[2532,1.4489],[2536,1.2476],[2762,0.566],[2787,-0.6711],[2927,-0.675],[2938,-0.6779],[2986,-6.2135],[3058,-6.2135],[3070,6.5804],[3099,6.5804],[3185,9.1052],[3217,6.7563],[3278,-1.5067],[3298,-1.2396],[3319,1.2595],[3363,-0.9905],[3389,0.8253],[3439,-1.225],[3476,1.2754],[3477,1.2745],[3482,1.2844],[3523,-1.0036],[3548,0.5947],[3598,-1.5067],[3618,-1.2396],[3639,1.2595],[3683,-0.9905],[3709,0.8253],[3759,-1.225],[3796,1.2754],[3797,1.2745],[3802,1.2844],[3843,-1.0036],[3868,0.5947],[3920,1.2754],[3921,1.2745],[3926,1.2844],[3967,-1.0036],[3992,0.5947],[4044,1.2754],[4045,1.2745],[4050,1.2844],[4091,-1.0036],[4116,0.5947],[4169,1.2754],[4170,1.2745],[4175,1.2844],[4216,-1.0036],[4241,0.5947],[4295,1.2754],[4296,1.2745],[4301,1.2844],[4342,-1.0036],[4360,-0.2059],[4364,-0.2831],[4387,-1.5067],[4407,-1.2396],[4428,1.2595],[4434,1.2232],[4436,0]]
       }
       neck_01_ctrl.Scale.X {type=MovieSceneScriptingFloatChannel, default=1, const [1,12]}
       neck_01_ctrl.Scale.Y {type=MovieSceneScriptingFloatChannel, default=1, const [1,12]}
       neck_01_ctrl.Scale.Z {type=MovieSceneScriptingFloatChannel, default=1, const [1,12]}
       neck_02_ctrl.Location.X {type=MovieSceneScriptingFloatChannel, default=0, thinned_from=33, keys [[1790,0],[4436,0]]}
       neck_02_ctrl.Location.Y {type=MovieSceneScriptingFloatChannel, default=0, thinned_from=30, keys [[1790,0],[4436,0]]}
       neck_02_ctrl.Location.Z {type=MovieSceneScriptingFloatChannel, default=0, thinned_from=19, keys [[1790,0],[4436,0]]}
       neck_02_ctrl.Rotation.X {
        type=MovieSceneScriptingFloatChannel
        default=0
        thinned_from=20
        keys [[1790,0],[2522,0],[2757,0.3695],[2761,2.1908],[2772,1.4281],[2784,0],[4436,0]]
       }
       neck_02_ctrl.Rotation.Y {type=MovieSceneScriptingFloatChannel, default=0, thinned_from=37, keys [[1790,0],[4436,0]]}
       neck_02_ctrl.Rotation.Z {type=MovieSceneScriptingFloatChannel, default=0, thinned_from=35, keys [[1790,0],[4436,0]]}
       neck_02_ctrl.Scale.X {type=MovieSceneScriptingFloatChannel, default=1, const [1,12]}
       neck_02_ctrl.Scale.Y {type=MovieSceneScriptingFloatChannel, default=1, const [1,12]}
       neck_02_ctrl.Scale.Z {type=MovieSceneScriptingFloatChannel, default=1, const [1,12]}
       head_ik_ctrl.Location.X {
        type=MovieSceneScriptingFloatChannel
        default=-0.0776
        keys [[1790,-0.0776],[1822,0.219],[1856,-0.1163],[1904,0.0795],[1931,-0.1379],[1982,0.2067],[2023,-0.1934],[2064,0.0796],[2091,-0.1363],[2142,0.219],[2176,-0.1163],[2224,0.0795],[2251,-0.1379],[2302,0.2067],[2343,-0.1934],[2384,0.0796],[2411,-0.1363],[2462,0.219],[2469,0.1316],[2496,-2.8261],[2504,-2.7895],[2507,-2.7941],[2509,-2.6427],[2512,-2.6787],[2518,-0.1286],[2519,-0.0705],[2520,-0.1185],[2524,0.0342],[2527,0.0154],[2760,0.2177],[2927,0.03],[2938,-0.5271],[2986,1.6489],[3058,1.6489],[3070,-1.2482],[3083,-1.2482],[3099,-3.1367],[3185,-3.1367],[3257,0.219],[3291,-0.1163],[3339,0.0795],[3366,-0.1379],[3417,0.2067],[3458,-0.1934],[3499,0.0796],[3526,-0.1363],[3577,0.219],[3611,-0.1163],[3659,0.0795],[3686,-0.1379],[3737,0.2067],[3778,-0.1934],[3819,0.0796],[3846,-0.1363],[3902,-0.1934],[3943,0.0796],[3970,-0.1363],[4026,-0.1934],[4067,0.0796],[4094,-0.1363],[4151,-0.1934],[4192,0.0796],[4219,-0.1363],[4277,-0.1934],[4318,0.0796],[4345,-0.1363],[4360,0.2795],[4364,0.2773],[4400,-0.1163],[4435,0.0373],[4436,0]]
       }
       head_ik_ctrl.Location.Y {
        type=MovieSceneScriptingFloatChannel
        default=1.8741
        keys [[1790,1.8741],[1822,2.2326],[1862,1.8265],[1902,2.2421],[1942,1.823],[1982,2.234],[2022,1.8853],[2062,2.242],[2102,1.823],[2142,2.2326],[2182,1.8265],[2222,2.2421],[2262,1.823],[2302,2.234],[2342,1.8853],[2382,2.242],[2422,1.823],[2462,2.2326],[2502,2.24],[2508,2.7475],[2511,1.3086],[2513,1.6659],[2519,3.6118],[2521,2.7355],[2524,1.8157],[2529,1.4356],[2761,2.5434],[2780,2.2474],[2791,2.2805],[2927,2.552],[2938,3.0695],[2986,3.0762],[3058,3.0762],[3070,2.6366],[3083,2.6366],[3099,3.8639],[3185,3.8639],[3217,1.823],[3257,2.2326],[3297,1.8265],[3337,2.2421],[3377,1.823],[3417,2.234],[3457,1.8853],[3497,2.242],[3537,1.823],[3577,2.2326],[3617,1.8265],[3657,2.2421],[3697,1.823],[3737,2.234],[3777,1.8853],[3817,2.242],[3857,1.823],[3901,1.8853],[3941,2.242],[3981,1.823],[4025,1.8853],[4065,2.242],[4105,1.823],[4150,1.8853],[4190,2.242],[4230,1.823],[4276,1.8853],[4316,2.242],[4356,1.823],[4360,-1.358],[4364,-1.2586],[4406,1.8265],[4435,2.1865],[4436,0]]
       }
       head_ik_ctrl.Location.Z {
        type=MovieSceneScriptingFloatChannel
        default=-1.1793
        keys [[1790,-1.1793],[1816,-1.3333],[1858,-1.1526],[1896,-1.3328],[1939,-1.1512],[1976,-1.3333],[2018,-1.2059],[2055,-1.3356],[2099,-1.1511],[2136,-1.3333],[2178,-1.1526],[2216,-1.3328],[2259,-1.1512],[2296,-1.3333],[2338,-1.2059],[2375,-1.3356],[2419,-1.1511],[2456,-1.3333],[2468,-1.2983],[2471,-1.3454],[2496,-1.838],[2507,-0.9527],[2509,-3.3636],[2511,-4.8174],[2515,-1.3521],[2518,-0.1413],[2520,0.2939],[2525,0.1088],[2531,0.1352],[2766,-0.9578],[2778,-1.2473],[2927,0.0804],[2938,0.4261],[2986,-0.8948],[3058,-0.8948],[3070,0.1598],[3083,0.1598],[3099,-0.3219],[3185,-0.3219],[3251,-1.3333],[3293,-1.1526],[3331,-1.3328],[3374,-1.1512],[3411,-1.3333],[3453,-1.2059],[3490,-1.3356],[3534,-1.1511],[3571,-1.3333],[3613,-1.1526],[3651,-1.3328],[3694,-1.1512],[3731,-1.3333],[3773,-1.2059],[3810,-1.3356],[3854,-1.1511],[3897,-1.2059],[3934,-1.3356],[3978,-1.1511],[4021,-1.2059],[4058,-1.3356],[4102,-1.1511],[4146,-1.2059],[4183,-1.3356],[4227,-1.1511],[4272,-1.2059],[4309,-1.3356],[4353,-1.1511],[4360,0.2299],[4364,-1.2131],[4366,-1.3333],[4402,-1.1526],[4435,-1.3229],[4436,0]]
       }
       head_ik_ctrl.Rotation.X {
        type=MovieSceneScriptingFloatChannel
        default=34.321
        thinned_from=108
        keys [[1790,34.321],[1806,34.3305],[1835,33.3965],[1870,34.3304],[1886,34.3226],[1915,33.4199],[1950,34.3196],[1966,34.328],[1990,33.5298],[2027,34.7973],[2075,33.42],[2110,34.321],[2126,34.3305],[2155,33.3965],[2190,34.3304],[2206,34.3226],[2235,33.4199],[2270,34.3196],[2286,34.328],[2310,33.5298],[2347,34.7973],[2395,33.42],[2430,34.321],[2446,34.3305],[2467,33.5737],[2507,-20.6485],[2509,29.2585],[2511,59.9008],[2513,47.0465],[2515,44.6668],[2520,14.1744],[2526,4.9315],[2763,38.5119],[2772,44.4481],[2784,40.6601],[2927,40.2945],[2938,49.262],[2986,38.9746],[3058,38.9746],[3070,17.5089],[3083,17.5089],[3099,65.5906],[3185,65.5906],[3225,34.321],[3241,34.3305],[3270,33.3965],[3305,34.3304],[3321,34.3226],[3350,33.4199],[3385,34.3196],[3401,34.328],[3425,33.5298],[3462,34.7973],[3510,33.42],[3545,34.321],[3561,34.3305],[3590,33.3965],[3625,34.3304],[3641,34.3226],[3670,33.4199],[3705,34.3196],[3721,34.328],[3745,33.5298],[3782,34.7973],[3830,33.42],[3865,34.321],[3881,34.3305],[3906,34.7973],[3954,33.42],[3989,34.321],[4005,34.3305],[4030,34.7973],[4078,33.42],[4113,34.321],[4129,34.3305],[4155,34.7973],[4203,33.42],[4238,34.321],[4254,34.3305],[4281,34.7973],[4329,33.42],[4360,-27.4236],[4364,-16.3468],[4379,33.3965],[4414,34.3304],[4435,34.3017],[4436,0]]
       }
       head_ik_ctrl.Rotation.Y {
        type=MovieSceneScriptingFloatChannel
        default=0.4717
        keys [[1790,0.4717],[1809,-0.7624],[1852,3.1711],[1888,-0.5562],[1931,3.1613],[1969,-0.8859],[2011,3.3978],[2048,-0.5744],[2090,3.1596],[2129,-0.7624],[2172,3.1711],[2208,-0.5562],[2251,3.1613],[2289,-0.8859],[2331,3.3978],[2368,-0.5744],[2410,3.1596],[2449,-0.7624],[2502,3.1042],[2506,2.0125],[2511,-5.7388],[2518,1.6821],[2521,-0.692],[2523,-0.4775],[2525,-0.4956],[2760,-0.6055],[2781,0.7746],[2927,0.3759],[2938,0.3617],[2986,-10.338],[3058,-10.338],[3070,17.4855],[3083,17.4855],[3099,7.8386],[3185,7.8386],[3244,-0.7624],[3287,3.1711],[3323,-0.5562],[3366,3.1613],[3404,-0.8859],[3446,3.3978],[3483,-0.5744],[3525,3.1596],[3564,-0.7624],[3607,3.1711],[3643,-0.5562],[3686,3.1613],[3724,-0.8859],[3766,3.3978],[3803,-0.5744],[3845,3.1596],[3884,-0.7624],[3890,3.3978],[3927,-0.5744],[3969,3.1596],[4008,-0.7624],[4014,3.3978],[4051,-0.5744],[4093,3.1596],[4132,-0.7624],[4139,3.3978],[4176,-0.5744],[4218,3.1596],[4257,-0.7624],[4265,3.3978],[4302,-0.5744],[4344,3.1596],[4360,-0.6152],[4364,-4.3895],[4396,3.1711],[4432,-0.5562],[4436,0]]
       }
       head_ik_ctrl.Rotation.Z {
        type=MovieSceneScriptingFloatChannel
        default=-1.4396
        thinned_from=103
        keys [[1790,-1.4396],[1817,-3.9415],[1857,-2.2687],[1868,-2.4616],[1880,-2.1259],[1897,-2.944],[1944,-1.2317],[1978,-3.6893],[2027,-0.8278],[2057,-2.9383],[2061,-2.9308],[2065,-2.94],[2104,-1.2396],[2137,-3.9415],[2177,-2.2687],[2188,-2.4616],[2200,-2.1259],[2217,-2.944],[2264,-1.2317],[2298,-3.6893],[2347,-0.8278],[2377,-2.9383],[2381,-2.9308],[2385,-2.94],[2424,-1.2396],[2457,-3.9415],[2466,-3.839],[2471,-2.7883],[2496,1.9537],[2501,2.0195],[2507,0.2386],[2508,0.6233],[2511,-0.8785],[2513,5.4909],[2522,-3.9345],[2525,-0.7454],[2526,-0.7501],[2529,-0.4248],[2531,-0.7408],[2532,-0.587],[2538,-0.9585],[2792,-3.7227],[2938,-3.9383],[2986,-18.3718],[3058,-18.3718],[3070,21.5166],[3083,21.5166],[3099,21.6632],[3185,21.6632],[3219,-1.2396],[3252,-3.9415],[3292,-2.2687],[3303,-2.4616],[3315,-2.1259],[3332,-2.9441],[3379,-1.2317],[3413,-3.6893],[3462,-0.8278],[3492,-2.9383],[3496,-2.9308],[3500,-2.94],[3539,-1.2396],[3572,-3.9415],[3612,-2.2687],[3623,-2.4616],[3635,-2.1259],[3652,-2.9441],[3699,-1.2317],[3733,-3.6894],[3782,-0.8278],[3812,-2.9383],[3816,-2.9308],[3820,-2.94],[3859,-1.2396],[3906,-0.8278],[3936,-2.9383],[3940,-2.9308],[3944,-2.94],[3983,-1.2396],[4030,-0.8278],[4060,-2.9383],[4064,-2.9308],[4068,-2.94],[4107,-1.2396],[4155,-0.8278],[4185,-2.9383],[4189,-2.9308],[4193,-2.94],[4232,-1.2396],[4281,-0.8278],[4311,-2.9383],[4315,-2.9308],[4319,-2.94],[4360,-3.8786],[4364,-3.9252],[4367,-3.9415],[4401,-2.2687],[4412,-2.4616],[4424,-2.1259],[4435,-2.7907],[4436,0]]
       }
       head_ik_ctrl.Scale.X {type=MovieSceneScriptingFloatChannel, default=1, const [1,12]}
       head_ik_ctrl.Scale.Y {type=MovieSceneScriptingFloatChannel, default=1, const [1,12]}
       head_ik_ctrl.Scale.Z {type=MovieSceneScriptingFloatChannel, default=1, const [1,12]}
       chest_ctrl.Location.X {
        type=MovieSceneScriptingFloatChannel
        default=0.2165
        keys [[1790,0.2165],[1821,0.0564],[1856,0.2486],[1899,0.0575],[1900,0.0577],[1901,0.0576],[1936,0.2447],[1979,0.0572],[1980,0.0575],[1981,0.0574],[2017,0.2444],[2060,0.0573],[2096,0.2448],[2141,0.0564],[2176,0.2486],[2219,0.0575],[2220,0.0577],[2221,0.0576],[2256,0.2447],[2299,0.0572],[2300,0.0575],[2301,0.0574],[2337,0.2444],[2380,0.0573],[2416,0.2448],[2461,0.0564],[2468,0.0654],[2502,0.5456],[2511,0.8196],[2513,0.6239],[2518,-0.1384],[2521,-0.0361],[2527,-0.1654],[2528,-0.1505],[2531,-0.176],[2532,-0.1694],[2536,-0.2042],[2927,2.5805],[2938,3.8793],[2986,3.5143],[3058,3.5143],[3070,-1.7188],[3083,-1.7188],[3099,-1.546],[3185,-1.546],[3256,0.0564],[3291,0.2486],[3334,0.0575],[3335,0.0577],[3336,0.0576],[3371,0.2447],[3414,0.0572],[3415,0.0575],[3416,0.0574],[3452,0.2444],[3495,0.0573],[3531,0.2448],[3576,0.0564],[3611,0.2486],[3654,0.0575],[3655,0.0577],[3656,0.0576],[3691,0.2447],[3734,0.0572],[3735,0.0575],[3736,0.0574],[3772,0.2444],[3815,0.0573],[3851,0.2448],[3896,0.2444],[3939,0.0573],[3975,0.2448],[4020,0.2444],[4063,0.0573],[4099,0.2448],[4145,0.2444],[4188,0.0573],[4224,0.2448],[4271,0.2444],[4314,0.0573],[4350,0.2448],[4360,2.4464],[4364,2.5492],[4400,0.2486],[4436,0]]
       }
       chest_ctrl.Location.Y {
        type=MovieSceneScriptingFloatChannel
        default=1.1311
        thinned_from=79
        keys [[1790,1.1311],[1813,1.0495],[1854,1.1585],[1893,1.0495],[1934,1.1588],[1973,1.0495],[2014,1.1613],[2055,1.0489],[2094,1.1585],[2133,1.0495],[2174,1.1585],[2213,1.0495],[2254,1.1588],[2293,1.0495],[2334,1.1613],[2375,1.0489],[2414,1.1585],[2453,1.0495],[2464,1.0657],[2507,0.7431],[2511,2.8246],[2513,1.8709],[2515,-0.5074],[2518,-2.6161],[2519,-2.6502],[2520,-2.6169],[2523,-2.7088],[2765,-0.35],[2776,0.2646],[2780,0.1965],[2927,0.4418],[2938,3.3833],[2986,1.0675],[3058,1.0675],[3070,-0.9251],[3083,-0.9251],[3099,5.1784],[3185,5.1784],[3248,1.0495],[3289,1.1585],[3328,1.0495],[3369,1.1588],[3408,1.0495],[3449,1.1613],[3490,1.0489],[3529,1.1585],[3568,1.0495],[3609,1.1585],[3648,1.0495],[3689,1.1588],[3728,1.0495],[3769,1.1613],[3810,1.0489],[3849,1.1585],[3888,1.0495],[3893,1.1613],[3934,1.0489],[3973,1.1585],[4012,1.0495],[4017,1.1613],[4058,1.0489],[4097,1.1585],[4136,1.0495],[4142,1.1613],[4183,1.0489],[4222,1.1585],[4261,1.0495],[4268,1.1613],[4309,1.0489],[4348,1.1585],[4360,-0.3904],[4364,-0.0108],[4398,1.1585],[4417,1.1229],[4426,1.083],[4431,1.0649],[4435,1.0547],[4436,0]]
       }
       chest_ctrl.Location.Z {
        type=MovieSceneScriptingFloatChannel
        default=-0.2907
        keys [[1790,-0.2907],[1809,-0.3048],[1843,-0.2243],[1890,-0.3048],[1922,-0.224],[1970,-0.3048],[2002,-0.2241],[2049,-0.3049],[2082,-0.224],[2129,-0.3048],[2163,-0.2243],[2210,-0.3048],[2242,-0.224],[2290,-0.3048],[2322,-0.2241],[2369,-0.3049],[2402,-0.224],[2449,-0.3048],[2466,-0.2533],[2496,-0.0344],[2506,0.0122],[2511,-0.3544],[2515,0.6532],[2518,-0.2544],[2519,-0.6563],[2520,-0.6532],[2524,-0.9478],[2759,-0.5138],[2763,0.2876],[2771,0.901],[2788,1.4091],[2927,2.2414],[2938,2.5308],[2986,1.0995],[3058,1.0995],[3070,0.4372],[3083,0.4372],[3099,-0.4425],[3185,-0.4425],[3244,-0.3048],[3278,-0.2243],[3325,-0.3048],[3357,-0.224],[3405,-0.3048],[3437,-0.2241],[3484,-0.3049],[3517,-0.224],[3564,-0.3048],[3598,-0.2243],[3645,-0.3048],[3677,-0.224],[3725,-0.3048],[3757,-0.2241],[3804,-0.3049],[3837,-0.224],[3884,-0.3048],[3928,-0.3049],[3961,-0.224],[4008,-0.3048],[4052,-0.3049],[4085,-0.224],[4132,-0.3048],[4177,-0.3049],[4210,-0.224],[4257,-0.3048],[4303,-0.3049],[4336,-0.224],[4360,-1.0238],[4364,-1.9903],[4387,-0.2243],[4434,-0.3048],[4436,0]]
       }
       chest_ctrl.Rotation.X {
        type=MovieSceneScriptingFloatChannel
        default=5.8869
        keys [[1790,5.8869],[1820,5.2542],[1860,5.995],[1900,5.2537],[1940,5.995],[1979,5.2537],[2020,6.0131],[2059,5.2517],[2100,5.9949],[2140,5.2542],[2180,5.995],[2220,5.2537],[2260,5.995],[2299,5.2537],[2340,6.0131],[2379,5.2517],[2420,5.9949],[2460,5.2542],[2468,5.4619],[2507,8.32],[2509,34.1613],[2511,49.9374],[2513,36.7116],[2515,11.3499],[2518,-14.6534],[2519,-17.1846],[2520,-17.1235],[2522,-17.3072],[2766,-1.7261],[2777,1.7571],[2780,1.5315],[2927,3.865],[2938,28.9931],[2986,0.6961],[3058,0.6961],[3070,-7.0137],[3083,-7.0137],[3099,35.8965],[3185,35.8965],[3215,5.9949],[3255,5.2542],[3295,5.995],[3335,5.2537],[3375,5.995],[3414,5.2537],[3455,6.0131],[3494,5.2517],[3535,5.9949],[3575,5.2542],[3615,5.995],[3655,5.2537],[3695,5.995],[3734,5.2537],[3775,6.0131],[3814,5.2517],[3855,5.9949],[3899,6.0131],[3938,5.2517],[3979,5.9949],[4023,6.0131],[4062,5.2517],[4103,5.9949],[4148,6.0131],[4187,5.2517],[4228,5.9949],[4274,6.0131],[4313,5.2517],[4354,5.9949],[4360,5.619],[4364,5.6442],[4404,5.995],[4428,5.4889],[4434,5.3369],[4436,0]]
       }
       chest_ctrl.Rotation.Y {
        type=MovieSceneScriptingFloatChannel
        default=-0.9232
        thinned_from=84
        keys [[1790,-0.9232],[1825,-0.2108],[1863,-0.9632],[1906,-0.2038],[1944,-0.939],[1987,-0.1953],[2027,-0.9636],[2066,-0.2055],[2103,-0.9383],[2105,-0.9383],[2145,-0.2108],[2183,-0.9632],[2226,-0.2038],[2264,-0.939],[2307,-0.1953],[2347,-0.9636],[2386,-0.2055],[2423,-0.9383],[2425,-0.9383],[2465,-0.2108],[2467,-0.214],[2507,-9.5063],[2511,-0.8344],[2515,-2.8141],[2519,-0.321],[2520,-0.5389],[2529,0.3316],[2531,0.2736],[2533,0.34],[2927,-8.5091],[2938,-26.1413],[2986,-13.8953],[3058,-13.8953],[3070,5.7747],[3083,5.7747],[3099,19.5493],[3185,19.5493],[3218,-0.9383],[3220,-0.9383],[3260,-0.2108],[3298,-0.9631],[3341,-0.2038],[3379,-0.939],[3422,-0.1953],[3462,-0.9636],[3501,-0.2055],[3538,-0.9383],[3540,-0.9383],[3580,-0.2108],[3618,-0.9632],[3661,-0.2038],[3699,-0.939],[3742,-0.1953],[3782,-0.9636],[3821,-0.2055],[3858,-0.9383],[3860,-0.9383],[3906,-0.9636],[3945,-0.2055],[3982,-0.9383],[3984,-0.9383],[4030,-0.9636],[4069,-0.2055],[4106,-0.9383],[4108,-0.9383],[4155,-0.9636],[4194,-0.2055],[4231,-0.9383],[4233,-0.9383],[4281,-0.9636],[4320,-0.2055],[4357,-0.9383],[4360,-20.9985],[4364,-21.2466],[4407,-0.9631],[4436,0]]
       }
       chest_ctrl.Rotation.Z {
        type=MovieSceneScriptingFloatChannel
        default=-1.3441
        keys [[1790,-1.3441],[1815,-2.7196],[1855,-0.7062],[1895,-2.7199],[1935,-0.7123],[1974,-2.7161],[2014,-0.5922],[2055,-2.7207],[2095,-0.7121],[2135,-2.7196],[2175,-0.7062],[2215,-2.7199],[2255,-0.7123],[2294,-2.7161],[2334,-0.5922],[2375,-2.7207],[2415,-0.7121],[2455,-2.7196],[2468,-2.3258],[2503,-19.0101],[2507,-18.4068],[2509,-19.3327],[2522,-1.609],[2526,-2.4042],[2529,-1.7234],[2531,-2.2601],[2532,-1.9204],[2536,-2.1183],[2760,-2.1676],[2927,0.8755],[2938,7.7451],[2986,20.8466],[3058,20.8466],[3070,-9.5872],[3083,-9.5872],[3099,-33.848],[3185,-33.848],[3250,-2.7196],[3290,-0.7062],[3330,-2.7199],[3370,-0.7123],[3409,-2.7161],[3449,-0.5922],[3490,-2.7207],[3530,-0.7121],[3570,-2.7196],[3610,-0.7062],[3650,-2.7199],[3690,-0.7123],[3729,-2.7161],[3769,-0.5922],[3810,-2.7207],[3850,-0.7121],[3893,-0.5922],[3934,-2.7207],[3974,-0.7121],[4017,-0.5922],[4058,-2.7207],[4098,-0.7121],[4142,-0.5922],[4183,-2.7207],[4223,-0.7121],[4268,-0.5922],[4309,-2.7207],[4349,-0.7121],[4360,9.0304],[4364,-2.4274],[4365,-2.7196],[4399,-0.7062],[4435,-2.6871],[4436,0]]
       }
       chest_ctrl.Scale.X {type=MovieSceneScriptingFloatChannel, default=1, const [1,12]}
       chest_ctrl.Scale.Y {type=MovieSceneScriptingFloatChannel, default=1, const [1,12]}
       chest_ctrl.Scale.Z {type=MovieSceneScriptingFloatChannel, default=1, const [1,12]}
       chest_tan_ctrl.Location.X {
        type=MovieSceneScriptingFloatChannel
        default=0.0069
        thinned_from=589
        keys [[1790,0.0069],[1791,0.0063],[1792,0.0066],[1793,0.006],[1795,0.0068],[1797,0.0045],[1799,0.0062],[1802,0.0027],[1804,0.005],[1807,0.0003],[1809,0.0014],[1812,-0.002],[1814,0.0002],[1824,-0.0054],[1826,-0.0052],[1830,-0.0059],[1835,-0.0029],[1836,-0.003],[1840,-0.0002],[1842,-0.0005],[1844,0.001],[1846,0.0003],[1864,0.0057],[1865,0.0055],[1869,0.008],[1873,0.006],[1875,0.0071],[1877,0.0057],[1878,0.0068],[1882,0.0029],[1884,0.005],[1887,-0.0006],[1889,0.0023],[1892,-0.0034],[1894,-0.0003],[1909,-0.0062],[1911,-0.0062],[1912,-0.0058],[1913,-0.0059],[1915,-0.0031],[1917,-0.004],[1919,-0.0008],[1922,-0.0023],[1924,0.0016],[1927,0.0004],[1929,0.0013],[1931,0.0003],[1937,0.003],[1938,0.0027],[1940,0.0038],[1941,0.0037],[1949,0.0074],[1950,0.0072],[1951,0.0072],[1953,0.0067],[1955,0.0081],[1957,0.0042],[1959,0.0053],[1962,0.0025],[1964,0.0047],[1967,-0.0004],[1969,0.0018],[1972,-0.003],[1974,-0.0003],[1976,-0.0022],[1977,-0.0022],[1983,-0.0051],[1984,-0.0048],[1988,-0.0067],[1990,-0.0059],[1991,-0.0061],[1992,-0.0058],[1993,-0.006],[1999,-0.0009],[2001,-0.0016],[2004,0.0021],[2006,0.0007],[2029,0.0076],[2031,0.007],[2034,0.0076],[2037,0.0065],[2039,0.0066],[2041,0.0051],[2045,0.0064],[2049,-0.0017],[2050,-0.0016],[2051,-0.0017],[2053,-0.0013],[2070,-0.0059],[2072,-0.0056],[2073,-0.0059],[2075,-0.0031],[2077,-0.004],[2079,-0.0008],[2082,-0.0022],[2084,0.0016],[2087,0.0004],[2104,0.005],[2105,0.0048],[2109,0.0073],[2111,0.0063],[2112,0.0066],[2113,0.006],[2115,0.0068],[2117,0.0045],[2119,0.0062],[2122,0.0027],[2124,0.005],[2127,0.0003],[2129,0.0014],[2132,-0.002],[2134,0.0002],[2144,-0.0054],[2146,-0.0052],[2150,-0.0059],[2155,-0.0029],[2156,-0.003],[2160,-0.0002],[2162,-0.0005],[2164,0.001],[2166,0.0003],[2184,0.0057],[2185,0.0055],[2189,0.008],[2193,0.006],[2195,0.0071],[2197,0.0057],[2198,0.0068],[2202,0.0029],[2204,0.005],[2207,-0.0006],[2209,0.0023],[2212,-0.0034],[2214,-0.0003],[2229,-0.0062],[2231,-0.0062],[2232,-0.0058],[2233,-0.0059],[2235,-0.0031],[2237,-0.004],[2239,-0.0008],[2242,-0.0023],[2244,0.0016],[2247,0.0004],[2249,0.0013],[2251,0.0003],[2257,0.003],[2258,0.0027],[2260,0.0038],[2261,0.0037],[2269,0.0074],[2270,0.0072],[2271,0.0072],[2273,0.0067],[2275,0.0081],[2277,0.0042],[2279,0.0053],[2282,0.0025],[2284,0.0047],[2287,-0.0004],[2289,0.0018],[2292,-0.003],[2294,-0.0003],[2296,-0.0022],[2297,-0.0022],[2303,-0.0051],[2304,-0.0048],[2308,-0.0067],[2310,-0.0059],[2311,-0.0061],[2312,-0.0058],[2313,-0.006],[2319,-0.0009],[2321,-0.0016],[2324,0.0021],[2326,0.0007],[2349,0.0076],[2351,0.007],[2354,0.0076],[2357,0.0065],[2359,0.0066],[2361,0.0051],[2365,0.0064],[2369,-0.0017],[2370,-0.0016],[2371,-0.0017],[2373,-0.0013],[2390,-0.0059],[2392,-0.0056],[2393,-0.0059],[2395,-0.0031],[2397,-0.004],[2399,-0.0008],[2402,-0.0022],[2404,0.0016],[2407,0.0004],[2424,0.005],[2425,0.0048],[2429,0.0073],[2431,0.0063],[2432,0.0066],[2433,0.006],[2435,0.0068],[2437,0.0045],[2439,0.0062],[2442,0.0027],[2444,0.005],[2447,0.0003],[2449,0.0014],[2452,-0.002],[2454,0.0002],[2464,-0.0054],[2466,-0.0052],[2507,0.806],[2510,0.1927],[2515,-0.1312],[2521,-0.0027],[2527,-0.025],[2529,-0.0217],[2531,-0.0248],[2532,-0.0224],[2538,-0.0256],[2927,-0.5831],[2938,-0.2775],[2986,0.3462],[3058,0.3462],[3070,-0.1217],[3083,-0.1217],[3099,1.139],[3185,1.139],[3219,0.005],[3220,0.0048],[3224,0.0073],[3226,0.0063],[3227,0.0066],[3228,0.006],[3230,0.0068],[3232,0.0045],[3234,0.0062],[3237,0.0027],[3239,0.005],[3242,0.0003],[3244,0.0014],[3247,-0.002],[3249,0.0002],[3259,-0.0054],[3261,-0.0052],[3265,-0.0059],[3270,-0.0029],[3271,-0.003],[3275,-0.0002],[3277,-0.0005],[3279,0.001],[3281,0.0003],[3299,0.0057],[3300,0.0055],[3304,0.008],[3308,0.006],[3310,0.0071],[3312,0.0057],[3313,0.0068],[3317,0.0029],[3319,0.005],[3322,-0.0006],[3324,0.0023],[3327,-0.0034],[3329,-0.0003],[3344,-0.0062],[3346,-0.0062],[3347,-0.0058],[3348,-0.0059],[3350,-0.0031],[3352,-0.004],[3354,-0.0008],[3357,-0.0023],[3359,0.0016],[3362,0.0004],[3364,0.0013],[3366,0.0004],[3372,0.003],[3373,0.0027],[3375,0.0038],[3376,0.0037],[3384,0.0074],[3385,0.0072],[3386,0.0072],[3388,0.0067],[3390,0.0081],[3392,0.0042],[3394,0.0053],[3397,0.0025],[3399,0.0047],[3402,-0.0004],[3404,0.0018],[3407,-0.003],[3409,-0.0003],[3411,-0.0022],[3412,-0.0022],[3418,-0.0051],[3419,-0.0048],[3423,-0.0067],[3425,-0.0059],[3426,-0.0061],[3427,-0.0058],[3428,-0.006],[3434,-0.0009],[3436,-0.0016],[3439,0.0021],[3441,0.0007],[3464,0.0076],[3466,0.007],[3469,0.0076],[3472,0.0066],[3474,0.0066],[3476,0.0051],[3480,0.0064],[3484,-0.0017],[3485,-0.0016],[3486,-0.0017],[3488,-0.0013],[3505,-0.0059],[3507,-0.0056],[3508,-0.0059],[3510,-0.0031],[3512,-0.004],[3514,-0.0008],[3517,-0.0022],[3519,0.0016],[3522,0.0004],[3539,0.005],[3540,0.0048],[3544,0.0073],[3546,0.0063],[3547,0.0066],[3548,0.006],[3550,0.0068],[3552,0.0045],[3554,0.0062],[3557,0.0027],[3559,0.005],[3562,0.0003],[3564,0.0014],[3567,-0.002],[3569,0.0002],[3579,-0.0054],[3581,-0.0052],[3585,-0.0059],[3590,-0.0029],[3591,-0.003],[3595,-0.0002],[3597,-0.0005],[3599,0.001],[3601,0.0003],[3619,0.0057],[3620,0.0055],[3624,0.008],[3628,0.006],[3630,0.0071],[3632,0.0057],[3633,0.0068],[3637,0.0029],[3639,0.005],[3642,-0.0006],[3644,0.0023],[3647,-0.0034],[3649,-0.0003],[3664,-0.0062],[3666,-0.0062],[3667,-0.0058],[3668,-0.0059],[3670,-0.0031],[3672,-0.004],[3674,-0.0008],[3677,-0.0023],[3679,0.0016],[3682,0.0004],[3684,0.0013],[3686,0.0004],[3692,0.003],[3693,0.0027],[3695,0.0038],[3696,0.0037],[3704,0.0074],[3705,0.0072],[3706,0.0072],[3708,0.0067],[3710,0.0081],[3712,0.0042],[3714,0.0053],[3717,0.0025],[3719,0.0047],[3722,-0.0004],[3724,0.0018],[3727,-0.003],[3729,-0.0003],[3731,-0.0022],[3732,-0.0022],[3738,-0.0051],[3739,-0.0048],[3743,-0.0067],[3745,-0.0059],[3746,-0.0061],[3747,-0.0058],[3748,-0.006],[3754,-0.0009],[3756,-0.0016],[3759,0.0021],[3761,0.0007],[3784,0.0076],[3786,0.007],[3789,0.0076],[3792,0.0066],[3794,0.0066],[3796,0.0051],[3800,0.0064],[3804,-0.0017],[3805,-0.0016],[3806,-0.0017],[3808,-0.0013],[3825,-0.0059],[3827,-0.0056],[3828,-0.0059],[3830,-0.0031],[3832,-0.004],[3834,-0.0008],[3837,-0.0022],[3839,0.0016],[3842,0.0004],[3859,0.005],[3860,0.0048],[3864,0.0073],[3866,0.0063],[3867,0.0066],[3868,0.006],[3870,0.0068],[3872,0.0045],[3874,0.0062],[3877,0.0027],[3879,0.005],[3882,0.0003],[3884,0.0014],[3887,-0.002],[3908,0.0076],[3910,0.007],[3913,0.0076],[3916,0.0066],[3918,0.0066],[3920,0.0051],[3924,0.0064],[3928,-0.0017],[3929,-0.0016],[3930,-0.0017],[3932,-0.0013],[3949,-0.0059],[3951,-0.0056],[3952,-0.0059],[3954,-0.0031],[3956,-0.004],[3958,-0.0008],[3961,-0.0022],[3963,0.0016],[3966,0.0004],[3983,0.005],[3984,0.0048],[3988,0.0073],[3990,0.0063],[3991,0.0066],[3992,0.006],[3994,0.0068],[3996,0.0045],[3998,0.0062],[4001,0.0027],[4003,0.005],[4006,0.0003],[4008,0.0014],[4011,-0.002],[4032,0.0076],[4034,0.007],[4037,0.0076],[4040,0.0066],[4042,0.0066],[4044,0.0051],[4048,0.0064],[4052,-0.0017],[4053,-0.0016],[4054,-0.0017],[4056,-0.0013],[4073,-0.0059],[4075,-0.0056],[4076,-0.0059],[4078,-0.0031],[4080,-0.004],[4082,-0.0008],[4085,-0.0022],[4087,0.0016],[4090,0.0004],[4107,0.005],[4108,0.0048],[4112,0.0073],[4114,0.0063],[4115,0.0066],[4116,0.006],[4118,0.0068],[4120,0.0045],[4122,0.0062],[4125,0.0027],[4127,0.005],[4130,0.0003],[4132,0.0014],[4135,-0.002],[4157,0.0076],[4159,0.007],[4162,0.0076],[4165,0.0066],[4167,0.0066],[4169,0.0051],[4173,0.0064],[4177,-0.0017],[4178,-0.0016],[4179,-0.0017],[4181,-0.0013],[4198,-0.0059],[4200,-0.0056],[4201,-0.0059],[4203,-0.0031],[4205,-0.004],[4207,-0.0008],[4210,-0.0022],[4212,0.0016],[4215,0.0004],[4232,0.005],[4233,0.0048],[4237,0.0073],[4239,0.0063],[4240,0.0066],[4241,0.006],[4243,0.0068],[4245,0.0045],[4247,0.0062],[4250,0.0027],[4252,0.005],[4255,0.0003],[4257,0.0014],[4260,-0.002],[4283,0.0076],[4285,0.007],[4288,0.0076],[4291,0.0066],[4293,0.0066],[4295,0.0051],[4299,0.0064],[4303,-0.0017],[4304,-0.0016],[4305,-0.0017],[4307,-0.0013],[4324,-0.0059],[4326,-0.0056],[4327,-0.0059],[4329,-0.0031],[4331,-0.004],[4333,-0.0008],[4336,-0.0022],[4338,0.0016],[4341,0.0004],[4360,0.0596],[4364,-0.0348],[4379,-0.0029],[4380,-0.003],[4384,-0.0002],[4386,-0.0005],[4388,0.001],[4390,0.0003],[4408,0.0057],[4409,0.0055],[4413,0.008],[4417,0.006],[4419,0.0071],[4421,0.0057],[4422,0.0068],[4426,0.0029],[4428,0.005],[4431,-0.0006],[4433,0.0023],[4435,-0.0013],[4436,0]]
       }
       chest_tan_ctrl.Location.Y {
        type=MovieSceneScriptingFloatChannel
        default=0.0644
        thinned_from=529
        keys [[1790,0.0644],[1793,0.0584],[1795,0.0603],[1797,0.0554],[1799,0.0565],[1801,0.0527],[1803,0.0551],[1806,0.0477],[1808,0.0504],[1811,0.0444],[1813,0.0483],[1818,0.0444],[1819,0.0445],[1820,0.0441],[1822,0.0446],[1825,0.0445],[1845,0.0609],[1847,0.0607],[1861,0.0658],[1868,0.061],[1870,0.0644],[1873,0.0583],[1875,0.0602],[1877,0.0553],[1879,0.0564],[1881,0.053],[1883,0.0544],[1886,0.0476],[1888,0.0503],[1891,0.0438],[1893,0.0483],[1899,0.0443],[1902,0.0442],[1916,0.0534],[1917,0.0518],[1925,0.0613],[1927,0.0609],[1930,0.0633],[1931,0.0632],[1935,0.0653],[1937,0.0648],[1939,0.0661],[1941,0.065],[1942,0.0654],[1945,0.0646],[1947,0.0657],[1949,0.0637],[1950,0.0643],[1956,0.0555],[1957,0.0565],[1961,0.053],[1963,0.0545],[1966,0.0477],[1968,0.0503],[1971,0.0437],[1973,0.0483],[1980,0.0437],[1984,0.0445],[1985,0.0444],[1988,0.0461],[1989,0.0457],[1996,0.0536],[1997,0.0521],[2005,0.0617],[2007,0.0613],[2021,0.0663],[2028,0.0613],[2030,0.065],[2033,0.0584],[2035,0.0608],[2037,0.0548],[2039,0.0571],[2051,0.0457],[2052,0.0458],[2053,0.0446],[2055,0.0457],[2061,0.0441],[2065,0.0447],[2066,0.0446],[2076,0.0534],[2077,0.0518],[2085,0.0614],[2087,0.0609],[2100,0.0658],[2108,0.0609],[2110,0.0644],[2113,0.0584],[2115,0.0603],[2117,0.0554],[2119,0.0565],[2121,0.0527],[2123,0.0551],[2126,0.0477],[2128,0.0504],[2131,0.0444],[2133,0.0483],[2138,0.0444],[2139,0.0445],[2140,0.0441],[2142,0.0446],[2145,0.0445],[2165,0.0609],[2167,0.0607],[2181,0.0658],[2188,0.061],[2190,0.0644],[2193,0.0583],[2195,0.0602],[2197,0.0553],[2199,0.0564],[2201,0.053],[2203,0.0544],[2206,0.0476],[2208,0.0503],[2211,0.0438],[2213,0.0483],[2219,0.0443],[2222,0.0442],[2236,0.0534],[2237,0.0518],[2245,0.0613],[2247,0.0609],[2250,0.0633],[2251,0.0632],[2255,0.0653],[2257,0.0648],[2259,0.0661],[2261,0.065],[2262,0.0654],[2265,0.0646],[2267,0.0657],[2269,0.0637],[2270,0.0643],[2276,0.0555],[2277,0.0565],[2281,0.053],[2283,0.0545],[2286,0.0477],[2288,0.0503],[2291,0.0437],[2293,0.0483],[2300,0.0437],[2304,0.0445],[2305,0.0444],[2308,0.0461],[2309,0.0457],[2316,0.0536],[2317,0.0521],[2325,0.0617],[2327,0.0613],[2341,0.0663],[2348,0.0613],[2350,0.065],[2353,0.0584],[2355,0.0608],[2357,0.0548],[2359,0.0571],[2371,0.0457],[2372,0.0458],[2373,0.0446],[2375,0.0457],[2381,0.0441],[2385,0.0447],[2386,0.0446],[2396,0.0534],[2397,0.0518],[2405,0.0614],[2407,0.0609],[2420,0.0658],[2428,0.0609],[2430,0.0644],[2433,0.0584],[2435,0.0603],[2437,0.0554],[2439,0.0565],[2441,0.0527],[2443,0.0551],[2446,0.0477],[2448,0.0504],[2451,0.0444],[2453,0.0483],[2458,0.0444],[2459,0.0445],[2460,0.0441],[2462,0.0446],[2465,0.0445],[2469,0.049],[2470,0.045],[2507,-0.7629],[2509,1.4412],[2511,2.8569],[2513,2.1848],[2515,1.3545],[2518,0.5404],[2520,0.4597],[2781,0.0589],[2927,-0.6942],[2938,0.3742],[2986,-0.2346],[3058,-0.2346],[3070,-0.1745],[3083,-0.1745],[3099,-3.2683],[3185,-3.2683],[3215,0.0658],[3223,0.0609],[3225,0.0644],[3228,0.0584],[3230,0.0603],[3232,0.0554],[3234,0.0565],[3236,0.0527],[3238,0.0551],[3241,0.0477],[3243,0.0504],[3246,0.0444],[3248,0.0483],[3253,0.0444],[3254,0.0445],[3255,0.0441],[3257,0.0446],[3260,0.0445],[3280,0.0609],[3282,0.0607],[3296,0.0658],[3303,0.061],[3305,0.0644],[3308,0.0584],[3310,0.0602],[3312,0.0553],[3314,0.0564],[3316,0.053],[3318,0.0544],[3321,0.0476],[3323,0.0503],[3326,0.0438],[3328,0.0483],[3334,0.0443],[3337,0.0442],[3351,0.0534],[3352,0.0518],[3360,0.0613],[3362,0.0609],[3365,0.0633],[3366,0.0632],[3370,0.0653],[3372,0.0648],[3374,0.0661],[3376,0.065],[3377,0.0653],[3380,0.0646],[3382,0.0657],[3384,0.0637],[3385,0.0643],[3391,0.0555],[3392,0.0565],[3396,0.053],[3398,0.0545],[3401,0.0477],[3403,0.0503],[3406,0.0437],[3408,0.0483],[3415,0.0437],[3419,0.0445],[3420,0.0444],[3423,0.0461],[3424,0.0457],[3431,0.0536],[3432,0.0521],[3440,0.0617],[3442,0.0613],[3456,0.0663],[3463,0.0613],[3465,0.065],[3468,0.0584],[3470,0.0608],[3472,0.0548],[3474,0.0571],[3486,0.0457],[3487,0.0458],[3488,0.0446],[3490,0.0457],[3496,0.0441],[3500,0.0447],[3501,0.0446],[3511,0.0534],[3512,0.0518],[3520,0.0614],[3522,0.0609],[3535,0.0658],[3543,0.0609],[3545,0.0644],[3548,0.0584],[3550,0.0603],[3552,0.0554],[3554,0.0565],[3556,0.0527],[3558,0.0551],[3561,0.0477],[3563,0.0504],[3566,0.0444],[3568,0.0483],[3573,0.0444],[3574,0.0445],[3575,0.0441],[3577,0.0446],[3580,0.0445],[3600,0.0609],[3602,0.0607],[3616,0.0658],[3623,0.061],[3625,0.0644],[3628,0.0584],[3630,0.0602],[3632,0.0553],[3634,0.0564],[3636,0.053],[3638,0.0544],[3641,0.0476],[3643,0.0503],[3646,0.0438],[3648,0.0483],[3654,0.0443],[3657,0.0442],[3671,0.0534],[3672,0.0518],[3680,0.0613],[3682,0.0609],[3685,0.0633],[3686,0.0632],[3690,0.0653],[3692,0.0648],[3694,0.0661],[3696,0.065],[3697,0.0653],[3700,0.0646],[3702,0.0657],[3704,0.0637],[3705,0.0643],[3711,0.0555],[3712,0.0565],[3716,0.053],[3718,0.0545],[3721,0.0477],[3723,0.0503],[3726,0.0437],[3728,0.0483],[3735,0.0437],[3739,0.0445],[3740,0.0444],[3743,0.0461],[3744,0.0457],[3751,0.0536],[3752,0.0521],[3760,0.0617],[3762,0.0613],[3776,0.0663],[3783,0.0613],[3785,0.065],[3788,0.0584],[3790,0.0608],[3792,0.0548],[3794,0.0571],[3806,0.0457],[3807,0.0458],[3808,0.0446],[3810,0.0457],[3816,0.0441],[3820,0.0447],[3821,0.0446],[3831,0.0534],[3832,0.0518],[3840,0.0614],[3842,0.0609],[3855,0.0658],[3863,0.0609],[3865,0.0644],[3868,0.0584],[3870,0.0603],[3872,0.0554],[3874,0.0565],[3876,0.0527],[3878,0.0551],[3881,0.0477],[3883,0.0504],[3886,0.0444],[3888,0.0483],[3900,0.0663],[3907,0.0613],[3909,0.065],[3912,0.0584],[3914,0.0608],[3916,0.0548],[3918,0.0571],[3930,0.0457],[3931,0.0458],[3932,0.0446],[3934,0.0457],[3940,0.0441],[3944,0.0447],[3945,0.0446],[3955,0.0534],[3956,0.0518],[3964,0.0614],[3966,0.0609],[3979,0.0658],[3987,0.0609],[3989,0.0644],[3992,0.0584],[3994,0.0603],[3996,0.0554],[3998,0.0565],[4000,0.0527],[4002,0.0551],[4005,0.0477],[4007,0.0504],[4010,0.0444],[4012,0.0483],[4024,0.0663],[4031,0.0613],[4033,0.065],[4036,0.0584],[4038,0.0608],[4040,0.0548],[4042,0.0571],[4054,0.0457],[4055,0.0458],[4056,0.0446],[4058,0.0457],[4064,0.0441],[4068,0.0447],[4069,0.0446],[4079,0.0534],[4080,0.0518],[4088,0.0614],[4090,0.0609],[4103,0.0658],[4111,0.0609],[4113,0.0644],[4116,0.0584],[4118,0.0603],[4120,0.0554],[4122,0.0565],[4124,0.0527],[4126,0.0551],[4129,0.0477],[4131,0.0504],[4134,0.0444],[4136,0.0483],[4149,0.0663],[4156,0.0613],[4158,0.065],[4161,0.0584],[4163,0.0608],[4165,0.0548],[4167,0.0571],[4179,0.0457],[4180,0.0458],[4181,0.0446],[4183,0.0457],[4189,0.0441],[4193,0.0447],[4194,0.0446],[4204,0.0534],[4205,0.0518],[4213,0.0614],[4215,0.0609],[4228,0.0658],[4236,0.0609],[4238,0.0644],[4241,0.0584],[4243,0.0603],[4245,0.0554],[4247,0.0565],[4249,0.0527],[4251,0.0551],[4254,0.0477],[4256,0.0504],[4259,0.0444],[4261,0.0483],[4275,0.0663],[4282,0.0613],[4284,0.065],[4287,0.0584],[4289,0.0608],[4291,0.0548],[4293,0.0571],[4305,0.0457],[4306,0.0458],[4307,0.0446],[4309,0.0457],[4315,0.0441],[4319,0.0447],[4320,0.0446],[4330,0.0534],[4331,0.0518],[4339,0.0614],[4341,0.0609],[4354,0.0658],[4360,0.9804],[4364,0.9275],[4389,0.0609],[4391,0.0607],[4405,0.0658],[4412,0.061],[4414,0.0644],[4417,0.0584],[4419,0.0602],[4421,0.0553],[4423,0.0564],[4425,0.053],[4427,0.0544],[4430,0.0476],[4432,0.0503],[4436,-0.0303]]
       }
       chest_tan_ctrl.Location.Z {
        type=MovieSceneScriptingFloatChannel
        default=0.4394
        thinned_from=448
        keys [[1790,0.4394],[1793,0.4386],[1796,0.4435],[1797,0.4423],[1799,0.4426],[1801,0.4418],[1803,0.4441],[1806,0.4429],[1808,0.4433],[1812,0.4411],[1813,0.4417],[1828,0.4136],[1829,0.4138],[1831,0.4096],[1832,0.4104],[1840,0.4066],[1843,0.4067],[1867,0.4351],[1868,0.4349],[1870,0.4394],[1873,0.4386],[1876,0.4434],[1877,0.4423],[1879,0.4425],[1881,0.4418],[1883,0.444],[1886,0.4428],[1888,0.4433],[1892,0.4414],[1893,0.4416],[1918,0.4067],[1920,0.4073],[1922,0.4067],[1925,0.4086],[1927,0.4082],[1932,0.4143],[1933,0.4124],[1948,0.44],[1950,0.4348],[1952,0.4373],[1953,0.4369],[1959,0.4423],[1961,0.4418],[1963,0.4439],[1966,0.4429],[1968,0.4433],[1972,0.4415],[1973,0.4416],[1988,0.4135],[1989,0.4165],[1998,0.4067],[2000,0.4072],[2002,0.4067],[2005,0.4085],[2007,0.4082],[2027,0.4348],[2028,0.4344],[2030,0.44],[2034,0.4382],[2036,0.4433],[2040,0.4418],[2045,0.4434],[2049,0.443],[2078,0.4067],[2080,0.4073],[2082,0.4067],[2085,0.4082],[2086,0.4079],[2107,0.4351],[2108,0.4349],[2110,0.4394],[2113,0.4386],[2116,0.4435],[2117,0.4423],[2119,0.4426],[2121,0.4418],[2123,0.4441],[2126,0.4429],[2128,0.4433],[2132,0.4411],[2133,0.4417],[2148,0.4136],[2149,0.4138],[2151,0.4096],[2152,0.4104],[2160,0.4066],[2163,0.4067],[2187,0.4351],[2188,0.4349],[2190,0.4394],[2193,0.4386],[2196,0.4434],[2197,0.4423],[2199,0.4425],[2201,0.4418],[2203,0.444],[2206,0.4428],[2208,0.4433],[2212,0.4414],[2213,0.4416],[2238,0.4067],[2240,0.4073],[2242,0.4067],[2245,0.4086],[2247,0.4082],[2252,0.4143],[2253,0.4124],[2268,0.44],[2270,0.4348],[2272,0.4373],[2273,0.4369],[2279,0.4423],[2281,0.4418],[2283,0.4439],[2286,0.4429],[2288,0.4433],[2292,0.4415],[2293,0.4416],[2308,0.4135],[2309,0.4165],[2318,0.4067],[2320,0.4072],[2322,0.4067],[2325,0.4085],[2327,0.4082],[2347,0.4348],[2348,0.4344],[2350,0.44],[2354,0.4382],[2356,0.4433],[2360,0.4418],[2365,0.4434],[2369,0.443],[2398,0.4067],[2400,0.4073],[2402,0.4067],[2405,0.4082],[2406,0.4079],[2427,0.4351],[2428,0.4349],[2430,0.4394],[2433,0.4386],[2436,0.4435],[2437,0.4423],[2439,0.4426],[2441,0.4418],[2443,0.4441],[2446,0.4429],[2448,0.4433],[2452,0.4411],[2453,0.4417],[2466,0.4177],[2506,0.1998],[2507,0.2024],[2508,0.2002],[2511,0.565],[2518,0.1199],[2525,0.3782],[2758,0.3163],[2769,-0.0988],[2780,-0.2962],[2802,-0.451],[2927,-0.4406],[2938,-0.6028],[2986,-0.2589],[3058,-0.2589],[3070,0.1486],[3083,0.1486],[3099,-1.6501],[3185,-1.6501],[3222,0.4351],[3223,0.4349],[3225,0.4394],[3228,0.4386],[3231,0.4435],[3232,0.4423],[3234,0.4426],[3236,0.4418],[3238,0.4441],[3241,0.4429],[3243,0.4433],[3247,0.4411],[3248,0.4417],[3263,0.4136],[3264,0.4138],[3266,0.4096],[3267,0.4104],[3275,0.4066],[3278,0.4067],[3302,0.4351],[3303,0.4349],[3305,0.4394],[3308,0.4386],[3311,0.4434],[3312,0.4423],[3314,0.4425],[3316,0.4418],[3318,0.444],[3321,0.4428],[3323,0.4433],[3327,0.4414],[3328,0.4416],[3353,0.4067],[3355,0.4073],[3357,0.4067],[3360,0.4086],[3362,0.4082],[3367,0.4143],[3368,0.4124],[3383,0.44],[3385,0.4348],[3387,0.4373],[3388,0.4369],[3394,0.4423],[3396,0.4418],[3398,0.4439],[3401,0.4429],[3403,0.4433],[3407,0.4415],[3408,0.4416],[3423,0.4134],[3424,0.4165],[3433,0.4067],[3435,0.4072],[3437,0.4067],[3440,0.4085],[3442,0.4082],[3462,0.4348],[3463,0.4344],[3465,0.44],[3469,0.4382],[3471,0.4433],[3475,0.4418],[3480,0.4434],[3484,0.443],[3513,0.4067],[3515,0.4073],[3517,0.4067],[3520,0.4082],[3521,0.4079],[3542,0.4351],[3543,0.4349],[3545,0.4394],[3548,0.4386],[3551,0.4435],[3552,0.4423],[3554,0.4426],[3556,0.4418],[3558,0.4441],[3561,0.4429],[3563,0.4433],[3567,0.4411],[3568,0.4417],[3583,0.4136],[3584,0.4138],[3586,0.4096],[3587,0.4104],[3595,0.4066],[3598,0.4067],[3622,0.4351],[3623,0.4349],[3625,0.4394],[3628,0.4386],[3631,0.4434],[3632,0.4423],[3634,0.4425],[3636,0.4418],[3638,0.444],[3641,0.4428],[3643,0.4433],[3647,0.4414],[3648,0.4416],[3673,0.4067],[3675,0.4073],[3677,0.4067],[3680,0.4086],[3682,0.4082],[3687,0.4143],[3688,0.4124],[3703,0.44],[3705,0.4348],[3707,0.4373],[3708,0.4369],[3714,0.4423],[3716,0.4418],[3718,0.4439],[3721,0.4429],[3723,0.4433],[3727,0.4415],[3728,0.4416],[3743,0.4134],[3744,0.4165],[3753,0.4067],[3755,0.4072],[3757,0.4067],[3760,0.4085],[3762,0.4082],[3782,0.4348],[3783,0.4344],[3785,0.44],[3789,0.4382],[3791,0.4433],[3795,0.4418],[3800,0.4434],[3804,0.443],[3833,0.4067],[3835,0.4073],[3837,0.4067],[3840,0.4082],[3841,0.4079],[3862,0.4351],[3863,0.4349],[3865,0.4394],[3868,0.4386],[3871,0.4435],[3872,0.4423],[3874,0.4426],[3876,0.4418],[3878,0.4441],[3881,0.4429],[3883,0.4433],[3887,0.4411],[3888,0.4417],[3907,0.4344],[3909,0.44],[3913,0.4382],[3915,0.4433],[3919,0.4418],[3924,0.4434],[3928,0.443],[3957,0.4067],[3959,0.4073],[3961,0.4067],[3964,0.4082],[3965,0.4079],[3986,0.4351],[3987,0.4349],[3989,0.4394],[3992,0.4386],[3995,0.4435],[3996,0.4423],[3998,0.4426],[4000,0.4418],[4002,0.4441],[4005,0.4429],[4007,0.4433],[4011,0.4411],[4012,0.4417],[4031,0.4344],[4033,0.44],[4037,0.4382],[4039,0.4433],[4043,0.4418],[4048,0.4434],[4052,0.443],[4081,0.4067],[4083,0.4073],[4085,0.4067],[4088,0.4082],[4089,0.4079],[4110,0.4351],[4111,0.4349],[4113,0.4394],[4116,0.4386],[4119,0.4435],[4120,0.4423],[4122,0.4426],[4124,0.4418],[4126,0.4441],[4129,0.4429],[4131,0.4433],[4135,0.4411],[4136,0.4417],[4156,0.4344],[4158,0.44],[4162,0.4382],[4164,0.4433],[4168,0.4418],[4173,0.4434],[4177,0.443],[4206,0.4067],[4208,0.4073],[4210,0.4067],[4213,0.4082],[4214,0.4079],[4235,0.4351],[4236,0.4349],[4238,0.4394],[4241,0.4386],[4244,0.4435],[4245,0.4423],[4247,0.4426],[4249,0.4418],[4251,0.4441],[4254,0.4429],[4256,0.4433],[4260,0.4411],[4261,0.4417],[4282,0.4344],[4284,0.44],[4288,0.4382],[4290,0.4433],[4294,0.4418],[4299,0.4434],[4303,0.443],[4332,0.4067],[4334,0.4073],[4336,0.4067],[4339,0.4082],[4340,0.4079],[4360,0.4683],[4364,2.4219],[4372,0.4136],[4373,0.4138],[4375,0.4096],[4376,0.4104],[4384,0.4066],[4387,0.4067],[4411,0.4351],[4412,0.4349],[4414,0.4394],[4417,0.4386],[4420,0.4434],[4421,0.4423],[4423,0.4425],[4425,0.4418],[4427,0.444],[4430,0.4428],[4432,0.4433],[4435,0.4418],[4436,0.265]]
       }
       chest_tan_ctrl.Rotation.X {
        type=MovieSceneScriptingFloatChannel
        default=3.7769
        thinned_from=74
        keys [[1790,3.7769],[1814,3.509],[1853,3.8697],[1894,3.5084],[1934,3.8695],[1974,3.5084],[2014,3.8773],[2054,3.5033],[2093,3.8697],[2134,3.509],[2173,3.8697],[2214,3.5084],[2254,3.8695],[2294,3.5084],[2334,3.8773],[2374,3.5033],[2413,3.8697],[2454,3.509],[2464,3.5715],[2469,4.0048],[2502,10.7051],[2507,7.6935],[2509,17.033],[2511,22.7362],[2513,15.9516],[2515,1.8236],[2518,0],[2778,0.9134],[2780,0.8612],[2927,0.8364],[2938,-11.5712],[2986,0.5386],[3058,0.5386],[3070,2.3862],[3083,2.3862],[3099,2.5044],[3185,2.5044],[3249,3.509],[3288,3.8697],[3329,3.5084],[3369,3.8695],[3409,3.5084],[3449,3.8773],[3489,3.5033],[3528,3.8697],[3569,3.509],[3608,3.8697],[3649,3.5084],[3689,3.8695],[3729,3.5084],[3769,3.8773],[3809,3.5033],[3848,3.8697],[3893,3.8773],[3933,3.5033],[3972,3.8697],[4017,3.8773],[4057,3.5033],[4096,3.8697],[4142,3.8773],[4182,3.5033],[4221,3.8697],[4268,3.8773],[4308,3.5033],[4347,3.8697],[4360,-6.2732],[4364,-6.5512],[4397,3.8697],[4416,3.7568],[4426,3.6184],[4433,3.534],[4435,3.5183],[4436,0]]
       }
       chest_tan_ctrl.Rotation.Y {
        type=MovieSceneScriptingFloatChannel
        default=0
        thinned_from=22
        keys [[1790,0],[2470,0],[2529,0.272],[2530,0.265],[2763,0],[2927,0.1514],[2938,2.2454],[2986,2.7091],[3058,2.7091],[3070,0.0639],[3083,0.0639],[3099,0.3364],[3185,0.3364],[4360,1.6829],[4364,1.6645],[4436,0]]
       }
       chest_tan_ctrl.Rotation.Z {
        type=MovieSceneScriptingFloatChannel
        default=0
        thinned_from=14
        keys [[1790,0],[2801,0],[2927,-0.7418],[2938,-0.6795],[3058,-0.679],[3070,2.1768],[3083,2.1768],[3099,-0.7],[3185,-0.7],[4360,0.6158],[4364,0.6082],[4436,0]]
       }
       chest_tan_ctrl.Scale.X {
        type=MovieSceneScriptingFloatChannel
        default=1
        thinned_from=12
        keys [[1790,1],[2927,0.9479],[2938,0.9555],[3058,0.9558],[3070,0.9623],[3083,0.9623],[3099,0.9532],[3185,0.9532],[4364,1.003],[4436,1]]
       }
       chest_tan_ctrl.Scale.Y {
        type=MovieSceneScriptingFloatChannel
        default=1
        thinned_from=12
        keys [[1790,1],[2927,0.9479],[2938,0.9555],[3058,0.9558],[3070,0.9623],[3083,0.9623],[3099,0.9532],[3185,0.9532],[4364,1.003],[4436,1]]
       }
       chest_tan_ctrl.Scale.Z {
        type=MovieSceneScriptingFloatChannel
        default=1
        thinned_from=12
        keys [[1790,1],[2927,0.9479],[2938,0.9555],[3058,0.9558],[3070,0.9623],[3083,0.9623],[3099,0.9532],[3185,0.9532],[4364,1.003],[4436,1]]
       }
       Weight {type=MovieSceneScriptingFloatChannel, default=1, const [1,10]}
       foot_l_ik_ctrl.Space {
        type=MovieSceneScriptingSpaceChannel
        const [
         "<Struct 'MovieSceneControlRigSpaceBaseKey' (0x8523f60cda0) {space_type: World, control_rig_element: {type: Reference, name: \"WorldSpace\"}}>"
         1
        ]
       }
       leg_l_pv_ik_ctrl.Space {
        type=MovieSceneScriptingSpaceChannel
        const [
         "<Struct 'MovieSceneControlRigSpaceBaseKey' (0x8523f60cb60) {space_type: World, control_rig_element: {type: Reference, name: \"WorldSpace\"}}>"
         1
        ]
       }
       foot_r_ik_ctrl.Space {
        type=MovieSceneScriptingSpaceChannel
        const [
         "<Struct 'MovieSceneControlRigSpaceBaseKey' (0x8523f60ce60) {space_type: World, control_rig_element: {type: Reference, name: \"WorldSpace\"}}>"
         1
        ]
       }
       leg_r_pv_ik_ctrl.Space {
        type=MovieSceneScriptingSpaceChannel
        const [
         "<Struct 'MovieSceneControlRigSpaceBaseKey' (0x8523f60c620) {space_type: World, control_rig_element: {type: Reference, name: \"WorldSpace\"}}>"
         1
        ]
       }
       body_ctrl.Space {
        type=MovieSceneScriptingSpaceChannel
        const [
         "<Struct 'MovieSceneControlRigSpaceBaseKey' (0x8523f60c740) {space_type: World, control_rig_element: {type: Reference, name: \"WorldSpace\"}}>"
         1
        ]
       }
      ]
     }
     {
      type=MovieSceneControlRigParameterSection
      channels [
       global_ctrl.Location.X {type=MovieSceneScriptingFloatChannel, default=0, const [0,16]}
       global_ctrl.Location.Y {type=MovieSceneScriptingFloatChannel, default=0, const [0,16]}
       global_ctrl.Location.Z {type=MovieSceneScriptingFloatChannel, default=0, const [0,16]}
       global_ctrl.Rotation.X {type=MovieSceneScriptingFloatChannel, default=0, const [0,16]}
       global_ctrl.Rotation.Y {type=MovieSceneScriptingFloatChannel, default=0, const [0,16]}
       global_ctrl.Rotation.Z {type=MovieSceneScriptingFloatChannel, default=0, const [0,16]}
       root_ctrl.Location.X {type=MovieSceneScriptingFloatChannel, default=0, const [0,16]}
       root_ctrl.Location.Y {type=MovieSceneScriptingFloatChannel, default=0, const [0,16]}
       root_ctrl.Location.Z {type=MovieSceneScriptingFloatChannel, default=0, const [0,16]}
       root_ctrl.Rotation.X {type=MovieSceneScriptingFloatChannel, default=0, const [0,16]}
       root_ctrl.Rotation.Y {type=MovieSceneScriptingFloatChannel, default=0, const [0,16]}
       root_ctrl.Rotation.Z {type=MovieSceneScriptingFloatChannel, default=0, const [0,16]}
       root_ctrl.Scale.X {type=MovieSceneScriptingFloatChannel, default=0, const [0,16]}
       root_ctrl.Scale.Y {type=MovieSceneScriptingFloatChannel, default=0, const [0,16]}
       root_ctrl.Scale.Z {type=MovieSceneScriptingFloatChannel, default=0, const [0,16]}
       body_offset_ctrl.Location.X {type=MovieSceneScriptingFloatChannel, default=0, const [0,16]}
       body_offset_ctrl.Location.Y {type=MovieSceneScriptingFloatChannel, default=0, const [0,16]}
       body_offset_ctrl.Location.Z {type=MovieSceneScriptingFloatChannel, default=0, const [0,16]}
       body_offset_ctrl.Rotation.X {type=MovieSceneScriptingFloatChannel, default=0, const [0,16]}
       body_offset_ctrl.Rotation.Y {type=MovieSceneScriptingFloatChannel, default=0, const [0,16]}
       body_offset_ctrl.Rotation.Z {type=MovieSceneScriptingFloatChannel, default=0, const [0,16]}
       body_offset_ctrl.Scale.X {type=MovieSceneScriptingFloatChannel, default=0, const [0,16]}
       body_offset_ctrl.Scale.Y {type=MovieSceneScriptingFloatChannel, default=0, const [0,16]}
       body_offset_ctrl.Scale.Z {type=MovieSceneScriptingFloatChannel, default=0, const [0,16]}
       thigh_l_fk_ctrl.Location.X {type=MovieSceneScriptingFloatChannel, default=0, thinned_from=21, keys [[1790,0],[4366,0]]}
       thigh_l_fk_ctrl.Location.Y {type=MovieSceneScriptingFloatChannel, default=0, thinned_from=21, keys [[1790,0],[4366,0]]}
       thigh_l_fk_ctrl.Location.Z {type=MovieSceneScriptingFloatChannel, default=0, thinned_from=21, keys [[1790,0],[4366,0]]}
       thigh_l_fk_ctrl.Rotation.X {
        type=MovieSceneScriptingFloatChannel
        default=0
        thinned_from=27
        keys [[1790,5.8464],[2113,12.8113],[2217,5.8464],[2259,4.2735],[2290,33.9109],[2342,4.568],[2378,5.1591],[2529,5.0462],[2556,5.0141],[2754,4.5559],[2803,5.3762],[2964,5.3762],[3087,3.0939],[3199,4.6738],[3509,1.6945],[3560,1.7023],[3584,2.4725],[3610,1.7181],[3716,1.7456],[3884,0.8489],[4003,11.0133],[4364,11.0133],[4366,6.9375]]
       }
       thigh_l_fk_ctrl.Rotation.Y {
        type=MovieSceneScriptingFloatChannel
        default=0
        thinned_from=28
        keys [[1790,3.7683],[2113,5.049],[2217,3.7683],[2259,-9.9162],[2290,-8.2257],[2342,4.203],[2378,3.2246],[2523,2.7382],[2529,2.7266],[2556,2.6445],[2754,2.0993],[2803,7.0218],[2944,7.0218],[2962,-6.3308],[3027,-3.6278],[3091,-1.9175],[3220,-0.8414],[3509,0.5211],[3560,0.3949],[3584,1.1431],[3610,0.1001],[3716,-0.5453],[3884,-0.9719],[4003,0.1752],[4364,0.1752],[4366,13.7014]]
       }
       thigh_l_fk_ctrl.Rotation.Z {
        type=MovieSceneScriptingFloatChannel
        default=0
        thinned_from=34
        keys [[1790,3.8845],[2113,5.3073],[2217,3.8845],[2259,-5.4788],[2290,-7.8844],[2342,-1.0496],[2378,3.3559],[2529,3.2525],[2556,3.2232],[2754,2.8037],[2803,15.9843],[2944,15.9843],[2954,30.6193],[2969,23.2053],[2975,17.4946],[3011,28.9699],[3030,30.9591],[3087,31.8916],[3199,13.4127],[3509,2.5056],[3560,4.2499],[3579,-10.514],[3585,-12.4071],[3610,7.7856],[3716,13.967],[3757,13.9816],[3890,-35.2282],[4003,-46.9236],[4201,-63.7548],[4364,-57.5629],[4366,-47.7906]]
       }
       thigh_l_fk_ctrl.Scale.X {type=MovieSceneScriptingFloatChannel, default=0, const [0,21]}
       thigh_l_fk_ctrl.Scale.Y {type=MovieSceneScriptingFloatChannel, default=0, const [0,21]}
       thigh_l_fk_ctrl.Scale.Z {type=MovieSceneScriptingFloatChannel, default=0, const [0,21]}
       calf_l_fk_ctrl.Location.X {type=MovieSceneScriptingFloatChannel, default=0, thinned_from=20, keys [[1790,0],[4366,0]]}
       calf_l_fk_ctrl.Location.Y {type=MovieSceneScriptingFloatChannel, default=0, thinned_from=20, keys [[1790,0],[4366,0]]}
       calf_l_fk_ctrl.Location.Z {type=MovieSceneScriptingFloatChannel, default=0, thinned_from=20, keys [[1790,0],[4366,0]]}
       calf_l_fk_ctrl.Rotation.X {
        type=MovieSceneScriptingFloatChannel
        default=0
        thinned_from=26
        keys [[1790,0.5769],[2113,0.5218],[2217,0.5769],[2334,0.4366],[2529,0.4564],[2556,0.4606],[2754,0.517],[2803,0.3295],[2944,0.3295],[2962,0.042],[2984,0.3295],[2994,0.1578],[3022,0.3538],[3199,0.422],[3305,0.3802],[3560,0.3802],[3584,1.0748],[3604,0.3802],[3884,0.3802],[4003,1.4574],[4364,1.4574],[4366,0.3802]]
       }
       calf_l_fk_ctrl.Rotation.Y {
        type=MovieSceneScriptingFloatChannel
        default=0
        thinned_from=26
        keys [[1790,1.1444],[2113,1.1558],[2217,1.1444],[2334,1.1665],[2556,1.1585],[2754,1.1397],[2803,1.2068],[2944,1.2068],[2962,1.3481],[2984,1.2068],[2994,1.2686],[3022,1.1992],[3199,1.1753],[3305,1.164],[3560,1.164],[3584,0.8405],[3604,1.164],[3884,1.164],[4003,0.0487],[4364,0.0487],[4366,1.164]]
       }
       calf_l_fk_ctrl.Rotation.Z {
        type=MovieSceneScriptingFloatChannel
        default=0
        thinned_from=30
        keys [[1790,-13.2881],[2113,-15.5675],[2217,-17.7758],[2334,-24.398],[2529,-23.3883],[2556,-23.1716],[2754,-20.2899],[2803,-29.5979],[2941,-29.5979],[2945,-25.6797],[2962,-45.1132],[2971,-42.195],[2984,-29.5979],[2992,-32.9863],[3047,-20.0127],[3068,-15.8251],[3199,-24.912],[3305,-26.6744],[3560,-26.6744],[3580,48.2156],[3584,51.0781],[3604,-26.6744],[3764,-26.6744],[3883,53.5853],[4364,74.2446],[4366,-26.6744]]
       }
       calf_l_fk_ctrl.Scale.X {type=MovieSceneScriptingFloatChannel, default=0, const [0,20]}
       calf_l_fk_ctrl.Scale.Y {type=MovieSceneScriptingFloatChannel, default=0, const [0,20]}
       calf_l_fk_ctrl.Scale.Z {type=MovieSceneScriptingFloatChannel, default=0, const [0,20]}
       foot_l_fk_ctrl.Location.X {type=MovieSceneScriptingFloatChannel, default=0, thinned_from=16, keys [[1790,0],[4366,0]]}
       foot_l_fk_ctrl.Location.Y {type=MovieSceneScriptingFloatChannel, default=0, thinned_from=16, keys [[1790,0.0398],[4366,0.0398]]}
       foot_l_fk_ctrl.Location.Z {type=MovieSceneScriptingFloatChannel, default=0, thinned_from=16, keys [[1790,-0.8885],[4366,-0.8885]]}
       foot_l_fk_ctrl.Rotation.X {
        type=MovieSceneScriptingFloatChannel
        default=0
        thinned_from=19
        keys [[1790,12.0877],[2113,5.6724],[2217,12.0877],[2529,10.9763],[2556,10.8316],[2754,9.1116],[2803,15.2338],[2964,15.2338],[3199,12.7251],[3305,21.8213],[4004,21.8213],[4364,9.7279],[4366,0.5099]]
       }
       foot_l_fk_ctrl.Rotation.Y {
        type=MovieSceneScriptingFloatChannel
        default=0
        thinned_from=19
        keys [[1790,-6.9397],[2113,-4.317],[2217,-6.9397],[2529,-6.3025],[2556,-6.2196],[2754,-5.2335],[2803,-6.5358],[2964,-6.5358],[3199,-5.4401],[3305,-1.6965],[4364,-1.6965],[4366,-0.7572]]
       }
       foot_l_fk_ctrl.Rotation.Z {
        type=MovieSceneScriptingFloatChannel
        default=0
        thinned_from=19
        keys [[1790,-0.0505],[2113,1.7115],[2217,-0.0505],[2516,-0.1799],[2523,-0.1833],[2529,-0.1847],[2556,-0.1945],[2754,-0.2498],[2803,10.9782],[2964,10.9782],[3199,9.188],[3305,15.2731],[4364,15.2731],[4366,29.798]]
       }
       foot_l_fk_ctrl.Scale.X {type=MovieSceneScriptingFloatChannel, default=0, const [0,16]}
       foot_l_fk_ctrl.Scale.Y {type=MovieSceneScriptingFloatChannel, default=0, const [0,16]}
       foot_l_fk_ctrl.Scale.Z {type=MovieSceneScriptingFloatChannel, default=0, const [0,16]}
       ball_l_fk_ctrl.Location.X {type=MovieSceneScriptingFloatChannel, default=0, thinned_from=16, keys [[1790,0],[4366,0]]}
       ball_l_fk_ctrl.Location.Y {type=MovieSceneScriptingFloatChannel, default=0, thinned_from=16, keys [[1790,0],[4366,0]]}
       ball_l_fk_ctrl.Location.Z {type=MovieSceneScriptingFloatChannel, default=0, thinned_from=16, keys [[1790,0],[4366,0]]}
       ball_l_fk_ctrl.Rotation.X {type=MovieSceneScriptingFloatChannel, default=0, const [0,16]}
       ball_l_fk_ctrl.Rotation.Y {type=MovieSceneScriptingFloatChannel, default=0, const [0,16]}
       ball_l_fk_ctrl.Rotation.Z {type=MovieSceneScriptingFloatChannel, default=0, const [0,16]}
       ball_l_fk_ctrl.Scale.X {type=MovieSceneScriptingFloatChannel, default=0, const [0,16]}
       ball_l_fk_ctrl.Scale.Y {type=MovieSceneScriptingFloatChannel, default=0, const [0,16]}
       ball_l_fk_ctrl.Scale.Z {type=MovieSceneScriptingFloatChannel, default=0, const [0,16]}
       foot_l_ik_ctrl.Location.X {
        type=MovieSceneScriptingFloatChannel
        default=0
        thinned_from=20
        keys [[1790,11.4245],[2217,11.4245],[2279,19.3721],[2298,4.3961],[2364,-14.766],[2419,-14.7152],[2516,-8.7886],[2523,-10.2521],[2529,-13.3584],[2556,-20.6827],[2754,8.7873],[2944,0.3141],[2964,0.3141],[3199,42.165],[3509,-115.9637],[3841,0],[4067,-192.9328],[4364,276.2446],[4366,-257.9788]]
       }
       foot_l_ik_ctrl.Location.Y {
        type=MovieSceneScriptingFloatChannel
        default=0
        thinned_from=20
        keys [[1790,-18.7925],[2217,-18.7925],[2279,-16.7475],[2298,1.929],[2364,15.5329],[2419,15.4794],[2516,28.5907],[2523,8.951],[2529,14.0156],[2556,53.3195],[2754,177.9919],[2944,97.2661],[2964,97.2661],[3199,85.2655],[3509,-288.7845],[3841,0],[4067,630.3357],[4364,376.0631],[4366,195.7186]]
       }
       foot_l_ik_ctrl.Location.Z {
        type=MovieSceneScriptingFloatChannel
        default=0
        keys [[1790,3.6351],[2113,3.6351],[2217,3.1666],[2279,2.2379],[2298,-1.7146],[2346,2.0023],[2419,3.0592],[2516,-6.6429],[2523,2.6424],[2529,4.4528],[2556,10.5919],[2754,-2.6273],[2944,-2.2514],[2964,-2.2514],[3199,29.8964],[3509,-152.6163],[3841,0],[4067,308.9721],[4364,120.9357],[4366,945.1497]]
       }
       foot_l_ik_ctrl.Rotation.X {
        type=MovieSceneScriptingFloatChannel
        default=0
        thinned_from=20
        keys [[1790,12.1579],[2217,12.1579],[2279,12.2144],[2298,-0.8637],[2364,-0.9967],[2419,-0.9932],[2516,-19.9129],[2523,-18.3434],[2529,-19.0569],[2556,-25.7091],[2754,-3.3043],[2944,-32.0858],[2964,-32.0858],[3199,-37.6993],[3509,-38.0579],[3841,0],[4067,28.298],[4364,-39.9347],[4366,-80.0206]]
       }
       foot_l_ik_ctrl.Rotation.Y {
        type=MovieSceneScriptingFloatChannel
        default=0
        thinned_from=19
        keys [[1790,10.3366],[2217,10.3366],[2298,-4.6606],[2364,6.1216],[2419,6.1005],[2516,4.8564],[2523,0.4777],[2529,1.7102],[2556,7.3999],[2754,0.5306],[2944,3.6758],[2964,3.6758],[3199,-18.3776],[3509,-5.7216],[3841,0],[4067,3.1101],[4364,-23.2084],[4366,-47.0778]]
       }
       foot_l_ik_ctrl.Rotation.Z {
        type=MovieSceneScriptingFloatChannel
        default=0
        thinned_from=19
        keys [[1790,-15.6473],[2217,-15.6473],[2298,-6.1729],[2364,-20.3309],[2419,-20.2608],[2516,-15.5422],[2523,-13.8404],[2529,-13.8939],[2556,-20.1901],[2754,-17.4942],[2944,-17.8111],[2964,-17.8111],[3199,-11.4865],[3509,-9.2744],[3841,0],[4067,-13.14],[4364,-13.0988],[4366,117.8687]]
       }
       foot_l_ik_ctrl.Scale.X {
        type=MovieSceneScriptingFloatChannel
        default=0
        thinned_from=17
        keys [[1790,0],[2529,0],[2556,-0.0165],[2754,0.0285],[2944,-0.0146],[2964,-0.0146],[3199,-0.0095],[3509,-0.0017],[3841,0],[4067,-0.0041],[4366,-0.0058]]
       }
       foot_l_ik_ctrl.Scale.Y {
        type=MovieSceneScriptingFloatChannel
        default=0
        thinned_from=17
        keys [[1790,0],[2529,0],[2556,-0.0165],[2754,0.0285],[2944,-0.0146],[2964,-0.0146],[3199,-0.0095],[3509,-0.0017],[3841,0],[4067,-0.0041],[4366,-0.0058]]
       }
       foot_l_ik_ctrl.Scale.Z {
        type=MovieSceneScriptingFloatChannel
        default=0
        thinned_from=17
        keys [[1790,0],[2529,0],[2556,-0.0165],[2754,0.0285],[2944,-0.0146],[2964,-0.0146],[3199,-0.0095],[3509,-0.0017],[3841,0],[4067,-0.0041],[4366,-0.0058]]
       }
       ball_l_ik_ctrl.Location.X {type=MovieSceneScriptingFloatChannel, default=0, thinned_from=16, keys [[1790,0],[4366,0]]}
       ball_l_ik_ctrl.Location.Y {type=MovieSceneScriptingFloatChannel, default=0, thinned_from=16, keys [[1790,0],[4366,0]]}
       ball_l_ik_ctrl.Location.Z {type=MovieSceneScriptingFloatChannel, default=0, thinned_from=16, keys [[1790,0],[4366,0]]}
       ball_l_ik_ctrl.Rotation.X {
        type=MovieSceneScriptingFloatChannel
        default=0
        thinned_from=16
        keys [[1790,0],[2217,0],[2516,0.0025],[2523,0.0017],[2529,0.0007],[2556,-0.0112],[2754,0.0151],[2944,-0.0093],[2964,-0.0093],[3199,0.0074],[3509,0],[4366,0]]
       }
       ball_l_ik_ctrl.Rotation.Y {
        type=MovieSceneScriptingFloatChannel
        default=0
        thinned_from=16
        keys [[1790,0],[2217,0],[2516,0.0293],[2523,0.0082],[2529,0.0138],[2556,0.1464],[2754,0.7078],[2944,0.1296],[2964,0.1296],[3199,0.2055],[3509,0],[4366,0]]
       }
       ball_l_ik_ctrl.Rotation.Z {
        type=MovieSceneScriptingFloatChannel
        default=0
        thinned_from=16
        keys [[1790,0],[2217,0],[2516,-0.0524],[2523,0.0002],[2556,0.0001],[2754,0.0021],[2944,0.0001],[2964,0.0001],[3199,-0.061],[3509,0],[4366,0]]
       }
       ball_l_ik_ctrl.Scale.X {type=MovieSceneScriptingFloatChannel, default=0, const [0,16]}
       ball_l_ik_ctrl.Scale.Y {type=MovieSceneScriptingFloatChannel, default=0, const [0,16]}
       ball_l_ik_ctrl.Scale.Z {type=MovieSceneScriptingFloatChannel, default=0, const [0,16]}
       foot_bk1_l_ctrl.Location.X {type=MovieSceneScriptingFloatChannel, default=0, const [0,16]}
       foot_bk1_l_ctrl.Location.Y {type=MovieSceneScriptingFloatChannel, default=0, const [0,16]}
       foot_bk1_l_ctrl.Location.Z {type=MovieSceneScriptingFloatChannel, default=0, const [0,16]}
       foot_bk1_l_ctrl.Rotation.X {type=MovieSceneScriptingFloatChannel, default=0, const [0,16]}
       foot_bk1_l_ctrl.Rotation.Y {type=MovieSceneScriptingFloatChannel, default=0, const [0,16]}
       foot_bk1_l_ctrl.Rotation.Z {type=MovieSceneScriptingFloatChannel, default=0, const [0,16]}
       foot_bk1_l_ctrl.Scale.X {type=MovieSceneScriptingFloatChannel, default=0, const [0,16]}
       foot_bk1_l_ctrl.Scale.Y {type=MovieSceneScriptingFloatChannel, default=0, const [0,16]}
       foot_bk1_l_ctrl.Scale.Z {type=MovieSceneScriptingFloatChannel, default=0, const [0,16]}
       foot_roll_l_ctrl.Location.X {type=MovieSceneScriptingFloatChannel, default=0, const [0,16]}
       foot_roll_l_ctrl.Location.Y {type=MovieSceneScriptingFloatChannel, default=0, const [0,16]}
       foot_roll_l_ctrl.Location.Z {type=MovieSceneScriptingFloatChannel, default=0, const [0,16]}
       foot_roll_l_ctrl.Rotation.X {type=MovieSceneScriptingFloatChannel, default=0, const [0,16]}
       foot_roll_l_ctrl.Rotation.Y {type=MovieSceneScriptingFloatChannel, default=0, const [0,16]}
       foot_roll_l_ctrl.Rotation.Z {type=MovieSceneScriptingFloatChannel, default=0, const [0,16]}
       foot_roll_l_ctrl.Scale.X {type=MovieSceneScriptingFloatChannel, default=0, const [0,16]}
       foot_roll_l_ctrl.Scale.Y {type=MovieSceneScriptingFloatChannel, default=0, const [0,16]}
       foot_roll_l_ctrl.Scale.Z {type=MovieSceneScriptingFloatChannel, default=0, const [0,16]}
       tip_l_ctrl.Location.X {type=MovieSceneScriptingFloatChannel, default=0, const [0,16]}
       tip_l_ctrl.Location.Y {type=MovieSceneScriptingFloatChannel, default=0, const [0,16]}
       tip_l_ctrl.Location.Z {type=MovieSceneScriptingFloatChannel, default=0, const [0,16]}
       tip_l_ctrl.Rotation.X {type=MovieSceneScriptingFloatChannel, default=0, const [0,16]}
       tip_l_ctrl.Rotation.Y {type=MovieSceneScriptingFloatChannel, default=0, const [0,16]}
       tip_l_ctrl.Rotation.Z {type=MovieSceneScriptingFloatChannel, default=0, const [0,16]}
       tip_l_ctrl.Scale.X {type=MovieSceneScriptingFloatChannel, default=0, const [0,16]}
       tip_l_ctrl.Scale.Y {type=MovieSceneScriptingFloatChannel, default=0, const [0,16]}
       tip_l_ctrl.Scale.Z {type=MovieSceneScriptingFloatChannel, default=0, const [0,16]}
       heel_l_ctrl.Location.X {type=MovieSceneScriptingFloatChannel, default=0, const [0,16]}
       heel_l_ctrl.Location.Y {type=MovieSceneScriptingFloatChannel, default=0, const [0,16]}
       heel_l_ctrl.Location.Z {type=MovieSceneScriptingFloatChannel, default=0, const [0,16]}
       heel_l_ctrl.Rotation.X {type=MovieSceneScriptingFloatChannel, default=0, const [0,16]}
       heel_l_ctrl.Rotation.Y {type=MovieSceneScriptingFloatChannel, default=0, const [0,16]}
       heel_l_ctrl.Rotation.Z {type=MovieSceneScriptingFloatChannel, default=0, const [0,16]}
       heel_l_ctrl.Scale.X {type=MovieSceneScriptingFloatChannel, default=0, const [0,16]}
       heel_l_ctrl.Scale.Y {type=MovieSceneScriptingFloatChannel, default=0, const [0,16]}
       heel_l_ctrl.Scale.Z {type=MovieSceneScriptingFloatChannel, default=0, const [0,16]}
       leg_l_softness {type=MovieSceneScriptingFloatChannel, default=0, const [0,17]}
       leg_l_pv_ik_ctrl.Location.X {
        type=MovieSceneScriptingFloatChannel
        default=0
        thinned_from=18
        keys [[1790,0.5985],[2217,0.5985],[2326,-3.2269],[2348,-1.572],[2516,45.1766],[2523,47.5706],[2529,49.482],[2556,58.9591],[2754,54.2578],[2944,74.0452],[2964,74.0452],[3199,178.7102],[3509,-18.7854],[3841,0],[4067,239.1334],[4364,-334.8559],[4366,280.5852]]
       }
       leg_l_pv_ik_ctrl.Location.Y {
        type=MovieSceneScriptingFloatChannel
        default=0
        thinned_from=19
        keys [[1790,-0.2876],[2217,-0.2876],[2298,38.9806],[2326,40.549],[2348,39.9155],[2516,36.6356],[2523,23.2551],[2529,27.914],[2556,56.5534],[2754,187.7948],[2944,114.3105],[2964,114.3105],[3199,20.3179],[3509,-218.9878],[3841,0],[4067,516.7639],[4364,-310.0058],[4366,-416.2257]]
       }
       leg_l_pv_ik_ctrl.Location.Z {
        type=MovieSceneScriptingFloatChannel
        default=0
        thinned_from=19
        keys [[1790,-6.4618],[2217,-6.4618],[2298,-8.1596],[2326,25.0582],[2348,15.191],[2516,20.6191],[2523,22.3057],[2529,23.2766],[2556,31.5294],[2754,13.7137],[2944,13.9521],[2964,13.9521],[3199,-68.0975],[3509,87.7694],[3841,0],[4067,167.1416],[4364,-1097.766],[4366,-1161.9304]]
       }
       leg_l_pv_ik_ctrl.Rotation.X {
        type=MovieSceneScriptingFloatChannel
        default=0
        thinned_from=16
        keys [[1790,0],[2217,0],[2516,34.1922],[2523,16.3033],[2529,13.9406],[2556,22.0428],[2754,4.8852],[2944,1.042],[2964,1.042],[3199,-17.9042],[3509,-28.9858],[3841,0],[4067,-119.2947],[4364,-15.551],[4366,-5.4843]]
       }
       leg_l_pv_ik_ctrl.Rotation.Y {
        type=MovieSceneScriptingFloatChannel
        default=0
        thinned_from=16
        keys [[1790,0],[2217,0],[2516,-2.8301],[2523,-3.4592],[2529,-4.5593],[2556,-8.4805],[2754,7.0579],[2944,-6.5508],[2964,-6.5508],[3199,-7.8608],[3509,-4.4035],[3841,0],[4067,-6.1905],[4364,32.6203],[4366,-38.3662]]
       }
       leg_l_pv_ik_ctrl.Rotation.Z {
        type=MovieSceneScriptingFloatChannel
        default=0
        thinned_from=16
        keys [[1790,0],[2217,0],[2516,23.0245],[2523,3.6952],[2529,2.9625],[2556,15.3292],[2754,-4.4662],[2944,6.1245],[2964,6.1245],[3199,111.6361],[3509,-12.9217],[3841,0],[4067,-84.8797],[4364,-123.8534],[4366,-122.6971]]
       }
       leg_l_pv_ik_ctrl.Scale.X {
        type=MovieSceneScriptingFloatChannel
        default=0
        thinned_from=16
        keys [[1790,0],[2529,0],[2556,0.1543],[2754,-0.0973],[2944,0.1366],[2964,0.1366],[3199,0.1037],[3509,0.0155],[3841,0],[4067,0.0479],[4364,0.0177],[4366,0.0298]]
       }
       leg_l_pv_ik_ctrl.Scale.Y {
        type=MovieSceneScriptingFloatChannel
        default=0
        thinned_from=16
        keys [[1790,0],[2529,0],[2556,0.1543],[2754,-0.0973],[2944,0.1366],[2964,0.1366],[3199,0.1037],[3509,0.0155],[3841,0],[4067,0.0479],[4364,0.0177],[4366,0.0298]]
       }
       leg_l_pv_ik_ctrl.Scale.Z {
        type=MovieSceneScriptingFloatChannel
        default=0
        thinned_from=16
        keys [[1790,0],[2529,0],[2556,0.1543],[2754,-0.0973],[2944,0.1366],[2964,0.1366],[3199,0.1037],[3509,0.0155],[3841,0],[4067,0.0479],[4364,0.0177],[4366,0.0298]]
       }
       thigh_r_fk_ctrl.Location.X {type=MovieSceneScriptingFloatChannel, default=0, thinned_from=22, keys [[1790,0],[4366,0]]}
       thigh_r_fk_ctrl.Location.Y {type=MovieSceneScriptingFloatChannel, default=0, thinned_from=22, keys [[1790,0],[4366,0]]}
       thigh_r_fk_ctrl.Location.Z {type=MovieSceneScriptingFloatChannel, default=0, thinned_from=22, keys [[1790,0],[4366,0]]}
       thigh_r_fk_ctrl.Rotation.X {
        type=MovieSceneScriptingFloatChannel
        default=0
        thinned_from=30
        keys [[1790,6.8645],[2113,5.1073],[2202,6.8645],[2241,5.4085],[2340,7.7871],[2409,6.6654],[2523,6.163],[2529,6.1474],[2556,6.0371],[2754,5.2817],[2803,5.8644],[2944,5.8644],[2953,6.6506],[2968,5.8644],[3087,-10.4502],[3153,-9.5771],[3305,4.7751],[3509,4.7751],[3560,5.3305],[3584,6.547],[3612,6.4562],[3716,8.4242],[3803,6.6153],[3851,13.3515],[3893,16.6984],[3940,19.4044],[4364,19.4044],[4366,23.2371]]
       }
       thigh_r_fk_ctrl.Rotation.Y {
        type=MovieSceneScriptingFloatChannel
        default=0
        thinned_from=31
        keys [[1790,6.1047],[2113,5.6281],[2202,6.1047],[2241,3.5896],[2340,4.9736],[2409,5.8547],[2529,5.7304],[2556,5.6849],[2754,4.9849],[2803,7.2741],[2944,7.2741],[2953,6.3354],[2962,0.1265],[3003,7.6582],[3087,8.339],[3179,9.4251],[3305,10.8011],[3509,10.8011],[3560,10.3952],[3584,9.1086],[3612,9.5724],[3716,8.134],[3789,9.2469],[3816,7.6634],[3903,-1.5406],[3940,-3.1916],[4364,-3.1916],[4366,-3.0429]]
       }
       thigh_r_fk_ctrl.Rotation.Z {
        type=MovieSceneScriptingFloatChannel
        default=0
        thinned_from=33
        keys [[1790,1.5933],[2113,0.9595],[2202,1.5933],[2256,-15.7787],[2340,6.5607],[2409,1.5179],[2523,1.2803],[2529,1.2737],[2556,1.2286],[2754,0.9501],[2803,4.9639],[2944,4.9639],[2953,10.2668],[2967,9.0825],[3094,-21.3406],[3187,-28.0482],[3235,-2.5958],[3251,0.0838],[3286,-0.0529],[3509,4.4017],[3560,4.3447],[3576,-40.4699],[3584,-42.1731],[3612,4.2074],[3695,3.845],[3803,-1.1496],[3820,-13.0792],[3863,-26.1045],[3940,-29.5464],[4364,-29.5464],[4366,-47.9204]]
       }
       thigh_r_fk_ctrl.Scale.X {type=MovieSceneScriptingFloatChannel, default=0, const [0,22]}
       thigh_r_fk_ctrl.Scale.Y {type=MovieSceneScriptingFloatChannel, default=0, const [0,22]}
       thigh_r_fk_ctrl.Scale.Z {type=MovieSceneScriptingFloatChannel, default=0, const [0,22]}
       calf_r_fk_ctrl.Location.X {type=MovieSceneScriptingFloatChannel, default=0, thinned_from=20, keys [[1790,0],[4366,0]]}
       calf_r_fk_ctrl.Location.Y {type=MovieSceneScriptingFloatChannel, default=0, thinned_from=20, keys [[1790,0],[4366,0]]}
       calf_r_fk_ctrl.Location.Z {type=MovieSceneScriptingFloatChannel, default=0, thinned_from=20, keys [[1790,0],[4366,0]]}
       calf_r_fk_ctrl.Rotation.X {
        type=MovieSceneScriptingFloatChannel
        default=0
        thinned_from=27
        keys [[1790,0.4824],[2113,0.505],[2202,0.4824],[2241,1.1748],[2265,1.2005],[2334,0.3141],[2529,0.3332],[2556,0.3373],[2754,0.3918],[2944,0.4612],[2957,0.1914],[2979,0.4612],[3199,0.5586],[3244,0.0199],[3306,0.3316],[3448,0.4872],[3560,0.4872],[3583,1.1396],[3605,0.4872],[3940,0.4872],[4014,0.6986],[4104,0.6918],[4364,0.6986],[4366,0.4872]]
       }
       calf_r_fk_ctrl.Rotation.Y {
        type=MovieSceneScriptingFloatChannel
        default=0
        thinned_from=27
        keys [[1790,1.1483],[2113,1.1437],[2202,1.1483],[2241,0.6493],[2265,0.5143],[2334,1.1678],[2556,1.1618],[2754,1.1478],[2944,1.1299],[2957,1.2487],[2979,1.1299],[3199,1.1048],[3244,1.1586],[3306,1.1632],[3448,1.126],[3560,1.126],[3583,0.7508],[3605,1.126],[3940,1.126],[4014,1.0628],[4364,1.0628],[4366,1.126]]
       }
       calf_r_fk_ctrl.Rotation.Z {
        type=MovieSceneScriptingFloatChannel
        default=0
        thinned_from=34
        keys [[1790,-5.0061],[2113,-1.0552],[2202,-5.0061],[2241,17.8639],[2265,24.1272],[2334,-22.7084],[2529,-21.7554],[2556,-21.5509],[2754,-18.831],[2944,-15.3709],[2957,-28.8692],[2967,-33.1953],[2998,-21.8222],[3031,-8.7091],[3090,-3.1911],[3199,-10.5119],[3244,-37.8158],[3306,-22.6662],[3448,-14.158],[3560,-14.158],[3577,81.8887],[3581,84.5624],[3599,8.053],[3605,-14.158],[3873,-14.158],[3947,-3.4164],[4004,5.4976],[4045,6.4216],[4225,2.0663],[4364,-0.3725],[4366,-14.158]]
       }
       calf_r_fk_ctrl.Scale.X {type=MovieSceneScriptingFloatChannel, default=0, const [0,20]}
       calf_r_fk_ctrl.Scale.Y {type=MovieSceneScriptingFloatChannel, default=0, const [0,20]}
       calf_r_fk_ctrl.Scale.Z {type=MovieSceneScriptingFloatChannel, default=0, const [0,20]}
       foot_r_fk_ctrl.Location.X {type=MovieSceneScriptingFloatChannel, default=0, thinned_from=16, keys [[1790,0],[4366,0]]}
       foot_r_fk_ctrl.Location.Y {type=MovieSceneScriptingFloatChannel, default=0, thinned_from=16, keys [[1790,-0.0329],[4366,-0.0329]]}
       foot_r_fk_ctrl.Location.Z {type=MovieSceneScriptingFloatChannel, default=0, thinned_from=16, keys [[1790,-0.9075],[4366,-0.9075]]}
       foot_r_fk_ctrl.Rotation.X {
        type=MovieSceneScriptingFloatChannel
        default=0
        thinned_from=19
        keys [[1790,5.6486],[2113,7.2109],[2217,5.6486],[2523,4.1914],[2529,4.1745],[2556,4.0563],[2754,3.3037],[2803,9.1468],[2964,9.1468],[3199,7.5829],[3305,6.5198],[4012,6.5198],[4364,-2.2916],[4366,6.9537]]
       }
       foot_r_fk_ctrl.Rotation.Y {
        type=MovieSceneScriptingFloatChannel
        default=0
        thinned_from=20
        keys [[1790,-6.6142],[2113,-7.1766],[2217,-6.6142],[2523,-5.1204],[2529,-5.1022],[2556,-4.9733],[2754,-4.1137],[2803,-3.3451],[2964,-3.3451],[3062,-3.1894],[3098,16.0748],[3305,2.128],[4364,2.128],[4366,1.4046]]
       }
       foot_r_fk_ctrl.Rotation.Z {
        type=MovieSceneScriptingFloatChannel
        default=0
        thinned_from=21
        keys [[1790,-1.1257],[2113,-2.0454],[2217,9.6396],[2529,8.7636],[2556,8.6496],[2754,7.2939],[2803,21.3943],[2964,21.3943],[3092,5.4974],[3132,5.3518],[3220,17.2145],[3305,18.619],[4364,18.619],[4366,33.934]]
       }
       foot_r_fk_ctrl.Scale.X {type=MovieSceneScriptingFloatChannel, default=0, const [0,16]}
       foot_r_fk_ctrl.Scale.Y {type=MovieSceneScriptingFloatChannel, default=0, const [0,16]}
       foot_r_fk_ctrl.Scale.Z {type=MovieSceneScriptingFloatChannel, default=0, const [0,16]}
       ball_r_fk_ctrl.Location.X {type=MovieSceneScriptingFloatChannel, default=0, const [0,16]}
       ball_r_fk_ctrl.Location.Y {type=MovieSceneScriptingFloatChannel, default=0, const [0,16]}
       ball_r_fk_ctrl.Location.Z {type=MovieSceneScriptingFloatChannel, default=0, thinned_from=16, keys [[1790,0],[4366,0]]}
       ball_r_fk_ctrl.Rotation.X {type=MovieSceneScriptingFloatChannel, default=0, const [0,16]}
       ball_r_fk_ctrl.Rotation.Y {type=MovieSceneScriptingFloatChannel, default=0, const [0,16]}
       ball_r_fk_ctrl.Rotation.Z {type=MovieSceneScriptingFloatChannel, default=0, const [0,16]}
       ball_r_fk_ctrl.Scale.X {type=MovieSceneScriptingFloatChannel, default=0, const [0,16]}
       ball_r_fk_ctrl.Scale.Y {type=MovieSceneScriptingFloatChannel, default=0, const [0,16]}
       ball_r_fk_ctrl.Scale.Z {type=MovieSceneScriptingFloatChannel, default=0, const [0,16]}
       foot_r_ik_ctrl.Location.X {
        type=MovieSceneScriptingFloatChannel
        default=0
        thinned_from=20
        keys [[1790,-11.4243],[2113,-11.4243],[2217,-4.5503],[2344,16.3437],[2364,17.2094],[2386,17.1997],[2516,29.4215],[2523,14.9742],[2529,14.6802],[2556,15.3709],[2754,15.6941],[2944,28.0384],[2964,28.0384],[3199,62.1019],[3509,-16.8598],[3841,0],[4067,-210.8695],[4364,291.7783],[4366,-463.1559]]
       }
       foot_r_ik_ctrl.Location.Y {
        type=MovieSceneScriptingFloatChannel
        default=0
        thinned_from=20
        keys [[1790,-19.3647],[2113,-19.3647],[2217,-18.8881],[2344,12.2814],[2364,13.922],[2386,13.9142],[2516,-30.9985],[2523,11.0284],[2529,16.4559],[2556,55.7374],[2754,171.2795],[2944,97.6458],[2964,97.6458],[3199,69.4932],[3509,-226.3427],[3841,0],[4067,611.9754],[4364,319.351],[4366,312.6044]]
       }
       foot_r_ik_ctrl.Location.Z {
        type=MovieSceneScriptingFloatChannel
        default=0
        keys [[1790,4.1115],[2113,4.1115],[2217,4.6674],[2305,0.3994],[2323,0.8576],[2344,2.0385],[2374,4.1829],[2409,3.7757],[2516,-9.1328],[2523,7.0268],[2529,9.3833],[2556,6.8458],[2754,5.4645],[2944,-5.1436],[2964,-5.1436],[3199,125.2001],[3509,-88.0703],[3841,0],[4067,349.1529],[4364,140.4113],[4366,1034.6997]]
       }
       foot_r_ik_ctrl.Rotation.X {
        type=MovieSceneScriptingFloatChannel
        default=0
        keys [[1790,9.2443],[2113,9.2443],[2217,19.6458],[2344,13.1868],[2364,12.5783],[2374,12.4409],[2386,12.3312],[2516,-32.4083],[2523,-8.5277],[2529,-8.9008],[2556,-8.3873],[2754,-3.5745],[2944,-17.4665],[2964,-17.4665],[3199,-57.5565],[3509,-36.3022],[3841,0],[4067,19.6433],[4364,-41.9162],[4366,-96.1571]]
       }
       foot_r_ik_ctrl.Rotation.Y {
        type=MovieSceneScriptingFloatChannel
        default=0
        thinned_from=19
        keys [[1790,-7.7313],[2217,-7.7313],[2344,-7.603],[2386,-7.5292],[2516,-14.221],[2523,-6.8346],[2529,-7.1689],[2556,-6.6091],[2754,-2.3169],[2944,-3.0606],[2964,-3.0606],[3199,-8.3606],[3509,-1.1379],[3841,0],[4067,-26.1792],[4364,-0.083],[4366,26.3244]]
       }
       foot_r_ik_ctrl.Rotation.Z {
        type=MovieSceneScriptingFloatChannel
        default=0
        thinned_from=19
        keys [[1790,12.1225],[2217,12.1225],[2344,11.9213],[2386,11.8056],[2516,16.2707],[2523,7.8143],[2529,7.0212],[2556,8.561],[2754,8.6733],[2944,7.4947],[2964,7.4947],[3199,21.9243],[3509,29.2636],[3841,0],[4067,7.581],[4364,31.4162],[4366,151.5119]]
       }
       foot_r_ik_ctrl.Scale.X {
        type=MovieSceneScriptingFloatChannel
        default=0
        thinned_from=19
        keys [[1790,0],[2529,0],[2556,0.0165],[2754,-0.0285],[2944,0.0146],[2964,0.0146],[3199,0.0095],[3509,0.0017],[3841,0],[4067,0.0041],[4364,0.0047],[4366,0.0056]]
       }
       foot_r_ik_ctrl.Scale.Y {
        type=MovieSceneScriptingFloatChannel
        default=0
        thinned_from=19
        keys [[1790,0],[2529,0],[2556,-0.0165],[2754,0.0285],[2944,-0.0146],[2964,-0.0146],[3199,-0.0095],[3509,-0.0017],[3841,0],[4067,-0.0041],[4364,-0.0047],[4366,-0.0056]]
       }
       foot_r_ik_ctrl.Scale.Z {
        type=MovieSceneScriptingFloatChannel
        default=0
        thinned_from=19
        keys [[1790,0],[2529,0],[2556,-0.0165],[2754,0.0285],[2944,-0.0146],[2964,-0.0146],[3199,-0.0095],[3509,-0.0017],[3841,0],[4067,-0.0041],[4364,-0.0047],[4366,-0.0056]]
       }
       ball_r_ik_ctrl.Location.X {type=MovieSceneScriptingFloatChannel, default=0, const [0,16]}
       ball_r_ik_ctrl.Location.Y {type=MovieSceneScriptingFloatChannel, default=0, const [0,16]}
       ball_r_ik_ctrl.Location.Z {type=MovieSceneScriptingFloatChannel, default=0, const [0,16]}
       ball_r_ik_ctrl.Rotation.X {type=MovieSceneScriptingFloatChannel, default=0, const [0,16]}
       ball_r_ik_ctrl.Rotation.Y {type=MovieSceneScriptingFloatChannel, default=0, const [0,16]}
       ball_r_ik_ctrl.Rotation.Z {type=MovieSceneScriptingFloatChannel, default=0, const [0,16]}
       ball_r_ik_ctrl.Scale.X {type=MovieSceneScriptingFloatChannel, default=0, const [0,16]}
       ball_r_ik_ctrl.Scale.Y {type=MovieSceneScriptingFloatChannel, default=0, const [0,16]}
       ball_r_ik_ctrl.Scale.Z {type=MovieSceneScriptingFloatChannel, default=0, const [0,16]}
       foot_bk1_r_ctrl.Location.X {type=MovieSceneScriptingFloatChannel, default=0, const [0,16]}
       foot_bk1_r_ctrl.Location.Y {type=MovieSceneScriptingFloatChannel, default=0, const [0,16]}
       foot_bk1_r_ctrl.Location.Z {type=MovieSceneScriptingFloatChannel, default=0, const [0,16]}
       foot_bk1_r_ctrl.Rotation.X {type=MovieSceneScriptingFloatChannel, default=0, const [0,16]}
       foot_bk1_r_ctrl.Rotation.Y {type=MovieSceneScriptingFloatChannel, default=0, const [0,16]}
       foot_bk1_r_ctrl.Rotation.Z {type=MovieSceneScriptingFloatChannel, default=0, const [0,16]}
       foot_bk1_r_ctrl.Scale.X {type=MovieSceneScriptingFloatChannel, default=0, const [0,16]}
       foot_bk1_r_ctrl.Scale.Y {type=MovieSceneScriptingFloatChannel, default=0, const [0,16]}
       foot_bk1_r_ctrl.Scale.Z {type=MovieSceneScriptingFloatChannel, default=0, const [0,16]}
       foot_roll_r_ctrl.Location.X {type=MovieSceneScriptingFloatChannel, default=0, const [0,16]}
       foot_roll_r_ctrl.Location.Y {type=MovieSceneScriptingFloatChannel, default=0, const [0,16]}
       foot_roll_r_ctrl.Location.Z {type=MovieSceneScriptingFloatChannel, default=0, const [0,16]}
       foot_roll_r_ctrl.Rotation.X {type=MovieSceneScriptingFloatChannel, default=0, const [0,16]}
       foot_roll_r_ctrl.Rotation.Y {type=MovieSceneScriptingFloatChannel, default=0, const [0,16]}
       foot_roll_r_ctrl.Rotation.Z {type=MovieSceneScriptingFloatChannel, default=0, const [0,16]}
       foot_roll_r_ctrl.Scale.X {type=MovieSceneScriptingFloatChannel, default=0, const [0,16]}
       foot_roll_r_ctrl.Scale.Y {type=MovieSceneScriptingFloatChannel, default=0, const [0,16]}
       foot_roll_r_ctrl.Scale.Z {type=MovieSceneScriptingFloatChannel, default=0, const [0,16]}
       tip_r_ctrl.Location.X {type=MovieSceneScriptingFloatChannel, default=0, const [0,16]}
       tip_r_ctrl.Location.Y {type=MovieSceneScriptingFloatChannel, default=0, const [0,16]}
       tip_r_ctrl.Location.Z {type=MovieSceneScriptingFloatChannel, default=0, const [0,16]}
       tip_r_ctrl.Rotation.X {type=MovieSceneScriptingFloatChannel, default=0, const [0,16]}
       tip_r_ctrl.Rotation.Y {type=MovieSceneScriptingFloatChannel, default=0, const [0,16]}
       tip_r_ctrl.Rotation.Z {type=MovieSceneScriptingFloatChannel, default=0, const [0,16]}
       tip_r_ctrl.Scale.X {type=MovieSceneScriptingFloatChannel, default=0, const [0,16]}
       tip_r_ctrl.Scale.Y {type=MovieSceneScriptingFloatChannel, default=0, const [0,16]}
       tip_r_ctrl.Scale.Z {type=MovieSceneScriptingFloatChannel, default=0, const [0,16]}
       heel_r_ctrl.Location.X {type=MovieSceneScriptingFloatChannel, default=0, const [0,16]}
       heel_r_ctrl.Location.Y {type=MovieSceneScriptingFloatChannel, default=0, const [0,16]}
       heel_r_ctrl.Location.Z {type=MovieSceneScriptingFloatChannel, default=0, const [0,16]}
       heel_r_ctrl.Rotation.X {type=MovieSceneScriptingFloatChannel, default=0, const [0,16]}
       heel_r_ctrl.Rotation.Y {type=MovieSceneScriptingFloatChannel, default=0, const [0,16]}
       heel_r_ctrl.Rotation.Z {type=MovieSceneScriptingFloatChannel, default=0, const [0,16]}
       heel_r_ctrl.Scale.X {type=MovieSceneScriptingFloatChannel, default=0, const [0,16]}
       heel_r_ctrl.Scale.Y {type=MovieSceneScriptingFloatChannel, default=0, const [0,16]}
       heel_r_ctrl.Scale.Z {type=MovieSceneScriptingFloatChannel, default=0, const [0,16]}
       leg_r_softness {type=MovieSceneScriptingFloatChannel, default=0, const [0,19]}
       leg_r_pv_ik_ctrl.Location.X {
        type=MovieSceneScriptingFloatChannel
        default=0
        thinned_from=18
        keys [[1790,-0.8591],[2217,-0.8591],[2298,-5.3347],[2348,-3.4615],[2516,-93.3185],[2523,-51.952],[2529,-53.7994],[2556,-54.9946],[2754,-70.7253],[2944,-68.9032],[2964,-68.9032],[3199,-374.9974],[3509,-86.3042],[3841,0],[4067,-390.164],[4364,629.9579],[4366,-372.2671]]
       }
       leg_r_pv_ik_ctrl.Location.Y {
        type=MovieSceneScriptingFloatChannel
        default=0
        thinned_from=18
        keys [[1790,0.4116],[2217,0.4116],[2298,-18.3947],[2348,-19.2391],[2516,-5.0245],[2523,21.5594],[2529,30.3221],[2556,65.5423],[2754,183.1402],[2944,122.0178],[2964,122.0178],[3199,29.8071],[3509,-185.6304],[3841,0],[4067,509.0811],[4364,-188.5059],[4366,-103.6672]]
       }
       leg_r_pv_ik_ctrl.Location.Z {
        type=MovieSceneScriptingFloatChannel
        default=0
        thinned_from=18
        keys [[1790,9.2793],[2217,9.2793],[2298,0.4316],[2348,-19.6637],[2516,-34.8514],[2523,-25.6423],[2529,-23.9101],[2556,-31.6023],[2754,-33.1525],[2944,-42.4542],[2964,-42.4542],[3199,-69.5717],[3509,42.671],[3841,0],[4067,128.8178],[4364,-1093.2546],[4366,-244.4672]]
       }
       leg_r_pv_ik_ctrl.Rotation.X {
        type=MovieSceneScriptingFloatChannel
        default=0
        thinned_from=16
        keys [[1790,0],[2217,0],[2516,-31.4638],[2523,-0.5494],[2529,-2.8758],[2556,-4.0846],[2754,3.678],[2944,-10.2277],[2964,-10.2277],[3199,-63.1505],[3509,-47.1797],[3841,0],[4067,15.8638],[4364,-16.0117],[4366,-45.5052]]
       }
       leg_r_pv_ik_ctrl.Rotation.Y {
        type=MovieSceneScriptingFloatChannel
        default=0
        thinned_from=16
        keys [[1790,0],[2217,0],[2516,12.052],[2523,5.9615],[2529,5.5009],[2556,5.1821],[2754,7.3878],[2944,4.9225],[2964,4.9225],[3199,7.0914],[3509,20.7137],[3841,0],[4067,-2.7293],[4364,-27.5187],[4366,69.8289]]
       }
       leg_r_pv_ik_ctrl.Rotation.Z {
        type=MovieSceneScriptingFloatChannel
        default=0
        thinned_from=16
        keys [[1790,0],[2217,0],[2516,-9.0848],[2523,11.3523],[2529,13.0372],[2556,10.9985],[2754,0.6657],[2944,4.3176],[2964,4.3176],[3199,14.7492],[3509,24.3655],[3841,0],[4067,8.2602],[4364,59.3598],[4366,18.5715]]
       }
       leg_r_pv_ik_ctrl.Scale.X {
        type=MovieSceneScriptingFloatChannel
        default=0
        thinned_from=16
        keys [[1790,0],[2529,0],[2556,0.0855],[2754,-0.0803],[2944,0.0757],[2964,0.0757],[3199,0.0474],[3509,0.0086],[3841,0],[4067,0.0203],[4366,0.0102]]
       }
       leg_r_pv_ik_ctrl.Scale.Y {
        type=MovieSceneScriptingFloatChannel
        default=0
        thinned_from=16
        keys [[1790,0],[2529,0],[2556,0.0855],[2754,-0.0803],[2944,0.0757],[2964,0.0757],[3199,0.0474],[3509,0.0086],[3841,0],[4067,0.0203],[4366,0.0102]]
       }
       leg_r_pv_ik_ctrl.Scale.Z {
        type=MovieSceneScriptingFloatChannel
        default=0
        thinned_from=16
        keys [[1790,0],[2529,0],[2556,0.0855],[2754,-0.0803],[2944,0.0757],[2964,0.0757],[3199,0.0474],[3509,0.0086],[3841,0],[4067,0.0203],[4366,0.0102]]
       }
       upperarm_l_fk_ctrl.Location.X {type=MovieSceneScriptingFloatChannel, default=0, thinned_from=25, keys [[1790,0],[4366,0]]}
       upperarm_l_fk_ctrl.Location.Y {type=MovieSceneScriptingFloatChannel, default=0, thinned_from=25, keys [[1790,0],[4366,0]]}
       upperarm_l_fk_ctrl.Location.Z {type=MovieSceneScriptingFloatChannel, default=0, thinned_from=25, keys [[1790,0],[4366,0]]}
       upperarm_l_fk_ctrl.Rotation.X {
        type=MovieSceneScriptingFloatChannel
        default=0
        thinned_from=36
        keys [[1790,0],[2113,0],[2191,0.4075],[2213,18.1605],[2236,14.5289],[2253,7.4782],[2274,1.1151],[2291,1.4455],[2312,2.1531],[2330,2.6675],[2385,2.6587],[2523,2.5696],[2532,2.9355],[2600,2.5396],[2651,2.5396],[2754,2.4954],[2944,2.2442],[2964,2.2442],[3051,4.6323],[3199,1.6108],[3568,0.3279],[3584,15.849],[3586,13.2953],[3594,-31.0931],[3599,-39.0074],[3617,-47.3526],[3672,-55.485],[3825,-54.6938],[3891,-54.9602],[4067,-50.3222],[4354,-49.4605],[4364,-50.139],[4365,28.2929],[4366,107.5421],[4372,83.3821]]
       }
       upperarm_l_fk_ctrl.Rotation.Y {
        type=MovieSceneScriptingFloatChannel
        default=0
        thinned_from=37
        keys [[1790,0],[2113,0],[2191,0.824],[2213,6.9951],[2236,-8.7055],[2253,-14.689],[2274,-16.0964],[2291,2.2657],[2312,6.869],[2330,2.7444],[2385,2.7353],[2523,2.6436],[2532,-7.7491],[2600,2.6128],[2651,2.6128],[2754,2.5673],[2944,2.3089],[2964,2.3089],[3051,-8.6586],[3199,1.6572],[3568,0.3373],[3578,14.6152],[3582,16.6747],[3586,16.0524],[3600,1.6973],[3606,0.3595],[3705,73.8779],[3834,113.4897],[3909,127.9222],[4045,90.3778],[4119,88.1475],[4354,107.8811],[4364,105.5182],[4365,30.387],[4366,20.5267],[4372,10.5866]]
       }
       upperarm_l_fk_ctrl.Rotation.Z {
        type=MovieSceneScriptingFloatChannel
        default=0
        thinned_from=36
        keys [[1790,0],[2113,0],[2191,6.4337],[2213,3.9002],[2236,10.7667],[2253,19.3328],[2274,19.7056],[2291,21.0834],[2312,18.2801],[2330,16.3818],[2385,15.6818],[2523,14.1738],[2532,2.2011],[2600,13.8342],[2651,13.8342],[2754,13.5936],[2944,12.2252],[2964,12.2252],[3051,15.0072],[3199,8.7749],[3568,1.7862],[3581,23.5822],[3587,39.6394],[3592,42.0854],[3604,40.1227],[3673,27.4834],[3754,26.8464],[3777,26.5563],[3834,26.0698],[3915,24.7885],[4067,25.6812],[4364,26.4833],[4365,137.211],[4366,163.0081],[4372,164.5833]]
       }
       upperarm_l_fk_ctrl.Scale.X {type=MovieSceneScriptingFloatChannel, default=0, const [0,25]}
       upperarm_l_fk_ctrl.Scale.Y {type=MovieSceneScriptingFloatChannel, default=0, const [0,25]}
       upperarm_l_fk_ctrl.Scale.Z {type=MovieSceneScriptingFloatChannel, default=0, const [0,25]}
       lowerarm_l_fk_ctrl.Location.X {type=MovieSceneScriptingFloatChannel, default=0, thinned_from=22, keys [[1790,0],[4484,0]]}
       lowerarm_l_fk_ctrl.Location.Y {type=MovieSceneScriptingFloatChannel, default=0, thinned_from=22, keys [[1790,0],[4484,0]]}
       lowerarm_l_fk_ctrl.Location.Z {type=MovieSceneScriptingFloatChannel, default=0, thinned_from=22, keys [[1790,0],[4484,0]]}
       lowerarm_l_fk_ctrl.Rotation.X {
        type=MovieSceneScriptingFloatChannel
        default=0
        thinned_from=27
        keys [[1790,0],[2191,0],[2216,0.1333],[2234,0],[2253,0.0211],[2516,0],[3565,0],[3686,0.0163],[3754,-0.0451],[3834,-0.0451],[3995,-0.0155],[4010,-0.0528],[4039,-0.018],[4364,-0.0451],[4366,-0.0473],[4368,-0.0535],[4409,-0.0451],[4484,0]]
       }
       lowerarm_l_fk_ctrl.Rotation.Y {
        type=MovieSceneScriptingFloatChannel
        default=0
        thinned_from=27
        keys [[1790,0],[2191,0],[2216,-0.0686],[2234,0],[2253,-0.0072],[2516,0],[3565,0],[3686,0.0664],[3754,0.0288],[3834,0.0288],[3995,0.019],[4010,0.0352],[4039,0.0199],[4364,0.0288],[4366,0.0216],[4368,0.0009],[4409,0.0288],[4484,0]]
       }
       lowerarm_l_fk_ctrl.Rotation.Z {
        type=MovieSceneScriptingFloatChannel
        default=0
        thinned_from=39
        keys [[1790,0],[2191,0],[2216,34.1306],[2234,0],[2253,-10.7127],[2516,0],[3565,0],[3574,62.8989],[3576,67.631],[3582,71.6347],[3587,72.3829],[3592,18.1672],[3595,2.5544],[3599,0.5046],[3628,16.865],[3655,15.2271],[3671,10.2217],[3689,7.8768],[3698,5.4542],[3744,-3.4386],[3770,-3.422],[3819,6.5986],[3858,33.9746],[3900,35.5127],[3957,32.8049],[4059,5.1425],[4292,-0.6987],[4366,-8.911],[4368,-8.8229],[4409,22.69],[4484,0]]
       }
       lowerarm_l_fk_ctrl.Scale.X {type=MovieSceneScriptingFloatChannel, default=0, const [0,22]}
       lowerarm_l_fk_ctrl.Scale.Y {type=MovieSceneScriptingFloatChannel, default=0, const [0,22]}
       lowerarm_l_fk_ctrl.Scale.Z {type=MovieSceneScriptingFloatChannel, default=0, const [0,22]}
       hand_l_fk_ctrl.Location.X {type=MovieSceneScriptingFloatChannel, default=0, thinned_from=20, keys [[1790,0],[4484,0]]}
       hand_l_fk_ctrl.Location.Y {type=MovieSceneScriptingFloatChannel, default=0, thinned_from=20, keys [[1790,0],[4484,0]]}
       hand_l_fk_ctrl.Location.Z {type=MovieSceneScriptingFloatChannel, default=0, thinned_from=20, keys [[1790,0],[4484,0]]}
       hand_l_fk_ctrl.Rotation.X {
        type=MovieSceneScriptingFloatChannel
        default=0
        thinned_from=25
        keys [[1790,0],[3568,0],[3584,-106.8474],[3593,-111.5498],[3610,-145.0322],[3645,-150.4313],[3758,-111.9268],[3799,-93.8042],[3999,-51.0892],[4055,-49.7336],[4084,-48.7836],[4364,-53.7856],[4366,-1.8141],[4409,-82.3245],[4484,0]]
       }
       hand_l_fk_ctrl.Rotation.Y {
        type=MovieSceneScriptingFloatChannel
        default=0
        thinned_from=29
        keys [[1790,0],[3568,0],[3587,0.6305],[3595,-5.0923],[3612,-52.3923],[3648,-52.3277],[3750,-55.8794],[3781,-55.1993],[3813,-36.6689],[3858,-22.9104],[3998,-51.8705],[4027,-46.8272],[4064,-44.2875],[4098,-45.3034],[4188,-36.1983],[4364,-28.5663],[4366,-16.6091],[4409,-28.5663],[4484,0]]
       }
       hand_l_fk_ctrl.Rotation.Z {
        type=MovieSceneScriptingFloatChannel
        default=0
        thinned_from=27
        keys [[1790,0],[3568,0],[3605,3.365],[3693,18.041],[3752,17.9584],[3777,10.3109],[3801,-13.8081],[3823,-23.9308],[3889,-12.1993],[3923,11.8548],[4010,-16.1154],[4051,2.987],[4081,9.2183],[4364,-4.1752],[4366,13.2803],[4409,-4.1752],[4484,0]]
       }
       hand_l_fk_ctrl.Scale.X {type=MovieSceneScriptingFloatChannel, default=0, const [0,20]}
       hand_l_fk_ctrl.Scale.Y {type=MovieSceneScriptingFloatChannel, default=0, const [0,20]}
       hand_l_fk_ctrl.Scale.Z {type=MovieSceneScriptingFloatChannel, default=0, const [0,20]}
       thumb_01_l_ctrl.Location.X {type=MovieSceneScriptingFloatChannel, default=0, thinned_from=18, keys [[1790,0],[4172,0]]}
       thumb_01_l_ctrl.Location.Y {type=MovieSceneScriptingFloatChannel, default=0, thinned_from=18, keys [[1790,0],[4172,0]]}
       thumb_01_l_ctrl.Location.Z {type=MovieSceneScriptingFloatChannel, default=0, thinned_from=18, keys [[1790,0],[4172,0]]}
       thumb_01_l_ctrl.Rotation.X {
        type=MovieSceneScriptingFloatChannel
        default=0
        thinned_from=22
        keys [[1790,1.0421],[2217,1.0421],[2529,0.9532],[2556,0.9416],[2754,0.804],[2803,-5.5182],[2964,-5.5182],[3199,-4.662],[3509,-1.0307],[3571,-0.32],[3587,-0.2906],[3601,-0.2158],[3632,0.2959],[3673,0.3581],[3738,0.3906],[4170,0],[4172,0]]
       }
       thumb_01_l_ctrl.Rotation.Y {
        type=MovieSceneScriptingFloatChannel
        default=0
        thinned_from=22
        keys [[1790,4.3973],[2217,4.3973],[2529,4.0221],[2556,3.9732],[2754,3.3925],[2803,12.8017],[2964,12.8017],[3199,10.8155],[3509,2.3911],[3571,0.7423],[3587,-2.0013],[3601,0.5007],[3632,1.754],[3673,2.39],[3738,2.7389],[4170,0],[4172,0]]
       }
       thumb_01_l_ctrl.Rotation.Z {
        type=MovieSceneScriptingFloatChannel
        default=0
        thinned_from=22
        keys [[1790,18.9553],[2217,18.9553],[2529,17.3377],[2556,17.1272],[2754,14.6236],[2803,12.4704],[2964,12.4704],[3199,10.5356],[3509,2.3292],[3571,0.7231],[3587,-9.2157],[3601,0.4877],[3632,7.4468],[3673,10.1473],[3738,11.6287],[4170,0],[4172,0]]
       }
       thumb_01_l_ctrl.Scale.X {
        type=MovieSceneScriptingFloatChannel
        default=0
        thinned_from=19
        keys [[1790,0],[3601,0],[3632,0.0406],[3673,0.0553],[3738,0.0634],[4170,0],[4172,0]]
       }
       thumb_01_l_ctrl.Scale.Y {
        type=MovieSceneScriptingFloatChannel
        default=0
        thinned_from=19
        keys [[1790,0],[3601,0],[3632,0.0406],[3673,0.0553],[3738,0.0634],[4170,0],[4172,0]]
       }
       thumb_01_l_ctrl.Scale.Z {
        type=MovieSceneScriptingFloatChannel
        default=0
        thinned_from=19
        keys [[1790,0],[3601,0],[3632,0.0406],[3673,0.0553],[3738,0.0634],[4170,0],[4172,0]]
       }
       thumb_02_l_ctrl.Location.X {type=MovieSceneScriptingFloatChannel, default=0, thinned_from=16, keys [[1790,0],[3673,0]]}
       thumb_02_l_ctrl.Location.Y {type=MovieSceneScriptingFloatChannel, default=0, thinned_from=16, keys [[1790,0],[3673,0]]}
       thumb_02_l_ctrl.Location.Z {type=MovieSceneScriptingFloatChannel, default=0, thinned_from=16, keys [[1790,0],[3673,0]]}
       thumb_02_l_ctrl.Rotation.X {
        type=MovieSceneScriptingFloatChannel
        default=0
        thinned_from=20
        keys [[1790,-1.2945],[2217,-1.2945],[2529,-1.184],[2556,-1.1697],[2754,-0.9987],[2803,-1.0723],[2964,-1.0723],[3199,-0.9059],[3509,-0.2003],[3571,-0.0622],[3587,-0.0439],[3601,-0.0419],[3632,0.6597],[3673,0.8825],[3738,1.0037]]
       }
       thumb_02_l_ctrl.Rotation.Y {
        type=MovieSceneScriptingFloatChannel
        default=0
        thinned_from=20
        keys [[1790,-0.8473],[2217,-0.8473],[2529,-0.775],[2556,-0.7656],[2754,-0.6537],[2803,-2.0135],[2964,-2.0135],[3199,-1.7011],[3509,-0.3761],[3571,-0.1168],[3587,-0.0905],[3601,-0.0787],[3632,0.2247],[3673,0.2779],[3738,0.3059]]
       }
       thumb_02_l_ctrl.Rotation.Z {
        type=MovieSceneScriptingFloatChannel
        default=0
        thinned_from=20
        keys [[1790,-20.3796],[2217,-20.3796],[2529,-18.6404],[2556,-18.4141],[2754,-15.7224],[2803,-61.0334],[2964,-61.0334],[3199,-51.5639],[3509,-11.3996],[3571,-3.5389],[3587,-2.7389],[3601,-2.387],[3632,8.6336],[3673,10.8321],[3738,11.977]]
       }
       thumb_02_l_ctrl.Scale.X {type=MovieSceneScriptingFloatChannel, default=0, thinned_from=17, keys [[1790,0],[3601,0],[3632,0.0385],[3673,0.0524],[3738,0.0601]]}
       thumb_02_l_ctrl.Scale.Y {type=MovieSceneScriptingFloatChannel, default=0, thinned_from=17, keys [[1790,0],[3601,0],[3632,0.0385],[3673,0.0524],[3738,0.0601]]}
       thumb_02_l_ctrl.Scale.Z {type=MovieSceneScriptingFloatChannel, default=0, thinned_from=17, keys [[1790,0],[3601,0],[3632,0.0385],[3673,0.0524],[3738,0.0601]]}
       thumb_03_l_ctrl.Location.X {type=MovieSceneScriptingFloatChannel, default=0, thinned_from=15, keys [[1790,0],[3673,0]]}
       thumb_03_l_ctrl.Location.Y {type=MovieSceneScriptingFloatChannel, default=0, thinned_from=15, keys [[1790,0],[3673,0]]}
       thumb_03_l_ctrl.Location.Z {type=MovieSceneScriptingFloatChannel, default=0, thinned_from=15, keys [[1790,0],[3673,0]]}
       thumb_03_l_ctrl.Rotation.X {
        type=MovieSceneScriptingFloatChannel
        default=0
        thinned_from=17
        keys [[1790,0],[3571,0],[3587,0.389],[3601,0],[3673,-0.2467],[3738,-0.2827]]
       }
       thumb_03_l_ctrl.Rotation.Y {
        type=MovieSceneScriptingFloatChannel
        default=0
        thinned_from=17
        keys [[1790,0],[3571,0],[3587,0.328],[3601,0],[3673,-0.1333],[3738,-0.1528]]
       }
       thumb_03_l_ctrl.Rotation.Z {
        type=MovieSceneScriptingFloatChannel
        default=0
        thinned_from=17
        keys [[1790,0],[3571,0],[3587,-22.0423],[3601,0],[3673,10.4201],[3738,11.9413]]
       }
       thumb_03_l_ctrl.Scale.X {type=MovieSceneScriptingFloatChannel, default=0, thinned_from=16, keys [[1790,0],[3601,0],[3673,0.0499],[3738,0.0571]]}
       thumb_03_l_ctrl.Scale.Y {type=MovieSceneScriptingFloatChannel, default=0, thinned_from=16, keys [[1790,0],[3601,0],[3673,0.0499],[3738,0.0571]]}
       thumb_03_l_ctrl.Scale.Z {type=MovieSceneScriptingFloatChannel, default=0, thinned_from=16, keys [[1790,0],[3601,0],[3673,0.0499],[3738,0.0571]]}
       index_metacarpal_l_ctrl.Location.X {type=MovieSceneScriptingFloatChannel, default=0, thinned_from=18, keys [[1790,0],[4366,0]]}
       index_metacarpal_l_ctrl.Location.Y {type=MovieSceneScriptingFloatChannel, default=0, thinned_from=18, keys [[1790,0],[4366,0]]}
       index_metacarpal_l_ctrl.Location.Z {type=MovieSceneScriptingFloatChannel, default=0, thinned_from=18, keys [[1790,0],[4366,0]]}
       index_metacarpal_l_ctrl.Rotation.X {
        type=MovieSceneScriptingFloatChannel
        default=0
        thinned_from=18
        keys [[1790,0.8268],[2217,0.8268],[2529,0.7562],[2556,0.747],[2754,0.6378],[2944,0.5239],[2964,0.5239],[3199,0.367],[3509,0.0739],[3682,0.023],[3712,0.0155],[3841,0],[4366,0]]
       }
       index_metacarpal_l_ctrl.Rotation.Y {
        type=MovieSceneScriptingFloatChannel
        default=0
        thinned_from=18
        keys [[1790,1.0139],[2217,1.0139],[2529,0.9274],[2556,0.9161],[2754,0.7822],[2944,0.6424],[2964,0.6424],[3199,0.4501],[3509,0.0907],[3682,0.0282],[3712,0.019],[3841,0],[4366,0]]
       }
       index_metacarpal_l_ctrl.Rotation.Z {
        type=MovieSceneScriptingFloatChannel
        default=0
        thinned_from=18
        keys [[1790,9.7],[2217,9.7],[2529,8.8722],[2556,8.7645],[2754,7.4834],[2944,6.1464],[2964,6.1464],[3199,4.3062],[3509,0.8676],[3682,0.2693],[3712,0.1817],[3841,0],[4366,0]]
       }
       index_metacarpal_l_ctrl.Scale.X {type=MovieSceneScriptingFloatChannel, default=0, const [0,18]}
       index_metacarpal_l_ctrl.Scale.Y {type=MovieSceneScriptingFloatChannel, default=0, const [0,18]}
       index_metacarpal_l_ctrl.Scale.Z {type=MovieSceneScriptingFloatChannel, default=0, const [0,18]}
       index_01_l_ctrl.Location.X {type=MovieSceneScriptingFloatChannel, default=0, thinned_from=17, keys [[1790,0],[3852,0]]}
       index_01_l_ctrl.Location.Y {type=MovieSceneScriptingFloatChannel, default=0, thinned_from=17, keys [[1790,0],[3852,0]]}
       index_01_l_ctrl.Location.Z {type=MovieSceneScriptingFloatChannel, default=0, thinned_from=17, keys [[1790,0],[3852,0]]}
       index_01_l_ctrl.Rotation.X {
        type=MovieSceneScriptingFloatChannel
        default=0
        thinned_from=22
        keys [[1790,5.5843],[2217,5.5843],[2529,5.1078],[2556,5.0458],[2754,4.3082],[2803,18.7496],[2964,18.7496],[3199,15.8405],[3509,3.502],[3571,1.0872],[3587,-2.2287],[3601,0.7333],[3632,-0.7601],[3673,-0.9073],[3738,-0.9839],[3852,-0.9839],[3938,-1.2444]]
       }
       index_01_l_ctrl.Rotation.Y {
        type=MovieSceneScriptingFloatChannel
        default=0
        thinned_from=22
        keys [[1790,-28.5381],[2217,-28.5381],[2529,-26.1026],[2556,-25.7857],[2754,-22.0165],[2803,-8.3349],[2964,-8.3349],[3199,-7.0417],[3509,-1.5568],[3571,-0.4833],[3587,2.1352],[3601,-0.326],[3632,-11.9081],[3673,-16.2264],[3738,-18.5953],[3852,-18.5953],[3938,-9.248]]
       }
       index_01_l_ctrl.Rotation.Z {
        type=MovieSceneScriptingFloatChannel
        default=0
        thinned_from=22
        keys [[1790,19.015],[2217,19.015],[2529,17.3923],[2556,17.1811],[2754,14.6697],[2803,-38.028],[2964,-38.028],[3199,-32.1279],[3509,-7.1028],[3571,-2.205],[3587,-50.9778],[3601,-1.4872],[3632,11.1036],[3673,14.5493],[3738,16.4014],[3852,16.4014],[3938,-32.1688]]
       }
       index_01_l_ctrl.Scale.X {type=MovieSceneScriptingFloatChannel, default=0, const [0,17]}
       index_01_l_ctrl.Scale.Y {type=MovieSceneScriptingFloatChannel, default=0, const [0,17]}
       index_01_l_ctrl.Scale.Z {type=MovieSceneScriptingFloatChannel, default=0, const [0,17]}
       index_02_l_ctrl.Location.X {type=MovieSceneScriptingFloatChannel, default=0, thinned_from=16, keys [[1790,0],[3673,0]]}
       index_02_l_ctrl.Location.Y {type=MovieSceneScriptingFloatChannel, default=0, thinned_from=16, keys [[1790,0],[3673,0]]}
       index_02_l_ctrl.Location.Z {type=MovieSceneScriptingFloatChannel, default=0, thinned_from=16, keys [[1790,0],[3673,0]]}
       index_02_l_ctrl.Rotation.X {
        type=MovieSceneScriptingFloatChannel
        default=0
        thinned_from=21
        keys [[1790,0],[2754,0],[2803,0.7112],[2964,0.7112],[3199,0.6009],[3509,0.1328],[3571,0.0412],[3587,2.9199],[3601,0.0278],[3632,-0.9029],[3673,-1.2194],[3738,-1.3923],[3938,1.5663]]
       }
       index_02_l_ctrl.Rotation.Y {
        type=MovieSceneScriptingFloatChannel
        default=0
        thinned_from=21
        keys [[1790,0],[2754,0],[2803,0.8586],[2964,0.8586],[3199,0.7254],[3509,0.1604],[3571,0.0498],[3587,6.2438],[3601,0.0336],[3632,-0.619],[3673,-0.8304],[3738,-0.9454],[3938,1.776]]
       }
       index_02_l_ctrl.Rotation.Z {
        type=MovieSceneScriptingFloatChannel
        default=0
        thinned_from=21
        keys [[1790,0],[2754,0],[2803,-23.8809],[2964,-23.8809],[3199,-20.1757],[3509,-4.4604],[3571,-1.3847],[3587,-50.0046],[3601,-0.934],[3632,8.0466],[3673,10.5998],[3738,11.9765],[3938,-14.4619]]
       }
       index_02_l_ctrl.Scale.X {type=MovieSceneScriptingFloatChannel, default=0, const [0,16]}
       index_02_l_ctrl.Scale.Y {type=MovieSceneScriptingFloatChannel, default=0, const [0,16]}
       index_02_l_ctrl.Scale.Z {type=MovieSceneScriptingFloatChannel, default=0, const [0,16]}
       index_03_l_ctrl.Location.X {type=MovieSceneScriptingFloatChannel, default=0, thinned_from=16, keys [[1790,0],[3673,0]]}
       index_03_l_ctrl.Location.Y {type=MovieSceneScriptingFloatChannel, default=0, thinned_from=16, keys [[1790,0],[3673,0]]}
       index_03_l_ctrl.Location.Z {type=MovieSceneScriptingFloatChannel, default=0, thinned_from=16, keys [[1790,0],[3673,0]]}
       index_03_l_ctrl.Rotation.X {
        type=MovieSceneScriptingFloatChannel
        default=0
        thinned_from=21
        keys [[1790,0.2449],[2217,0.2449],[2529,0.224],[2556,0.2213],[2754,0.1889],[2803,0.2116],[2964,0.2116],[3199,0.1787],[3509,0.0395],[3571,0.0123],[3587,0.4068],[3601,0.0083],[3632,-0.1019],[3673,-0.1356],[3738,-0.1539],[3938,0.0774]]
       }
       index_03_l_ctrl.Rotation.Y {
        type=MovieSceneScriptingFloatChannel
        default=0
        thinned_from=21
        keys [[1790,0.1863],[2217,0.1863],[2529,0.1704],[2556,0.1683],[2754,0.1437],[2803,0.3561],[2964,0.3561],[3199,0.3008],[3509,0.0665],[3571,0.0206],[3587,0.5174],[3601,0.0139],[3632,-0.0448],[3673,-0.0557],[3738,-0.0614],[3938,0.0432]]
       }
       index_03_l_ctrl.Rotation.Z {
        type=MovieSceneScriptingFloatChannel
        default=0
        thinned_from=21
        keys [[1790,-22.9041],[2217,-22.9041],[2529,-20.9494],[2556,-20.6951],[2754,-17.67],[2803,-40.7336],[2964,-40.7336],[3199,-34.4136],[3509,-7.6081],[3571,-2.3619],[3587,-51.0712],[3601,-1.593],[3632,8.2892],[3673,10.6729],[3738,11.9398],[3938,-5.6292]]
       }
       index_03_l_ctrl.Scale.X {type=MovieSceneScriptingFloatChannel, default=0, thinned_from=17, keys [[1790,0],[3601,0],[3632,0.0406],[3673,0.0553],[3738,0.0634]]}
       index_03_l_ctrl.Scale.Y {type=MovieSceneScriptingFloatChannel, default=0, thinned_from=17, keys [[1790,0],[3601,0],[3632,0.0406],[3673,0.0553],[3738,0.0634]]}
       index_03_l_ctrl.Scale.Z {type=MovieSceneScriptingFloatChannel, default=0, thinned_from=17, keys [[1790,0],[3601,0],[3632,0.0406],[3673,0.0553],[3738,0.0634]]}
       middle_metacarpal_l_ctrl.Location.X {type=MovieSceneScriptingFloatChannel, default=0, thinned_from=16, keys [[1790,0],[4366,0]]}
       middle_metacarpal_l_ctrl.Location.Y {type=MovieSceneScriptingFloatChannel, default=0, thinned_from=16, keys [[1790,0],[4366,0]]}
       middle_metacarpal_l_ctrl.Location.Z {type=MovieSceneScriptingFloatChannel, default=0, thinned_from=16, keys [[1790,0],[4366,0]]}
       middle_metacarpal_l_ctrl.Rotation.X {
        type=MovieSceneScriptingFloatChannel
        default=0
        thinned_from=16
        keys [[1790,0.2973],[2217,0.2973],[2529,0.272],[2556,0.2687],[2754,0.2294],[2944,0.1884],[2964,0.1884],[3199,0.132],[3509,0.0266],[3841,0],[4366,0]]
       }
       middle_metacarpal_l_ctrl.Rotation.Y {
        type=MovieSceneScriptingFloatChannel
        default=0
        thinned_from=16
        keys [[1790,0.1439],[2217,0.1439],[2529,0.1316],[2556,0.13],[2754,0.111],[2944,0.0912],[2964,0.0912],[3199,0.0639],[3509,0.0129],[3841,0],[4366,0]]
       }
       middle_metacarpal_l_ctrl.Rotation.Z {
        type=MovieSceneScriptingFloatChannel
        default=0
        thinned_from=16
        keys [[1790,9.1927],[2217,9.1927],[2529,8.4082],[2556,8.3061],[2754,7.092],[2944,5.8249],[2964,5.8249],[3199,4.081],[3509,0.8222],[3841,0],[4366,0]]
       }
       middle_metacarpal_l_ctrl.Scale.X {type=MovieSceneScriptingFloatChannel, default=0, const [0,16]}
       middle_metacarpal_l_ctrl.Scale.Y {type=MovieSceneScriptingFloatChannel, default=0, const [0,16]}
       middle_metacarpal_l_ctrl.Scale.Z {type=MovieSceneScriptingFloatChannel, default=0, const [0,16]}
       middle_01_l_ctrl.Location.X {type=MovieSceneScriptingFloatChannel, default=0, thinned_from=17, keys [[1790,0],[3852,0]]}
       middle_01_l_ctrl.Location.Y {type=MovieSceneScriptingFloatChannel, default=0, thinned_from=17, keys [[1790,0],[3852,0]]}
       middle_01_l_ctrl.Location.Z {type=MovieSceneScriptingFloatChannel, default=0, thinned_from=17, keys [[1790,0],[3852,0]]}
       middle_01_l_ctrl.Rotation.X {
        type=MovieSceneScriptingFloatChannel
        default=0
        thinned_from=22
        keys [[1790,1.4207],[2217,1.4207],[2529,1.2995],[2556,1.2837],[2754,1.0961],[2803,3.0934],[2964,3.0934],[3199,2.6135],[3509,0.5778],[3571,0.1794],[3587,-3.9299],[3601,0.121],[3632,1.6041],[3673,2.1857],[3738,2.5048],[3852,2.5048],[3938,-19.9552]]
       }
       middle_01_l_ctrl.Rotation.Y {
        type=MovieSceneScriptingFloatChannel
        default=0
        thinned_from=22
        keys [[1790,-0.1132],[2217,-0.1132],[2529,-0.1036],[2556,-0.1023],[2754,-0.0873],[2803,-2.0517],[2964,-2.0517],[3199,-1.7333],[3509,-0.3832],[3571,-0.119],[3587,-2.0982],[3601,-0.0802],[3632,-0.301],[3673,-0.4102],[3738,-0.4701],[3852,-0.4701],[3938,-11.2335]]
       }
       middle_01_l_ctrl.Rotation.Z {
        type=MovieSceneScriptingFloatChannel
        default=0
        thinned_from=22
        keys [[1790,15.1923],[2217,15.1923],[2529,13.8958],[2556,13.7271],[2754,11.7205],[2803,-39.2515],[2964,-39.2515],[3199,-33.1615],[3509,-7.3313],[3571,-2.2759],[3587,-51.1233],[3601,-1.5351],[3632,18.1491],[3673,24.131],[3738,27.3731],[3852,27.3731],[3938,-21.119]]
       }
       middle_01_l_ctrl.Scale.X {type=MovieSceneScriptingFloatChannel, default=0, const [0,17]}
       middle_01_l_ctrl.Scale.Y {type=MovieSceneScriptingFloatChannel, default=0, const [0,17]}
       middle_01_l_ctrl.Scale.Z {type=MovieSceneScriptingFloatChannel, default=0, const [0,17]}
       middle_02_l_ctrl.Location.X {type=MovieSceneScriptingFloatChannel, default=0, thinned_from=16, keys [[1790,0],[3673,0]]}
       middle_02_l_ctrl.Location.Y {type=MovieSceneScriptingFloatChannel, default=0, thinned_from=16, keys [[1790,0],[3673,0]]}
       middle_02_l_ctrl.Location.Z {type=MovieSceneScriptingFloatChannel, default=0, thinned_from=16, keys [[1790,0],[3673,0]]}
       middle_02_l_ctrl.Rotation.X {
        type=MovieSceneScriptingFloatChannel
        default=0
        thinned_from=21
        keys [[1790,0],[2754,0],[2803,1.5058],[2964,1.5058],[3199,1.2722],[3509,0.2813],[3571,0.0873],[3587,0.496],[3601,0.0589],[3632,-0.0064],[3673,-0.0074],[3738,-0.008],[3938,-0.0328]]
       }
       middle_02_l_ctrl.Rotation.Y {
        type=MovieSceneScriptingFloatChannel
        default=0
        thinned_from=21
        keys [[1790,0],[2754,0],[2803,-1.8692],[2964,-1.8692],[3199,-1.5792],[3509,-0.3491],[3571,-0.1084],[3587,-0.6166],[3601,-0.0731],[3632,0.1376],[3673,0.1685],[3738,0.1847],[3938,0.0522]]
       }
       middle_02_l_ctrl.Rotation.Z {
        type=MovieSceneScriptingFloatChannel
        default=0
        thinned_from=21
        keys [[1790,0],[2754,0],[2803,-50.4127],[2964,-50.4127],[3199,-42.591],[3509,-9.4159],[3571,-2.9231],[3587,-51.5185],[3601,-1.9716],[3632,8.4343],[3673,10.7228],[3738,11.9277],[3938,-14.4852]]
       }
       middle_02_l_ctrl.Scale.X {type=MovieSceneScriptingFloatChannel, default=0, const [0,16]}
       middle_02_l_ctrl.Scale.Y {type=MovieSceneScriptingFloatChannel, default=0, const [0,16]}
       middle_02_l_ctrl.Scale.Z {type=MovieSceneScriptingFloatChannel, default=0, const [0,16]}
       middle_03_l_ctrl.Location.X {type=MovieSceneScriptingFloatChannel, default=0, const [0,16]}
       middle_03_l_ctrl.Location.Y {type=MovieSceneScriptingFloatChannel, default=0, const [0,16]}
       middle_03_l_ctrl.Location.Z {type=MovieSceneScriptingFloatChannel, default=0, const [0,16]}
       middle_03_l_ctrl.Rotation.X {
        type=MovieSceneScriptingFloatChannel
        default=0
        thinned_from=21
        keys [[1790,0],[2754,0],[2803,-0.646],[2964,-0.646],[3199,-0.5458],[3509,-0.1207],[3571,-0.0375],[3587,-0.1536],[3601,-0.0253],[3632,0.0649],[3673,0.0801],[3738,0.0881],[3938,0.0595]]
       }
       middle_03_l_ctrl.Rotation.Y {
        type=MovieSceneScriptingFloatChannel
        default=0
        thinned_from=21
        keys [[1790,0],[2754,0],[2803,-1.2122],[2964,-1.2122],[3199,-1.0241],[3509,-0.2264],[3571,-0.0703],[3587,-0.3768],[3601,-0.0474],[3632,0.0411],[3673,0.0486],[3738,0.0524],[3938,0.0415]]
       }
       middle_03_l_ctrl.Rotation.Z {
        type=MovieSceneScriptingFloatChannel
        default=0
        thinned_from=21
        keys [[1790,0],[2754,0],[2803,-50.4058],[2964,-50.4058],[3199,-42.5852],[3509,-9.4146],[3571,-2.9227],[3587,-51.5183],[3601,-1.9713],[3632,8.4415],[3673,10.7328],[3738,11.9392],[3938,-5.6286]]
       }
       middle_03_l_ctrl.Scale.X {type=MovieSceneScriptingFloatChannel, default=0, const [0,16]}
       middle_03_l_ctrl.Scale.Y {type=MovieSceneScriptingFloatChannel, default=0, const [0,16]}
       middle_03_l_ctrl.Scale.Z {type=MovieSceneScriptingFloatChannel, default=0, const [0,16]}
       ring_metacarpal_l_ctrl.Location.X {type=MovieSceneScriptingFloatChannel, default=0, const [0,16]}
       ring_metacarpal_l_ctrl.Location.Y {type=MovieSceneScriptingFloatChannel, default=0, const [0,16]}
       ring_metacarpal_l_ctrl.Location.Z {type=MovieSceneScriptingFloatChannel, default=0, const [0,16]}
       ring_metacarpal_l_ctrl.Rotation.X {
        type=MovieSceneScriptingFloatChannel
        default=0
        thinned_from=16
        keys [[1790,-0.0288],[2217,-0.0288],[2556,-0.0261],[2754,-0.0223],[2944,-0.0183],[2964,-0.0183],[3199,-0.0128],[3509,-0.0026],[3841,0],[4366,0]]
       }
       ring_metacarpal_l_ctrl.Rotation.Y {
        type=MovieSceneScriptingFloatChannel
        default=0
        thinned_from=16
        keys [[1790,1.3838],[2217,1.3838],[2529,1.2657],[2556,1.2503],[2754,1.0675],[2944,0.8768],[2964,0.8768],[3199,0.6143],[3509,0.1238],[3841,0],[4366,0]]
       }
       ring_metacarpal_l_ctrl.Rotation.Z {
        type=MovieSceneScriptingFloatChannel
        default=0
        thinned_from=16
        keys [[1790,-0.3536],[2217,-0.3536],[2529,-0.3234],[2556,-0.3195],[2754,-0.2728],[2944,-0.224],[2964,-0.224],[3199,-0.157],[3509,-0.0316],[3841,0],[4366,0]]
       }
       ring_metacarpal_l_ctrl.Scale.X {type=MovieSceneScriptingFloatChannel, default=0, const [0,16]}
       ring_metacarpal_l_ctrl.Scale.Y {type=MovieSceneScriptingFloatChannel, default=0, const [0,16]}
       ring_metacarpal_l_ctrl.Scale.Z {type=MovieSceneScriptingFloatChannel, default=0, const [0,16]}
       ring_01_l_ctrl.Location.X {type=MovieSceneScriptingFloatChannel, default=0, const [0,17]}
       ring_01_l_ctrl.Location.Y {type=MovieSceneScriptingFloatChannel, default=0, thinned_from=17, keys [[1790,0],[3852,0]]}
       ring_01_l_ctrl.Location.Z {type=MovieSceneScriptingFloatChannel, default=0, thinned_from=17, keys [[1790,0],[3852,0]]}
       ring_01_l_ctrl.Rotation.X {
        type=MovieSceneScriptingFloatChannel
        default=0
        thinned_from=22
        keys [[1790,5.6985],[2217,5.6985],[2529,5.2121],[2556,5.1489],[2754,4.3962],[2803,2.4441],[2964,2.4441],[3199,2.0649],[3509,0.4565],[3571,0.1417],[3587,-3.8546],[3601,0.0956],[3632,2.009],[3673,2.7375],[3738,3.1372],[3852,3.1372],[3938,-14.7385]]
       }
       ring_01_l_ctrl.Rotation.Y {
        type=MovieSceneScriptingFloatChannel
        default=0
        thinned_from=22
        keys [[1790,11.3694],[2217,11.3694],[2529,10.3991],[2556,10.2729],[2754,8.7712],[2803,9.9498],[2964,9.9498],[3199,8.4061],[3509,1.8584],[3571,0.5769],[3587,10.8195],[3601,0.3891],[3632,4.6375],[3673,6.3192],[3738,7.2418],[3852,7.2418],[3938,11.2155]]
       }
       ring_01_l_ctrl.Rotation.Z {
        type=MovieSceneScriptingFloatChannel
        default=0
        thinned_from=22
        keys [[1790,26.2551],[2217,26.2551],[2529,24.0144],[2556,23.7229],[2754,20.2552],[2803,-13.2533],[2964,-13.2533],[3199,-11.197],[3509,-2.4754],[3571,-0.7685],[3587,-48.9186],[3601,-0.5183],[3632,23.9785],[3673,32.4715],[3738,37.1172],[3852,37.1172],[3938,-16.6666]]
       }
       ring_01_l_ctrl.Scale.X {type=MovieSceneScriptingFloatChannel, default=0, const [0,17]}
       ring_01_l_ctrl.Scale.Y {type=MovieSceneScriptingFloatChannel, default=0, const [0,17]}
       ring_01_l_ctrl.Scale.Z {type=MovieSceneScriptingFloatChannel, default=0, const [0,17]}
       ring_02_l_ctrl.Location.X {type=MovieSceneScriptingFloatChannel, default=0, thinned_from=16, keys [[1790,0],[3673,0]]}
       ring_02_l_ctrl.Location.Y {type=MovieSceneScriptingFloatChannel, default=0, thinned_from=16, keys [[1790,0],[3673,0]]}
       ring_02_l_ctrl.Location.Z {type=MovieSceneScriptingFloatChannel, default=0, thinned_from=16, keys [[1790,0],[3673,0]]}
       ring_02_l_ctrl.Rotation.X {
        type=MovieSceneScriptingFloatChannel
        default=0
        thinned_from=21
        keys [[1790,0],[2754,0],[2803,0.0741],[2964,0.0741],[3199,0.0626],[3509,0.0138],[3571,0.0043],[3587,0.2688],[3601,0.0029],[3632,0.0348],[3673,0.0474],[3738,0.0543],[3938,-0.0138]]
       }
       ring_02_l_ctrl.Rotation.Y {
        type=MovieSceneScriptingFloatChannel
        default=0
        thinned_from=21
        keys [[1790,0],[2754,0],[2803,-0.374],[2964,-0.374],[3199,-0.316],[3509,-0.0699],[3571,-0.0217],[3587,-0.9093],[3601,-0.0146],[3632,0.1623],[3673,0.2155],[3738,0.2442],[3938,-0.3235]]
       }
       ring_02_l_ctrl.Rotation.Z {
        type=MovieSceneScriptingFloatChannel
        default=0
        thinned_from=21
        keys [[1790,0],[2754,0],[2803,-32.5534],[2964,-32.5534],[3199,-27.5027],[3509,-6.0802],[3571,-1.8876],[3587,-50.6871],[3601,-1.2731],[3632,8.158],[3673,10.6191],[3738,11.9366],[3938,-14.4708]]
       }
       ring_02_l_ctrl.Scale.X {type=MovieSceneScriptingFloatChannel, default=0, const [0,16]}
       ring_02_l_ctrl.Scale.Y {type=MovieSceneScriptingFloatChannel, default=0, const [0,16]}
       ring_02_l_ctrl.Scale.Z {type=MovieSceneScriptingFloatChannel, default=0, const [0,16]}
       ring_03_l_ctrl.Location.X {type=MovieSceneScriptingFloatChannel, default=0, const [0,16]}
       ring_03_l_ctrl.Location.Y {type=MovieSceneScriptingFloatChannel, default=0, thinned_from=16, keys [[1790,0],[3673,0]]}
       ring_03_l_ctrl.Location.Z {type=MovieSceneScriptingFloatChannel, default=0, thinned_from=16, keys [[1790,0],[3673,0]]}
       ring_03_l_ctrl.Rotation.X {
        type=MovieSceneScriptingFloatChannel
        default=0
        thinned_from=21
        keys [[1790,0.0319],[2217,0.0319],[2529,0.0292],[2556,0.0289],[2754,0.0246],[2803,0.1535],[2964,0.1535],[3199,0.1297],[3509,0.0287],[3571,0.0089],[3587,0.1335],[3601,0.006],[3632,0.0152],[3673,0.0207],[3738,0.0238],[3938,-0.0033]]
       }
       ring_03_l_ctrl.Rotation.Y {
        type=MovieSceneScriptingFloatChannel
        default=0
        thinned_from=21
        keys [[1790,-0.2508],[2217,-0.2508],[2529,-0.2294],[2556,-0.2266],[2754,-0.1935],[2803,-0.3067],[2964,-0.3067],[3199,-0.2592],[3509,-0.0573],[3571,-0.0178],[3587,-0.397],[3601,-0.012],[3632,0.0669],[3673,0.0864],[3738,0.0968],[3938,-0.07]]
       }
       ring_03_l_ctrl.Rotation.Z {
        type=MovieSceneScriptingFloatChannel
        default=0
        thinned_from=21
        keys [[1790,-30.2598],[2217,-30.2598],[2529,-27.6773],[2556,-27.3413],[2754,-23.3448],[2803,-54.8036],[2964,-54.8036],[3199,-46.3007],[3509,-10.236],[3571,-3.1777],[3587,-51.7212],[3601,-2.1433],[3632,8.5106],[3673,10.7597],[3738,11.9386],[3938,-5.6284]]
       }
       ring_03_l_ctrl.Scale.X {type=MovieSceneScriptingFloatChannel, default=0, thinned_from=17, keys [[1790,0],[3601,0],[3632,0.0406],[3673,0.0553],[3738,0.0634]]}
       ring_03_l_ctrl.Scale.Y {type=MovieSceneScriptingFloatChannel, default=0, thinned_from=17, keys [[1790,0],[3601,0],[3632,0.0406],[3673,0.0553],[3738,0.0634]]}
       ring_03_l_ctrl.Scale.Z {type=MovieSceneScriptingFloatChannel, default=0, thinned_from=17, keys [[1790,0],[3601,0],[3632,0.0406],[3673,0.0553],[3738,0.0634]]}
       pinky_metacarpal_l_ctrl.Location.X {type=MovieSceneScriptingFloatChannel, default=0, const [0,16]}
       pinky_metacarpal_l_ctrl.Location.Y {type=MovieSceneScriptingFloatChannel, default=0, const [0,16]}
       pinky_metacarpal_l_ctrl.Location.Z {type=MovieSceneScriptingFloatChannel, default=0, const [0,16]}
       pinky_metacarpal_l_ctrl.Rotation.X {
        type=MovieSceneScriptingFloatChannel
        default=0
        thinned_from=16
        keys [[1790,-0.2501],[2217,-0.2501],[2529,-0.2287],[2556,-0.226],[2754,-0.1929],[2944,-0.1585],[2964,-0.1585],[3199,-0.111],[3509,-0.0224],[3841,0],[4366,0]]
       }
       pinky_metacarpal_l_ctrl.Rotation.Y {
        type=MovieSceneScriptingFloatChannel
        default=0
        thinned_from=16
        keys [[1790,7.6411],[2217,7.6411],[2529,6.989],[2556,6.9041],[2754,5.8949],[2944,4.8417],[2964,4.8417],[3199,3.3922],[3509,0.6834],[3841,0],[4366,0]]
       }
       pinky_metacarpal_l_ctrl.Rotation.Z {
        type=MovieSceneScriptingFloatChannel
        default=0
        thinned_from=16
        keys [[1790,-2.652],[2217,-2.652],[2529,-2.4257],[2556,-2.3962],[2754,-2.046],[2944,-1.6804],[2964,-1.6804],[3199,-1.1773],[3509,-0.2372],[3841,0],[4366,0]]
       }
       pinky_metacarpal_l_ctrl.Scale.X {type=MovieSceneScriptingFloatChannel, default=0, thinned_from=17, keys [[1790,0],[3841,0],[4067,0.0553],[4132,0.0634],[4364,0],[4366,0]]}
       pinky_metacarpal_l_ctrl.Scale.Y {type=MovieSceneScriptingFloatChannel, default=0, thinned_from=17, keys [[1790,0],[3841,0],[4067,0.0553],[4132,0.0634],[4364,0],[4366,0]]}
       pinky_metacarpal_l_ctrl.Scale.Z {type=MovieSceneScriptingFloatChannel, default=0, thinned_from=17, keys [[1790,0],[3841,0],[4067,0.0553],[4132,0.0634],[4364,0],[4366,0]]}
       pinky_01_l_ctrl.Location.X {type=MovieSceneScriptingFloatChannel, default=0, thinned_from=18, keys [[1790,0],[3852,0]]}
       pinky_01_l_ctrl.Location.Y {type=MovieSceneScriptingFloatChannel, default=0, thinned_from=18, keys [[1790,0],[3852,0]]}
       pinky_01_l_ctrl.Location.Z {type=MovieSceneScriptingFloatChannel, default=0, thinned_from=18, keys [[1790,0],[3852,0]]}
       pinky_01_l_ctrl.Rotation.X {
        type=MovieSceneScriptingFloatChannel
        default=0
        thinned_from=22
        keys [[1790,1.9755],[2217,1.9755],[2529,1.8069],[2556,1.7849],[2754,1.524],[2803,-3.2827],[2964,-3.2827],[3199,-2.7734],[3509,-0.6131],[3571,-0.1903],[3587,-9.3558],[3601,-0.1284],[3632,7.7091],[3673,10.4545],[3738,11.9572],[3852,11.9572],[3938,-11.5863]]
       }
       pinky_01_l_ctrl.Rotation.Y {
        type=MovieSceneScriptingFloatChannel
        default=0
        thinned_from=23
        keys [[1790,27.0269],[2217,27.0269],[2529,24.7204],[2556,24.4202],[2754,20.8507],[2803,18.112],[2964,18.112],[3199,15.3019],[3509,3.3829],[3571,1.0502],[3587,7.4059],[3601,0.7083],[3632,17.02],[3658,24.9152],[3673,23.192],[3738,26.5777],[3852,26.5777],[3938,15.0645]]
       }
       pinky_01_l_ctrl.Rotation.Z {
        type=MovieSceneScriptingFloatChannel
        default=0
        thinned_from=22
        keys [[1790,5.878],[2217,5.878],[2529,5.3763],[2556,5.3111],[2754,4.5347],[2803,-19.9578],[2964,-19.9578],[3199,-16.8613],[3509,-3.7277],[3571,-1.1572],[3587,-50.5069],[3601,-0.7805],[3632,22.3144],[3673,30.1015],[3738,34.3532],[3852,34.3532],[3938,-31.0074]]
       }
       pinky_01_l_ctrl.Scale.X {
        type=MovieSceneScriptingFloatChannel
        default=0
        thinned_from=18
        keys [[1790,0],[3601,0],[3632,0.0769],[3673,0.1049],[3738,0.1202],[3852,0.1202]]
       }
       pinky_01_l_ctrl.Scale.Y {
        type=MovieSceneScriptingFloatChannel
        default=0
        thinned_from=18
        keys [[1790,0],[3601,0],[3632,0.0769],[3673,0.1049],[3738,0.1202],[3852,0.1202]]
       }
       pinky_01_l_ctrl.Scale.Z {
        type=MovieSceneScriptingFloatChannel
        default=0
        thinned_from=18
        keys [[1790,0],[3601,0],[3632,0.0769],[3673,0.1049],[3738,0.1202],[3852,0.1202]]
       }
       pinky_02_l_ctrl.Location.X {type=MovieSceneScriptingFloatChannel, default=0, thinned_from=17, keys [[1790,0],[3738,0]]}
       pinky_02_l_ctrl.Location.Y {type=MovieSceneScriptingFloatChannel, default=0, thinned_from=17, keys [[1790,0],[3738,0]]}
       pinky_02_l_ctrl.Location.Z {type=MovieSceneScriptingFloatChannel, default=0, thinned_from=17, keys [[1790,0],[3738,0]]}
       pinky_02_l_ctrl.Rotation.X {
        type=MovieSceneScriptingFloatChannel
        default=0
        thinned_from=21
        keys [[1790,0],[2754,0],[2803,0.0677],[2964,0.0677],[3199,0.0572],[3509,0.0126],[3571,0.0039],[3587,0.8899],[3601,0.0026],[3632,0.0438],[3673,0.0596],[3738,0.0684],[3938,0.47]]
       }
       pinky_02_l_ctrl.Rotation.Y {
        type=MovieSceneScriptingFloatChannel
        default=0
        thinned_from=21
        keys [[1790,0],[2754,0],[2803,-0.3586],[2964,-0.3586],[3199,-0.303],[3509,-0.067],[3571,-0.0208],[3587,-2.0541],[3601,-0.014],[3632,0.3461],[3673,0.4661],[3738,0.5316],[3938,-1.6541]]
       }
       pinky_02_l_ctrl.Rotation.Z {
        type=MovieSceneScriptingFloatChannel
        default=0
        thinned_from=21
        keys [[1790,0],[2754,0],[2803,-23.8834],[2964,-23.8834],[3199,-20.1778],[3509,-4.4609],[3571,-1.3848],[3587,-50.2612],[3601,-0.9341],[3632,8.0149],[3673,10.5566],[3738,11.927],[3938,-32.0042]]
       }
       pinky_02_l_ctrl.Scale.X {type=MovieSceneScriptingFloatChannel, default=0, thinned_from=17, keys [[1790,0],[3601,0],[3632,0.0747],[3673,0.1018],[3738,0.1166]]}
       pinky_02_l_ctrl.Scale.Y {type=MovieSceneScriptingFloatChannel, default=0, thinned_from=17, keys [[1790,0],[3601,0],[3632,0.0747],[3673,0.1018],[3738,0.1166]]}
       pinky_02_l_ctrl.Scale.Z {type=MovieSceneScriptingFloatChannel, default=0, thinned_from=17, keys [[1790,0],[3601,0],[3632,0.0747],[3673,0.1018],[3738,0.1166]]}
       pinky_03_l_ctrl.Location.X {type=MovieSceneScriptingFloatChannel, default=0, thinned_from=17, keys [[1790,0],[3738,0]]}
       pinky_03_l_ctrl.Location.Y {type=MovieSceneScriptingFloatChannel, default=0, thinned_from=17, keys [[1790,0],[3738,0]]}
       pinky_03_l_ctrl.Location.Z {type=MovieSceneScriptingFloatChannel, default=0, thinned_from=17, keys [[1790,0],[3738,0]]}
       pinky_03_l_ctrl.Rotation.X {
        type=MovieSceneScriptingFloatChannel
        default=0
        thinned_from=21
        keys [[1790,0],[2754,0],[2803,0.1357],[2964,0.1357],[3199,0.1146],[3509,0.0253],[3571,0.0079],[3587,0.9875],[3601,0.0053],[3632,-0.0866],[3673,-0.116],[3738,-0.132],[3938,0.1163]]
       }
       pinky_03_l_ctrl.Rotation.Y {
        type=MovieSceneScriptingFloatChannel
        default=0
        thinned_from=21
        keys [[1790,0],[2754,0],[2803,-0.1286],[2964,-0.1286],[3199,-0.1087],[3509,-0.024],[3571,-0.0075],[3587,-0.5883],[3601,-0.005],[3632,0.1591],[3673,0.2149],[3738,0.2454],[3938,-0.1489]]
       }
       pinky_03_l_ctrl.Rotation.Z {
        type=MovieSceneScriptingFloatChannel
        default=0
        thinned_from=21
        keys [[1790,0],[2754,0],[2803,-23.8799],[2964,-23.8799],[3199,-20.1749],[3509,-4.4602],[3571,-1.3846],[3587,-50.2991],[3601,-0.9339],[3632,8.0214],[3673,10.5655],[3738,11.9371],[3938,-5.6279]]
       }
       pinky_03_l_ctrl.Scale.X {type=MovieSceneScriptingFloatChannel, default=0, thinned_from=17, keys [[1790,0],[3601,0],[3632,0.0727],[3673,0.099],[3738,0.1135]]}
       pinky_03_l_ctrl.Scale.Y {type=MovieSceneScriptingFloatChannel, default=0, thinned_from=17, keys [[1790,0],[3601,0],[3632,0.0727],[3673,0.099],[3738,0.1135]]}
       pinky_03_l_ctrl.Scale.Z {type=MovieSceneScriptingFloatChannel, default=0, thinned_from=17, keys [[1790,0],[3601,0],[3632,0.0727],[3673,0.099],[3738,0.1135]]}
       hand_l_ik_ctrl.Location.X {
        type=MovieSceneScriptingFloatChannel
        default=0
        keys [[1790,-3.5446],[2113,1040.9978],[2217,-3.5446],[2516,957.2751],[2523,922.2101],[2529,876.3285],[2556,715.7339],[2754,472.7214],[2944,570.5413],[2964,570.5413],[3199,562.3593],[3509,184.8778],[3841,138.2399],[4067,34.3683],[4364,120.66],[4366,-131.553]]
       }
       hand_l_ik_ctrl.Location.Y {
        type=MovieSceneScriptingFloatChannel
        default=0
        keys [[1790,-1.8527],[2113,-103.8224],[2217,-1.8527],[2516,-425.8712],[2523,-390.9834],[2529,-367.3867],[2556,-277.0844],[2754,-127.9869],[2944,-272.9925],[2964,-272.9925],[3199,-267.7654],[3509,-201.3014],[3841,-260.0282],[4067,-189.4913],[4364,-251.2126],[4366,-143.7039]]
       }
       hand_l_ik_ctrl.Location.Z {
        type=MovieSceneScriptingFloatChannel
        default=0
        thinned_from=16
        keys [[1790,2.5398],[2113,2524.6692],[2217,2.5398],[2516,4784.9395],[2529,4317.9878],[2556,3483.8389],[2754,2288.8223],[2944,3150.1201],[2964,3150.1201],[3199,3003.4082],[3509,2389.6086],[3841,2296.0754],[4067,2369.637],[4364,2345.5579],[4366,2414.5115]]
       }
       hand_l_ik_ctrl.Rotation.X {
        type=MovieSceneScriptingFloatChannel
        default=0
        keys [[1790,11.0884],[2113,9.7291],[2217,11.0884],[2516,75.9039],[2523,131.0381],[2529,128.3942],[2556,110.7661],[2754,68.0427],[2944,112.9356],[2964,112.9356],[3199,8.8518],[3509,24.4499],[3841,177.0497],[4067,-4.5951],[4364,127.9513],[4366,170.5222]]
       }
       hand_l_ik_ctrl.Rotation.Y {
        type=MovieSceneScriptingFloatChannel
        default=0
        keys [[1790,-13.8691],[2113,-14.6937],[2217,-13.8691],[2516,-97.0926],[2523,-45.2533],[2529,-38.1933],[2556,-34.6024],[2754,3.6961],[2944,-39.0374],[2964,-39.0374],[3199,6.3679],[3509,-2.9744],[3841,-18.9302],[4067,36.9137],[4364,39.6538],[4366,29.8825]]
       }
       hand_l_ik_ctrl.Rotation.Z {
        type=MovieSceneScriptingFloatChannel
        default=0
        keys [[1790,-1.5505],[2113,-3.2564],[2217,-1.5505],[2516,-159.4373],[2523,-135.6465],[2529,-128.9854],[2556,-136.7429],[2754,-257.1259],[2944,-103.4979],[2964,-103.4979],[3199,-101.4043],[3509,205.283],[3841,78.4743],[4067,148.0081],[4364,231.5284],[4366,130.8446]]
       }
       hand_l_ik_ctrl.Scale.X {
        type=MovieSceneScriptingFloatChannel
        default=0
        thinned_from=16
        keys [[1790,0],[2217,0],[2516,-0.0346],[2523,0],[2529,-0.0353],[2556,-0.0537],[2754,0.0065],[2944,-0.0475],[2964,-0.0475],[3199,-0.0497],[3509,-0.0178],[3841,0],[4067,-0.0045],[4366,-0.0026]]
       }
       hand_l_ik_ctrl.Scale.Y {
        type=MovieSceneScriptingFloatChannel
        default=0
        thinned_from=16
        keys [[1790,0],[2217,0],[2516,-0.0346],[2523,0],[2529,-0.0353],[2556,-0.0537],[2754,0.0065],[2944,-0.0475],[2964,-0.0475],[3199,-0.0497],[3509,-0.0178],[3841,0],[4067,-0.0045],[4366,-0.0026]]
       }
       hand_l_ik_ctrl.Scale.Z {
        type=MovieSceneScriptingFloatChannel
        default=0
        thinned_from=16
        keys [[1790,0],[2217,0],[2516,-0.0346],[2523,0],[2529,-0.0353],[2556,-0.0537],[2754,0.0065],[2944,-0.0475],[2964,-0.0475],[3199,-0.0497],[3509,-0.0178],[3841,0],[4067,-0.0045],[4366,-0.0026]]
       }
       Softness {type=MovieSceneScriptingFloatChannel, default=0, const [0,16]}
       hand_l_ikorient_ctrl.Rotation.Roll {type=MovieSceneScriptingFloatChannel, default=0, const [0,16]}
       hand_l_ikorient_ctrl.Rotation.Pitch {type=MovieSceneScriptingFloatChannel, default=0, const [0,16]}
       hand_l_ikorient_ctrl.Rotation.Yaw {type=MovieSceneScriptingFloatChannel, default=0, const [0,16]}
       arm_l_pv_ik_ctrl.Location.X {
        type=MovieSceneScriptingFloatChannel
        default=0
        keys [[1790,0.0852],[2113,-2.2253],[2217,0.0852],[2516,-81.4591],[2523,-44.9388],[2529,-48.1942],[2556,-48.4818],[2754,-64.8844],[2944,-54.5715],[2964,-54.5715],[3199,-52.1424],[3509,-132.5821],[3841,-73.2134],[4067,-201.3212],[4364,-88.4239],[4366,-279.2833]]
       }
       arm_l_pv_ik_ctrl.Location.Y {
        type=MovieSceneScriptingFloatChannel
        default=0
        keys [[1790,6.0234],[2113,2731.2705],[2217,6.0234],[2516,4798.1733],[2523,4535.3301],[2529,4325.5703],[2556,3506.5627],[2754,2322.6111],[2944,3145.0857],[2964,3145.0857],[3199,3042.9871],[3509,2292.9617],[3841,2177.8833],[4067,2189.092],[4364,2214.1772],[4366,2123.6345]]
       }
       arm_l_pv_ik_ctrl.Location.Z {
        type=MovieSceneScriptingFloatChannel
        default=0
        keys [[1790,11.6151],[2113,195.8024],[2217,11.6151],[2516,1292.0974],[2523,1212.708],[2529,1156.3749],[2556,927.1965],[2754,609.652],[2944,917.1295],[2964,917.1295],[3199,823.6518],[3509,881.0798],[3841,919.3496],[4067,964.1494],[4364,907.0991],[4366,1140.6099]]
       }
       arm_l_pv_ik_ctrl.Rotation.X {
        type=MovieSceneScriptingFloatChannel
        default=0
        keys [[1790,17.9981],[2113,17.4843],[2217,17.9981],[2516,5.2519],[2523,5.6651],[2529,5.8901],[2556,6.8467],[2754,4.9638],[2944,0.1817],[2964,0.1817],[3199,5.4771],[3509,7.3301],[3841,20.0164],[4067,-126.7192],[4364,-109.647],[4366,-18.4982]]
       }
       arm_l_pv_ik_ctrl.Rotation.Y {
        type=MovieSceneScriptingFloatChannel
        default=0
        keys [[1790,4.1959],[2113,2.5772],[2217,4.1959],[2516,-2.674],[2523,3.381],[2529,1.3042],[2556,-0.2379],[2754,-4.647],[2944,-3.0283],[2964,-3.0283],[3199,-11.0406],[3509,-1.0717],[3841,3.6358],[4067,29.7617],[4364,9.1345],[4366,47.3373]]
       }
       arm_l_pv_ik_ctrl.Rotation.Z {
        type=MovieSceneScriptingFloatChannel
        default=0
        thinned_from=16
        keys [[1790,6.345],[2113,3.87],[2217,6.345],[2516,-163.9927],[2529,-163.0097],[2556,-165.4417],[2754,-201.3461],[2944,-165.4512],[2964,-165.4512],[3199,-175.2064],[3509,-165.4785],[3841,-153.0486],[4067,-246.3057],[4364,-211.1253],[4366,-330.2451]]
       }
       arm_l_pv_ik_ctrl.Scale.X {type=MovieSceneScriptingFloatChannel, default=0, const [0,16]}
       arm_l_pv_ik_ctrl.Scale.Y {type=MovieSceneScriptingFloatChannel, default=0, const [0,16]}
       arm_l_pv_ik_ctrl.Scale.Z {type=MovieSceneScriptingFloatChannel, default=0, const [0,16]}
       upperarm_r_fk_ctrl.Location.X {type=MovieSceneScriptingFloatChannel, default=0, thinned_from=27, keys [[1790,0],[4366,0]]}
       upperarm_r_fk_ctrl.Location.Y {type=MovieSceneScriptingFloatChannel, default=0, thinned_from=27, keys [[1790,0],[4366,0]]}
       upperarm_r_fk_ctrl.Location.Z {type=MovieSceneScriptingFloatChannel, default=0, thinned_from=27, keys [[1790,0],[4366,0]]}
       upperarm_r_fk_ctrl.Rotation.X {
        type=MovieSceneScriptingFloatChannel
        default=0
        thinned_from=41
        keys [[1790,0],[2113,0],[2186,0.3842],[2204,14.8751],[2236,0.6223],[2312,1.9141],[2385,1.9812],[2529,1.9333],[2556,1.919],[2634,1.8555],[2644,1.5099],[2680,1.7116],[2944,1.414],[2970,0.863],[2992,-49.1835],[3052,-49.1835],[3078,-1.0444],[3199,0.9919],[3561,0.1999],[3583,35.375],[3585,26.1932],[3593,-26.7793],[3598,-38.3206],[3615,-50.3043],[3676,-65.4341],[3687,-67.6772],[3811,-53.5703],[3839,-68.544],[3930,-40.3081],[3950,-41.8361],[4023,-81.7247],[4038,-83.8842],[4067,-86.2873],[4120,-89.5731],[4364,-87.0099],[4365,34.5531],[4366,179.8712],[4372,180.1569]]
       }
       upperarm_r_fk_ctrl.Rotation.Y {
        type=MovieSceneScriptingFloatChannel
        default=0
        thinned_from=38
        keys [[1790,0],[2113,0],[2186,0.4304],[2204,1.8555],[2236,-14.0394],[2312,5.297],[2385,1.045],[2523,0.9069],[2529,0.9038],[2556,0.8819],[2634,0.8329],[2644,-6.5088],[2680,0.7488],[2944,0.6145],[2969,0.6145],[2992,34.8391],[3052,34.8391],[3061,40.2425],[3078,12.2569],[3199,0.4305],[3501,0.0867],[3543,1.3049],[3566,-0.4893],[3583,17.2834],[3592,8.0185],[3597,1.1156],[3605,-9.9077],[3633,10.6433],[3723,80.9032],[3811,98.9019],[3835,102.3105],[3922,111.8937],[4025,113.0801],[4364,106.3912],[4365,-0.3137],[4366,-6.6282],[4372,2.2443]]
       }
       upperarm_r_fk_ctrl.Rotation.Z {
        type=MovieSceneScriptingFloatChannel
        default=0
        thinned_from=33
        keys [[1790,0],[2113,0],[2186,5.1643],[2204,3.7929],[2236,10.3328],[2312,14.1196],[2385,14.3906],[2529,14.0431],[2556,13.9393],[2634,13.4778],[2644,6.4294],[2680,12.4323],[2944,10.271],[2964,10.271],[2992,23.3295],[3000,24.5586],[3069,26.6273],[3199,7.2047],[3561,1.4523],[3584,39.0279],[3590,25.0487],[3597,23.5873],[3613,23.0864],[3673,29.419],[3804,9.5542],[4023,9.9306],[4038,9.3584],[4067,8.7305],[4104,8.1852],[4358,11.7835],[4372,73.038]]
       }
       upperarm_r_fk_ctrl.Scale.X {type=MovieSceneScriptingFloatChannel, default=0, const [0,27]}
       upperarm_r_fk_ctrl.Scale.Y {type=MovieSceneScriptingFloatChannel, default=0, const [0,27]}
       upperarm_r_fk_ctrl.Scale.Z {type=MovieSceneScriptingFloatChannel, default=0, const [0,27]}
       lowerarm_r_fk_ctrl.Location.X {type=MovieSceneScriptingFloatChannel, default=0, thinned_from=24, keys [[1790,0],[4484,0]]}
       lowerarm_r_fk_ctrl.Location.Y {type=MovieSceneScriptingFloatChannel, default=0, thinned_from=24, keys [[1790,0],[4484,0]]}
       lowerarm_r_fk_ctrl.Location.Z {type=MovieSceneScriptingFloatChannel, default=0, thinned_from=24, keys [[1790,0],[4484,0]]}
       lowerarm_r_fk_ctrl.Rotation.X {
        type=MovieSceneScriptingFloatChannel
        default=0
        thinned_from=26
        keys [[1790,0],[2186,0],[2204,-0.0296],[2217,0],[3541,0],[3686,-0.0769],[3754,-0.0122],[3834,-0.0122],[3845,-0.0113],[3972,0.084],[3982,-0.0377],[4011,0.0778],[4364,-0.0122],[4366,-0.0231],[4409,-0.0122],[4484,0]]
       }
       lowerarm_r_fk_ctrl.Rotation.Y {
        type=MovieSceneScriptingFloatChannel
        default=0
        thinned_from=26
        keys [[1790,0],[2186,0],[2204,0.014],[2217,0],[3541,0],[3686,0.0389],[3754,0.0066],[3834,0.0066],[3845,0.0064],[3972,-0.0123],[3982,0.0213],[4011,-0.0107],[4364,0.0066],[4366,0.0022],[4409,0.0066],[4484,0]]
       }
       lowerarm_r_fk_ctrl.Rotation.Z {
        type=MovieSceneScriptingFloatChannel
        default=0
        thinned_from=37
        keys [[1790,0],[2186,0],[2204,15.1826],[2217,0],[3541,0],[3554,2.3227],[3581,74.018],[3585,75.374],[3589,73.9251],[3593,33.3964],[3597,9.5114],[3604,6.0054],[3626,13.9511],[3673,1.6877],[3682,0.2856],[3764,-0.5064],[3792,-0.1067],[3811,9.8772],[3820,16.6097],[3830,30.8136],[3847,9.9555],[3899,16.3364],[3907,16.9148],[3937,8.9529],[3966,4.9852],[4121,-0.9688],[4335,-0.4566],[4484,0]]
       }
       lowerarm_r_fk_ctrl.Scale.X {type=MovieSceneScriptingFloatChannel, default=0, const [0,24]}
       lowerarm_r_fk_ctrl.Scale.Y {type=MovieSceneScriptingFloatChannel, default=0, const [0,24]}
       lowerarm_r_fk_ctrl.Scale.Z {type=MovieSceneScriptingFloatChannel, default=0, const [0,24]}
       hand_r_fk_ctrl.Location.X {type=MovieSceneScriptingFloatChannel, default=0, thinned_from=23, keys [[1790,0],[4484,0]]}
       hand_r_fk_ctrl.Location.Y {type=MovieSceneScriptingFloatChannel, default=0, thinned_from=23, keys [[1790,0],[4484,0]]}
       hand_r_fk_ctrl.Location.Z {type=MovieSceneScriptingFloatChannel, default=0, thinned_from=23, keys [[1790,0],[4484,0]]}
       hand_r_fk_ctrl.Rotation.X {
        type=MovieSceneScriptingFloatChannel
        default=0
        thinned_from=28
        keys [[1790,0],[2964,0],[3077,-22.0347],[3104,-22.7303],[3250,-5.9085],[3568,0],[3586,-111.1951],[3612,-150.1052],[3772,-106.9331],[3849,-40.9065],[3900,-46.6959],[3929,-49.7167],[4023,-61.5339],[4038,-60.8158],[4067,-55.0213],[4108,-55.7352],[4364,-40.9065],[4409,-40.9065],[4484,0]]
       }
       hand_r_fk_ctrl.Rotation.Y {
        type=MovieSceneScriptingFloatChannel
        default=0
        thinned_from=29
        keys [[1790,0],[3568,0],[3593,-4.542],[3606,-44.137],[3648,-35.3502],[3742,-42.3544],[3796,-13.2419],[3802,-12.0864],[3900,-16.6115],[3930,-17.1494],[3989,-20.5277],[4023,-20.3693],[4037,-15.1726],[4067,-25.2651],[4107,-34.4435],[4364,-41.0667],[4409,-41.0667],[4484,0]]
       }
       hand_r_fk_ctrl.Rotation.Z {
        type=MovieSceneScriptingFloatChannel
        default=0
        thinned_from=25
        keys [[1790,0],[3568,0],[3694,5.8671],[3737,6.7217],[3776,6.3945],[3838,4.4314],[3900,-18.8283],[3966,1.579],[4017,16.6772],[4041,20.5589],[4093,11.6625],[4196,-16.1808],[4409,-16.1808],[4484,0]]
       }
       hand_r_fk_ctrl.Scale.X {type=MovieSceneScriptingFloatChannel, default=0, const [0,23]}
       hand_r_fk_ctrl.Scale.Y {type=MovieSceneScriptingFloatChannel, default=0, const [0,23]}
       hand_r_fk_ctrl.Scale.Z {type=MovieSceneScriptingFloatChannel, default=0, const [0,23]}
       thumb_01_r_ctrl.Location.X {type=MovieSceneScriptingFloatChannel, default=0, thinned_from=14, keys [[1790,0],[3695,0]]}
       thumb_01_r_ctrl.Location.Y {type=MovieSceneScriptingFloatChannel, default=0, thinned_from=14, keys [[1790,0],[3695,0]]}
       thumb_01_r_ctrl.Location.Z {type=MovieSceneScriptingFloatChannel, default=0, thinned_from=14, keys [[1790,0],[3695,0]]}
       thumb_01_r_ctrl.Rotation.X {
        type=MovieSceneScriptingFloatChannel
        default=0
        thinned_from=18
        keys [[1790,0.3084],[2217,0.3084],[2529,0.2821],[2556,0.2786],[2754,0.2379],[2803,-0.0987],[2964,-0.0987],[3199,-0.0834],[3509,-0.0184],[3637,0.2782],[3651,0.0362],[3695,0.5853],[3760,0.6694]]
       }
       thumb_01_r_ctrl.Rotation.Y {
        type=MovieSceneScriptingFloatChannel
        default=0
        thinned_from=18
        keys [[1790,2.1181],[2217,2.1181],[2529,1.9373],[2556,1.9138],[2754,1.634],[2803,12.9534],[2964,12.9534],[3199,10.9437],[3509,2.4194],[3637,1.8468],[3651,1.1179],[3695,3.9349],[3760,4.5094]]
       }
       thumb_01_r_ctrl.Rotation.Z {
        type=MovieSceneScriptingFloatChannel
        default=0
        thinned_from=18
        keys [[1790,8.9053],[2217,8.9053],[2529,8.1453],[2556,8.0464],[2754,6.8702],[2803,4.1897],[2964,4.1897],[3199,3.5397],[3509,0.7825],[3637,5.5408],[3651,1.236],[3695,11.8058],[3760,13.5293]]
       }
       thumb_01_r_ctrl.Scale.X {type=MovieSceneScriptingFloatChannel, default=0, const [0,14]}
       thumb_01_r_ctrl.Scale.Y {type=MovieSceneScriptingFloatChannel, default=0, const [0,14]}
       thumb_01_r_ctrl.Scale.Z {type=MovieSceneScriptingFloatChannel, default=0, const [0,14]}
       thumb_02_r_ctrl.Location.X {type=MovieSceneScriptingFloatChannel, default=0, thinned_from=14, keys [[1790,0],[3695,0]]}
       thumb_02_r_ctrl.Location.Y {type=MovieSceneScriptingFloatChannel, default=0, thinned_from=14, keys [[1790,0],[3695,0]]}
       thumb_02_r_ctrl.Location.Z {type=MovieSceneScriptingFloatChannel, default=0, thinned_from=14, keys [[1790,0],[3695,0]]}
       thumb_02_r_ctrl.Rotation.X {
        type=MovieSceneScriptingFloatChannel
        default=0
        thinned_from=18
        keys [[1790,-0.4306],[2217,-0.4306],[2529,-0.3939],[2556,-0.3891],[2754,-0.3322],[2803,-0.92],[2964,-0.92],[3199,-0.7773],[3509,-0.1718],[3637,0.4684],[3651,0.1391],[3695,0.9288],[3760,1.0522]]
       }
       thumb_02_r_ctrl.Rotation.Y {
        type=MovieSceneScriptingFloatChannel
        default=0
        thinned_from=18
        keys [[1790,-0.2115],[2217,-0.2115],[2529,-0.1935],[2556,-0.1911],[2754,-0.1632],[2803,-0.6317],[2964,-0.6317],[3199,-0.5337],[3509,-0.118],[3637,0.1716],[3651,0.0782],[3695,0.3181],[3760,0.3562]]
       }
       thumb_02_r_ctrl.Rotation.Z {
        type=MovieSceneScriptingFloatChannel
        default=0
        thinned_from=18
        keys [[1790,-6.2905],[2217,-6.2905],[2529,-5.7537],[2556,-5.6838],[2754,-4.853],[2803,-26.7461],[2964,-26.7461],[3199,-22.5963],[3509,-4.9955],[3637,6.9366],[3651,2.2369],[3695,12.7671],[3760,14.2762]]
       }
       thumb_02_r_ctrl.Scale.X {type=MovieSceneScriptingFloatChannel, default=0, const [0,14]}
       thumb_02_r_ctrl.Scale.Y {type=MovieSceneScriptingFloatChannel, default=0, const [0,14]}
       thumb_02_r_ctrl.Scale.Z {type=MovieSceneScriptingFloatChannel, default=0, const [0,14]}
       thumb_03_r_ctrl.Location.X {type=MovieSceneScriptingFloatChannel, default=0, const [0,14]}
       thumb_03_r_ctrl.Location.Y {type=MovieSceneScriptingFloatChannel, default=0, const [0,14]}
       thumb_03_r_ctrl.Location.Z {type=MovieSceneScriptingFloatChannel, default=0, const [0,14]}
       thumb_03_r_ctrl.Rotation.X {
        type=MovieSceneScriptingFloatChannel
        default=0
        thinned_from=16
        keys [[1790,0.3803],[2217,0.3803],[2529,0.3479],[2556,0.3436],[2754,0.2934],[2944,0.241],[2964,0.241],[3199,0.1688],[3545,0.034],[3637,-0.106],[3651,-0.0216],[3695,-0.2874],[3760,-0.3401]]
       }
       thumb_03_r_ctrl.Rotation.Y {
        type=MovieSceneScriptingFloatChannel
        default=0
        thinned_from=16
        keys [[1790,0.3699],[2217,0.3699],[2529,0.3383],[2556,0.3342],[2754,0.2854],[2944,0.2344],[2964,0.2344],[3199,0.1642],[3545,0.0331],[3637,-0.0578],[3651,-0.0033],[3695,-0.1492],[3760,-0.1754]]
       }
       thumb_03_r_ctrl.Rotation.Z {
        type=MovieSceneScriptingFloatChannel
        default=0
        thinned_from=16
        keys [[1790,-20.0168],[2217,-20.0168],[2529,-18.3086],[2556,-18.0863],[2754,-15.4426],[2944,-12.6836],[2964,-12.6836],[3199,-8.8863],[3545,-1.7904],[3637,4.5124],[3651,0.7208],[3695,12.0616],[3760,14.245]]
       }
       thumb_03_r_ctrl.Scale.X {type=MovieSceneScriptingFloatChannel, default=0, const [0,14]}
       thumb_03_r_ctrl.Scale.Y {type=MovieSceneScriptingFloatChannel, default=0, const [0,14]}
       thumb_03_r_ctrl.Scale.Z {type=MovieSceneScriptingFloatChannel, default=0, const [0,14]}
       index_metacarpal_r_ctrl.Location.X {type=MovieSceneScriptingFloatChannel, default=0, const [0,16]}
       index_metacarpal_r_ctrl.Location.Y {type=MovieSceneScriptingFloatChannel, default=0, thinned_from=16, keys [[1790,0],[4366,0]]}
       index_metacarpal_r_ctrl.Location.Z {type=MovieSceneScriptingFloatChannel, default=0, thinned_from=16, keys [[1790,0],[4366,0]]}
       index_metacarpal_r_ctrl.Rotation.X {
        type=MovieSceneScriptingFloatChannel
        default=0
        thinned_from=17
        keys [[1790,0.628],[2217,0.628],[2529,0.5744],[2556,0.5675],[2754,0.4845],[2944,0.398],[2964,0.398],[3199,0.2788],[3512,0.0562],[3844,0],[4067,0.0164],[4132,0.0188],[4364,0],[4366,0]]
       }
       index_metacarpal_r_ctrl.Rotation.Y {
        type=MovieSceneScriptingFloatChannel
        default=0
        thinned_from=17
        keys [[1790,0.8007],[2217,0.8007],[2529,0.7323],[2556,0.7234],[2754,0.6177],[2944,0.5073],[2964,0.5073],[3199,0.3554],[3512,0.0716],[3844,0],[4067,-3.6317],[4132,-4.1684],[4364,0],[4366,0]]
       }
       index_metacarpal_r_ctrl.Rotation.Z {
        type=MovieSceneScriptingFloatChannel
        default=0
        thinned_from=17
        keys [[1790,7.5399],[2217,7.5399],[2529,6.8965],[2556,6.8127],[2754,5.8169],[2944,4.7776],[2964,4.7776],[3199,3.3473],[3512,0.6744],[3844,0],[4067,0.4087],[4132,0.4697],[4364,0],[4366,0]]
       }
       index_metacarpal_r_ctrl.Scale.X {type=MovieSceneScriptingFloatChannel, default=0, const [0,16]}
       index_metacarpal_r_ctrl.Scale.Y {type=MovieSceneScriptingFloatChannel, default=0, const [0,16]}
       index_metacarpal_r_ctrl.Scale.Z {type=MovieSceneScriptingFloatChannel, default=0, const [0,16]}
       index_01_r_ctrl.Location.X {type=MovieSceneScriptingFloatChannel, default=0, const [0,17]}
       index_01_r_ctrl.Location.Y {type=MovieSceneScriptingFloatChannel, default=0, const [0,17]}
       index_01_r_ctrl.Location.Z {type=MovieSceneScriptingFloatChannel, default=0, const [0,17]}
       index_01_r_ctrl.Rotation.X {
        type=MovieSceneScriptingFloatChannel
        default=0
        thinned_from=21
        keys [[1790,-3.4599],[2217,-3.4599],[2529,-3.1646],[2556,-3.1262],[2754,-2.6692],[2803,15.8233],[2964,15.8233],[3199,13.3683],[3509,2.9554],[3581,0.9542],[3595,-2.027],[3610,0.6483],[3637,7.2232],[3651,14.871],[3695,15.3904],[3760,17.6372]]
       }
       index_01_r_ctrl.Rotation.Y {
        type=MovieSceneScriptingFloatChannel
        default=0
        thinned_from=21
        keys [[1790,-32.1543],[2217,-32.1543],[2529,-29.4101],[2556,-29.0531],[2754,-24.8063],[2803,-16.2734],[2964,-16.2734],[3199,-13.7485],[3509,-3.0395],[3581,-0.9813],[3595,1.8242],[3610,-0.6667],[3637,-13.4716],[3651,-20.2065],[3695,-28.7039],[3760,-32.8944]]
       }
       index_01_r_ctrl.Rotation.Z {
        type=MovieSceneScriptingFloatChannel
        default=0
        thinned_from=21
        keys [[1790,9.6727],[2217,9.6727],[2529,8.8472],[2556,8.7398],[2754,7.4623],[2803,-42.7625],[2964,-42.7625],[3199,-36.1278],[3509,-7.987],[3581,-2.5787],[3595,-51.8152],[3610,-1.752],[3637,6.6584],[3651,1.8399],[3695,12.4351],[3760,13.9417]]
       }
       index_01_r_ctrl.Scale.X {type=MovieSceneScriptingFloatChannel, default=0, const [0,17]}
       index_01_r_ctrl.Scale.Y {type=MovieSceneScriptingFloatChannel, default=0, const [0,17]}
       index_01_r_ctrl.Scale.Z {type=MovieSceneScriptingFloatChannel, default=0, const [0,17]}
       index_02_r_ctrl.Location.X {type=MovieSceneScriptingFloatChannel, default=0, thinned_from=17, keys [[1790,0],[3695,0]]}
       index_02_r_ctrl.Location.Y {type=MovieSceneScriptingFloatChannel, default=0, thinned_from=17, keys [[1790,0],[3695,0]]}
       index_02_r_ctrl.Location.Z {type=MovieSceneScriptingFloatChannel, default=0, thinned_from=17, keys [[1790,0],[3695,0]]}
       index_02_r_ctrl.Rotation.X {
        type=MovieSceneScriptingFloatChannel
        default=0
        thinned_from=21
        keys [[1790,0],[2754,0],[2803,1.15],[2964,1.15],[3199,0.9716],[3509,0.2148],[3581,0.0693],[3595,2.821],[3610,0.0471],[3637,-0.7258],[3651,-1.0813],[3695,-1.4992],[3760,-1.7098]]
       }
       index_02_r_ctrl.Rotation.Y {
        type=MovieSceneScriptingFloatChannel
        default=0
        thinned_from=21
        keys [[1790,0],[2754,0],[2803,2.557],[2964,2.557],[3199,2.1603],[3509,0.4776],[3581,0.1542],[3595,6.298],[3610,0.1048],[3637,3.0353],[3651,5.693],[3695,6.4672],[3760,7.4114]]
       }
       index_02_r_ctrl.Rotation.Z {
        type=MovieSceneScriptingFloatChannel
        default=0
        thinned_from=21
        keys [[1790,0],[2754,0],[2803,-47.0693],[2964,-47.0693],[3199,-39.7664],[3509,-8.7915],[3581,-2.8385],[3595,-51.7248],[3610,-1.9285],[3637,7.1136],[3651,2.6208],[3695,13.2285],[3760,14.8199]]
       }
       index_02_r_ctrl.Scale.X {type=MovieSceneScriptingFloatChannel, default=0, const [0,17]}
       index_02_r_ctrl.Scale.Y {type=MovieSceneScriptingFloatChannel, default=0, const [0,17]}
       index_02_r_ctrl.Scale.Z {type=MovieSceneScriptingFloatChannel, default=0, const [0,17]}
       index_03_r_ctrl.Location.X {type=MovieSceneScriptingFloatChannel, default=0, const [0,17]}
       index_03_r_ctrl.Location.Y {type=MovieSceneScriptingFloatChannel, default=0, const [0,17]}
       index_03_r_ctrl.Location.Z {type=MovieSceneScriptingFloatChannel, default=0, const [0,17]}
       index_03_r_ctrl.Rotation.X {
        type=MovieSceneScriptingFloatChannel
        default=0
        thinned_from=21
        keys [[1790,0.2893],[2217,0.2893],[2529,0.2647],[2556,0.2614],[2754,0.2232],[2803,0.159],[2964,0.159],[3199,0.1343],[3509,0.0297],[3581,0.0096],[3595,0.3467],[3610,0.0065],[3637,-0.0791],[3651,-0.0245],[3695,-0.162],[3760,-0.1845]]
       }
       index_03_r_ctrl.Rotation.Y {
        type=MovieSceneScriptingFloatChannel
        default=0
        thinned_from=21
        keys [[1790,0.339],[2217,0.339],[2529,0.3101],[2556,0.3063],[2754,0.2615],[2803,0.5062],[2964,0.5062],[3199,0.4277],[3509,0.0946],[3581,0.0305],[3595,0.6044],[3610,0.0207],[3637,-0.0509],[3651,-0.0205],[3695,-0.0878],[3760,-0.097]]
       }
       index_03_r_ctrl.Rotation.Z {
        type=MovieSceneScriptingFloatChannel
        default=0
        thinned_from=21
        keys [[1790,-31.9233],[2217,-31.9233],[2529,-29.1989],[2556,-28.8444],[2754,-24.6281],[2803,-46.8739],[2964,-46.8739],[3199,-39.6013],[3509,-8.755],[3581,-2.8267],[3595,-52.0013],[3610,-1.9205],[3637,6.8729],[3651,2.1999],[3695,12.7237],[3760,14.2428]]
       }
       index_03_r_ctrl.Scale.X {type=MovieSceneScriptingFloatChannel, default=0, const [0,17]}
       index_03_r_ctrl.Scale.Y {type=MovieSceneScriptingFloatChannel, default=0, const [0,17]}
       index_03_r_ctrl.Scale.Z {type=MovieSceneScriptingFloatChannel, default=0, const [0,17]}
       middle_metacarpal_r_ctrl.Location.X {type=MovieSceneScriptingFloatChannel, default=0, thinned_from=16, keys [[1790,0],[4366,0]]}
       middle_metacarpal_r_ctrl.Location.Y {type=MovieSceneScriptingFloatChannel, default=0, thinned_from=16, keys [[1790,0],[4366,0]]}
       middle_metacarpal_r_ctrl.Location.Z {type=MovieSceneScriptingFloatChannel, default=0, thinned_from=16, keys [[1790,0],[4366,0]]}
       middle_metacarpal_r_ctrl.Rotation.X {
        type=MovieSceneScriptingFloatChannel
        default=0
        thinned_from=16
        keys [[1790,-0.0985],[2217,-0.0985],[2529,-0.0901],[2556,-0.089],[2754,-0.076],[2944,-0.0624],[2964,-0.0624],[3199,-0.0437],[3509,-0.0088],[3841,0],[4366,0]]
       }
       middle_metacarpal_r_ctrl.Rotation.Y {
        type=MovieSceneScriptingFloatChannel
        default=0
        thinned_from=16
        keys [[1790,-0.0616],[2217,-0.0616],[2529,-0.0564],[2556,-0.0557],[2754,-0.0475],[2944,-0.039],[2964,-0.039],[3199,-0.0274],[3509,-0.0055],[3841,0],[4366,0]]
       }
       middle_metacarpal_r_ctrl.Rotation.Z {
        type=MovieSceneScriptingFloatChannel
        default=0
        thinned_from=16
        keys [[1790,-3.2289],[2217,-3.2289],[2529,-2.9533],[2556,-2.9175],[2754,-2.491],[2944,-2.0459],[2964,-2.0459],[3199,-1.4334],[3509,-0.2888],[3841,0],[4366,0]]
       }
       middle_metacarpal_r_ctrl.Scale.X {type=MovieSceneScriptingFloatChannel, default=0, const [0,16]}
       middle_metacarpal_r_ctrl.Scale.Y {type=MovieSceneScriptingFloatChannel, default=0, const [0,16]}
       middle_metacarpal_r_ctrl.Scale.Z {type=MovieSceneScriptingFloatChannel, default=0, const [0,16]}
       middle_01_r_ctrl.Location.X {type=MovieSceneScriptingFloatChannel, default=0, thinned_from=17, keys [[1790,0],[3695,0]]}
       middle_01_r_ctrl.Location.Y {type=MovieSceneScriptingFloatChannel, default=0, thinned_from=17, keys [[1790,0],[3695,0]]}
       middle_01_r_ctrl.Location.Z {type=MovieSceneScriptingFloatChannel, default=0, thinned_from=17, keys [[1790,0],[3695,0]]}
       middle_01_r_ctrl.Rotation.X {
        type=MovieSceneScriptingFloatChannel
        default=0
        thinned_from=21
        keys [[1790,-2.8432],[2217,-2.8432],[2529,-2.6006],[2556,-2.569],[2754,-2.1935],[2803,4.9667],[2964,4.9667],[3199,4.1961],[3509,0.9277],[3581,0.2995],[3595,-3.676],[3610,0.2035],[3637,0.5651],[3651,2.3852],[3695,1.2041],[3760,1.3799]]
       }
       middle_01_r_ctrl.Rotation.Y {
        type=MovieSceneScriptingFloatChannel
        default=0
        thinned_from=21
        keys [[1790,-7.8125],[2217,-7.8125],[2529,-7.1458],[2556,-7.059],[2754,-6.0272],[2803,-8.1244],[2964,-8.1244],[3199,-6.8639],[3509,-1.5175],[3581,-0.4899],[3595,-2.2574],[3610,-0.3329],[3637,-9.0906],[3651,-15.077],[3695,-19.3692],[3760,-22.1969]]
       }
       middle_01_r_ctrl.Rotation.Z {
        type=MovieSceneScriptingFloatChannel
        default=0
        thinned_from=21
        keys [[1790,30.9221],[2217,30.9221],[2529,28.2831],[2556,27.9398],[2754,23.8557],[2803,-24.8145],[2964,-24.8145],[3199,-20.9645],[3509,-4.6348],[3581,-1.4964],[3595,-51.0555],[3610,-1.0167],[3637,6.2389],[3651,1.5063],[3695,12.2767],[3760,13.8898]]
       }
       middle_01_r_ctrl.Scale.X {type=MovieSceneScriptingFloatChannel, default=0, const [0,17]}
       middle_01_r_ctrl.Scale.Y {type=MovieSceneScriptingFloatChannel, default=0, const [0,17]}
       middle_01_r_ctrl.Scale.Z {type=MovieSceneScriptingFloatChannel, default=0, const [0,17]}
       middle_02_r_ctrl.Location.X {type=MovieSceneScriptingFloatChannel, default=0, const [0,17]}
       middle_02_r_ctrl.Location.Y {type=MovieSceneScriptingFloatChannel, default=0, const [0,17]}
       middle_02_r_ctrl.Location.Z {type=MovieSceneScriptingFloatChannel, default=0, const [0,17]}
       middle_02_r_ctrl.Rotation.X {
        type=MovieSceneScriptingFloatChannel
        default=0
        thinned_from=21
        keys [[1790,0],[2754,0],[2803,0.579],[2964,0.579],[3199,0.4891],[3509,0.1081],[3581,0.0349],[3595,1.4408],[3610,0.0237],[3637,-0.3223],[3651,-0.0944],[3695,-0.663],[3760,-0.7556]]
       }
       middle_02_r_ctrl.Rotation.Y {
        type=MovieSceneScriptingFloatChannel
        default=0
        thinned_from=21
        keys [[1790,0],[2754,0],[2803,0.9787],[2964,0.9787],[3199,0.8269],[3509,0.1828],[3581,0.059],[3595,2.5166],[3610,0.0401],[3637,-0.2008],[3651,-0.0816],[3695,-0.3877],[3760,-0.4372]]
       }
       middle_02_r_ctrl.Rotation.Z {
        type=MovieSceneScriptingFloatChannel
        default=0
        thinned_from=21
        keys [[1790,0],[2754,0],[2803,-47.1156],[2964,-47.1156],[3199,-39.8055],[3509,-8.8001],[3581,-2.8412],[3595,-51.9764],[3610,-1.9304],[3637,6.8833],[3651,2.2044],[3695,12.7359],[3760,14.2551]]
       }
       middle_02_r_ctrl.Scale.X {type=MovieSceneScriptingFloatChannel, default=0, const [0,17]}
       middle_02_r_ctrl.Scale.Y {type=MovieSceneScriptingFloatChannel, default=0, const [0,17]}
       middle_02_r_ctrl.Scale.Z {type=MovieSceneScriptingFloatChannel, default=0, const [0,17]}
       middle_03_r_ctrl.Location.X {type=MovieSceneScriptingFloatChannel, default=0, thinned_from=17, keys [[1790,0],[3695,0]]}
       middle_03_r_ctrl.Location.Y {type=MovieSceneScriptingFloatChannel, default=0, thinned_from=17, keys [[1790,0],[3695,0]]}
       middle_03_r_ctrl.Location.Z {type=MovieSceneScriptingFloatChannel, default=0, thinned_from=17, keys [[1790,0],[3695,0]]}
       middle_03_r_ctrl.Rotation.X {
        type=MovieSceneScriptingFloatChannel
        default=0
        thinned_from=21
        keys [[1790,0],[2754,0],[2803,0.6879],[2964,0.6879],[3199,0.5812],[3509,0.1285],[3581,0.0415],[3595,1.7877],[3610,0.0282],[3637,-0.1887],[3651,-0.0509],[3695,-0.374],[3760,-0.4236]]
       }
       middle_03_r_ctrl.Rotation.Y {
        type=MovieSceneScriptingFloatChannel
        default=0
        thinned_from=21
        keys [[1790,0],[2754,0],[2803,0.0099],[2964,0.0099],[3199,0.0083],[3509,0.0018],[3581,0.0006],[3595,-0.0392],[3610,0.0004],[3637,0.1292],[3651,0.0535],[3695,0.2753],[3760,0.3155]]
       }
       middle_03_r_ctrl.Rotation.Z {
        type=MovieSceneScriptingFloatChannel
        default=0
        thinned_from=21
        keys [[1790,0],[2754,0],[2803,-47.1162],[2964,-47.1162],[3199,-39.806],[3509,-8.8002],[3581,-2.8413],[3595,-52.044],[3610,-1.9304],[3637,6.8791],[3651,2.2033],[3695,12.7271],[3760,14.2449]]
       }
       middle_03_r_ctrl.Scale.X {type=MovieSceneScriptingFloatChannel, default=0, const [0,17]}
       middle_03_r_ctrl.Scale.Y {type=MovieSceneScriptingFloatChannel, default=0, const [0,17]}
       middle_03_r_ctrl.Scale.Z {type=MovieSceneScriptingFloatChannel, default=0, const [0,17]}
       ring_metacarpal_r_ctrl.Location.X {type=MovieSceneScriptingFloatChannel, default=0, thinned_from=16, keys [[1790,0],[4366,0]]}
       ring_metacarpal_r_ctrl.Location.Y {type=MovieSceneScriptingFloatChannel, default=0, thinned_from=16, keys [[1790,0],[4366,0]]}
       ring_metacarpal_r_ctrl.Location.Z {type=MovieSceneScriptingFloatChannel, default=0, thinned_from=16, keys [[1790,0],[4366,0]]}
       ring_metacarpal_r_ctrl.Rotation.X {type=MovieSceneScriptingFloatChannel, default=0, const [0,16]}
       ring_metacarpal_r_ctrl.Rotation.Y {type=MovieSceneScriptingFloatChannel, default=0, const [0,16]}
       ring_metacarpal_r_ctrl.Rotation.Z {type=MovieSceneScriptingFloatChannel, default=0, const [0,16]}
       ring_metacarpal_r_ctrl.Scale.X {type=MovieSceneScriptingFloatChannel, default=0, const [0,16]}
       ring_metacarpal_r_ctrl.Scale.Y {type=MovieSceneScriptingFloatChannel, default=0, const [0,16]}
       ring_metacarpal_r_ctrl.Scale.Z {type=MovieSceneScriptingFloatChannel, default=0, const [0,16]}
       ring_01_r_ctrl.Location.X {type=MovieSceneScriptingFloatChannel, default=0, const [0,17]}
       ring_01_r_ctrl.Location.Y {type=MovieSceneScriptingFloatChannel, default=0, const [0,17]}
       ring_01_r_ctrl.Location.Z {type=MovieSceneScriptingFloatChannel, default=0, const [0,17]}
       ring_01_r_ctrl.Rotation.X {
        type=MovieSceneScriptingFloatChannel
        default=0
        thinned_from=21
        keys [[1790,-12.8691],[2217,-12.8691],[2529,-11.7708],[2556,-11.6279],[2754,-9.9282],[2803,-5.9924],[2964,-5.9924],[3199,-5.0626],[3509,-1.1192],[3581,-0.3614],[3595,-3.8141],[3610,-0.2455],[3637,-0.2794],[3651,-0.1194],[3695,-0.5954],[3760,-0.6823]]
       }
       ring_01_r_ctrl.Rotation.Y {
        type=MovieSceneScriptingFloatChannel
        default=0
        thinned_from=21
        keys [[1790,8.1966],[2217,8.1966],[2529,7.4971],[2556,7.406],[2754,6.3235],[2803,-1.7489],[2964,-1.7489],[3199,-1.4775],[3509,-0.3267],[3581,-0.1055],[3595,10.2814],[3610,-0.0717],[3637,-1.2759],[3651,-0.3289],[3695,-2.7185],[3760,-3.1154]]
       }
       ring_01_r_ctrl.Rotation.Z {
        type=MovieSceneScriptingFloatChannel
        default=0
        thinned_from=21
        keys [[1790,29.3086],[2217,29.3086],[2529,26.8073],[2556,26.4819],[2754,22.611],[2803,-25.0787],[2964,-25.0787],[3199,-21.1877],[3509,-4.6841],[3581,-1.5123],[3595,-50.0522],[3610,-1.0275],[3637,6.2553],[3651,1.8413],[3695,12.3007],[3760,13.9154]]
       }
       ring_01_r_ctrl.Scale.X {type=MovieSceneScriptingFloatChannel, default=0, const [0,17]}
       ring_01_r_ctrl.Scale.Y {type=MovieSceneScriptingFloatChannel, default=0, const [0,17]}
       ring_01_r_ctrl.Scale.Z {type=MovieSceneScriptingFloatChannel, default=0, const [0,17]}
       ring_02_r_ctrl.Location.X {type=MovieSceneScriptingFloatChannel, default=0, thinned_from=17, keys [[1790,0],[3695,0]]}
       ring_02_r_ctrl.Location.Y {type=MovieSceneScriptingFloatChannel, default=0, thinned_from=17, keys [[1790,0],[3695,0]]}
       ring_02_r_ctrl.Location.Z {type=MovieSceneScriptingFloatChannel, default=0, thinned_from=17, keys [[1790,0],[3695,0]]}
       ring_02_r_ctrl.Rotation.X {
        type=MovieSceneScriptingFloatChannel
        default=0
        thinned_from=21
        keys [[1790,0],[2754,0],[2803,0.0732],[2964,0.0732],[3199,0.0619],[3509,0.0137],[3581,0.0044],[3595,0.2739],[3610,0.003],[3637,0.0316],[3651,0.014],[3695,0.0672],[3760,0.077]]
       }
       ring_02_r_ctrl.Rotation.Y {
        type=MovieSceneScriptingFloatChannel
        default=0
        thinned_from=21
        keys [[1790,0],[2754,0],[2803,-0.2894],[2964,-0.2894],[3199,-0.2445],[3509,-0.0541],[3581,-0.0175],[3595,-0.9125],[3610,-0.0119],[3637,0.1095],[3651,0.0248],[3695,0.2214],[3760,0.2516]]
       }
       ring_02_r_ctrl.Rotation.Z {
        type=MovieSceneScriptingFloatChannel
        default=0
        thinned_from=21
        keys [[1790,0],[2754,0],[2803,-47.1213],[2964,-47.1213],[3199,-39.8103],[3509,-8.8012],[3581,-2.8416],[3595,-52.0084],[3610,-1.9306],[3637,6.8772],[3651,2.2034],[3695,12.7227],[3760,14.2398]]
       }
       ring_02_r_ctrl.Scale.X {type=MovieSceneScriptingFloatChannel, default=0, const [0,17]}
       ring_02_r_ctrl.Scale.Y {type=MovieSceneScriptingFloatChannel, default=0, const [0,17]}
       ring_02_r_ctrl.Scale.Z {type=MovieSceneScriptingFloatChannel, default=0, const [0,17]}
       ring_03_r_ctrl.Location.X {type=MovieSceneScriptingFloatChannel, default=0, thinned_from=17, keys [[1790,0],[3695,0]]}
       ring_03_r_ctrl.Location.Y {type=MovieSceneScriptingFloatChannel, default=0, thinned_from=17, keys [[1790,0],[3695,0]]}
       ring_03_r_ctrl.Location.Z {type=MovieSceneScriptingFloatChannel, default=0, thinned_from=17, keys [[1790,0],[3695,0]]}
       ring_03_r_ctrl.Rotation.X {
        type=MovieSceneScriptingFloatChannel
        default=0
        thinned_from=21
        keys [[1790,0.0106],[2217,0.0106],[2556,0.0095],[2754,0.0081],[2803,0.1425],[2964,0.1425],[3199,0.1204],[3509,0.0266],[3581,0.0086],[3595,0.1346],[3610,0.0058],[3637,0.0125],[3651,0.005],[3695,0.0265],[3760,0.0304]]
       }
       ring_03_r_ctrl.Rotation.Y {
        type=MovieSceneScriptingFloatChannel
        default=0
        thinned_from=21
        keys [[1790,-0.1863],[2217,-0.1863],[2529,-0.1704],[2556,-0.1683],[2754,-0.1437],[2803,-0.2276],[2964,-0.2276],[3199,-0.1923],[3509,-0.0425],[3581,-0.0137],[3595,-0.3997],[3610,-0.0093],[3637,0.0489],[3651,0.0095],[3695,0.0949],[3760,0.1071]]
       }
       ring_03_r_ctrl.Rotation.Z {
        type=MovieSceneScriptingFloatChannel
        default=0
        thinned_from=21
        keys [[1790,-22.2124],[2217,-22.2124],[2529,-20.3168],[2556,-20.0701],[2754,-17.1364],[2803,-63.4507],[2964,-63.4507],[3199,-53.6062],[3509,-11.8511],[3581,-3.8263],[3595,-52.7981],[3610,-2.5996],[3637,7.2401],[3651,2.4482],[3695,12.8271],[3760,14.2416]]
       }
       ring_03_r_ctrl.Scale.X {type=MovieSceneScriptingFloatChannel, default=0, const [0,17]}
       ring_03_r_ctrl.Scale.Y {type=MovieSceneScriptingFloatChannel, default=0, const [0,17]}
       ring_03_r_ctrl.Scale.Z {type=MovieSceneScriptingFloatChannel, default=0, const [0,17]}
       pinky_metacarpal_r_ctrl.Location.X {type=MovieSceneScriptingFloatChannel, default=0, const [0,16]}
       pinky_metacarpal_r_ctrl.Location.Y {type=MovieSceneScriptingFloatChannel, default=0, const [0,16]}
       pinky_metacarpal_r_ctrl.Location.Z {type=MovieSceneScriptingFloatChannel, default=0, const [0,16]}
       pinky_metacarpal_r_ctrl.Rotation.X {
        type=MovieSceneScriptingFloatChannel
        default=0
        thinned_from=16
        keys [[1790,-0.0971],[2217,-0.0971],[2529,-0.0888],[2556,-0.0878],[2754,-0.0749],[2944,-0.0615],[2964,-0.0615],[3199,-0.0431],[3509,-0.0087],[3841,0],[4366,0]]
       }
       pinky_metacarpal_r_ctrl.Rotation.Y {
        type=MovieSceneScriptingFloatChannel
        default=0
        thinned_from=16
        keys [[1790,1.8007],[2217,1.8007],[2529,1.6471],[2556,1.6271],[2754,1.3892],[2944,1.141],[2964,1.141],[3199,0.7994],[3509,0.1611],[3841,0],[4366,0]]
       }
       pinky_metacarpal_r_ctrl.Rotation.Z {
        type=MovieSceneScriptingFloatChannel
        default=0
        thinned_from=16
        keys [[1790,-6.1178],[2217,-6.1178],[2529,-5.5957],[2556,-5.5278],[2754,-4.7198],[2944,-3.8765],[2964,-3.8765],[3199,-2.716],[3509,-0.5472],[3841,0],[4366,0]]
       }
       pinky_metacarpal_r_ctrl.Scale.X {type=MovieSceneScriptingFloatChannel, default=0, const [0,16]}
       pinky_metacarpal_r_ctrl.Scale.Y {type=MovieSceneScriptingFloatChannel, default=0, const [0,16]}
       pinky_metacarpal_r_ctrl.Scale.Z {type=MovieSceneScriptingFloatChannel, default=0, const [0,16]}
       pinky_01_r_ctrl.Location.X {type=MovieSceneScriptingFloatChannel, default=0, thinned_from=17, keys [[1790,0],[3695,0]]}
       pinky_01_r_ctrl.Location.Y {type=MovieSceneScriptingFloatChannel, default=0, thinned_from=17, keys [[1790,0],[3695,0]]}
       pinky_01_r_ctrl.Location.Z {type=MovieSceneScriptingFloatChannel, default=0, thinned_from=17, keys [[1790,0],[3695,0]]}
       pinky_01_r_ctrl.Rotation.X {
        type=MovieSceneScriptingFloatChannel
        default=0
        thinned_from=21
        keys [[1790,2.9269],[2217,2.9269],[2529,2.6771],[2556,2.6446],[2754,2.258],[2803,-3.6904],[2964,-3.6904],[3199,-3.1178],[3509,-0.6893],[3581,-0.2225],[3595,-9.2767],[3610,-0.1512],[3637,1.2873],[3651,-1.4947],[3695,2.5917],[3760,2.9434]]
       }
       pinky_01_r_ctrl.Rotation.Y {
        type=MovieSceneScriptingFloatChannel
        default=0
        thinned_from=21
        keys [[1790,22.2355],[2217,22.2355],[2529,20.3379],[2556,20.0909],[2754,17.1542],[2803,12.8418],[2964,12.8418],[3199,10.8493],[3509,2.3985],[3581,0.7744],[3595,7.2986],[3610,0.5261],[3637,10.3015],[3651,18.9011],[3695,21.9493],[3760,25.1536]]
       }
       pinky_01_r_ctrl.Rotation.Z {
        type=MovieSceneScriptingFloatChannel
        default=0
        thinned_from=21
        keys [[1790,35.2197],[2217,35.2197],[2529,32.214],[2556,31.8229],[2754,27.1712],[2803,-21.4371],[2964,-21.4371],[3199,-18.111],[3509,-4.0039],[3581,-1.2927],[3595,-51.1402],[3610,-0.8783],[3637,8.8711],[3651,5.5303],[3695,18.0234],[3760,20.4999]]
       }
       pinky_01_r_ctrl.Scale.X {type=MovieSceneScriptingFloatChannel, default=0, const [0,17]}
       pinky_01_r_ctrl.Scale.Y {type=MovieSceneScriptingFloatChannel, default=0, const [0,17]}
       pinky_01_r_ctrl.Scale.Z {type=MovieSceneScriptingFloatChannel, default=0, const [0,17]}
       pinky_02_r_ctrl.Location.X {type=MovieSceneScriptingFloatChannel, default=0, thinned_from=17, keys [[1790,0],[3695,0]]}
       pinky_02_r_ctrl.Location.Y {type=MovieSceneScriptingFloatChannel, default=0, thinned_from=17, keys [[1790,0],[3695,0]]}
       pinky_02_r_ctrl.Location.Z {type=MovieSceneScriptingFloatChannel, default=0, thinned_from=17, keys [[1790,0],[3695,0]]}
       pinky_02_r_ctrl.Rotation.X {
        type=MovieSceneScriptingFloatChannel
        default=0
        thinned_from=21
        keys [[1790,0],[2754,0],[2803,0.3784],[2964,0.3784],[3199,0.3197],[3509,0.0707],[3581,0.0228],[3595,0.9395],[3610,0.0155],[3637,0.0383],[3651,0.017],[3695,0.0816],[3760,0.0935]]
       }
       pinky_02_r_ctrl.Rotation.Y {
        type=MovieSceneScriptingFloatChannel
        default=0
        thinned_from=21
        keys [[1790,0],[2754,0],[2803,-0.8963],[2964,-0.8963],[3199,-0.7572],[3509,-0.1674],[3581,-0.054],[3595,-2.0746],[3610,-0.0367],[3637,0.2791],[3651,0.0756],[3695,0.558],[3760,0.633]]
       }
       pinky_02_r_ctrl.Rotation.Z {
        type=MovieSceneScriptingFloatChannel
        default=0
        thinned_from=21
        keys [[1790,0],[2754,0],[2803,-47.1158],[2964,-47.1158],[3199,-39.8057],[3509,-8.8001],[3581,-2.8413],[3595,-51.9842],[3610,-1.9304],[3637,6.8722],[3651,2.2021],[3695,12.7123],[3760,14.228]]
       }
       pinky_02_r_ctrl.Scale.X {type=MovieSceneScriptingFloatChannel, default=0, const [0,17]}
       pinky_02_r_ctrl.Scale.Y {type=MovieSceneScriptingFloatChannel, default=0, const [0,17]}
       pinky_02_r_ctrl.Scale.Z {type=MovieSceneScriptingFloatChannel, default=0, const [0,17]}
       pinky_03_r_ctrl.Location.X {type=MovieSceneScriptingFloatChannel, default=0, thinned_from=17, keys [[1790,0],[3695,0]]}
       pinky_03_r_ctrl.Location.Y {type=MovieSceneScriptingFloatChannel, default=0, thinned_from=17, keys [[1790,0],[3695,0]]}
       pinky_03_r_ctrl.Location.Z {type=MovieSceneScriptingFloatChannel, default=0, thinned_from=17, keys [[1790,0],[3695,0]]}
       pinky_03_r_ctrl.Rotation.X {
        type=MovieSceneScriptingFloatChannel
        default=0
        thinned_from=21
        keys [[1790,0.7163],[2217,0.7163],[2529,0.6551],[2556,0.6472],[2754,0.5526],[2803,0.9999],[2964,0.9999],[3199,0.8448],[3509,0.1868],[3581,0.0603],[3595,1.0274],[3610,0.041],[3637,-0.0823],[3651,-0.0259],[3695,-0.1374],[3760,-0.1513]]
       }
       pinky_03_r_ctrl.Rotation.Y {
        type=MovieSceneScriptingFloatChannel
        default=0
        thinned_from=21
        keys [[1790,-0.5302],[2217,-0.5302],[2529,-0.4849],[2556,-0.479],[2754,-0.409],[2803,-0.0742],[2964,-0.0742],[3199,-0.0627],[3509,-0.0139],[3581,-0.0045],[3595,-0.556],[3610,-0.003],[3637,0.1228],[3651,0.0334],[3695,0.2585],[3760,0.2957]]
       }
       pinky_03_r_ctrl.Rotation.Z {
        type=MovieSceneScriptingFloatChannel
        default=0
        thinned_from=21
        keys [[1790,-38.8119],[2217,-38.8119],[2529,-35.4996],[2556,-35.0686],[2754,-29.9425],[2803,-75.6566],[2964,-75.6566],[3199,-63.9183],[3509,-14.1309],[3581,-4.5624],[3595,-53.3916],[3610,-3.0997],[3637,7.5101],[3651,2.6309],[3695,12.9024],[3760,14.2398]]
       }
       pinky_03_r_ctrl.Scale.X {type=MovieSceneScriptingFloatChannel, default=0, const [0,17]}
       pinky_03_r_ctrl.Scale.Y {type=MovieSceneScriptingFloatChannel, default=0, const [0,17]}
       pinky_03_r_ctrl.Scale.Z {type=MovieSceneScriptingFloatChannel, default=0, const [0,17]}
       hand_r_ik_ctrl.Location.X {
        type=MovieSceneScriptingFloatChannel
        default=0
        thinned_from=16
        keys [[1790,-2.2062],[2113,1041.7537],[2217,-2.2062],[2516,967.7421],[2523,932.5381],[2529,889.2967],[2556,732.0424],[2754,500.6022],[2944,587.4734],[3199,592.0345],[3509,267.9951],[3841,153.7856],[4067,107.2455],[4364,172.3599],[4366,60.7654]]
       }
       hand_r_ik_ctrl.Location.Y {
        type=MovieSceneScriptingFloatChannel
        default=0
        keys [[1790,-0.7834],[2113,-105.1248],[2217,-0.7834],[2516,-415.9671],[2523,-398.9077],[2529,-376.3524],[2556,-283.0567],[2754,-161.33],[2944,-278.3087],[2964,-278.3087],[3199,-228.1667],[3509,-288.6009],[3841,-263.6385],[4067,-385.8498],[4364,-277.8524],[4366,-628.1564]]
       }
       hand_r_ik_ctrl.Location.Z {
        type=MovieSceneScriptingFloatChannel
        default=0
        thinned_from=16
        keys [[1790,1.7688],[2113,2524.0215],[2217,1.7688],[2516,4788.1431],[2529,4331.9805],[2556,3497.8022],[2754,2311.1812],[2944,3183.2905],[2964,3183.2905],[3199,2880.3779],[3509,2387.3445],[3841,2324.1345],[4067,2347.3323],[4364,2344.8088],[4366,2330.9021]]
       }
       hand_r_ik_ctrl.Rotation.X {
        type=MovieSceneScriptingFloatChannel
        default=0
        keys [[1790,9.386],[2113,10.702],[2217,9.386],[2516,134.9381],[2523,137.5601],[2529,137.2726],[2556,135.1108],[2754,27.9825],[2944,130.2079],[2964,130.2079],[3199,76.059],[3509,73.1761],[3841,145.6496],[4067,22.4547],[4364,128.2583],[4366,222.3809]]
       }
       hand_r_ik_ctrl.Rotation.Y {
        type=MovieSceneScriptingFloatChannel
        default=0
        keys [[1790,-14.4964],[2113,-13.9288],[2217,-14.4964],[2516,-63.168],[2523,-25.194],[2529,-25.8309],[2556,-21.4371],[2754,24.0945],[2944,-19.291],[2964,-19.291],[3199,-28.0084],[3509,-13.9203],[3841,-0.3391],[4067,66.7956],[4364,7.5841],[4366,-36.8038]]
       }
       hand_r_ik_ctrl.Rotation.Z {
        type=MovieSceneScriptingFloatChannel
        default=0
        keys [[1790,1.3595],[2113,-3.8506],[2217,1.3595],[2516,-125.5399],[2523,-127.4444],[2529,-128.3048],[2556,-166.4905],[2754,-264.1559],[2944,-126.526],[2964,-126.526],[3199,-204.4685],[3509,1.1484],[3841,73.8212],[4067,152.4864],[4364,74.1653],[4366,69.1774]]
       }
       hand_r_ik_ctrl.Scale.X {
        type=MovieSceneScriptingFloatChannel
        default=0
        thinned_from=16
        keys [[1790,0],[2217,0],[2516,-0.0346],[2523,0],[2529,-0.0353],[2556,-0.0537],[2754,0.0065],[2944,-0.0475],[2964,-0.0475],[3199,-0.0497],[3509,-0.0178],[3841,0],[4067,-0.0045],[4366,-0.0026]]
       }
       hand_r_ik_ctrl.Scale.Y {
        type=MovieSceneScriptingFloatChannel
        default=0
        thinned_from=16
        keys [[1790,0],[2217,0],[2516,-0.0346],[2523,0],[2529,-0.0353],[2556,-0.0537],[2754,0.0065],[2944,-0.0475],[2964,-0.0475],[3199,-0.0497],[3509,-0.0178],[3841,0],[4067,-0.0045],[4366,-0.0026]]
       }
       hand_r_ik_ctrl.Scale.Z {
        type=MovieSceneScriptingFloatChannel
        default=0
        thinned_from=16
        keys [[1790,0],[2217,0],[2516,-0.0346],[2523,0],[2529,-0.0353],[2556,-0.0537],[2754,0.0065],[2944,-0.0475],[2964,-0.0475],[3199,-0.0497],[3509,-0.0178],[3841,0],[4067,-0.0045],[4366,-0.0026]]
       }
       Softness_2 {type=MovieSceneScriptingFloatChannel, default=0, const [0,16]}
       hand_r_ikorient_ctrl.Rotation.Roll {type=MovieSceneScriptingFloatChannel, default=0, const [0,16]}
       hand_r_ikorient_ctrl.Rotation.Pitch {type=MovieSceneScriptingFloatChannel, default=0, const [0,16]}
       hand_r_ikorient_ctrl.Rotation.Yaw {type=MovieSceneScriptingFloatChannel, default=0, const [0,16]}
       arm_r_pv_ik_ctrl.Location.X {
        type=MovieSceneScriptingFloatChannel
        default=0
        keys [[1790,-0.071],[2113,-0.5203],[2217,-0.071],[2516,-73.1378],[2523,-30.317],[2529,-27.0225],[2556,-23.5164],[2754,-14.8014],[2944,-24.9409],[2964,-24.9409],[3199,-47.8431],[3509,5.0432],[3841,-53.7066],[4067,14.8704],[4364,-49.4816],[4366,253.6819]]
       }
       arm_r_pv_ik_ctrl.Location.Y {
        type=MovieSceneScriptingFloatChannel
        default=0
        thinned_from=16
        keys [[1790,6.0647],[2113,2731.3503],[2217,6.0647],[2516,4802.8628],[2529,4336.1548],[2556,3514.7024],[2754,2340.9353],[2944,3175.3303],[2964,3175.3303],[3199,2894.6365],[3509,2329.9778],[3841,2202.8506],[4067,2200.146],[4364,2221.5615],[4366,2206.1614]]
       }
       arm_r_pv_ik_ctrl.Location.Z {
        type=MovieSceneScriptingFloatChannel
        default=0
        keys [[1790,10.899],[2113,196.2475],[2217,10.899],[2516,1290.1155],[2523,1211.6354],[2529,1154.2878],[2556,917.1862],[2754,580.5941],[2944,912.9149],[2964,912.9149],[3199,797.4828],[3509,868.7269],[3841,919.5526],[4067,957.1972],[4364,915.0555],[4366,1153.2024]]
       }
       arm_r_pv_ik_ctrl.Rotation.X {
        type=MovieSceneScriptingFloatChannel
        default=0
        keys [[1790,18.4923],[2113,18.3007],[2217,18.4923],[2516,9.2933],[2523,6.9943],[2529,7.3528],[2556,7.2557],[2754,10.7119],[2944,1.0023],[2964,1.0023],[3199,115.8271],[3509,131.1264],[3841,20.0088],[4067,-122.1212],[4364,15.3564],[4366,76.6396]]
       }
       arm_r_pv_ik_ctrl.Rotation.Y {
        type=MovieSceneScriptingFloatChannel
        default=0
        keys [[1790,2.3102],[2113,3.3144],[2217,2.3102],[2516,-1.0997],[2523,-0.3683],[2529,-1.8754],[2556,-5.5588],[2754,-5.4809],[2944,-6.694],[2964,-6.694],[3199,5.8729],[3509,7.8173],[3841,2.8963],[4067,15.9909],[4364,-15.2431],[4366,51.1307]]
       }
       arm_r_pv_ik_ctrl.Rotation.Z {
        type=MovieSceneScriptingFloatChannel
        default=0
        keys [[1790,6.1325],[2113,5.3149],[2217,6.1325],[2516,181.2106],[2523,178.6726],[2529,177.5527],[2556,177.8695],[2754,166.8277],[2944,174.0632],[2964,174.0632],[3199,304.6606],[3509,296.1543],[3841,167.1185],[4067,104.6754],[4364,200.253],[4366,251.2132]]
       }
       arm_r_pv_ik_ctrl.Scale.X {type=MovieSceneScriptingFloatChannel, default=0, const [0,16]}
       arm_r_pv_ik_ctrl.Scale.Y {type=MovieSceneScriptingFloatChannel, default=0, const [0,16]}
       arm_r_pv_ik_ctrl.Scale.Z {type=MovieSceneScriptingFloatChannel, default=0, const [0,16]}
       body_ctrl.Location.X {
        type=MovieSceneScriptingFloatChannel
        default=0
        thinned_from=40
        keys [[1790,0],[2523,0],[2535,58.7504],[2538,66.1631],[2550,25.1858],[2553,21.5555],[2573,35.4398],[2586,22.2685],[2596,-2.185],[2608,7.7677],[2621,-0.4054],[2634,0],[2643,-39.154],[2646,-51.6481],[2662,-11.8998],[2666,-9.2482],[2680,-16.2455],[2694,-2.861],[2701,-65.1846],[2706,-80.0834],[2721,-4.6783],[2740,-18.7584],[2755,-2.989],[2762,0.2784],[2944,0],[4366,0]]
       }
       body_ctrl.Location.Y {
        type=MovieSceneScriptingFloatChannel
        default=5.7485
        thinned_from=24
        keys [[1790,5.7485],[2526,5.7485],[2543,40.9821],[2583,41.1197],[2599,80.4332],[2619,82.1124],[2653,116.2824],[2701,116.5378],[2720,155.7881],[2778,157.0854],[3019,155.8298],[4366,155.8298]]
       }
       body_ctrl.Location.Z {
        type=MovieSceneScriptingFloatChannel
        default=0
        thinned_from=38
        keys [[1790,0],[2523,0],[2532,-19.3313],[2536,-21.8183],[2554,-0.1618],[2583,0],[2585,-0.8823],[2591,-33.6668],[2596,-55.1285],[2613,-21.1371],[2622,-24.4884],[2634,-14.404],[2645,-56.0826],[2659,-25.8161],[2667,-18.981],[2679,-1.452],[2697,-0.6492],[2705,-10.2165],[2707,-12.3352],[2718,-8.2688],[2728,0.8203],[2754,0],[4366,0]]
       }
       body_ctrl.Rotation.X {
        type=MovieSceneScriptingFloatChannel
        default=17.4861
        thinned_from=36
        keys [[1790,17.4861],[2300,17.4861],[2338,16.659],[2402,17.4861],[2523,17.4861],[2529,21.4558],[2536,21.9921],[2556,17.4861],[2583,17.4861],[2587,20.9404],[2590,23.1823],[2634,17.4861],[2642,21.1788],[2701,17.4861],[2715,14.7489],[2754,8.9723],[2803,9.9759],[3199,9.9759],[3305,10.0077],[3539,10.0077],[3561,10.0269],[3590,-5.8507],[3621,-7.1257],[3750,-0.0505],[3767,0.1853],[3791,0.2843],[3841,0.1473],[4044,-0.371],[4364,-12.1617],[4366,30.5987]]
       }
       body_ctrl.Rotation.Y {type=MovieSceneScriptingFloatChannel, default=0, thinned_from=27, keys [[1790,0],[2300,0],[2338,-1.6891],[2402,0],[4364,0],[4366,-34.5297]]}
       body_ctrl.Rotation.Z {
        type=MovieSceneScriptingFloatChannel
        default=0
        thinned_from=27
        keys [[1790,0],[2300,0],[2348,-15.1261],[2402,-15.5559],[4364,-15.5559],[4366,78.0319]]
       }
       body_ctrl.Scale.X {type=MovieSceneScriptingFloatChannel, default=0, const [0,26]}
       body_ctrl.Scale.Y {type=MovieSceneScriptingFloatChannel, default=0, const [0,26]}
       body_ctrl.Scale.Z {type=MovieSceneScriptingFloatChannel, default=0, const [0,26]}
       hips_ctrl.Location.X {
        type=MovieSceneScriptingFloatChannel
        default=0
        thinned_from=20
        keys [[1790,0],[2217,0],[2348,0.0669],[2529,0.0645],[2556,0.064],[2754,0.0564],[2944,0.0465],[3003,0.0465],[3199,0.0326],[3509,0.0066],[3554,0.0053],[3612,0.0043],[3793,0.0003],[3841,0],[4366,0]]
       }
       hips_ctrl.Location.Y {
        type=MovieSceneScriptingFloatChannel
        default=0
        thinned_from=20
        keys [[1790,0],[2217,0],[2348,0.847],[2529,0.8168],[2556,0.8097],[2754,0.7144],[2944,0.5893],[3003,0.5893],[3199,0.4133],[3509,0.0833],[3554,0.0672],[3612,0.0539],[3793,0.0042],[3841,0],[4366,0]]
       }
       hips_ctrl.Location.Z {
        type=MovieSceneScriptingFloatChannel
        default=0
        thinned_from=20
        keys [[1790,0],[2217,0],[2348,0.2276],[2529,0.2195],[2556,0.2176],[2754,0.192],[2944,0.1584],[3003,0.1584],[3199,0.1111],[3509,0.0224],[3554,0.0181],[3612,0.0145],[3793,0.0011],[3841,0],[4366,0]]
       }
       hips_ctrl.Rotation.X {
        type=MovieSceneScriptingFloatChannel
        default=0
        thinned_from=33
        keys [[1790,9.8173],[2217,9.8173],[2348,-5.9678],[2516,-5.7624],[2523,-5.7612],[2529,-2.0072],[2556,-7.6623],[2711,-1.7381],[2754,-7.2827],[2803,-2.2216],[2944,-2.2216],[2964,7.7786],[2976,0.6138],[2986,-0.7836],[3003,-2.2216],[3199,-1.8769],[3509,-0.415],[3554,-0.3348],[3579,-34.9425],[3581,-35.4634],[3591,-15.6805],[3603,4.2916],[3615,6.5775],[3729,0.9126],[3793,0.2155],[3807,4.839],[3832,3.0203],[3884,0],[4364,0],[4366,-27.6546]]
       }
       hips_ctrl.Rotation.Y {
        type=MovieSceneScriptingFloatChannel
        default=0
        thinned_from=21
        keys [[1790,0],[2944,0],[2976,9.6439],[3003,0],[3554,0],[3577,2.2689],[3612,0],[4366,0]]
       }
       hips_ctrl.Rotation.Z {
        type=MovieSceneScriptingFloatChannel
        default=0
        thinned_from=21
        keys [[1790,0],[2944,0],[2976,9.3137],[3003,0],[3554,0],[3577,-0.5518],[3612,0],[4366,0]]
       }
       hips_ctrl.Scale.X {type=MovieSceneScriptingFloatChannel, default=0, const [0,19]}
       hips_ctrl.Scale.Y {type=MovieSceneScriptingFloatChannel, default=0, const [0,19]}
       hips_ctrl.Scale.Z {type=MovieSceneScriptingFloatChannel, default=0, const [0,19]}
       hips_tan_ctrl.Location.X {
        type=MovieSceneScriptingFloatChannel
        default=0
        keys [[1790,0.2916],[2113,0.2488],[2217,0.2916],[2516,-0.1553],[2523,-0.1401],[2529,-0.1725],[2556,-0.1219],[2754,-0.007],[2944,0.0164],[2964,0.0164],[3199,0.4384],[3509,0.6791],[3841,0.4201],[4067,0.1754],[4364,0.7715],[4366,4.1417]]
       }
       hips_tan_ctrl.Location.Y {
        type=MovieSceneScriptingFloatChannel
        default=0
        keys [[1790,0.3586],[2113,0.3668],[2217,0.3586],[2516,-0.2419],[2523,-1.669],[2529,-1.3682],[2556,-1.3392],[2754,0.2601],[2944,0.2868],[2964,0.2868],[3199,-0.1877],[3509,0.6437],[3841,0.7316],[4067,1.7916],[4364,0.4434],[4366,9.0922]]
       }
       hips_tan_ctrl.Location.Z {
        type=MovieSceneScriptingFloatChannel
        default=0
        keys [[1790,-0.0307],[2113,-0.0304],[2217,-0.0307],[2516,-0.2431],[2523,-0.6817],[2529,-0.589],[2556,-0.5895],[2754,-0.5581],[2944,-0.5462],[2964,-0.5462],[3199,-0.8021],[3509,-0.3361],[3841,-0.04],[4067,-0.0338],[4364,0.1047],[4366,-6.4226]]
       }
       hips_tan_ctrl.Rotation.X {type=MovieSceneScriptingFloatChannel, default=0, const [0,16]}
       hips_tan_ctrl.Rotation.Y {type=MovieSceneScriptingFloatChannel, default=0, const [0,16]}
       hips_tan_ctrl.Rotation.Z {type=MovieSceneScriptingFloatChannel, default=0, const [0,16]}
       hips_tan_ctrl.Scale.X {type=MovieSceneScriptingFloatChannel, default=0, const [0,16]}
       hips_tan_ctrl.Scale.Y {type=MovieSceneScriptingFloatChannel, default=0, const [0,16]}
       hips_tan_ctrl.Scale.Z {type=MovieSceneScriptingFloatChannel, default=0, const [0,16]}
       spine_01_ctrl.Location.X {type=MovieSceneScriptingFloatChannel, default=0, const [0,16]}
       spine_01_ctrl.Location.Y {type=MovieSceneScriptingFloatChannel, default=0, const [0,16]}
       spine_01_ctrl.Location.Z {type=MovieSceneScriptingFloatChannel, default=0, const [0,16]}
       spine_01_ctrl.Rotation.X {
        type=MovieSceneScriptingFloatChannel
        default=0
        thinned_from=20
        keys [[1790,0],[2113,0],[2217,-11.8162],[2289,-10.9223],[2342,-11.5169],[2523,-11.0953],[2529,-10.3594],[2556,-10.9858],[2754,-10.4414],[2803,-9.2982],[2964,-9.2982],[3199,-7.8556],[3305,-15.5369],[4366,-15.5369]]
       }
       spine_01_ctrl.Rotation.Y {
        type=MovieSceneScriptingFloatChannel
        default=0
        thinned_from=19
        keys [[1790,0],[2217,0],[2289,4.0962],[2342,0],[2523,0],[2529,-4.3185],[2556,0],[2754,0],[2803,3.0045],[2964,3.0045],[3199,2.5384],[3509,0.5612],[4366,0.5612]]
       }
       spine_01_ctrl.Rotation.Z {
        type=MovieSceneScriptingFloatChannel
        default=0
        thinned_from=19
        keys [[1790,0],[2217,0],[2289,24.5987],[2342,0],[2523,0],[2529,-17.446],[2556,0],[2754,0],[2803,-7.1638],[2964,-7.1638],[3199,-6.0523],[3509,-1.338],[4366,-1.338]]
       }
       spine_01_ctrl.Scale.X {type=MovieSceneScriptingFloatChannel, default=0, const [0,16]}
       spine_01_ctrl.Scale.Y {type=MovieSceneScriptingFloatChannel, default=0, const [0,16]}
       spine_01_ctrl.Scale.Z {type=MovieSceneScriptingFloatChannel, default=0, const [0,16]}
       spine_02_ctrl.Location.X {type=MovieSceneScriptingFloatChannel, default=0, thinned_from=18, keys [[1790,0],[4366,0],[4372,13.2359]]}
       spine_02_ctrl.Location.Y {type=MovieSceneScriptingFloatChannel, default=0, thinned_from=18, keys [[1790,0],[4366,0],[4372,-2.3003]]}
       spine_02_ctrl.Location.Z {type=MovieSceneScriptingFloatChannel, default=0, thinned_from=18, keys [[1790,0],[4366,0],[4372,-2.5777]]}
       spine_02_ctrl.Rotation.X {
        type=MovieSceneScriptingFloatChannel
        default=0
        thinned_from=24
        keys [[1790,-7.9916],[2217,-7.9916],[2516,-7.3468],[2523,9.2491],[2529,-6.9417],[2556,1.656],[2701,1.6351],[2708,0.3794],[2737,6.4004],[2754,6.3968],[2803,6.2275],[2964,6.2275],[3199,5.2613],[3509,1.1632],[3561,-6.6586],[4364,-6.6586],[4366,49.4725],[4372,51.1939]]
       }
       spine_02_ctrl.Rotation.Y {
        type=MovieSceneScriptingFloatChannel
        default=0
        thinned_from=23
        keys [[1790,0],[2701,0],[2708,0.9599],[2737,0.1176],[2754,0.1151],[2803,4.7673],[2964,4.7673],[3199,4.0276],[3509,0.8904],[3680,0.8917],[4364,0.8904],[4366,-6.5469],[4372,-11.7194]]
       }
       spine_02_ctrl.Rotation.Z {
        type=MovieSceneScriptingFloatChannel
        default=0
        thinned_from=23
        keys [[1790,0],[2701,0],[2708,9.7163],[2737,0.7674],[2754,0.7515],[2803,1.4131],[2964,1.4131],[3199,1.1938],[3509,0.2639],[3590,0.2639],[3680,0.2614],[4067,0.2639],[4364,0.2639],[4366,-8.9079],[4372,-16.7874]]
       }
       spine_02_ctrl.Scale.X {type=MovieSceneScriptingFloatChannel, default=0, const [0,17]}
       spine_02_ctrl.Scale.Y {type=MovieSceneScriptingFloatChannel, default=0, const [0,17]}
       spine_02_ctrl.Scale.Z {type=MovieSceneScriptingFloatChannel, default=0, const [0,17]}
       spine_03_ctrl.Location.X {type=MovieSceneScriptingFloatChannel, default=0, thinned_from=25, keys [[1790,0],[4364,0],[4366,0.5027]]}
       spine_03_ctrl.Location.Y {type=MovieSceneScriptingFloatChannel, default=0, thinned_from=25, keys [[1790,0],[3715,0],[3747,0.1058],[4067,0.0475],[4364,0],[4366,6.196]]}
       spine_03_ctrl.Location.Z {type=MovieSceneScriptingFloatChannel, default=0, thinned_from=25, keys [[1790,0],[4364,0],[4366,6.7832]]}
       spine_03_ctrl.Rotation.X {
        type=MovieSceneScriptingFloatChannel
        default=0
        thinned_from=39
        keys [[1790,0],[2244,0],[2282,-0.5671],[2328,0],[2523,0],[2529,0.4742],[2556,11.7036],[2583,11.6919],[2590,-9.0152],[2615,5.3463],[2634,11.6919],[2639,-3.3285],[2641,-7.0427],[2650,-3.3819],[2659,2.6499],[2666,5.3463],[2691,5.3304],[2698,-8.9609],[2746,-4.551],[2754,-4.5117],[2944,-3.7113],[2964,-3.7113],[2998,-3.2482],[3051,-2.2653],[3199,-2.601],[3432,-0.5241],[3558,0.1385],[3562,2.9752],[3579,42.9673],[3587,46.9065],[3597,-0.3016],[3610,-3.9848],[3644,0.3749],[4067,-5.3007],[4364,-0.5241],[4365,69.4041],[4366,24.9931]]
       }
       spine_03_ctrl.Rotation.Y {
        type=MovieSceneScriptingFloatChannel
        default=0
        thinned_from=29
        keys [[1790,0],[2244,0],[2282,-1.1925],[2328,0],[2691,0],[2698,-0.7254],[2754,-0.7226],[2944,-0.6625],[2964,-0.6625],[2998,2.3284],[3051,3.1794],[3192,2.9673],[3509,-0.1014],[4366,-0.1014]]
       }
       spine_03_ctrl.Rotation.Z {
        type=MovieSceneScriptingFloatChannel
        default=0
        thinned_from=36
        keys [[1790,0],[2244,0],[2282,39.4539],[2328,0],[2523,0],[2530,-25.736],[2550,19.0302],[2576,0],[2634,0],[2639,33.3436],[2641,35.0243],[2654,-10.1788],[2666,5.202],[2679,0.1126],[2691,0],[2698,38.4753],[2699,39.461],[2714,-13.1341],[2726,8.0648],[2754,9.2767],[2944,8.506],[2964,8.506],[3008,36.4591],[3047,40.0742],[3096,0.369],[3165,-14.3832],[3232,-3.8075],[3509,1.3022],[4366,1.3022]]
       }
       spine_03_ctrl.Scale.X {type=MovieSceneScriptingFloatChannel, default=0, const [0,25]}
       spine_03_ctrl.Scale.Y {type=MovieSceneScriptingFloatChannel, default=0, const [0,25]}
       spine_03_ctrl.Scale.Z {type=MovieSceneScriptingFloatChannel, default=0, const [0,25]}
       spine_04_ctrl.Location.X {type=MovieSceneScriptingFloatChannel, default=0, thinned_from=23, keys [[1790,0],[4364,0],[4366,0.0794]]}
       spine_04_ctrl.Location.Y {type=MovieSceneScriptingFloatChannel, default=0, thinned_from=23, keys [[1790,0],[4364,0],[4366,4.6142]]}
       spine_04_ctrl.Location.Z {type=MovieSceneScriptingFloatChannel, default=0, thinned_from=23, keys [[1790,0],[4364,0],[4366,11.8935]]}
       spine_04_ctrl.Rotation.X {
        type=MovieSceneScriptingFloatChannel
        default=0
        thinned_from=35
        keys [[1790,0],[2523,0],[2529,-6.1534],[2556,0],[2584,0],[2591,-9.294],[2622,-0.0828],[2635,0],[2642,-12.2829],[2673,-0.0828],[2694,-0.0828],[2704,-5.8984],[2754,-0.0751],[2803,-4.7939],[2964,-4.7939],[3199,-4.0501],[3385,1.5976],[3474,1.7721],[3510,1.4544],[3560,2.483],[3584,4.2024],[3594,5.497],[3597,5.0572],[3614,-2.6468],[3673,-4.7593],[3882,-4.0553],[4067,-9.9777],[4335,-31.7236],[4364,-20.5359],[4366,0]]
       }
       spine_04_ctrl.Rotation.Y {
        type=MovieSceneScriptingFloatChannel
        default=0
        thinned_from=25
        keys [[1790,0],[2523,0],[2529,-2.4443],[2556,0],[2754,0],[2803,0.1043],[2964,0.1043],[3199,0.0881],[3509,0.0195],[3594,0.007],[3724,0.0031],[3841,0],[4366,0]]
       }
       spine_04_ctrl.Rotation.Z {
        type=MovieSceneScriptingFloatChannel
        default=0
        thinned_from=25
        keys [[1790,0],[2523,0],[2529,-9.4788],[2556,0],[2754,0],[2803,-5.7514],[2964,-5.7514],[3199,-4.8591],[3509,-1.0742],[3594,-0.388],[3724,-0.1691],[3841,0],[4366,0]]
       }
       spine_04_ctrl.Scale.X {type=MovieSceneScriptingFloatChannel, default=0, const [0,23]}
       spine_04_ctrl.Scale.Y {type=MovieSceneScriptingFloatChannel, default=0, const [0,23]}
       spine_04_ctrl.Scale.Z {type=MovieSceneScriptingFloatChannel, default=0, const [0,23]}
       head_ctrl.Location.X {type=MovieSceneScriptingFloatChannel, default=0, thinned_from=27, keys [[1790,0],[4366,0]]}
       head_ctrl.Location.Y {type=MovieSceneScriptingFloatChannel, default=0, thinned_from=27, keys [[1790,0],[4366,0]]}
       head_ctrl.Location.Z {type=MovieSceneScriptingFloatChannel, default=0, thinned_from=27, keys [[1790,0],[4366,0]]}
       head_ctrl.Rotation.X {
        type=MovieSceneScriptingFloatChannel
        default=0
        thinned_from=49
        keys [[1790,-16.6246],[2190,-16.6246],[2230,-50.4182],[2250,-52.2825],[2279,-20.4092],[2315,-16.3656],[2330,-16.4571],[2356,-15.9096],[2516,-14.1482],[2522,-14.0197],[2533,-32.2244],[2537,-33.3105],[2547,-25.2031],[2548,-24.1768],[2550,-23.9616],[2582,-33.9408],[2595,-89.5217],[2598,-90.4704],[2613,-55.6212],[2621,-54.5386],[2629,-54.4292],[2635,-56.591],[2647,-77.0857],[2672,-64.1714],[2704,-66.1157],[2715,-76.0332],[2738,-68.245],[2754,-68.2106],[2803,-2.5232],[2944,-38.0833],[2964,-38.0833],[2983,-64.5001],[3199,-56.5554],[3305,-41.6126],[3509,-41.6126],[3557,-38.2689],[3576,5.7485],[3579,5.258],[3596,-26.7058],[3617,-37.294],[3650,-55.222],[3707,-78.5749],[3776,-83.6909],[4067,-91.2352],[4364,-74.742],[4365,47.1276],[4366,26.0345]]
       }
       head_ctrl.Rotation.Y {
        type=MovieSceneScriptingFloatChannel
        default=0
        thinned_from=31
        keys [[1790,0],[2190,0],[2230,-7.6721],[2279,5.5515],[2315,0],[2330,-1.2408],[2356,0],[2529,0],[2540,-0.0001],[3777,0],[4067,-2.0591],[4364,0],[4365,21.0904],[4366,0]]
       }
       head_ctrl.Rotation.Z {
        type=MovieSceneScriptingFloatChannel
        default=0
        thinned_from=33
        keys [[1790,0],[2190,0],[2230,48.8478],[2279,43.0912],[2315,0],[2330,-9.4819],[2356,0],[2529,0],[2540,-0.0026],[2556,0],[2584,0],[2621,0.0013],[2635,0],[2672,0.0013],[3199,0.001],[3509,0.0002],[3777,0.0002],[4067,-3.2083],[4364,0.0002],[4365,61.2894],[4366,0.0002]]
       }
       head_ctrl.Scale.X {type=MovieSceneScriptingFloatChannel, default=0, const [0,27]}
       head_ctrl.Scale.Y {type=MovieSceneScriptingFloatChannel, default=0, const [0,27]}
       head_ctrl.Scale.Z {type=MovieSceneScriptingFloatChannel, default=0, const [0,27]}
       clavicle_l_ctrl.Location.X {type=MovieSceneScriptingFloatChannel, default=0, thinned_from=21, keys [[1790,0],[4366,0]]}
       clavicle_l_ctrl.Location.Y {type=MovieSceneScriptingFloatChannel, default=0, thinned_from=21, keys [[1790,0],[4366,0]]}
       clavicle_l_ctrl.Location.Z {type=MovieSceneScriptingFloatChannel, default=0, thinned_from=21, keys [[1790,0],[4366,0]]}
       clavicle_l_ctrl.Rotation.X {
        type=MovieSceneScriptingFloatChannel
        default=0
        thinned_from=24
        keys [[1790,0],[3841,0],[3933,1.0193],[4023,2.7311],[4038,3.2742],[4067,3.1908],[4079,3.2799],[4265,7.336],[4366,8.0728]]
       }
       clavicle_l_ctrl.Rotation.Y {
        type=MovieSceneScriptingFloatChannel
        default=0
        thinned_from=29
        keys [[1790,0],[3560,0],[3575,10.7449],[3579,11.5781],[3606,0],[3678,0],[3691,2.6324],[3705,-2.7834],[3731,2.0864],[3802,0.7698],[3841,0],[3933,-0.6379],[4023,-0.8169],[4038,3.0308],[4067,3.5521],[4079,5.4036],[4265,10.0284],[4366,-2.2416]]
       }
       clavicle_l_ctrl.Rotation.Z {
        type=MovieSceneScriptingFloatChannel
        default=0
        thinned_from=25
        keys [[1790,0],[3802,0],[3816,-8.3547],[3868,-0.7505],[3933,4.0307],[4023,10.8003],[4038,12.6842],[4067,12.7299],[4079,13.0292],[4265,28.7297],[4366,31.9232]]
       }
       clavicle_l_ctrl.Scale.X {type=MovieSceneScriptingFloatChannel, default=0, const [0,21]}
       clavicle_l_ctrl.Scale.Y {type=MovieSceneScriptingFloatChannel, default=0, const [0,21]}
       clavicle_l_ctrl.Scale.Z {type=MovieSceneScriptingFloatChannel, default=0, const [0,21]}
       clavicle_r_ctrl.Location.X {type=MovieSceneScriptingFloatChannel, default=0, thinned_from=20, keys [[1790,0],[4366,0]]}
       clavicle_r_ctrl.Location.Y {type=MovieSceneScriptingFloatChannel, default=0, thinned_from=20, keys [[1790,0],[4366,0]]}
       clavicle_r_ctrl.Location.Z {type=MovieSceneScriptingFloatChannel, default=0, thinned_from=20, keys [[1790,0],[4366,0]]}
       clavicle_r_ctrl.Rotation.X {
        type=MovieSceneScriptingFloatChannel
        default=0
        thinned_from=23
        keys [[1790,0],[3841,0],[3933,0.901],[4023,2.4141],[4038,2.8976],[4067,2.821],[4079,2.9005],[4265,9.1323],[4366,7.1359]]
       }
       clavicle_r_ctrl.Rotation.Y {
        type=MovieSceneScriptingFloatChannel
        default=0
        thinned_from=28
        keys [[1790,0],[3561,0],[3575,13.8188],[3582,15.5762],[3596,0],[3678,0],[3691,2.6324],[3705,-2.7876],[3731,2.0687],[3841,0],[3933,-0.4827],[4023,-0.6182],[4038,5.7833],[4067,3.7843],[4079,5.647],[4265,6.2191],[4366,-1.6963]]
       }
       clavicle_r_ctrl.Rotation.Z {
        type=MovieSceneScriptingFloatChannel
        default=0
        thinned_from=24
        keys [[1790,0],[3784,0],[3804,-6.8538],[3822,-9.9435],[3933,3.4829],[4023,9.3326],[4038,10.9797],[4067,11.0029],[4079,11.2655],[4265,39.2749],[4366,27.5847]]
       }
       clavicle_r_ctrl.Scale.X {type=MovieSceneScriptingFloatChannel, default=0, const [0,20]}
       clavicle_r_ctrl.Scale.Y {type=MovieSceneScriptingFloatChannel, default=0, const [0,20]}
       clavicle_r_ctrl.Scale.Z {type=MovieSceneScriptingFloatChannel, default=0, const [0,20]}
       neck_01_ctrl.Location.X {type=MovieSceneScriptingFloatChannel, default=0, thinned_from=18, keys [[1790,0],[4366,0]]}
       neck_01_ctrl.Location.Y {type=MovieSceneScriptingFloatChannel, default=0, thinned_from=18, keys [[1790,0],[4366,0]]}
       neck_01_ctrl.Location.Z {type=MovieSceneScriptingFloatChannel, default=0, thinned_from=18, keys [[1790,0],[4366,0]]}
       neck_01_ctrl.Rotation.X {
        type=MovieSceneScriptingFloatChannel
        default=0
        thinned_from=27
        keys [[1790,-12.8985],[2217,-12.8985],[2529,-11.7977],[2549,-26.5556],[2578,-27.0066],[2629,-27.0066],[2738,-38.1581],[2754,-38.1389],[2803,-0.7796],[2944,-28.4545],[2964,-28.4545],[3199,-24.0397],[3509,-5.3146],[3561,-19.8593],[3592,-19.2985],[3683,-14.5837],[3695,-14.8729],[3706,-28.9396],[3719,-21.3165],[3738,-22.105],[4067,-37.1128],[4364,-28.1635],[4365,5.0797],[4366,-15.5683]]
       }
       neck_01_ctrl.Rotation.Y {type=MovieSceneScriptingFloatChannel, default=0, thinned_from=19, keys [[1790,0],[4364,0],[4365,12.4399],[4366,0]]}
       neck_01_ctrl.Rotation.Z {type=MovieSceneScriptingFloatChannel, default=0, thinned_from=19, keys [[1790,0],[4364,0],[4365,4.4454],[4366,0]]}
       neck_01_ctrl.Scale.X {type=MovieSceneScriptingFloatChannel, default=0, const [0,18]}
       neck_01_ctrl.Scale.Y {type=MovieSceneScriptingFloatChannel, default=0, const [0,18]}
       neck_01_ctrl.Scale.Z {type=MovieSceneScriptingFloatChannel, default=0, const [0,18]}
       neck_02_ctrl.Location.X {type=MovieSceneScriptingFloatChannel, default=0, thinned_from=15, keys [[1790,0],[4366,0]]}
       neck_02_ctrl.Location.Y {type=MovieSceneScriptingFloatChannel, default=0, thinned_from=15, keys [[1790,0],[4366,0]]}
       neck_02_ctrl.Location.Z {type=MovieSceneScriptingFloatChannel, default=0, thinned_from=15, keys [[1790,0],[4366,0]]}
       neck_02_ctrl.Rotation.X {type=MovieSceneScriptingFloatChannel, default=0, thinned_from=15, keys [[1790,0],[4364,0],[4366,17.3182]]}
       neck_02_ctrl.Rotation.Y {type=MovieSceneScriptingFloatChannel, default=0, const [0,15]}
       neck_02_ctrl.Rotation.Z {type=MovieSceneScriptingFloatChannel, default=0, const [0,15]}
       neck_02_ctrl.Scale.X {type=MovieSceneScriptingFloatChannel, default=0, const [0,15]}
       neck_02_ctrl.Scale.Y {type=MovieSceneScriptingFloatChannel, default=0, const [0,15]}
       neck_02_ctrl.Scale.Z {type=MovieSceneScriptingFloatChannel, default=0, const [0,15]}
       head_ik_ctrl.Location.X {
        type=MovieSceneScriptingFloatChannel
        default=0
        keys [[1790,0.0057],[2113,-0.1614],[2217,0.0057],[2516,-0.2227],[2523,0.0002],[2529,-0.0213],[2556,-0.0396],[2754,0.2632],[2944,0.6201],[2964,0.6201],[3199,0.906],[3509,0.174],[3841,-0.0563],[4067,-0.3126],[4364,-0.1385],[4366,0.0604]]
       }
       head_ik_ctrl.Location.Y {
        type=MovieSceneScriptingFloatChannel
        default=0
        keys [[1790,0.1459],[2113,-1.8785],[2217,0.1459],[2516,-1.7044],[2523,-2.0018],[2529,-1.9628],[2556,-1.942],[2754,-7.5725],[2944,-6.0628],[2964,-6.0628],[3199,-4.5896],[3509,-1.2607],[3841,-0.0003],[4067,-5.9116],[4364,-1.6864],[4366,0.5736]]
       }
       head_ik_ctrl.Location.Z {
        type=MovieSceneScriptingFloatChannel
        default=0
        keys [[1790,0.2718],[2113,1.2048],[2217,0.2718],[2516,1.274],[2523,-0.2687],[2529,-0.4227],[2556,-0.3754],[2754,-2.4239],[2944,0.4573],[2964,0.4573],[3199,1.4048],[3509,1.3771],[3841,-0.0015],[4067,0.326],[4364,1.4956],[4366,0.3861]]
       }
       head_ik_ctrl.Rotation.X {
        type=MovieSceneScriptingFloatChannel
        default=0
        keys [[1790,0],[2113,-16.6246],[2217,0],[2516,-14.1482],[2523,-14.0615],[2529,-14.0197],[2556,-13.7215],[2754,-93.3328],[2944,-63.2861],[2964,-63.2861],[3199,-64.7568],[3509,-58.1202],[3841,0],[4067,-91.3633],[4364,-40.7464],[4366,49.0217]]
       }
       head_ik_ctrl.Rotation.Y {type=MovieSceneScriptingFloatChannel, default=0, thinned_from=16, keys [[1790,0],[3841,0],[4067,-2.0591],[4364,3.7743],[4366,17.1372]]}
       head_ik_ctrl.Rotation.Z {
        type=MovieSceneScriptingFloatChannel
        default=0
        thinned_from=16
        keys [[1790,0],[2556,0],[2754,0.0013],[3199,0.0009],[3509,0.0002],[3841,0],[4067,-3.2083],[4364,0.0468],[4366,0.0281]]
       }
       head_ik_ctrl.Scale.X {type=MovieSceneScriptingFloatChannel, default=0, const [0,16]}
       head_ik_ctrl.Scale.Y {type=MovieSceneScriptingFloatChannel, default=0, const [0,16]}
       head_ik_ctrl.Scale.Z {type=MovieSceneScriptingFloatChannel, default=0, const [0,16]}
       chest_ctrl.Location.X {
        type=MovieSceneScriptingFloatChannel
        default=0
        keys [[1790,-0.0172],[2113,-0.035],[2217,-0.0172],[2516,-0.1724],[2523,-0.0213],[2529,0.0568],[2556,0.067],[2754,0.261],[2944,0.0591],[2964,0.0591],[3199,3.2118],[3509,2.6532],[3841,-0.0018],[4067,-0.4343],[4364,0.0114],[4366,1.715]]
       }
       chest_ctrl.Location.Y {
        type=MovieSceneScriptingFloatChannel
        default=0
        keys [[1790,-1.1046],[2113,-1.0935],[2217,-1.1046],[2516,-3.71],[2523,-3.2876],[2529,-3.2433],[2556,-3.2598],[2754,-2.9267],[2944,-0.9226],[2964,-0.9226],[3199,2.4572],[3509,-0.5423],[3841,0.001],[4067,-4.512],[4364,-0.5976],[4366,16.8016]]
       }
       chest_ctrl.Location.Z {
        type=MovieSceneScriptingFloatChannel
        default=0
        keys [[1790,-0.1057],[2113,-0.1163],[2217,-0.1057],[2516,-0.9545],[2523,-1.6127],[2529,-1.5747],[2556,-1.5545],[2754,-0.5941],[2944,-0.9095],[2964,-0.9095],[3199,-0.4591],[3509,-0.4199],[3841,-0.0064],[4067,-0.9386],[4364,1.2964],[4366,5.8868]]
       }
       chest_ctrl.Rotation.X {
        type=MovieSceneScriptingFloatChannel
        default=0
        keys [[1790,-7.9917],[2113,-8.0018],[2217,-7.9917],[2516,-18.4694],[2523,-18.2848],[2529,-17.3022],[2556,-17.9168],[2754,-22.1368],[2944,-6.708],[2964,-6.708],[3199,30.6132],[3509,8.8202],[3841,-0.0092],[4067,-27.5209],[4364,0.6401],[4366,82.0642]]
       }
       chest_ctrl.Rotation.Y {
        type=MovieSceneScriptingFloatChannel
        default=0
        keys [[1790,-0.0712],[2113,-0.0548],[2217,-0.0712],[2516,-0.1833],[2523,-0.5331],[2529,-0.6611],[2556,-0.7014],[2754,6.2579],[2944,-1.2431],[2964,-1.2431],[3199,-22.9658],[3509,-23.7319],[3841,-0.053],[4067,1.2616],[4364,-2.6876],[4366,-4.336]]
       }
       chest_ctrl.Rotation.Z {
        type=MovieSceneScriptingFloatChannel
        default=0
        keys [[1790,0.027],[2113,0.0309],[2217,0.027],[2516,5.5338],[2523,0.1211],[2529,0.001],[2556,0.0414],[2754,8.1589],[2944,9.1448],[2964,9.1448],[3199,22.1589],[3509,14.2263],[3841,-0.0031],[4067,0.0791],[4364,13.8221],[4366,6.7505]]
       }
       chest_ctrl.Scale.X {type=MovieSceneScriptingFloatChannel, default=0, const [0,16]}
       chest_ctrl.Scale.Y {type=MovieSceneScriptingFloatChannel, default=0, const [0,16]}
       chest_ctrl.Scale.Z {type=MovieSceneScriptingFloatChannel, default=0, const [0,16]}
       chest_tan_ctrl.Location.X {
        type=MovieSceneScriptingFloatChannel
        default=0
        keys [[1790,-0.0058],[2113,0.0155],[2217,-0.0058],[2516,-0.0611],[2523,0.035],[2529,0.0535],[2556,0.0195],[2754,-0.6416],[2944,0.0932],[2964,0.0932],[3199,0.2476],[3509,0.1612],[3841,0.0101],[4067,0.0851],[4364,0.2555],[4366,-3.6285]]
       }
       chest_tan_ctrl.Location.Y {
        type=MovieSceneScriptingFloatChannel
        default=0
        keys [[1790,0.6431],[2113,0.6446],[2217,0.6431],[2516,0.4012],[2523,0.4452],[2529,0.5127],[2556,0.5209],[2754,-1.034],[2944,-1.2541],[2964,-1.2541],[3199,-0.936],[3509,0.3123],[3841,0.0003],[4067,-1.1098],[4364,0.0266],[4366,-2.6513]]
       }
       chest_tan_ctrl.Location.Z {
        type=MovieSceneScriptingFloatChannel
        default=0
        keys [[1790,0.043],[2113,0.0463],[2217,0.043],[2516,0.2325],[2523,0.0127],[2529,0.0399],[2556,0.025],[2754,0.0811],[2944,-0.1855],[2964,-0.1855],[3199,-0.5214],[3509,-0.1621],[3841,-0.009],[4067,0.084],[4364,-2.7802],[4366,-24.8768]]
       }
       chest_tan_ctrl.Rotation.X {
        type=MovieSceneScriptingFloatChannel
        default=0
        keys [[1790,-3.7769],[2113,-3.6959],[2217,-3.7769],[2516,-1.8236],[2523,1.639],[2529,1.0596],[2556,1.4008],[2754,7.0531],[2944,3.5083],[2964,3.5083],[3199,-12.3767],[3509,-16.5423],[3841,-3.8476],[4067,1.2315],[4364,-6.1201],[4366,-6.2537]]
       }
       chest_tan_ctrl.Rotation.Y {
        type=MovieSceneScriptingFloatChannel
        default=0
        keys [[1790,0.0993],[2113,0.2331],[2217,0.0993],[2516,0.4683],[2523,-0.2683],[2529,-0.272],[2556,-0.2177],[2754,-0.0006],[2944,-0.1335],[2964,-0.1335],[3199,1.9948],[3509,2.1755],[3841,0.0924],[4067,-1.3459],[4364,0.6922],[4366,0.4646]]
       }
       chest_tan_ctrl.Rotation.Z {
        type=MovieSceneScriptingFloatChannel
        default=0
        keys [[1790,0.5188],[2113,0.5851],[2217,0.5188],[2516,3.0792],[2523,0.8224],[2529,0.5938],[2556,0.7792],[2754,0],[2944,0.7272],[2964,0.7272],[3199,0.7449],[3509,0.2407],[3841,0.3261],[4067,-0.4112],[4364,-0.5626],[4366,0.2699]]
       }
       chest_tan_ctrl.Scale.X {
        type=MovieSceneScriptingFloatChannel
        default=0
        thinned_from=16
        keys [[1790,0],[2217,0],[2516,0.0366],[2523,0],[2529,0.0373],[2556,0.0567],[2754,0.0488],[2944,0.0502],[2964,0.0502],[3199,0.0525],[3509,0.0188],[3841,0],[4067,0.0047],[4366,0.0028]]
       }
       chest_tan_ctrl.Scale.Y {
        type=MovieSceneScriptingFloatChannel
        default=0
        thinned_from=16
        keys [[1790,0],[2217,0],[2516,0.0366],[2523,0],[2529,0.0373],[2556,0.0567],[2754,0.0488],[2944,0.0502],[2964,0.0502],[3199,0.0525],[3509,0.0188],[3841,0],[4067,0.0047],[4366,0.0028]]
       }
       chest_tan_ctrl.Scale.Z {
        type=MovieSceneScriptingFloatChannel
        default=0
        thinned_from=16
        keys [[1790,0],[2217,0],[2516,0.0366],[2523,0],[2529,0.0373],[2556,0.0567],[2754,0.0488],[2944,0.0502],[2964,0.0502],[3199,0.0525],[3509,0.0188],[3841,0],[4067,0.0047],[4366,0.0028]]
       }
       Weight {type=MovieSceneScriptingFloatChannel, default=1, const [1,16]}
      ]
     }
     {
      type=MovieSceneControlRigParameterSection
      channels [
       foot_l_ik_ctrl.Location.X {type=MovieSceneScriptingFloatChannel, default=0, thinned_from=9, keys [[1895,0],[2138,0],[3395,-111.4193],[3417,-97.0785]]}
       foot_l_ik_ctrl.Location.Y {type=MovieSceneScriptingFloatChannel, default=0, thinned_from=9, keys [[1895,0],[2138,0],[3395,-201.903],[3417,-183.485]]}
       foot_l_ik_ctrl.Location.Z {
        type=MovieSceneScriptingFloatChannel
        default=0
        keys [[1895,0],[1943,-28.2213],[1984,32.4187],[2023,0.3879],[2063,82.0966],[2102,17.2971],[2138,58.1006],[2179,-32.8301],[3395,-173.2859],[3417,-163.8372]]
       }
       foot_l_ik_ctrl.Rotation.X {type=MovieSceneScriptingFloatChannel, default=0, thinned_from=9, keys [[1895,0],[2138,0],[3395,-18.6818],[3417,-18.47]]}
       foot_l_ik_ctrl.Rotation.Y {type=MovieSceneScriptingFloatChannel, default=0, thinned_from=9, keys [[1895,0],[2138,0],[3395,12.0393],[3417,11.5312]]}
       foot_l_ik_ctrl.Rotation.Z {type=MovieSceneScriptingFloatChannel, default=0, thinned_from=9, keys [[1895,0],[2138,0],[3395,-17.5344],[3417,-16.7111]]}
       foot_l_ik_ctrl.Scale.X {type=MovieSceneScriptingFloatChannel, default=0, thinned_from=9, keys [[1895,0],[2138,0],[3395,0.0045],[3417,0.0041]]}
       foot_l_ik_ctrl.Scale.Y {type=MovieSceneScriptingFloatChannel, default=0, thinned_from=9, keys [[1895,0],[2138,0],[3395,0.0045],[3417,0.0041]]}
       foot_l_ik_ctrl.Scale.Z {type=MovieSceneScriptingFloatChannel, default=0, thinned_from=9, keys [[1895,0],[2138,0],[3395,0.0045],[3417,0.0041]]}
       leg_l_softness {type=MovieSceneScriptingFloatChannel, default=0, const [0,9]}
       leg_l_pv_ik_ctrl.Location.X {
        type=MovieSceneScriptingFloatChannel
        default=0
        thinned_from=10
        keys [[1895,0],[1984,0],[2023,18.4138],[2063,2.7944],[2102,30.5524],[2138,4.3745],[2179,-89.0837],[3395,-196.314],[3417,-176.9795]]
       }
       leg_l_pv_ik_ctrl.Location.Y {
        type=MovieSceneScriptingFloatChannel
        default=0
        thinned_from=10
        keys [[1895,0],[1984,0],[2023,0.0752],[2063,-1.9133],[2102,0.1255],[2138,-2.7166],[2179,-1.1211],[3395,-73.4102],[3417,-62.5071]]
       }
       leg_l_pv_ik_ctrl.Location.Z {
        type=MovieSceneScriptingFloatChannel
        default=0
        thinned_from=10
        keys [[1895,0],[1984,0],[2023,1.7018],[2063,-23.7212],[2102,2.8237],[2138,-30.9934],[2179,-17.0075],[3395,146.1126],[3417,133.1355]]
       }
       leg_l_pv_ik_ctrl.Rotation.X {type=MovieSceneScriptingFloatChannel, default=0, thinned_from=9, keys [[1895,0],[2138,0],[3395,79.8483],[3417,103.3588]]}
       leg_l_pv_ik_ctrl.Rotation.Y {type=MovieSceneScriptingFloatChannel, default=0, thinned_from=9, keys [[1895,0],[2138,0],[3395,13.2853],[3417,16.8431]]}
       leg_l_pv_ik_ctrl.Rotation.Z {type=MovieSceneScriptingFloatChannel, default=0, thinned_from=9, keys [[1895,0],[2138,0],[3395,11.2116],[3417,43.1121]]}
       leg_l_pv_ik_ctrl.Scale.X {type=MovieSceneScriptingFloatChannel, default=0, thinned_from=9, keys [[1895,0],[2138,0],[3395,-0.0525],[3417,-0.0479]]}
       leg_l_pv_ik_ctrl.Scale.Y {type=MovieSceneScriptingFloatChannel, default=0, thinned_from=9, keys [[1895,0],[2138,0],[3395,-0.0525],[3417,-0.0479]]}
       leg_l_pv_ik_ctrl.Scale.Z {type=MovieSceneScriptingFloatChannel, default=0, thinned_from=9, keys [[1895,0],[2138,0],[3395,-0.0525],[3417,-0.0479]]}
       foot_r_ik_ctrl.Location.X {type=MovieSceneScriptingFloatChannel, default=0, thinned_from=9, keys [[1895,0],[2138,0],[3395,-20.2774],[3417,-15.6156]]}
       foot_r_ik_ctrl.Location.Y {type=MovieSceneScriptingFloatChannel, default=0, thinned_from=9, keys [[1895,0],[2138,0],[3395,-209.9278],[3417,-190.9524]]}
       foot_r_ik_ctrl.Location.Z {
        type=MovieSceneScriptingFloatChannel
        default=0
        keys [[1895,0],[1943,-28.2213],[1984,32.4187],[2023,0.3879],[2063,82.0966],[2102,17.2971],[2138,58.1006],[2179,-32.8301],[3395,-200.5837],[3417,-191.3679]]
       }
       foot_r_ik_ctrl.Rotation.X {type=MovieSceneScriptingFloatChannel, default=0, thinned_from=9, keys [[1895,0],[2138,0],[3395,-4.1271],[3417,-5.5608]]}
       foot_r_ik_ctrl.Rotation.Y {type=MovieSceneScriptingFloatChannel, default=0, thinned_from=9, keys [[1895,0],[2138,0],[3395,7.0831],[3417,6.7638]]}
       foot_r_ik_ctrl.Rotation.Z {type=MovieSceneScriptingFloatChannel, default=0, thinned_from=9, keys [[1895,0],[2138,0],[3395,-10.7289],[3417,-11.7235]]}
       foot_r_ik_ctrl.Scale.X {type=MovieSceneScriptingFloatChannel, default=0, thinned_from=9, keys [[1895,0],[2138,0],[3395,-0.0045],[3417,-0.0041]]}
       foot_r_ik_ctrl.Scale.Y {type=MovieSceneScriptingFloatChannel, default=0, thinned_from=9, keys [[1895,0],[2138,0],[3395,0.0045],[3417,0.0041]]}
       foot_r_ik_ctrl.Scale.Z {type=MovieSceneScriptingFloatChannel, default=0, thinned_from=9, keys [[1895,0],[2138,0],[3395,0.0045],[3417,0.0041]]}
       leg_r_softness {type=MovieSceneScriptingFloatChannel, default=0, const [0,9]}
       leg_r_pv_ik_ctrl.Location.X {
        type=MovieSceneScriptingFloatChannel
        default=0
        thinned_from=10
        keys [[1895,0],[1984,0],[2023,-18.4137],[2063,-3.5708],[2102,-30.5524],[2138,-2.1499],[2179,88.4061],[3395,261.7101],[3417,242.5807]]
       }
       leg_r_pv_ik_ctrl.Location.Y {
        type=MovieSceneScriptingFloatChannel
        default=0
        thinned_from=10
        keys [[1895,0],[1984,0],[2023,-0.0752],[2063,-1.0542],[2102,-0.1255],[2138,-0.7229],[2179,-0.5607],[3395,-107.0283],[3417,-91.7311]]
       }
       leg_r_pv_ik_ctrl.Location.Z {
        type=MovieSceneScriptingFloatChannel
        default=0
        thinned_from=10
        keys [[1895,0],[1984,0],[2023,-1.7018],[2063,25.0392],[2102,-2.8236],[2138,16.6362],[2179,22.166],[3395,124.0935],[3417,118.2594]]
       }
       leg_r_pv_ik_ctrl.Rotation.X {type=MovieSceneScriptingFloatChannel, default=0, thinned_from=9, keys [[1895,0],[2138,0],[3395,-33.2315],[3417,-40.6177]]}
       leg_r_pv_ik_ctrl.Rotation.Y {type=MovieSceneScriptingFloatChannel, default=0, thinned_from=9, keys [[1895,0],[2138,0],[3395,-1.3755],[3417,-4.5218]]}
       leg_r_pv_ik_ctrl.Rotation.Z {type=MovieSceneScriptingFloatChannel, default=0, thinned_from=9, keys [[1895,0],[2138,0],[3395,25.2548],[3417,31.8657]]}
       leg_r_pv_ik_ctrl.Scale.X {type=MovieSceneScriptingFloatChannel, default=0, thinned_from=9, keys [[1895,0],[2138,0],[3395,-0.0222],[3417,-0.0203]]}
       leg_r_pv_ik_ctrl.Scale.Y {type=MovieSceneScriptingFloatChannel, default=0, thinned_from=9, keys [[1895,0],[2138,0],[3395,-0.0222],[3417,-0.0203]]}
       leg_r_pv_ik_ctrl.Scale.Z {type=MovieSceneScriptingFloatChannel, default=0, thinned_from=9, keys [[1895,0],[2138,0],[3395,-0.0222],[3417,-0.0203]]}
       upperarm_l_fk_ctrl.Location.X {type=MovieSceneScriptingFloatChannel, default=0, const [0,9]}
       upperarm_l_fk_ctrl.Location.Y {type=MovieSceneScriptingFloatChannel, default=0, const [0,9]}
       upperarm_l_fk_ctrl.Location.Z {type=MovieSceneScriptingFloatChannel, default=0, const [0,9]}
       upperarm_l_fk_ctrl.Rotation.X {
        type=MovieSceneScriptingFloatChannel
        default=0
        thinned_from=10
        keys [[1925,0],[1973,0],[2013,0.001],[2014,0],[2053,0],[2095,0.0768],[2130,0.1806],[2168,0.0904],[3417,0.0904]]
       }
       upperarm_l_fk_ctrl.Rotation.Y {
        type=MovieSceneScriptingFloatChannel
        default=0
        thinned_from=10
        keys [[1925,0],[1973,0],[2013,-0.0568],[2014,0],[2053,0],[2095,-5.7185],[2130,-2.8366],[2168,-6.1918],[3417,-6.1918]]
       }
       upperarm_l_fk_ctrl.Rotation.Z {
        type=MovieSceneScriptingFloatChannel
        default=0
        thinned_from=10
        keys [[1925,0],[1973,0],[2013,0.0063],[2014,0],[2053,0],[2095,0.6758],[2130,0.6383],[2168,0.8934],[3417,0.8934]]
       }
       upperarm_l_fk_ctrl.Scale.X {type=MovieSceneScriptingFloatChannel, default=0, const [0,9]}
       upperarm_l_fk_ctrl.Scale.Y {type=MovieSceneScriptingFloatChannel, default=0, const [0,9]}
       upperarm_l_fk_ctrl.Scale.Z {type=MovieSceneScriptingFloatChannel, default=0, const [0,9]}
       hand_l_fk_ctrl.Location.X {type=MovieSceneScriptingFloatChannel, default=0, thinned_from=9, keys [[1895,0],[3417,0]]}
       hand_l_fk_ctrl.Location.Y {type=MovieSceneScriptingFloatChannel, default=0, thinned_from=9, keys [[1895,0],[3417,0]]}
       hand_l_fk_ctrl.Location.Z {type=MovieSceneScriptingFloatChannel, default=0, thinned_from=9, keys [[1895,0],[3417,0]]}
       hand_l_fk_ctrl.Rotation.X {
        type=MovieSceneScriptingFloatChannel
        default=0
        thinned_from=9
        keys [[1895,0],[1984,0],[2023,-2.7804],[2063,0],[2102,-1.9755],[2138,0],[3417,0]]
       }
       hand_l_fk_ctrl.Rotation.Y {
        type=MovieSceneScriptingFloatChannel
        default=0
        thinned_from=9
        keys [[1895,0],[1984,0],[2023,-0.3967],[2063,0],[2102,-0.123],[2138,0],[3417,0]]
       }
       hand_l_fk_ctrl.Rotation.Z {
        type=MovieSceneScriptingFloatChannel
        default=0
        thinned_from=9
        keys [[1895,0],[1984,0],[2023,-25.3263],[2063,0],[2102,-17.592],[2138,0],[3417,0]]
       }
       hand_l_fk_ctrl.Scale.X {type=MovieSceneScriptingFloatChannel, default=0, const [0,9]}
       hand_l_fk_ctrl.Scale.Y {type=MovieSceneScriptingFloatChannel, default=0, const [0,9]}
       hand_l_fk_ctrl.Scale.Z {type=MovieSceneScriptingFloatChannel, default=0, const [0,9]}
       thumb_01_l_ctrl.Location.X {type=MovieSceneScriptingFloatChannel, default=0, const [0,9]}
       thumb_01_l_ctrl.Location.Y {type=MovieSceneScriptingFloatChannel, default=0, const [0,9]}
       thumb_01_l_ctrl.Location.Z {type=MovieSceneScriptingFloatChannel, default=0, const [0,9]}
       thumb_01_l_ctrl.Rotation.X {type=MovieSceneScriptingFloatChannel, default=0, const [0,9]}
       thumb_01_l_ctrl.Rotation.Y {type=MovieSceneScriptingFloatChannel, default=0, const [0,9]}
       thumb_01_l_ctrl.Rotation.Z {type=MovieSceneScriptingFloatChannel, default=0, const [0,9]}
       thumb_01_l_ctrl.Scale.X {type=MovieSceneScriptingFloatChannel, default=0, const [0,9]}
       thumb_01_l_ctrl.Scale.Y {type=MovieSceneScriptingFloatChannel, default=0, const [0,9]}
       thumb_01_l_ctrl.Scale.Z {type=MovieSceneScriptingFloatChannel, default=0, const [0,9]}
       thumb_02_l_ctrl.Location.X {type=MovieSceneScriptingFloatChannel, default=0, const [0,9]}
       thumb_02_l_ctrl.Location.Y {type=MovieSceneScriptingFloatChannel, default=0, const [0,9]}
       thumb_02_l_ctrl.Location.Z {type=MovieSceneScriptingFloatChannel, default=0, const [0,9]}
       thumb_02_l_ctrl.Rotation.X {type=MovieSceneScriptingFloatChannel, default=0, const [0,9]}
       thumb_02_l_ctrl.Rotation.Y {type=MovieSceneScriptingFloatChannel, default=0, const [0,9]}
       thumb_02_l_ctrl.Rotation.Z {type=MovieSceneScriptingFloatChannel, default=0, const [0,9]}
       thumb_02_l_ctrl.Scale.X {type=MovieSceneScriptingFloatChannel, default=0, const [0,9]}
       thumb_02_l_ctrl.Scale.Y {type=MovieSceneScriptingFloatChannel, default=0, const [0,9]}
       thumb_02_l_ctrl.Scale.Z {type=MovieSceneScriptingFloatChannel, default=0, const [0,9]}
       thumb_03_l_ctrl.Location.X {type=MovieSceneScriptingFloatChannel, default=0, const [0,9]}
       thumb_03_l_ctrl.Location.Y {type=MovieSceneScriptingFloatChannel, default=0, const [0,9]}
       thumb_03_l_ctrl.Location.Z {type=MovieSceneScriptingFloatChannel, default=0, const [0,9]}
       thumb_03_l_ctrl.Rotation.X {type=MovieSceneScriptingFloatChannel, default=0, const [0,9]}
       thumb_03_l_ctrl.Rotation.Y {type=MovieSceneScriptingFloatChannel, default=0, const [0,9]}
       thumb_03_l_ctrl.Rotation.Z {type=MovieSceneScriptingFloatChannel, default=0, const [0,9]}
       thumb_03_l_ctrl.Scale.X {type=MovieSceneScriptingFloatChannel, default=0, const [0,9]}
       thumb_03_l_ctrl.Scale.Y {type=MovieSceneScriptingFloatChannel, default=0, const [0,9]}
       thumb_03_l_ctrl.Scale.Z {type=MovieSceneScriptingFloatChannel, default=0, const [0,9]}
       index_metacarpal_l_ctrl.Location.X {type=MovieSceneScriptingFloatChannel, default=0, const [0,9]}
       index_metacarpal_l_ctrl.Location.Y {type=MovieSceneScriptingFloatChannel, default=0, const [0,9]}
       index_metacarpal_l_ctrl.Location.Z {type=MovieSceneScriptingFloatChannel, default=0, const [0,9]}
       index_metacarpal_l_ctrl.Rotation.X {type=MovieSceneScriptingFloatChannel, default=0, const [0,9]}
       index_metacarpal_l_ctrl.Rotation.Y {type=MovieSceneScriptingFloatChannel, default=0, const [0,9]}
       index_metacarpal_l_ctrl.Rotation.Z {type=MovieSceneScriptingFloatChannel, default=0, const [0,9]}
       index_metacarpal_l_ctrl.Scale.X {type=MovieSceneScriptingFloatChannel, default=0, const [0,9]}
       index_metacarpal_l_ctrl.Scale.Y {type=MovieSceneScriptingFloatChannel, default=0, const [0,9]}
       index_metacarpal_l_ctrl.Scale.Z {type=MovieSceneScriptingFloatChannel, default=0, const [0,9]}
       index_01_l_ctrl.Location.X {type=MovieSceneScriptingFloatChannel, default=0, const [0,10]}
       index_01_l_ctrl.Location.Y {type=MovieSceneScriptingFloatChannel, default=0, const [0,10]}
       index_01_l_ctrl.Location.Z {type=MovieSceneScriptingFloatChannel, default=0, const [0,10]}
       index_01_l_ctrl.Rotation.X {type=MovieSceneScriptingFloatChannel, default=0, thinned_from=11, keys [[1895,0],[3927,0],[4164,9.2431]]}
       index_01_l_ctrl.Rotation.Y {type=MovieSceneScriptingFloatChannel, default=0, thinned_from=11, keys [[1895,0],[3927,0],[4164,4.7328]]}
       index_01_l_ctrl.Rotation.Z {type=MovieSceneScriptingFloatChannel, default=0, thinned_from=11, keys [[1895,0],[3927,0],[4164,-37.6464]]}
       index_01_l_ctrl.Scale.X {type=MovieSceneScriptingFloatChannel, default=0, const [0,10]}
       index_01_l_ctrl.Scale.Y {type=MovieSceneScriptingFloatChannel, default=0, const [0,10]}
       index_01_l_ctrl.Scale.Z {type=MovieSceneScriptingFloatChannel, default=0, const [0,10]}
       index_02_l_ctrl.Location.X {type=MovieSceneScriptingFloatChannel, default=0, thinned_from=10, keys [[1895,0],[3927,0]]}
       index_02_l_ctrl.Location.Y {type=MovieSceneScriptingFloatChannel, default=0, thinned_from=10, keys [[1895,0],[3927,0]]}
       index_02_l_ctrl.Location.Z {type=MovieSceneScriptingFloatChannel, default=0, thinned_from=10, keys [[1895,0],[3927,0]]}
       index_02_l_ctrl.Rotation.X {type=MovieSceneScriptingFloatChannel, default=0, thinned_from=11, keys [[1895,0],[3927,0],[4164,3.5599]]}
       index_02_l_ctrl.Rotation.Y {type=MovieSceneScriptingFloatChannel, default=0, thinned_from=11, keys [[1895,0],[3927,0],[4164,3.787]]}
       index_02_l_ctrl.Rotation.Z {type=MovieSceneScriptingFloatChannel, default=0, thinned_from=11, keys [[1895,0],[3927,0],[4164,-36.807]]}
       index_02_l_ctrl.Scale.X {type=MovieSceneScriptingFloatChannel, default=0, const [0,10]}
       index_02_l_ctrl.Scale.Y {type=MovieSceneScriptingFloatChannel, default=0, const [0,10]}
       index_02_l_ctrl.Scale.Z {type=MovieSceneScriptingFloatChannel, default=0, const [0,10]}
       index_03_l_ctrl.Location.X {type=MovieSceneScriptingFloatChannel, default=0, const [0,10]}
       index_03_l_ctrl.Location.Y {type=MovieSceneScriptingFloatChannel, default=0, const [0,10]}
       index_03_l_ctrl.Location.Z {type=MovieSceneScriptingFloatChannel, default=0, const [0,10]}
       index_03_l_ctrl.Rotation.X {type=MovieSceneScriptingFloatChannel, default=0, thinned_from=11, keys [[1895,0],[3927,0],[4164,0.4192]]}
       index_03_l_ctrl.Rotation.Y {type=MovieSceneScriptingFloatChannel, default=0, thinned_from=11, keys [[1895,0],[3927,0],[4164,0.267]]}
       index_03_l_ctrl.Rotation.Z {type=MovieSceneScriptingFloatChannel, default=0, thinned_from=11, keys [[1895,0],[3927,0],[4164,-36.8303]]}
       index_03_l_ctrl.Scale.X {type=MovieSceneScriptingFloatChannel, default=0, const [0,10]}
       index_03_l_ctrl.Scale.Y {type=MovieSceneScriptingFloatChannel, default=0, const [0,10]}
       index_03_l_ctrl.Scale.Z {type=MovieSceneScriptingFloatChannel, default=0, const [0,10]}
       middle_metacarpal_l_ctrl.Location.X {type=MovieSceneScriptingFloatChannel, default=0, thinned_from=9, keys [[1895,0],[3417,0]]}
       middle_metacarpal_l_ctrl.Location.Y {type=MovieSceneScriptingFloatChannel, default=0, thinned_from=9, keys [[1895,0],[3417,0]]}
       middle_metacarpal_l_ctrl.Location.Z {type=MovieSceneScriptingFloatChannel, default=0, thinned_from=9, keys [[1895,0],[3417,0]]}
       middle_metacarpal_l_ctrl.Rotation.X {type=MovieSceneScriptingFloatChannel, default=0, const [0,9]}
       middle_metacarpal_l_ctrl.Rotation.Y {type=MovieSceneScriptingFloatChannel, default=0, const [0,9]}
       middle_metacarpal_l_ctrl.Rotation.Z {type=MovieSceneScriptingFloatChannel, default=0, const [0,9]}
       middle_metacarpal_l_ctrl.Scale.X {type=MovieSceneScriptingFloatChannel, default=0, const [0,9]}
       middle_metacarpal_l_ctrl.Scale.Y {type=MovieSceneScriptingFloatChannel, default=0, const [0,9]}
       middle_metacarpal_l_ctrl.Scale.Z {type=MovieSceneScriptingFloatChannel, default=0, const [0,9]}
       middle_01_l_ctrl.Location.X {type=MovieSceneScriptingFloatChannel, default=0, const [0,10]}
       middle_01_l_ctrl.Location.Y {type=MovieSceneScriptingFloatChannel, default=0, const [0,10]}
       middle_01_l_ctrl.Location.Z {type=MovieSceneScriptingFloatChannel, default=0, const [0,10]}
       middle_01_l_ctrl.Rotation.X {type=MovieSceneScriptingFloatChannel, default=0, thinned_from=11, keys [[1895,0],[3927,0],[4164,-3.4326]]}
       middle_01_l_ctrl.Rotation.Y {type=MovieSceneScriptingFloatChannel, default=0, thinned_from=11, keys [[1895,0],[3927,0],[4164,0.3333]]}
       middle_01_l_ctrl.Rotation.Z {type=MovieSceneScriptingFloatChannel, default=0, thinned_from=11, keys [[1895,0],[3927,0],[4164,-36.9815]]}
       middle_01_l_ctrl.Scale.X {type=MovieSceneScriptingFloatChannel, default=0, const [0,10]}
       middle_01_l_ctrl.Scale.Y {type=MovieSceneScriptingFloatChannel, default=0, const [0,10]}
       middle_01_l_ctrl.Scale.Z {type=MovieSceneScriptingFloatChannel, default=0, const [0,10]}
       middle_02_l_ctrl.Location.X {type=MovieSceneScriptingFloatChannel, default=0, const [0,10]}
       middle_02_l_ctrl.Location.Y {type=MovieSceneScriptingFloatChannel, default=0, const [0,10]}
       middle_02_l_ctrl.Location.Z {type=MovieSceneScriptingFloatChannel, default=0, const [0,10]}
       middle_02_l_ctrl.Rotation.X {type=MovieSceneScriptingFloatChannel, default=0, thinned_from=11, keys [[1895,0],[3927,0],[4164,0.1116]]}
       middle_02_l_ctrl.Rotation.Y {type=MovieSceneScriptingFloatChannel, default=0, thinned_from=11, keys [[1895,0],[3927,0],[4164,-0.4859]]}
       middle_02_l_ctrl.Rotation.Z {type=MovieSceneScriptingFloatChannel, default=0, thinned_from=11, keys [[1895,0],[3927,0],[4164,-36.826]]}
       middle_02_l_ctrl.Scale.X {type=MovieSceneScriptingFloatChannel, default=0, const [0,10]}
       middle_02_l_ctrl.Scale.Y {type=MovieSceneScriptingFloatChannel, default=0, const [0,10]}
       middle_02_l_ctrl.Scale.Z {type=MovieSceneScriptingFloatChannel, default=0, const [0,10]}
       middle_03_l_ctrl.Location.X {type=MovieSceneScriptingFloatChannel, default=0, const [0,10]}
       middle_03_l_ctrl.Location.Y {type=MovieSceneScriptingFloatChannel, default=0, const [0,10]}
       middle_03_l_ctrl.Location.Z {type=MovieSceneScriptingFloatChannel, default=0, const [0,10]}
       middle_03_l_ctrl.Rotation.X {type=MovieSceneScriptingFloatChannel, default=0, thinned_from=11, keys [[1895,0],[3927,0],[4164,-0.1777]]}
       middle_03_l_ctrl.Rotation.Y {type=MovieSceneScriptingFloatChannel, default=0, thinned_from=11, keys [[1895,0],[3927,0],[4164,-0.1795]]}
       middle_03_l_ctrl.Rotation.Z {type=MovieSceneScriptingFloatChannel, default=0, thinned_from=11, keys [[1895,0],[3927,0],[4164,-36.8289]]}
       middle_03_l_ctrl.Scale.X {type=MovieSceneScriptingFloatChannel, default=0, const [0,10]}
       middle_03_l_ctrl.Scale.Y {type=MovieSceneScriptingFloatChannel, default=0, const [0,10]}
       middle_03_l_ctrl.Scale.Z {type=MovieSceneScriptingFloatChannel, default=0, const [0,10]}
       ring_metacarpal_l_ctrl.Location.X {type=MovieSceneScriptingFloatChannel, default=0, const [0,9]}
       ring_metacarpal_l_ctrl.Location.Y {type=MovieSceneScriptingFloatChannel, default=0, const [0,9]}
       ring_metacarpal_l_ctrl.Location.Z {type=MovieSceneScriptingFloatChannel, default=0, const [0,9]}
       ring_metacarpal_l_ctrl.Rotation.X {type=MovieSceneScriptingFloatChannel, default=0, const [0,9]}
       ring_metacarpal_l_ctrl.Rotation.Y {type=MovieSceneScriptingFloatChannel, default=0, const [0,9]}
       ring_metacarpal_l_ctrl.Rotation.Z {type=MovieSceneScriptingFloatChannel, default=0, const [0,9]}
       ring_metacarpal_l_ctrl.Scale.X {type=MovieSceneScriptingFloatChannel, default=0, const [0,9]}
       ring_metacarpal_l_ctrl.Scale.Y {type=MovieSceneScriptingFloatChannel, default=0, const [0,9]}
       ring_metacarpal_l_ctrl.Scale.Z {type=MovieSceneScriptingFloatChannel, default=0, const [0,9]}
       ring_01_l_ctrl.Location.X {type=MovieSceneScriptingFloatChannel, default=0, const [0,10]}
       ring_01_l_ctrl.Location.Y {type=MovieSceneScriptingFloatChannel, default=0, const [0,10]}
       ring_01_l_ctrl.Location.Z {type=MovieSceneScriptingFloatChannel, default=0, const [0,10]}
       ring_01_l_ctrl.Rotation.X {type=MovieSceneScriptingFloatChannel, default=0, thinned_from=11, keys [[1895,0],[3927,0],[4164,-6.7579]]}
       ring_01_l_ctrl.Rotation.Y {type=MovieSceneScriptingFloatChannel, default=0, thinned_from=11, keys [[1895,0],[3927,0],[4164,8.572]]}
       ring_01_l_ctrl.Rotation.Z {type=MovieSceneScriptingFloatChannel, default=0, thinned_from=11, keys [[1895,0],[3927,0],[4164,-36.4619]]}
       ring_01_l_ctrl.Scale.X {type=MovieSceneScriptingFloatChannel, default=0, const [0,10]}
       ring_01_l_ctrl.Scale.Y {type=MovieSceneScriptingFloatChannel, default=0, const [0,10]}
       ring_01_l_ctrl.Scale.Z {type=MovieSceneScriptingFloatChannel, default=0, const [0,10]}
       ring_02_l_ctrl.Location.X {type=MovieSceneScriptingFloatChannel, default=0, const [0,10]}
       ring_02_l_ctrl.Location.Y {type=MovieSceneScriptingFloatChannel, default=0, const [0,10]}
       ring_02_l_ctrl.Location.Z {type=MovieSceneScriptingFloatChannel, default=0, const [0,10]}
       ring_02_l_ctrl.Rotation.X {type=MovieSceneScriptingFloatChannel, default=0, thinned_from=11, keys [[1895,0],[3927,0],[4164,-0.0388]]}
       ring_02_l_ctrl.Rotation.Y {type=MovieSceneScriptingFloatChannel, default=0, thinned_from=11, keys [[1895,0],[3927,0],[4164,-0.7104]]}
       ring_02_l_ctrl.Rotation.Z {type=MovieSceneScriptingFloatChannel, default=0, thinned_from=11, keys [[1895,0],[3927,0],[4164,-36.8223]]}
       ring_02_l_ctrl.Scale.X {type=MovieSceneScriptingFloatChannel, default=0, const [0,10]}
       ring_02_l_ctrl.Scale.Y {type=MovieSceneScriptingFloatChannel, default=0, const [0,10]}
       ring_02_l_ctrl.Scale.Z {type=MovieSceneScriptingFloatChannel, default=0, const [0,10]}
       ring_03_l_ctrl.Location.X {type=MovieSceneScriptingFloatChannel, default=0, const [0,10]}
       ring_03_l_ctrl.Location.Y {type=MovieSceneScriptingFloatChannel, default=0, const [0,10]}
       ring_03_l_ctrl.Location.Z {type=MovieSceneScriptingFloatChannel, default=0, const [0,10]}
       ring_03_l_ctrl.Rotation.X {type=MovieSceneScriptingFloatChannel, default=0, thinned_from=11, keys [[1895,0],[3927,0],[4164,-0.0073]]}
       ring_03_l_ctrl.Rotation.Y {type=MovieSceneScriptingFloatChannel, default=0, thinned_from=11, keys [[1895,0],[3927,0],[4164,-0.3119]]}
       ring_03_l_ctrl.Rotation.Z {type=MovieSceneScriptingFloatChannel, default=0, thinned_from=11, keys [[1895,0],[3927,0],[4164,-36.8276]]}
       ring_03_l_ctrl.Scale.X {type=MovieSceneScriptingFloatChannel, default=0, const [0,10]}
       ring_03_l_ctrl.Scale.Y {type=MovieSceneScriptingFloatChannel, default=0, const [0,10]}
       ring_03_l_ctrl.Scale.Z {type=MovieSceneScriptingFloatChannel, default=0, const [0,10]}
       pinky_metacarpal_l_ctrl.Location.X {type=MovieSceneScriptingFloatChannel, default=0, thinned_from=9, keys [[1895,0],[3417,0]]}
       pinky_metacarpal_l_ctrl.Location.Y {type=MovieSceneScriptingFloatChannel, default=0, thinned_from=9, keys [[1895,0],[3417,0]]}
       pinky_metacarpal_l_ctrl.Location.Z {type=MovieSceneScriptingFloatChannel, default=0, const [0,9]}
       pinky_metacarpal_l_ctrl.Rotation.X {type=MovieSceneScriptingFloatChannel, default=0, const [0,9]}
       pinky_metacarpal_l_ctrl.Rotation.Y {type=MovieSceneScriptingFloatChannel, default=0, const [0,9]}
       pinky_metacarpal_l_ctrl.Rotation.Z {type=MovieSceneScriptingFloatChannel, default=0, const [0,9]}
       pinky_metacarpal_l_ctrl.Scale.X {type=MovieSceneScriptingFloatChannel, default=0, const [0,9]}
       pinky_metacarpal_l_ctrl.Scale.Y {type=MovieSceneScriptingFloatChannel, default=0, const [0,9]}
       pinky_metacarpal_l_ctrl.Scale.Z {type=MovieSceneScriptingFloatChannel, default=0, const [0,9]}
       pinky_01_l_ctrl.Location.X {type=MovieSceneScriptingFloatChannel, default=0, const [0,10]}
       pinky_01_l_ctrl.Location.Y {type=MovieSceneScriptingFloatChannel, default=0, const [0,10]}
       pinky_01_l_ctrl.Location.Z {type=MovieSceneScriptingFloatChannel, default=0, const [0,10]}
       pinky_01_l_ctrl.Rotation.X {type=MovieSceneScriptingFloatChannel, default=0, thinned_from=11, keys [[1895,0],[3927,0],[4164,-26.5083]]}
       pinky_01_l_ctrl.Rotation.Y {type=MovieSceneScriptingFloatChannel, default=0, thinned_from=11, keys [[1895,0],[3927,0],[4164,-7.4498]]}
       pinky_01_l_ctrl.Rotation.Z {type=MovieSceneScriptingFloatChannel, default=0, thinned_from=11, keys [[1895,0],[3927,0],[4164,-43.9245]]}
       pinky_01_l_ctrl.Scale.X {type=MovieSceneScriptingFloatChannel, default=0, const [0,10]}
       pinky_01_l_ctrl.Scale.Y {type=MovieSceneScriptingFloatChannel, default=0, const [0,10]}
       pinky_01_l_ctrl.Scale.Z {type=MovieSceneScriptingFloatChannel, default=0, const [0,10]}
       pinky_02_l_ctrl.Location.X {type=MovieSceneScriptingFloatChannel, default=0, const [0,10]}
       pinky_02_l_ctrl.Location.Y {type=MovieSceneScriptingFloatChannel, default=0, const [0,10]}
       pinky_02_l_ctrl.Location.Z {type=MovieSceneScriptingFloatChannel, default=0, const [0,10]}
       pinky_02_l_ctrl.Rotation.X {type=MovieSceneScriptingFloatChannel, default=0, thinned_from=11, keys [[1895,0],[3927,0],[4164,0.158]]}
       pinky_02_l_ctrl.Rotation.Y {type=MovieSceneScriptingFloatChannel, default=0, thinned_from=11, keys [[1895,0],[3927,0],[4164,-1.6709]]}
       pinky_02_l_ctrl.Rotation.Z {type=MovieSceneScriptingFloatChannel, default=0, thinned_from=11, keys [[1895,0],[3927,0],[4164,-36.7926]]}
       pinky_02_l_ctrl.Scale.X {type=MovieSceneScriptingFloatChannel, default=0, const [0,10]}
       pinky_02_l_ctrl.Scale.Y {type=MovieSceneScriptingFloatChannel, default=0, const [0,10]}
       pinky_02_l_ctrl.Scale.Z {type=MovieSceneScriptingFloatChannel, default=0, const [0,10]}
       pinky_03_l_ctrl.Location.X {type=MovieSceneScriptingFloatChannel, default=0, const [0,10]}
       pinky_03_l_ctrl.Location.Y {type=MovieSceneScriptingFloatChannel, default=0, const [0,10]}
       pinky_03_l_ctrl.Location.Z {type=MovieSceneScriptingFloatChannel, default=0, const [0,10]}
       pinky_03_l_ctrl.Rotation.X {type=MovieSceneScriptingFloatChannel, default=0, thinned_from=11, keys [[1895,0],[3927,0],[4164,0.5672]]}
       pinky_03_l_ctrl.Rotation.Y {type=MovieSceneScriptingFloatChannel, default=0, thinned_from=11, keys [[1895,0],[3927,0],[4164,-0.6728]]}
       pinky_03_l_ctrl.Rotation.Z {type=MovieSceneScriptingFloatChannel, default=0, thinned_from=11, keys [[1895,0],[3927,0],[4164,-36.8272]]}
       pinky_03_l_ctrl.Scale.X {type=MovieSceneScriptingFloatChannel, default=0, const [0,10]}
       pinky_03_l_ctrl.Scale.Y {type=MovieSceneScriptingFloatChannel, default=0, const [0,10]}
       pinky_03_l_ctrl.Scale.Z {type=MovieSceneScriptingFloatChannel, default=0, const [0,10]}
       upperarm_r_fk_ctrl.Location.X {type=MovieSceneScriptingFloatChannel, default=0, const [0,9]}
       upperarm_r_fk_ctrl.Location.Y {type=MovieSceneScriptingFloatChannel, default=0, const [0,9]}
       upperarm_r_fk_ctrl.Location.Z {type=MovieSceneScriptingFloatChannel, default=0, const [0,9]}
       upperarm_r_fk_ctrl.Rotation.X {
        type=MovieSceneScriptingFloatChannel
        default=0
        thinned_from=10
        keys [[1925,0],[1973,0],[2013,0.0009],[2014,0],[2053,0],[2095,0.0507],[2130,-0.0303],[2168,0.095],[3417,0.095]]
       }
       upperarm_r_fk_ctrl.Rotation.Y {
        type=MovieSceneScriptingFloatChannel
        default=0
        thinned_from=10
        keys [[1925,0],[1973,0],[2013,-0.057],[2014,0],[2053,0],[2095,-4.836],[2130,1.2543],[2168,-6.2179],[3417,-6.2179]]
       }
       upperarm_r_fk_ctrl.Rotation.Z {
        type=MovieSceneScriptingFloatChannel
        default=0
        thinned_from=10
        keys [[1925,0],[1973,0],[2013,0.0048],[2014,0],[2053,0],[2095,0.3818],[2130,-0.1055],[2168,0.6884],[3417,0.6884]]
       }
       upperarm_r_fk_ctrl.Scale.X {type=MovieSceneScriptingFloatChannel, default=0, const [0,9]}
       upperarm_r_fk_ctrl.Scale.Y {type=MovieSceneScriptingFloatChannel, default=0, const [0,9]}
       upperarm_r_fk_ctrl.Scale.Z {type=MovieSceneScriptingFloatChannel, default=0, const [0,9]}
       hand_r_fk_ctrl.Location.X {type=MovieSceneScriptingFloatChannel, default=0, thinned_from=9, keys [[1895,0],[3417,0]]}
       hand_r_fk_ctrl.Location.Y {type=MovieSceneScriptingFloatChannel, default=0, thinned_from=9, keys [[1895,0],[3417,0]]}
       hand_r_fk_ctrl.Location.Z {type=MovieSceneScriptingFloatChannel, default=0, thinned_from=9, keys [[1895,0],[3417,0]]}
       hand_r_fk_ctrl.Rotation.X {
        type=MovieSceneScriptingFloatChannel
        default=0
        thinned_from=9
        keys [[1895,0],[1984,0],[2023,-2.852],[2063,0],[2102,-2.0695],[2138,0],[3417,0]]
       }
       hand_r_fk_ctrl.Rotation.Y {
        type=MovieSceneScriptingFloatChannel
        default=0
        thinned_from=9
        keys [[1895,0],[1984,0],[2023,-0.4248],[2063,0],[2102,-0.1444],[2138,0],[3417,0]]
       }
       hand_r_fk_ctrl.Rotation.Z {
        type=MovieSceneScriptingFloatChannel
        default=0
        thinned_from=9
        keys [[1895,0],[1984,0],[2023,-26.0387],[2063,0],[2102,-18.4649],[2138,0],[3417,0]]
       }
       hand_r_fk_ctrl.Scale.X {type=MovieSceneScriptingFloatChannel, default=0, const [0,9]}
       hand_r_fk_ctrl.Scale.Y {type=MovieSceneScriptingFloatChannel, default=0, const [0,9]}
       hand_r_fk_ctrl.Scale.Z {type=MovieSceneScriptingFloatChannel, default=0, const [0,9]}
       thumb_01_r_ctrl.Location.X {type=MovieSceneScriptingFloatChannel, default=0, const [0,9]}
       thumb_01_r_ctrl.Location.Y {type=MovieSceneScriptingFloatChannel, default=0, const [0,9]}
       thumb_01_r_ctrl.Location.Z {type=MovieSceneScriptingFloatChannel, default=0, const [0,9]}
       thumb_01_r_ctrl.Rotation.X {type=MovieSceneScriptingFloatChannel, default=0, const [0,9]}
       thumb_01_r_ctrl.Rotation.Y {type=MovieSceneScriptingFloatChannel, default=0, const [0,9]}
       thumb_01_r_ctrl.Rotation.Z {type=MovieSceneScriptingFloatChannel, default=0, const [0,9]}
       thumb_01_r_ctrl.Scale.X {type=MovieSceneScriptingFloatChannel, default=0, const [0,9]}
       thumb_01_r_ctrl.Scale.Y {type=MovieSceneScriptingFloatChannel, default=0, const [0,9]}
       thumb_01_r_ctrl.Scale.Z {type=MovieSceneScriptingFloatChannel, default=0, const [0,9]}
       thumb_02_r_ctrl.Location.X {type=MovieSceneScriptingFloatChannel, default=0, const [0,9]}
       thumb_02_r_ctrl.Location.Y {type=MovieSceneScriptingFloatChannel, default=0, const [0,9]}
       thumb_02_r_ctrl.Location.Z {type=MovieSceneScriptingFloatChannel, default=0, const [0,9]}
       thumb_02_r_ctrl.Rotation.X {type=MovieSceneScriptingFloatChannel, default=0, const [0,9]}
       thumb_02_r_ctrl.Rotation.Y {type=MovieSceneScriptingFloatChannel, default=0, const [0,9]}
       thumb_02_r_ctrl.Rotation.Z {type=MovieSceneScriptingFloatChannel, default=0, const [0,9]}
       thumb_02_r_ctrl.Scale.X {type=MovieSceneScriptingFloatChannel, default=0, const [0,9]}
       thumb_02_r_ctrl.Scale.Y {type=MovieSceneScriptingFloatChannel, default=0, const [0,9]}
       thumb_02_r_ctrl.Scale.Z {type=MovieSceneScriptingFloatChannel, default=0, const [0,9]}
       thumb_03_r_ctrl.Location.X {type=MovieSceneScriptingFloatChannel, default=0, const [0,9]}
       thumb_03_r_ctrl.Location.Y {type=MovieSceneScriptingFloatChannel, default=0, const [0,9]}
       thumb_03_r_ctrl.Location.Z {type=MovieSceneScriptingFloatChannel, default=0, const [0,9]}
       thumb_03_r_ctrl.Rotation.X {type=MovieSceneScriptingFloatChannel, default=0, const [0,9]}
       thumb_03_r_ctrl.Rotation.Y {type=MovieSceneScriptingFloatChannel, default=0, const [0,9]}
       thumb_03_r_ctrl.Rotation.Z {type=MovieSceneScriptingFloatChannel, default=0, const [0,9]}
       thumb_03_r_ctrl.Scale.X {type=MovieSceneScriptingFloatChannel, default=0, const [0,9]}
       thumb_03_r_ctrl.Scale.Y {type=MovieSceneScriptingFloatChannel, default=0, const [0,9]}
       thumb_03_r_ctrl.Scale.Z {type=MovieSceneScriptingFloatChannel, default=0, const [0,9]}
       index_metacarpal_r_ctrl.Location.X {type=MovieSceneScriptingFloatChannel, default=0, const [0,9]}
       index_metacarpal_r_ctrl.Location.Y {type=MovieSceneScriptingFloatChannel, default=0, const [0,9]}
       index_metacarpal_r_ctrl.Location.Z {type=MovieSceneScriptingFloatChannel, default=0, const [0,9]}
       index_metacarpal_r_ctrl.Rotation.X {type=MovieSceneScriptingFloatChannel, default=0, const [0,9]}
       index_metacarpal_r_ctrl.Rotation.Y {type=MovieSceneScriptingFloatChannel, default=0, const [0,9]}
       index_metacarpal_r_ctrl.Rotation.Z {type=MovieSceneScriptingFloatChannel, default=0, const [0,9]}
       index_metacarpal_r_ctrl.Scale.X {type=MovieSceneScriptingFloatChannel, default=0, const [0,9]}
       index_metacarpal_r_ctrl.Scale.Y {type=MovieSceneScriptingFloatChannel, default=0, const [0,9]}
       index_metacarpal_r_ctrl.Scale.Z {type=MovieSceneScriptingFloatChannel, default=0, const [0,9]}
       index_01_r_ctrl.Location.X {type=MovieSceneScriptingFloatChannel, default=0, const [0,10]}
       index_01_r_ctrl.Location.Y {type=MovieSceneScriptingFloatChannel, default=0, const [0,10]}
       index_01_r_ctrl.Location.Z {type=MovieSceneScriptingFloatChannel, default=0, const [0,10]}
       index_01_r_ctrl.Rotation.X {type=MovieSceneScriptingFloatChannel, default=0, thinned_from=11, keys [[1895,0],[3906,0],[4140,12.9178]]}
       index_01_r_ctrl.Rotation.Y {type=MovieSceneScriptingFloatChannel, default=0, thinned_from=11, keys [[1895,0],[3906,0],[4140,16.5599]]}
       index_01_r_ctrl.Rotation.Z {type=MovieSceneScriptingFloatChannel, default=0, thinned_from=11, keys [[1895,0],[3906,0],[4140,-33.1195]]}
       index_01_r_ctrl.Scale.X {type=MovieSceneScriptingFloatChannel, default=0, const [0,10]}
       index_01_r_ctrl.Scale.Y {type=MovieSceneScriptingFloatChannel, default=0, const [0,10]}
       index_01_r_ctrl.Scale.Z {type=MovieSceneScriptingFloatChannel, default=0, const [0,10]}
       index_02_r_ctrl.Location.X {type=MovieSceneScriptingFloatChannel, default=0, thinned_from=10, keys [[1895,0],[3906,0]]}
       index_02_r_ctrl.Location.Y {type=MovieSceneScriptingFloatChannel, default=0, thinned_from=10, keys [[1895,0],[3906,0]]}
       index_02_r_ctrl.Location.Z {type=MovieSceneScriptingFloatChannel, default=0, thinned_from=10, keys [[1895,0],[3906,0]]}
       index_02_r_ctrl.Rotation.X {type=MovieSceneScriptingFloatChannel, default=0, thinned_from=11, keys [[1895,0],[3906,0],[4140,-1.3243]]}
       index_02_r_ctrl.Rotation.Y {type=MovieSceneScriptingFloatChannel, default=0, thinned_from=11, keys [[1895,0],[3906,0],[4140,1.7752]]}
       index_02_r_ctrl.Rotation.Z {type=MovieSceneScriptingFloatChannel, default=0, thinned_from=11, keys [[1895,0],[3906,0],[4140,-34.6459]]}
       index_02_r_ctrl.Scale.X {type=MovieSceneScriptingFloatChannel, default=0, const [0,10]}
       index_02_r_ctrl.Scale.Y {type=MovieSceneScriptingFloatChannel, default=0, const [0,10]}
       index_02_r_ctrl.Scale.Z {type=MovieSceneScriptingFloatChannel, default=0, const [0,10]}
       index_03_r_ctrl.Location.X {type=MovieSceneScriptingFloatChannel, default=0, const [0,10]}
       index_03_r_ctrl.Location.Y {type=MovieSceneScriptingFloatChannel, default=0, const [0,10]}
       index_03_r_ctrl.Location.Z {type=MovieSceneScriptingFloatChannel, default=0, const [0,10]}
       index_03_r_ctrl.Rotation.X {type=MovieSceneScriptingFloatChannel, default=0, thinned_from=11, keys [[1895,0],[3906,0],[4140,0.3926]]}
       index_03_r_ctrl.Rotation.Y {type=MovieSceneScriptingFloatChannel, default=0, thinned_from=11, keys [[1895,0],[3906,0],[4140,0.2873]]}
       index_03_r_ctrl.Rotation.Z {type=MovieSceneScriptingFloatChannel, default=0, thinned_from=11, keys [[1895,0],[3906,0],[4140,-34.6665]]}
       index_03_r_ctrl.Scale.X {type=MovieSceneScriptingFloatChannel, default=0, const [0,10]}
       index_03_r_ctrl.Scale.Y {type=MovieSceneScriptingFloatChannel, default=0, const [0,10]}
       index_03_r_ctrl.Scale.Z {type=MovieSceneScriptingFloatChannel, default=0, const [0,10]}
       middle_metacarpal_r_ctrl.Location.X {type=MovieSceneScriptingFloatChannel, default=0, thinned_from=9, keys [[1895,0],[3417,0]]}
       middle_metacarpal_r_ctrl.Location.Y {type=MovieSceneScriptingFloatChannel, default=0, thinned_from=9, keys [[1895,0],[3417,0]]}
       middle_metacarpal_r_ctrl.Location.Z {type=MovieSceneScriptingFloatChannel, default=0, thinned_from=9, keys [[1895,0],[3417,0]]}
       middle_metacarpal_r_ctrl.Rotation.X {type=MovieSceneScriptingFloatChannel, default=0, const [0,9]}
       middle_metacarpal_r_ctrl.Rotation.Y {type=MovieSceneScriptingFloatChannel, default=0, const [0,9]}
       middle_metacarpal_r_ctrl.Rotation.Z {type=MovieSceneScriptingFloatChannel, default=0, const [0,9]}
       middle_metacarpal_r_ctrl.Scale.X {type=MovieSceneScriptingFloatChannel, default=0, const [0,9]}
       middle_metacarpal_r_ctrl.Scale.Y {type=MovieSceneScriptingFloatChannel, default=0, const [0,9]}
       middle_metacarpal_r_ctrl.Scale.Z {type=MovieSceneScriptingFloatChannel, default=0, const [0,9]}
       middle_01_r_ctrl.Location.X {type=MovieSceneScriptingFloatChannel, default=0, thinned_from=10, keys [[1895,0],[3906,0]]}
       middle_01_r_ctrl.Location.Y {type=MovieSceneScriptingFloatChannel, default=0, thinned_from=10, keys [[1895,0],[3906,0]]}
       middle_01_r_ctrl.Location.Z {type=MovieSceneScriptingFloatChannel, default=0, thinned_from=10, keys [[1895,0],[3906,0]]}
       middle_01_r_ctrl.Rotation.X {type=MovieSceneScriptingFloatChannel, default=0, thinned_from=11, keys [[1895,0],[3906,0],[4140,9.5299]]}
       middle_01_r_ctrl.Rotation.Y {type=MovieSceneScriptingFloatChannel, default=0, thinned_from=11, keys [[1895,0],[3906,0],[4140,3.6636]]}
       middle_01_r_ctrl.Rotation.Z {type=MovieSceneScriptingFloatChannel, default=0, thinned_from=11, keys [[1895,0],[3906,0],[4140,-35.7315]]}
       middle_01_r_ctrl.Scale.X {type=MovieSceneScriptingFloatChannel, default=0, const [0,10]}
       middle_01_r_ctrl.Scale.Y {type=MovieSceneScriptingFloatChannel, default=0, const [0,10]}
       middle_01_r_ctrl.Scale.Z {type=MovieSceneScriptingFloatChannel, default=0, const [0,10]}
       middle_02_r_ctrl.Location.X {type=MovieSceneScriptingFloatChannel, default=0, const [0,10]}
       middle_02_r_ctrl.Location.Y {type=MovieSceneScriptingFloatChannel, default=0, const [0,10]}
       middle_02_r_ctrl.Location.Z {type=MovieSceneScriptingFloatChannel, default=0, const [0,10]}
       middle_02_r_ctrl.Rotation.X {type=MovieSceneScriptingFloatChannel, default=0, thinned_from=11, keys [[1895,0],[3906,0],[4140,1.6203]]}
       middle_02_r_ctrl.Rotation.Y {type=MovieSceneScriptingFloatChannel, default=0, thinned_from=11, keys [[1895,0],[3906,0],[4140,1.2633]]}
       middle_02_r_ctrl.Rotation.Z {type=MovieSceneScriptingFloatChannel, default=0, thinned_from=11, keys [[1895,0],[3906,0],[4140,-34.6799]]}
       middle_02_r_ctrl.Scale.X {type=MovieSceneScriptingFloatChannel, default=0, const [0,10]}
       middle_02_r_ctrl.Scale.Y {type=MovieSceneScriptingFloatChannel, default=0, const [0,10]}
       middle_02_r_ctrl.Scale.Z {type=MovieSceneScriptingFloatChannel, default=0, const [0,10]}
       middle_03_r_ctrl.Location.X {type=MovieSceneScriptingFloatChannel, default=0, const [0,10]}
       middle_03_r_ctrl.Location.Y {type=MovieSceneScriptingFloatChannel, default=0, const [0,10]}
       middle_03_r_ctrl.Location.Z {type=MovieSceneScriptingFloatChannel, default=0, const [0,10]}
       middle_03_r_ctrl.Rotation.X {type=MovieSceneScriptingFloatChannel, default=0, thinned_from=11, keys [[1895,0],[3906,0],[4140,1.1272]]}
       middle_03_r_ctrl.Rotation.Y {type=MovieSceneScriptingFloatChannel, default=0, thinned_from=11, keys [[1895,0],[3906,0],[4140,-0.5088]]}
       middle_03_r_ctrl.Rotation.Z {type=MovieSceneScriptingFloatChannel, default=0, thinned_from=11, keys [[1895,0],[3906,0],[4140,-34.6796]]}
       middle_03_r_ctrl.Scale.X {type=MovieSceneScriptingFloatChannel, default=0, const [0,10]}
       middle_03_r_ctrl.Scale.Y {type=MovieSceneScriptingFloatChannel, default=0, const [0,10]}
       middle_03_r_ctrl.Scale.Z {type=MovieSceneScriptingFloatChannel, default=0, const [0,10]}
       ring_metacarpal_r_ctrl.Location.X {type=MovieSceneScriptingFloatChannel, default=0, const [0,9]}
       ring_metacarpal_r_ctrl.Location.Y {type=MovieSceneScriptingFloatChannel, default=0, const [0,9]}
       ring_metacarpal_r_ctrl.Location.Z {type=MovieSceneScriptingFloatChannel, default=0, const [0,9]}
       ring_metacarpal_r_ctrl.Rotation.X {type=MovieSceneScriptingFloatChannel, default=0, const [0,9]}
       ring_metacarpal_r_ctrl.Rotation.Y {type=MovieSceneScriptingFloatChannel, default=0, const [0,9]}
       ring_metacarpal_r_ctrl.Rotation.Z {type=MovieSceneScriptingFloatChannel, default=0, const [0,9]}
       ring_metacarpal_r_ctrl.Scale.X {type=MovieSceneScriptingFloatChannel, default=0, const [0,9]}
       ring_metacarpal_r_ctrl.Scale.Y {type=MovieSceneScriptingFloatChannel, default=0, const [0,9]}
       ring_metacarpal_r_ctrl.Scale.Z {type=MovieSceneScriptingFloatChannel, default=0, const [0,9]}
       ring_01_r_ctrl.Location.X {type=MovieSceneScriptingFloatChannel, default=0, const [0,10]}
       ring_01_r_ctrl.Location.Y {type=MovieSceneScriptingFloatChannel, default=0, const [0,10]}
       ring_01_r_ctrl.Location.Z {type=MovieSceneScriptingFloatChannel, default=0, const [0,10]}
       ring_01_r_ctrl.Rotation.X {type=MovieSceneScriptingFloatChannel, default=0, thinned_from=11, keys [[1895,0],[3906,0],[4140,0.3134]]}
       ring_01_r_ctrl.Rotation.Y {type=MovieSceneScriptingFloatChannel, default=0, thinned_from=11, keys [[1895,0],[3906,0],[4140,7.6521]]}
       ring_01_r_ctrl.Rotation.Z {type=MovieSceneScriptingFloatChannel, default=0, thinned_from=11, keys [[1895,0],[3906,0],[4140,-33.8371]]}
       ring_01_r_ctrl.Scale.X {type=MovieSceneScriptingFloatChannel, default=0, const [0,10]}
       ring_01_r_ctrl.Scale.Y {type=MovieSceneScriptingFloatChannel, default=0, const [0,10]}
       ring_01_r_ctrl.Scale.Z {type=MovieSceneScriptingFloatChannel, default=0, const [0,10]}
       ring_02_r_ctrl.Location.X {type=MovieSceneScriptingFloatChannel, default=0, const [0,10]}
       ring_02_r_ctrl.Location.Y {type=MovieSceneScriptingFloatChannel, default=0, const [0,10]}
       ring_02_r_ctrl.Location.Z {type=MovieSceneScriptingFloatChannel, default=0, const [0,10]}
       ring_02_r_ctrl.Rotation.X {type=MovieSceneScriptingFloatChannel, default=0, thinned_from=11, keys [[1895,0],[3906,0],[4140,-0.0653]]}
       ring_02_r_ctrl.Rotation.Y {type=MovieSceneScriptingFloatChannel, default=0, thinned_from=11, keys [[1895,0],[3906,0],[4140,-0.6551]]}
       ring_02_r_ctrl.Rotation.Z {type=MovieSceneScriptingFloatChannel, default=0, thinned_from=11, keys [[1895,0],[3906,0],[4140,-34.6596]]}
       ring_02_r_ctrl.Scale.X {type=MovieSceneScriptingFloatChannel, default=0, const [0,10]}
       ring_02_r_ctrl.Scale.Y {type=MovieSceneScriptingFloatChannel, default=0, const [0,10]}
       ring_02_r_ctrl.Scale.Z {type=MovieSceneScriptingFloatChannel, default=0, const [0,10]}
       ring_03_r_ctrl.Location.X {type=MovieSceneScriptingFloatChannel, default=0, const [0,10]}
       ring_03_r_ctrl.Location.Y {type=MovieSceneScriptingFloatChannel, default=0, const [0,10]}
       ring_03_r_ctrl.Location.Z {type=MovieSceneScriptingFloatChannel, default=0, const [0,10]}
       ring_03_r_ctrl.Rotation.X {type=MovieSceneScriptingFloatChannel, default=0, thinned_from=11, keys [[1895,0],[3906,0],[4140,-0.0192]]}
       ring_03_r_ctrl.Rotation.Y {type=MovieSceneScriptingFloatChannel, default=0, thinned_from=11, keys [[1895,0],[3906,0],[4140,-0.2877]]}
       ring_03_r_ctrl.Rotation.Z {type=MovieSceneScriptingFloatChannel, default=0, thinned_from=11, keys [[1895,0],[3906,0],[4140,-34.6644]]}
       ring_03_r_ctrl.Scale.X {type=MovieSceneScriptingFloatChannel, default=0, const [0,10]}
       ring_03_r_ctrl.Scale.Y {type=MovieSceneScriptingFloatChannel, default=0, const [0,10]}
       ring_03_r_ctrl.Scale.Z {type=MovieSceneScriptingFloatChannel, default=0, const [0,10]}
       pinky_metacarpal_r_ctrl.Location.X {type=MovieSceneScriptingFloatChannel, default=0, thinned_from=9, keys [[1895,0],[3417,0]]}
       pinky_metacarpal_r_ctrl.Location.Y {type=MovieSceneScriptingFloatChannel, default=0, thinned_from=9, keys [[1895,0],[3417,0]]}
       pinky_metacarpal_r_ctrl.Location.Z {type=MovieSceneScriptingFloatChannel, default=0, thinned_from=9, keys [[1895,0],[3417,0]]}
       pinky_metacarpal_r_ctrl.Rotation.X {type=MovieSceneScriptingFloatChannel, default=0, const [0,9]}
       pinky_metacarpal_r_ctrl.Rotation.Y {type=MovieSceneScriptingFloatChannel, default=0, const [0,9]}
       pinky_metacarpal_r_ctrl.Rotation.Z {type=MovieSceneScriptingFloatChannel, default=0, const [0,9]}
       pinky_metacarpal_r_ctrl.Scale.X {type=MovieSceneScriptingFloatChannel, default=0, const [0,9]}
       pinky_metacarpal_r_ctrl.Scale.Y {type=MovieSceneScriptingFloatChannel, default=0, const [0,9]}
       pinky_metacarpal_r_ctrl.Scale.Z {type=MovieSceneScriptingFloatChannel, default=0, const [0,9]}
       pinky_01_r_ctrl.Location.X {type=MovieSceneScriptingFloatChannel, default=0, thinned_from=10, keys [[1895,0],[3906,0]]}
       pinky_01_r_ctrl.Location.Y {type=MovieSceneScriptingFloatChannel, default=0, thinned_from=10, keys [[1895,0],[3906,0]]}
       pinky_01_r_ctrl.Location.Z {type=MovieSceneScriptingFloatChannel, default=0, thinned_from=10, keys [[1895,0],[3906,0]]}
       pinky_01_r_ctrl.Rotation.X {type=MovieSceneScriptingFloatChannel, default=0, thinned_from=11, keys [[1895,0],[3906,0],[4140,-22.0456]]}
       pinky_01_r_ctrl.Rotation.Y {type=MovieSceneScriptingFloatChannel, default=0, thinned_from=11, keys [[1895,0],[3906,0],[4140,1.6667]]}
       pinky_01_r_ctrl.Rotation.Z {type=MovieSceneScriptingFloatChannel, default=0, thinned_from=11, keys [[1895,0],[3906,0],[4140,-40.8666]]}
       pinky_01_r_ctrl.Scale.X {type=MovieSceneScriptingFloatChannel, default=0, const [0,10]}
       pinky_01_r_ctrl.Scale.Y {type=MovieSceneScriptingFloatChannel, default=0, const [0,10]}
       pinky_01_r_ctrl.Scale.Z {type=MovieSceneScriptingFloatChannel, default=0, const [0,10]}
       pinky_02_r_ctrl.Location.X {type=MovieSceneScriptingFloatChannel, default=0, thinned_from=10, keys [[1895,0],[3906,0]]}
       pinky_02_r_ctrl.Location.Y {type=MovieSceneScriptingFloatChannel, default=0, thinned_from=10, keys [[1895,0],[3906,0]]}
       pinky_02_r_ctrl.Location.Z {type=MovieSceneScriptingFloatChannel, default=0, thinned_from=10, keys [[1895,0],[3906,0]]}
       pinky_02_r_ctrl.Rotation.X {type=MovieSceneScriptingFloatChannel, default=0, thinned_from=11, keys [[1895,0],[3906,0],[4140,0.0505]]}
       pinky_02_r_ctrl.Rotation.Y {type=MovieSceneScriptingFloatChannel, default=0, thinned_from=11, keys [[1895,0],[3906,0],[4140,-1.5579]]}
       pinky_02_r_ctrl.Rotation.Z {type=MovieSceneScriptingFloatChannel, default=0, thinned_from=11, keys [[1895,0],[3906,0],[4140,-34.6316]]}
       pinky_02_r_ctrl.Scale.X {type=MovieSceneScriptingFloatChannel, default=0, const [0,10]}
       pinky_02_r_ctrl.Scale.Y {type=MovieSceneScriptingFloatChannel, default=0, const [0,10]}
       pinky_02_r_ctrl.Scale.Z {type=MovieSceneScriptingFloatChannel, default=0, const [0,10]}
       pinky_03_r_ctrl.Location.X {type=MovieSceneScriptingFloatChannel, default=0, thinned_from=10, keys [[1895,0],[3906,0]]}
       pinky_03_r_ctrl.Location.Y {type=MovieSceneScriptingFloatChannel, default=0, thinned_from=10, keys [[1895,0],[3906,0]]}
       pinky_03_r_ctrl.Location.Z {type=MovieSceneScriptingFloatChannel, default=0, thinned_from=10, keys [[1895,0],[3906,0]]}
       pinky_03_r_ctrl.Rotation.X {type=MovieSceneScriptingFloatChannel, default=0, thinned_from=11, keys [[1895,0],[3906,0],[4140,0.4817]]}
       pinky_03_r_ctrl.Rotation.Y {type=MovieSceneScriptingFloatChannel, default=0, thinned_from=11, keys [[1895,0],[3906,0],[4140,-0.6355]]}
       pinky_03_r_ctrl.Rotation.Z {type=MovieSceneScriptingFloatChannel, default=0, thinned_from=11, keys [[1895,0],[3906,0],[4140,-34.6631]]}
       pinky_03_r_ctrl.Scale.X {type=MovieSceneScriptingFloatChannel, default=0, const [0,10]}
       pinky_03_r_ctrl.Scale.Y {type=MovieSceneScriptingFloatChannel, default=0, const [0,10]}
       pinky_03_r_ctrl.Scale.Z {type=MovieSceneScriptingFloatChannel, default=0, const [0,10]}
       body_ctrl.Location.X {
        type=MovieSceneScriptingFloatChannel
        default=0
        thinned_from=16
        keys [[1895,0],[2776,0],[2799,-5.2435],[2854,0.3979],[2911,-1.6329],[2964,1.6424],[3005,0],[4041,0]]
       }
       body_ctrl.Location.Y {
        type=MovieSceneScriptingFloatChannel
        default=0
        thinned_from=17
        keys [[1895,0],[2776,0],[2799,20.7414],[2854,-0.8497],[2911,6.4011],[2964,-5.949],[3005,0],[3417,0],[3809,-82.6178],[4041,-82.6178]]
       }
       body_ctrl.Location.Z {
        type=MovieSceneScriptingFloatChannel
        default=0
        thinned_from=47
        keys [[1895,0],[1943,-28.2213],[1984,14.8],[2023,0.3879],[2063,29.2255],[2102,17.2971],[2138,40.4819],[2179,-32.8301],[2776,-32.8301],[2787,-61.2185],[2811,-138.276],[2868,-32.9355],[2911,-66.7238],[2964,-0.1588],[3005,-32.8301],[3417,-32.8301],[3465,-86.0466],[3505,-9.2561],[3519,-3.0679],[3547,-29.7252],[3558,-47.7832],[3579,-234.486],[3583,-251.1526],[3604,-62.4953],[3623,31.1958],[3649,41.5403],[3706,22.6697],[3717,28.6403],[3751,83.7216],[3778,136.911],[3809,172.2197],[3843,184.7462],[3886,153.7708],[3908,137.0463],[3940,166.692],[3978,143.9945],[4010,162.8288],[4036,156.214],[4060,160.719],[4094,153.7708],[4130,160.7419],[4165,152.6706],[4197,157.5412],[4224,153.7708],[4427,153.7708]]
       }
       body_ctrl.Rotation.X {type=MovieSceneScriptingFloatChannel, default=0, const [0,12]}
       body_ctrl.Rotation.Y {type=MovieSceneScriptingFloatChannel, default=0, const [0,13]}
       body_ctrl.Rotation.Z {type=MovieSceneScriptingFloatChannel, default=0, const [0,12]}
       body_ctrl.Scale.X {type=MovieSceneScriptingFloatChannel, default=0, const [0,12]}
       body_ctrl.Scale.Y {type=MovieSceneScriptingFloatChannel, default=0, const [0,12]}
       body_ctrl.Scale.Z {type=MovieSceneScriptingFloatChannel, default=0, const [0,12]}
       Weight {type=MovieSceneScriptingFloatChannel, default=1, const [1,9]}
      ]
     }
     {
      type=MovieSceneControlRigParameterSection
      channels [
       thigh_l_fk_ctrl.Location.X {type=MovieSceneScriptingFloatChannel, default=0, const [0,2]}
       thigh_l_fk_ctrl.Location.Y {type=MovieSceneScriptingFloatChannel, default=0, const [0,2]}
       thigh_l_fk_ctrl.Location.Z {type=MovieSceneScriptingFloatChannel, default=0, const [0,2]}
       thigh_l_fk_ctrl.Rotation.X {type=MovieSceneScriptingFloatChannel, default=0, const [0,2]}
       thigh_l_fk_ctrl.Rotation.Y {type=MovieSceneScriptingFloatChannel, default=0, const [0,2]}
       thigh_l_fk_ctrl.Rotation.Z {type=MovieSceneScriptingFloatChannel, default=0, const [0,2]}
       thigh_l_fk_ctrl.Scale.X {type=MovieSceneScriptingFloatChannel, default=0, const [0,2]}
       thigh_l_fk_ctrl.Scale.Y {type=MovieSceneScriptingFloatChannel, default=0, const [0,2]}
       thigh_l_fk_ctrl.Scale.Z {type=MovieSceneScriptingFloatChannel, default=0, const [0,2]}
       calf_l_fk_ctrl.Location.X {type=MovieSceneScriptingFloatChannel, default=0, const [0,2]}
       calf_l_fk_ctrl.Location.Y {type=MovieSceneScriptingFloatChannel, default=0, const [0,2]}
       calf_l_fk_ctrl.Location.Z {type=MovieSceneScriptingFloatChannel, default=0, const [0,2]}
       calf_l_fk_ctrl.Rotation.X {type=MovieSceneScriptingFloatChannel, default=0, const [0,2]}
       calf_l_fk_ctrl.Rotation.Y {type=MovieSceneScriptingFloatChannel, default=0, const [0,2]}
       calf_l_fk_ctrl.Rotation.Z {type=MovieSceneScriptingFloatChannel, default=0, const [0,2]}
       calf_l_fk_ctrl.Scale.X {type=MovieSceneScriptingFloatChannel, default=0, const [0,2]}
       calf_l_fk_ctrl.Scale.Y {type=MovieSceneScriptingFloatChannel, default=0, const [0,2]}
       calf_l_fk_ctrl.Scale.Z {type=MovieSceneScriptingFloatChannel, default=0, const [0,2]}
       foot_l_fk_ctrl.Location.X {type=MovieSceneScriptingFloatChannel, default=0, const [0,2]}
       foot_l_fk_ctrl.Location.Y {type=MovieSceneScriptingFloatChannel, default=0, const [0,2]}
       foot_l_fk_ctrl.Location.Z {type=MovieSceneScriptingFloatChannel, default=0, const [0,2]}
       foot_l_fk_ctrl.Rotation.X {type=MovieSceneScriptingFloatChannel, default=0, const [0,2]}
       foot_l_fk_ctrl.Rotation.Y {type=MovieSceneScriptingFloatChannel, default=0, const [0,2]}
       foot_l_fk_ctrl.Rotation.Z {type=MovieSceneScriptingFloatChannel, default=0, const [0,2]}
       foot_l_fk_ctrl.Scale.X {type=MovieSceneScriptingFloatChannel, default=0, const [0,2]}
       foot_l_fk_ctrl.Scale.Y {type=MovieSceneScriptingFloatChannel, default=0, const [0,2]}
       foot_l_fk_ctrl.Scale.Z {type=MovieSceneScriptingFloatChannel, default=0, const [0,2]}
       ball_l_fk_ctrl.Location.X {type=MovieSceneScriptingFloatChannel, default=0, const [0,2]}
       ball_l_fk_ctrl.Location.Y {type=MovieSceneScriptingFloatChannel, default=0, const [0,2]}
       ball_l_fk_ctrl.Location.Z {type=MovieSceneScriptingFloatChannel, default=0, const [0,2]}
       ball_l_fk_ctrl.Rotation.X {type=MovieSceneScriptingFloatChannel, default=0, const [0,2]}
       ball_l_fk_ctrl.Rotation.Y {type=MovieSceneScriptingFloatChannel, default=0, const [0,2]}
       ball_l_fk_ctrl.Rotation.Z {type=MovieSceneScriptingFloatChannel, default=0, const [0,2]}
       ball_l_fk_ctrl.Scale.X {type=MovieSceneScriptingFloatChannel, default=0, const [0,2]}
       ball_l_fk_ctrl.Scale.Y {type=MovieSceneScriptingFloatChannel, default=0, const [0,2]}
       ball_l_fk_ctrl.Scale.Z {type=MovieSceneScriptingFloatChannel, default=0, const [0,2]}
       foot_l_ik_ctrl.Location.X {type=MovieSceneScriptingFloatChannel, default=0, const [-213.0783,2]}
       foot_l_ik_ctrl.Location.Y {type=MovieSceneScriptingFloatChannel, default=0, const [393.0834,2]}
       foot_l_ik_ctrl.Location.Z {type=MovieSceneScriptingFloatChannel, default=0, const [155.77,2]}
       foot_l_ik_ctrl.Rotation.X {type=MovieSceneScriptingFloatChannel, default=0, const [44.8034,2]}
       foot_l_ik_ctrl.Rotation.Y {type=MovieSceneScriptingFloatChannel, default=0, const [25.9825,2]}
       foot_l_ik_ctrl.Rotation.Z {type=MovieSceneScriptingFloatChannel, default=0, const [-4.5759,2]}
       foot_l_ik_ctrl.Scale.X {type=MovieSceneScriptingFloatChannel, default=0, const [0.0017,2]}
       foot_l_ik_ctrl.Scale.Y {type=MovieSceneScriptingFloatChannel, default=0, const [0.0017,2]}
       foot_l_ik_ctrl.Scale.Z {type=MovieSceneScriptingFloatChannel, default=0, const [0.0017,2]}
       foot_bk1_l_ctrl.Location.X {type=MovieSceneScriptingFloatChannel, default=0, const [0,2]}
       foot_bk1_l_ctrl.Location.Y {type=MovieSceneScriptingFloatChannel, default=0, const [0,2]}
       foot_bk1_l_ctrl.Location.Z {type=MovieSceneScriptingFloatChannel, default=0, const [0,2]}
       foot_bk1_l_ctrl.Rotation.X {type=MovieSceneScriptingFloatChannel, default=0, const [0,2]}
       foot_bk1_l_ctrl.Rotation.Y {type=MovieSceneScriptingFloatChannel, default=0, const [0,2]}
       foot_bk1_l_ctrl.Rotation.Z {type=MovieSceneScriptingFloatChannel, default=0, const [0,2]}
       foot_bk1_l_ctrl.Scale.X {type=MovieSceneScriptingFloatChannel, default=0, const [0,2]}
       foot_bk1_l_ctrl.Scale.Y {type=MovieSceneScriptingFloatChannel, default=0, const [0,2]}
       foot_bk1_l_ctrl.Scale.Z {type=MovieSceneScriptingFloatChannel, default=0, const [0,2]}
       tip_l_ctrl.Location.X {type=MovieSceneScriptingFloatChannel, default=0, const [0,2]}
       tip_l_ctrl.Location.Y {type=MovieSceneScriptingFloatChannel, default=0, const [0,2]}
       tip_l_ctrl.Location.Z {type=MovieSceneScriptingFloatChannel, default=0, const [0,2]}
       tip_l_ctrl.Rotation.X {type=MovieSceneScriptingFloatChannel, default=0, const [0,2]}
       tip_l_ctrl.Rotation.Y {type=MovieSceneScriptingFloatChannel, default=0, const [0,2]}
       tip_l_ctrl.Rotation.Z {type=MovieSceneScriptingFloatChannel, default=0, const [0,2]}
       tip_l_ctrl.Scale.X {type=MovieSceneScriptingFloatChannel, default=0, const [0,2]}
       tip_l_ctrl.Scale.Y {type=MovieSceneScriptingFloatChannel, default=0, const [0,2]}
       tip_l_ctrl.Scale.Z {type=MovieSceneScriptingFloatChannel, default=0, const [0,2]}
       heel_l_ctrl.Location.X {type=MovieSceneScriptingFloatChannel, default=0, const [0,2]}
       heel_l_ctrl.Location.Y {type=MovieSceneScriptingFloatChannel, default=0, const [0,2]}
       heel_l_ctrl.Location.Z {type=MovieSceneScriptingFloatChannel, default=0, const [0,2]}
       heel_l_ctrl.Rotation.X {type=MovieSceneScriptingFloatChannel, default=0, const [0,2]}
       heel_l_ctrl.Rotation.Y {type=MovieSceneScriptingFloatChannel, default=0, const [0,2]}
       heel_l_ctrl.Rotation.Z {type=MovieSceneScriptingFloatChannel, default=0, const [0,2]}
       heel_l_ctrl.Scale.X {type=MovieSceneScriptingFloatChannel, default=0, const [0,2]}
       heel_l_ctrl.Scale.Y {type=MovieSceneScriptingFloatChannel, default=0, const [0,2]}
       heel_l_ctrl.Scale.Z {type=MovieSceneScriptingFloatChannel, default=0, const [0,2]}
       leg_l_softness {type=MovieSceneScriptingFloatChannel, default=0, const [0,2]}
       leg_l_pv_ik_ctrl.Location.X {type=MovieSceneScriptingFloatChannel, default=0, const [4.9004,2]}
       leg_l_pv_ik_ctrl.Location.Y {type=MovieSceneScriptingFloatChannel, default=0, const [359.2141,2]}
       leg_l_pv_ik_ctrl.Location.Z {type=MovieSceneScriptingFloatChannel, default=0, const [237.8187,2]}
       leg_l_pv_ik_ctrl.Rotation.X {type=MovieSceneScriptingFloatChannel, default=0, const [61.4321,2]}
       leg_l_pv_ik_ctrl.Rotation.Y {type=MovieSceneScriptingFloatChannel, default=0, const [2.9411,2]}
       leg_l_pv_ik_ctrl.Rotation.Z {type=MovieSceneScriptingFloatChannel, default=0, const [25.2171,2]}
       leg_l_pv_ik_ctrl.Scale.X {type=MovieSceneScriptingFloatChannel, default=0, const [-0.0155,2]}
       leg_l_pv_ik_ctrl.Scale.Y {type=MovieSceneScriptingFloatChannel, default=0, const [-0.0155,2]}
       leg_l_pv_ik_ctrl.Scale.Z {type=MovieSceneScriptingFloatChannel, default=0, const [-0.0155,2]}
       thigh_r_fk_ctrl.Location.X {type=MovieSceneScriptingFloatChannel, default=0, const [0,2]}
       thigh_r_fk_ctrl.Location.Y {type=MovieSceneScriptingFloatChannel, default=0, const [0,2]}
       thigh_r_fk_ctrl.Location.Z {type=MovieSceneScriptingFloatChannel, default=0, const [0,2]}
       thigh_r_fk_ctrl.Rotation.X {type=MovieSceneScriptingFloatChannel, default=0, const [0,2]}
       thigh_r_fk_ctrl.Rotation.Y {type=MovieSceneScriptingFloatChannel, default=0, const [0,2]}
       thigh_r_fk_ctrl.Rotation.Z {type=MovieSceneScriptingFloatChannel, default=0, const [0,2]}
       thigh_r_fk_ctrl.Scale.X {type=MovieSceneScriptingFloatChannel, default=0, const [0,2]}
       thigh_r_fk_ctrl.Scale.Y {type=MovieSceneScriptingFloatChannel, default=0, const [0,2]}
       thigh_r_fk_ctrl.Scale.Z {type=MovieSceneScriptingFloatChannel, default=0, const [0,2]}
       calf_r_fk_ctrl.Location.X {type=MovieSceneScriptingFloatChannel, default=0, const [0,2]}
       calf_r_fk_ctrl.Location.Y {type=MovieSceneScriptingFloatChannel, default=0, const [0,2]}
       calf_r_fk_ctrl.Location.Z {type=MovieSceneScriptingFloatChannel, default=0, const [0,2]}
       calf_r_fk_ctrl.Rotation.X {type=MovieSceneScriptingFloatChannel, default=0, const [0,2]}
       calf_r_fk_ctrl.Rotation.Y {type=MovieSceneScriptingFloatChannel, default=0, const [0,2]}
       calf_r_fk_ctrl.Rotation.Z {type=MovieSceneScriptingFloatChannel, default=0, const [0,2]}
       calf_r_fk_ctrl.Scale.X {type=MovieSceneScriptingFloatChannel, default=0, const [0,2]}
       calf_r_fk_ctrl.Scale.Y {type=MovieSceneScriptingFloatChannel, default=0, const [0,2]}
       calf_r_fk_ctrl.Scale.Z {type=MovieSceneScriptingFloatChannel, default=0, const [0,2]}
       foot_r_fk_ctrl.Location.X {type=MovieSceneScriptingFloatChannel, default=0, const [0,2]}
       foot_r_fk_ctrl.Location.Y {type=MovieSceneScriptingFloatChannel, default=0, const [0,2]}
       foot_r_fk_ctrl.Location.Z {type=MovieSceneScriptingFloatChannel, default=0, const [0,2]}
       foot_r_fk_ctrl.Rotation.X {type=MovieSceneScriptingFloatChannel, default=0, const [0,2]}
       foot_r_fk_ctrl.Rotation.Y {type=MovieSceneScriptingFloatChannel, default=0, const [0,2]}
       foot_r_fk_ctrl.Rotation.Z {type=MovieSceneScriptingFloatChannel, default=0, const [0,2]}
       foot_r_fk_ctrl.Scale.X {type=MovieSceneScriptingFloatChannel, default=0, const [0,2]}
       foot_r_fk_ctrl.Scale.Y {type=MovieSceneScriptingFloatChannel, default=0, const [0,2]}
       foot_r_fk_ctrl.Scale.Z {type=MovieSceneScriptingFloatChannel, default=0, const [0,2]}
       ball_r_fk_ctrl.Location.X {type=MovieSceneScriptingFloatChannel, default=0, const [0,2]}
       ball_r_fk_ctrl.Location.Y {type=MovieSceneScriptingFloatChannel, default=0, const [0,2]}
       ball_r_fk_ctrl.Location.Z {type=MovieSceneScriptingFloatChannel, default=0, const [0,2]}
       ball_r_fk_ctrl.Rotation.X {type=MovieSceneScriptingFloatChannel, default=0, const [0,2]}
       ball_r_fk_ctrl.Rotation.Y {type=MovieSceneScriptingFloatChannel, default=0, const [0,2]}
       ball_r_fk_ctrl.Rotation.Z {type=MovieSceneScriptingFloatChannel, default=0, const [0,2]}
       ball_r_fk_ctrl.Scale.X {type=MovieSceneScriptingFloatChannel, default=0, const [0,2]}
       ball_r_fk_ctrl.Scale.Y {type=MovieSceneScriptingFloatChannel, default=0, const [0,2]}
       ball_r_fk_ctrl.Scale.Z {type=MovieSceneScriptingFloatChannel, default=0, const [0,2]}
       foot_r_ik_ctrl.Location.X {type=MovieSceneScriptingFloatChannel, default=0, const [-263.093,2]}
       foot_r_ik_ctrl.Location.Y {type=MovieSceneScriptingFloatChannel, default=0, const [386.7035,2]}
       foot_r_ik_ctrl.Location.Z {type=MovieSceneScriptingFloatChannel, default=0, const [115.051,2]}
       foot_r_ik_ctrl.Rotation.X {type=MovieSceneScriptingFloatChannel, default=0, const [63.1239,2]}
       foot_r_ik_ctrl.Rotation.Y {type=MovieSceneScriptingFloatChannel, default=0, const [-21.1291,2]}
       foot_r_ik_ctrl.Rotation.Z {type=MovieSceneScriptingFloatChannel, default=0, const [-26.6026,2]}
       foot_r_ik_ctrl.Scale.X {type=MovieSceneScriptingFloatChannel, default=0, const [-0.0017,2]}
       foot_r_ik_ctrl.Scale.Y {type=MovieSceneScriptingFloatChannel, default=0, const [0.0017,2]}
       foot_r_ik_ctrl.Scale.Z {type=MovieSceneScriptingFloatChannel, default=0, const [0.0017,2]}
       foot_bk1_r_ctrl.Location.X {type=MovieSceneScriptingFloatChannel, default=0, const [0,2]}
       foot_bk1_r_ctrl.Location.Y {type=MovieSceneScriptingFloatChannel, default=0, const [0,2]}
       foot_bk1_r_ctrl.Location.Z {type=MovieSceneScriptingFloatChannel, default=0, const [0,2]}
       foot_bk1_r_ctrl.Rotation.X {type=MovieSceneScriptingFloatChannel, default=0, const [0,2]}
       foot_bk1_r_ctrl.Rotation.Y {type=MovieSceneScriptingFloatChannel, default=0, const [0,2]}
       foot_bk1_r_ctrl.Rotation.Z {type=MovieSceneScriptingFloatChannel, default=0, const [0,2]}
       foot_bk1_r_ctrl.Scale.X {type=MovieSceneScriptingFloatChannel, default=0, const [0,2]}
       foot_bk1_r_ctrl.Scale.Y {type=MovieSceneScriptingFloatChannel, default=0, const [0,2]}
       foot_bk1_r_ctrl.Scale.Z {type=MovieSceneScriptingFloatChannel, default=0, const [0,2]}
       tip_r_ctrl.Location.X {type=MovieSceneScriptingFloatChannel, default=0, const [0,2]}
       tip_r_ctrl.Location.Y {type=MovieSceneScriptingFloatChannel, default=0, const [0,2]}
       tip_r_ctrl.Location.Z {type=MovieSceneScriptingFloatChannel, default=0, const [0,2]}
       tip_r_ctrl.Rotation.X {type=MovieSceneScriptingFloatChannel, default=0, const [0,2]}
       tip_r_ctrl.Rotation.Y {type=MovieSceneScriptingFloatChannel, default=0, const [0,2]}
       tip_r_ctrl.Rotation.Z {type=MovieSceneScriptingFloatChannel, default=0, const [0,2]}
       tip_r_ctrl.Scale.X {type=MovieSceneScriptingFloatChannel, default=0, const [0,2]}
       tip_r_ctrl.Scale.Y {type=MovieSceneScriptingFloatChannel, default=0, const [0,2]}
       tip_r_ctrl.Scale.Z {type=MovieSceneScriptingFloatChannel, default=0, const [0,2]}
       heel_r_ctrl.Location.X {type=MovieSceneScriptingFloatChannel, default=0, const [0,2]}
       heel_r_ctrl.Location.Y {type=MovieSceneScriptingFloatChannel, default=0, const [0,2]}
       heel_r_ctrl.Location.Z {type=MovieSceneScriptingFloatChannel, default=0, const [0,2]}
       heel_r_ctrl.Rotation.X {type=MovieSceneScriptingFloatChannel, default=0, const [0,2]}
       heel_r_ctrl.Rotation.Y {type=MovieSceneScriptingFloatChannel, default=0, const [0,2]}
       heel_r_ctrl.Rotation.Z {type=MovieSceneScriptingFloatChannel, default=0, const [0,2]}
       heel_r_ctrl.Scale.X {type=MovieSceneScriptingFloatChannel, default=0, const [0,2]}
       heel_r_ctrl.Scale.Y {type=MovieSceneScriptingFloatChannel, default=0, const [0,2]}
       heel_r_ctrl.Scale.Z {type=MovieSceneScriptingFloatChannel, default=0, const [0,2]}
       leg_r_softness {type=MovieSceneScriptingFloatChannel, default=0, const [0,2]}
       leg_r_pv_ik_ctrl.Location.X {type=MovieSceneScriptingFloatChannel, default=0, const [42.8506,2]}
       leg_r_pv_ik_ctrl.Location.Y {type=MovieSceneScriptingFloatChannel, default=0, const [335.9202,2]}
       leg_r_pv_ik_ctrl.Location.Z {type=MovieSceneScriptingFloatChannel, default=0, const [221.975,2]}
       leg_r_pv_ik_ctrl.Rotation.X {type=MovieSceneScriptingFloatChannel, default=0, const [45.689,2]}
       leg_r_pv_ik_ctrl.Rotation.Y {type=MovieSceneScriptingFloatChannel, default=0, const [-17.7618,2]}
       leg_r_pv_ik_ctrl.Rotation.Z {type=MovieSceneScriptingFloatChannel, default=0, const [4.8566,2]}
       leg_r_pv_ik_ctrl.Scale.X {type=MovieSceneScriptingFloatChannel, default=0, const [-0.0086,2]}
       leg_r_pv_ik_ctrl.Scale.Y {type=MovieSceneScriptingFloatChannel, default=0, const [-0.0086,2]}
       leg_r_pv_ik_ctrl.Scale.Z {type=MovieSceneScriptingFloatChannel, default=0, const [-0.0086,2]}
       upperarm_l_fk_ctrl.Location.X {type=MovieSceneScriptingFloatChannel, default=0, const [0,2]}
       upperarm_l_fk_ctrl.Location.Y {type=MovieSceneScriptingFloatChannel, default=0, const [0,2]}
       upperarm_l_fk_ctrl.Location.Z {type=MovieSceneScriptingFloatChannel, default=0, const [0,2]}
       upperarm_l_fk_ctrl.Rotation.X {type=MovieSceneScriptingFloatChannel, default=0, const [0,2]}
       upperarm_l_fk_ctrl.Rotation.Y {type=MovieSceneScriptingFloatChannel, default=0, const [0,2]}
       upperarm_l_fk_ctrl.Rotation.Z {type=MovieSceneScriptingFloatChannel, default=0, const [0,2]}
       upperarm_l_fk_ctrl.Scale.X {type=MovieSceneScriptingFloatChannel, default=0, const [0,2]}
       upperarm_l_fk_ctrl.Scale.Y {type=MovieSceneScriptingFloatChannel, default=0, const [0,2]}
       upperarm_l_fk_ctrl.Scale.Z {type=MovieSceneScriptingFloatChannel, default=0, const [0,2]}
       lowerarm_l_fk_ctrl.Location.X {type=MovieSceneScriptingFloatChannel, default=0, thinned_from=3, keys [[3004,0],[3171,0]]}
       lowerarm_l_fk_ctrl.Location.Y {type=MovieSceneScriptingFloatChannel, default=0, thinned_from=3, keys [[3004,0],[3171,0]]}
       lowerarm_l_fk_ctrl.Location.Z {type=MovieSceneScriptingFloatChannel, default=0, thinned_from=3, keys [[3004,0],[3171,0]]}
       lowerarm_l_fk_ctrl.Rotation.X {type=MovieSceneScriptingFloatChannel, default=0, const [0,3]}
       lowerarm_l_fk_ctrl.Rotation.Y {type=MovieSceneScriptingFloatChannel, default=0, const [0,3]}
       lowerarm_l_fk_ctrl.Rotation.Z {type=MovieSceneScriptingFloatChannel, default=0, const [0,3]}
       lowerarm_l_fk_ctrl.Scale.X {type=MovieSceneScriptingFloatChannel, default=0, const [0,3]}
       lowerarm_l_fk_ctrl.Scale.Y {type=MovieSceneScriptingFloatChannel, default=0, const [0,3]}
       lowerarm_l_fk_ctrl.Scale.Z {type=MovieSceneScriptingFloatChannel, default=0, const [0,3]}
       hand_l_fk_ctrl.Location.X {type=MovieSceneScriptingFloatChannel, default=0, const [0,2]}
       hand_l_fk_ctrl.Location.Y {type=MovieSceneScriptingFloatChannel, default=0, const [0,2]}
       hand_l_fk_ctrl.Location.Z {type=MovieSceneScriptingFloatChannel, default=0, const [0,2]}
       hand_l_fk_ctrl.Rotation.X {type=MovieSceneScriptingFloatChannel, default=0, const [0,2]}
       hand_l_fk_ctrl.Rotation.Y {type=MovieSceneScriptingFloatChannel, default=0, const [0,2]}
       hand_l_fk_ctrl.Rotation.Z {type=MovieSceneScriptingFloatChannel, default=0, const [0,2]}
       hand_l_fk_ctrl.Scale.X {type=MovieSceneScriptingFloatChannel, default=0, const [0,2]}
       hand_l_fk_ctrl.Scale.Y {type=MovieSceneScriptingFloatChannel, default=0, const [0,2]}
       hand_l_fk_ctrl.Scale.Z {type=MovieSceneScriptingFloatChannel, default=0, const [0,2]}
       thumb_01_l_ctrl.Location.X {type=MovieSceneScriptingFloatChannel, default=0, const [0,2]}
       thumb_01_l_ctrl.Location.Y {type=MovieSceneScriptingFloatChannel, default=0, const [0,2]}
       thumb_01_l_ctrl.Location.Z {type=MovieSceneScriptingFloatChannel, default=0, const [0,2]}
       thumb_01_l_ctrl.Rotation.X {type=MovieSceneScriptingFloatChannel, default=0, const [0,2]}
       thumb_01_l_ctrl.Rotation.Y {type=MovieSceneScriptingFloatChannel, default=0, const [0,2]}
       thumb_01_l_ctrl.Rotation.Z {type=MovieSceneScriptingFloatChannel, default=0, const [0,2]}
       thumb_01_l_ctrl.Scale.X {type=MovieSceneScriptingFloatChannel, default=0, const [0,2]}
       thumb_01_l_ctrl.Scale.Y {type=MovieSceneScriptingFloatChannel, default=0, const [0,2]}
       thumb_01_l_ctrl.Scale.Z {type=MovieSceneScriptingFloatChannel, default=0, const [0,2]}
       thumb_02_l_ctrl.Location.X {type=MovieSceneScriptingFloatChannel, default=0, const [0,2]}
       thumb_02_l_ctrl.Location.Y {type=MovieSceneScriptingFloatChannel, default=0, const [0,2]}
       thumb_02_l_ctrl.Location.Z {type=MovieSceneScriptingFloatChannel, default=0, const [0,2]}
       thumb_02_l_ctrl.Rotation.X {type=MovieSceneScriptingFloatChannel, default=0, const [0,2]}
       thumb_02_l_ctrl.Rotation.Y {type=MovieSceneScriptingFloatChannel, default=0, const [0,2]}
       thumb_02_l_ctrl.Rotation.Z {type=MovieSceneScriptingFloatChannel, default=0, const [0,2]}
       thumb_02_l_ctrl.Scale.X {type=MovieSceneScriptingFloatChannel, default=0, const [0,2]}
       thumb_02_l_ctrl.Scale.Y {type=MovieSceneScriptingFloatChannel, default=0, const [0,2]}
       thumb_02_l_ctrl.Scale.Z {type=MovieSceneScriptingFloatChannel, default=0, const [0,2]}
       thumb_03_l_ctrl.Location.X {type=MovieSceneScriptingFloatChannel, default=0, const [0,2]}
       thumb_03_l_ctrl.Location.Y {type=MovieSceneScriptingFloatChannel, default=0, const [0,2]}
       thumb_03_l_ctrl.Location.Z {type=MovieSceneScriptingFloatChannel, default=0, const [0,2]}
       thumb_03_l_ctrl.Rotation.X {type=MovieSceneScriptingFloatChannel, default=0, const [0,2]}
       thumb_03_l_ctrl.Rotation.Y {type=MovieSceneScriptingFloatChannel, default=0, const [0,2]}
       thumb_03_l_ctrl.Rotation.Z {type=MovieSceneScriptingFloatChannel, default=0, const [0,2]}
       thumb_03_l_ctrl.Scale.X {type=MovieSceneScriptingFloatChannel, default=0, const [0,2]}
       thumb_03_l_ctrl.Scale.Y {type=MovieSceneScriptingFloatChannel, default=0, const [0,2]}
       thumb_03_l_ctrl.Scale.Z {type=MovieSceneScriptingFloatChannel, default=0, const [0,2]}
       index_metacarpal_l_ctrl.Location.X {type=MovieSceneScriptingFloatChannel, default=0, const [0,2]}
       index_metacarpal_l_ctrl.Location.Y {type=MovieSceneScriptingFloatChannel, default=0, const [0,2]}
       index_metacarpal_l_ctrl.Location.Z {type=MovieSceneScriptingFloatChannel, default=0, const [0,2]}
       index_metacarpal_l_ctrl.Rotation.X {type=MovieSceneScriptingFloatChannel, default=0, const [0,2]}
       index_metacarpal_l_ctrl.Rotation.Y {type=MovieSceneScriptingFloatChannel, default=0, const [0,2]}
       index_metacarpal_l_ctrl.Rotation.Z {type=MovieSceneScriptingFloatChannel, default=0, const [0,2]}
       index_metacarpal_l_ctrl.Scale.X {type=MovieSceneScriptingFloatChannel, default=0, const [0,2]}
       index_metacarpal_l_ctrl.Scale.Y {type=MovieSceneScriptingFloatChannel, default=0, const [0,2]}
       index_metacarpal_l_ctrl.Scale.Z {type=MovieSceneScriptingFloatChannel, default=0, const [0,2]}
       index_01_l_ctrl.Location.X {type=MovieSceneScriptingFloatChannel, default=0, const [0,2]}
       index_01_l_ctrl.Location.Y {type=MovieSceneScriptingFloatChannel, default=0, const [0,2]}
       index_01_l_ctrl.Location.Z {type=MovieSceneScriptingFloatChannel, default=0, const [0,2]}
       index_01_l_ctrl.Rotation.X {type=MovieSceneScriptingFloatChannel, default=0, const [0,2]}
       index_01_l_ctrl.Rotation.Y {type=MovieSceneScriptingFloatChannel, default=0, const [0,2]}
       index_01_l_ctrl.Rotation.Z {type=MovieSceneScriptingFloatChannel, default=0, const [0,2]}
       index_01_l_ctrl.Scale.X {type=MovieSceneScriptingFloatChannel, default=0, const [0,2]}
       index_01_l_ctrl.Scale.Y {type=MovieSceneScriptingFloatChannel, default=0, const [0,2]}
       index_01_l_ctrl.Scale.Z {type=MovieSceneScriptingFloatChannel, default=0, const [0,2]}
       index_02_l_ctrl.Location.X {type=MovieSceneScriptingFloatChannel, default=0, const [0,2]}
       index_02_l_ctrl.Location.Y {type=MovieSceneScriptingFloatChannel, default=0, const [0,2]}
       index_02_l_ctrl.Location.Z {type=MovieSceneScriptingFloatChannel, default=0, const [0,2]}
       index_02_l_ctrl.Rotation.X {type=MovieSceneScriptingFloatChannel, default=0, const [0,2]}
       index_02_l_ctrl.Rotation.Y {type=MovieSceneScriptingFloatChannel, default=0, const [0,2]}
       index_02_l_ctrl.Rotation.Z {type=MovieSceneScriptingFloatChannel, default=0, const [0,2]}
       index_02_l_ctrl.Scale.X {type=MovieSceneScriptingFloatChannel, default=0, const [0,2]}
       index_02_l_ctrl.Scale.Y {type=MovieSceneScriptingFloatChannel, default=0, const [0,2]}
       index_02_l_ctrl.Scale.Z {type=MovieSceneScriptingFloatChannel, default=0, const [0,2]}
       index_03_l_ctrl.Location.X {type=MovieSceneScriptingFloatChannel, default=0, const [0,2]}
       index_03_l_ctrl.Location.Y {type=MovieSceneScriptingFloatChannel, default=0, const [0,2]}
       index_03_l_ctrl.Location.Z {type=MovieSceneScriptingFloatChannel, default=0, const [0,2]}
       index_03_l_ctrl.Rotation.X {type=MovieSceneScriptingFloatChannel, default=0, const [0,2]}
       index_03_l_ctrl.Rotation.Y {type=MovieSceneScriptingFloatChannel, default=0, const [0,2]}
       index_03_l_ctrl.Rotation.Z {type=MovieSceneScriptingFloatChannel, default=0, const [0,2]}
       index_03_l_ctrl.Scale.X {type=MovieSceneScriptingFloatChannel, default=0, const [0,2]}
       index_03_l_ctrl.Scale.Y {type=MovieSceneScriptingFloatChannel, default=0, const [0,2]}
       index_03_l_ctrl.Scale.Z {type=MovieSceneScriptingFloatChannel, default=0, const [0,2]}
       middle_metacarpal_l_ctrl.Location.X {type=MovieSceneScriptingFloatChannel, default=0, const [0,2]}
       middle_metacarpal_l_ctrl.Location.Y {type=MovieSceneScriptingFloatChannel, default=0, const [0,2]}
       middle_metacarpal_l_ctrl.Location.Z {type=MovieSceneScriptingFloatChannel, default=0, const [0,2]}
       middle_metacarpal_l_ctrl.Rotation.X {type=MovieSceneScriptingFloatChannel, default=0, const [0,2]}
       middle_metacarpal_l_ctrl.Rotation.Y {type=MovieSceneScriptingFloatChannel, default=0, const [0,2]}
       middle_metacarpal_l_ctrl.Rotation.Z {type=MovieSceneScriptingFloatChannel, default=0, const [0,2]}
       middle_metacarpal_l_ctrl.Scale.X {type=MovieSceneScriptingFloatChannel, default=0, const [0,2]}
       middle_metacarpal_l_ctrl.Scale.Y {type=MovieSceneScriptingFloatChannel, default=0, const [0,2]}
       middle_metacarpal_l_ctrl.Scale.Z {type=MovieSceneScriptingFloatChannel, default=0, const [0,2]}
       middle_01_l_ctrl.Location.X {type=MovieSceneScriptingFloatChannel, default=0, const [0,2]}
       middle_01_l_ctrl.Location.Y {type=MovieSceneScriptingFloatChannel, default=0, const [0,2]}
       middle_01_l_ctrl.Location.Z {type=MovieSceneScriptingFloatChannel, default=0, const [0,2]}
       middle_01_l_ctrl.Rotation.X {type=MovieSceneScriptingFloatChannel, default=0, const [0,2]}
       middle_01_l_ctrl.Rotation.Y {type=MovieSceneScriptingFloatChannel, default=0, const [0,2]}
       middle_01_l_ctrl.Rotation.Z {type=MovieSceneScriptingFloatChannel, default=0, const [0,2]}
       middle_01_l_ctrl.Scale.X {type=MovieSceneScriptingFloatChannel, default=0, const [0,2]}
       middle_01_l_ctrl.Scale.Y {type=MovieSceneScriptingFloatChannel, default=0, const [0,2]}
       middle_01_l_ctrl.Scale.Z {type=MovieSceneScriptingFloatChannel, default=0, const [0,2]}
       middle_02_l_ctrl.Location.X {type=MovieSceneScriptingFloatChannel, default=0, const [0,2]}
       middle_02_l_ctrl.Location.Y {type=MovieSceneScriptingFloatChannel, default=0, const [0,2]}
       middle_02_l_ctrl.Location.Z {type=MovieSceneScriptingFloatChannel, default=0, const [0,2]}
       middle_02_l_ctrl.Rotation.X {type=MovieSceneScriptingFloatChannel, default=0, const [0,2]}
       middle_02_l_ctrl.Rotation.Y {type=MovieSceneScriptingFloatChannel, default=0, const [0,2]}
       middle_02_l_ctrl.Rotation.Z {type=MovieSceneScriptingFloatChannel, default=0, const [0,2]}
       middle_02_l_ctrl.Scale.X {type=MovieSceneScriptingFloatChannel, default=0, const [0,2]}
       middle_02_l_ctrl.Scale.Y {type=MovieSceneScriptingFloatChannel, default=0, const [0,2]}
       middle_02_l_ctrl.Scale.Z {type=MovieSceneScriptingFloatChannel, default=0, const [0,2]}
       middle_03_l_ctrl.Location.X {type=MovieSceneScriptingFloatChannel, default=0, const [0,2]}
       middle_03_l_ctrl.Location.Y {type=MovieSceneScriptingFloatChannel, default=0, const [0,2]}
       middle_03_l_ctrl.Location.Z {type=MovieSceneScriptingFloatChannel, default=0, const [0,2]}
       middle_03_l_ctrl.Rotation.X {type=MovieSceneScriptingFloatChannel, default=0, const [0,2]}
       middle_03_l_ctrl.Rotation.Y {type=MovieSceneScriptingFloatChannel, default=0, const [0,2]}
       middle_03_l_ctrl.Rotation.Z {type=MovieSceneScriptingFloatChannel, default=0, const [0,2]}
       middle_03_l_ctrl.Scale.X {type=MovieSceneScriptingFloatChannel, default=0, const [0,2]}
       middle_03_l_ctrl.Scale.Y {type=MovieSceneScriptingFloatChannel, default=0, const [0,2]}
       middle_03_l_ctrl.Scale.Z {type=MovieSceneScriptingFloatChannel, default=0, const [0,2]}
       ring_metacarpal_l_ctrl.Location.X {type=MovieSceneScriptingFloatChannel, default=0, const [0,2]}
       ring_metacarpal_l_ctrl.Location.Y {type=MovieSceneScriptingFloatChannel, default=0, const [0,2]}
       ring_metacarpal_l_ctrl.Location.Z {type=MovieSceneScriptingFloatChannel, default=0, const [0,2]}
       ring_metacarpal_l_ctrl.Rotation.X {type=MovieSceneScriptingFloatChannel, default=0, const [0,2]}
       ring_metacarpal_l_ctrl.Rotation.Y {type=MovieSceneScriptingFloatChannel, default=0, const [0,2]}
       ring_metacarpal_l_ctrl.Rotation.Z {type=MovieSceneScriptingFloatChannel, default=0, const [0,2]}
       ring_metacarpal_l_ctrl.Scale.X {type=MovieSceneScriptingFloatChannel, default=0, const [0,2]}
       ring_metacarpal_l_ctrl.Scale.Y {type=MovieSceneScriptingFloatChannel, default=0, const [0,2]}
       ring_metacarpal_l_ctrl.Scale.Z {type=MovieSceneScriptingFloatChannel, default=0, const [0,2]}
       ring_01_l_ctrl.Location.X {type=MovieSceneScriptingFloatChannel, default=0, const [0,2]}
       ring_01_l_ctrl.Location.Y {type=MovieSceneScriptingFloatChannel, default=0, const [0,2]}
       ring_01_l_ctrl.Location.Z {type=MovieSceneScriptingFloatChannel, default=0, const [0,2]}
       ring_01_l_ctrl.Rotation.X {type=MovieSceneScriptingFloatChannel, default=0, const [0,2]}
       ring_01_l_ctrl.Rotation.Y {type=MovieSceneScriptingFloatChannel, default=0, const [0,2]}
       ring_01_l_ctrl.Rotation.Z {type=MovieSceneScriptingFloatChannel, default=0, const [0,2]}
       ring_01_l_ctrl.Scale.X {type=MovieSceneScriptingFloatChannel, default=0, const [0,2]}
       ring_01_l_ctrl.Scale.Y {type=MovieSceneScriptingFloatChannel, default=0, const [0,2]}
       ring_01_l_ctrl.Scale.Z {type=MovieSceneScriptingFloatChannel, default=0, const [0,2]}
       ring_02_l_ctrl.Location.X {type=MovieSceneScriptingFloatChannel, default=0, const [0,2]}
       ring_02_l_ctrl.Location.Y {type=MovieSceneScriptingFloatChannel, default=0, const [0,2]}
       ring_02_l_ctrl.Location.Z {type=MovieSceneScriptingFloatChannel, default=0, const [0,2]}
       ring_02_l_ctrl.Rotation.X {type=MovieSceneScriptingFloatChannel, default=0, const [0,2]}
       ring_02_l_ctrl.Rotation.Y {type=MovieSceneScriptingFloatChannel, default=0, const [0,2]}
       ring_02_l_ctrl.Rotation.Z {type=MovieSceneScriptingFloatChannel, default=0, const [0,2]}
       ring_02_l_ctrl.Scale.X {type=MovieSceneScriptingFloatChannel, default=0, const [0,2]}
       ring_02_l_ctrl.Scale.Y {type=MovieSceneScriptingFloatChannel, default=0, const [0,2]}
       ring_02_l_ctrl.Scale.Z {type=MovieSceneScriptingFloatChannel, default=0, const [0,2]}
       ring_03_l_ctrl.Location.X {type=MovieSceneScriptingFloatChannel, default=0, const [0,2]}
       ring_03_l_ctrl.Location.Y {type=MovieSceneScriptingFloatChannel, default=0, const [0,2]}
       ring_03_l_ctrl.Location.Z {type=MovieSceneScriptingFloatChannel, default=0, const [0,2]}
       ring_03_l_ctrl.Rotation.X {type=MovieSceneScriptingFloatChannel, default=0, const [0,2]}
       ring_03_l_ctrl.Rotation.Y {type=MovieSceneScriptingFloatChannel, default=0, const [0,2]}
       ring_03_l_ctrl.Rotation.Z {type=MovieSceneScriptingFloatChannel, default=0, const [0,2]}
       ring_03_l_ctrl.Scale.X {type=MovieSceneScriptingFloatChannel, default=0, const [0,2]}
       ring_03_l_ctrl.Scale.Y {type=MovieSceneScriptingFloatChannel, default=0, const [0,2]}
       ring_03_l_ctrl.Scale.Z {type=MovieSceneScriptingFloatChannel, default=0, const [0,2]}
       pinky_metacarpal_l_ctrl.Location.X {type=MovieSceneScriptingFloatChannel, default=0, const [0,2]}
       pinky_metacarpal_l_ctrl.Location.Y {type=MovieSceneScriptingFloatChannel, default=0, const [0,2]}
       pinky_metacarpal_l_ctrl.Location.Z {type=MovieSceneScriptingFloatChannel, default=0, const [0,2]}
       pinky_metacarpal_l_ctrl.Rotation.X {type=MovieSceneScriptingFloatChannel, default=0, const [0,2]}
       pinky_metacarpal_l_ctrl.Rotation.Y {type=MovieSceneScriptingFloatChannel, default=0, const [0,2]}
       pinky_metacarpal_l_ctrl.Rotation.Z {type=MovieSceneScriptingFloatChannel, default=0, const [0,2]}
       pinky_metacarpal_l_ctrl.Scale.X {type=MovieSceneScriptingFloatChannel, default=0, const [0,2]}
       pinky_metacarpal_l_ctrl.Scale.Y {type=MovieSceneScriptingFloatChannel, default=0, const [0,2]}
       pinky_metacarpal_l_ctrl.Scale.Z {type=MovieSceneScriptingFloatChannel, default=0, const [0,2]}
       pinky_01_l_ctrl.Location.X {type=MovieSceneScriptingFloatChannel, default=0, const [0,2]}
       pinky_01_l_ctrl.Location.Y {type=MovieSceneScriptingFloatChannel, default=0, const [0,2]}
       pinky_01_l_ctrl.Location.Z {type=MovieSceneScriptingFloatChannel, default=0, const [0,2]}
       pinky_01_l_ctrl.Rotation.X {type=MovieSceneScriptingFloatChannel, default=0, const [0,2]}
       pinky_01_l_ctrl.Rotation.Y {type=MovieSceneScriptingFloatChannel, default=0, const [0,2]}
       pinky_01_l_ctrl.Rotation.Z {type=MovieSceneScriptingFloatChannel, default=0, const [0,2]}
       pinky_01_l_ctrl.Scale.X {type=MovieSceneScriptingFloatChannel, default=0, const [0,2]}
       pinky_01_l_ctrl.Scale.Y {type=MovieSceneScriptingFloatChannel, default=0, const [0,2]}
       pinky_01_l_ctrl.Scale.Z {type=MovieSceneScriptingFloatChannel, default=0, const [0,2]}
       pinky_02_l_ctrl.Location.X {type=MovieSceneScriptingFloatChannel, default=0, const [0,2]}
       pinky_02_l_ctrl.Location.Y {type=MovieSceneScriptingFloatChannel, default=0, const [0,2]}
       pinky_02_l_ctrl.Location.Z {type=MovieSceneScriptingFloatChannel, default=0, const [0,2]}
       pinky_02_l_ctrl.Rotation.X {type=MovieSceneScriptingFloatChannel, default=0, const [0,2]}
       pinky_02_l_ctrl.Rotation.Y {type=MovieSceneScriptingFloatChannel, default=0, const [0,2]}
       pinky_02_l_ctrl.Rotation.Z {type=MovieSceneScriptingFloatChannel, default=0, const [0,2]}
       pinky_02_l_ctrl.Scale.X {type=MovieSceneScriptingFloatChannel, default=0, const [0,2]}
       pinky_02_l_ctrl.Scale.Y {type=MovieSceneScriptingFloatChannel, default=0, const [0,2]}
       pinky_02_l_ctrl.Scale.Z {type=MovieSceneScriptingFloatChannel, default=0, const [0,2]}
       pinky_03_l_ctrl.Location.X {type=MovieSceneScriptingFloatChannel, default=0, const [0,2]}
       pinky_03_l_ctrl.Location.Y {type=MovieSceneScriptingFloatChannel, default=0, const [0,2]}
       pinky_03_l_ctrl.Location.Z {type=MovieSceneScriptingFloatChannel, default=0, const [0,2]}
       pinky_03_l_ctrl.Rotation.X {type=MovieSceneScriptingFloatChannel, default=0, const [0,2]}
       pinky_03_l_ctrl.Rotation.Y {type=MovieSceneScriptingFloatChannel, default=0, const [0,2]}
       pinky_03_l_ctrl.Rotation.Z {type=MovieSceneScriptingFloatChannel, default=0, const [0,2]}
       pinky_03_l_ctrl.Scale.X {type=MovieSceneScriptingFloatChannel, default=0, const [0,2]}
       pinky_03_l_ctrl.Scale.Y {type=MovieSceneScriptingFloatChannel, default=0, const [0,2]}
       pinky_03_l_ctrl.Scale.Z {type=MovieSceneScriptingFloatChannel, default=0, const [0,2]}
       hand_l_ik_ctrl.Location.X {type=MovieSceneScriptingFloatChannel, default=0, const [-51.2055,2]}
       hand_l_ik_ctrl.Location.Y {type=MovieSceneScriptingFloatChannel, default=0, const [45.144,2]}
       hand_l_ik_ctrl.Location.Z {type=MovieSceneScriptingFloatChannel, default=0, const [-29.4292,2]}
       hand_l_ik_ctrl.Rotation.X {type=MovieSceneScriptingFloatChannel, default=0, const [152.6857,2]}
       hand_l_ik_ctrl.Rotation.Y {type=MovieSceneScriptingFloatChannel, default=0, const [-15.7639,2]}
       hand_l_ik_ctrl.Rotation.Z {type=MovieSceneScriptingFloatChannel, default=0, const [-131.0282,2]}
       hand_l_ik_ctrl.Scale.X {type=MovieSceneScriptingFloatChannel, default=0, const [0.0054,2]}
       hand_l_ik_ctrl.Scale.Y {type=MovieSceneScriptingFloatChannel, default=0, const [0.0054,2]}
       hand_l_ik_ctrl.Scale.Z {type=MovieSceneScriptingFloatChannel, default=0, const [0.0054,2]}
       Softness {type=MovieSceneScriptingFloatChannel, default=0, const [0,2]}
       arm_l_pv_ik_ctrl.Location.X {type=MovieSceneScriptingFloatChannel, default=0, const [-50.7251,2]}
       arm_l_pv_ik_ctrl.Location.Y {type=MovieSceneScriptingFloatChannel, default=0, const [-61.5684,2]}
       arm_l_pv_ik_ctrl.Location.Z {type=MovieSceneScriptingFloatChannel, default=0, const [33.2734,2]}
       arm_l_pv_ik_ctrl.Rotation.X {type=MovieSceneScriptingFloatChannel, default=0, const [12.6785,2]}
       arm_l_pv_ik_ctrl.Rotation.Y {type=MovieSceneScriptingFloatChannel, default=0, const [0.9096,2]}
       arm_l_pv_ik_ctrl.Rotation.Z {type=MovieSceneScriptingFloatChannel, default=0, const [10.6662,2]}
       arm_l_pv_ik_ctrl.Scale.X {type=MovieSceneScriptingFloatChannel, default=0, const [0,2]}
       arm_l_pv_ik_ctrl.Scale.Y {type=MovieSceneScriptingFloatChannel, default=0, const [0,2]}
       arm_l_pv_ik_ctrl.Scale.Z {type=MovieSceneScriptingFloatChannel, default=0, const [0,2]}
       upperarm_r_fk_ctrl.Location.X {type=MovieSceneScriptingFloatChannel, default=0, const [0,2]}
       upperarm_r_fk_ctrl.Location.Y {type=MovieSceneScriptingFloatChannel, default=0, const [0,2]}
       upperarm_r_fk_ctrl.Location.Z {type=MovieSceneScriptingFloatChannel, default=0, const [0,2]}
       upperarm_r_fk_ctrl.Rotation.X {type=MovieSceneScriptingFloatChannel, default=0, keys [[3004,0],[3012,1.5683],[3018,1.1288],[3024,1.4809],[3171,0]]}
       upperarm_r_fk_ctrl.Rotation.Y {type=MovieSceneScriptingFloatChannel, default=0, keys [[3004,0],[3012,-3.0361],[3018,-2.1038],[3024,-2.8523],[3171,0]]}
       upperarm_r_fk_ctrl.Rotation.Z {type=MovieSceneScriptingFloatChannel, default=0, keys [[3004,0],[3012,5.5213],[3018,3.8734],[3024,5.1956],[3171,0]]}
       upperarm_r_fk_ctrl.Scale.X {type=MovieSceneScriptingFloatChannel, default=0, const [0,2]}
       upperarm_r_fk_ctrl.Scale.Y {type=MovieSceneScriptingFloatChannel, default=0, const [0,2]}
       upperarm_r_fk_ctrl.Scale.Z {type=MovieSceneScriptingFloatChannel, default=0, const [0,2]}
       lowerarm_r_fk_ctrl.Location.X {type=MovieSceneScriptingFloatChannel, default=0, const [0,3]}
       lowerarm_r_fk_ctrl.Location.Y {type=MovieSceneScriptingFloatChannel, default=0, const [0,3]}
       lowerarm_r_fk_ctrl.Location.Z {type=MovieSceneScriptingFloatChannel, default=0, const [0,3]}
       lowerarm_r_fk_ctrl.Rotation.X {
        type=MovieSceneScriptingFloatChannel
        default=0
        keys [[3004,0],[3013,0],[3018,0.6477],[3024,0.2654],[3030,0.5524],[3035,0.4692],[3042,0.7428],[3050,0.4124],[3058,0.6459],[3171,0]]
       }
       lowerarm_r_fk_ctrl.Rotation.Y {
        type=MovieSceneScriptingFloatChannel
        default=0
        keys [[3004,0],[3013,0],[3018,-0.1557],[3024,-0.079],[3030,-0.1397],[3035,-0.1244],[3042,-0.1655],[3050,-0.1103],[3058,-0.1501],[3171,0]]
       }
       lowerarm_r_fk_ctrl.Rotation.Z {
        type=MovieSceneScriptingFloatChannel
        default=0
        keys [[3004,0],[3013,0],[3018,-8.5726],[3021,-8.7104],[3025,-9.467],[3028,-8.8371],[3032,-9.456],[3035,-8.6315],[3038,-8.9617],[3041,-8.4384],[3046,-9.4524],[3050,-8.8996],[3058,-10.626],[3171,0]]
       }
       lowerarm_r_fk_ctrl.Scale.X {type=MovieSceneScriptingFloatChannel, default=0, const [0,3]}
       lowerarm_r_fk_ctrl.Scale.Y {type=MovieSceneScriptingFloatChannel, default=0, const [0,3]}
       lowerarm_r_fk_ctrl.Scale.Z {type=MovieSceneScriptingFloatChannel, default=0, const [0,3]}
       hand_r_fk_ctrl.Location.X {type=MovieSceneScriptingFloatChannel, default=0, const [0,2]}
       hand_r_fk_ctrl.Location.Y {type=MovieSceneScriptingFloatChannel, default=0, const [0,2]}
       hand_r_fk_ctrl.Location.Z {type=MovieSceneScriptingFloatChannel, default=0, const [0,2]}
       hand_r_fk_ctrl.Rotation.X {type=MovieSceneScriptingFloatChannel, default=0, const [0,2]}
       hand_r_fk_ctrl.Rotation.Y {type=MovieSceneScriptingFloatChannel, default=0, const [0,2]}
       hand_r_fk_ctrl.Rotation.Z {type=MovieSceneScriptingFloatChannel, default=0, const [0,2]}
       hand_r_fk_ctrl.Scale.X {type=MovieSceneScriptingFloatChannel, default=0, const [0,2]}
       hand_r_fk_ctrl.Scale.Y {type=MovieSceneScriptingFloatChannel, default=0, const [0,2]}
       hand_r_fk_ctrl.Scale.Z {type=MovieSceneScriptingFloatChannel, default=0, const [0,2]}
       thumb_01_r_ctrl.Location.X {type=MovieSceneScriptingFloatChannel, default=0, const [0,2]}
       thumb_01_r_ctrl.Location.Y {type=MovieSceneScriptingFloatChannel, default=0, const [0,2]}
       thumb_01_r_ctrl.Location.Z {type=MovieSceneScriptingFloatChannel, default=0, const [0,2]}
       thumb_01_r_ctrl.Rotation.X {type=MovieSceneScriptingFloatChannel, default=0, const [0,2]}
       thumb_01_r_ctrl.Rotation.Y {type=MovieSceneScriptingFloatChannel, default=0, const [0,2]}
       thumb_01_r_ctrl.Rotation.Z {type=MovieSceneScriptingFloatChannel, default=0, const [0,2]}
       thumb_01_r_ctrl.Scale.X {type=MovieSceneScriptingFloatChannel, default=0, const [0,2]}
       thumb_01_r_ctrl.Scale.Y {type=MovieSceneScriptingFloatChannel, default=0, const [0,2]}
       thumb_01_r_ctrl.Scale.Z {type=MovieSceneScriptingFloatChannel, default=0, const [0,2]}
       thumb_02_r_ctrl.Location.X {type=MovieSceneScriptingFloatChannel, default=0, const [0,2]}
       thumb_02_r_ctrl.Location.Y {type=MovieSceneScriptingFloatChannel, default=0, const [0,2]}
       thumb_02_r_ctrl.Location.Z {type=MovieSceneScriptingFloatChannel, default=0, const [0,2]}
       thumb_02_r_ctrl.Rotation.X {type=MovieSceneScriptingFloatChannel, default=0, const [0,2]}
       thumb_02_r_ctrl.Rotation.Y {type=MovieSceneScriptingFloatChannel, default=0, const [0,2]}
       thumb_02_r_ctrl.Rotation.Z {type=MovieSceneScriptingFloatChannel, default=0, const [0,2]}
       thumb_02_r_ctrl.Scale.X {type=MovieSceneScriptingFloatChannel, default=0, const [0,2]}
       thumb_02_r_ctrl.Scale.Y {type=MovieSceneScriptingFloatChannel, default=0, const [0,2]}
       thumb_02_r_ctrl.Scale.Z {type=MovieSceneScriptingFloatChannel, default=0, const [0,2]}
       thumb_03_r_ctrl.Location.X {type=MovieSceneScriptingFloatChannel, default=0, const [0,2]}
       thumb_03_r_ctrl.Location.Y {type=MovieSceneScriptingFloatChannel, default=0, const [0,2]}
       thumb_03_r_ctrl.Location.Z {type=MovieSceneScriptingFloatChannel, default=0, const [0,2]}
       thumb_03_r_ctrl.Rotation.X {type=MovieSceneScriptingFloatChannel, default=0, const [0,2]}
       thumb_03_r_ctrl.Rotation.Y {type=MovieSceneScriptingFloatChannel, default=0, const [0,2]}
       thumb_03_r_ctrl.Rotation.Z {type=MovieSceneScriptingFloatChannel, default=0, const [0,2]}
       thumb_03_r_ctrl.Scale.X {type=MovieSceneScriptingFloatChannel, default=0, const [0,2]}
       thumb_03_r_ctrl.Scale.Y {type=MovieSceneScriptingFloatChannel, default=0, const [0,2]}
       thumb_03_r_ctrl.Scale.Z {type=MovieSceneScriptingFloatChannel, default=0, const [0,2]}
       index_metacarpal_r_ctrl.Location.X {type=MovieSceneScriptingFloatChannel, default=0, const [0,2]}
       index_metacarpal_r_ctrl.Location.Y {type=MovieSceneScriptingFloatChannel, default=0, const [0,2]}
       index_metacarpal_r_ctrl.Location.Z {type=MovieSceneScriptingFloatChannel, default=0, const [0,2]}
       index_metacarpal_r_ctrl.Rotation.X {type=MovieSceneScriptingFloatChannel, default=0, const [0,2]}
       index_metacarpal_r_ctrl.Rotation.Y {type=MovieSceneScriptingFloatChannel, default=0, const [0,2]}
       index_metacarpal_r_ctrl.Rotation.Z {type=MovieSceneScriptingFloatChannel, default=0, const [0,2]}
       index_metacarpal_r_ctrl.Scale.X {type=MovieSceneScriptingFloatChannel, default=0, const [0,2]}
       index_metacarpal_r_ctrl.Scale.Y {type=MovieSceneScriptingFloatChannel, default=0, const [0,2]}
       index_metacarpal_r_ctrl.Scale.Z {type=MovieSceneScriptingFloatChannel, default=0, const [0,2]}
       index_01_r_ctrl.Location.X {type=MovieSceneScriptingFloatChannel, default=0, const [0,2]}
       index_01_r_ctrl.Location.Y {type=MovieSceneScriptingFloatChannel, default=0, const [0,2]}
       index_01_r_ctrl.Location.Z {type=MovieSceneScriptingFloatChannel, default=0, const [0,2]}
       index_01_r_ctrl.Rotation.X {type=MovieSceneScriptingFloatChannel, default=0, const [0,2]}
       index_01_r_ctrl.Rotation.Y {type=MovieSceneScriptingFloatChannel, default=0, const [0,2]}
       index_01_r_ctrl.Rotation.Z {type=MovieSceneScriptingFloatChannel, default=0, const [0,2]}
       index_01_r_ctrl.Scale.X {type=MovieSceneScriptingFloatChannel, default=0, const [0,2]}
       index_01_r_ctrl.Scale.Y {type=MovieSceneScriptingFloatChannel, default=0, const [0,2]}
       index_01_r_ctrl.Scale.Z {type=MovieSceneScriptingFloatChannel, default=0, const [0,2]}
       index_02_r_ctrl.Location.X {type=MovieSceneScriptingFloatChannel, default=0, const [0,2]}
       index_02_r_ctrl.Location.Y {type=MovieSceneScriptingFloatChannel, default=0, const [0,2]}
       index_02_r_ctrl.Location.Z {type=MovieSceneScriptingFloatChannel, default=0, const [0,2]}
       index_02_r_ctrl.Rotation.X {type=MovieSceneScriptingFloatChannel, default=0, const [0,2]}
       index_02_r_ctrl.Rotation.Y {type=MovieSceneScriptingFloatChannel, default=0, const [0,2]}
       index_02_r_ctrl.Rotation.Z {type=MovieSceneScriptingFloatChannel, default=0, const [0,2]}
       index_02_r_ctrl.Scale.X {type=MovieSceneScriptingFloatChannel, default=0, const [0,2]}
       index_02_r_ctrl.Scale.Y {type=MovieSceneScriptingFloatChannel, default=0, const [0,2]}
       index_02_r_ctrl.Scale.Z {type=MovieSceneScriptingFloatChannel, default=0, const [0,2]}
       index_03_r_ctrl.Location.X {type=MovieSceneScriptingFloatChannel, default=0, const [0,2]}
       index_03_r_ctrl.Location.Y {type=MovieSceneScriptingFloatChannel, default=0, const [0,2]}
       index_03_r_ctrl.Location.Z {type=MovieSceneScriptingFloatChannel, default=0, const [0,2]}
       index_03_r_ctrl.Rotation.X {type=MovieSceneScriptingFloatChannel, default=0, const [0,2]}
       index_03_r_ctrl.Rotation.Y {type=MovieSceneScriptingFloatChannel, default=0, const [0,2]}
       index_03_r_ctrl.Rotation.Z {type=MovieSceneScriptingFloatChannel, default=0, const [0,2]}
       index_03_r_ctrl.Scale.X {type=MovieSceneScriptingFloatChannel, default=0, const [0,2]}
       index_03_r_ctrl.Scale.Y {type=MovieSceneScriptingFloatChannel, default=0, const [0,2]}
       index_03_r_ctrl.Scale.Z {type=MovieSceneScriptingFloatChannel, default=0, const [0,2]}
       middle_metacarpal_r_ctrl.Location.X {type=MovieSceneScriptingFloatChannel, default=0, const [0,2]}
       middle_metacarpal_r_ctrl.Location.Y {type=MovieSceneScriptingFloatChannel, default=0, const [0,2]}
       middle_metacarpal_r_ctrl.Location.Z {type=MovieSceneScriptingFloatChannel, default=0, const [0,2]}
       middle_metacarpal_r_ctrl.Rotation.X {type=MovieSceneScriptingFloatChannel, default=0, const [0,2]}
       middle_metacarpal_r_ctrl.Rotation.Y {type=MovieSceneScriptingFloatChannel, default=0, const [0,2]}
       middle_metacarpal_r_ctrl.Rotation.Z {type=MovieSceneScriptingFloatChannel, default=0, const [0,2]}
       middle_metacarpal_r_ctrl.Scale.X {type=MovieSceneScriptingFloatChannel, default=0, const [0,2]}
       middle_metacarpal_r_ctrl.Scale.Y {type=MovieSceneScriptingFloatChannel, default=0, const [0,2]}
       middle_metacarpal_r_ctrl.Scale.Z {type=MovieSceneScriptingFloatChannel, default=0, const [0,2]}
       middle_01_r_ctrl.Location.X {type=MovieSceneScriptingFloatChannel, default=0, const [0,2]}
       middle_01_r_ctrl.Location.Y {type=MovieSceneScriptingFloatChannel, default=0, const [0,2]}
       middle_01_r_ctrl.Location.Z {type=MovieSceneScriptingFloatChannel, default=0, const [0,2]}
       middle_01_r_ctrl.Rotation.X {type=MovieSceneScriptingFloatChannel, default=0, const [0,2]}
       middle_01_r_ctrl.Rotation.Y {type=MovieSceneScriptingFloatChannel, default=0, const [0,2]}
       middle_01_r_ctrl.Rotation.Z {type=MovieSceneScriptingFloatChannel, default=0, const [0,2]}
       middle_01_r_ctrl.Scale.X {type=MovieSceneScriptingFloatChannel, default=0, const [0,2]}
       middle_01_r_ctrl.Scale.Y {type=MovieSceneScriptingFloatChannel, default=0, const [0,2]}
       middle_01_r_ctrl.Scale.Z {type=MovieSceneScriptingFloatChannel, default=0, const [0,2]}
       middle_02_r_ctrl.Location.X {type=MovieSceneScriptingFloatChannel, default=0, const [0,2]}
       middle_02_r_ctrl.Location.Y {type=MovieSceneScriptingFloatChannel, default=0, const [0,2]}
       middle_02_r_ctrl.Location.Z {type=MovieSceneScriptingFloatChannel, default=0, const [0,2]}
       middle_02_r_ctrl.Rotation.X {type=MovieSceneScriptingFloatChannel, default=0, const [0,2]}
       middle_02_r_ctrl.Rotation.Y {type=MovieSceneScriptingFloatChannel, default=0, const [0,2]}
       middle_02_r_ctrl.Rotation.Z {type=MovieSceneScriptingFloatChannel, default=0, const [0,2]}
       middle_02_r_ctrl.Scale.X {type=MovieSceneScriptingFloatChannel, default=0, const [0,2]}
       middle_02_r_ctrl.Scale.Y {type=MovieSceneScriptingFloatChannel, default=0, const [0,2]}
       middle_02_r_ctrl.Scale.Z {type=MovieSceneScriptingFloatChannel, default=0, const [0,2]}
       middle_03_r_ctrl.Location.X {type=MovieSceneScriptingFloatChannel, default=0, const [0,2]}
       middle_03_r_ctrl.Location.Y {type=MovieSceneScriptingFloatChannel, default=0, const [0,2]}
       middle_03_r_ctrl.Location.Z {type=MovieSceneScriptingFloatChannel, default=0, const [0,2]}
       middle_03_r_ctrl.Rotation.X {type=MovieSceneScriptingFloatChannel, default=0, const [0,2]}
       middle_03_r_ctrl.Rotation.Y {type=MovieSceneScriptingFloatChannel, default=0, const [0,2]}
       middle_03_r_ctrl.Rotation.Z {type=MovieSceneScriptingFloatChannel, default=0, const [0,2]}
       middle_03_r_ctrl.Scale.X {type=MovieSceneScriptingFloatChannel, default=0, const [0,2]}
       middle_03_r_ctrl.Scale.Y {type=MovieSceneScriptingFloatChannel, default=0, const [0,2]}
       middle_03_r_ctrl.Scale.Z {type=MovieSceneScriptingFloatChannel, default=0, const [0,2]}
       ring_metacarpal_r_ctrl.Location.X {type=MovieSceneScriptingFloatChannel, default=0, const [0,2]}
       ring_metacarpal_r_ctrl.Location.Y {type=MovieSceneScriptingFloatChannel, default=0, const [0,2]}
       ring_metacarpal_r_ctrl.Location.Z {type=MovieSceneScriptingFloatChannel, default=0, const [0,2]}
       ring_metacarpal_r_ctrl.Rotation.X {type=MovieSceneScriptingFloatChannel, default=0, const [0,2]}
       ring_metacarpal_r_ctrl.Rotation.Y {type=MovieSceneScriptingFloatChannel, default=0, const [0,2]}
       ring_metacarpal_r_ctrl.Rotation.Z {type=MovieSceneScriptingFloatChannel, default=0, const [0,2]}
       ring_metacarpal_r_ctrl.Scale.X {type=MovieSceneScriptingFloatChannel, default=0, const [0,2]}
       ring_metacarpal_r_ctrl.Scale.Y {type=MovieSceneScriptingFloatChannel, default=0, const [0,2]}
       ring_metacarpal_r_ctrl.Scale.Z {type=MovieSceneScriptingFloatChannel, default=0, const [0,2]}
       ring_01_r_ctrl.Location.X {type=MovieSceneScriptingFloatChannel, default=0, const [0,2]}
       ring_01_r_ctrl.Location.Y {type=MovieSceneScriptingFloatChannel, default=0, const [0,2]}
       ring_01_r_ctrl.Location.Z {type=MovieSceneScriptingFloatChannel, default=0, const [0,2]}
       ring_01_r_ctrl.Rotation.X {type=MovieSceneScriptingFloatChannel, default=0, const [0,2]}
       ring_01_r_ctrl.Rotation.Y {type=MovieSceneScriptingFloatChannel, default=0, const [0,2]}
       ring_01_r_ctrl.Rotation.Z {type=MovieSceneScriptingFloatChannel, default=0, const [0,2]}
       ring_01_r_ctrl.Scale.X {type=MovieSceneScriptingFloatChannel, default=0, const [0,2]}
       ring_01_r_ctrl.Scale.Y {type=MovieSceneScriptingFloatChannel, default=0, const [0,2]}
       ring_01_r_ctrl.Scale.Z {type=MovieSceneScriptingFloatChannel, default=0, const [0,2]}
       ring_02_r_ctrl.Location.X {type=MovieSceneScriptingFloatChannel, default=0, const [0,2]}
       ring_02_r_ctrl.Location.Y {type=MovieSceneScriptingFloatChannel, default=0, const [0,2]}
       ring_02_r_ctrl.Location.Z {type=MovieSceneScriptingFloatChannel, default=0, const [0,2]}
       ring_02_r_ctrl.Rotation.X {type=MovieSceneScriptingFloatChannel, default=0, const [0,2]}
       ring_02_r_ctrl.Rotation.Y {type=MovieSceneScriptingFloatChannel, default=0, const [0,2]}
       ring_02_r_ctrl.Rotation.Z {type=MovieSceneScriptingFloatChannel, default=0, const [0,2]}
       ring_02_r_ctrl.Scale.X {type=MovieSceneScriptingFloatChannel, default=0, const [0,2]}
       ring_02_r_ctrl.Scale.Y {type=MovieSceneScriptingFloatChannel, default=0, const [0,2]}
       ring_02_r_ctrl.Scale.Z {type=MovieSceneScriptingFloatChannel, default=0, const [0,2]}
       ring_03_r_ctrl.Location.X {type=MovieSceneScriptingFloatChannel, default=0, const [0,2]}
       ring_03_r_ctrl.Location.Y {type=MovieSceneScriptingFloatChannel, default=0, const [0,2]}
       ring_03_r_ctrl.Location.Z {type=MovieSceneScriptingFloatChannel, default=0, const [0,2]}
       ring_03_r_ctrl.Rotation.X {type=MovieSceneScriptingFloatChannel, default=0, const [0,2]}
       ring_03_r_ctrl.Rotation.Y {type=MovieSceneScriptingFloatChannel, default=0, const [0,2]}
       ring_03_r_ctrl.Rotation.Z {type=MovieSceneScriptingFloatChannel, default=0, const [0,2]}
       ring_03_r_ctrl.Scale.X {type=MovieSceneScriptingFloatChannel, default=0, const [0,2]}
       ring_03_r_ctrl.Scale.Y {type=MovieSceneScriptingFloatChannel, default=0, const [0,2]}
       ring_03_r_ctrl.Scale.Z {type=MovieSceneScriptingFloatChannel, default=0, const [0,2]}
       pinky_metacarpal_r_ctrl.Location.X {type=MovieSceneScriptingFloatChannel, default=0, const [0,2]}
       pinky_metacarpal_r_ctrl.Location.Y {type=MovieSceneScriptingFloatChannel, default=0, const [0,2]}
       pinky_metacarpal_r_ctrl.Location.Z {type=MovieSceneScriptingFloatChannel, default=0, const [0,2]}
       pinky_metacarpal_r_ctrl.Rotation.X {type=MovieSceneScriptingFloatChannel, default=0, const [0,2]}
       pinky_metacarpal_r_ctrl.Rotation.Y {type=MovieSceneScriptingFloatChannel, default=0, const [0,2]}
       pinky_metacarpal_r_ctrl.Rotation.Z {type=MovieSceneScriptingFloatChannel, default=0, const [0,2]}
       pinky_metacarpal_r_ctrl.Scale.X {type=MovieSceneScriptingFloatChannel, default=0, const [0,2]}
       pinky_metacarpal_r_ctrl.Scale.Y {type=MovieSceneScriptingFloatChannel, default=0, const [0,2]}
       pinky_metacarpal_r_ctrl.Scale.Z {type=MovieSceneScriptingFloatChannel, default=0, const [0,2]}
       pinky_01_r_ctrl.Location.X {type=MovieSceneScriptingFloatChannel, default=0, const [0,2]}
       pinky_01_r_ctrl.Location.Y {type=MovieSceneScriptingFloatChannel, default=0, const [0,2]}
       pinky_01_r_ctrl.Location.Z {type=MovieSceneScriptingFloatChannel, default=0, const [0,2]}
       pinky_01_r_ctrl.Rotation.X {type=MovieSceneScriptingFloatChannel, default=0, const [0,2]}
       pinky_01_r_ctrl.Rotation.Y {type=MovieSceneScriptingFloatChannel, default=0, const [0,2]}
       pinky_01_r_ctrl.Rotation.Z {type=MovieSceneScriptingFloatChannel, default=0, const [0,2]}
       pinky_01_r_ctrl.Scale.X {type=MovieSceneScriptingFloatChannel, default=0, const [0,2]}
       pinky_01_r_ctrl.Scale.Y {type=MovieSceneScriptingFloatChannel, default=0, const [0,2]}
       pinky_01_r_ctrl.Scale.Z {type=MovieSceneScriptingFloatChannel, default=0, const [0,2]}
       pinky_02_r_ctrl.Location.X {type=MovieSceneScriptingFloatChannel, default=0, const [0,2]}
       pinky_02_r_ctrl.Location.Y {type=MovieSceneScriptingFloatChannel, default=0, const [0,2]}
       pinky_02_r_ctrl.Location.Z {type=MovieSceneScriptingFloatChannel, default=0, const [0,2]}
       pinky_02_r_ctrl.Rotation.X {type=MovieSceneScriptingFloatChannel, default=0, const [0,2]}
       pinky_02_r_ctrl.Rotation.Y {type=MovieSceneScriptingFloatChannel, default=0, const [0,2]}
       pinky_02_r_ctrl.Rotation.Z {type=MovieSceneScriptingFloatChannel, default=0, const [0,2]}
       pinky_02_r_ctrl.Scale.X {type=MovieSceneScriptingFloatChannel, default=0, const [0,2]}
       pinky_02_r_ctrl.Scale.Y {type=MovieSceneScriptingFloatChannel, default=0, const [0,2]}
       pinky_02_r_ctrl.Scale.Z {type=MovieSceneScriptingFloatChannel, default=0, const [0,2]}
       pinky_03_r_ctrl.Location.X {type=MovieSceneScriptingFloatChannel, default=0, const [0,2]}
       pinky_03_r_ctrl.Location.Y {type=MovieSceneScriptingFloatChannel, default=0, const [0,2]}
       pinky_03_r_ctrl.Location.Z {type=MovieSceneScriptingFloatChannel, default=0, const [0,2]}
       pinky_03_r_ctrl.Rotation.X {type=MovieSceneScriptingFloatChannel, default=0, const [0,2]}
       pinky_03_r_ctrl.Rotation.Y {type=MovieSceneScriptingFloatChannel, default=0, const [0,2]}
       pinky_03_r_ctrl.Rotation.Z {type=MovieSceneScriptingFloatChannel, default=0, const [0,2]}
       pinky_03_r_ctrl.Scale.X {type=MovieSceneScriptingFloatChannel, default=0, const [0,2]}
       pinky_03_r_ctrl.Scale.Y {type=MovieSceneScriptingFloatChannel, default=0, const [0,2]}
       pinky_03_r_ctrl.Scale.Z {type=MovieSceneScriptingFloatChannel, default=0, const [0,2]}
       hand_r_ik_ctrl.Location.X {type=MovieSceneScriptingFloatChannel, default=0, const [-48.597,2]}
       hand_r_ik_ctrl.Location.Y {type=MovieSceneScriptingFloatChannel, default=0, const [-82.6449,2]}
       hand_r_ik_ctrl.Location.Z {type=MovieSceneScriptingFloatChannel, default=0, const [-39.1777,2]}
       hand_r_ik_ctrl.Rotation.X {type=MovieSceneScriptingFloatChannel, default=0, const [69.3602,2]}
       hand_r_ik_ctrl.Rotation.Y {type=MovieSceneScriptingFloatChannel, default=0, const [10.8865,2]}
       hand_r_ik_ctrl.Rotation.Z {type=MovieSceneScriptingFloatChannel, default=0, const [66.8644,2]}
       hand_r_ik_ctrl.Scale.X {type=MovieSceneScriptingFloatChannel, default=0, const [0.0054,2]}
       hand_r_ik_ctrl.Scale.Y {type=MovieSceneScriptingFloatChannel, default=0, const [0.0054,2]}
       hand_r_ik_ctrl.Scale.Z {type=MovieSceneScriptingFloatChannel, default=0, const [0.0054,2]}
       Softness_2 {type=MovieSceneScriptingFloatChannel, default=0, const [0,2]}
       arm_r_pv_ik_ctrl.Location.X {type=MovieSceneScriptingFloatChannel, default=0, const [57.0185,2]}
       arm_r_pv_ik_ctrl.Location.Y {type=MovieSceneScriptingFloatChannel, default=0, const [-75.2212,2]}
       arm_r_pv_ik_ctrl.Location.Z {type=MovieSceneScriptingFloatChannel, default=0, const [43.3861,2]}
       arm_r_pv_ik_ctrl.Rotation.X {type=MovieSceneScriptingFloatChannel, default=0, const [-111.0304,2]}
       arm_r_pv_ik_ctrl.Rotation.Y {type=MovieSceneScriptingFloatChannel, default=0, const [-11.3733,2]}
       arm_r_pv_ik_ctrl.Rotation.Z {type=MovieSceneScriptingFloatChannel, default=0, const [-132.7263,2]}
       arm_r_pv_ik_ctrl.Scale.X {type=MovieSceneScriptingFloatChannel, default=0, const [0,2]}
       arm_r_pv_ik_ctrl.Scale.Y {type=MovieSceneScriptingFloatChannel, default=0, const [0,2]}
       arm_r_pv_ik_ctrl.Scale.Z {type=MovieSceneScriptingFloatChannel, default=0, const [0,2]}
       body_ctrl.Location.X {type=MovieSceneScriptingFloatChannel, default=0, const [0,2]}
       body_ctrl.Location.Y {type=MovieSceneScriptingFloatChannel, default=0, const [0,2]}
       body_ctrl.Location.Z {type=MovieSceneScriptingFloatChannel, default=0, const [0,2]}
       body_ctrl.Rotation.X {type=MovieSceneScriptingFloatChannel, default=0, const [0,2]}
       body_ctrl.Rotation.Y {type=MovieSceneScriptingFloatChannel, default=0, const [0,2]}
       body_ctrl.Rotation.Z {type=MovieSceneScriptingFloatChannel, default=0, const [0,2]}
       body_ctrl.Scale.X {type=MovieSceneScriptingFloatChannel, default=0, const [0,2]}
       body_ctrl.Scale.Y {type=MovieSceneScriptingFloatChannel, default=0, const [0,2]}
       body_ctrl.Scale.Z {type=MovieSceneScriptingFloatChannel, default=0, const [0,2]}
       hips_ctrl.Location.X {type=MovieSceneScriptingFloatChannel, default=0, const [0,2]}
       hips_ctrl.Location.Y {type=MovieSceneScriptingFloatChannel, default=0, const [0,2]}
       hips_ctrl.Location.Z {type=MovieSceneScriptingFloatChannel, default=0, const [0,2]}
       hips_ctrl.Rotation.X {type=MovieSceneScriptingFloatChannel, default=0, const [0,2]}
       hips_ctrl.Rotation.Y {type=MovieSceneScriptingFloatChannel, default=0, const [0,2]}
       hips_ctrl.Rotation.Z {type=MovieSceneScriptingFloatChannel, default=0, const [0,2]}
       hips_ctrl.Scale.X {type=MovieSceneScriptingFloatChannel, default=0, const [0,2]}
       hips_ctrl.Scale.Y {type=MovieSceneScriptingFloatChannel, default=0, const [0,2]}
       hips_ctrl.Scale.Z {type=MovieSceneScriptingFloatChannel, default=0, const [0,2]}
       hips_tan_ctrl.Location.X {type=MovieSceneScriptingFloatChannel, default=0, const [-0.2255,2]}
       hips_tan_ctrl.Location.Y {type=MovieSceneScriptingFloatChannel, default=0, const [1.221,2]}
       hips_tan_ctrl.Location.Z {type=MovieSceneScriptingFloatChannel, default=0, const [0.363,2]}
       hips_tan_ctrl.Rotation.X {type=MovieSceneScriptingFloatChannel, default=0, const [0,2]}
       hips_tan_ctrl.Rotation.Y {type=MovieSceneScriptingFloatChannel, default=0, const [0,2]}
       hips_tan_ctrl.Rotation.Z {type=MovieSceneScriptingFloatChannel, default=0, const [0,2]}
       hips_tan_ctrl.Scale.X {type=MovieSceneScriptingFloatChannel, default=0, const [0,2]}
       hips_tan_ctrl.Scale.Y {type=MovieSceneScriptingFloatChannel, default=0, const [0,2]}
       hips_tan_ctrl.Scale.Z {type=MovieSceneScriptingFloatChannel, default=0, const [0,2]}
       spine_01_ctrl.Location.X {type=MovieSceneScriptingFloatChannel, default=0, const [0,2]}
       spine_01_ctrl.Location.Y {type=MovieSceneScriptingFloatChannel, default=0, const [0,2]}
       spine_01_ctrl.Location.Z {type=MovieSceneScriptingFloatChannel, default=0, const [0,2]}
       spine_01_ctrl.Rotation.X {type=MovieSceneScriptingFloatChannel, default=0, const [0,2]}
       spine_01_ctrl.Rotation.Y {type=MovieSceneScriptingFloatChannel, default=0, const [0,2]}
       spine_01_ctrl.Rotation.Z {type=MovieSceneScriptingFloatChannel, default=0, const [0,2]}
       spine_01_ctrl.Scale.X {type=MovieSceneScriptingFloatChannel, default=0, const [0,2]}
       spine_01_ctrl.Scale.Y {type=MovieSceneScriptingFloatChannel, default=0, const [0,2]}
       spine_01_ctrl.Scale.Z {type=MovieSceneScriptingFloatChannel, default=0, const [0,2]}
       spine_02_ctrl.Location.X {type=MovieSceneScriptingFloatChannel, default=0, const [0,2]}
       spine_02_ctrl.Location.Y {type=MovieSceneScriptingFloatChannel, default=0, const [0,2]}
       spine_02_ctrl.Location.Z {type=MovieSceneScriptingFloatChannel, default=0, const [0,2]}
       spine_02_ctrl.Rotation.X {type=MovieSceneScriptingFloatChannel, default=0, const [0,2]}
       spine_02_ctrl.Rotation.Y {type=MovieSceneScriptingFloatChannel, default=0, const [0,2]}
       spine_02_ctrl.Rotation.Z {type=MovieSceneScriptingFloatChannel, default=0, const [0,2]}
       spine_02_ctrl.Scale.X {type=MovieSceneScriptingFloatChannel, default=0, const [0,2]}
       spine_02_ctrl.Scale.Y {type=MovieSceneScriptingFloatChannel, default=0, const [0,2]}
       spine_02_ctrl.Scale.Z {type=MovieSceneScriptingFloatChannel, default=0, const [0,2]}
       spine_03_ctrl.Location.X {type=MovieSceneScriptingFloatChannel, default=0, const [0,2]}
       spine_03_ctrl.Location.Y {type=MovieSceneScriptingFloatChannel, default=0, const [0,2]}
       spine_03_ctrl.Location.Z {type=MovieSceneScriptingFloatChannel, default=0, const [0,2]}
       spine_03_ctrl.Rotation.X {
        type=MovieSceneScriptingFloatChannel
        default=0
        keys [[3004,0],[3019,0.1894],[3026,0.108],[3034,0.3612],[3039,0.2848],[3046,0.4669],[3053,0.3166],[3059,0.8311],[3171,0]]
       }
       spine_03_ctrl.Rotation.Y {
        type=MovieSceneScriptingFloatChannel
        default=0
        keys [[3004,0],[3019,2.3424],[3026,1.7537],[3034,3.2089],[3039,2.8336],[3046,3.5957],[3053,2.9106],[3059,4.7798],[3171,0]]
       }
       spine_03_ctrl.Rotation.Z {
        type=MovieSceneScriptingFloatChannel
        default=0
        keys [[3004,0],[3019,8.628],[3026,6.4446],[3034,11.8903],[3039,10.4753],[3046,13.3794],[3053,10.7742],[3059,18.0097],[3171,0]]
       }
       spine_03_ctrl.Scale.X {type=MovieSceneScriptingFloatChannel, default=0, const [0,2]}
       spine_03_ctrl.Scale.Y {type=MovieSceneScriptingFloatChannel, default=0, const [0,2]}
       spine_03_ctrl.Scale.Z {type=MovieSceneScriptingFloatChannel, default=0, const [0,2]}
       spine_04_ctrl.Location.X {type=MovieSceneScriptingFloatChannel, default=0}
       spine_04_ctrl.Location.Y {type=MovieSceneScriptingFloatChannel, default=0}
       spine_04_ctrl.Location.Z {type=MovieSceneScriptingFloatChannel, default=0}
       spine_04_ctrl.Rotation.X {type=MovieSceneScriptingFloatChannel, default=0}
       spine_04_ctrl.Rotation.Y {type=MovieSceneScriptingFloatChannel, default=0}
       spine_04_ctrl.Rotation.Z {type=MovieSceneScriptingFloatChannel, default=0}
       spine_04_ctrl.Scale.X {type=MovieSceneScriptingFloatChannel, default=0}
       spine_04_ctrl.Scale.Y {type=MovieSceneScriptingFloatChannel, default=0}
       spine_04_ctrl.Scale.Z {type=MovieSceneScriptingFloatChannel, default=0}
       head_ctrl.Location.X {type=MovieSceneScriptingFloatChannel, default=0, const [0,2]}
       head_ctrl.Location.Y {type=MovieSceneScriptingFloatChannel, default=0, const [0,2]}
       head_ctrl.Location.Z {type=MovieSceneScriptingFloatChannel, default=0, const [0,2]}
       head_ctrl.Rotation.X {type=MovieSceneScriptingFloatChannel, default=0, const [0,2]}
       head_ctrl.Rotation.Y {type=MovieSceneScriptingFloatChannel, default=0, const [0,2]}
       head_ctrl.Rotation.Z {type=MovieSceneScriptingFloatChannel, default=0, const [0,2]}
       head_ctrl.Scale.X {type=MovieSceneScriptingFloatChannel, default=0, const [0,2]}
       head_ctrl.Scale.Y {type=MovieSceneScriptingFloatChannel, default=0, const [0,2]}
       head_ctrl.Scale.Z {type=MovieSceneScriptingFloatChannel, default=0, const [0,2]}
       clavicle_l_ctrl.Location.X {type=MovieSceneScriptingFloatChannel, default=0, const [0,2]}
       clavicle_l_ctrl.Location.Y {type=MovieSceneScriptingFloatChannel, default=0, const [0,2]}
       clavicle_l_ctrl.Location.Z {type=MovieSceneScriptingFloatChannel, default=0, const [0,2]}
       clavicle_l_ctrl.Rotation.X {type=MovieSceneScriptingFloatChannel, default=0, const [0,2]}
       clavicle_l_ctrl.Rotation.Y {type=MovieSceneScriptingFloatChannel, default=0, const [0,2]}
       clavicle_l_ctrl.Rotation.Z {type=MovieSceneScriptingFloatChannel, default=0, const [0,2]}
       clavicle_l_ctrl.Scale.X {type=MovieSceneScriptingFloatChannel, default=0, const [0,2]}
       clavicle_l_ctrl.Scale.Y {type=MovieSceneScriptingFloatChannel, default=0, const [0,2]}
       clavicle_l_ctrl.Scale.Z {type=MovieSceneScriptingFloatChannel, default=0, const [0,2]}
       clavicle_r_ctrl.Location.X {type=MovieSceneScriptingFloatChannel, default=0, const [0,2]}
       clavicle_r_ctrl.Location.Y {type=MovieSceneScriptingFloatChannel, default=0, const [0,2]}
       clavicle_r_ctrl.Location.Z {type=MovieSceneScriptingFloatChannel, default=0, const [0,2]}
       clavicle_r_ctrl.Rotation.X {type=MovieSceneScriptingFloatChannel, default=0, const [0,2]}
       clavicle_r_ctrl.Rotation.Y {type=MovieSceneScriptingFloatChannel, default=0, const [0,2]}
       clavicle_r_ctrl.Rotation.Z {type=MovieSceneScriptingFloatChannel, default=0, const [0,2]}
       clavicle_r_ctrl.Scale.X {type=MovieSceneScriptingFloatChannel, default=0, const [0,2]}
       clavicle_r_ctrl.Scale.Y {type=MovieSceneScriptingFloatChannel, default=0, const [0,2]}
       clavicle_r_ctrl.Scale.Z {type=MovieSceneScriptingFloatChannel, default=0, const [0,2]}
       neck_01_ctrl.Location.X {type=MovieSceneScriptingFloatChannel, default=0, const [0,2]}
       neck_01_ctrl.Location.Y {type=MovieSceneScriptingFloatChannel, default=0, const [0,2]}
       neck_01_ctrl.Location.Z {type=MovieSceneScriptingFloatChannel, default=0, const [0,2]}
       neck_01_ctrl.Rotation.X {type=MovieSceneScriptingFloatChannel, default=0, const [0,2]}
       neck_01_ctrl.Rotation.Y {type=MovieSceneScriptingFloatChannel, default=0, const [0,2]}
       neck_01_ctrl.Rotation.Z {type=MovieSceneScriptingFloatChannel, default=0, const [0,2]}
       neck_01_ctrl.Scale.X {type=MovieSceneScriptingFloatChannel, default=0, const [0,2]}
       neck_01_ctrl.Scale.Y {type=MovieSceneScriptingFloatChannel, default=0, const [0,2]}
       neck_01_ctrl.Scale.Z {type=MovieSceneScriptingFloatChannel, default=0, const [0,2]}
       neck_02_ctrl.Location.X {type=MovieSceneScriptingFloatChannel, default=0, const [0,2]}
       neck_02_ctrl.Location.Y {type=MovieSceneScriptingFloatChannel, default=0, const [0,2]}
       neck_02_ctrl.Location.Z {type=MovieSceneScriptingFloatChannel, default=0, const [0,2]}
       neck_02_ctrl.Rotation.X {type=MovieSceneScriptingFloatChannel, default=0, const [0,2]}
       neck_02_ctrl.Rotation.Y {type=MovieSceneScriptingFloatChannel, default=0, const [0,2]}
       neck_02_ctrl.Rotation.Z {type=MovieSceneScriptingFloatChannel, default=0, const [0,2]}
       neck_02_ctrl.Scale.X {type=MovieSceneScriptingFloatChannel, default=0, const [0,2]}
       neck_02_ctrl.Scale.Y {type=MovieSceneScriptingFloatChannel, default=0, const [0,2]}
       neck_02_ctrl.Scale.Z {type=MovieSceneScriptingFloatChannel, default=0, const [0,2]}
       head_ik_ctrl.Location.X {type=MovieSceneScriptingFloatChannel, default=0, const [-0.2355,2]}
       head_ik_ctrl.Location.Y {type=MovieSceneScriptingFloatChannel, default=0, const [0.6435,2]}
       head_ik_ctrl.Location.Z {type=MovieSceneScriptingFloatChannel, default=0, const [-0.5294,2]}
       head_ik_ctrl.Rotation.X {type=MovieSceneScriptingFloatChannel, default=0, const [16.5076,2]}
       head_ik_ctrl.Rotation.Y {type=MovieSceneScriptingFloatChannel, default=0, const [0,2]}
       head_ik_ctrl.Rotation.Z {type=MovieSceneScriptingFloatChannel, default=0, const [0,2]}
       head_ik_ctrl.Scale.X {type=MovieSceneScriptingFloatChannel, default=0, const [0,2]}
       head_ik_ctrl.Scale.Y {type=MovieSceneScriptingFloatChannel, default=0, const [0,2]}
       head_ik_ctrl.Scale.Z {type=MovieSceneScriptingFloatChannel, default=0, const [0,2]}
       chest_ctrl.Location.X {type=MovieSceneScriptingFloatChannel, default=0, const [-3.0194,2]}
       chest_ctrl.Location.Y {type=MovieSceneScriptingFloatChannel, default=0, const [-3.0052,2]}
       chest_ctrl.Location.Z {type=MovieSceneScriptingFloatChannel, default=0, const [-0.141,2]}
       chest_ctrl.Rotation.X {type=MovieSceneScriptingFloatChannel, default=0, const [-23.7232,2]}
       chest_ctrl.Rotation.Y {type=MovieSceneScriptingFloatChannel, default=0, const [25.177,2]}
       chest_ctrl.Rotation.Z {type=MovieSceneScriptingFloatChannel, default=0, const [-14.0956,2]}
       chest_ctrl.Scale.X {type=MovieSceneScriptingFloatChannel, default=0, const [0,2]}
       chest_ctrl.Scale.Y {type=MovieSceneScriptingFloatChannel, default=0, const [0,2]}
       chest_ctrl.Scale.Z {type=MovieSceneScriptingFloatChannel, default=0, const [0,2]}
       chest_tan_ctrl.Location.X {type=MovieSceneScriptingFloatChannel, default=0, const [-0.0803,2]}
       chest_tan_ctrl.Location.Y {type=MovieSceneScriptingFloatChannel, default=0, const [-1.1394,2]}
       chest_tan_ctrl.Location.Z {type=MovieSceneScriptingFloatChannel, default=0, const [0.1343,2]}
       chest_tan_ctrl.Rotation.X {type=MovieSceneScriptingFloatChannel, default=0, const [12.6947,2]}
       chest_tan_ctrl.Rotation.Y {type=MovieSceneScriptingFloatChannel, default=0, const [-2.1123,2]}
       chest_tan_ctrl.Rotation.Z {type=MovieSceneScriptingFloatChannel, default=0, const [-0.0456,2]}
       chest_tan_ctrl.Scale.X {type=MovieSceneScriptingFloatChannel, default=0, const [-0.0057,2]}
       chest_tan_ctrl.Scale.Y {type=MovieSceneScriptingFloatChannel, default=0, const [-0.0057,2]}
       chest_tan_ctrl.Scale.Z {type=MovieSceneScriptingFloatChannel, default=0, const [-0.0057,2]}
       Weight {type=MovieSceneScriptingFloatChannel, default=0, const [1,2]}
      ]
     }
     {
      type=MovieSceneControlRigParameterSection
      channels [
       upperarm_l_fk_ctrl.Location.X {type=MovieSceneScriptingFloatChannel, default=0, const [0,5]}
       upperarm_l_fk_ctrl.Location.Y {type=MovieSceneScriptingFloatChannel, default=0, const [0,5]}
       upperarm_l_fk_ctrl.Location.Z {type=MovieSceneScriptingFloatChannel, default=0, const [0,5]}
       upperarm_l_fk_ctrl.Rotation.X {
        type=MovieSceneScriptingFloatChannel
        default=0
        keys [[3786,0],[3793,0],[3800,36.7686],[3809,62.7983],[3812,57.197],[3820,26.8503],[3829,44.6222],[3844,76.1848],[3938,98.3073],[4364,76.1848],[4365,0]]
       }
       upperarm_l_fk_ctrl.Rotation.Y {
        type=MovieSceneScriptingFloatChannel
        default=0
        keys [[3786,0],[3793,0],[3804,12.8374],[3814,18.1344],[3826,-18.8254],[3832,-57.266],[3854,-97.6305],[3893,-111.7184],[4030,-99.6435],[4184,-93.6629],[4364,-94.4309],[4365,0]]
       }
       upperarm_l_fk_ctrl.Rotation.Z {
        type=MovieSceneScriptingFloatChannel
        default=0
        keys [[3786,0],[3793,0],[3801,30.7554],[3809,57.8447],[3813,61.9579],[3825,68.4537],[3834,69.8759],[3855,43.9204],[3904,30.669],[3932,29.9444],[4210,25.5202],[4364,19.4062],[4365,0]]
       }
       upperarm_l_fk_ctrl.Scale.X {type=MovieSceneScriptingFloatChannel, default=0, const [0,5]}
       upperarm_l_fk_ctrl.Scale.Y {type=MovieSceneScriptingFloatChannel, default=0, const [0,5]}
       upperarm_l_fk_ctrl.Scale.Z {type=MovieSceneScriptingFloatChannel, default=0, const [0,5]}
       hand_l_fk_ctrl.Location.X {type=MovieSceneScriptingFloatChannel, default=0, const [0,4]}
       hand_l_fk_ctrl.Location.Y {type=MovieSceneScriptingFloatChannel, default=0, const [0,4]}
       hand_l_fk_ctrl.Location.Z {type=MovieSceneScriptingFloatChannel, default=0, const [0,4]}
       hand_l_fk_ctrl.Rotation.X {
        type=MovieSceneScriptingFloatChannel
        default=27.7604
        keys [[3789,27.7604],[3795,27.7604],[3806,52.8436],[3814,75.6439],[3823,78.0701],[3900,77.5059],[4364,96.9676],[4365,27.7604]]
       }
       hand_l_fk_ctrl.Rotation.Y {
        type=MovieSceneScriptingFloatChannel
        default=9.8468
        keys [[3789,9.8468],[3798,9.8468],[3841,38.4879],[3965,52.5056],[4364,25.0298],[4365,9.8468]]
       }
       hand_l_fk_ctrl.Rotation.Z {
        type=MovieSceneScriptingFloatChannel
        default=-18.6379
        keys [[3789,-18.6379],[3798,-18.6379],[3824,-5.3656],[3895,12.2281],[4364,3.781],[4365,-18.6379]]
       }
       hand_l_fk_ctrl.Scale.X {type=MovieSceneScriptingFloatChannel, default=0, const [0,4]}
       hand_l_fk_ctrl.Scale.Y {type=MovieSceneScriptingFloatChannel, default=0, const [0,4]}
       hand_l_fk_ctrl.Scale.Z {type=MovieSceneScriptingFloatChannel, default=0, const [0,4]}
       upperarm_r_fk_ctrl.Location.X {type=MovieSceneScriptingFloatChannel, default=0, const [0,4]}
       upperarm_r_fk_ctrl.Location.Y {type=MovieSceneScriptingFloatChannel, default=0, const [0,4]}
       upperarm_r_fk_ctrl.Location.Z {type=MovieSceneScriptingFloatChannel, default=0, const [0,4]}
       upperarm_r_fk_ctrl.Rotation.X {
        type=MovieSceneScriptingFloatChannel
        default=0
        keys [[3789,0],[3798,0],[3802,18.4751],[3812,33.587],[3845,118.9348],[3922,123.0622],[4164,139.9772],[4197,140.9953],[4364,142.5955],[4365,3.9608]]
       }
       upperarm_r_fk_ctrl.Rotation.Y {
        type=MovieSceneScriptingFloatChannel
        default=0
        keys [[3789,0],[3798,0],[3803,10.7951],[3809,26.2587],[3816,28.5783],[3821,24.2994],[3844,-66.1723],[3866,-68.8474],[4153,-100.1972],[4364,-110.8468],[4365,14.199]]
       }
       upperarm_r_fk_ctrl.Rotation.Z {
        type=MovieSceneScriptingFloatChannel
        default=0
        keys [[3789,0],[3798,0],[3813,34.5532],[3834,78.9865],[3857,51.1508],[3893,47.6169],[4008,55.7477],[4182,69.7254],[4364,33.4099],[4365,15.3785]]
       }
       upperarm_r_fk_ctrl.Scale.X {type=MovieSceneScriptingFloatChannel, default=0, const [0,4]}
       upperarm_r_fk_ctrl.Scale.Y {type=MovieSceneScriptingFloatChannel, default=0, const [0,4]}
       upperarm_r_fk_ctrl.Scale.Z {type=MovieSceneScriptingFloatChannel, default=0, const [0,4]}
       hand_r_fk_ctrl.Location.X {type=MovieSceneScriptingFloatChannel, default=0, const [0,4]}
       hand_r_fk_ctrl.Location.Y {type=MovieSceneScriptingFloatChannel, default=0, const [0,4]}
       hand_r_fk_ctrl.Location.Z {type=MovieSceneScriptingFloatChannel, default=0, const [0,4]}
       hand_r_fk_ctrl.Rotation.X {type=MovieSceneScriptingFloatChannel, default=23.8346, keys [[3789,23.8346],[3813,65.7212],[3857,68.2545],[4364,60.053],[4365,23.8346]]}
       hand_r_fk_ctrl.Rotation.Y {type=MovieSceneScriptingFloatChannel, default=10.2075, keys [[3789,10.2075],[3798,10.2075],[4364,45.0487],[4365,10.2075]]}
       hand_r_fk_ctrl.Rotation.Z {
        type=MovieSceneScriptingFloatChannel
        default=-4.9075
        keys [[3789,-4.9075],[3822,-30.4707],[3828,-24.4396],[3864,9.737],[4364,12.6492],[4365,-4.9075]]
       }
       hand_r_fk_ctrl.Scale.X {type=MovieSceneScriptingFloatChannel, default=0, const [0,4]}
       hand_r_fk_ctrl.Scale.Y {type=MovieSceneScriptingFloatChannel, default=0, const [0,4]}
       hand_r_fk_ctrl.Scale.Z {type=MovieSceneScriptingFloatChannel, default=0, const [0,4]}
       body_ctrl.Location.X {type=MovieSceneScriptingFloatChannel, default=0, const [0,5]}
       body_ctrl.Location.Y {type=MovieSceneScriptingFloatChannel, default=0, const [0,5]}
       body_ctrl.Location.Z {type=MovieSceneScriptingFloatChannel, default=0, const [0,5]}
       body_ctrl.Rotation.X {type=MovieSceneScriptingFloatChannel, default=0, keys [[3789,0],[3802,-9.7452],[3837,22.1948],[3856,22.6671],[4364,22.6671],[4365,0]]}
       body_ctrl.Rotation.Y {type=MovieSceneScriptingFloatChannel, default=0, const [0,5]}
       body_ctrl.Rotation.Z {type=MovieSceneScriptingFloatChannel, default=0, const [0,5]}
       body_ctrl.Scale.X {type=MovieSceneScriptingFloatChannel, default=0, const [0,5]}
       body_ctrl.Scale.Y {type=MovieSceneScriptingFloatChannel, default=0, const [0,5]}
       body_ctrl.Scale.Z {type=MovieSceneScriptingFloatChannel, default=0, const [0,5]}
       hips_ctrl.Location.X {type=MovieSceneScriptingFloatChannel, default=0, const [0,4]}
       hips_ctrl.Location.Y {type=MovieSceneScriptingFloatChannel, default=0, const [0,4]}
       hips_ctrl.Location.Z {type=MovieSceneScriptingFloatChannel, default=0, const [0,4]}
       hips_ctrl.Rotation.X {type=MovieSceneScriptingFloatChannel, default=0, keys [[3789,0],[3802,8.0311],[4364,0],[4365,0]]}
       hips_ctrl.Rotation.Y {type=MovieSceneScriptingFloatChannel, default=0, const [0,4]}
       hips_ctrl.Rotation.Z {type=MovieSceneScriptingFloatChannel, default=0, const [0,4]}
       hips_ctrl.Scale.X {type=MovieSceneScriptingFloatChannel, default=0, const [0,4]}
       hips_ctrl.Scale.Y {type=MovieSceneScriptingFloatChannel, default=0, const [0,4]}
       hips_ctrl.Scale.Z {type=MovieSceneScriptingFloatChannel, default=0, const [0,4]}
       spine_01_ctrl.Location.X {type=MovieSceneScriptingFloatChannel, default=0, const [0,5]}
       spine_01_ctrl.Location.Y {type=MovieSceneScriptingFloatChannel, default=0, const [0,5]}
       spine_01_ctrl.Location.Z {type=MovieSceneScriptingFloatChannel, default=0, const [0,5]}
       spine_01_ctrl.Rotation.X {type=MovieSceneScriptingFloatChannel, default=0, keys [[3789,0],[3798,0],[3856,59.9468],[4184,72.0948],[4364,73.6448],[4365,0]]}
       spine_01_ctrl.Rotation.Y {type=MovieSceneScriptingFloatChannel, default=0, const [0,5]}
       spine_01_ctrl.Rotation.Z {type=MovieSceneScriptingFloatChannel, default=0, const [0,5]}
       spine_01_ctrl.Scale.X {type=MovieSceneScriptingFloatChannel, default=0, const [0,5]}
       spine_01_ctrl.Scale.Y {type=MovieSceneScriptingFloatChannel, default=0, const [0,5]}
       spine_01_ctrl.Scale.Z {type=MovieSceneScriptingFloatChannel, default=0, const [0,5]}
       spine_02_ctrl.Location.X {type=MovieSceneScriptingFloatChannel, default=0, const [0,5]}
       spine_02_ctrl.Location.Y {type=MovieSceneScriptingFloatChannel, default=0, const [0,5]}
       spine_02_ctrl.Location.Z {type=MovieSceneScriptingFloatChannel, default=0, const [0,5]}
       spine_02_ctrl.Rotation.X {type=MovieSceneScriptingFloatChannel, default=64.1557, keys [[3789,0],[3793,1.0028],[3803,0],[4011,5.1682],[4364,0],[4365,26.4821]]}
       spine_02_ctrl.Rotation.Y {type=MovieSceneScriptingFloatChannel, default=0, const [0,5]}
       spine_02_ctrl.Rotation.Z {type=MovieSceneScriptingFloatChannel, default=0, const [0,5]}
       spine_02_ctrl.Scale.X {type=MovieSceneScriptingFloatChannel, default=0, const [0,5]}
       spine_02_ctrl.Scale.Y {type=MovieSceneScriptingFloatChannel, default=0, const [0,5]}
       spine_02_ctrl.Scale.Z {type=MovieSceneScriptingFloatChannel, default=0, const [0,5]}
       spine_03_ctrl.Location.X {type=MovieSceneScriptingFloatChannel, default=0, const [0,6]}
       spine_03_ctrl.Location.Y {type=MovieSceneScriptingFloatChannel, default=0, const [0,6]}
       spine_03_ctrl.Location.Z {type=MovieSceneScriptingFloatChannel, default=0, const [0,6]}
       spine_03_ctrl.Rotation.X {
        type=MovieSceneScriptingFloatChannel
        default=0
        keys [[3784,0],[3785,0],[3796,5.1266],[3805,-6.7167],[3856,4.0834],[3902,13.7301],[4069,5.9343],[4364,5.9797],[4365,0]]
       }
       spine_03_ctrl.Rotation.Y {type=MovieSceneScriptingFloatChannel, default=0, const [0,6]}
       spine_03_ctrl.Rotation.Z {type=MovieSceneScriptingFloatChannel, default=0, const [0,6]}
       spine_03_ctrl.Scale.X {type=MovieSceneScriptingFloatChannel, default=0, const [0,6]}
       spine_03_ctrl.Scale.Y {type=MovieSceneScriptingFloatChannel, default=0, const [0,6]}
       spine_03_ctrl.Scale.Z {type=MovieSceneScriptingFloatChannel, default=0, const [0,6]}
       spine_04_ctrl.Location.X {type=MovieSceneScriptingFloatChannel, default=0, const [0,5]}
       spine_04_ctrl.Location.Y {type=MovieSceneScriptingFloatChannel, default=0, const [0,5]}
       spine_04_ctrl.Location.Z {type=MovieSceneScriptingFloatChannel, default=0, const [0,5]}
       spine_04_ctrl.Rotation.X {type=MovieSceneScriptingFloatChannel, default=0, keys [[3789,0],[3798,0],[3856,0.0299],[4364,0.0299],[4365,0]]}
       spine_04_ctrl.Rotation.Y {type=MovieSceneScriptingFloatChannel, default=0, const [0,5]}
       spine_04_ctrl.Rotation.Z {type=MovieSceneScriptingFloatChannel, default=0, const [0,5]}
       spine_04_ctrl.Scale.X {type=MovieSceneScriptingFloatChannel, default=0, const [0,5]}
       spine_04_ctrl.Scale.Y {type=MovieSceneScriptingFloatChannel, default=0, const [0,5]}
       spine_04_ctrl.Scale.Z {type=MovieSceneScriptingFloatChannel, default=0, const [0,5]}
       head_ctrl.Location.X {type=MovieSceneScriptingFloatChannel, default=0, const [0,6]}
       head_ctrl.Location.Y {type=MovieSceneScriptingFloatChannel, default=0, const [0,6]}
       head_ctrl.Location.Z {type=MovieSceneScriptingFloatChannel, default=0, const [0,6]}
       head_ctrl.Rotation.X {
        type=MovieSceneScriptingFloatChannel
        default=0
        keys [[3803,0],[3811,0],[3828,52.3412],[3844,101.21],[3856,101.838],[4364,101.838],[4365,97.4342]]
       }
       head_ctrl.Rotation.Y {type=MovieSceneScriptingFloatChannel, default=0, const [0,6]}
       head_ctrl.Rotation.Z {type=MovieSceneScriptingFloatChannel, default=0, const [0,6]}
       head_ctrl.Scale.X {type=MovieSceneScriptingFloatChannel, default=0, const [0,6]}
       head_ctrl.Scale.Y {type=MovieSceneScriptingFloatChannel, default=0, const [0,6]}
       head_ctrl.Scale.Z {type=MovieSceneScriptingFloatChannel, default=0, const [0,6]}
       neck_01_ctrl.Location.X {type=MovieSceneScriptingFloatChannel, default=0, const [0,6]}
       neck_01_ctrl.Location.Y {type=MovieSceneScriptingFloatChannel, default=0, const [0,6]}
       neck_01_ctrl.Location.Z {type=MovieSceneScriptingFloatChannel, default=0, const [0,6]}
       neck_01_ctrl.Rotation.X {type=MovieSceneScriptingFloatChannel, default=0, thinned_from=6, keys [[3789,0],[3800,0],[3856,37.8724],[4364,37.8724],[4365,0]]}
       neck_01_ctrl.Rotation.Y {type=MovieSceneScriptingFloatChannel, default=0, const [0,6]}
       neck_01_ctrl.Rotation.Z {type=MovieSceneScriptingFloatChannel, default=0, const [0,6]}
       neck_01_ctrl.Scale.X {type=MovieSceneScriptingFloatChannel, default=0, const [0,6]}
       neck_01_ctrl.Scale.Y {type=MovieSceneScriptingFloatChannel, default=0, const [0,6]}
       neck_01_ctrl.Scale.Z {type=MovieSceneScriptingFloatChannel, default=0, const [0,6]}
       Weight {type=MovieSceneScriptingFloatChannel, default=1, const [1,4]}
      ]
     }
    ]
   }
  ]
 }
]
display_rate {denominator=1, numerator=30}
tick_resolution {denominator=1, numerator=24000}
tracks []
